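# Niagara system for the prop-spawn effect of the SpriteSender emote, a burst built from sprite, mesh and light renderers.
# Several DefaultPropSpawn_Merged emitters sit alongside Flare, BigSparkles and Shockwave layers, some duplicated; BigSparkles is disabled.
# The dependencies include SpawnGroup and MergeHelper switch inputs, which point to several particle looks sharing each merged emitter.
# Platform, front-end and time-of-day helper inputs are also used to adjust values.
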NiagaraSystem /BRCosmetics/Animation/Game/MainPlayer/Emotes/SpriteSender/FX/NS_Emote_SpriteSender_PropSpawn
dependencies {
 usedDataInterfaces [
  {refPath=&NiagaraDataInterfacePlatformSet}
  {refPath=&NiagaraDataInterfaceCurve}
  {refPath=&NiagaraDataInterfaceArrayFloat}
  {refPath=&NiagaraDataInterfaceColorCurve}
  {refPath=&NiagaraDataInterfaceVector2DCurve}
 ]
 usedDynamicInputs [
  {refPath=&MakeIntFromFloat}
  {refPath=&FN_PickFloatDependOnPlatform}
  {refPath=&SpawnGroup}
  {refPath=&MergeHelper_FloatSwitch}
  {refPath=&UniformRangedFloat}
  {refPath=&MergeHelper_ColorSwitch}
  {refPath=&MergeHelper_VectorSwitch}
  {refPath=&UniformRangedVector}
  {refPath=&RandomRangeFloat}
  {refPath=&RandomRangeVector}
  {refPath=&MakeVector4}
  {refPath=&Multiply_Float}
  {refPath=&FloatFromCurve}
  {refPath=&Multiply_VectorByFloat}
  {refPath=&MaskFloatBySpawnGroup}
  {refPath=&MergeHelper_Vector2Switch}
  {refPath=&Vector2DFromFloat}
  {refPath=&FN_FrontEndFloatModifier}
  {refPath=&RandomRangeInt}
  {refPath=&SelectFloatFromArray}
  {refPath=&ReturnNormalizedExecIndex}
  {refPath=&Multiply_Vector2D_ByFloat}
  {refPath=&MultiplyLinearColor}
  {refPath=&MakeVector4_FromLinearColor}
  {refPath=&Multiply_Vector2}
  {refPath=&MakeFloatFromVector}
  {refPath=&MakeVectorFromLinearColorRGB}
  {refPath=&ColorFromCurve}
  {refPath=&VectorFromFloat}
  {refPath=&MakeLinearColorFromVectorAndFloat}
  {refPath=&Vector2DFromCurve}
  {refPath=&AddVectorToPosition}
  {refPath=&FNTimeOfDayScalar}
 ]
 usedModules [
  {refPath=&EmitterState}
  {refPath=&SpawnBurst_Instantaneous}
  {refPath=&self:DefaultPropSpawn_Merged_0.NiagaraScriptSource_1.NiagaraGraph_0.NiagaraNodeAssignment_1.SetVariables_#1}
  {refPath=&PartitionParticles}
  {refPath=&InitializeParticle}
  {refPath=&self:DefaultPropSpawn_Merged_0.NiagaraScriptSource_1.NiagaraGraph_0.NiagaraNodeAssignment_0.SetVariables_#2}
  {refPath=&AccelerationForce}
  {refPath=&ShapeLocation}
  {refPath=&CameraOffset}
  {refPath=&AddVelocity}
  {refPath=&ApplyInitialForces}
  {refPath=&self:DefaultPropSpawn_Merged_0.NiagaraScriptSource_1.NiagaraGraph_0.NiagaraNodeAssignment_5.SetVariables_#3}
  {refPath=&ParticleState}
  {refPath=&SpriteRotationRate}
  {refPath=&GravityForce}
  {refPath=/Niagara/Modules/Update/Forces/CurlNoiseForce}
  {refPath=&Drag}
  {refPath=&SolveForcesAndVelocity}
  {refPath=&ScaleSpriteSize}
  {refPath=&self:DefaultPropSpawn_Merged_0.NiagaraScriptSource_1.NiagaraGraph_0.NiagaraNodeAssignment_2.SetVariables_#4}
  {refPath=&ScaleColor}
  {refPath=&DynamicMaterialParameters}
  {refPath=&self:Flare_0.NiagaraScriptSource_0.NiagaraGraph_0.NiagaraNodeAssignment_0.SetVariables_#5}
  {refPath=&FN_Light_Attributes}
  {refPath=&self:BigSparkles_0.NiagaraScriptSource_0.NiagaraGraph_0.NiagaraNodeAssignment_1.SetVariables_#6}
  {refPath=&self:BigSparkles_0.NiagaraScriptSource_0.NiagaraGraph_0.NiagaraNodeAssignment_2.SetVariables_#7}
  {refPath=/Niagara/Modules/Update/Forces/V2/CurlNoiseForce}
  {refPath=&InitialMeshOrientation}
  {refPath=&ScaleMeshSize}
  {refPath=&self:Shockwave_0.NiagaraScriptSource_0.NiagaraGraph_0.NiagaraNodeAssignment_0.SetVariables_#8}
  {refPath=&self:DefaultPropSpawn_Merged001_0.NiagaraScriptSource_1.NiagaraGraph_0.NiagaraNodeAssignment_1.SetVariables_#1}
  {refPath=&self:DefaultPropSpawn_Merged001_0.NiagaraScriptSource_1.NiagaraGraph_0.NiagaraNodeAssignment_0.SetVariables_#2}
  {refPath=&self:DefaultPropSpawn_Merged001_0.NiagaraScriptSource_1.NiagaraGraph_0.NiagaraNodeAssignment_5.SetVariables_#3}
  {refPath=&self:DefaultPropSpawn_Merged001_0.NiagaraScriptSource_1.NiagaraGraph_0.NiagaraNodeAssignment_2.SetVariables_#4}
  {refPath=&self:DefaultPropSpawn_Merged001_0.NiagaraScriptSource_1.NiagaraGraph_0.NiagaraNodeAssignment_3.SetVariables_#9}
  {refPath=&self:DefaultPropSpawn_Merged002_0.NiagaraScriptSource_1.NiagaraGraph_0.NiagaraNodeAssignment_1.SetVariables_#1}
  {refPath=&self:DefaultPropSpawn_Merged002_0.NiagaraScriptSource_1.NiagaraGraph_0.NiagaraNodeAssignment_0.SetVariables_#2}
  {refPath=&self:DefaultPropSpawn_Merged002_0.NiagaraScriptSource_1.NiagaraGraph_0.NiagaraNodeAssignment_5.SetVariables_#3}
  {refPath=&self:DefaultPropSpawn_Merged002_0.NiagaraScriptSource_1.NiagaraGraph_0.NiagaraNodeAssignment_2.SetVariables_#4}
  {refPath=&self:DefaultPropSpawn_Merged002_0.NiagaraScriptSource_1.NiagaraGraph_0.NiagaraNodeAssignment_3.SetVariables_#9}
  {refPath=&self:DefaultPropSpawn_Merged003_0.NiagaraScriptSource_1.NiagaraGraph_0.NiagaraNodeAssignment_1.SetVariables_#1}
  {refPath=&self:DefaultPropSpawn_Merged003_0.NiagaraScriptSource_1.NiagaraGraph_0.NiagaraNodeAssignment_0.SetVariables_#2}
  {refPath=&self:DefaultPropSpawn_Merged003_0.NiagaraScriptSource_1.NiagaraGraph_0.NiagaraNodeAssignment_5.SetVariables_#3}
  {refPath=&self:DefaultPropSpawn_Merged003_0.NiagaraScriptSource_1.NiagaraGraph_0.NiagaraNodeAssignment_2.SetVariables_#4}
  {refPath=&self:DefaultPropSpawn_Merged003_0.NiagaraScriptSource_1.NiagaraGraph_0.NiagaraNodeAssignment_3.SetVariables_#9}
  {refPath=&self:DefaultPropSpawn_Merged004_0.NiagaraScriptSource_1.NiagaraGraph_0.NiagaraNodeAssignment_1.SetVariables_#1}
  {refPath=&self:DefaultPropSpawn_Merged004_0.NiagaraScriptSource_1.NiagaraGraph_0.NiagaraNodeAssignment_0.SetVariables_#2}
  {refPath=&self:DefaultPropSpawn_Merged004_0.NiagaraScriptSource_1.NiagaraGraph_0.NiagaraNodeAssignment_5.SetVariables_#3}
  {refPath=&self:DefaultPropSpawn_Merged004_0.NiagaraScriptSource_1.NiagaraGraph_0.NiagaraNodeAssignment_2.SetVariables_#4}
  {refPath=&self:DefaultPropSpawn_Merged004_0.NiagaraScriptSource_1.NiagaraGraph_0.NiagaraNodeAssignment_3.SetVariables_#9}
  {refPath=&self:DefaultPropSpawn_Merged005_0.NiagaraScriptSource_1.NiagaraGraph_0.NiagaraNodeAssignment_1.SetVariables_#1}
  {refPath=&self:DefaultPropSpawn_Merged005_0.NiagaraScriptSource_1.NiagaraGraph_0.NiagaraNodeAssignment_0.SetVariables_#2}
  {refPath=&self:DefaultPropSpawn_Merged005_0.NiagaraScriptSource_1.NiagaraGraph_0.NiagaraNodeAssignment_5.SetVariables_#3}
  {refPath=&self:DefaultPropSpawn_Merged005_0.NiagaraScriptSource_1.NiagaraGraph_0.NiagaraNodeAssignment_2.SetVariables_#4}
  {refPath=&self:Flare002_0.NiagaraScriptSource_0.NiagaraGraph_0.NiagaraNodeAssignment_0.SetVariables_#5}
  {refPath=&self:BigSparkles002_0.NiagaraScriptSource_0.NiagaraGraph_0.NiagaraNodeAssignment_1.SetVariables_#6}
  {refPath=&self:BigSparkles002_0.NiagaraScriptSource_0.NiagaraGraph_0.NiagaraNodeAssignment_2.SetVariables_#7}
  {refPath=&self:Shockwave002_0.NiagaraScriptSource_0.NiagaraGraph_0.NiagaraNodeAssignment_0.SetVariables_#8}
  {refPath=&self:SystemSpawnScript.SystemScriptSource.SystemScriptGraph.NiagaraNodeAssignment_0.SetVariables_#10}
  {refPath=&SystemState}
 ]
 usedRenderers [
  {refPath=&NiagaraSpriteRendererProperties}
  {refPath=&NiagaraLightRendererProperties}
  {refPath=&NiagaraMeshRendererProperties}
 ]
}
emitters {
 BigSparkles {
  bEnabled=false
  emitterSpawnScript {scriptName=EmitterSpawnScript, modules []}
  emitterUpdateScript {
   scriptName=EmitterUpdateScript
   modules [
    {
     bIsSetParametersModule=false
     enabled=true
     moduleName=EmitterState
     inputs [
      Life Cycle Mode {
       bIsDynamic=false
       bIsEditable=true
       bIsStaticSwitch=true
       bIsVisible=true
       type {classStructOrEnum {refPath=/Niagara/Enums/ENiagaraEmitterLifeCycleMode}}
      }
      Inactive Response {
       bIsDynamic=false
       bIsEditable=false
       bIsStaticSwitch=true
       bIsVisible=false
       type {classStructOrEnum {refPath=/Niagara/Enums/ENiagaraInactiveMode}}
      }
      Loop Behavior {
       bIsDynamic=false
       bIsEditable=false
       bIsStaticSwitch=true
       bIsVisible=false
       type {classStructOrEnum {refPath=/Niagara/Enums/ENiagara_EmitterStateOptions}}
      }
      Loop Duration Mode {
       bIsDynamic=false
       bIsEditable=false
       bIsStaticSwitch=true
       bIsVisible=false
       type {classStructOrEnum {refPath=/Niagara/Enums/ENiagara_InfiniteLoopDuration}}
      }
      Loop Count {
       bIsDynamic=false
       bIsEditable=false
       bIsStaticSwitch=false
       bIsVisible=false
       type {classStructOrEnum {refPath=/Script/Niagara.NiagaraInt32}}
      }
      Loop Duration {
       bIsDynamic=false
       bIsEditable=false
       bIsStaticSwitch=false
       bIsVisible=false
       type {classStructOrEnum {refPath=/Script/Niagara.NiagaraFloat}}
      }
      Recalculate Duration Each Loop {bIsDynamic=false, bIsEditable=false, bIsStaticSwitch=false, bIsVisible=false, type {classStructOrEnum {refPath=/Script/Niagara.NiagaraBool}}}
      UseLoopDelay {bIsDynamic=false, bIsEditable=false, bIsStaticSwitch=false, bIsVisible=false, type {classStructOrEnum {refPath=&NiagaraBool}}}
      Loop Delay {
       bIsDynamic=false
       bIsEditable=false
       bIsStaticSwitch=false
       bIsVisible=false
       type {classStructOrEnum {refPath=&NiagaraFloat}}
      }
      Delay First Loop Only {bIsDynamic=false, bIsEditable=false, bIsStaticSwitch=false, bIsVisible=false, type {classStructOrEnum {refPath=&NiagaraBool}}}
      Scalability Mode {
       bIsDynamic=false
       bIsEditable=true
       bIsStaticSwitch=true
       bIsVisible=true
       type {classStructOrEnum {refPath=/Niagara/Enums/ENiagaraEmitterScalabilityMode_Limited}}
      }
      Enable Distance Culling {bIsDynamic=false, bIsEditable=false, bIsStaticSwitch=true, bIsVisible=false, type {classStructOrEnum {refPath=&NiagaraBool}}}
      MinDistance {
       bIsDynamic=false
       bIsEditable=false
       bIsStaticSwitch=false
       bIsVisible=false
       type {classStructOrEnum {refPath=&NiagaraFloat}}
      }
      Min Distance Response {
       bIsDynamic=false
       bIsEditable=false
       bIsStaticSwitch=false
       bIsVisible=false
       type {classStructOrEnum {refPath=/Script/Niagara.ENiagaraExecutionStateManagement}}
      }
      MaxDistance {
       bIsDynamic=false
       bIsEditable=false
       bIsStaticSwitch=false
       bIsVisible=false
       type {classStructOrEnum {refPath=&NiagaraFloat}}
      }
      Max Distance Response {
       bIsDynamic=false
       bIsEditable=false
       bIsStaticSwitch=false
       bIsVisible=false
       type {classStructOrEnum {refPath=&ENiagaraExecutionStateManagement}}
      }
      Scale Spawn Count {bIsDynamic=false, bIsEditable=false, bIsStaticSwitch=true, bIsVisible=false, type {classStructOrEnum {refPath=&NiagaraBool}}}
      Spawn Count Scale {
       bIsDynamic=false
       bIsEditable=false
       bIsStaticSwitch=false
       bIsVisible=false
       type {classStructOrEnum {refPath=&NiagaraFloat}}
      }
      Spawn Count Scale By Distance {
       bIsDynamic=false
       bIsEditable=false
       bIsStaticSwitch=false
       bIsVisible=false
       type {classStructOrEnum {refPath=&NiagaraDataInterfaceCurve}}
      }
      Enable Visibility Culling {bIsDynamic=false, bIsEditable=false, bIsStaticSwitch=true, bIsVisible=false, type {classStructOrEnum {refPath=&NiagaraBool}}}
      Visibility Cull Response {
       bIsDynamic=false
       bIsEditable=false
       bIsStaticSwitch=false
       bIsVisible=false
       type {classStructOrEnum {refPath=&ENiagaraExecutionStateManagement}}
      }
      Visibility Cull Delay {
       bIsDynamic=false
       bIsEditable=false
       bIsStaticSwitch=false
       bIsVisible=false
       type {classStructOrEnum {refPath=&NiagaraFloat}}
      }
      Reset Age On Awaken {bIsDynamic=false, bIsEditable=false, bIsStaticSwitch=false, bIsVisible=false, type {classStructOrEnum {refPath=&NiagaraBool}}}
     ]
     moduleScript {refPath=&EmitterState}
    }
    {
     bIsSetParametersModule=false
     enabled=true
     moduleName=SpawnBurst_Instantaneous
     inputs [
      Spawn Count {bIsDynamic=false, bIsEditable=true, bIsStaticSwitch=false, bIsVisible=true, type {classStructOrEnum {refPath=&NiagaraInt32}}}
      Spawn Time {bIsDynamic=false, bIsEditable=true, bIsStaticSwitch=false, bIsVisible=true, type {classStructOrEnum {refPath=&NiagaraFloat}}}
      Use Spawn Probability {bIsDynamic=false, bIsEditable=true, bIsStaticSwitch=false, bIsVisible=true, type {classStructOrEnum {refPath=&NiagaraBool}}}
      Spawn Probability {bIsDynamic=false, bIsEditable=false, bIsStaticSwitch=false, bIsVisible=true, type {classStructOrEnum {refPath=&NiagaraFloat}}}
      UseLoopCountLimit {bIsDynamic=false, bIsEditable=true, bIsStaticSwitch=false, bIsVisible=true, type {classStructOrEnum {refPath=&NiagaraBool}}}
      Loop Count Limit {bIsDynamic=false, bIsEditable=false, bIsStaticSwitch=false, bIsVisible=true, type {classStructOrEnum {refPath=&NiagaraInt32}}}
      Spawn Group {bIsDynamic=false, bIsEditable=true, bIsStaticSwitch=false, bIsVisible=true, type {classStructOrEnum {refPath=&NiagaraInt32}}}
      Age {bIsDynamic=false, bIsEditable=true, bIsStaticSwitch=false, bIsVisible=true, type {classStructOrEnum {refPath=&NiagaraFloat}}}
     ]
     moduleScript {refPath=&SpawnBurst_Instantaneous}
    }
   ]
  }
  particleSpawnScript {
   scriptName=ParticleSpawnScript
   modules [
    {
     bIsSetParametersModule=true
     enabled=true
     moduleName=SetVariables_#6
     inputs [Particles.VisibilityTag {bIsDynamic=true, bIsEditable=true, bIsStaticSwitch=false, bIsVisible=true, type {classStructOrEnum {refPath=&NiagaraInt32}}}]
     moduleScript {refPath=&self:BigSparkles_0.NiagaraScriptSource_0.NiagaraGraph_0.NiagaraNodeAssignment_1.SetVariables_#6}
    }
    {
     bIsSetParametersModule=false
     enabled=true
     moduleName=PartitionParticles
     inputs [
      Partition Mode {
       bIsDynamic=false
       bIsEditable=true
       bIsStaticSwitch=true
       bIsVisible=true
       type {classStructOrEnum {refPath=/Niagara/Enums/Utility/ENiagara_PartitionMode_v2}}
      }
      Partition Count Mode {
       bIsDynamic=false
       bIsEditable=false
       bIsStaticSwitch=true
       bIsVisible=false
       type {classStructOrEnum {refPath=&ENiagara_InfiniteLoopDuration}}
      }
      A {
       bIsDynamic=false
       bIsEditable=false
       bIsStaticSwitch=false
       bIsVisible=false
       type {classStructOrEnum {refPath=&NiagaraFloat}}
      }
      Number Of Partitions {bIsDynamic=false, bIsEditable=true, bIsStaticSwitch=false, bIsVisible=true, type {classStructOrEnum {refPath=&NiagaraInt32}}}
      Particles Per Partition {
       bIsDynamic=false
       bIsEditable=false
       bIsStaticSwitch=false
       bIsVisible=false
       type {classStructOrEnum {refPath=&NiagaraInt32}}
      }
      Partition Duration {
       bIsDynamic=false
       bIsEditable=false
       bIsStaticSwitch=false
       bIsVisible=false
       type {classStructOrEnum {refPath=&NiagaraFloat}}
      }
      Time {
       bIsDynamic=false
       bIsEditable=false
       bIsStaticSwitch=false
       bIsVisible=false
       type {classStructOrEnum {refPath=&NiagaraFloat}}
      }
      Partitioning Frequency {
       bIsDynamic=false
       bIsEditable=false
       bIsStaticSwitch=true
       bIsVisible=false
       type {classStructOrEnum {refPath=/Niagara/Enums/ENiagaraRandomnessEvaluation}}
      }
      B {
       bIsDynamic=false
       bIsEditable=false
       bIsStaticSwitch=false
       bIsVisible=false
       type {classStructOrEnum {refPath=&NiagaraFloat}}
      }
      Distance {
       bIsDynamic=false
       bIsEditable=false
       bIsStaticSwitch=false
       bIsVisible=false
       type {classStructOrEnum {refPath=&NiagaraFloat}}
      }
      Comparison Type {
       bIsDynamic=false
       bIsEditable=false
       bIsStaticSwitch=true
       bIsVisible=false
       type {classStructOrEnum {refPath=/Niagara/Enums/ENiagaraBooleanLogicOps_v2}}
      }
      Evaluation Type {
       bIsDynamic=false
       bIsEditable=false
       bIsStaticSwitch=true
       bIsVisible=false
       type {classStructOrEnum {refPath=&ENiagaraRandomnessEvaluation}}
      }
      UseDistanceFalloff {bIsDynamic=false, bIsEditable=false, bIsStaticSwitch=false, bIsVisible=false, type {classStructOrEnum {refPath=&NiagaraBool}}}
      Distance Falloff Exponent {
       bIsDynamic=false
       bIsEditable=false
       bIsStaticSwitch=false
       bIsVisible=false
       type {classStructOrEnum {refPath=&NiagaraFloat}}
      }
      Start Position {
       bIsDynamic=false
       bIsEditable=false
       bIsStaticSwitch=false
       bIsVisible=false
       type {classStructOrEnum {refPath=/Script/Niagara.NiagaraPosition}}
      }
      End Position {
       bIsDynamic=false
       bIsEditable=false
       bIsStaticSwitch=false
       bIsVisible=false
       type {classStructOrEnum {refPath=&NiagaraPosition}}
      }
      ID to Partition {bIsDynamic=false, bIsEditable=true, bIsStaticSwitch=false, bIsVisible=true, type {classStructOrEnum {refPath=&NiagaraInt32}}}
      Randomness Mode {
       bIsDynamic=false
       bIsEditable=false
       bIsStaticSwitch=true
       bIsVisible=false
       type {classStructOrEnum {refPath=/Niagara/Enums/ENiagaraRandomnessMode}}
      }
      Override Seed {bIsDynamic=false, bIsEditable=false, bIsStaticSwitch=true, bIsVisible=false, type {classStructOrEnum {refPath=&NiagaraBool}}}
      Fixed Random Seed {bIsDynamic=false, bIsEditable=false, bIsStaticSwitch=true, bIsVisible=false, type {classStructOrEnum {refPath=&NiagaraBool}}}
      Color A {
       bIsDynamic=false
       bIsEditable=false
       bIsStaticSwitch=false
       bIsVisible=false
       type {classStructOrEnum {refPath=/Script/CoreUObject.LinearColor}}
      }
      Color B {
       bIsDynamic=false
       bIsEditable=false
       bIsStaticSwitch=false
       bIsVisible=false
       type {classStructOrEnum {refPath=&LinearColor}}
      }
      Color Comparison Type {bIsDynamic=false, bIsEditable=false, bIsStaticSwitch=true, bIsVisible=false, type {classStructOrEnum {refPath=&NiagaraInt32}}}
      Color Tolerance {
       bIsDynamic=false
       bIsEditable=false
       bIsStaticSwitch=false
       bIsVisible=false
       type {classStructOrEnum {refPath=&NiagaraFloat}}
      }
      Minimum Saturation {
       bIsDynamic=false
       bIsEditable=false
       bIsStaticSwitch=false
       bIsVisible=false
       type {classStructOrEnum {refPath=&NiagaraFloat}}
      }
      Direction A {
       bIsDynamic=false
       bIsEditable=false
       bIsStaticSwitch=false
       bIsVisible=false
       type {classStructOrEnum {refPath=/Script/CoreUObject.Vector3f}}
      }
      Direction B {
       bIsDynamic=false
       bIsEditable=false
       bIsStaticSwitch=false
       bIsVisible=false
       type {classStructOrEnum {refPath=&Vector3f}}
      }
      Direction Tolerance {
       bIsDynamic=false
       bIsEditable=false
       bIsStaticSwitch=false
       bIsVisible=false
       type {classStructOrEnum {refPath=&NiagaraFloat}}
      }
      Write Partition to Ribbon ID {
       bIsDynamic=false
       bIsEditable=true
       bIsStaticSwitch=true
       bIsVisible=true
       type {classStructOrEnum {refPath=/Niagara/Enums/ENiagara_AttributeSamplingApplyOutput}}
      }
      Write Partition Index to Ribbon Link Order {
       bIsDynamic=false
       bIsEditable=false
       bIsStaticSwitch=true
       bIsVisible=false
       type {classStructOrEnum {refPath=&ENiagara_AttributeSamplingApplyOutput}}
      }
      Normalized Index In Partition Includes One {bIsDynamic=false, bIsEditable=false, bIsStaticSwitch=true, bIsVisible=false, type {classStructOrEnum {refPath=&NiagaraBool}}}
      Normalized Index Includes One {bIsDynamic=false, bIsEditable=true, bIsStaticSwitch=true, bIsVisible=true, type {classStructOrEnum {refPath=&NiagaraBool}}}
     ]
     moduleScript {refPath=&PartitionParticles}
    }
    {
     bIsSetParametersModule=false
     enabled=true
     moduleName=InitializeParticle
     inputs [
      Lifetime Mode {
       bIsDynamic=false
       bIsEditable=true
       bIsStaticSwitch=true
       bIsVisible=true
       type {classStructOrEnum {refPath=/Niagara/Enums/ENiagara_LifetimeMode}}
      }
      Lifetime Min {
       bIsDynamic=false
       bIsEditable=false
       bIsStaticSwitch=false
       bIsVisible=false
       type {classStructOrEnum {refPath=&NiagaraFloat}}
      }
      Lifetime {bIsDynamic=true, bIsEditable=true, bIsStaticSwitch=false, bIsVisible=true, type {classStructOrEnum {refPath=&NiagaraFloat}}}
      Lifetime Max {
       bIsDynamic=false
       bIsEditable=false
       bIsStaticSwitch=false
       bIsVisible=false
       type {classStructOrEnum {refPath=&NiagaraFloat}}
      }
      Lifetime Randomness Mode {
       bIsDynamic=false
       bIsEditable=false
       bIsStaticSwitch=true
       bIsVisible=false
       type {classStructOrEnum {refPath=&ENiagaraRandomnessMode}}
      }
      Color Mode {
       bIsDynamic=false
       bIsEditable=true
       bIsStaticSwitch=true
       bIsVisible=true
       type {classStructOrEnum {refPath=/Niagara/Enums/ENiagara_ColorInitializationMode}}
      }
      Color Minimum {
       bIsDynamic=false
       bIsEditable=false
       bIsStaticSwitch=false
       bIsVisible=false
       type {classStructOrEnum {refPath=&LinearColor}}
      }
      Color {
       bIsDynamic=false
       bIsEditable=true
       bIsStaticSwitch=false
       bIsVisible=true
       type {classStructOrEnum {refPath=&LinearColor}}
      }
      AdjustHue {bIsDynamic=false, bIsEditable=false, bIsStaticSwitch=false, bIsVisible=false, type {classStructOrEnum {refPath=&NiagaraBool}}}
      Hue Shift Range {
       bIsDynamic=false
       bIsEditable=false
       bIsStaticSwitch=false
       bIsVisible=false
       type {classStructOrEnum {refPath=/Script/CoreUObject.Vector2f}}
      }
      Color Maximum {
       bIsDynamic=false
       bIsEditable=false
       bIsStaticSwitch=false
       bIsVisible=false
       type {classStructOrEnum {refPath=&LinearColor}}
      }
      AdjustSaturation {bIsDynamic=false, bIsEditable=false, bIsStaticSwitch=false, bIsVisible=false, type {classStructOrEnum {refPath=&NiagaraBool}}}
      Saturation Range {
       bIsDynamic=false
       bIsEditable=false
       bIsStaticSwitch=false
       bIsVisible=false
       type {classStructOrEnum {refPath=&Vector2f}}
      }
      AdjustValue {bIsDynamic=false, bIsEditable=false, bIsStaticSwitch=false, bIsVisible=false, type {classStructOrEnum {refPath=&NiagaraBool}}}
      Value Range {
       bIsDynamic=false
       bIsEditable=false
       bIsStaticSwitch=false
       bIsVisible=false
       type {classStructOrEnum {refPath=&Vector2f}}
      }
      AdjustAlpha {bIsDynamic=false, bIsEditable=false, bIsStaticSwitch=false, bIsVisible=false, type {classStructOrEnum {refPath=&NiagaraBool}}}
      Alpha Scale Range {
       bIsDynamic=false
       bIsEditable=false
       bIsStaticSwitch=false
       bIsVisible=false
       type {classStructOrEnum {refPath=&Vector2f}}
      }
      Color Channel Mode {
       bIsDynamic=false
       bIsEditable=false
       bIsStaticSwitch=true
       bIsVisible=false
       type {classStructOrEnum {refPath=/Niagara/Enums/ENiagaraChannelCorrelation}}
      }
      Color Randomness Mode {
       bIsDynamic=false
       bIsEditable=false
       bIsStaticSwitch=true
       bIsVisible=false
       type {classStructOrEnum {refPath=&ENiagaraRandomnessMode}}
      }
      Position Mode {
       bIsDynamic=false
       bIsEditable=true
       bIsStaticSwitch=true
       bIsVisible=true
       type {classStructOrEnum {refPath=/Niagara/Enums/ENiagara_PositionInitializationMode}}
      }
      Position {
       bIsDynamic=false
       bIsEditable=false
       bIsStaticSwitch=false
       bIsVisible=false
       type {classStructOrEnum {refPath=&NiagaraPosition}}
      }
      UsePositionOffset {bIsDynamic=false, bIsEditable=true, bIsStaticSwitch=false, bIsVisible=true, type {classStructOrEnum {refPath=&NiagaraBool}}}
      Position Offset {bIsDynamic=false, bIsEditable=true, bIsStaticSwitch=false, bIsVisible=true, type {classStructOrEnum {refPath=&Vector3f}}}
      Position Offset Coordinate Space {
       bIsDynamic=false
       bIsEditable=true
       bIsStaticSwitch=false
       bIsVisible=true
       type {classStructOrEnum {refPath=/Script/Niagara.ENiagaraCoordinateSpace}}
      }
      Mass Mode {
       bIsDynamic=false
       bIsEditable=true
       bIsStaticSwitch=true
       bIsVisible=true
       type {classStructOrEnum {refPath=/Niagara/Enums/ENiagara_MassInitializationMode}}
      }
      Mass Min {
       bIsDynamic=false
       bIsEditable=false
       bIsStaticSwitch=false
       bIsVisible=false
       type {classStructOrEnum {refPath=&NiagaraFloat}}
      }
      Mass {
       bIsDynamic=false
       bIsEditable=false
       bIsStaticSwitch=false
       bIsVisible=false
       type {classStructOrEnum {refPath=&NiagaraFloat}}
      }
      Mass Max {
       bIsDynamic=false
       bIsEditable=false
       bIsStaticSwitch=false
       bIsVisible=false
       type {classStructOrEnum {refPath=&NiagaraFloat}}
      }
      Mass Randomness Mode {
       bIsDynamic=false
       bIsEditable=false
       bIsStaticSwitch=true
       bIsVisible=false
       type {classStructOrEnum {refPath=&ENiagaraRandomnessMode}}
      }
      Write Material Random {bIsDynamic=false, bIsEditable=true, bIsStaticSwitch=false, bIsVisible=true, type {classStructOrEnum {refPath=&NiagaraBool}}}
      Material Random {bIsDynamic=false, bIsEditable=true, bIsStaticSwitch=false, bIsVisible=true, type {classStructOrEnum {refPath=&NiagaraFloat}}}
      Sprite Size Mode {
       bIsDynamic=false
       bIsEditable=true
       bIsStaticSwitch=true
       bIsVisible=true
       type {classStructOrEnum {refPath=/Niagara/Enums/ENiagara_SizeScaleMode}}
      }
      Sprite Size {
       bIsDynamic=false
       bIsEditable=false
       bIsStaticSwitch=false
       bIsVisible=false
       type {classStructOrEnum {refPath=&Vector2f}}
      }
      Uniform Sprite Size {bIsDynamic=true, bIsEditable=true, bIsStaticSwitch=false, bIsVisible=true, type {classStructOrEnum {refPath=&NiagaraFloat}}}
      Uniform Sprite Size Min {
       bIsDynamic=false
       bIsEditable=false
       bIsStaticSwitch=false
       bIsVisible=false
       type {classStructOrEnum {refPath=&NiagaraFloat}}
      }
      Sprite Size Min {
       bIsDynamic=false
       bIsEditable=false
       bIsStaticSwitch=false
       bIsVisible=false
       type {classStructOrEnum {refPath=&Vector2f}}
      }
      Sprite Size Max {
       bIsDynamic=false
       bIsEditable=false
       bIsStaticSwitch=false
       bIsVisible=false
       type {classStructOrEnum {refPath=&Vector2f}}
      }
      Uniform Sprite Size Max {
       bIsDynamic=false
       bIsEditable=false
       bIsStaticSwitch=false
       bIsVisible=false
       type {classStructOrEnum {refPath=&NiagaraFloat}}
      }
      Sprite Size Randomness Mode {
       bIsDynamic=false
       bIsEditable=false
       bIsStaticSwitch=true
       bIsVisible=false
       type {classStructOrEnum {refPath=&ENiagaraRandomnessMode}}
      }
      Sprite Rotation Mode {
       bIsDynamic=false
       bIsEditable=true
       bIsStaticSwitch=true
       bIsVisible=true
       type {classStructOrEnum {refPath=/Niagara/Enums/ENiagara_SpriteRotationMode}}
      }
      Sprite Rotation Angle Min {
       bIsDynamic=false
       bIsEditable=false
       bIsStaticSwitch=false
       bIsVisible=false
       type {classStructOrEnum {refPath=&NiagaraFloat}}
      }
      Sprite Rotation Angle {
       bIsDynamic=false
       bIsEditable=false
       bIsStaticSwitch=false
       bIsVisible=false
       type {classStructOrEnum {refPath=&NiagaraFloat}}
      }
      Sprite Rotation Angle Max {
       bIsDynamic=false
       bIsEditable=false
       bIsStaticSwitch=false
       bIsVisible=false
       type {classStructOrEnum {refPath=&NiagaraFloat}}
      }
      Sprite Rotation Randomness Mode {
       bIsDynamic=false
       bIsEditable=false
       bIsStaticSwitch=true
       bIsVisible=false
       type {classStructOrEnum {refPath=&ENiagaraRandomnessMode}}
      }
      Sprite UV Mode {
       bIsDynamic=false
       bIsEditable=true
       bIsStaticSwitch=true
       bIsVisible=true
       type {classStructOrEnum {refPath=/Niagara/Enums/ENiagara_UVFlippingMode}}
      }
      Sprite UV Randomness Mode {
       bIsDynamic=false
       bIsEditable=false
       bIsStaticSwitch=true
       bIsVisible=false
       type {classStructOrEnum {refPath=&ENiagaraRandomnessMode}}
      }
      Sprite UV Scale {
       bIsDynamic=false
       bIsEditable=false
       bIsStaticSwitch=false
       bIsVisible=false
       type {classStructOrEnum {refPath=&Vector2f}}
      }
      Mesh Scale Mode {
       bIsDynamic=false
       bIsEditable=true
       bIsStaticSwitch=true
       bIsVisible=true
       type {classStructOrEnum {refPath=&ENiagara_SizeScaleMode}}
      }
      Mesh Uniform Scale Min {
       bIsDynamic=false
       bIsEditable=false
       bIsStaticSwitch=false
       bIsVisible=false
       type {classStructOrEnum {refPath=&NiagaraFloat}}
      }
      Mesh Uniform Scale {
       bIsDynamic=false
       bIsEditable=false
       bIsStaticSwitch=false
       bIsVisible=false
       type {classStructOrEnum {refPath=&NiagaraFloat}}
      }
      Mesh Scale Min {
       bIsDynamic=false
       bIsEditable=false
       bIsStaticSwitch=false
       bIsVisible=false
       type {classStructOrEnum {refPath=&Vector3f}}
      }
      Mesh Scale {
       bIsDynamic=false
       bIsEditable=false
       bIsStaticSwitch=false
       bIsVisible=false
       type {classStructOrEnum {refPath=&Vector3f}}
      }
      Mesh Scale Max {
       bIsDynamic=false
       bIsEditable=false
       bIsStaticSwitch=false
       bIsVisible=false
       type {classStructOrEnum {refPath=&Vector3f}}
      }
      Mesh Uniform Scale Max {
       bIsDynamic=false
       bIsEditable=false
       bIsStaticSwitch=false
       bIsVisible=false
       type {classStructOrEnum {refPath=&NiagaraFloat}}
      }
      Mesh Scale Randomness Mode {
       bIsDynamic=false
       bIsEditable=false
       bIsStaticSwitch=true
       bIsVisible=false
       type {classStructOrEnum {refPath=&ENiagaraRandomnessMode}}
      }
      Mesh Renderer Array Visibility Mode {
       bIsDynamic=false
       bIsEditable=true
       bIsStaticSwitch=true
       bIsVisible=true
       type {classStructOrEnum {refPath=/Niagara/Enums/Ribbons/ENiagara_UnsetDirectSetRandom}}
      }
      Mesh Renderer Mesh Index {
       bIsDynamic=false
       bIsEditable=false
       bIsStaticSwitch=false
       bIsVisible=false
       type {classStructOrEnum {refPath=&NiagaraInt32}}
      }
      Wrap Index By Mesh Count {bIsDynamic=false, bIsEditable=false, bIsStaticSwitch=true, bIsVisible=false, type {classStructOrEnum {refPath=&NiagaraBool}}}
      Mesh Renderer Info {
       bIsDynamic=false
       bIsEditable=false
       bIsStaticSwitch=false
       bIsVisible=false
       type {classStructOrEnum {refPath=&NiagaraDataInterfaceMeshRendererInfo}}
      }
      Mesh Renderer Random Seed {
       bIsDynamic=false
       bIsEditable=false
       bIsStaticSwitch=false
       bIsVisible=false
       type {classStructOrEnum {refPath=&NiagaraInt32}}
      }
      Ribbon Width Mode {
       bIsDynamic=false
       bIsEditable=true
       bIsStaticSwitch=true
       bIsVisible=true
       type {classStructOrEnum {refPath=/Niagara/Enums/Ribbons/ENiagara_UnsetDirectSet}}
      }
      Ribbon Width {
       bIsDynamic=false
       bIsEditable=false
       bIsStaticSwitch=false
       bIsVisible=false
       type {classStructOrEnum {refPath=&NiagaraFloat}}
      }
      Ribbon Facing Mode {
       bIsDynamic=false
       bIsEditable=true
       bIsStaticSwitch=true
       bIsVisible=true
       type {classStructOrEnum {refPath=&ENiagara_UnsetDirectSet}}
      }
      Ribbon Facing Vector {
       bIsDynamic=false
       bIsEditable=false
       bIsStaticSwitch=false
       bIsVisible=false
       type {classStructOrEnum {refPath=&Vector3f}}
      }
      Ribbon Twist Mode {
       bIsDynamic=false
       bIsEditable=true
       bIsStaticSwitch=true
       bIsVisible=true
       type {classStructOrEnum {refPath=&ENiagara_UnsetDirectSet}}
      }
      "Ribbon Twist " {
       bIsDynamic=false
       bIsEditable=false
       bIsStaticSwitch=false
       bIsVisible=false
       type {classStructOrEnum {refPath=&NiagaraFloat}}
      }
     ]
     moduleScript {refPath=&InitializeParticle}
    }
    {
     bIsSetParametersModule=false
     enabled=true
     moduleName=CameraOffset
     inputs [Camera Offset Amount {bIsDynamic=false, bIsEditable=true, bIsStaticSwitch=false, bIsVisible=true, type {classStructOrEnum {refPath=&NiagaraFloat}}}]
     moduleScript {refPath=&CameraOffset}
    }
    {
     bIsSetParametersModule=false
     enabled=true
     moduleName=ShapeLocation
     inputs [
      Shape Primitive {
       bIsDynamic=false
       bIsEditable=true
       bIsStaticSwitch=true
       bIsVisible=true
       type {classStructOrEnum {refPath=/Niagara/Enums/Location/ENiagara_LocationShapes}}
      }
      Box / Plane Mode {
       bIsDynamic=false
       bIsEditable=false
       bIsStaticSwitch=true
       bIsVisible=false
       type {classStructOrEnum {refPath=/Niagara/Enums/Location/ENiagara_BoxPlaneMode}}
      }
      Cone Mode {
       bIsDynamic=false
       bIsEditable=false
       bIsStaticSwitch=true
       bIsVisible=false
       type {classStructOrEnum {refPath=/Niagara/Enums/Location/ENiagara_ConeMode}}
      }
      Cylinder Height {
       bIsDynamic=false
       bIsEditable=false
       bIsStaticSwitch=false
       bIsVisible=false
       type {classStructOrEnum {refPath=&NiagaraFloat}}
      }
      Sphere Radius {bIsDynamic=false, bIsEditable=true, bIsStaticSwitch=false, bIsVisible=true, type {classStructOrEnum {refPath=&NiagaraFloat}}}
      Ring / Disc Mode {
       bIsDynamic=false
       bIsEditable=false
       bIsStaticSwitch=true
       bIsVisible=false
       type {classStructOrEnum {refPath=/Niagara/Enums/Location/ENiagaraRingDiscMode}}
      }
      Torus Mode {
       bIsDynamic=false
       bIsEditable=false
       bIsStaticSwitch=true
       bIsVisible=false
       type {classStructOrEnum {refPath=/Niagara/Enums/ENiagaraShapeTorusMode}}
      }
      Knot Ring Loops {
       bIsDynamic=false
       bIsEditable=false
       bIsStaticSwitch=false
       bIsVisible=false
       type {classStructOrEnum {refPath=&NiagaraInt32}}
      }
      Knot Helix Loops {
       bIsDynamic=false
       bIsEditable=false
       bIsStaticSwitch=false
       bIsVisible=false
       type {classStructOrEnum {refPath=&NiagaraInt32}}
      }
      Enable Torus Knot Radius {bIsDynamic=false, bIsEditable=false, bIsStaticSwitch=false, bIsVisible=false, type {classStructOrEnum {refPath=&NiagaraBool}}}
      Torus Knot Tube Radius {
       bIsDynamic=false
       bIsEditable=false
       bIsStaticSwitch=false
       bIsVisible=false
       type {classStructOrEnum {refPath=&NiagaraFloat}}
      }
      Cylinder Radius {
       bIsDynamic=false
       bIsEditable=false
       bIsStaticSwitch=false
       bIsVisible=false
       type {classStructOrEnum {refPath=&NiagaraFloat}}
      }
      Cylinder Height Midpoint {
       bIsDynamic=false
       bIsEditable=false
       bIsStaticSwitch=false
       bIsVisible=false
       type {classStructOrEnum {refPath=&NiagaraFloat}}
      }
      Ring Radius {
       bIsDynamic=false
       bIsEditable=false
       bIsStaticSwitch=false
       bIsVisible=false
       type {classStructOrEnum {refPath=&NiagaraFloat}}
      }
      Box Size {
       bIsDynamic=false
       bIsEditable=false
       bIsStaticSwitch=false
       bIsVisible=false
       type {classStructOrEnum {refPath=&Vector3f}}
      }
      Box Midpoint {
       bIsDynamic=false
       bIsEditable=false
       bIsStaticSwitch=false
       bIsVisible=false
       type {classStructOrEnum {refPath=&Vector3f}}
      }
      Cone Length {
       bIsDynamic=false
       bIsEditable=false
       bIsStaticSwitch=false
       bIsVisible=false
       type {classStructOrEnum {refPath=&NiagaraFloat}}
      }
      Plane Size {
       bIsDynamic=false
       bIsEditable=false
       bIsStaticSwitch=false
       bIsVisible=false
       type {classStructOrEnum {refPath=&Vector2f}}
      }
      Wedge Horizontal Angle {
       bIsDynamic=false
       bIsEditable=false
       bIsStaticSwitch=false
       bIsVisible=false
       type {classStructOrEnum {refPath=&NiagaraFloat}}
      }
      Large Radius {
       bIsDynamic=false
       bIsEditable=false
       bIsStaticSwitch=false
       bIsVisible=false
       type {classStructOrEnum {refPath=&NiagaraFloat}}
      }
      Plane Midpoint {
       bIsDynamic=false
       bIsEditable=false
       bIsStaticSwitch=false
       bIsVisible=false
       type {classStructOrEnum {refPath=&Vector2f}}
      }
      Wedge Vertical Angle {
       bIsDynamic=false
       bIsEditable=false
       bIsStaticSwitch=false
       bIsVisible=false
       type {classStructOrEnum {refPath=&NiagaraFloat}}
      }
      Cone Angle {
       bIsDynamic=false
       bIsEditable=false
       bIsStaticSwitch=false
       bIsVisible=false
       type {classStructOrEnum {refPath=&NiagaraFloat}}
      }
      Handle Radius {
       bIsDynamic=false
       bIsEditable=false
       bIsStaticSwitch=false
       bIsVisible=false
       type {classStructOrEnum {refPath=&NiagaraFloat}}
      }
      Cone Inner Angle {
       bIsDynamic=false
       bIsEditable=false
       bIsStaticSwitch=false
       bIsVisible=false
       type {classStructOrEnum {refPath=&NiagaraFloat}}
      }
      Cone Radial Angle {
       bIsDynamic=false
       bIsEditable=false
       bIsStaticSwitch=false
       bIsVisible=false
       type {classStructOrEnum {refPath=&NiagaraFloat}}
      }
      Cone Angle Type {
       bIsDynamic=false
       bIsEditable=false
       bIsStaticSwitch=true
       bIsVisible=false
       type {classStructOrEnum {refPath=/Niagara/Enums/Angles/ENiagara_AngleInput}}
      }
      Flatten Endcaps {bIsDynamic=false, bIsEditable=false, bIsStaticSwitch=true, bIsVisible=false, type {classStructOrEnum {refPath=&NiagaraBool}}}
      Apply To Particle Position {
       bIsDynamic=false
       bIsEditable=true
       bIsStaticSwitch=true
       bIsVisible=true
       type {classStructOrEnum {refPath=&ENiagara_AttributeSamplingApplyOutput}}
      }
      Cone Surface Distribution {
       bIsDynamic=false
       bIsEditable=false
       bIsStaticSwitch=false
       bIsVisible=false
       type {classStructOrEnum {refPath=&NiagaraFloat}}
      }
      Ring / Disc Distribution Mode {
       bIsDynamic=false
       bIsEditable=false
       bIsStaticSwitch=true
       bIsVisible=false
       type {classStructOrEnum {refPath=/Niagara/Enums/ENiagaraSphereDistributionMode}}
      }
      Sphere Distribution {
       bIsDynamic=false
       bIsEditable=true
       bIsStaticSwitch=true
       bIsVisible=true
       type {classStructOrEnum {refPath=&ENiagaraSphereDistributionMode}}
      }
      Torus Distribution Mode {
       bIsDynamic=false
       bIsEditable=false
       bIsStaticSwitch=true
       bIsVisible=false
       type {classStructOrEnum {refPath=/Niagara/Enums/ENiagaraTorusDistributionMode}}
      }
      Sphere Surface Distribution {bIsDynamic=false, bIsEditable=true, bIsStaticSwitch=false, bIsVisible=true, type {classStructOrEnum {refPath=&NiagaraFloat}}}
      Hemisphere Distribution {bIsDynamic=false, bIsEditable=true, bIsStaticSwitch=false, bIsVisible=true, type {classStructOrEnum {refPath=&Vector2f}}}
      Cylinder Distribution {
       bIsDynamic=false
       bIsEditable=false
       bIsStaticSwitch=true
       bIsVisible=false
       type {classStructOrEnum {refPath=/Niagara/Enums/Location/ENiagara_CylinderMode}}
      }
      Surface Only {bIsDynamic=false, bIsEditable=false, bIsStaticSwitch=false, bIsVisible=false, type {classStructOrEnum {refPath=&NiagaraBool}}}
      Surface Only Band Thickness {
       bIsDynamic=false
       bIsEditable=false
       bIsStaticSwitch=false
       bIsVisible=false
       type {classStructOrEnum {refPath=&NiagaraFloat}}
      }
      Use Endcaps In Surface Only Mode {bIsDynamic=false, bIsEditable=false, bIsStaticSwitch=true, bIsVisible=false, type {classStructOrEnum {refPath=&NiagaraBool}}}
      Hemicircle X {bIsDynamic=false, bIsEditable=false, bIsStaticSwitch=true, bIsVisible=false, type {classStructOrEnum {refPath=&NiagaraBool}}}
      Hemicircle Y {bIsDynamic=false, bIsEditable=false, bIsStaticSwitch=true, bIsVisible=false, type {classStructOrEnum {refPath=&NiagaraBool}}}
      Disc Coverage {
       bIsDynamic=false
       bIsEditable=false
       bIsStaticSwitch=false
       bIsVisible=false
       type {classStructOrEnum {refPath=&NiagaraFloat}}
      }
      Hemicircle Internal Cap {bIsDynamic=false, bIsEditable=false, bIsStaticSwitch=true, bIsVisible=false, type {classStructOrEnum {refPath=&NiagaraBool}}}
      U Position {
       bIsDynamic=false
       bIsEditable=false
       bIsStaticSwitch=false
       bIsVisible=false
       type {classStructOrEnum {refPath=&NiagaraFloat}}
      }
      Enable Lathe Profile {bIsDynamic=false, bIsEditable=false, bIsStaticSwitch=true, bIsVisible=false, type {classStructOrEnum {refPath=&NiagaraBool}}}
      Lathe Profile {
       bIsDynamic=false
       bIsEditable=false
       bIsStaticSwitch=false
       bIsVisible=false
       type {classStructOrEnum {refPath=&NiagaraDataInterfaceCurve}}
      }
      Surface Distribution {
       bIsDynamic=false
       bIsEditable=false
       bIsStaticSwitch=false
       bIsVisible=false
       type {classStructOrEnum {refPath=&NiagaraFloat}}
      }
      Surface Expansion Mode {
       bIsDynamic=false
       bIsEditable=false
       bIsStaticSwitch=false
       bIsVisible=false
       type {classStructOrEnum {refPath=/Niagara/Enums/ENiagaraExpansionMode}}
      }
      U Distribution {
       bIsDynamic=false
       bIsEditable=false
       bIsStaticSwitch=false
       bIsVisible=false
       type {classStructOrEnum {refPath=&NiagaraFloat}}
      }
      V Position {bIsDynamic=true, bIsEditable=false, bIsStaticSwitch=false, bIsVisible=false, type {classStructOrEnum {refPath=&NiagaraFloat}}}
      Radius Position {
       bIsDynamic=false
       bIsEditable=false
       bIsStaticSwitch=false
       bIsVisible=false
       type {classStructOrEnum {refPath=&NiagaraFloat}}
      }
      Sphere Radius Position {
       bIsDynamic=false
       bIsEditable=false
       bIsStaticSwitch=false
       bIsVisible=false
       type {classStructOrEnum {refPath=&NiagaraFloat}}
      }
      V Distribution {
       bIsDynamic=false
       bIsEditable=false
       bIsStaticSwitch=false
       bIsVisible=false
       type {classStructOrEnum {refPath=&NiagaraFloat}}
      }
      Evaluation Type {
       bIsDynamic=false
       bIsEditable=false
       bIsStaticSwitch=true
       bIsVisible=false
       type {classStructOrEnum {refPath=&ENiagaraRandomnessEvaluation}}
      }
      Uniform Distribution {
       bIsDynamic=false
       bIsEditable=false
       bIsStaticSwitch=false
       bIsVisible=false
       type {classStructOrEnum {refPath=&NiagaraFloat}}
      }
      Uniform Spiral Amount {
       bIsDynamic=false
       bIsEditable=false
       bIsStaticSwitch=false
       bIsVisible=false
       type {classStructOrEnum {refPath=&NiagaraFloat}}
      }
      Uniform Spiral Falloff {
       bIsDynamic=false
       bIsEditable=false
       bIsStaticSwitch=false
       bIsVisible=false
       type {classStructOrEnum {refPath=&NiagaraFloat}}
      }
      Uniform Count {
       bIsDynamic=false
       bIsEditable=false
       bIsStaticSwitch=false
       bIsVisible=false
       type {classStructOrEnum {refPath=&NiagaraInt32}}
      }
      Uniform Particle Index {
       bIsDynamic=false
       bIsEditable=false
       bIsStaticSwitch=false
       bIsVisible=false
       type {classStructOrEnum {refPath=&NiagaraInt32}}
      }
      Hemisphere Angle Type {
       bIsDynamic=false
       bIsEditable=true
       bIsStaticSwitch=true
       bIsVisible=true
       type {classStructOrEnum {refPath=&ENiagara_AngleInput}}
      }
      Uniform Endpoint Offset {
       bIsDynamic=false
       bIsEditable=false
       bIsStaticSwitch=false
       bIsVisible=false
       type {classStructOrEnum {refPath=&NiagaraFloat}}
      }
      MaskSpawnGroup {bIsDynamic=false, bIsEditable=true, bIsStaticSwitch=false, bIsVisible=true, type {classStructOrEnum {refPath=&NiagaraBool}}}
      Spawn Group Mask {bIsDynamic=false, bIsEditable=false, bIsStaticSwitch=false, bIsVisible=true, type {classStructOrEnum {refPath=&NiagaraInt32}}}
      Spawn Group {
       bIsDynamic=false
       bIsEditable=false
       bIsStaticSwitch=false
       bIsVisible=false
       type {classStructOrEnum {refPath=&NiagaraInt32}}
      }
      Randomness Mode {
       bIsDynamic=false
       bIsEditable=true
       bIsStaticSwitch=true
       bIsVisible=true
       type {classStructOrEnum {refPath=&ENiagaraRandomnessMode}}
      }
      Override Seed {bIsDynamic=false, bIsEditable=true, bIsStaticSwitch=false, bIsVisible=true, type {classStructOrEnum {refPath=&NiagaraBool}}}
      Random Seed {bIsDynamic=false, bIsEditable=false, bIsStaticSwitch=false, bIsVisible=true, type {classStructOrEnum {refPath=&NiagaraInt32}}}
      Fixed Random Seed {bIsDynamic=false, bIsEditable=false, bIsStaticSwitch=true, bIsVisible=false, type {classStructOrEnum {refPath=&NiagaraBool}}}
      Transform Order {
       bIsDynamic=false
       bIsEditable=true
       bIsStaticSwitch=true
       bIsVisible=true
       type {classStructOrEnum {refPath=/Niagara/Enums/Transforms/ENiagara_TransformOrder}}
      }
      Scale Mode {
       bIsDynamic=false
       bIsEditable=true
       bIsStaticSwitch=true
       bIsVisible=true
       type {classStructOrEnum {refPath=/Niagara/Enums/Transforms/ENiagara_ScaleMode}}
      }
      Non Uniform Scale {bIsDynamic=false, bIsEditable=true, bIsStaticSwitch=false, bIsVisible=true, type {classStructOrEnum {refPath=&Vector3f}}}
      UseOwnerScale {bIsDynamic=false, bIsEditable=true, bIsStaticSwitch=false, bIsVisible=true, type {classStructOrEnum {refPath=&NiagaraBool}}}
      Apply Owner Scale {bIsDynamic=false, bIsEditable=true, bIsStaticSwitch=false, bIsVisible=true, type {classStructOrEnum {refPath=&Vector3f}}}
      Rotation Mode {
       bIsDynamic=false
       bIsEditable=true
       bIsStaticSwitch=true
       bIsVisible=true
       type {classStructOrEnum {refPath=/Niagara/Enums/Transforms/ENiagara_RotationMode}}
      }
      Rotation Axis {
       bIsDynamic=false
       bIsEditable=false
       bIsStaticSwitch=false
       bIsVisible=false
       type {classStructOrEnum {refPath=&Vector3f}}
      }
      Rotation Matrix {
       bIsDynamic=false
       bIsEditable=false
       bIsStaticSwitch=false
       bIsVisible=false
       type {classStructOrEnum {refPath=/Script/Niagara.NiagaraMatrix}}
      }
      Rotation Quaternion {bIsDynamic=false, bIsEditable=false, bIsStaticSwitch=false, bIsVisible=false, type {classStructOrEnum {refPath=/Script/CoreUObject.Quat4f}}}
      Yaw / Pitch / Roll {
       bIsDynamic=false
       bIsEditable=false
       bIsStaticSwitch=false
       bIsVisible=false
       type {classStructOrEnum {refPath=&Vector3f}}
      }
      Rotation Angle {
       bIsDynamic=false
       bIsEditable=false
       bIsStaticSwitch=false
       bIsVisible=false
       type {classStructOrEnum {refPath=&NiagaraFloat}}
      }
      Invert Rotation Quaternion {bIsDynamic=false, bIsEditable=false, bIsStaticSwitch=false, bIsVisible=false, type {classStructOrEnum {refPath=&NiagaraBool}}}
      Additional Quaternion Rotation {bIsDynamic=false, bIsEditable=false, bIsStaticSwitch=false, bIsVisible=false, type {classStructOrEnum {refPath=&NiagaraBool}}}
      Additional Yaw / Pitch / Roll {
       bIsDynamic=false
       bIsEditable=false
       bIsStaticSwitch=false
       bIsVisible=false
       type {classStructOrEnum {refPath=&Vector3f}}
      }
      Rotation Angle Type {
       bIsDynamic=false
       bIsEditable=false
       bIsStaticSwitch=true
       bIsVisible=false
       type {classStructOrEnum {refPath=&ENiagara_AngleInput}}
      }
      Rotation Coordinate Space {
       bIsDynamic=false
       bIsEditable=true
       bIsStaticSwitch=false
       bIsVisible=true
       type {classStructOrEnum {refPath=&ENiagaraCoordinateSpace}}
      }
      Offset Mode {
       bIsDynamic=false
       bIsEditable=true
       bIsStaticSwitch=true
       bIsVisible=true
       type {classStructOrEnum {refPath=/Niagara/Enums/Transforms/ENiagara_OffsetMode}}
      }
      Shape Origin {
       bIsDynamic=false
       bIsEditable=true
       bIsStaticSwitch=false
       bIsVisible=true
       type {classStructOrEnum {refPath=&NiagaraPosition}}
      }
      Offset {bIsDynamic=true, bIsEditable=true, bIsStaticSwitch=false, bIsVisible=true, type {classStructOrEnum {refPath=&Vector3f}}}
      Offset Coordinate Space {
       bIsDynamic=false
       bIsEditable=true
       bIsStaticSwitch=false
       bIsVisible=true
       type {classStructOrEnum {refPath=&ENiagaraCoordinateSpace}}
      }
      Custom Transform Matrix {
       bIsDynamic=false
       bIsEditable=false
       bIsStaticSwitch=false
       bIsVisible=false
       type {classStructOrEnum {refPath=&NiagaraMatrix}}
      }
      Transform Method {
       bIsDynamic=false
       bIsEditable=true
       bIsStaticSwitch=true
       bIsVisible=true
       type {classStructOrEnum {refPath=/Niagara/Enums/Transforms/ENiagara_TransformType}}
      }
     ]
     moduleScript {refPath=&ShapeLocation}
    }
    {
     bIsSetParametersModule=false
     enabled=true
     moduleName=AddVelocity
     inputs [
      Velocity Mode {
       bIsDynamic=false
       bIsEditable=true
       bIsStaticSwitch=true
       bIsVisible=true
       type {classStructOrEnum {refPath=/Niagara/Enums/Utility/ENiagara_VelocityMode}}
      }
      Velocity {bIsDynamic=true, bIsEditable=true, bIsStaticSwitch=false, bIsVisible=true, type {classStructOrEnum {refPath=&Vector3f}}}
      Velocity Speed {
       bIsDynamic=false
       bIsEditable=false
       bIsStaticSwitch=false
       bIsVisible=false
       type {classStructOrEnum {refPath=&NiagaraFloat}}
      }
      Velocity Speed Scale {bIsDynamic=false, bIsEditable=true, bIsStaticSwitch=false, bIsVisible=true, type {classStructOrEnum {refPath=&NiagaraFloat}}}
      Constrain To Radius {
       bIsDynamic=false
       bIsEditable=false
       bIsStaticSwitch=false
       bIsVisible=false
       type {classStructOrEnum {refPath=&NiagaraFloat}}
      }
      Distribution Along Cone Axis {
       bIsDynamic=false
       bIsEditable=false
       bIsStaticSwitch=false
       bIsVisible=false
       type {classStructOrEnum {refPath=&NiagaraFloat}}
      }
      Radius Falloff Near / Far {
       bIsDynamic=false
       bIsEditable=false
       bIsStaticSwitch=false
       bIsVisible=false
       type {classStructOrEnum {refPath=&Vector2f}}
      }
      Radius Falloff Exponent {
       bIsDynamic=false
       bIsEditable=false
       bIsStaticSwitch=false
       bIsVisible=false
       type {classStructOrEnum {refPath=&NiagaraFloat}}
      }
      Speed Falloff From Cone Axis {
       bIsDynamic=false
       bIsEditable=false
       bIsStaticSwitch=false
       bIsVisible=false
       type {classStructOrEnum {refPath=&NiagaraFloat}}
      }
      Invert Velocity Falloff {bIsDynamic=false, bIsEditable=false, bIsStaticSwitch=false, bIsVisible=false, type {classStructOrEnum {refPath=&NiagaraBool}}}
      Override Seed {bIsDynamic=false, bIsEditable=false, bIsStaticSwitch=false, bIsVisible=false, type {classStructOrEnum {refPath=&NiagaraBool}}}
      Random Seed {
       bIsDynamic=false
       bIsEditable=false
       bIsStaticSwitch=false
       bIsVisible=false
       type {classStructOrEnum {refPath=&NiagaraInt32}}
      }
      Fixed Random Seed {bIsDynamic=false, bIsEditable=false, bIsStaticSwitch=true, bIsVisible=false, type {classStructOrEnum {refPath=&NiagaraBool}}}
      Randomness Mode {
       bIsDynamic=false
       bIsEditable=false
       bIsStaticSwitch=true
       bIsVisible=false
       type {classStructOrEnum {refPath=&ENiagaraRandomnessMode}}
      }
      Cone Axis {
       bIsDynamic=false
       bIsEditable=false
       bIsStaticSwitch=false
       bIsVisible=false
       type {classStructOrEnum {refPath=&Vector3f}}
      }
      Cone Angle {
       bIsDynamic=false
       bIsEditable=false
       bIsStaticSwitch=false
       bIsVisible=false
       type {classStructOrEnum {refPath=&NiagaraFloat}}
      }
      Inner Cone Angle {
       bIsDynamic=false
       bIsEditable=false
       bIsStaticSwitch=false
       bIsVisible=false
       type {classStructOrEnum {refPath=&NiagaraFloat}}
      }
      Cone Angle Mode {
       bIsDynamic=false
       bIsEditable=false
       bIsStaticSwitch=true
       bIsVisible=false
       type {classStructOrEnum {refPath=&ENiagara_AngleInput}}
      }
      Rotation Mode {
       bIsDynamic=false
       bIsEditable=true
       bIsStaticSwitch=true
       bIsVisible=true
       type {classStructOrEnum {refPath=&ENiagara_RotationMode}}
      }
      Rotation Axis {
       bIsDynamic=false
       bIsEditable=false
       bIsStaticSwitch=false
       bIsVisible=false
       type {classStructOrEnum {refPath=&Vector3f}}
      }
      Yaw / Pitch / Roll {
       bIsDynamic=false
       bIsEditable=false
       bIsStaticSwitch=false
       bIsVisible=false
       type {classStructOrEnum {refPath=&Vector3f}}
      }
      Rotation Quaternion {bIsDynamic=false, bIsEditable=false, bIsStaticSwitch=false, bIsVisible=false, type {classStructOrEnum {refPath=&Quat4f}}}
      Rotation Matrix {
       bIsDynamic=false
       bIsEditable=false
       bIsStaticSwitch=false
       bIsVisible=false
       type {classStructOrEnum {refPath=&NiagaraMatrix}}
      }
      Rotation Angle {
       bIsDynamic=false
       bIsEditable=false
       bIsStaticSwitch=false
       bIsVisible=false
       type {classStructOrEnum {refPath=&NiagaraFloat}}
      }
      Invert Rotation Quaternion {bIsDynamic=false, bIsEditable=false, bIsStaticSwitch=false, bIsVisible=false, type {classStructOrEnum {refPath=&NiagaraBool}}}
      Additional Quaternion Rotation {bIsDynamic=false, bIsEditable=false, bIsStaticSwitch=false, bIsVisible=false, type {classStructOrEnum {refPath=&NiagaraBool}}}
      Additional Yaw / Pitch / Roll {
       bIsDynamic=false
       bIsEditable=false
       bIsStaticSwitch=false
       bIsVisible=false
       type {classStructOrEnum {refPath=&Vector3f}}
      }
      Rotation Angle Type {
       bIsDynamic=false
       bIsEditable=false
       bIsStaticSwitch=true
       bIsVisible=false
       type {classStructOrEnum {refPath=&ENiagara_AngleInput}}
      }
      Rotation Coordinate Space {
       bIsDynamic=false
       bIsEditable=true
       bIsStaticSwitch=false
       bIsVisible=true
       type {classStructOrEnum {refPath=&ENiagaraCoordinateSpace}}
      }
      Velocity Origin {
       bIsDynamic=false
       bIsEditable=false
       bIsStaticSwitch=false
       bIsVisible=false
       type {classStructOrEnum {refPath=&NiagaraPosition}}
      }
      Origin Offset {
       bIsDynamic=false
       bIsEditable=false
       bIsStaticSwitch=false
       bIsVisible=false
       type {classStructOrEnum {refPath=&Vector3f}}
      }
      Origin Offset Coordinate Space {
       bIsDynamic=false
       bIsEditable=false
       bIsStaticSwitch=false
       bIsVisible=false
       type {classStructOrEnum {refPath=&ENiagaraCoordinateSpace}}
      }
      Default Position {
       bIsDynamic=false
       bIsEditable=false
       bIsStaticSwitch=false
       bIsVisible=false
       type {classStructOrEnum {refPath=&NiagaraPosition}}
      }
     ]
     moduleScript {refPath=&AddVelocity}
    }
   ]
  }
  particleUpdateScript {
   scriptName=ParticleUpdateScript
   modules [
    {
     bIsSetParametersModule=false
     enabled=true
     moduleName=ParticleState
     inputs [
      Kill Particles When Lifetime Has Elapsed {bIsDynamic=false, bIsEditable=true, bIsStaticSwitch=true, bIsVisible=true, type {classStructOrEnum {refPath=&NiagaraBool}}}
      Loop Particles Lifetime {bIsDynamic=false, bIsEditable=false, bIsStaticSwitch=false, bIsVisible=false, type {classStructOrEnum {refPath=&NiagaraBool}}}
      Let Infinitely Lived Particles Die When Emitter Deactivates {bIsDynamic=false, bIsEditable=false, bIsStaticSwitch=false, bIsVisible=false, type {classStructOrEnum {refPath=&NiagaraBool}}}
      Lifetime {bIsDynamic=false, bIsEditable=true, bIsStaticSwitch=false, bIsVisible=true, type {classStructOrEnum {refPath=&NiagaraFloat}}}
      DeltaTime {bIsDynamic=false, bIsEditable=true, bIsStaticSwitch=false, bIsVisible=true, type {classStructOrEnum {refPath=&NiagaraFloat}}}
     ]
     moduleScript {refPath=&ParticleState}
    }
    {
     bIsSetParametersModule=false
     enabled=true
     moduleName=ScaleSpriteSize
     inputs [
      Scale Sprite Size Mode {
       bIsDynamic=false
       bIsEditable=true
       bIsStaticSwitch=true
       bIsVisible=true
       type {classStructOrEnum {refPath=/Niagara/Enums/SpriteRenderer/ENiagara_ScaleSpriteSize}}
      }
      Uniform Curve Sprite Scale {
       bIsDynamic=false
       bIsEditable=false
       bIsStaticSwitch=false
       bIsVisible=false
       type {classStructOrEnum {refPath=&NiagaraDataInterfaceCurve}}
      }
      Non-Uniform Curve Sprite Scale {
       bIsDynamic=false
       bIsEditable=false
       bIsStaticSwitch=false
       bIsVisible=false
       type {classStructOrEnum {refPath=&NiagaraDataInterfaceVector2DCurve}}
      }
      Uniform Curve Index {
       bIsDynamic=false
       bIsEditable=false
       bIsStaticSwitch=false
       bIsVisible=false
       type {classStructOrEnum {refPath=&NiagaraFloat}}
      }
      Non-Uniform Curve Index {
       bIsDynamic=false
       bIsEditable=false
       bIsStaticSwitch=false
       bIsVisible=false
       type {classStructOrEnum {refPath=&NiagaraFloat}}
      }
      Non-Uniform Curve Scale {
       bIsDynamic=false
       bIsEditable=false
       bIsStaticSwitch=false
       bIsVisible=false
       type {classStructOrEnum {refPath=&Vector2f}}
      }
      Uniform Curve Scale {
       bIsDynamic=false
       bIsEditable=false
       bIsStaticSwitch=false
       bIsVisible=false
       type {classStructOrEnum {refPath=&NiagaraFloat}}
      }
      Scale Factor {bIsDynamic=true, bIsEditable=true, bIsStaticSwitch=false, bIsVisible=true, type {classStructOrEnum {refPath=&Vector2f}}}
      Uniform Scale Factor {
       bIsDynamic=false
       bIsEditable=false
       bIsStaticSwitch=false
       bIsVisible=false
       type {classStructOrEnum {refPath=&NiagaraFloat}}
      }
      Initial Sprite Size {bIsDynamic=false, bIsEditable=true, bIsStaticSwitch=false, bIsVisible=true, type {classStructOrEnum {refPath=&Vector2f}}}
     ]
     moduleScript {refPath=&ScaleSpriteSize}
    }
    {
     bIsSetParametersModule=true
     enabled=true
     moduleName=SetVariables_#7
     inputs [
      Particles.Color02 {
       bIsDynamic=true
       bIsEditable=true
       bIsStaticSwitch=false
       bIsVisible=true
       type {classStructOrEnum {refPath=&LinearColor}}
      }
      Particles.SpriteSize02 {bIsDynamic=true, bIsEditable=true, bIsStaticSwitch=false, bIsVisible=true, type {classStructOrEnum {refPath=&Vector2f}}}
     ]
     moduleScript {refPath=&self:BigSparkles_0.NiagaraScriptSource_0.NiagaraGraph_0.NiagaraNodeAssignment_2.SetVariables_#7}
    }
    {
     bIsSetParametersModule=false
     enabled=true
     moduleName=ScaleColor
     inputs [
      Scale Mode {bIsDynamic=false, bIsEditable=true, bIsStaticSwitch=true, bIsVisible=true, type {classStructOrEnum {refPath=&NiagaraInt32}}}
      Linear Color Curve {
       bIsDynamic=false
       bIsEditable=false
       bIsStaticSwitch=false
       bIsVisible=false
       type {classStructOrEnum {refPath=&NiagaraDataInterfaceColorCurve}}
      }
      ScaleRGB {bIsDynamic=false, bIsEditable=true, bIsStaticSwitch=false, bIsVisible=true, type {classStructOrEnum {refPath=&NiagaraBool}}}
      Scale RGB {bIsDynamic=false, bIsEditable=true, bIsStaticSwitch=false, bIsVisible=true, type {classStructOrEnum {refPath=&Vector3f}}}
      ScaleRGBA {bIsDynamic=false, bIsEditable=false, bIsStaticSwitch=false, bIsVisible=false, type {classStructOrEnum {refPath=&NiagaraBool}}}
      Scale RGBA {
       bIsDynamic=false
       bIsEditable=false
       bIsStaticSwitch=false
       bIsVisible=false
       type {classStructOrEnum {refPath=/Script/CoreUObject.Vector4f}}
      }
      Curve Index {
       bIsDynamic=false
       bIsEditable=false
       bIsStaticSwitch=false
       bIsVisible=false
       type {classStructOrEnum {refPath=&NiagaraFloat}}
      }
      ScaleA {bIsDynamic=false, bIsEditable=true, bIsStaticSwitch=false, bIsVisible=true, type {classStructOrEnum {refPath=&NiagaraBool}}}
      Scale Alpha {bIsDynamic=true, bIsEditable=true, bIsStaticSwitch=false, bIsVisible=true, type {classStructOrEnum {refPath=&NiagaraFloat}}}
      Color Value To Scale {
       bIsDynamic=false
       bIsEditable=true
       bIsStaticSwitch=false
       bIsVisible=true
       type {classStructOrEnum {refPath=&LinearColor}}
      }
      Color Scale {
       bIsDynamic=false
       bIsEditable=false
       bIsStaticSwitch=false
       bIsVisible=false
       type {classStructOrEnum {refPath=&LinearColor}}
      }
     ]
     moduleScript {refPath=&ScaleColor}
    }
    {
     bIsSetParametersModule=false
     enabled=true
     moduleName=CurlNoiseForce
     inputs [
      Noise Strength {bIsDynamic=false, bIsEditable=true, bIsStaticSwitch=false, bIsVisible=true, type {classStructOrEnum {refPath=&NiagaraFloat}}}
      Noise Frequency {bIsDynamic=false, bIsEditable=true, bIsStaticSwitch=false, bIsVisible=true, type {classStructOrEnum {refPath=&NiagaraFloat}}}
      Noise Quality / Cost {
       bIsDynamic=false
       bIsEditable=true
       bIsStaticSwitch=true
       bIsVisible=true
       type {classStructOrEnum {refPath=/Niagara/Enums/ENiagaraCurlNoiseQuality}}
      }
      Pan Noise {bIsDynamic=false, bIsEditable=true, bIsStaticSwitch=false, bIsVisible=true, type {classStructOrEnum {refPath=&NiagaraBool}}}
      Pan Noise Field {bIsDynamic=false, bIsEditable=false, bIsStaticSwitch=false, bIsVisible=true, type {classStructOrEnum {refPath=&Vector3f}}}
      Bias Noise {bIsDynamic=false, bIsEditable=true, bIsStaticSwitch=false, bIsVisible=true, type {classStructOrEnum {refPath=&NiagaraBool}}}
      Bias Noise Field {bIsDynamic=false, bIsEditable=false, bIsStaticSwitch=false, bIsVisible=true, type {classStructOrEnum {refPath=&Vector3f}}}
      Write to Intrinsic Parameters {bIsDynamic=false, bIsEditable=true, bIsStaticSwitch=true, bIsVisible=true, type {classStructOrEnum {refPath=&NiagaraBool}}}
      Sample Position {
       bIsDynamic=false
       bIsEditable=true
       bIsStaticSwitch=false
       bIsVisible=true
       type {classStructOrEnum {refPath=&NiagaraPosition}}
      }
      Randomize Noise Sample {bIsDynamic=false, bIsEditable=true, bIsStaticSwitch=true, bIsVisible=true, type {classStructOrEnum {refPath=&NiagaraBool}}}
      Random Seed {bIsDynamic=false, bIsEditable=true, bIsStaticSwitch=false, bIsVisible=true, type {classStructOrEnum {refPath=&NiagaraInt32}}}
      Randomization Vector {bIsDynamic=false, bIsEditable=true, bIsStaticSwitch=false, bIsVisible=true, type {classStructOrEnum {refPath=&Vector3f}}}
      Scale Curl Lookup By Engine Owner Scale {bIsDynamic=false, bIsEditable=true, bIsStaticSwitch=true, bIsVisible=true, type {classStructOrEnum {refPath=&NiagaraBool}}}
      Coordinate Space {
       bIsDynamic=false
       bIsEditable=true
       bIsStaticSwitch=true
       bIsVisible=true
       type {classStructOrEnum {refPath=&ENiagaraCoordinateSpace}}
      }
      CurlNoise {
       bIsDynamic=false
       bIsEditable=false
       bIsStaticSwitch=false
       bIsVisible=false
       type {classStructOrEnum {refPath=&NiagaraDataInterfaceCurlNoise}}
      }
      Debug Every N Particles {
       bIsDynamic=false
       bIsEditable=false
       bIsStaticSwitch=false
       bIsVisible=false
       type {classStructOrEnum {refPath=&NiagaraInt32}}
      }
      Debug Force Line Length {
       bIsDynamic=false
       bIsEditable=false
       bIsStaticSwitch=false
       bIsVisible=false
       type {classStructOrEnum {refPath=&NiagaraFloat}}
      }
      Mask Curl Noise {bIsDynamic=false, bIsEditable=true, bIsStaticSwitch=true, bIsVisible=true, type {classStructOrEnum {refPath=&NiagaraBool}}}
      Curl Noise Cone Mask Angle {
       bIsDynamic=false
       bIsEditable=false
       bIsStaticSwitch=false
       bIsVisible=false
       type {classStructOrEnum {refPath=&NiagaraFloat}}
      }
      Curl Noise Cone Mask Falloff Angle {
       bIsDynamic=false
       bIsEditable=false
       bIsStaticSwitch=false
       bIsVisible=false
       type {classStructOrEnum {refPath=&NiagaraFloat}}
      }
      Curl Noise Cone Mask Axis {
       bIsDynamic=false
       bIsEditable=false
       bIsStaticSwitch=false
       bIsVisible=false
       type {classStructOrEnum {refPath=&Vector3f}}
      }
     ]
     moduleScript {refPath=&CurlNoiseForce}
    }
    {
     bIsSetParametersModule=false
     enabled=true
     moduleName=Drag
     inputs [
      Use Linear Drag {bIsDynamic=false, bIsEditable=true, bIsStaticSwitch=false, bIsVisible=true, type {classStructOrEnum {refPath=&NiagaraBool}}}
      Drag {bIsDynamic=true, bIsEditable=true, bIsStaticSwitch=false, bIsVisible=true, type {classStructOrEnum {refPath=&NiagaraFloat}}}
      Use Rotational Drag {bIsDynamic=false, bIsEditable=true, bIsStaticSwitch=false, bIsVisible=true, type {classStructOrEnum {refPath=&NiagaraBool}}}
      Rotational Drag {bIsDynamic=false, bIsEditable=true, bIsStaticSwitch=false, bIsVisible=true, type {classStructOrEnum {refPath=&NiagaraFloat}}}
      Ignore Mass {bIsDynamic=false, bIsEditable=true, bIsStaticSwitch=false, bIsVisible=true, type {classStructOrEnum {refPath=&NiagaraBool}}}
     ]
     moduleScript {refPath=&Drag}
    }
    {
     bIsSetParametersModule=false
     enabled=true
     moduleName=SolveForcesAndVelocity
     inputs [
      Clamp Velocity {bIsDynamic=false, bIsEditable=true, bIsStaticSwitch=false, bIsVisible=true, type {classStructOrEnum {refPath=&NiagaraBool}}}
      Speed Limit {bIsDynamic=false, bIsEditable=false, bIsStaticSwitch=false, bIsVisible=true, type {classStructOrEnum {refPath=&NiagaraFloat}}}
      Limit Acceleration {bIsDynamic=false, bIsEditable=true, bIsStaticSwitch=false, bIsVisible=true, type {classStructOrEnum {refPath=&NiagaraBool}}}
      Acceleration Limit {bIsDynamic=false, bIsEditable=false, bIsStaticSwitch=false, bIsVisible=true, type {classStructOrEnum {refPath=&NiagaraFloat}}}
      Manually Enable Rotational Solver {bIsDynamic=false, bIsEditable=true, bIsStaticSwitch=false, bIsVisible=true, type {classStructOrEnum {refPath=&NiagaraBool}}}
      Write to Intrinsic Properties {bIsDynamic=false, bIsEditable=true, bIsStaticSwitch=false, bIsVisible=true, type {classStructOrEnum {refPath=&NiagaraBool}}}
      Write to Presolve Properties {bIsDynamic=false, bIsEditable=true, bIsStaticSwitch=true, bIsVisible=true, type {classStructOrEnum {refPath=&NiagaraBool}}}
      DeltaTime {bIsDynamic=false, bIsEditable=true, bIsStaticSwitch=false, bIsVisible=true, type {classStructOrEnum {refPath=&NiagaraFloat}}}
      Mass {bIsDynamic=false, bIsEditable=true, bIsStaticSwitch=false, bIsVisible=true, type {classStructOrEnum {refPath=&NiagaraFloat}}}
      Position {
       bIsDynamic=false
       bIsEditable=true
       bIsStaticSwitch=false
       bIsVisible=true
       type {classStructOrEnum {refPath=&NiagaraPosition}}
      }
      Velocity {bIsDynamic=false, bIsEditable=true, bIsStaticSwitch=false, bIsVisible=true, type {classStructOrEnum {refPath=&Vector3f}}}
      PreviousVelocity {bIsDynamic=false, bIsEditable=true, bIsStaticSwitch=false, bIsVisible=true, type {classStructOrEnum {refPath=&Vector3f}}}
      Force {bIsDynamic=false, bIsEditable=true, bIsStaticSwitch=false, bIsVisible=true, type {classStructOrEnum {refPath=&Vector3f}}}
      Mesh Orientation {bIsDynamic=false, bIsEditable=false, bIsStaticSwitch=false, bIsVisible=false, type {classStructOrEnum {refPath=&Quat4f}}}
      Rotational Velocity {
       bIsDynamic=false
       bIsEditable=false
       bIsStaticSwitch=false
       bIsVisible=false
       type {classStructOrEnum {refPath=&Vector3f}}
      }
      Rotational Force {
       bIsDynamic=false
       bIsEditable=false
       bIsStaticSwitch=false
       bIsVisible=false
       type {classStructOrEnum {refPath=&Vector3f}}
      }
      Rotational Inertia {
       bIsDynamic=false
       bIsEditable=false
       bIsStaticSwitch=false
       bIsVisible=false
       type {classStructOrEnum {refPath=&NiagaraFloat}}
      }
     ]
     moduleScript {refPath=&SolveForcesAndVelocity}
    }
   ]
  }
  rendererClasses [{refPath=&NiagaraSpriteRendererProperties}]
  renderers [
   {rendererIndex=0, rendererClass {refPath=&NiagaraSpriteRendererProperties}}
   {rendererIndex=1, rendererClass {refPath=&NiagaraSpriteRendererProperties}}
  ]
 }
 BigSparkles002 {
  emitterSpawnScript {scriptName=EmitterSpawnScript, modules []}
  emitterUpdateScript {
   scriptName=EmitterUpdateScript
   modules [
    {
     bIsSetParametersModule=false
     enabled=true
     moduleName=EmitterState
     inputs [
      Life Cycle Mode {
       bIsDynamic=false
       bIsEditable=true
       bIsStaticSwitch=true
       bIsVisible=true
       type {classStructOrEnum {refPath=&ENiagaraEmitterLifeCycleMode}}
      }
      Inactive Response {
       bIsDynamic=false
       bIsEditable=false
       bIsStaticSwitch=true
       bIsVisible=false
       type {classStructOrEnum {refPath=&ENiagaraInactiveMode}}
      }
      Loop Behavior {
       bIsDynamic=false
       bIsEditable=false
       bIsStaticSwitch=true
       bIsVisible=false
       type {classStructOrEnum {refPath=&ENiagara_EmitterStateOptions}}
      }
      Loop Duration Mode {
       bIsDynamic=false
       bIsEditable=false
       bIsStaticSwitch=true
       bIsVisible=false
       type {classStructOrEnum {refPath=&ENiagara_InfiniteLoopDuration}}
      }
      Loop Count {
       bIsDynamic=false
       bIsEditable=false
       bIsStaticSwitch=false
       bIsVisible=false
       type {classStructOrEnum {refPath=&NiagaraInt32}}
      }
      Loop Duration {
       bIsDynamic=false
       bIsEditable=false
       bIsStaticSwitch=false
       bIsVisible=false
       type {classStructOrEnum {refPath=&NiagaraFloat}}
      }
      Recalculate Duration Each Loop {bIsDynamic=false, bIsEditable=false, bIsStaticSwitch=false, bIsVisible=false, type {classStructOrEnum {refPath=&NiagaraBool}}}
      UseLoopDelay {bIsDynamic=false, bIsEditable=false, bIsStaticSwitch=false, bIsVisible=false, type {classStructOrEnum {refPath=&NiagaraBool}}}
      Loop Delay {
       bIsDynamic=false
       bIsEditable=false
       bIsStaticSwitch=false
       bIsVisible=false
       type {classStructOrEnum {refPath=&NiagaraFloat}}
      }
      Delay First Loop Only {bIsDynamic=false, bIsEditable=false, bIsStaticSwitch=false, bIsVisible=false, type {classStructOrEnum {refPath=&NiagaraBool}}}
      Scalability Mode {
       bIsDynamic=false
       bIsEditable=true
       bIsStaticSwitch=true
       bIsVisible=true
       type {classStructOrEnum {refPath=&ENiagaraEmitterScalabilityMode_Limited}}
      }
      Enable Distance Culling {bIsDynamic=false, bIsEditable=false, bIsStaticSwitch=true, bIsVisible=false, type {classStructOrEnum {refPath=&NiagaraBool}}}
      MinDistance {
       bIsDynamic=false
       bIsEditable=false
       bIsStaticSwitch=false
       bIsVisible=false
       type {classStructOrEnum {refPath=&NiagaraFloat}}
      }
      Min Distance Response {
       bIsDynamic=false
       bIsEditable=false
       bIsStaticSwitch=false
       bIsVisible=false
       type {classStructOrEnum {refPath=&ENiagaraExecutionStateManagement}}
      }
      MaxDistance {
       bIsDynamic=false
       bIsEditable=false
       bIsStaticSwitch=false
       bIsVisible=false
       type {classStructOrEnum {refPath=&NiagaraFloat}}
      }
      Max Distance Response {
       bIsDynamic=false
       bIsEditable=false
       bIsStaticSwitch=false
       bIsVisible=false
       type {classStructOrEnum {refPath=&ENiagaraExecutionStateManagement}}
      }
      Scale Spawn Count {bIsDynamic=false, bIsEditable=false, bIsStaticSwitch=true, bIsVisible=false, type {classStructOrEnum {refPath=&NiagaraBool}}}
      Spawn Count Scale {
       bIsDynamic=false
       bIsEditable=false
       bIsStaticSwitch=false
       bIsVisible=false
       type {classStructOrEnum {refPath=&NiagaraFloat}}
      }
      Spawn Count Scale By Distance {
       bIsDynamic=false
       bIsEditable=false
       bIsStaticSwitch=false
       bIsVisible=false
       type {classStructOrEnum {refPath=&NiagaraDataInterfaceCurve}}
      }
      Enable Visibility Culling {bIsDynamic=false, bIsEditable=false, bIsStaticSwitch=true, bIsVisible=false, type {classStructOrEnum {refPath=&NiagaraBool}}}
      Visibility Cull Response {
       bIsDynamic=false
       bIsEditable=false
       bIsStaticSwitch=false
       bIsVisible=false
       type {classStructOrEnum {refPath=&ENiagaraExecutionStateManagement}}
      }
      Visibility Cull Delay {
       bIsDynamic=false
       bIsEditable=false
       bIsStaticSwitch=false
       bIsVisible=false
       type {classStructOrEnum {refPath=&NiagaraFloat}}
      }
      Reset Age On Awaken {bIsDynamic=false, bIsEditable=false, bIsStaticSwitch=false, bIsVisible=false, type {classStructOrEnum {refPath=&NiagaraBool}}}
     ]
     moduleScript {refPath=&EmitterState}
    }
    {
     bIsSetParametersModule=false
     enabled=true
     moduleName=SpawnBurst_Instantaneous
     inputs [
      Spawn Count {bIsDynamic=false, bIsEditable=true, bIsStaticSwitch=false, bIsVisible=true, type {classStructOrEnum {refPath=&NiagaraInt32}}}
      Spawn Time {bIsDynamic=false, bIsEditable=true, bIsStaticSwitch=false, bIsVisible=true, type {classStructOrEnum {refPath=&NiagaraFloat}}}
      Use Spawn Probability {bIsDynamic=false, bIsEditable=true, bIsStaticSwitch=false, bIsVisible=true, type {classStructOrEnum {refPath=&NiagaraBool}}}
      Spawn Probability {bIsDynamic=false, bIsEditable=false, bIsStaticSwitch=false, bIsVisible=true, type {classStructOrEnum {refPath=&NiagaraFloat}}}
      UseLoopCountLimit {bIsDynamic=false, bIsEditable=true, bIsStaticSwitch=false, bIsVisible=true, type {classStructOrEnum {refPath=&NiagaraBool}}}
      Loop Count Limit {bIsDynamic=false, bIsEditable=false, bIsStaticSwitch=false, bIsVisible=true, type {classStructOrEnum {refPath=&NiagaraInt32}}}
      Spawn Group {bIsDynamic=false, bIsEditable=true, bIsStaticSwitch=false, bIsVisible=true, type {classStructOrEnum {refPath=&NiagaraInt32}}}
      Age {bIsDynamic=false, bIsEditable=true, bIsStaticSwitch=false, bIsVisible=true, type {classStructOrEnum {refPath=&NiagaraFloat}}}
     ]
     moduleScript {refPath=&SpawnBurst_Instantaneous}
    }
   ]
  }
  particleSpawnScript {
   scriptName=ParticleSpawnScript
   modules [
    {
     bIsSetParametersModule=true
     enabled=true
     moduleName=SetVariables_#6
     inputs [Particles.VisibilityTag {bIsDynamic=true, bIsEditable=true, bIsStaticSwitch=false, bIsVisible=true, type {classStructOrEnum {refPath=&NiagaraInt32}}}]
     moduleScript {refPath=&self:BigSparkles002_0.NiagaraScriptSource_0.NiagaraGraph_0.NiagaraNodeAssignment_1.SetVariables_#6}
    }
    {
     bIsSetParametersModule=false
     enabled=true
     moduleName=PartitionParticles
     inputs [
      Partition Mode {
       bIsDynamic=false
       bIsEditable=true
       bIsStaticSwitch=true
       bIsVisible=true
       type {classStructOrEnum {refPath=&ENiagara_PartitionMode_v2}}
      }
      Partition Count Mode {
       bIsDynamic=false
       bIsEditable=false
       bIsStaticSwitch=true
       bIsVisible=false
       type {classStructOrEnum {refPath=&ENiagara_InfiniteLoopDuration}}
      }
      A {
       bIsDynamic=false
       bIsEditable=false
       bIsStaticSwitch=false
       bIsVisible=false
       type {classStructOrEnum {refPath=&NiagaraFloat}}
      }
      Number Of Partitions {bIsDynamic=false, bIsEditable=true, bIsStaticSwitch=false, bIsVisible=true, type {classStructOrEnum {refPath=&NiagaraInt32}}}
      Particles Per Partition {
       bIsDynamic=false
       bIsEditable=false
       bIsStaticSwitch=false
       bIsVisible=false
       type {classStructOrEnum {refPath=&NiagaraInt32}}
      }
      Partition Duration {
       bIsDynamic=false
       bIsEditable=false
       bIsStaticSwitch=false
       bIsVisible=false
       type {classStructOrEnum {refPath=&NiagaraFloat}}
      }
      Time {
       bIsDynamic=false
       bIsEditable=false
       bIsStaticSwitch=false
       bIsVisible=false
       type {classStructOrEnum {refPath=&NiagaraFloat}}
      }
      Partitioning Frequency {
       bIsDynamic=false
       bIsEditable=false
       bIsStaticSwitch=true
       bIsVisible=false
       type {classStructOrEnum {refPath=&ENiagaraRandomnessEvaluation}}
      }
      B {
       bIsDynamic=false
       bIsEditable=false
       bIsStaticSwitch=false
       bIsVisible=false
       type {classStructOrEnum {refPath=&NiagaraFloat}}
      }
      Distance {
       bIsDynamic=false
       bIsEditable=false
       bIsStaticSwitch=false
       bIsVisible=false
       type {classStructOrEnum {refPath=&NiagaraFloat}}
      }
      Comparison Type {
       bIsDynamic=false
       bIsEditable=false
       bIsStaticSwitch=true
       bIsVisible=false
       type {classStructOrEnum {refPath=&ENiagaraBooleanLogicOps_v2}}
      }
      Evaluation Type {
       bIsDynamic=false
       bIsEditable=false
       bIsStaticSwitch=true
       bIsVisible=false
       type {classStructOrEnum {refPath=&ENiagaraRandomnessEvaluation}}
      }
      UseDistanceFalloff {bIsDynamic=false, bIsEditable=false, bIsStaticSwitch=false, bIsVisible=false, type {classStructOrEnum {refPath=&NiagaraBool}}}
      Distance Falloff Exponent {
       bIsDynamic=false
       bIsEditable=false
       bIsStaticSwitch=false
       bIsVisible=false
       type {classStructOrEnum {refPath=&NiagaraFloat}}
      }
      Start Position {
       bIsDynamic=false
       bIsEditable=false
       bIsStaticSwitch=false
       bIsVisible=false
       type {classStructOrEnum {refPath=&NiagaraPosition}}
      }
      End Position {
       bIsDynamic=false
       bIsEditable=false
       bIsStaticSwitch=false
       bIsVisible=false
       type {classStructOrEnum {refPath=&NiagaraPosition}}
      }
      ID to Partition {bIsDynamic=false, bIsEditable=true, bIsStaticSwitch=false, bIsVisible=true, type {classStructOrEnum {refPath=&NiagaraInt32}}}
      Randomness Mode {
       bIsDynamic=false
       bIsEditable=false
       bIsStaticSwitch=true
       bIsVisible=false
       type {classStructOrEnum {refPath=&ENiagaraRandomnessMode}}
      }
      Override Seed {bIsDynamic=false, bIsEditable=false, bIsStaticSwitch=true, bIsVisible=false, type {classStructOrEnum {refPath=&NiagaraBool}}}
      Fixed Random Seed {bIsDynamic=false, bIsEditable=false, bIsStaticSwitch=true, bIsVisible=false, type {classStructOrEnum {refPath=&NiagaraBool}}}
      Color A {
       bIsDynamic=false
       bIsEditable=false
       bIsStaticSwitch=false
       bIsVisible=false
       type {classStructOrEnum {refPath=&LinearColor}}
      }
      Color B {
       bIsDynamic=false
       bIsEditable=false
       bIsStaticSwitch=false
       bIsVisible=false
       type {classStructOrEnum {refPath=&LinearColor}}
      }
      Color Comparison Type {bIsDynamic=false, bIsEditable=false, bIsStaticSwitch=true, bIsVisible=false, type {classStructOrEnum {refPath=&NiagaraInt32}}}
      Color Tolerance {
       bIsDynamic=false
       bIsEditable=false
       bIsStaticSwitch=false
       bIsVisible=false
       type {classStructOrEnum {refPath=&NiagaraFloat}}
      }
      Minimum Saturation {
       bIsDynamic=false
       bIsEditable=false
       bIsStaticSwitch=false
       bIsVisible=false
       type {classStructOrEnum {refPath=&NiagaraFloat}}
      }
      Direction A {
       bIsDynamic=false
       bIsEditable=false
       bIsStaticSwitch=false
       bIsVisible=false
       type {classStructOrEnum {refPath=&Vector3f}}
      }
      Direction B {
       bIsDynamic=false
       bIsEditable=false
       bIsStaticSwitch=false
       bIsVisible=false
       type {classStructOrEnum {refPath=&Vector3f}}
      }
      Direction Tolerance {
       bIsDynamic=false
       bIsEditable=false
       bIsStaticSwitch=false
       bIsVisible=false
       type {classStructOrEnum {refPath=&NiagaraFloat}}
      }
      Write Partition to Ribbon ID {
       bIsDynamic=false
       bIsEditable=true
       bIsStaticSwitch=true
       bIsVisible=true
       type {classStructOrEnum {refPath=&ENiagara_AttributeSamplingApplyOutput}}
      }
      Write Partition Index to Ribbon Link Order {
       bIsDynamic=false
       bIsEditable=false
       bIsStaticSwitch=true
       bIsVisible=false
       type {classStructOrEnum {refPath=&ENiagara_AttributeSamplingApplyOutput}}
      }
      Normalized Index In Partition Includes One {bIsDynamic=false, bIsEditable=false, bIsStaticSwitch=true, bIsVisible=false, type {classStructOrEnum {refPath=&NiagaraBool}}}
      Normalized Index Includes One {bIsDynamic=false, bIsEditable=true, bIsStaticSwitch=true, bIsVisible=true, type {classStructOrEnum {refPath=&NiagaraBool}}}
     ]
     moduleScript {refPath=&PartitionParticles}
    }
    {
     bIsSetParametersModule=false
     enabled=true
     moduleName=InitializeParticle
     inputs [
      Lifetime Mode {
       bIsDynamic=false
       bIsEditable=true
       bIsStaticSwitch=true
       bIsVisible=true
       type {classStructOrEnum {refPath=&ENiagara_LifetimeMode}}
      }
      Lifetime Min {
       bIsDynamic=false
       bIsEditable=false
       bIsStaticSwitch=false
       bIsVisible=false
       type {classStructOrEnum {refPath=&NiagaraFloat}}
      }
      Lifetime {bIsDynamic=true, bIsEditable=true, bIsStaticSwitch=false, bIsVisible=true, type {classStructOrEnum {refPath=&NiagaraFloat}}}
      Lifetime Max {
       bIsDynamic=false
       bIsEditable=false
       bIsStaticSwitch=false
       bIsVisible=false
       type {classStructOrEnum {refPath=&NiagaraFloat}}
      }
      Lifetime Randomness Mode {
       bIsDynamic=false
       bIsEditable=false
       bIsStaticSwitch=true
       bIsVisible=false
       type {classStructOrEnum {refPath=&ENiagaraRandomnessMode}}
      }
      Color Mode {
       bIsDynamic=false
       bIsEditable=true
       bIsStaticSwitch=true
       bIsVisible=true
       type {classStructOrEnum {refPath=&ENiagara_ColorInitializationMode}}
      }
      Color Minimum {
       bIsDynamic=false
       bIsEditable=false
       bIsStaticSwitch=false
       bIsVisible=false
       type {classStructOrEnum {refPath=&LinearColor}}
      }
      Color {
       bIsDynamic=false
       bIsEditable=true
       bIsStaticSwitch=false
       bIsVisible=true
       type {classStructOrEnum {refPath=&LinearColor}}
      }
      AdjustHue {bIsDynamic=false, bIsEditable=false, bIsStaticSwitch=false, bIsVisible=false, type {classStructOrEnum {refPath=&NiagaraBool}}}
      Hue Shift Range {
       bIsDynamic=false
       bIsEditable=false
       bIsStaticSwitch=false
       bIsVisible=false
       type {classStructOrEnum {refPath=&Vector2f}}
      }
      Color Maximum {
       bIsDynamic=false
       bIsEditable=false
       bIsStaticSwitch=false
       bIsVisible=false
       type {classStructOrEnum {refPath=&LinearColor}}
      }
      AdjustSaturation {bIsDynamic=false, bIsEditable=false, bIsStaticSwitch=false, bIsVisible=false, type {classStructOrEnum {refPath=&NiagaraBool}}}
      Saturation Range {
       bIsDynamic=false
       bIsEditable=false
       bIsStaticSwitch=false
       bIsVisible=false
       type {classStructOrEnum {refPath=&Vector2f}}
      }
      AdjustValue {bIsDynamic=false, bIsEditable=false, bIsStaticSwitch=false, bIsVisible=false, type {classStructOrEnum {refPath=&NiagaraBool}}}
      Value Range {
       bIsDynamic=false
       bIsEditable=false
       bIsStaticSwitch=false
       bIsVisible=false
       type {classStructOrEnum {refPath=&Vector2f}}
      }
      AdjustAlpha {bIsDynamic=false, bIsEditable=false, bIsStaticSwitch=false, bIsVisible=false, type {classStructOrEnum {refPath=&NiagaraBool}}}
      Alpha Scale Range {
       bIsDynamic=false
       bIsEditable=false
       bIsStaticSwitch=false
       bIsVisible=false
       type {classStructOrEnum {refPath=&Vector2f}}
      }
      Color Channel Mode {
       bIsDynamic=false
       bIsEditable=false
       bIsStaticSwitch=true
       bIsVisible=false
       type {classStructOrEnum {refPath=&ENiagaraChannelCorrelation}}
      }
      Color Randomness Mode {
       bIsDynamic=false
       bIsEditable=false
       bIsStaticSwitch=true
       bIsVisible=false
       type {classStructOrEnum {refPath=&ENiagaraRandomnessMode}}
      }
      Position Mode {
       bIsDynamic=false
       bIsEditable=true
       bIsStaticSwitch=true
       bIsVisible=true
       type {classStructOrEnum {refPath=&ENiagara_PositionInitializationMode}}
      }
      Position {
       bIsDynamic=false
       bIsEditable=false
       bIsStaticSwitch=false
       bIsVisible=false
       type {classStructOrEnum {refPath=&NiagaraPosition}}
      }
      UsePositionOffset {bIsDynamic=false, bIsEditable=true, bIsStaticSwitch=false, bIsVisible=true, type {classStructOrEnum {refPath=&NiagaraBool}}}
      Position Offset {bIsDynamic=false, bIsEditable=true, bIsStaticSwitch=false, bIsVisible=true, type {classStructOrEnum {refPath=&Vector3f}}}
      Position Offset Coordinate Space {
       bIsDynamic=false
       bIsEditable=true
       bIsStaticSwitch=false
       bIsVisible=true
       type {classStructOrEnum {refPath=&ENiagaraCoordinateSpace}}
      }
      Mass Mode {
       bIsDynamic=false
       bIsEditable=true
       bIsStaticSwitch=true
       bIsVisible=true
       type {classStructOrEnum {refPath=&ENiagara_MassInitializationMode}}
      }
      Mass Min {
       bIsDynamic=false
       bIsEditable=false
       bIsStaticSwitch=false
       bIsVisible=false
       type {classStructOrEnum {refPath=&NiagaraFloat}}
      }
      Mass {
       bIsDynamic=false
       bIsEditable=false
       bIsStaticSwitch=false
       bIsVisible=false
       type {classStructOrEnum {refPath=&NiagaraFloat}}
      }
      Mass Max {
       bIsDynamic=false
       bIsEditable=false
       bIsStaticSwitch=false
       bIsVisible=false
       type {classStructOrEnum {refPath=&NiagaraFloat}}
      }
      Mass Randomness Mode {
       bIsDynamic=false
       bIsEditable=false
       bIsStaticSwitch=true
       bIsVisible=false
       type {classStructOrEnum {refPath=&ENiagaraRandomnessMode}}
      }
      Write Material Random {bIsDynamic=false, bIsEditable=true, bIsStaticSwitch=false, bIsVisible=true, type {classStructOrEnum {refPath=&NiagaraBool}}}
      Material Random {bIsDynamic=false, bIsEditable=true, bIsStaticSwitch=false, bIsVisible=true, type {classStructOrEnum {refPath=&NiagaraFloat}}}
      Sprite Size Mode {
       bIsDynamic=false
       bIsEditable=true
       bIsStaticSwitch=true
       bIsVisible=true
       type {classStructOrEnum {refPath=&ENiagara_SizeScaleMode}}
      }
      Sprite Size {
       bIsDynamic=false
       bIsEditable=false
       bIsStaticSwitch=false
       bIsVisible=false
       type {classStructOrEnum {refPath=&Vector2f}}
      }
      Uniform Sprite Size {bIsDynamic=true, bIsEditable=true, bIsStaticSwitch=false, bIsVisible=true, type {classStructOrEnum {refPath=&NiagaraFloat}}}
      Uniform Sprite Size Min {
       bIsDynamic=false
       bIsEditable=false
       bIsStaticSwitch=false
       bIsVisible=false
       type {classStructOrEnum {refPath=&NiagaraFloat}}
      }
      Sprite Size Min {
       bIsDynamic=false
       bIsEditable=false
       bIsStaticSwitch=false
       bIsVisible=false
       type {classStructOrEnum {refPath=&Vector2f}}
      }
      Sprite Size Max {
       bIsDynamic=false
       bIsEditable=false
       bIsStaticSwitch=false
       bIsVisible=false
       type {classStructOrEnum {refPath=&Vector2f}}
      }
      Uniform Sprite Size Max {
       bIsDynamic=false
       bIsEditable=false
       bIsStaticSwitch=false
       bIsVisible=false
       type {classStructOrEnum {refPath=&NiagaraFloat}}
      }
      Sprite Size Randomness Mode {
       bIsDynamic=false
       bIsEditable=false
       bIsStaticSwitch=true
       bIsVisible=false
       type {classStructOrEnum {refPath=&ENiagaraRandomnessMode}}
      }
      Sprite Rotation Mode {
       bIsDynamic=false
       bIsEditable=true
       bIsStaticSwitch=true
       bIsVisible=true
       type {classStructOrEnum {refPath=&ENiagara_SpriteRotationMode}}
      }
      Sprite Rotation Angle Min {
       bIsDynamic=false
       bIsEditable=false
       bIsStaticSwitch=false
       bIsVisible=false
       type {classStructOrEnum {refPath=&NiagaraFloat}}
      }
      Sprite Rotation Angle {
       bIsDynamic=false
       bIsEditable=false
       bIsStaticSwitch=false
       bIsVisible=false
       type {classStructOrEnum {refPath=&NiagaraFloat}}
      }
      Sprite Rotation Angle Max {
       bIsDynamic=false
       bIsEditable=false
       bIsStaticSwitch=false
       bIsVisible=false
       type {classStructOrEnum {refPath=&NiagaraFloat}}
      }
      Sprite Rotation Randomness Mode {
       bIsDynamic=false
       bIsEditable=false
       bIsStaticSwitch=true
       bIsVisible=false
       type {classStructOrEnum {refPath=&ENiagaraRandomnessMode}}
      }
      Sprite UV Mode {
       bIsDynamic=false
       bIsEditable=true
       bIsStaticSwitch=true
       bIsVisible=true
       type {classStructOrEnum {refPath=&ENiagara_UVFlippingMode}}
      }
      Sprite UV Randomness Mode {
       bIsDynamic=false
       bIsEditable=false
       bIsStaticSwitch=true
       bIsVisible=false
       type {classStructOrEnum {refPath=&ENiagaraRandomnessMode}}
      }
      Sprite UV Scale {
       bIsDynamic=false
       bIsEditable=false
       bIsStaticSwitch=false
       bIsVisible=false
       type {classStructOrEnum {refPath=&Vector2f}}
      }
      Mesh Scale Mode {
       bIsDynamic=false
       bIsEditable=true
       bIsStaticSwitch=true
       bIsVisible=true
       type {classStructOrEnum {refPath=&ENiagara_SizeScaleMode}}
      }
      Mesh Uniform Scale Min {
       bIsDynamic=false
       bIsEditable=false
       bIsStaticSwitch=false
       bIsVisible=false
       type {classStructOrEnum {refPath=&NiagaraFloat}}
      }
      Mesh Uniform Scale {
       bIsDynamic=false
       bIsEditable=false
       bIsStaticSwitch=false
       bIsVisible=false
       type {classStructOrEnum {refPath=&NiagaraFloat}}
      }
      Mesh Scale Min {
       bIsDynamic=false
       bIsEditable=false
       bIsStaticSwitch=false
       bIsVisible=false
       type {classStructOrEnum {refPath=&Vector3f}}
      }
      Mesh Scale {
       bIsDynamic=false
       bIsEditable=false
       bIsStaticSwitch=false
       bIsVisible=false
       type {classStructOrEnum {refPath=&Vector3f}}
      }
      Mesh Scale Max {
       bIsDynamic=false
       bIsEditable=false
       bIsStaticSwitch=false
       bIsVisible=false
       type {classStructOrEnum {refPath=&Vector3f}}
      }
      Mesh Uniform Scale Max {
       bIsDynamic=false
       bIsEditable=false
       bIsStaticSwitch=false
       bIsVisible=false
       type {classStructOrEnum {refPath=&NiagaraFloat}}
      }
      Mesh Scale Randomness Mode {
       bIsDynamic=false
       bIsEditable=false
       bIsStaticSwitch=true
       bIsVisible=false
       type {classStructOrEnum {refPath=&ENiagaraRandomnessMode}}
      }
      Mesh Renderer Array Visibility Mode {
       bIsDynamic=false
       bIsEditable=true
       bIsStaticSwitch=true
       bIsVisible=true
       type {classStructOrEnum {refPath=&ENiagara_UnsetDirectSetRandom}}
      }
      Mesh Renderer Mesh Index {
       bIsDynamic=false
       bIsEditable=false
       bIsStaticSwitch=false
       bIsVisible=false
       type {classStructOrEnum {refPath=&NiagaraInt32}}
      }
      Wrap Index By Mesh Count {bIsDynamic=false, bIsEditable=false, bIsStaticSwitch=true, bIsVisible=false, type {classStructOrEnum {refPath=&NiagaraBool}}}
      Mesh Renderer Info {
       bIsDynamic=false
       bIsEditable=false
       bIsStaticSwitch=false
       bIsVisible=false
       type {classStructOrEnum {refPath=&NiagaraDataInterfaceMeshRendererInfo}}
      }
      Mesh Renderer Random Seed {
       bIsDynamic=false
       bIsEditable=false
       bIsStaticSwitch=false
       bIsVisible=false
       type {classStructOrEnum {refPath=&NiagaraInt32}}
      }
      Ribbon Width Mode {
       bIsDynamic=false
       bIsEditable=true
       bIsStaticSwitch=true
       bIsVisible=true
       type {classStructOrEnum {refPath=&ENiagara_UnsetDirectSet}}
      }
      Ribbon Width {
       bIsDynamic=false
       bIsEditable=false
       bIsStaticSwitch=false
       bIsVisible=false
       type {classStructOrEnum {refPath=&NiagaraFloat}}
      }
      Ribbon Facing Mode {
       bIsDynamic=false
       bIsEditable=true
       bIsStaticSwitch=true
       bIsVisible=true
       type {classStructOrEnum {refPath=&ENiagara_UnsetDirectSet}}
      }
      Ribbon Facing Vector {
       bIsDynamic=false
       bIsEditable=false
       bIsStaticSwitch=false
       bIsVisible=false
       type {classStructOrEnum {refPath=&Vector3f}}
      }
      Ribbon Twist Mode {
       bIsDynamic=false
       bIsEditable=true
       bIsStaticSwitch=true
       bIsVisible=true
       type {classStructOrEnum {refPath=&ENiagara_UnsetDirectSet}}
      }
      "Ribbon Twist " {
       bIsDynamic=false
       bIsEditable=false
       bIsStaticSwitch=false
       bIsVisible=false
       type {classStructOrEnum {refPath=&NiagaraFloat}}
      }
     ]
     moduleScript {refPath=&InitializeParticle}
    }
    {
     bIsSetParametersModule=false
     enabled=true
     moduleName=CameraOffset
     inputs [Camera Offset Amount {bIsDynamic=false, bIsEditable=true, bIsStaticSwitch=false, bIsVisible=true, type {classStructOrEnum {refPath=&NiagaraFloat}}}]
     moduleScript {refPath=&CameraOffset}
    }
    {
     bIsSetParametersModule=false
     enabled=true
     moduleName=ShapeLocation
     inputs [
      Shape Primitive {
       bIsDynamic=false
       bIsEditable=true
       bIsStaticSwitch=true
       bIsVisible=true
       type {classStructOrEnum {refPath=&ENiagara_LocationShapes}}
      }
      Box / Plane Mode {
       bIsDynamic=false
       bIsEditable=false
       bIsStaticSwitch=true
       bIsVisible=false
       type {classStructOrEnum {refPath=&ENiagara_BoxPlaneMode}}
      }
      Cone Mode {
       bIsDynamic=false
       bIsEditable=false
       bIsStaticSwitch=true
       bIsVisible=false
       type {classStructOrEnum {refPath=&ENiagara_ConeMode}}
      }
      Cylinder Height {
       bIsDynamic=false
       bIsEditable=false
       bIsStaticSwitch=false
       bIsVisible=false
       type {classStructOrEnum {refPath=&NiagaraFloat}}
      }
      Sphere Radius {bIsDynamic=false, bIsEditable=true, bIsStaticSwitch=false, bIsVisible=true, type {classStructOrEnum {refPath=&NiagaraFloat}}}
      Ring / Disc Mode {
       bIsDynamic=false
       bIsEditable=false
       bIsStaticSwitch=true
       bIsVisible=false
       type {classStructOrEnum {refPath=&ENiagaraRingDiscMode}}
      }
      Torus Mode {
       bIsDynamic=false
       bIsEditable=false
       bIsStaticSwitch=true
       bIsVisible=false
       type {classStructOrEnum {refPath=&ENiagaraShapeTorusMode}}
      }
      Knot Ring Loops {
       bIsDynamic=false
       bIsEditable=false
       bIsStaticSwitch=false
       bIsVisible=false
       type {classStructOrEnum {refPath=&NiagaraInt32}}
      }
      Knot Helix Loops {
       bIsDynamic=false
       bIsEditable=false
       bIsStaticSwitch=false
       bIsVisible=false
       type {classStructOrEnum {refPath=&NiagaraInt32}}
      }
      Enable Torus Knot Radius {bIsDynamic=false, bIsEditable=false, bIsStaticSwitch=false, bIsVisible=false, type {classStructOrEnum {refPath=&NiagaraBool}}}
      Torus Knot Tube Radius {
       bIsDynamic=false
       bIsEditable=false
       bIsStaticSwitch=false
       bIsVisible=false
       type {classStructOrEnum {refPath=&NiagaraFloat}}
      }
      Cylinder Radius {
       bIsDynamic=false
       bIsEditable=false
       bIsStaticSwitch=false
       bIsVisible=false
       type {classStructOrEnum {refPath=&NiagaraFloat}}
      }
      Cylinder Height Midpoint {
       bIsDynamic=false
       bIsEditable=false
       bIsStaticSwitch=false
       bIsVisible=false
       type {classStructOrEnum {refPath=&NiagaraFloat}}
      }
      Ring Radius {
       bIsDynamic=false
       bIsEditable=false
       bIsStaticSwitch=false
       bIsVisible=false
       type {classStructOrEnum {refPath=&NiagaraFloat}}
      }
      Box Size {
       bIsDynamic=false
       bIsEditable=false
       bIsStaticSwitch=false
       bIsVisible=false
       type {classStructOrEnum {refPath=&Vector3f}}
      }
      Box Midpoint {
       bIsDynamic=false
       bIsEditable=false
       bIsStaticSwitch=false
       bIsVisible=false
       type {classStructOrEnum {refPath=&Vector3f}}
      }
      Cone Length {
       bIsDynamic=false
       bIsEditable=false
       bIsStaticSwitch=false
       bIsVisible=false
       type {classStructOrEnum {refPath=&NiagaraFloat}}
      }
      Plane Size {
       bIsDynamic=false
       bIsEditable=false
       bIsStaticSwitch=false
       bIsVisible=false
       type {classStructOrEnum {refPath=&Vector2f}}
      }
      Wedge Horizontal Angle {
       bIsDynamic=false
       bIsEditable=false
       bIsStaticSwitch=false
       bIsVisible=false
       type {classStructOrEnum {refPath=&NiagaraFloat}}
      }
      Large Radius {
       bIsDynamic=false
       bIsEditable=false
       bIsStaticSwitch=false
       bIsVisible=false
       type {classStructOrEnum {refPath=&NiagaraFloat}}
      }
      Plane Midpoint {
       bIsDynamic=false
       bIsEditable=false
       bIsStaticSwitch=false
       bIsVisible=false
       type {classStructOrEnum {refPath=&Vector2f}}
      }
      Wedge Vertical Angle {
       bIsDynamic=false
       bIsEditable=false
       bIsStaticSwitch=false
       bIsVisible=false
       type {classStructOrEnum {refPath=&NiagaraFloat}}
      }
      Cone Angle {
       bIsDynamic=false
       bIsEditable=false
       bIsStaticSwitch=false
       bIsVisible=false
       type {classStructOrEnum {refPath=&NiagaraFloat}}
      }
      Handle Radius {
       bIsDynamic=false
       bIsEditable=false
       bIsStaticSwitch=false
       bIsVisible=false
       type {classStructOrEnum {refPath=&NiagaraFloat}}
      }
      Cone Inner Angle {
       bIsDynamic=false
       bIsEditable=false
       bIsStaticSwitch=false
       bIsVisible=false
       type {classStructOrEnum {refPath=&NiagaraFloat}}
      }
      Cone Radial Angle {
       bIsDynamic=false
       bIsEditable=false
       bIsStaticSwitch=false
       bIsVisible=false
       type {classStructOrEnum {refPath=&NiagaraFloat}}
      }
      Cone Angle Type {
       bIsDynamic=false
       bIsEditable=false
       bIsStaticSwitch=true
       bIsVisible=false
       type {classStructOrEnum {refPath=&ENiagara_AngleInput}}
      }
      Flatten Endcaps {bIsDynamic=false, bIsEditable=false, bIsStaticSwitch=true, bIsVisible=false, type {classStructOrEnum {refPath=&NiagaraBool}}}
      Apply To Particle Position {
       bIsDynamic=false
       bIsEditable=true
       bIsStaticSwitch=true
       bIsVisible=true
       type {classStructOrEnum {refPath=&ENiagara_AttributeSamplingApplyOutput}}
      }
      Cone Surface Distribution {
       bIsDynamic=false
       bIsEditable=false
       bIsStaticSwitch=false
       bIsVisible=false
       type {classStructOrEnum {refPath=&NiagaraFloat}}
      }
      Ring / Disc Distribution Mode {
       bIsDynamic=false
       bIsEditable=false
       bIsStaticSwitch=true
       bIsVisible=false
       type {classStructOrEnum {refPath=&ENiagaraSphereDistributionMode}}
      }
      Sphere Distribution {
       bIsDynamic=false
       bIsEditable=true
       bIsStaticSwitch=true
       bIsVisible=true
       type {classStructOrEnum {refPath=&ENiagaraSphereDistributionMode}}
      }
      Torus Distribution Mode {
       bIsDynamic=false
       bIsEditable=false
       bIsStaticSwitch=true
       bIsVisible=false
       type {classStructOrEnum {refPath=&ENiagaraTorusDistributionMode}}
      }
      Sphere Surface Distribution {bIsDynamic=false, bIsEditable=true, bIsStaticSwitch=false, bIsVisible=true, type {classStructOrEnum {refPath=&NiagaraFloat}}}
      Hemisphere Distribution {bIsDynamic=false, bIsEditable=true, bIsStaticSwitch=false, bIsVisible=true, type {classStructOrEnum {refPath=&Vector2f}}}
      Cylinder Distribution {
       bIsDynamic=false
       bIsEditable=false
       bIsStaticSwitch=true
       bIsVisible=false
       type {classStructOrEnum {refPath=&ENiagara_CylinderMode}}
      }
      Surface Only {bIsDynamic=false, bIsEditable=false, bIsStaticSwitch=false, bIsVisible=false, type {classStructOrEnum {refPath=&NiagaraBool}}}
      Surface Only Band Thickness {
       bIsDynamic=false
       bIsEditable=false
       bIsStaticSwitch=false
       bIsVisible=false
       type {classStructOrEnum {refPath=&NiagaraFloat}}
      }
      Use Endcaps In Surface Only Mode {bIsDynamic=false, bIsEditable=false, bIsStaticSwitch=true, bIsVisible=false, type {classStructOrEnum {refPath=&NiagaraBool}}}
      Hemicircle X {bIsDynamic=false, bIsEditable=false, bIsStaticSwitch=true, bIsVisible=false, type {classStructOrEnum {refPath=&NiagaraBool}}}
      Hemicircle Y {bIsDynamic=false, bIsEditable=false, bIsStaticSwitch=true, bIsVisible=false, type {classStructOrEnum {refPath=&NiagaraBool}}}
      Disc Coverage {
       bIsDynamic=false
       bIsEditable=false
       bIsStaticSwitch=false
       bIsVisible=false
       type {classStructOrEnum {refPath=&NiagaraFloat}}
      }
      Hemicircle Internal Cap {bIsDynamic=false, bIsEditable=false, bIsStaticSwitch=true, bIsVisible=false, type {classStructOrEnum {refPath=&NiagaraBool}}}
      U Position {
       bIsDynamic=false
       bIsEditable=false
       bIsStaticSwitch=false
       bIsVisible=false
       type {classStructOrEnum {refPath=&NiagaraFloat}}
      }
      Enable Lathe Profile {bIsDynamic=false, bIsEditable=false, bIsStaticSwitch=true, bIsVisible=false, type {classStructOrEnum {refPath=&NiagaraBool}}}
      Lathe Profile {
       bIsDynamic=false
       bIsEditable=false
       bIsStaticSwitch=false
       bIsVisible=false
       type {classStructOrEnum {refPath=&NiagaraDataInterfaceCurve}}
      }
      Surface Distribution {
       bIsDynamic=false
       bIsEditable=false
       bIsStaticSwitch=false
       bIsVisible=false
       type {classStructOrEnum {refPath=&NiagaraFloat}}
      }
      Surface Expansion Mode {
       bIsDynamic=false
       bIsEditable=false
       bIsStaticSwitch=false
       bIsVisible=false
       type {classStructOrEnum {refPath=&ENiagaraExpansionMode}}
      }
      U Distribution {
       bIsDynamic=false
       bIsEditable=false
       bIsStaticSwitch=false
       bIsVisible=false
       type {classStructOrEnum {refPath=&NiagaraFloat}}
      }
      V Position {bIsDynamic=true, bIsEditable=false, bIsStaticSwitch=false, bIsVisible=false, type {classStructOrEnum {refPath=&NiagaraFloat}}}
      Radius Position {
       bIsDynamic=false
       bIsEditable=false
       bIsStaticSwitch=false
       bIsVisible=false
       type {classStructOrEnum {refPath=&NiagaraFloat}}
      }
      Sphere Radius Position {
       bIsDynamic=false
       bIsEditable=false
       bIsStaticSwitch=false
       bIsVisible=false
       type {classStructOrEnum {refPath=&NiagaraFloat}}
      }
      V Distribution {
       bIsDynamic=false
       bIsEditable=false
       bIsStaticSwitch=false
       bIsVisible=false
       type {classStructOrEnum {refPath=&NiagaraFloat}}
      }
      Evaluation Type {
       bIsDynamic=false
       bIsEditable=false
       bIsStaticSwitch=true
       bIsVisible=false
       type {classStructOrEnum {refPath=&ENiagaraRandomnessEvaluation}}
      }
      Uniform Distribution {
       bIsDynamic=false
       bIsEditable=false
       bIsStaticSwitch=false
       bIsVisible=false
       type {classStructOrEnum {refPath=&NiagaraFloat}}
      }
      Uniform Spiral Amount {
       bIsDynamic=false
       bIsEditable=false
       bIsStaticSwitch=false
       bIsVisible=false
       type {classStructOrEnum {refPath=&NiagaraFloat}}
      }
      Uniform Spiral Falloff {
       bIsDynamic=false
       bIsEditable=false
       bIsStaticSwitch=false
       bIsVisible=false
       type {classStructOrEnum {refPath=&NiagaraFloat}}
      }
      Uniform Count {
       bIsDynamic=false
       bIsEditable=false
       bIsStaticSwitch=false
       bIsVisible=false
       type {classStructOrEnum {refPath=&NiagaraInt32}}
      }
      Uniform Particle Index {
       bIsDynamic=false
       bIsEditable=false
       bIsStaticSwitch=false
       bIsVisible=false
       type {classStructOrEnum {refPath=&NiagaraInt32}}
      }
      Hemisphere Angle Type {
       bIsDynamic=false
       bIsEditable=true
       bIsStaticSwitch=true
       bIsVisible=true
       type {classStructOrEnum {refPath=&ENiagara_AngleInput}}
      }
      Uniform Endpoint Offset {
       bIsDynamic=false
       bIsEditable=false
       bIsStaticSwitch=false
       bIsVisible=false
       type {classStructOrEnum {refPath=&NiagaraFloat}}
      }
      MaskSpawnGroup {bIsDynamic=false, bIsEditable=true, bIsStaticSwitch=false, bIsVisible=true, type {classStructOrEnum {refPath=&NiagaraBool}}}
      Spawn Group Mask {bIsDynamic=false, bIsEditable=false, bIsStaticSwitch=false, bIsVisible=true, type {classStructOrEnum {refPath=&NiagaraInt32}}}
      Spawn Group {
       bIsDynamic=false
       bIsEditable=false
       bIsStaticSwitch=false
       bIsVisible=false
       type {classStructOrEnum {refPath=&NiagaraInt32}}
      }
      Randomness Mode {
       bIsDynamic=false
       bIsEditable=true
       bIsStaticSwitch=true
       bIsVisible=true
       type {classStructOrEnum {refPath=&ENiagaraRandomnessMode}}
      }
      Override Seed {bIsDynamic=false, bIsEditable=true, bIsStaticSwitch=false, bIsVisible=true, type {classStructOrEnum {refPath=&NiagaraBool}}}
      Random Seed {bIsDynamic=false, bIsEditable=false, bIsStaticSwitch=false, bIsVisible=true, type {classStructOrEnum {refPath=&NiagaraInt32}}}
      Fixed Random Seed {bIsDynamic=false, bIsEditable=false, bIsStaticSwitch=true, bIsVisible=false, type {classStructOrEnum {refPath=&NiagaraBool}}}
      Transform Order {
       bIsDynamic=false
       bIsEditable=true
       bIsStaticSwitch=true
       bIsVisible=true
       type {classStructOrEnum {refPath=&ENiagara_TransformOrder}}
      }
      Scale Mode {
       bIsDynamic=false
       bIsEditable=true
       bIsStaticSwitch=true
       bIsVisible=true
       type {classStructOrEnum {refPath=&ENiagara_ScaleMode}}
      }
      Non Uniform Scale {bIsDynamic=false, bIsEditable=true, bIsStaticSwitch=false, bIsVisible=true, type {classStructOrEnum {refPath=&Vector3f}}}
      UseOwnerScale {bIsDynamic=false, bIsEditable=true, bIsStaticSwitch=false, bIsVisible=true, type {classStructOrEnum {refPath=&NiagaraBool}}}
      Apply Owner Scale {bIsDynamic=false, bIsEditable=true, bIsStaticSwitch=false, bIsVisible=true, type {classStructOrEnum {refPath=&Vector3f}}}
      Rotation Mode {
       bIsDynamic=false
       bIsEditable=true
       bIsStaticSwitch=true
       bIsVisible=true
       type {classStructOrEnum {refPath=&ENiagara_RotationMode}}
      }
      Rotation Axis {
       bIsDynamic=false
       bIsEditable=false
       bIsStaticSwitch=false
       bIsVisible=false
       type {classStructOrEnum {refPath=&Vector3f}}
      }
      Rotation Matrix {
       bIsDynamic=false
       bIsEditable=false
       bIsStaticSwitch=false
       bIsVisible=false
       type {classStructOrEnum {refPath=&NiagaraMatrix}}
      }
      Rotation Quaternion {bIsDynamic=false, bIsEditable=false, bIsStaticSwitch=false, bIsVisible=false, type {classStructOrEnum {refPath=&Quat4f}}}
      Yaw / Pitch / Roll {
       bIsDynamic=false
       bIsEditable=false
       bIsStaticSwitch=false
       bIsVisible=false
       type {classStructOrEnum {refPath=&Vector3f}}
      }
      Rotation Angle {
       bIsDynamic=false
       bIsEditable=false
       bIsStaticSwitch=false
       bIsVisible=false
       type {classStructOrEnum {refPath=&NiagaraFloat}}
      }
      Invert Rotation Quaternion {bIsDynamic=false, bIsEditable=false, bIsStaticSwitch=false, bIsVisible=false, type {classStructOrEnum {refPath=&NiagaraBool}}}
      Additional Quaternion Rotation {bIsDynamic=false, bIsEditable=false, bIsStaticSwitch=false, bIsVisible=false, type {classStructOrEnum {refPath=&NiagaraBool}}}
      Additional Yaw / Pitch / Roll {
       bIsDynamic=false
       bIsEditable=false
       bIsStaticSwitch=false
       bIsVisible=false
       type {classStructOrEnum {refPath=&Vector3f}}
      }
      Rotation Angle Type {
       bIsDynamic=false
       bIsEditable=false
       bIsStaticSwitch=true
       bIsVisible=false
       type {classStructOrEnum {refPath=&ENiagara_AngleInput}}
      }
      Rotation Coordinate Space {
       bIsDynamic=false
       bIsEditable=true
       bIsStaticSwitch=false
       bIsVisible=true
       type {classStructOrEnum {refPath=&ENiagaraCoordinateSpace}}
      }
      Offset Mode {
       bIsDynamic=false
       bIsEditable=true
       bIsStaticSwitch=true
       bIsVisible=true
       type {classStructOrEnum {refPath=&ENiagara_OffsetMode}}
      }
      Shape Origin {
       bIsDynamic=false
       bIsEditable=true
       bIsStaticSwitch=false
       bIsVisible=true
       type {classStructOrEnum {refPath=&NiagaraPosition}}
      }
      Offset {bIsDynamic=true, bIsEditable=true, bIsStaticSwitch=false, bIsVisible=true, type {classStructOrEnum {refPath=&Vector3f}}}
      Offset Coordinate Space {
       bIsDynamic=false
       bIsEditable=true
       bIsStaticSwitch=false
       bIsVisible=true
       type {classStructOrEnum {refPath=&ENiagaraCoordinateSpace}}
      }
      Custom Transform Matrix {
       bIsDynamic=false
       bIsEditable=false
       bIsStaticSwitch=false
       bIsVisible=false
       type {classStructOrEnum {refPath=&NiagaraMatrix}}
      }
      Transform Method {
       bIsDynamic=false
       bIsEditable=true
       bIsStaticSwitch=true
       bIsVisible=true
       type {classStructOrEnum {refPath=&ENiagara_TransformType}}
      }
     ]
     moduleScript {refPath=&ShapeLocation}
    }
    {
     bIsSetParametersModule=false
     enabled=true
     moduleName=AddVelocity
     inputs [
      Velocity Mode {
       bIsDynamic=false
       bIsEditable=true
       bIsStaticSwitch=true
       bIsVisible=true
       type {classStructOrEnum {refPath=&ENiagara_VelocityMode}}
      }
      Velocity {bIsDynamic=true, bIsEditable=true, bIsStaticSwitch=false, bIsVisible=true, type {classStructOrEnum {refPath=&Vector3f}}}
      Velocity Speed {
       bIsDynamic=false
       bIsEditable=false
       bIsStaticSwitch=false
       bIsVisible=false
       type {classStructOrEnum {refPath=&NiagaraFloat}}
      }
      Velocity Speed Scale {bIsDynamic=false, bIsEditable=true, bIsStaticSwitch=false, bIsVisible=true, type {classStructOrEnum {refPath=&NiagaraFloat}}}
      Constrain To Radius {
       bIsDynamic=false
       bIsEditable=false
       bIsStaticSwitch=false
       bIsVisible=false
       type {classStructOrEnum {refPath=&NiagaraFloat}}
      }
      Distribution Along Cone Axis {
       bIsDynamic=false
       bIsEditable=false
       bIsStaticSwitch=false
       bIsVisible=false
       type {classStructOrEnum {refPath=&NiagaraFloat}}
      }
      Radius Falloff Near / Far {
       bIsDynamic=false
       bIsEditable=false
       bIsStaticSwitch=false
       bIsVisible=false
       type {classStructOrEnum {refPath=&Vector2f}}
      }
      Radius Falloff Exponent {
       bIsDynamic=false
       bIsEditable=false
       bIsStaticSwitch=false
       bIsVisible=false
       type {classStructOrEnum {refPath=&NiagaraFloat}}
      }
      Speed Falloff From Cone Axis {
       bIsDynamic=false
       bIsEditable=false
       bIsStaticSwitch=false
       bIsVisible=false
       type {classStructOrEnum {refPath=&NiagaraFloat}}
      }
      Invert Velocity Falloff {bIsDynamic=false, bIsEditable=false, bIsStaticSwitch=false, bIsVisible=false, type {classStructOrEnum {refPath=&NiagaraBool}}}
      Override Seed {bIsDynamic=false, bIsEditable=false, bIsStaticSwitch=false, bIsVisible=false, type {classStructOrEnum {refPath=&NiagaraBool}}}
      Random Seed {
       bIsDynamic=false
       bIsEditable=false
       bIsStaticSwitch=false
       bIsVisible=false
       type {classStructOrEnum {refPath=&NiagaraInt32}}
      }
      Fixed Random Seed {bIsDynamic=false, bIsEditable=false, bIsStaticSwitch=true, bIsVisible=false, type {classStructOrEnum {refPath=&NiagaraBool}}}
      Randomness Mode {
       bIsDynamic=false
       bIsEditable=false
       bIsStaticSwitch=true
       bIsVisible=false
       type {classStructOrEnum {refPath=&ENiagaraRandomnessMode}}
      }
      Cone Axis {
       bIsDynamic=false
       bIsEditable=false
       bIsStaticSwitch=false
       bIsVisible=false
       type {classStructOrEnum {refPath=&Vector3f}}
      }
      Cone Angle {
       bIsDynamic=false
       bIsEditable=false
       bIsStaticSwitch=false
       bIsVisible=false
       type {classStructOrEnum {refPath=&NiagaraFloat}}
      }
      Inner Cone Angle {
       bIsDynamic=false
       bIsEditable=false
       bIsStaticSwitch=false
       bIsVisible=false
       type {classStructOrEnum {refPath=&NiagaraFloat}}
      }
      Cone Angle Mode {
       bIsDynamic=false
       bIsEditable=false
       bIsStaticSwitch=true
       bIsVisible=false
       type {classStructOrEnum {refPath=&ENiagara_AngleInput}}
      }
      Rotation Mode {
       bIsDynamic=false
       bIsEditable=true
       bIsStaticSwitch=true
       bIsVisible=true
       type {classStructOrEnum {refPath=&ENiagara_RotationMode}}
      }
      Rotation Axis {
       bIsDynamic=false
       bIsEditable=false
       bIsStaticSwitch=false
       bIsVisible=false
       type {classStructOrEnum {refPath=&Vector3f}}
      }
      Yaw / Pitch / Roll {
       bIsDynamic=false
       bIsEditable=false
       bIsStaticSwitch=false
       bIsVisible=false
       type {classStructOrEnum {refPath=&Vector3f}}
      }
      Rotation Quaternion {bIsDynamic=false, bIsEditable=false, bIsStaticSwitch=false, bIsVisible=false, type {classStructOrEnum {refPath=&Quat4f}}}
      Rotation Matrix {
       bIsDynamic=false
       bIsEditable=false
       bIsStaticSwitch=false
       bIsVisible=false
       type {classStructOrEnum {refPath=&NiagaraMatrix}}
      }
      Rotation Angle {
       bIsDynamic=false
       bIsEditable=false
       bIsStaticSwitch=false
       bIsVisible=false
       type {classStructOrEnum {refPath=&NiagaraFloat}}
      }
      Invert Rotation Quaternion {bIsDynamic=false, bIsEditable=false, bIsStaticSwitch=false, bIsVisible=false, type {classStructOrEnum {refPath=&NiagaraBool}}}
      Additional Quaternion Rotation {bIsDynamic=false, bIsEditable=false, bIsStaticSwitch=false, bIsVisible=false, type {classStructOrEnum {refPath=&NiagaraBool}}}
      Additional Yaw / Pitch / Roll {
       bIsDynamic=false
       bIsEditable=false
       bIsStaticSwitch=false
       bIsVisible=false
       type {classStructOrEnum {refPath=&Vector3f}}
      }
      Rotation Angle Type {
       bIsDynamic=false
       bIsEditable=false
       bIsStaticSwitch=true
       bIsVisible=false
       type {classStructOrEnum {refPath=&ENiagara_AngleInput}}
      }
      Rotation Coordinate Space {
       bIsDynamic=false
       bIsEditable=true
       bIsStaticSwitch=false
       bIsVisible=true
       type {classStructOrEnum {refPath=&ENiagaraCoordinateSpace}}
      }
      Velocity Origin {
       bIsDynamic=false
       bIsEditable=false
       bIsStaticSwitch=false
       bIsVisible=false
       type {classStructOrEnum {refPath=&NiagaraPosition}}
      }
      Origin Offset {
       bIsDynamic=false
       bIsEditable=false
       bIsStaticSwitch=false
       bIsVisible=false
       type {classStructOrEnum {refPath=&Vector3f}}
      }
      Origin Offset Coordinate Space {
       bIsDynamic=false
       bIsEditable=false
       bIsStaticSwitch=false
       bIsVisible=false
       type {classStructOrEnum {refPath=&ENiagaraCoordinateSpace}}
      }
      Default Position {
       bIsDynamic=false
       bIsEditable=false
       bIsStaticSwitch=false
       bIsVisible=false
       type {classStructOrEnum {refPath=&NiagaraPosition}}
      }
     ]
     moduleScript {refPath=&AddVelocity}
    }
   ]
  }
  particleUpdateScript {
   scriptName=ParticleUpdateScript
   modules [
    {
     bIsSetParametersModule=false
     enabled=true
     moduleName=ParticleState
     inputs [
      Kill Particles When Lifetime Has Elapsed {bIsDynamic=false, bIsEditable=true, bIsStaticSwitch=true, bIsVisible=true, type {classStructOrEnum {refPath=&NiagaraBool}}}
      Loop Particles Lifetime {bIsDynamic=false, bIsEditable=false, bIsStaticSwitch=false, bIsVisible=false, type {classStructOrEnum {refPath=&NiagaraBool}}}
      Let Infinitely Lived Particles Die When Emitter Deactivates {bIsDynamic=false, bIsEditable=false, bIsStaticSwitch=false, bIsVisible=false, type {classStructOrEnum {refPath=&NiagaraBool}}}
      Lifetime {bIsDynamic=false, bIsEditable=true, bIsStaticSwitch=false, bIsVisible=true, type {classStructOrEnum {refPath=&NiagaraFloat}}}
      DeltaTime {bIsDynamic=false, bIsEditable=true, bIsStaticSwitch=false, bIsVisible=true, type {classStructOrEnum {refPath=&NiagaraFloat}}}
     ]
     moduleScript {refPath=&ParticleState}
    }
    {
     bIsSetParametersModule=false
     enabled=true
     moduleName=ScaleSpriteSize
     inputs [
      Scale Sprite Size Mode {
       bIsDynamic=false
       bIsEditable=true
       bIsStaticSwitch=true
       bIsVisible=true
       type {classStructOrEnum {refPath=&ENiagara_ScaleSpriteSize}}
      }
      Uniform Curve Sprite Scale {
       bIsDynamic=false
       bIsEditable=false
       bIsStaticSwitch=false
       bIsVisible=false
       type {classStructOrEnum {refPath=&NiagaraDataInterfaceCurve}}
      }
      Non-Uniform Curve Sprite Scale {
       bIsDynamic=false
       bIsEditable=false
       bIsStaticSwitch=false
       bIsVisible=false
       type {classStructOrEnum {refPath=&NiagaraDataInterfaceVector2DCurve}}
      }
      Uniform Curve Index {
       bIsDynamic=false
       bIsEditable=false
       bIsStaticSwitch=false
       bIsVisible=false
       type {classStructOrEnum {refPath=&NiagaraFloat}}
      }
      Non-Uniform Curve Index {
       bIsDynamic=false
       bIsEditable=false
       bIsStaticSwitch=false
       bIsVisible=false
       type {classStructOrEnum {refPath=&NiagaraFloat}}
      }
      Non-Uniform Curve Scale {
       bIsDynamic=false
       bIsEditable=false
       bIsStaticSwitch=false
       bIsVisible=false
       type {classStructOrEnum {refPath=&Vector2f}}
      }
      Uniform Curve Scale {
       bIsDynamic=false
       bIsEditable=false
       bIsStaticSwitch=false
       bIsVisible=false
       type {classStructOrEnum {refPath=&NiagaraFloat}}
      }
      Scale Factor {bIsDynamic=true, bIsEditable=true, bIsStaticSwitch=false, bIsVisible=true, type {classStructOrEnum {refPath=&Vector2f}}}
      Uniform Scale Factor {
       bIsDynamic=false
       bIsEditable=false
       bIsStaticSwitch=false
       bIsVisible=false
       type {classStructOrEnum {refPath=&NiagaraFloat}}
      }
      Initial Sprite Size {bIsDynamic=false, bIsEditable=true, bIsStaticSwitch=false, bIsVisible=true, type {classStructOrEnum {refPath=&Vector2f}}}
     ]
     moduleScript {refPath=&ScaleSpriteSize}
    }
    {
     bIsSetParametersModule=true
     enabled=true
     moduleName=SetVariables_#7
     inputs [
      Particles.Color02 {
       bIsDynamic=true
       bIsEditable=true
       bIsStaticSwitch=false
       bIsVisible=true
       type {classStructOrEnum {refPath=&LinearColor}}
      }
      Particles.SpriteSize02 {bIsDynamic=true, bIsEditable=true, bIsStaticSwitch=false, bIsVisible=true, type {classStructOrEnum {refPath=&Vector2f}}}
     ]
     moduleScript {refPath=&self:BigSparkles002_0.NiagaraScriptSource_0.NiagaraGraph_0.NiagaraNodeAssignment_2.SetVariables_#7}
    }
    {
     bIsSetParametersModule=false
     enabled=true
     moduleName=ScaleColor
     inputs [
      Scale Mode {bIsDynamic=false, bIsEditable=true, bIsStaticSwitch=true, bIsVisible=true, type {classStructOrEnum {refPath=&NiagaraInt32}}}
      Linear Color Curve {
       bIsDynamic=false
       bIsEditable=false
       bIsStaticSwitch=false
       bIsVisible=false
       type {classStructOrEnum {refPath=&NiagaraDataInterfaceColorCurve}}
      }
      ScaleRGB {bIsDynamic=false, bIsEditable=true, bIsStaticSwitch=false, bIsVisible=true, type {classStructOrEnum {refPath=&NiagaraBool}}}
      Scale RGB {bIsDynamic=false, bIsEditable=true, bIsStaticSwitch=false, bIsVisible=true, type {classStructOrEnum {refPath=&Vector3f}}}
      ScaleRGBA {bIsDynamic=false, bIsEditable=false, bIsStaticSwitch=false, bIsVisible=false, type {classStructOrEnum {refPath=&NiagaraBool}}}
      Scale RGBA {
       bIsDynamic=false
       bIsEditable=false
       bIsStaticSwitch=false
       bIsVisible=false
       type {classStructOrEnum {refPath=&Vector4f}}
      }
      Curve Index {
       bIsDynamic=false
       bIsEditable=false
       bIsStaticSwitch=false
       bIsVisible=false
       type {classStructOrEnum {refPath=&NiagaraFloat}}
      }
      ScaleA {bIsDynamic=false, bIsEditable=true, bIsStaticSwitch=false, bIsVisible=true, type {classStructOrEnum {refPath=&NiagaraBool}}}
      Scale Alpha {bIsDynamic=true, bIsEditable=true, bIsStaticSwitch=false, bIsVisible=true, type {classStructOrEnum {refPath=&NiagaraFloat}}}
      Color Value To Scale {
       bIsDynamic=false
       bIsEditable=true
       bIsStaticSwitch=false
       bIsVisible=true
       type {classStructOrEnum {refPath=&LinearColor}}
      }
      Color Scale {
       bIsDynamic=false
       bIsEditable=false
       bIsStaticSwitch=false
       bIsVisible=false
       type {classStructOrEnum {refPath=&LinearColor}}
      }
     ]
     moduleScript {refPath=&ScaleColor}
    }
    {
     bIsSetParametersModule=false
     enabled=true
     moduleName=CurlNoiseForce
     inputs [
      Noise Strength {bIsDynamic=false, bIsEditable=true, bIsStaticSwitch=false, bIsVisible=true, type {classStructOrEnum {refPath=&NiagaraFloat}}}
      Noise Frequency {bIsDynamic=false, bIsEditable=true, bIsStaticSwitch=false, bIsVisible=true, type {classStructOrEnum {refPath=&NiagaraFloat}}}
      Noise Quality / Cost {
       bIsDynamic=false
       bIsEditable=true
       bIsStaticSwitch=true
       bIsVisible=true
       type {classStructOrEnum {refPath=&ENiagaraCurlNoiseQuality}}
      }
      Pan Noise {bIsDynamic=false, bIsEditable=true, bIsStaticSwitch=false, bIsVisible=true, type {classStructOrEnum {refPath=&NiagaraBool}}}
      Pan Noise Field {bIsDynamic=false, bIsEditable=false, bIsStaticSwitch=false, bIsVisible=true, type {classStructOrEnum {refPath=&Vector3f}}}
      Bias Noise {bIsDynamic=false, bIsEditable=true, bIsStaticSwitch=false, bIsVisible=true, type {classStructOrEnum {refPath=&NiagaraBool}}}
      Bias Noise Field {bIsDynamic=false, bIsEditable=false, bIsStaticSwitch=false, bIsVisible=true, type {classStructOrEnum {refPath=&Vector3f}}}
      Write to Intrinsic Parameters {bIsDynamic=false, bIsEditable=true, bIsStaticSwitch=true, bIsVisible=true, type {classStructOrEnum {refPath=&NiagaraBool}}}
      Sample Position {
       bIsDynamic=false
       bIsEditable=true
       bIsStaticSwitch=false
       bIsVisible=true
       type {classStructOrEnum {refPath=&NiagaraPosition}}
      }
      Randomize Noise Sample {bIsDynamic=false, bIsEditable=true, bIsStaticSwitch=true, bIsVisible=true, type {classStructOrEnum {refPath=&NiagaraBool}}}
      Random Seed {bIsDynamic=false, bIsEditable=true, bIsStaticSwitch=false, bIsVisible=true, type {classStructOrEnum {refPath=&NiagaraInt32}}}
      Randomization Vector {bIsDynamic=false, bIsEditable=true, bIsStaticSwitch=false, bIsVisible=true, type {classStructOrEnum {refPath=&Vector3f}}}
      Scale Curl Lookup By Engine Owner Scale {bIsDynamic=false, bIsEditable=true, bIsStaticSwitch=true, bIsVisible=true, type {classStructOrEnum {refPath=&NiagaraBool}}}
      Coordinate Space {
       bIsDynamic=false
       bIsEditable=true
       bIsStaticSwitch=true
       bIsVisible=true
       type {classStructOrEnum {refPath=&ENiagaraCoordinateSpace}}
      }
      CurlNoise {
       bIsDynamic=false
       bIsEditable=false
       bIsStaticSwitch=false
       bIsVisible=false
       type {classStructOrEnum {refPath=&NiagaraDataInterfaceCurlNoise}}
      }
      Debug Every N Particles {
       bIsDynamic=false
       bIsEditable=false
       bIsStaticSwitch=false
       bIsVisible=false
       type {classStructOrEnum {refPath=&NiagaraInt32}}
      }
      Debug Force Line Length {
       bIsDynamic=false
       bIsEditable=false
       bIsStaticSwitch=false
       bIsVisible=false
       type {classStructOrEnum {refPath=&NiagaraFloat}}
      }
      Mask Curl Noise {bIsDynamic=false, bIsEditable=true, bIsStaticSwitch=true, bIsVisible=true, type {classStructOrEnum {refPath=&NiagaraBool}}}
      Curl Noise Cone Mask Angle {
       bIsDynamic=false
       bIsEditable=false
       bIsStaticSwitch=false
       bIsVisible=false
       type {classStructOrEnum {refPath=&NiagaraFloat}}
      }
      Curl Noise Cone Mask Falloff Angle {
       bIsDynamic=false
       bIsEditable=false
       bIsStaticSwitch=false
       bIsVisible=false
       type {classStructOrEnum {refPath=&NiagaraFloat}}
      }
      Curl Noise Cone Mask Axis {
       bIsDynamic=false
       bIsEditable=false
       bIsStaticSwitch=false
       bIsVisible=false
       type {classStructOrEnum {refPath=&Vector3f}}
      }
     ]
     moduleScript {refPath=&CurlNoiseForce}
    }
    {
     bIsSetParametersModule=false
     enabled=true
     moduleName=Drag
     inputs [
      Use Linear Drag {bIsDynamic=false, bIsEditable=true, bIsStaticSwitch=false, bIsVisible=true, type {classStructOrEnum {refPath=&NiagaraBool}}}
      Drag {bIsDynamic=true, bIsEditable=true, bIsStaticSwitch=false, bIsVisible=true, type {classStructOrEnum {refPath=&NiagaraFloat}}}
      Use Rotational Drag {bIsDynamic=false, bIsEditable=true, bIsStaticSwitch=false, bIsVisible=true, type {classStructOrEnum {refPath=&NiagaraBool}}}
      Rotational Drag {bIsDynamic=false, bIsEditable=true, bIsStaticSwitch=false, bIsVisible=true, type {classStructOrEnum {refPath=&NiagaraFloat}}}
      Ignore Mass {bIsDynamic=false, bIsEditable=true, bIsStaticSwitch=false, bIsVisible=true, type {classStructOrEnum {refPath=&NiagaraBool}}}
     ]
     moduleScript {refPath=&Drag}
    }
    {
     bIsSetParametersModule=false
     enabled=true
     moduleName=SolveForcesAndVelocity
     inputs [
      Clamp Velocity {bIsDynamic=false, bIsEditable=true, bIsStaticSwitch=false, bIsVisible=true, type {classStructOrEnum {refPath=&NiagaraBool}}}
      Speed Limit {bIsDynamic=false, bIsEditable=false, bIsStaticSwitch=false, bIsVisible=true, type {classStructOrEnum {refPath=&NiagaraFloat}}}
      Limit Acceleration {bIsDynamic=false, bIsEditable=true, bIsStaticSwitch=false, bIsVisible=true, type {classStructOrEnum {refPath=&NiagaraBool}}}
      Acceleration Limit {bIsDynamic=false, bIsEditable=false, bIsStaticSwitch=false, bIsVisible=true, type {classStructOrEnum {refPath=&NiagaraFloat}}}
      Manually Enable Rotational Solver {bIsDynamic=false, bIsEditable=true, bIsStaticSwitch=false, bIsVisible=true, type {classStructOrEnum {refPath=&NiagaraBool}}}
      Write to Intrinsic Properties {bIsDynamic=false, bIsEditable=true, bIsStaticSwitch=false, bIsVisible=true, type {classStructOrEnum {refPath=&NiagaraBool}}}
      Write to Presolve Properties {bIsDynamic=false, bIsEditable=true, bIsStaticSwitch=true, bIsVisible=true, type {classStructOrEnum {refPath=&NiagaraBool}}}
      DeltaTime {bIsDynamic=false, bIsEditable=true, bIsStaticSwitch=false, bIsVisible=true, type {classStructOrEnum {refPath=&NiagaraFloat}}}
      Mass {bIsDynamic=false, bIsEditable=true, bIsStaticSwitch=false, bIsVisible=true, type {classStructOrEnum {refPath=&NiagaraFloat}}}
      Position {
       bIsDynamic=false
       bIsEditable=true
       bIsStaticSwitch=false
       bIsVisible=true
       type {classStructOrEnum {refPath=&NiagaraPosition}}
      }
      Velocity {bIsDynamic=false, bIsEditable=true, bIsStaticSwitch=false, bIsVisible=true, type {classStructOrEnum {refPath=&Vector3f}}}
      PreviousVelocity {bIsDynamic=false, bIsEditable=true, bIsStaticSwitch=false, bIsVisible=true, type {classStructOrEnum {refPath=&Vector3f}}}
      Force {bIsDynamic=false, bIsEditable=true, bIsStaticSwitch=false, bIsVisible=true, type {classStructOrEnum {refPath=&Vector3f}}}
      Mesh Orientation {bIsDynamic=false, bIsEditable=false, bIsStaticSwitch=false, bIsVisible=false, type {classStructOrEnum {refPath=&Quat4f}}}
      Rotational Velocity {
       bIsDynamic=false
       bIsEditable=false
       bIsStaticSwitch=false
       bIsVisible=false
       type {classStructOrEnum {refPath=&Vector3f}}
      }
      Rotational Force {
       bIsDynamic=false
       bIsEditable=false
       bIsStaticSwitch=false
       bIsVisible=false
       type {classStructOrEnum {refPath=&Vector3f}}
      }
      Rotational Inertia {
       bIsDynamic=false
       bIsEditable=false
       bIsStaticSwitch=false
       bIsVisible=false
       type {classStructOrEnum {refPath=&NiagaraFloat}}
      }
     ]
     moduleScript {refPath=&SolveForcesAndVelocity}
    }
   ]
  }
  rendererClasses [{refPath=&NiagaraSpriteRendererProperties}]
  renderers [
   {rendererIndex=0, rendererClass {refPath=&NiagaraSpriteRendererProperties}}
   {rendererIndex=1, rendererClass {refPath=&NiagaraSpriteRendererProperties}}
  ]
 }
 DefaultPropSpawn_Merged {
  bEnabled=false
  emitterSpawnScript {scriptName=EmitterSpawnScript, modules []}
  emitterUpdateScript {
   scriptName=EmitterUpdateScript
   modules [
    {
     bIsSetParametersModule=false
     enabled=true
     moduleName=EmitterState
     inputs [
      Life Cycle Mode {
       bIsDynamic=false
       bIsEditable=true
       bIsStaticSwitch=true
       bIsVisible=true
       type {classStructOrEnum {refPath=&ENiagaraEmitterLifeCycleMode}}
      }
      Inactive Response {
       bIsDynamic=false
       bIsEditable=false
       bIsStaticSwitch=true
       bIsVisible=false
       type {classStructOrEnum {refPath=&ENiagaraInactiveMode}}
      }
      Loop Behavior {
       bIsDynamic=false
       bIsEditable=false
       bIsStaticSwitch=true
       bIsVisible=false
       type {classStructOrEnum {refPath=&ENiagara_EmitterStateOptions}}
      }
      Loop Duration Mode {
       bIsDynamic=false
       bIsEditable=false
       bIsStaticSwitch=true
       bIsVisible=false
       type {classStructOrEnum {refPath=&ENiagara_InfiniteLoopDuration}}
      }
      Loop Count {
       bIsDynamic=false
       bIsEditable=false
       bIsStaticSwitch=false
       bIsVisible=false
       type {classStructOrEnum {refPath=&NiagaraInt32}}
      }
      Loop Duration {
       bIsDynamic=false
       bIsEditable=false
       bIsStaticSwitch=false
       bIsVisible=false
       type {classStructOrEnum {refPath=&NiagaraFloat}}
      }
      Recalculate Duration Each Loop {bIsDynamic=false, bIsEditable=false, bIsStaticSwitch=false, bIsVisible=false, type {classStructOrEnum {refPath=&NiagaraBool}}}
      UseLoopDelay {bIsDynamic=false, bIsEditable=false, bIsStaticSwitch=false, bIsVisible=false, type {classStructOrEnum {refPath=&NiagaraBool}}}
      Loop Delay {
       bIsDynamic=false
       bIsEditable=false
       bIsStaticSwitch=false
       bIsVisible=false
       type {classStructOrEnum {refPath=&NiagaraFloat}}
      }
      Delay First Loop Only {bIsDynamic=false, bIsEditable=false, bIsStaticSwitch=false, bIsVisible=false, type {classStructOrEnum {refPath=&NiagaraBool}}}
      Scalability Mode {
       bIsDynamic=false
       bIsEditable=true
       bIsStaticSwitch=true
       bIsVisible=true
       type {classStructOrEnum {refPath=&ENiagaraEmitterScalabilityMode_Limited}}
      }
      Enable Distance Culling {bIsDynamic=false, bIsEditable=false, bIsStaticSwitch=true, bIsVisible=false, type {classStructOrEnum {refPath=&NiagaraBool}}}
      MinDistance {
       bIsDynamic=false
       bIsEditable=false
       bIsStaticSwitch=false
       bIsVisible=false
       type {classStructOrEnum {refPath=&NiagaraFloat}}
      }
      Min Distance Response {
       bIsDynamic=false
       bIsEditable=false
       bIsStaticSwitch=false
       bIsVisible=false
       type {classStructOrEnum {refPath=&ENiagaraExecutionStateManagement}}
      }
      MaxDistance {
       bIsDynamic=false
       bIsEditable=false
       bIsStaticSwitch=false
       bIsVisible=false
       type {classStructOrEnum {refPath=&NiagaraFloat}}
      }
      Max Distance Response {
       bIsDynamic=false
       bIsEditable=false
       bIsStaticSwitch=false
       bIsVisible=false
       type {classStructOrEnum {refPath=&ENiagaraExecutionStateManagement}}
      }
      Scale Spawn Count {bIsDynamic=false, bIsEditable=false, bIsStaticSwitch=true, bIsVisible=false, type {classStructOrEnum {refPath=&NiagaraBool}}}
      Spawn Count Scale {
       bIsDynamic=false
       bIsEditable=false
       bIsStaticSwitch=false
       bIsVisible=false
       type {classStructOrEnum {refPath=&NiagaraFloat}}
      }
      Spawn Count Scale By Distance {
       bIsDynamic=false
       bIsEditable=false
       bIsStaticSwitch=false
       bIsVisible=false
       type {classStructOrEnum {refPath=&NiagaraDataInterfaceCurve}}
      }
      Enable Visibility Culling {bIsDynamic=false, bIsEditable=false, bIsStaticSwitch=true, bIsVisible=false, type {classStructOrEnum {refPath=&NiagaraBool}}}
      Visibility Cull Response {
       bIsDynamic=false
       bIsEditable=false
       bIsStaticSwitch=false
       bIsVisible=false
       type {classStructOrEnum {refPath=&ENiagaraExecutionStateManagement}}
      }
      Visibility Cull Delay {
       bIsDynamic=false
       bIsEditable=false
       bIsStaticSwitch=false
       bIsVisible=false
       type {classStructOrEnum {refPath=&NiagaraFloat}}
      }
      Reset Age On Awaken {bIsDynamic=false, bIsEditable=false, bIsStaticSwitch=false, bIsVisible=false, type {classStructOrEnum {refPath=&NiagaraBool}}}
     ]
     moduleScript {refPath=&EmitterState}
    }
    {
     bIsSetParametersModule=false
     enabled=true
     moduleName=SpawnBurst_Instantaneous001
     inputs [
      Spawn Count {bIsDynamic=true, bIsEditable=true, bIsStaticSwitch=false, bIsVisible=true, type {classStructOrEnum {refPath=&NiagaraInt32}}}
      Spawn Time {bIsDynamic=false, bIsEditable=true, bIsStaticSwitch=false, bIsVisible=true, type {classStructOrEnum {refPath=&NiagaraFloat}}}
      Use Spawn Probability {bIsDynamic=false, bIsEditable=true, bIsStaticSwitch=false, bIsVisible=true, type {classStructOrEnum {refPath=&NiagaraBool}}}
      Spawn Probability {bIsDynamic=false, bIsEditable=false, bIsStaticSwitch=false, bIsVisible=true, type {classStructOrEnum {refPath=&NiagaraFloat}}}
      UseLoopCountLimit {bIsDynamic=false, bIsEditable=true, bIsStaticSwitch=false, bIsVisible=true, type {classStructOrEnum {refPath=&NiagaraBool}}}
      Loop Count Limit {bIsDynamic=false, bIsEditable=false, bIsStaticSwitch=false, bIsVisible=true, type {classStructOrEnum {refPath=&NiagaraInt32}}}
      Spawn Group {bIsDynamic=false, bIsEditable=true, bIsStaticSwitch=false, bIsVisible=true, type {classStructOrEnum {refPath=&NiagaraInt32}}}
      Age {bIsDynamic=false, bIsEditable=true, bIsStaticSwitch=false, bIsVisible=true, type {classStructOrEnum {refPath=&NiagaraFloat}}}
     ]
     moduleScript {refPath=&SpawnBurst_Instantaneous}
    }
    {
     bIsSetParametersModule=false
     enabled=true
     moduleName=SpawnBurst_Instantaneous
     inputs [
      Spawn Count {bIsDynamic=false, bIsEditable=true, bIsStaticSwitch=false, bIsVisible=true, type {classStructOrEnum {refPath=&NiagaraInt32}}}
      Spawn Time {bIsDynamic=false, bIsEditable=true, bIsStaticSwitch=false, bIsVisible=true, type {classStructOrEnum {refPath=&NiagaraFloat}}}
      Use Spawn Probability {bIsDynamic=false, bIsEditable=true, bIsStaticSwitch=false, bIsVisible=true, type {classStructOrEnum {refPath=&NiagaraBool}}}
      Spawn Probability {bIsDynamic=false, bIsEditable=false, bIsStaticSwitch=false, bIsVisible=true, type {classStructOrEnum {refPath=&NiagaraFloat}}}
      UseLoopCountLimit {bIsDynamic=false, bIsEditable=true, bIsStaticSwitch=false, bIsVisible=true, type {classStructOrEnum {refPath=&NiagaraBool}}}
      Loop Count Limit {bIsDynamic=false, bIsEditable=false, bIsStaticSwitch=false, bIsVisible=true, type {classStructOrEnum {refPath=&NiagaraInt32}}}
      Spawn Group {bIsDynamic=false, bIsEditable=true, bIsStaticSwitch=false, bIsVisible=true, type {classStructOrEnum {refPath=&NiagaraInt32}}}
      Age {bIsDynamic=false, bIsEditable=true, bIsStaticSwitch=false, bIsVisible=true, type {classStructOrEnum {refPath=&NiagaraFloat}}}
     ]
     moduleScript {refPath=&SpawnBurst_Instantaneous}
    }
    {
     bIsSetParametersModule=false
     enabled=true
     moduleName=SpawnBurst_Instantaneous002
     inputs [
      Spawn Count {bIsDynamic=true, bIsEditable=true, bIsStaticSwitch=false, bIsVisible=true, type {classStructOrEnum {refPath=&NiagaraInt32}}}
      Spawn Time {bIsDynamic=false, bIsEditable=true, bIsStaticSwitch=false, bIsVisible=true, type {classStructOrEnum {refPath=&NiagaraFloat}}}
      Use Spawn Probability {bIsDynamic=false, bIsEditable=true, bIsStaticSwitch=false, bIsVisible=true, type {classStructOrEnum {refPath=&NiagaraBool}}}
      Spawn Probability {bIsDynamic=false, bIsEditable=false, bIsStaticSwitch=false, bIsVisible=true, type {classStructOrEnum {refPath=&NiagaraFloat}}}
      UseLoopCountLimit {bIsDynamic=false, bIsEditable=true, bIsStaticSwitch=false, bIsVisible=true, type {classStructOrEnum {refPath=&NiagaraBool}}}
      Loop Count Limit {bIsDynamic=false, bIsEditable=false, bIsStaticSwitch=false, bIsVisible=true, type {classStructOrEnum {refPath=&NiagaraInt32}}}
      Spawn Group {bIsDynamic=false, bIsEditable=true, bIsStaticSwitch=false, bIsVisible=true, type {classStructOrEnum {refPath=&NiagaraInt32}}}
      Age {bIsDynamic=false, bIsEditable=true, bIsStaticSwitch=false, bIsVisible=true, type {classStructOrEnum {refPath=&NiagaraFloat}}}
     ]
     moduleScript {refPath=&SpawnBurst_Instantaneous}
    }
   ]
  }
  particleSpawnScript {
   scriptName=ParticleSpawnScript
   modules [
    {
     bIsSetParametersModule=true
     enabled=true
     moduleName=SetVariables_#1
     inputs [Particles.VisibilityTag {bIsDynamic=true, bIsEditable=true, bIsStaticSwitch=false, bIsVisible=true, type {classStructOrEnum {refPath=&NiagaraInt32}}}]
     moduleScript {refPath=&self:DefaultPropSpawn_Merged_0.NiagaraScriptSource_1.NiagaraGraph_0.NiagaraNodeAssignment_1.SetVariables_#1}
    }
    {
     bIsSetParametersModule=false
     enabled=true
     moduleName=PartitionParticles
     inputs [
      Partition Mode {
       bIsDynamic=false
       bIsEditable=true
       bIsStaticSwitch=true
       bIsVisible=true
       type {classStructOrEnum {refPath=&ENiagara_PartitionMode_v2}}
      }
      Partition Count Mode {
       bIsDynamic=false
       bIsEditable=false
       bIsStaticSwitch=true
       bIsVisible=false
       type {classStructOrEnum {refPath=&ENiagara_InfiniteLoopDuration}}
      }
      A {
       bIsDynamic=false
       bIsEditable=false
       bIsStaticSwitch=false
       bIsVisible=false
       type {classStructOrEnum {refPath=&NiagaraFloat}}
      }
      Number Of Partitions {bIsDynamic=false, bIsEditable=true, bIsStaticSwitch=false, bIsVisible=true, type {classStructOrEnum {refPath=&NiagaraInt32}}}
      Particles Per Partition {
       bIsDynamic=false
       bIsEditable=false
       bIsStaticSwitch=false
       bIsVisible=false
       type {classStructOrEnum {refPath=&NiagaraInt32}}
      }
      Partition Duration {
       bIsDynamic=false
       bIsEditable=false
       bIsStaticSwitch=false
       bIsVisible=false
       type {classStructOrEnum {refPath=&NiagaraFloat}}
      }
      Time {
       bIsDynamic=false
       bIsEditable=false
       bIsStaticSwitch=false
       bIsVisible=false
       type {classStructOrEnum {refPath=&NiagaraFloat}}
      }
      Partitioning Frequency {
       bIsDynamic=false
       bIsEditable=false
       bIsStaticSwitch=true
       bIsVisible=false
       type {classStructOrEnum {refPath=&ENiagaraRandomnessEvaluation}}
      }
      B {
       bIsDynamic=false
       bIsEditable=false
       bIsStaticSwitch=false
       bIsVisible=false
       type {classStructOrEnum {refPath=&NiagaraFloat}}
      }
      Distance {
       bIsDynamic=false
       bIsEditable=false
       bIsStaticSwitch=false
       bIsVisible=false
       type {classStructOrEnum {refPath=&NiagaraFloat}}
      }
      Comparison Type {
       bIsDynamic=false
       bIsEditable=false
       bIsStaticSwitch=true
       bIsVisible=false
       type {classStructOrEnum {refPath=&ENiagaraBooleanLogicOps_v2}}
      }
      Evaluation Type {
       bIsDynamic=false
       bIsEditable=false
       bIsStaticSwitch=true
       bIsVisible=false
       type {classStructOrEnum {refPath=&ENiagaraRandomnessEvaluation}}
      }
      UseDistanceFalloff {bIsDynamic=false, bIsEditable=false, bIsStaticSwitch=false, bIsVisible=false, type {classStructOrEnum {refPath=&NiagaraBool}}}
      Distance Falloff Exponent {
       bIsDynamic=false
       bIsEditable=false
       bIsStaticSwitch=false
       bIsVisible=false
       type {classStructOrEnum {refPath=&NiagaraFloat}}
      }
      Start Position {
       bIsDynamic=false
       bIsEditable=false
       bIsStaticSwitch=false
       bIsVisible=false
       type {classStructOrEnum {refPath=&NiagaraPosition}}
      }
      End Position {
       bIsDynamic=false
       bIsEditable=false
       bIsStaticSwitch=false
       bIsVisible=false
       type {classStructOrEnum {refPath=&NiagaraPosition}}
      }
      ID to Partition {bIsDynamic=false, bIsEditable=true, bIsStaticSwitch=false, bIsVisible=true, type {classStructOrEnum {refPath=&NiagaraInt32}}}
      Randomness Mode {
       bIsDynamic=false
       bIsEditable=false
       bIsStaticSwitch=true
       bIsVisible=false
       type {classStructOrEnum {refPath=&ENiagaraRandomnessMode}}
      }
      Override Seed {bIsDynamic=false, bIsEditable=false, bIsStaticSwitch=true, bIsVisible=false, type {classStructOrEnum {refPath=&NiagaraBool}}}
      Fixed Random Seed {bIsDynamic=false, bIsEditable=false, bIsStaticSwitch=true, bIsVisible=false, type {classStructOrEnum {refPath=&NiagaraBool}}}
      Color A {
       bIsDynamic=false
       bIsEditable=false
       bIsStaticSwitch=false
       bIsVisible=false
       type {classStructOrEnum {refPath=&LinearColor}}
      }
      Color B {
       bIsDynamic=false
       bIsEditable=false
       bIsStaticSwitch=false
       bIsVisible=false
       type {classStructOrEnum {refPath=&LinearColor}}
      }
      Color Comparison Type {bIsDynamic=false, bIsEditable=false, bIsStaticSwitch=true, bIsVisible=false, type {classStructOrEnum {refPath=&NiagaraInt32}}}
      Color Tolerance {
       bIsDynamic=false
       bIsEditable=false
       bIsStaticSwitch=false
       bIsVisible=false
       type {classStructOrEnum {refPath=&NiagaraFloat}}
      }
      Minimum Saturation {
       bIsDynamic=false
       bIsEditable=false
       bIsStaticSwitch=false
       bIsVisible=false
       type {classStructOrEnum {refPath=&NiagaraFloat}}
      }
      Direction A {
       bIsDynamic=false
       bIsEditable=false
       bIsStaticSwitch=false
       bIsVisible=false
       type {classStructOrEnum {refPath=&Vector3f}}
      }
      Direction B {
       bIsDynamic=false
       bIsEditable=false
       bIsStaticSwitch=false
       bIsVisible=false
       type {classStructOrEnum {refPath=&Vector3f}}
      }
      Direction Tolerance {
       bIsDynamic=false
       bIsEditable=false
       bIsStaticSwitch=false
       bIsVisible=false
       type {classStructOrEnum {refPath=&NiagaraFloat}}
      }
      Write Partition to Ribbon ID {
       bIsDynamic=false
       bIsEditable=true
       bIsStaticSwitch=true
       bIsVisible=true
       type {classStructOrEnum {refPath=&ENiagara_AttributeSamplingApplyOutput}}
      }
      Write Partition Index to Ribbon Link Order {
       bIsDynamic=false
       bIsEditable=false
       bIsStaticSwitch=true
       bIsVisible=false
       type {classStructOrEnum {refPath=&ENiagara_AttributeSamplingApplyOutput}}
      }
      Normalized Index In Partition Includes One {bIsDynamic=false, bIsEditable=false, bIsStaticSwitch=true, bIsVisible=false, type {classStructOrEnum {refPath=&NiagaraBool}}}
      Normalized Index Includes One {bIsDynamic=false, bIsEditable=true, bIsStaticSwitch=true, bIsVisible=true, type {classStructOrEnum {refPath=&NiagaraBool}}}
     ]
     moduleScript {refPath=&PartitionParticles}
    }
    {
     bIsSetParametersModule=false
     enabled=true
     moduleName=InitializeParticle001
     inputs [
      Lifetime Mode {
       bIsDynamic=false
       bIsEditable=true
       bIsStaticSwitch=true
       bIsVisible=true
       type {classStructOrEnum {refPath=&ENiagara_LifetimeMode}}
      }
      Lifetime Min {
       bIsDynamic=false
       bIsEditable=false
       bIsStaticSwitch=false
       bIsVisible=false
       type {classStructOrEnum {refPath=&NiagaraFloat}}
      }
      Lifetime {bIsDynamic=true, bIsEditable=true, bIsStaticSwitch=false, bIsVisible=true, type {classStructOrEnum {refPath=&NiagaraFloat}}}
      Lifetime Max {
       bIsDynamic=false
       bIsEditable=false
       bIsStaticSwitch=false
       bIsVisible=false
       type {classStructOrEnum {refPath=&NiagaraFloat}}
      }
      Lifetime Randomness Mode {
       bIsDynamic=false
       bIsEditable=false
       bIsStaticSwitch=true
       bIsVisible=false
       type {classStructOrEnum {refPath=&ENiagaraRandomnessMode}}
      }
      Color Mode {
       bIsDynamic=false
       bIsEditable=true
       bIsStaticSwitch=true
       bIsVisible=true
       type {classStructOrEnum {refPath=&ENiagara_ColorInitializationMode}}
      }
      Color Minimum {
       bIsDynamic=false
       bIsEditable=false
       bIsStaticSwitch=false
       bIsVisible=false
       type {classStructOrEnum {refPath=&LinearColor}}
      }
      Color {
       bIsDynamic=true
       bIsEditable=true
       bIsStaticSwitch=false
       bIsVisible=true
       type {classStructOrEnum {refPath=&LinearColor}}
      }
      AdjustHue {bIsDynamic=false, bIsEditable=false, bIsStaticSwitch=false, bIsVisible=false, type {classStructOrEnum {refPath=&NiagaraBool}}}
      Hue Shift Range {
       bIsDynamic=false
       bIsEditable=false
       bIsStaticSwitch=false
       bIsVisible=false
       type {classStructOrEnum {refPath=&Vector2f}}
      }
      Color Maximum {
       bIsDynamic=false
       bIsEditable=false
       bIsStaticSwitch=false
       bIsVisible=false
       type {classStructOrEnum {refPath=&LinearColor}}
      }
      AdjustSaturation {bIsDynamic=false, bIsEditable=false, bIsStaticSwitch=false, bIsVisible=false, type {classStructOrEnum {refPath=&NiagaraBool}}}
      Saturation Range {
       bIsDynamic=false
       bIsEditable=false
       bIsStaticSwitch=false
       bIsVisible=false
       type {classStructOrEnum {refPath=&Vector2f}}
      }
      AdjustValue {bIsDynamic=false, bIsEditable=false, bIsStaticSwitch=false, bIsVisible=false, type {classStructOrEnum {refPath=&NiagaraBool}}}
      Value Range {
       bIsDynamic=false
       bIsEditable=false
       bIsStaticSwitch=false
       bIsVisible=false
       type {classStructOrEnum {refPath=&Vector2f}}
      }
      AdjustAlpha {bIsDynamic=false, bIsEditable=false, bIsStaticSwitch=false, bIsVisible=false, type {classStructOrEnum {refPath=&NiagaraBool}}}
      Alpha Scale Range {
       bIsDynamic=false
       bIsEditable=false
       bIsStaticSwitch=false
       bIsVisible=false
       type {classStructOrEnum {refPath=&Vector2f}}
      }
      Color Channel Mode {
       bIsDynamic=false
       bIsEditable=false
       bIsStaticSwitch=true
       bIsVisible=false
       type {classStructOrEnum {refPath=&ENiagaraChannelCorrelation}}
      }
      Color Randomness Mode {
       bIsDynamic=false
       bIsEditable=false
       bIsStaticSwitch=true
       bIsVisible=false
       type {classStructOrEnum {refPath=&ENiagaraRandomnessMode}}
      }
      Position Mode {
       bIsDynamic=false
       bIsEditable=true
       bIsStaticSwitch=true
       bIsVisible=true
       type {classStructOrEnum {refPath=&ENiagara_PositionInitializationMode}}
      }
      Position {
       bIsDynamic=false
       bIsEditable=false
       bIsStaticSwitch=false
       bIsVisible=false
       type {classStructOrEnum {refPath=&NiagaraPosition}}
      }
      UsePositionOffset {bIsDynamic=false, bIsEditable=true, bIsStaticSwitch=false, bIsVisible=true, type {classStructOrEnum {refPath=&NiagaraBool}}}
      Position Offset {bIsDynamic=false, bIsEditable=true, bIsStaticSwitch=false, bIsVisible=true, type {classStructOrEnum {refPath=&Vector3f}}}
      Position Offset Coordinate Space {
       bIsDynamic=false
       bIsEditable=true
       bIsStaticSwitch=false
       bIsVisible=true
       type {classStructOrEnum {refPath=&ENiagaraCoordinateSpace}}
      }
      Mass Mode {
       bIsDynamic=false
       bIsEditable=true
       bIsStaticSwitch=true
       bIsVisible=true
       type {classStructOrEnum {refPath=&ENiagara_MassInitializationMode}}
      }
      Mass Min {
       bIsDynamic=false
       bIsEditable=false
       bIsStaticSwitch=false
       bIsVisible=false
       type {classStructOrEnum {refPath=&NiagaraFloat}}
      }
      Mass {
       bIsDynamic=false
       bIsEditable=false
       bIsStaticSwitch=false
       bIsVisible=false
       type {classStructOrEnum {refPath=&NiagaraFloat}}
      }
      Mass Max {
       bIsDynamic=false
       bIsEditable=false
       bIsStaticSwitch=false
       bIsVisible=false
       type {classStructOrEnum {refPath=&NiagaraFloat}}
      }
      Mass Randomness Mode {
       bIsDynamic=false
       bIsEditable=false
       bIsStaticSwitch=true
       bIsVisible=false
       type {classStructOrEnum {refPath=&ENiagaraRandomnessMode}}
      }
      Write Material Random {bIsDynamic=false, bIsEditable=true, bIsStaticSwitch=false, bIsVisible=true, type {classStructOrEnum {refPath=&NiagaraBool}}}
      Material Random {bIsDynamic=false, bIsEditable=true, bIsStaticSwitch=false, bIsVisible=true, type {classStructOrEnum {refPath=&NiagaraFloat}}}
      Sprite Size Mode {
       bIsDynamic=false
       bIsEditable=true
       bIsStaticSwitch=true
       bIsVisible=true
       type {classStructOrEnum {refPath=&ENiagara_SizeScaleMode}}
      }
      Sprite Size {
       bIsDynamic=false
       bIsEditable=false
       bIsStaticSwitch=false
       bIsVisible=false
       type {classStructOrEnum {refPath=&Vector2f}}
      }
      Uniform Sprite Size {bIsDynamic=true, bIsEditable=true, bIsStaticSwitch=false, bIsVisible=true, type {classStructOrEnum {refPath=&NiagaraFloat}}}
      Uniform Sprite Size Min {
       bIsDynamic=false
       bIsEditable=false
       bIsStaticSwitch=false
       bIsVisible=false
       type {classStructOrEnum {refPath=&NiagaraFloat}}
      }
      Sprite Size Min {
       bIsDynamic=false
       bIsEditable=false
       bIsStaticSwitch=false
       bIsVisible=false
       type {classStructOrEnum {refPath=&Vector2f}}
      }
      Sprite Size Max {
       bIsDynamic=false
       bIsEditable=false
       bIsStaticSwitch=false
       bIsVisible=false
       type {classStructOrEnum {refPath=&Vector2f}}
      }
      Uniform Sprite Size Max {
       bIsDynamic=false
       bIsEditable=false
       bIsStaticSwitch=false
       bIsVisible=false
       type {classStructOrEnum {refPath=&NiagaraFloat}}
      }
      Sprite Size Randomness Mode {
       bIsDynamic=false
       bIsEditable=false
       bIsStaticSwitch=true
       bIsVisible=false
       type {classStructOrEnum {refPath=&ENiagaraRandomnessMode}}
      }
      Sprite Rotation Mode {
       bIsDynamic=false
       bIsEditable=true
       bIsStaticSwitch=true
       bIsVisible=true
       type {classStructOrEnum {refPath=&ENiagara_SpriteRotationMode}}
      }
      Sprite Rotation Angle Min {
       bIsDynamic=false
       bIsEditable=false
       bIsStaticSwitch=false
       bIsVisible=false
       type {classStructOrEnum {refPath=&NiagaraFloat}}
      }
      Sprite Rotation Angle {bIsDynamic=true, bIsEditable=true, bIsStaticSwitch=false, bIsVisible=true, type {classStructOrEnum {refPath=&NiagaraFloat}}}
      Sprite Rotation Angle Max {
       bIsDynamic=false
       bIsEditable=false
       bIsStaticSwitch=false
       bIsVisible=false
       type {classStructOrEnum {refPath=&NiagaraFloat}}
      }
      Sprite Rotation Randomness Mode {
       bIsDynamic=false
       bIsEditable=false
       bIsStaticSwitch=true
       bIsVisible=false
       type {classStructOrEnum {refPath=&ENiagaraRandomnessMode}}
      }
      Sprite UV Mode {
       bIsDynamic=false
       bIsEditable=true
       bIsStaticSwitch=true
       bIsVisible=true
       type {classStructOrEnum {refPath=&ENiagara_UVFlippingMode}}
      }
      Sprite UV Randomness Mode {
       bIsDynamic=false
       bIsEditable=false
       bIsStaticSwitch=true
       bIsVisible=false
       type {classStructOrEnum {refPath=&ENiagaraRandomnessMode}}
      }
      Sprite UV Scale {
       bIsDynamic=false
       bIsEditable=false
       bIsStaticSwitch=false
       bIsVisible=false
       type {classStructOrEnum {refPath=&Vector2f}}
      }
      Mesh Scale Mode {
       bIsDynamic=false
       bIsEditable=true
       bIsStaticSwitch=true
       bIsVisible=true
       type {classStructOrEnum {refPath=&ENiagara_SizeScaleMode}}
      }
      Mesh Uniform Scale Min {
       bIsDynamic=false
       bIsEditable=false
       bIsStaticSwitch=false
       bIsVisible=false
       type {classStructOrEnum {refPath=&NiagaraFloat}}
      }
      Mesh Uniform Scale {
       bIsDynamic=false
       bIsEditable=false
       bIsStaticSwitch=false
       bIsVisible=false
       type {classStructOrEnum {refPath=&NiagaraFloat}}
      }
      Mesh Scale Min {
       bIsDynamic=false
       bIsEditable=false
       bIsStaticSwitch=false
       bIsVisible=false
       type {classStructOrEnum {refPath=&Vector3f}}
      }
      Mesh Scale {
       bIsDynamic=false
       bIsEditable=false
       bIsStaticSwitch=false
       bIsVisible=false
       type {classStructOrEnum {refPath=&Vector3f}}
      }
      Mesh Scale Max {
       bIsDynamic=false
       bIsEditable=false
       bIsStaticSwitch=false
       bIsVisible=false
       type {classStructOrEnum {refPath=&Vector3f}}
      }
      Mesh Uniform Scale Max {
       bIsDynamic=false
       bIsEditable=false
       bIsStaticSwitch=false
       bIsVisible=false
       type {classStructOrEnum {refPath=&NiagaraFloat}}
      }
      Mesh Scale Randomness Mode {
       bIsDynamic=false
       bIsEditable=false
       bIsStaticSwitch=true
       bIsVisible=false
       type {classStructOrEnum {refPath=&ENiagaraRandomnessMode}}
      }
      Mesh Renderer Array Visibility Mode {
       bIsDynamic=false
       bIsEditable=true
       bIsStaticSwitch=true
       bIsVisible=true
       type {classStructOrEnum {refPath=&ENiagara_UnsetDirectSetRandom}}
      }
      Mesh Renderer Mesh Index {
       bIsDynamic=false
       bIsEditable=false
       bIsStaticSwitch=false
       bIsVisible=false
       type {classStructOrEnum {refPath=&NiagaraInt32}}
      }
      Wrap Index By Mesh Count {bIsDynamic=false, bIsEditable=false, bIsStaticSwitch=true, bIsVisible=false, type {classStructOrEnum {refPath=&NiagaraBool}}}
      Mesh Renderer Info {
       bIsDynamic=false
       bIsEditable=false
       bIsStaticSwitch=false
       bIsVisible=false
       type {classStructOrEnum {refPath=&NiagaraDataInterfaceMeshRendererInfo}}
      }
      Mesh Renderer Random Seed {
       bIsDynamic=false
       bIsEditable=false
       bIsStaticSwitch=false
       bIsVisible=false
       type {classStructOrEnum {refPath=&NiagaraInt32}}
      }
      Ribbon Width Mode {
       bIsDynamic=false
       bIsEditable=true
       bIsStaticSwitch=true
       bIsVisible=true
       type {classStructOrEnum {refPath=&ENiagara_UnsetDirectSet}}
      }
      Ribbon Width {
       bIsDynamic=false
       bIsEditable=false
       bIsStaticSwitch=false
       bIsVisible=false
       type {classStructOrEnum {refPath=&NiagaraFloat}}
      }
      Ribbon Facing Mode {
       bIsDynamic=false
       bIsEditable=true
       bIsStaticSwitch=true
       bIsVisible=true
       type {classStructOrEnum {refPath=&ENiagara_UnsetDirectSet}}
      }
      Ribbon Facing Vector {
       bIsDynamic=false
       bIsEditable=false
       bIsStaticSwitch=false
       bIsVisible=false
       type {classStructOrEnum {refPath=&Vector3f}}
      }
      Ribbon Twist Mode {
       bIsDynamic=false
       bIsEditable=true
       bIsStaticSwitch=true
       bIsVisible=true
       type {classStructOrEnum {refPath=&ENiagara_UnsetDirectSet}}
      }
      "Ribbon Twist " {
       bIsDynamic=false
       bIsEditable=false
       bIsStaticSwitch=false
       bIsVisible=false
       type {classStructOrEnum {refPath=&NiagaraFloat}}
      }
     ]
     moduleScript {refPath=&InitializeParticle}
    }
    {
     bIsSetParametersModule=true
     enabled=true
     moduleName=SetVariables_#2
     inputs [Particles.InitSpriteRotRate {bIsDynamic=true, bIsEditable=true, bIsStaticSwitch=false, bIsVisible=true, type {classStructOrEnum {refPath=&NiagaraFloat}}}]
     moduleScript {refPath=&self:DefaultPropSpawn_Merged_0.NiagaraScriptSource_1.NiagaraGraph_0.NiagaraNodeAssignment_0.SetVariables_#2}
    }
    {
     bIsSetParametersModule=false
     enabled=true
     moduleName=AccelerationForce
     inputs [
      Acceleration {bIsDynamic=true, bIsEditable=true, bIsStaticSwitch=false, bIsVisible=true, type {classStructOrEnum {refPath=&Vector3f}}}
      Coordinate Space {
       bIsDynamic=false
       bIsEditable=true
       bIsStaticSwitch=false
       bIsVisible=true
       type {classStructOrEnum {refPath=&ENiagaraCoordinateSpace}}
      }
     ]
     moduleScript {refPath=&AccelerationForce}
    }
    {
     bIsSetParametersModule=false
     enabled=true
     moduleName=ShapeLocation
     inputs [
      Shape Primitive {
       bIsDynamic=false
       bIsEditable=true
       bIsStaticSwitch=true
       bIsVisible=true
       type {classStructOrEnum {refPath=&ENiagara_LocationShapes}}
      }
      Box / Plane Mode {
       bIsDynamic=false
       bIsEditable=false
       bIsStaticSwitch=true
       bIsVisible=false
       type {classStructOrEnum {refPath=&ENiagara_BoxPlaneMode}}
      }
      Cone Mode {
       bIsDynamic=false
       bIsEditable=false
       bIsStaticSwitch=true
       bIsVisible=false
       type {classStructOrEnum {refPath=&ENiagara_ConeMode}}
      }
      Cylinder Height {
       bIsDynamic=false
       bIsEditable=false
       bIsStaticSwitch=false
       bIsVisible=false
       type {classStructOrEnum {refPath=&NiagaraFloat}}
      }
      Sphere Radius {bIsDynamic=true, bIsEditable=true, bIsStaticSwitch=false, bIsVisible=true, type {classStructOrEnum {refPath=&NiagaraFloat}}}
      Ring / Disc Mode {
       bIsDynamic=false
       bIsEditable=false
       bIsStaticSwitch=true
       bIsVisible=false
       type {classStructOrEnum {refPath=&ENiagaraRingDiscMode}}
      }
      Torus Mode {
       bIsDynamic=false
       bIsEditable=false
       bIsStaticSwitch=true
       bIsVisible=false
       type {classStructOrEnum {refPath=&ENiagaraShapeTorusMode}}
      }
      Knot Ring Loops {
       bIsDynamic=false
       bIsEditable=false
       bIsStaticSwitch=false
       bIsVisible=false
       type {classStructOrEnum {refPath=&NiagaraInt32}}
      }
      Knot Helix Loops {
       bIsDynamic=false
       bIsEditable=false
       bIsStaticSwitch=false
       bIsVisible=false
       type {classStructOrEnum {refPath=&NiagaraInt32}}
      }
      Enable Torus Knot Radius {bIsDynamic=false, bIsEditable=false, bIsStaticSwitch=false, bIsVisible=false, type {classStructOrEnum {refPath=&NiagaraBool}}}
      Torus Knot Tube Radius {
       bIsDynamic=false
       bIsEditable=false
       bIsStaticSwitch=false
       bIsVisible=false
       type {classStructOrEnum {refPath=&NiagaraFloat}}
      }
      Cylinder Radius {
       bIsDynamic=false
       bIsEditable=false
       bIsStaticSwitch=false
       bIsVisible=false
       type {classStructOrEnum {refPath=&NiagaraFloat}}
      }
      Cylinder Height Midpoint {
       bIsDynamic=false
       bIsEditable=false
       bIsStaticSwitch=false
       bIsVisible=false
       type {classStructOrEnum {refPath=&NiagaraFloat}}
      }
      Ring Radius {
       bIsDynamic=false
       bIsEditable=false
       bIsStaticSwitch=false
       bIsVisible=false
       type {classStructOrEnum {refPath=&NiagaraFloat}}
      }
      Box Size {
       bIsDynamic=false
       bIsEditable=false
       bIsStaticSwitch=false
       bIsVisible=false
       type {classStructOrEnum {refPath=&Vector3f}}
      }
      Box Midpoint {
       bIsDynamic=false
       bIsEditable=false
       bIsStaticSwitch=false
       bIsVisible=false
       type {classStructOrEnum {refPath=&Vector3f}}
      }
      Cone Length {
       bIsDynamic=false
       bIsEditable=false
       bIsStaticSwitch=false
       bIsVisible=false
       type {classStructOrEnum {refPath=&NiagaraFloat}}
      }
      Plane Size {
       bIsDynamic=false
       bIsEditable=false
       bIsStaticSwitch=false
       bIsVisible=false
       type {classStructOrEnum {refPath=&Vector2f}}
      }
      Wedge Horizontal Angle {
       bIsDynamic=false
       bIsEditable=false
       bIsStaticSwitch=false
       bIsVisible=false
       type {classStructOrEnum {refPath=&NiagaraFloat}}
      }
      Large Radius {
       bIsDynamic=false
       bIsEditable=false
       bIsStaticSwitch=false
       bIsVisible=false
       type {classStructOrEnum {refPath=&NiagaraFloat}}
      }
      Plane Midpoint {
       bIsDynamic=false
       bIsEditable=false
       bIsStaticSwitch=false
       bIsVisible=false
       type {classStructOrEnum {refPath=&Vector2f}}
      }
      Wedge Vertical Angle {
       bIsDynamic=false
       bIsEditable=false
       bIsStaticSwitch=false
       bIsVisible=false
       type {classStructOrEnum {refPath=&NiagaraFloat}}
      }
      Cone Angle {
       bIsDynamic=false
       bIsEditable=false
       bIsStaticSwitch=false
       bIsVisible=false
       type {classStructOrEnum {refPath=&NiagaraFloat}}
      }
      Handle Radius {
       bIsDynamic=false
       bIsEditable=false
       bIsStaticSwitch=false
       bIsVisible=false
       type {classStructOrEnum {refPath=&NiagaraFloat}}
      }
      Cone Inner Angle {
       bIsDynamic=false
       bIsEditable=false
       bIsStaticSwitch=false
       bIsVisible=false
       type {classStructOrEnum {refPath=&NiagaraFloat}}
      }
      Cone Radial Angle {
       bIsDynamic=false
       bIsEditable=false
       bIsStaticSwitch=false
       bIsVisible=false
       type {classStructOrEnum {refPath=&NiagaraFloat}}
      }
      Cone Angle Type {
       bIsDynamic=false
       bIsEditable=false
       bIsStaticSwitch=true
       bIsVisible=false
       type {classStructOrEnum {refPath=&ENiagara_AngleInput}}
      }
      Flatten Endcaps {bIsDynamic=false, bIsEditable=false, bIsStaticSwitch=true, bIsVisible=false, type {classStructOrEnum {refPath=&NiagaraBool}}}
      Apply To Particle Position {
       bIsDynamic=false
       bIsEditable=true
       bIsStaticSwitch=true
       bIsVisible=true
       type {classStructOrEnum {refPath=&ENiagara_AttributeSamplingApplyOutput}}
      }
      Cone Surface Distribution {
       bIsDynamic=false
       bIsEditable=false
       bIsStaticSwitch=false
       bIsVisible=false
       type {classStructOrEnum {refPath=&NiagaraFloat}}
      }
      Ring / Disc Distribution Mode {
       bIsDynamic=false
       bIsEditable=false
       bIsStaticSwitch=true
       bIsVisible=false
       type {classStructOrEnum {refPath=&ENiagaraSphereDistributionMode}}
      }
      Sphere Distribution {
       bIsDynamic=false
       bIsEditable=true
       bIsStaticSwitch=true
       bIsVisible=true
       type {classStructOrEnum {refPath=&ENiagaraSphereDistributionMode}}
      }
      Torus Distribution Mode {
       bIsDynamic=false
       bIsEditable=false
       bIsStaticSwitch=true
       bIsVisible=false
       type {classStructOrEnum {refPath=&ENiagaraTorusDistributionMode}}
      }
      Sphere Surface Distribution {bIsDynamic=false, bIsEditable=true, bIsStaticSwitch=false, bIsVisible=true, type {classStructOrEnum {refPath=&NiagaraFloat}}}
      Hemisphere Distribution {bIsDynamic=false, bIsEditable=true, bIsStaticSwitch=false, bIsVisible=true, type {classStructOrEnum {refPath=&Vector2f}}}
      Cylinder Distribution {
       bIsDynamic=false
       bIsEditable=false
       bIsStaticSwitch=true
       bIsVisible=false
       type {classStructOrEnum {refPath=&ENiagara_CylinderMode}}
      }
      Surface Only {bIsDynamic=false, bIsEditable=false, bIsStaticSwitch=false, bIsVisible=false, type {classStructOrEnum {refPath=&NiagaraBool}}}
      Surface Only Band Thickness {
       bIsDynamic=false
       bIsEditable=false
       bIsStaticSwitch=false
       bIsVisible=false
       type {classStructOrEnum {refPath=&NiagaraFloat}}
      }
      Use Endcaps In Surface Only Mode {bIsDynamic=false, bIsEditable=false, bIsStaticSwitch=true, bIsVisible=false, type {classStructOrEnum {refPath=&NiagaraBool}}}
      Hemicircle X {bIsDynamic=false, bIsEditable=false, bIsStaticSwitch=true, bIsVisible=false, type {classStructOrEnum {refPath=&NiagaraBool}}}
      Hemicircle Y {bIsDynamic=false, bIsEditable=false, bIsStaticSwitch=true, bIsVisible=false, type {classStructOrEnum {refPath=&NiagaraBool}}}
      Disc Coverage {
       bIsDynamic=false
       bIsEditable=false
       bIsStaticSwitch=false
       bIsVisible=false
       type {classStructOrEnum {refPath=&NiagaraFloat}}
      }
      Hemicircle Internal Cap {bIsDynamic=false, bIsEditable=false, bIsStaticSwitch=true, bIsVisible=false, type {classStructOrEnum {refPath=&NiagaraBool}}}
      U Position {
       bIsDynamic=false
       bIsEditable=false
       bIsStaticSwitch=false
       bIsVisible=false
       type {classStructOrEnum {refPath=&NiagaraFloat}}
      }
      Enable Lathe Profile {bIsDynamic=false, bIsEditable=false, bIsStaticSwitch=true, bIsVisible=false, type {classStructOrEnum {refPath=&NiagaraBool}}}
      Lathe Profile {
       bIsDynamic=false
       bIsEditable=false
       bIsStaticSwitch=false
       bIsVisible=false
       type {classStructOrEnum {refPath=&NiagaraDataInterfaceCurve}}
      }
      Surface Distribution {
       bIsDynamic=false
       bIsEditable=false
       bIsStaticSwitch=false
       bIsVisible=false
       type {classStructOrEnum {refPath=&NiagaraFloat}}
      }
      Surface Expansion Mode {
       bIsDynamic=false
       bIsEditable=false
       bIsStaticSwitch=false
       bIsVisible=false
       type {classStructOrEnum {refPath=&ENiagaraExpansionMode}}
      }
      U Distribution {
       bIsDynamic=false
       bIsEditable=false
       bIsStaticSwitch=false
       bIsVisible=false
       type {classStructOrEnum {refPath=&NiagaraFloat}}
      }
      V Position {
       bIsDynamic=false
       bIsEditable=false
       bIsStaticSwitch=false
       bIsVisible=false
       type {classStructOrEnum {refPath=&NiagaraFloat}}
      }
      Radius Position {
       bIsDynamic=false
       bIsEditable=false
       bIsStaticSwitch=false
       bIsVisible=false
       type {classStructOrEnum {refPath=&NiagaraFloat}}
      }
      Sphere Radius Position {
       bIsDynamic=false
       bIsEditable=false
       bIsStaticSwitch=false
       bIsVisible=false
       type {classStructOrEnum {refPath=&NiagaraFloat}}
      }
      V Distribution {
       bIsDynamic=false
       bIsEditable=false
       bIsStaticSwitch=false
       bIsVisible=false
       type {classStructOrEnum {refPath=&NiagaraFloat}}
      }
      Evaluation Type {
       bIsDynamic=false
       bIsEditable=false
       bIsStaticSwitch=true
       bIsVisible=false
       type {classStructOrEnum {refPath=&ENiagaraRandomnessEvaluation}}
      }
      Uniform Distribution {
       bIsDynamic=false
       bIsEditable=false
       bIsStaticSwitch=false
       bIsVisible=false
       type {classStructOrEnum {refPath=&NiagaraFloat}}
      }
      Uniform Spiral Amount {
       bIsDynamic=false
       bIsEditable=false
       bIsStaticSwitch=false
       bIsVisible=false
       type {classStructOrEnum {refPath=&NiagaraFloat}}
      }
      Uniform Spiral Falloff {
       bIsDynamic=false
       bIsEditable=false
       bIsStaticSwitch=false
       bIsVisible=false
       type {classStructOrEnum {refPath=&NiagaraFloat}}
      }
      Uniform Count {
       bIsDynamic=false
       bIsEditable=false
       bIsStaticSwitch=false
       bIsVisible=false
       type {classStructOrEnum {refPath=&NiagaraInt32}}
      }
      Uniform Particle Index {
       bIsDynamic=false
       bIsEditable=false
       bIsStaticSwitch=false
       bIsVisible=false
       type {classStructOrEnum {refPath=&NiagaraInt32}}
      }
      Hemisphere Angle Type {
       bIsDynamic=false
       bIsEditable=true
       bIsStaticSwitch=true
       bIsVisible=true
       type {classStructOrEnum {refPath=&ENiagara_AngleInput}}
      }
      Uniform Endpoint Offset {
       bIsDynamic=false
       bIsEditable=false
       bIsStaticSwitch=false
       bIsVisible=false
       type {classStructOrEnum {refPath=&NiagaraFloat}}
      }
      MaskSpawnGroup {bIsDynamic=false, bIsEditable=true, bIsStaticSwitch=false, bIsVisible=true, type {classStructOrEnum {refPath=&NiagaraBool}}}
      Spawn Group Mask {bIsDynamic=false, bIsEditable=false, bIsStaticSwitch=false, bIsVisible=true, type {classStructOrEnum {refPath=&NiagaraInt32}}}
      Spawn Group {
       bIsDynamic=false
       bIsEditable=false
       bIsStaticSwitch=false
       bIsVisible=false
       type {classStructOrEnum {refPath=&NiagaraInt32}}
      }
      Randomness Mode {
       bIsDynamic=false
       bIsEditable=true
       bIsStaticSwitch=true
       bIsVisible=true
       type {classStructOrEnum {refPath=&ENiagaraRandomnessMode}}
      }
      Override Seed {bIsDynamic=false, bIsEditable=true, bIsStaticSwitch=false, bIsVisible=true, type {classStructOrEnum {refPath=&NiagaraBool}}}
      Random Seed {bIsDynamic=false, bIsEditable=false, bIsStaticSwitch=false, bIsVisible=true, type {classStructOrEnum {refPath=&NiagaraInt32}}}
      Fixed Random Seed {bIsDynamic=false, bIsEditable=false, bIsStaticSwitch=true, bIsVisible=false, type {classStructOrEnum {refPath=&NiagaraBool}}}
      Transform Order {
       bIsDynamic=false
       bIsEditable=true
       bIsStaticSwitch=true
       bIsVisible=true
       type {classStructOrEnum {refPath=&ENiagara_TransformOrder}}
      }
      Scale Mode {
       bIsDynamic=false
       bIsEditable=true
       bIsStaticSwitch=true
       bIsVisible=true
       type {classStructOrEnum {refPath=&ENiagara_ScaleMode}}
      }
      Non Uniform Scale {bIsDynamic=false, bIsEditable=true, bIsStaticSwitch=false, bIsVisible=true, type {classStructOrEnum {refPath=&Vector3f}}}
      UseOwnerScale {bIsDynamic=false, bIsEditable=true, bIsStaticSwitch=false, bIsVisible=true, type {classStructOrEnum {refPath=&NiagaraBool}}}
      Apply Owner Scale {bIsDynamic=false, bIsEditable=true, bIsStaticSwitch=false, bIsVisible=true, type {classStructOrEnum {refPath=&Vector3f}}}
      Rotation Mode {
       bIsDynamic=false
       bIsEditable=true
       bIsStaticSwitch=true
       bIsVisible=true
       type {classStructOrEnum {refPath=&ENiagara_RotationMode}}
      }
      Rotation Axis {
       bIsDynamic=false
       bIsEditable=false
       bIsStaticSwitch=false
       bIsVisible=false
       type {classStructOrEnum {refPath=&Vector3f}}
      }
      Rotation Matrix {
       bIsDynamic=false
       bIsEditable=false
       bIsStaticSwitch=false
       bIsVisible=false
       type {classStructOrEnum {refPath=&NiagaraMatrix}}
      }
      Rotation Quaternion {bIsDynamic=false, bIsEditable=false, bIsStaticSwitch=false, bIsVisible=false, type {classStructOrEnum {refPath=&Quat4f}}}
      Yaw / Pitch / Roll {
       bIsDynamic=false
       bIsEditable=false
       bIsStaticSwitch=false
       bIsVisible=false
       type {classStructOrEnum {refPath=&Vector3f}}
      }
      Rotation Angle {
       bIsDynamic=false
       bIsEditable=false
       bIsStaticSwitch=false
       bIsVisible=false
       type {classStructOrEnum {refPath=&NiagaraFloat}}
      }
      Invert Rotation Quaternion {bIsDynamic=false, bIsEditable=false, bIsStaticSwitch=false, bIsVisible=false, type {classStructOrEnum {refPath=&NiagaraBool}}}
      Additional Quaternion Rotation {bIsDynamic=false, bIsEditable=false, bIsStaticSwitch=false, bIsVisible=false, type {classStructOrEnum {refPath=&NiagaraBool}}}
      Additional Yaw / Pitch / Roll {
       bIsDynamic=false
       bIsEditable=false
       bIsStaticSwitch=false
       bIsVisible=false
       type {classStructOrEnum {refPath=&Vector3f}}
      }
      Rotation Angle Type {
       bIsDynamic=false
       bIsEditable=false
       bIsStaticSwitch=true
       bIsVisible=false
       type {classStructOrEnum {refPath=&ENiagara_AngleInput}}
      }
      Rotation Coordinate Space {
       bIsDynamic=false
       bIsEditable=true
       bIsStaticSwitch=false
       bIsVisible=true
       type {classStructOrEnum {refPath=&ENiagaraCoordinateSpace}}
      }
      Offset Mode {
       bIsDynamic=false
       bIsEditable=true
       bIsStaticSwitch=true
       bIsVisible=true
       type {classStructOrEnum {refPath=&ENiagara_OffsetMode}}
      }
      Shape Origin {
       bIsDynamic=false
       bIsEditable=true
       bIsStaticSwitch=false
       bIsVisible=true
       type {classStructOrEnum {refPath=&NiagaraPosition}}
      }
      Offset {bIsDynamic=true, bIsEditable=true, bIsStaticSwitch=false, bIsVisible=true, type {classStructOrEnum {refPath=&Vector3f}}}
      Offset Coordinate Space {
       bIsDynamic=false
       bIsEditable=true
       bIsStaticSwitch=false
       bIsVisible=true
       type {classStructOrEnum {refPath=&ENiagaraCoordinateSpace}}
      }
      Custom Transform Matrix {
       bIsDynamic=false
       bIsEditable=false
       bIsStaticSwitch=false
       bIsVisible=false
       type {classStructOrEnum {refPath=&NiagaraMatrix}}
      }
      Transform Method {
       bIsDynamic=false
       bIsEditable=true
       bIsStaticSwitch=true
       bIsVisible=true
       type {classStructOrEnum {refPath=&ENiagara_TransformType}}
      }
     ]
     moduleScript {refPath=&ShapeLocation}
    }
    {
     bIsSetParametersModule=false
     enabled=true
     moduleName=CameraOffset
     inputs [Camera Offset Amount {bIsDynamic=true, bIsEditable=true, bIsStaticSwitch=false, bIsVisible=true, type {classStructOrEnum {refPath=&NiagaraFloat}}}]
     moduleScript {refPath=&CameraOffset}
    }
    {
     bIsSetParametersModule=false
     enabled=true
     moduleName=AddVelocity001
     inputs [
      Velocity Mode {
       bIsDynamic=false
       bIsEditable=true
       bIsStaticSwitch=true
       bIsVisible=true
       type {classStructOrEnum {refPath=&ENiagara_VelocityMode}}
      }
      Velocity {bIsDynamic=true, bIsEditable=false, bIsStaticSwitch=false, bIsVisible=false, type {classStructOrEnum {refPath=&Vector3f}}}
      Velocity Speed {bIsDynamic=true, bIsEditable=true, bIsStaticSwitch=false, bIsVisible=true, type {classStructOrEnum {refPath=&NiagaraFloat}}}
      Velocity Speed Scale {
       bIsDynamic=false
       bIsEditable=false
       bIsStaticSwitch=false
       bIsVisible=false
       type {classStructOrEnum {refPath=&NiagaraFloat}}
      }
      Module.Use Falloff Distance {bIsDynamic=false, bIsEditable=true, bIsStaticSwitch=false, bIsVisible=true, type {classStructOrEnum {refPath=&NiagaraBool}}}
      Constrain To Radius {bIsDynamic=false, bIsEditable=false, bIsStaticSwitch=false, bIsVisible=true, type {classStructOrEnum {refPath=&NiagaraFloat}}}
      Distribution Along Cone Axis {
       bIsDynamic=false
       bIsEditable=false
       bIsStaticSwitch=false
       bIsVisible=false
       type {classStructOrEnum {refPath=&NiagaraFloat}}
      }
      Radius Falloff Near / Far {
       bIsDynamic=false
       bIsEditable=false
       bIsStaticSwitch=false
       bIsVisible=false
       type {classStructOrEnum {refPath=&Vector2f}}
      }
      Radius Falloff Exponent {
       bIsDynamic=false
       bIsEditable=false
       bIsStaticSwitch=false
       bIsVisible=false
       type {classStructOrEnum {refPath=&NiagaraFloat}}
      }
      Speed Falloff From Cone Axis {
       bIsDynamic=false
       bIsEditable=false
       bIsStaticSwitch=false
       bIsVisible=false
       type {classStructOrEnum {refPath=&NiagaraFloat}}
      }
      Invert Velocity Falloff {bIsDynamic=false, bIsEditable=false, bIsStaticSwitch=false, bIsVisible=false, type {classStructOrEnum {refPath=&NiagaraBool}}}
      Override Seed {bIsDynamic=false, bIsEditable=true, bIsStaticSwitch=false, bIsVisible=true, type {classStructOrEnum {refPath=&NiagaraBool}}}
      Random Seed {bIsDynamic=false, bIsEditable=false, bIsStaticSwitch=false, bIsVisible=true, type {classStructOrEnum {refPath=&NiagaraInt32}}}
      Fixed Random Seed {bIsDynamic=false, bIsEditable=false, bIsStaticSwitch=true, bIsVisible=false, type {classStructOrEnum {refPath=&NiagaraBool}}}
      Randomness Mode {
       bIsDynamic=false
       bIsEditable=true
       bIsStaticSwitch=true
       bIsVisible=true
       type {classStructOrEnum {refPath=&ENiagaraRandomnessMode}}
      }
      Cone Axis {
       bIsDynamic=false
       bIsEditable=false
       bIsStaticSwitch=false
       bIsVisible=false
       type {classStructOrEnum {refPath=&Vector3f}}
      }
      Cone Angle {
       bIsDynamic=false
       bIsEditable=false
       bIsStaticSwitch=false
       bIsVisible=false
       type {classStructOrEnum {refPath=&NiagaraFloat}}
      }
      Inner Cone Angle {
       bIsDynamic=false
       bIsEditable=false
       bIsStaticSwitch=false
       bIsVisible=false
       type {classStructOrEnum {refPath=&NiagaraFloat}}
      }
      Cone Angle Mode {
       bIsDynamic=false
       bIsEditable=false
       bIsStaticSwitch=true
       bIsVisible=false
       type {classStructOrEnum {refPath=&ENiagara_AngleInput}}
      }
      Rotation Mode {
       bIsDynamic=false
       bIsEditable=false
       bIsStaticSwitch=true
       bIsVisible=false
       type {classStructOrEnum {refPath=&ENiagara_RotationMode}}
      }
      Rotation Axis {
       bIsDynamic=false
       bIsEditable=false
       bIsStaticSwitch=false
       bIsVisible=false
       type {classStructOrEnum {refPath=&Vector3f}}
      }
      Yaw / Pitch / Roll {
       bIsDynamic=false
       bIsEditable=false
       bIsStaticSwitch=false
       bIsVisible=false
       type {classStructOrEnum {refPath=&Vector3f}}
      }
      Rotation Quaternion {bIsDynamic=false, bIsEditable=false, bIsStaticSwitch=false, bIsVisible=false, type {classStructOrEnum {refPath=&Quat4f}}}
      Rotation Matrix {
       bIsDynamic=false
       bIsEditable=false
       bIsStaticSwitch=false
       bIsVisible=false
       type {classStructOrEnum {refPath=&NiagaraMatrix}}
      }
      Rotation Angle {
       bIsDynamic=false
       bIsEditable=false
       bIsStaticSwitch=false
       bIsVisible=false
       type {classStructOrEnum {refPath=&NiagaraFloat}}
      }
      Invert Rotation Quaternion {bIsDynamic=false, bIsEditable=false, bIsStaticSwitch=false, bIsVisible=false, type {classStructOrEnum {refPath=&NiagaraBool}}}
      Additional Quaternion Rotation {bIsDynamic=false, bIsEditable=false, bIsStaticSwitch=false, bIsVisible=false, type {classStructOrEnum {refPath=&NiagaraBool}}}
      Additional Yaw / Pitch / Roll {
       bIsDynamic=false
       bIsEditable=false
       bIsStaticSwitch=false
       bIsVisible=false
       type {classStructOrEnum {refPath=&Vector3f}}
      }
      Rotation Angle Type {
       bIsDynamic=false
       bIsEditable=false
       bIsStaticSwitch=true
       bIsVisible=false
       type {classStructOrEnum {refPath=&ENiagara_AngleInput}}
      }
      Rotation Coordinate Space {
       bIsDynamic=false
       bIsEditable=false
       bIsStaticSwitch=false
       bIsVisible=false
       type {classStructOrEnum {refPath=&ENiagaraCoordinateSpace}}
      }
      Velocity Origin {
       bIsDynamic=false
       bIsEditable=true
       bIsStaticSwitch=false
       bIsVisible=true
       type {classStructOrEnum {refPath=&NiagaraPosition}}
      }
      Origin Offset {bIsDynamic=true, bIsEditable=true, bIsStaticSwitch=false, bIsVisible=true, type {classStructOrEnum {refPath=&Vector3f}}}
      Origin Offset Coordinate Space {
       bIsDynamic=false
       bIsEditable=true
       bIsStaticSwitch=false
       bIsVisible=true
       type {classStructOrEnum {refPath=&ENiagaraCoordinateSpace}}
      }
      Default Position {
       bIsDynamic=false
       bIsEditable=true
       bIsStaticSwitch=false
       bIsVisible=true
       type {classStructOrEnum {refPath=&NiagaraPosition}}
      }
     ]
     moduleScript {refPath=&AddVelocity}
    }
    {
     bIsSetParametersModule=false
     enabled=true
     moduleName=AddVelocity
     inputs [
      Velocity Mode {
       bIsDynamic=false
       bIsEditable=true
       bIsStaticSwitch=true
       bIsVisible=true
       type {classStructOrEnum {refPath=&ENiagara_VelocityMode}}
      }
      Velocity {bIsDynamic=true, bIsEditable=true, bIsStaticSwitch=false, bIsVisible=true, type {classStructOrEnum {refPath=&Vector3f}}}
      Velocity Speed {
       bIsDynamic=false
       bIsEditable=false
       bIsStaticSwitch=false
       bIsVisible=false
       type {classStructOrEnum {refPath=&NiagaraFloat}}
      }
      Velocity Speed Scale {bIsDynamic=false, bIsEditable=true, bIsStaticSwitch=false, bIsVisible=true, type {classStructOrEnum {refPath=&NiagaraFloat}}}
      Constrain To Radius {
       bIsDynamic=false
       bIsEditable=false
       bIsStaticSwitch=false
       bIsVisible=false
       type {classStructOrEnum {refPath=&NiagaraFloat}}
      }
      Distribution Along Cone Axis {
       bIsDynamic=false
       bIsEditable=false
       bIsStaticSwitch=false
       bIsVisible=false
       type {classStructOrEnum {refPath=&NiagaraFloat}}
      }
      Radius Falloff Near / Far {
       bIsDynamic=false
       bIsEditable=false
       bIsStaticSwitch=false
       bIsVisible=false
       type {classStructOrEnum {refPath=&Vector2f}}
      }
      Radius Falloff Exponent {
       bIsDynamic=false
       bIsEditable=false
       bIsStaticSwitch=false
       bIsVisible=false
       type {classStructOrEnum {refPath=&NiagaraFloat}}
      }
      Speed Falloff From Cone Axis {
       bIsDynamic=false
       bIsEditable=false
       bIsStaticSwitch=false
       bIsVisible=false
       type {classStructOrEnum {refPath=&NiagaraFloat}}
      }
      Invert Velocity Falloff {bIsDynamic=false, bIsEditable=false, bIsStaticSwitch=false, bIsVisible=false, type {classStructOrEnum {refPath=&NiagaraBool}}}
      Override Seed {bIsDynamic=false, bIsEditable=false, bIsStaticSwitch=false, bIsVisible=false, type {classStructOrEnum {refPath=&NiagaraBool}}}
      Random Seed {
       bIsDynamic=false
       bIsEditable=false
       bIsStaticSwitch=false
       bIsVisible=false
       type {classStructOrEnum {refPath=&NiagaraInt32}}
      }
      Fixed Random Seed {bIsDynamic=false, bIsEditable=false, bIsStaticSwitch=true, bIsVisible=false, type {classStructOrEnum {refPath=&NiagaraBool}}}
      Randomness Mode {
       bIsDynamic=false
       bIsEditable=false
       bIsStaticSwitch=true
       bIsVisible=false
       type {classStructOrEnum {refPath=&ENiagaraRandomnessMode}}
      }
      Cone Axis {
       bIsDynamic=false
       bIsEditable=false
       bIsStaticSwitch=false
       bIsVisible=false
       type {classStructOrEnum {refPath=&Vector3f}}
      }
      Cone Angle {
       bIsDynamic=false
       bIsEditable=false
       bIsStaticSwitch=false
       bIsVisible=false
       type {classStructOrEnum {refPath=&NiagaraFloat}}
      }
      Inner Cone Angle {
       bIsDynamic=false
       bIsEditable=false
       bIsStaticSwitch=false
       bIsVisible=false
       type {classStructOrEnum {refPath=&NiagaraFloat}}
      }
      Cone Angle Mode {
       bIsDynamic=false
       bIsEditable=false
       bIsStaticSwitch=true
       bIsVisible=false
       type {classStructOrEnum {refPath=&ENiagara_AngleInput}}
      }
      Rotation Mode {
       bIsDynamic=false
       bIsEditable=true
       bIsStaticSwitch=true
       bIsVisible=true
       type {classStructOrEnum {refPath=&ENiagara_RotationMode}}
      }
      Rotation Axis {
       bIsDynamic=false
       bIsEditable=false
       bIsStaticSwitch=false
       bIsVisible=false
       type {classStructOrEnum {refPath=&Vector3f}}
      }
      Yaw / Pitch / Roll {
       bIsDynamic=false
       bIsEditable=false
       bIsStaticSwitch=false
       bIsVisible=false
       type {classStructOrEnum {refPath=&Vector3f}}
      }
      Rotation Quaternion {bIsDynamic=false, bIsEditable=false, bIsStaticSwitch=false, bIsVisible=false, type {classStructOrEnum {refPath=&Quat4f}}}
      Rotation Matrix {
       bIsDynamic=false
       bIsEditable=false
       bIsStaticSwitch=false
       bIsVisible=false
       type {classStructOrEnum {refPath=&NiagaraMatrix}}
      }
      Rotation Angle {
       bIsDynamic=false
       bIsEditable=false
       bIsStaticSwitch=false
       bIsVisible=false
       type {classStructOrEnum {refPath=&NiagaraFloat}}
      }
      Invert Rotation Quaternion {bIsDynamic=false, bIsEditable=false, bIsStaticSwitch=false, bIsVisible=false, type {classStructOrEnum {refPath=&NiagaraBool}}}
      Additional Quaternion Rotation {bIsDynamic=false, bIsEditable=false, bIsStaticSwitch=false, bIsVisible=false, type {classStructOrEnum {refPath=&NiagaraBool}}}
      Additional Yaw / Pitch / Roll {
       bIsDynamic=false
       bIsEditable=false
       bIsStaticSwitch=false
       bIsVisible=false
       type {classStructOrEnum {refPath=&Vector3f}}
      }
      Rotation Angle Type {
       bIsDynamic=false
       bIsEditable=false
       bIsStaticSwitch=true
       bIsVisible=false
       type {classStructOrEnum {refPath=&ENiagara_AngleInput}}
      }
      Rotation Coordinate Space {
       bIsDynamic=false
       bIsEditable=true
       bIsStaticSwitch=false
       bIsVisible=true
       type {classStructOrEnum {refPath=&ENiagaraCoordinateSpace}}
      }
      Velocity Origin {
       bIsDynamic=false
       bIsEditable=false
       bIsStaticSwitch=false
       bIsVisible=false
       type {classStructOrEnum {refPath=&NiagaraPosition}}
      }
      Origin Offset {
       bIsDynamic=false
       bIsEditable=false
       bIsStaticSwitch=false
       bIsVisible=false
       type {classStructOrEnum {refPath=&Vector3f}}
      }
      Origin Offset Coordinate Space {
       bIsDynamic=false
       bIsEditable=false
       bIsStaticSwitch=false
       bIsVisible=false
       type {classStructOrEnum {refPath=&ENiagaraCoordinateSpace}}
      }
      Default Position {
       bIsDynamic=false
       bIsEditable=false
       bIsStaticSwitch=false
       bIsVisible=false
       type {classStructOrEnum {refPath=&NiagaraPosition}}
      }
     ]
     moduleScript {refPath=&AddVelocity}
    }
    {
     bIsSetParametersModule=false
     enabled=true
     moduleName=ApplyInitialForces
     inputs [
      Apply Force to Velocity {bIsDynamic=false, bIsEditable=true, bIsStaticSwitch=false, bIsVisible=true, type {classStructOrEnum {refPath=&NiagaraBool}}}
      Apply Rotational Force to Rotational Velocity {bIsDynamic=false, bIsEditable=true, bIsStaticSwitch=false, bIsVisible=true, type {classStructOrEnum {refPath=&NiagaraBool}}}
      Apply Force to Position {bIsDynamic=false, bIsEditable=true, bIsStaticSwitch=false, bIsVisible=true, type {classStructOrEnum {refPath=&NiagaraBool}}}
      Positional Force Warmup Time {bIsDynamic=false, bIsEditable=false, bIsStaticSwitch=false, bIsVisible=true, type {classStructOrEnum {refPath=&NiagaraFloat}}}
     ]
     moduleScript {refPath=&ApplyInitialForces}
    }
    {
     bIsSetParametersModule=true
     enabled=true
     moduleName=SetVariables_#3
     inputs [Particles.DynamicMaterialParameter_Air {bIsDynamic=true, bIsEditable=true, bIsStaticSwitch=false, bIsVisible=true, type {classStructOrEnum {refPath=&Vector4f}}}]
     moduleScript {refPath=&self:DefaultPropSpawn_Merged_0.NiagaraScriptSource_1.NiagaraGraph_0.NiagaraNodeAssignment_5.SetVariables_#3}
    }
   ]
  }
  particleUpdateScript {
   scriptName=ParticleUpdateScript
   modules [
    {
     bIsSetParametersModule=false
     enabled=true
     moduleName=ParticleState
     inputs [
      Kill Particles When Lifetime Has Elapsed {bIsDynamic=false, bIsEditable=true, bIsStaticSwitch=true, bIsVisible=true, type {classStructOrEnum {refPath=&NiagaraBool}}}
      Loop Particles Lifetime {bIsDynamic=false, bIsEditable=false, bIsStaticSwitch=false, bIsVisible=false, type {classStructOrEnum {refPath=&NiagaraBool}}}
      Let Infinitely Lived Particles Die When Emitter Deactivates {bIsDynamic=false, bIsEditable=false, bIsStaticSwitch=false, bIsVisible=false, type {classStructOrEnum {refPath=&NiagaraBool}}}
      Lifetime {bIsDynamic=false, bIsEditable=true, bIsStaticSwitch=false, bIsVisible=true, type {classStructOrEnum {refPath=&NiagaraFloat}}}
      DeltaTime {bIsDynamic=false, bIsEditable=true, bIsStaticSwitch=false, bIsVisible=true, type {classStructOrEnum {refPath=&NiagaraFloat}}}
     ]
     moduleScript {refPath=&ParticleState}
    }
    {
     bIsSetParametersModule=false
     enabled=true
     moduleName=SpriteRotationRate
     inputs [
      Rotation Rate {bIsDynamic=true, bIsEditable=true, bIsStaticSwitch=false, bIsVisible=true, type {classStructOrEnum {refPath=&NiagaraFloat}}}
      Delta Time {bIsDynamic=false, bIsEditable=true, bIsStaticSwitch=false, bIsVisible=true, type {classStructOrEnum {refPath=&NiagaraFloat}}}
     ]
     moduleScript {refPath=&SpriteRotationRate}
    }
    {
     bIsSetParametersModule=false
     enabled=true
     moduleName=GravityForce
     inputs [
      Gravity {bIsDynamic=true, bIsEditable=true, bIsStaticSwitch=false, bIsVisible=true, type {classStructOrEnum {refPath=&Vector3f}}}
      Coordinate Space {
       bIsDynamic=false
       bIsEditable=true
       bIsStaticSwitch=true
       bIsVisible=true
       type {classStructOrEnum {refPath=&ENiagaraCoordinateSpace}}
      }
     ]
     moduleScript {refPath=&GravityForce}
    }
    {
     bIsSetParametersModule=false
     enabled=true
     moduleName=CurlNoiseForce
     inputs [
      Noise Strength {bIsDynamic=true, bIsEditable=true, bIsStaticSwitch=false, bIsVisible=true, type {classStructOrEnum {refPath=&NiagaraFloat}}}
      Noise Frequency {bIsDynamic=true, bIsEditable=true, bIsStaticSwitch=false, bIsVisible=true, type {classStructOrEnum {refPath=&NiagaraFloat}}}
      Noise Quality / Cost {
       bIsDynamic=false
       bIsEditable=true
       bIsStaticSwitch=true
       bIsVisible=true
       type {classStructOrEnum {refPath=&ENiagaraCurlNoiseQuality}}
      }
      Random Seed {bIsDynamic=false, bIsEditable=true, bIsStaticSwitch=false, bIsVisible=true, type {classStructOrEnum {refPath=&NiagaraInt32}}}
      Pan Noise {bIsDynamic=false, bIsEditable=true, bIsStaticSwitch=false, bIsVisible=true, type {classStructOrEnum {refPath=&NiagaraBool}}}
      Pan Noise Field {bIsDynamic=false, bIsEditable=false, bIsStaticSwitch=false, bIsVisible=true, type {classStructOrEnum {refPath=&Vector3f}}}
      Write to Intrinsic Parameters {bIsDynamic=false, bIsEditable=true, bIsStaticSwitch=true, bIsVisible=true, type {classStructOrEnum {refPath=&NiagaraBool}}}
      Sample Position {bIsDynamic=false, bIsEditable=true, bIsStaticSwitch=false, bIsVisible=true, type {classStructOrEnum {refPath=&Vector3f}}}
      CurlNoise {
       bIsDynamic=false
       bIsEditable=false
       bIsStaticSwitch=false
       bIsVisible=false
       type {classStructOrEnum {refPath=&NiagaraDataInterfaceCurlNoise}}
      }
      Mask Curl Noise {bIsDynamic=false, bIsEditable=true, bIsStaticSwitch=true, bIsVisible=true, type {classStructOrEnum {refPath=&NiagaraBool}}}
      Curl Noise Cone Mask Angle {
       bIsDynamic=false
       bIsEditable=false
       bIsStaticSwitch=false
       bIsVisible=false
       type {classStructOrEnum {refPath=&NiagaraFloat}}
      }
      Curl Noise Cone Mask Falloff Angle {
       bIsDynamic=false
       bIsEditable=false
       bIsStaticSwitch=false
       bIsVisible=false
       type {classStructOrEnum {refPath=&NiagaraFloat}}
      }
      Curl Noise Cone Mask Axis {
       bIsDynamic=false
       bIsEditable=false
       bIsStaticSwitch=false
       bIsVisible=false
       type {classStructOrEnum {refPath=&Vector3f}}
      }
     ]
     moduleScript {refPath=/Niagara/Modules/Update/Forces/CurlNoiseForce}
    }
    {
     bIsSetParametersModule=false
     enabled=true
     moduleName=Drag
     inputs [
      Use Linear Drag {bIsDynamic=false, bIsEditable=true, bIsStaticSwitch=false, bIsVisible=true, type {classStructOrEnum {refPath=&NiagaraBool}}}
      Drag {bIsDynamic=true, bIsEditable=true, bIsStaticSwitch=false, bIsVisible=true, type {classStructOrEnum {refPath=&NiagaraFloat}}}
      Use Rotational Drag {bIsDynamic=false, bIsEditable=true, bIsStaticSwitch=false, bIsVisible=true, type {classStructOrEnum {refPath=&NiagaraBool}}}
      Rotational Drag {bIsDynamic=false, bIsEditable=true, bIsStaticSwitch=false, bIsVisible=true, type {classStructOrEnum {refPath=&NiagaraFloat}}}
      Ignore Mass {bIsDynamic=false, bIsEditable=true, bIsStaticSwitch=false, bIsVisible=true, type {classStructOrEnum {refPath=&NiagaraBool}}}
     ]
     moduleScript {refPath=&Drag}
    }
    {
     bIsSetParametersModule=false
     enabled=true
     moduleName=SolveForcesAndVelocity
     inputs [
      Clamp Velocity {bIsDynamic=false, bIsEditable=true, bIsStaticSwitch=false, bIsVisible=true, type {classStructOrEnum {refPath=&NiagaraBool}}}
      Speed Limit {bIsDynamic=false, bIsEditable=false, bIsStaticSwitch=false, bIsVisible=true, type {classStructOrEnum {refPath=&NiagaraFloat}}}
      Limit Acceleration {bIsDynamic=false, bIsEditable=true, bIsStaticSwitch=false, bIsVisible=true, type {classStructOrEnum {refPath=&NiagaraBool}}}
      Acceleration Limit {bIsDynamic=false, bIsEditable=false, bIsStaticSwitch=false, bIsVisible=true, type {classStructOrEnum {refPath=&NiagaraFloat}}}
      Manually Enable Rotational Solver {bIsDynamic=false, bIsEditable=true, bIsStaticSwitch=false, bIsVisible=true, type {classStructOrEnum {refPath=&NiagaraBool}}}
      Write to Intrinsic Properties {bIsDynamic=false, bIsEditable=true, bIsStaticSwitch=false, bIsVisible=true, type {classStructOrEnum {refPath=&NiagaraBool}}}
      Write to Presolve Properties {bIsDynamic=false, bIsEditable=true, bIsStaticSwitch=true, bIsVisible=true, type {classStructOrEnum {refPath=&NiagaraBool}}}
      DeltaTime {bIsDynamic=false, bIsEditable=true, bIsStaticSwitch=false, bIsVisible=true, type {classStructOrEnum {refPath=&NiagaraFloat}}}
      Mass {bIsDynamic=false, bIsEditable=true, bIsStaticSwitch=false, bIsVisible=true, type {classStructOrEnum {refPath=&NiagaraFloat}}}
      Position {
       bIsDynamic=false
       bIsEditable=true
       bIsStaticSwitch=false
       bIsVisible=true
       type {classStructOrEnum {refPath=&NiagaraPosition}}
      }
      Velocity {bIsDynamic=false, bIsEditable=true, bIsStaticSwitch=false, bIsVisible=true, type {classStructOrEnum {refPath=&Vector3f}}}
      PreviousVelocity {bIsDynamic=false, bIsEditable=true, bIsStaticSwitch=false, bIsVisible=true, type {classStructOrEnum {refPath=&Vector3f}}}
      Force {bIsDynamic=false, bIsEditable=true, bIsStaticSwitch=false, bIsVisible=true, type {classStructOrEnum {refPath=&Vector3f}}}
      Mesh Orientation {bIsDynamic=false, bIsEditable=false, bIsStaticSwitch=false, bIsVisible=false, type {classStructOrEnum {refPath=&Quat4f}}}
      Rotational Velocity {
       bIsDynamic=false
       bIsEditable=false
       bIsStaticSwitch=false
       bIsVisible=false
       type {classStructOrEnum {refPath=&Vector3f}}
      }
      Rotational Force {
       bIsDynamic=false
       bIsEditable=false
       bIsStaticSwitch=false
       bIsVisible=false
       type {classStructOrEnum {refPath=&Vector3f}}
      }
      Rotational Inertia {
       bIsDynamic=false
       bIsEditable=false
       bIsStaticSwitch=false
       bIsVisible=false
       type {classStructOrEnum {refPath=&NiagaraFloat}}
      }
     ]
     moduleScript {refPath=&SolveForcesAndVelocity}
    }
    {
     bIsSetParametersModule=false
     enabled=true
     moduleName=ScaleSpriteSize
     inputs [
      Scale Sprite Size Mode {
       bIsDynamic=false
       bIsEditable=true
       bIsStaticSwitch=true
       bIsVisible=true
       type {classStructOrEnum {refPath=&ENiagara_ScaleSpriteSize}}
      }
      Uniform Curve Sprite Scale {
       bIsDynamic=false
       bIsEditable=false
       bIsStaticSwitch=false
       bIsVisible=false
       type {classStructOrEnum {refPath=&NiagaraDataInterfaceCurve}}
      }
      Non-Uniform Curve Sprite Scale {
       bIsDynamic=false
       bIsEditable=false
       bIsStaticSwitch=false
       bIsVisible=false
       type {classStructOrEnum {refPath=&NiagaraDataInterfaceVector2DCurve}}
      }
      Uniform Curve Index {
       bIsDynamic=false
       bIsEditable=false
       bIsStaticSwitch=false
       bIsVisible=false
       type {classStructOrEnum {refPath=&NiagaraFloat}}
      }
      Non-Uniform Curve Index {
       bIsDynamic=false
       bIsEditable=false
       bIsStaticSwitch=false
       bIsVisible=false
       type {classStructOrEnum {refPath=&NiagaraFloat}}
      }
      Non-Uniform Curve Scale {
       bIsDynamic=false
       bIsEditable=false
       bIsStaticSwitch=false
       bIsVisible=false
       type {classStructOrEnum {refPath=&Vector2f}}
      }
      Uniform Curve Scale {
       bIsDynamic=false
       bIsEditable=false
       bIsStaticSwitch=false
       bIsVisible=false
       type {classStructOrEnum {refPath=&NiagaraFloat}}
      }
      Scale Factor {bIsDynamic=true, bIsEditable=true, bIsStaticSwitch=false, bIsVisible=true, type {classStructOrEnum {refPath=&Vector2f}}}
      Uniform Scale Factor {
       bIsDynamic=false
       bIsEditable=false
       bIsStaticSwitch=false
       bIsVisible=false
       type {classStructOrEnum {refPath=&NiagaraFloat}}
      }
      Initial Sprite Size {bIsDynamic=false, bIsEditable=true, bIsStaticSwitch=false, bIsVisible=true, type {classStructOrEnum {refPath=&Vector2f}}}
     ]
     moduleScript {refPath=&ScaleSpriteSize}
    }
    {
     bIsSetParametersModule=true
     enabled=true
     moduleName=SetVariables_#4
     inputs [
      Particles.DynParSmoke0 {bIsDynamic=true, bIsEditable=true, bIsStaticSwitch=false, bIsVisible=true, type {classStructOrEnum {refPath=&Vector4f}}}
      Particles.DynParSmoke1 {bIsDynamic=true, bIsEditable=true, bIsStaticSwitch=false, bIsVisible=true, type {classStructOrEnum {refPath=&Vector4f}}}
      Particles.DynParSmoke2 {bIsDynamic=true, bIsEditable=true, bIsStaticSwitch=false, bIsVisible=true, type {classStructOrEnum {refPath=&Vector4f}}}
     ]
     moduleScript {refPath=&self:DefaultPropSpawn_Merged_0.NiagaraScriptSource_1.NiagaraGraph_0.NiagaraNodeAssignment_2.SetVariables_#4}
    }
    {
     bIsSetParametersModule=false
     enabled=true
     moduleName=ScaleColor001
     inputs [
      Scale Mode {bIsDynamic=false, bIsEditable=true, bIsStaticSwitch=true, bIsVisible=true, type {classStructOrEnum {refPath=&NiagaraInt32}}}
      Linear Color Curve {
       bIsDynamic=false
       bIsEditable=false
       bIsStaticSwitch=false
       bIsVisible=false
       type {classStructOrEnum {refPath=&NiagaraDataInterfaceColorCurve}}
      }
      ScaleRGB {bIsDynamic=false, bIsEditable=true, bIsStaticSwitch=false, bIsVisible=true, type {classStructOrEnum {refPath=&NiagaraBool}}}
      Scale RGB {bIsDynamic=false, bIsEditable=true, bIsStaticSwitch=false, bIsVisible=true, type {classStructOrEnum {refPath=&Vector3f}}}
      ScaleRGBA {bIsDynamic=false, bIsEditable=false, bIsStaticSwitch=false, bIsVisible=false, type {classStructOrEnum {refPath=&NiagaraBool}}}
      Scale RGBA {
       bIsDynamic=false
       bIsEditable=false
       bIsStaticSwitch=false
       bIsVisible=false
       type {classStructOrEnum {refPath=&Vector4f}}
      }
      Curve Index {
       bIsDynamic=false
       bIsEditable=false
       bIsStaticSwitch=false
       bIsVisible=false
       type {classStructOrEnum {refPath=&NiagaraFloat}}
      }
      ScaleA {bIsDynamic=false, bIsEditable=true, bIsStaticSwitch=false, bIsVisible=true, type {classStructOrEnum {refPath=&NiagaraBool}}}
      Scale Alpha {bIsDynamic=true, bIsEditable=true, bIsStaticSwitch=false, bIsVisible=true, type {classStructOrEnum {refPath=&NiagaraFloat}}}
      Color Value To Scale {
       bIsDynamic=false
       bIsEditable=true
       bIsStaticSwitch=false
       bIsVisible=true
       type {classStructOrEnum {refPath=&LinearColor}}
      }
      Color Scale {
       bIsDynamic=false
       bIsEditable=false
       bIsStaticSwitch=false
       bIsVisible=false
       type {classStructOrEnum {refPath=&LinearColor}}
      }
     ]
     moduleScript {refPath=&ScaleColor}
    }
   ]
  }
  rendererClasses [{refPath=&NiagaraSpriteRendererProperties}]
  renderers [
   {rendererIndex=0, rendererClass {refPath=&NiagaraSpriteRendererProperties}}
   {rendererIndex=1, rendererClass {refPath=&NiagaraSpriteRendererProperties}}
   {rendererIndex=2, rendererClass {refPath=&NiagaraSpriteRendererProperties}}
   {rendererIndex=3, rendererClass {refPath=&NiagaraSpriteRendererProperties}}
  ]
 }
 DefaultPropSpawn_Merged001 {
  bEnabled=false
  emitterSpawnScript {scriptName=EmitterSpawnScript, modules []}
  emitterUpdateScript {
   scriptName=EmitterUpdateScript
   modules [
    {
     bIsSetParametersModule=false
     enabled=true
     moduleName=EmitterState
     inputs [
      Life Cycle Mode {
       bIsDynamic=false
       bIsEditable=true
       bIsStaticSwitch=true
       bIsVisible=true
       type {classStructOrEnum {refPath=&ENiagaraEmitterLifeCycleMode}}
      }
      Inactive Response {
       bIsDynamic=false
       bIsEditable=false
       bIsStaticSwitch=true
       bIsVisible=false
       type {classStructOrEnum {refPath=&ENiagaraInactiveMode}}
      }
      Loop Behavior {
       bIsDynamic=false
       bIsEditable=false
       bIsStaticSwitch=true
       bIsVisible=false
       type {classStructOrEnum {refPath=&ENiagara_EmitterStateOptions}}
      }
      Loop Duration Mode {
       bIsDynamic=false
       bIsEditable=false
       bIsStaticSwitch=true
       bIsVisible=false
       type {classStructOrEnum {refPath=&ENiagara_InfiniteLoopDuration}}
      }
      Loop Count {
       bIsDynamic=false
       bIsEditable=false
       bIsStaticSwitch=false
       bIsVisible=false
       type {classStructOrEnum {refPath=&NiagaraInt32}}
      }
      Loop Duration {
       bIsDynamic=false
       bIsEditable=false
       bIsStaticSwitch=false
       bIsVisible=false
       type {classStructOrEnum {refPath=&NiagaraFloat}}
      }
      Recalculate Duration Each Loop {bIsDynamic=false, bIsEditable=false, bIsStaticSwitch=false, bIsVisible=false, type {classStructOrEnum {refPath=&NiagaraBool}}}
      UseLoopDelay {bIsDynamic=false, bIsEditable=false, bIsStaticSwitch=false, bIsVisible=false, type {classStructOrEnum {refPath=&NiagaraBool}}}
      Loop Delay {
       bIsDynamic=false
       bIsEditable=false
       bIsStaticSwitch=false
       bIsVisible=false
       type {classStructOrEnum {refPath=&NiagaraFloat}}
      }
      Delay First Loop Only {bIsDynamic=false, bIsEditable=false, bIsStaticSwitch=false, bIsVisible=false, type {classStructOrEnum {refPath=&NiagaraBool}}}
      Scalability Mode {
       bIsDynamic=false
       bIsEditable=true
       bIsStaticSwitch=true
       bIsVisible=true
       type {classStructOrEnum {refPath=&ENiagaraEmitterScalabilityMode_Limited}}
      }
      Enable Distance Culling {bIsDynamic=false, bIsEditable=false, bIsStaticSwitch=true, bIsVisible=false, type {classStructOrEnum {refPath=&NiagaraBool}}}
      MinDistance {
       bIsDynamic=false
       bIsEditable=false
       bIsStaticSwitch=false
       bIsVisible=false
       type {classStructOrEnum {refPath=&NiagaraFloat}}
      }
      Min Distance Response {
       bIsDynamic=false
       bIsEditable=false
       bIsStaticSwitch=false
       bIsVisible=false
       type {classStructOrEnum {refPath=&ENiagaraExecutionStateManagement}}
      }
      MaxDistance {
       bIsDynamic=false
       bIsEditable=false
       bIsStaticSwitch=false
       bIsVisible=false
       type {classStructOrEnum {refPath=&NiagaraFloat}}
      }
      Max Distance Response {
       bIsDynamic=false
       bIsEditable=false
       bIsStaticSwitch=false
       bIsVisible=false
       type {classStructOrEnum {refPath=&ENiagaraExecutionStateManagement}}
      }
      Scale Spawn Count {bIsDynamic=false, bIsEditable=false, bIsStaticSwitch=true, bIsVisible=false, type {classStructOrEnum {refPath=&NiagaraBool}}}
      Spawn Count Scale {
       bIsDynamic=false
       bIsEditable=false
       bIsStaticSwitch=false
       bIsVisible=false
       type {classStructOrEnum {refPath=&NiagaraFloat}}
      }
      Spawn Count Scale By Distance {
       bIsDynamic=false
       bIsEditable=false
       bIsStaticSwitch=false
       bIsVisible=false
       type {classStructOrEnum {refPath=&NiagaraDataInterfaceCurve}}
      }
      Enable Visibility Culling {bIsDynamic=false, bIsEditable=false, bIsStaticSwitch=true, bIsVisible=false, type {classStructOrEnum {refPath=&NiagaraBool}}}
      Visibility Cull Response {
       bIsDynamic=false
       bIsEditable=false
       bIsStaticSwitch=false
       bIsVisible=false
       type {classStructOrEnum {refPath=&ENiagaraExecutionStateManagement}}
      }
      Visibility Cull Delay {
       bIsDynamic=false
       bIsEditable=false
       bIsStaticSwitch=false
       bIsVisible=false
       type {classStructOrEnum {refPath=&NiagaraFloat}}
      }
      Reset Age On Awaken {bIsDynamic=false, bIsEditable=false, bIsStaticSwitch=false, bIsVisible=false, type {classStructOrEnum {refPath=&NiagaraBool}}}
     ]
     moduleScript {refPath=&EmitterState}
    }
    {
     bIsSetParametersModule=false
     enabled=true
     moduleName=SpawnBurst_Instantaneous001
     inputs [
      Spawn Count {bIsDynamic=true, bIsEditable=true, bIsStaticSwitch=false, bIsVisible=true, type {classStructOrEnum {refPath=&NiagaraInt32}}}
      Spawn Time {bIsDynamic=false, bIsEditable=true, bIsStaticSwitch=false, bIsVisible=true, type {classStructOrEnum {refPath=&NiagaraFloat}}}
      Use Spawn Probability {bIsDynamic=false, bIsEditable=true, bIsStaticSwitch=false, bIsVisible=true, type {classStructOrEnum {refPath=&NiagaraBool}}}
      Spawn Probability {bIsDynamic=false, bIsEditable=false, bIsStaticSwitch=false, bIsVisible=true, type {classStructOrEnum {refPath=&NiagaraFloat}}}
      UseLoopCountLimit {bIsDynamic=false, bIsEditable=true, bIsStaticSwitch=false, bIsVisible=true, type {classStructOrEnum {refPath=&NiagaraBool}}}
      Loop Count Limit {bIsDynamic=false, bIsEditable=false, bIsStaticSwitch=false, bIsVisible=true, type {classStructOrEnum {refPath=&NiagaraInt32}}}
      Spawn Group {bIsDynamic=false, bIsEditable=true, bIsStaticSwitch=false, bIsVisible=true, type {classStructOrEnum {refPath=&NiagaraInt32}}}
      Age {bIsDynamic=false, bIsEditable=true, bIsStaticSwitch=false, bIsVisible=true, type {classStructOrEnum {refPath=&NiagaraFloat}}}
     ]
     moduleScript {refPath=&SpawnBurst_Instantaneous}
    }
    {
     bIsSetParametersModule=false
     enabled=true
     moduleName=SpawnBurst_Instantaneous
     inputs [
      Spawn Count {bIsDynamic=false, bIsEditable=true, bIsStaticSwitch=false, bIsVisible=true, type {classStructOrEnum {refPath=&NiagaraInt32}}}
      Spawn Time {bIsDynamic=false, bIsEditable=true, bIsStaticSwitch=false, bIsVisible=true, type {classStructOrEnum {refPath=&NiagaraFloat}}}
      Use Spawn Probability {bIsDynamic=false, bIsEditable=true, bIsStaticSwitch=false, bIsVisible=true, type {classStructOrEnum {refPath=&NiagaraBool}}}
      Spawn Probability {bIsDynamic=false, bIsEditable=false, bIsStaticSwitch=false, bIsVisible=true, type {classStructOrEnum {refPath=&NiagaraFloat}}}
      UseLoopCountLimit {bIsDynamic=false, bIsEditable=true, bIsStaticSwitch=false, bIsVisible=true, type {classStructOrEnum {refPath=&NiagaraBool}}}
      Loop Count Limit {bIsDynamic=false, bIsEditable=false, bIsStaticSwitch=false, bIsVisible=true, type {classStructOrEnum {refPath=&NiagaraInt32}}}
      Spawn Group {bIsDynamic=false, bIsEditable=true, bIsStaticSwitch=false, bIsVisible=true, type {classStructOrEnum {refPath=&NiagaraInt32}}}
      Age {bIsDynamic=false, bIsEditable=true, bIsStaticSwitch=false, bIsVisible=true, type {classStructOrEnum {refPath=&NiagaraFloat}}}
     ]
     moduleScript {refPath=&SpawnBurst_Instantaneous}
    }
    {
     bIsSetParametersModule=false
     enabled=true
     moduleName=SpawnBurst_Instantaneous002
     inputs [
      Spawn Count {bIsDynamic=true, bIsEditable=true, bIsStaticSwitch=false, bIsVisible=true, type {classStructOrEnum {refPath=&NiagaraInt32}}}
      Spawn Time {bIsDynamic=false, bIsEditable=true, bIsStaticSwitch=false, bIsVisible=true, type {classStructOrEnum {refPath=&NiagaraFloat}}}
      Use Spawn Probability {bIsDynamic=false, bIsEditable=true, bIsStaticSwitch=false, bIsVisible=true, type {classStructOrEnum {refPath=&NiagaraBool}}}
      Spawn Probability {bIsDynamic=false, bIsEditable=false, bIsStaticSwitch=false, bIsVisible=true, type {classStructOrEnum {refPath=&NiagaraFloat}}}
      UseLoopCountLimit {bIsDynamic=false, bIsEditable=true, bIsStaticSwitch=false, bIsVisible=true, type {classStructOrEnum {refPath=&NiagaraBool}}}
      Loop Count Limit {bIsDynamic=false, bIsEditable=false, bIsStaticSwitch=false, bIsVisible=true, type {classStructOrEnum {refPath=&NiagaraInt32}}}
      Spawn Group {bIsDynamic=false, bIsEditable=true, bIsStaticSwitch=false, bIsVisible=true, type {classStructOrEnum {refPath=&NiagaraInt32}}}
      Age {bIsDynamic=false, bIsEditable=true, bIsStaticSwitch=false, bIsVisible=true, type {classStructOrEnum {refPath=&NiagaraFloat}}}
     ]
     moduleScript {refPath=&SpawnBurst_Instantaneous}
    }
    {
     bIsSetParametersModule=false
     enabled=true
     moduleName=SpawnBurst_Instantaneous003
     inputs [
      Spawn Count {bIsDynamic=false, bIsEditable=true, bIsStaticSwitch=false, bIsVisible=true, type {classStructOrEnum {refPath=&NiagaraInt32}}}
      Spawn Time {bIsDynamic=false, bIsEditable=true, bIsStaticSwitch=false, bIsVisible=true, type {classStructOrEnum {refPath=&NiagaraFloat}}}
      Use Spawn Probability {bIsDynamic=false, bIsEditable=true, bIsStaticSwitch=false, bIsVisible=true, type {classStructOrEnum {refPath=&NiagaraBool}}}
      Spawn Probability {bIsDynamic=false, bIsEditable=false, bIsStaticSwitch=false, bIsVisible=true, type {classStructOrEnum {refPath=&NiagaraFloat}}}
      UseLoopCountLimit {bIsDynamic=false, bIsEditable=true, bIsStaticSwitch=false, bIsVisible=true, type {classStructOrEnum {refPath=&NiagaraBool}}}
      Loop Count Limit {bIsDynamic=false, bIsEditable=false, bIsStaticSwitch=false, bIsVisible=true, type {classStructOrEnum {refPath=&NiagaraInt32}}}
      Spawn Group {bIsDynamic=false, bIsEditable=true, bIsStaticSwitch=false, bIsVisible=true, type {classStructOrEnum {refPath=&NiagaraInt32}}}
      Age {bIsDynamic=false, bIsEditable=true, bIsStaticSwitch=false, bIsVisible=true, type {classStructOrEnum {refPath=&NiagaraFloat}}}
     ]
     moduleScript {refPath=&SpawnBurst_Instantaneous}
    }
    {
     bIsSetParametersModule=false
     enabled=true
     moduleName=SpawnBurst_Instantaneous004
     inputs [
      Spawn Count {bIsDynamic=false, bIsEditable=true, bIsStaticSwitch=false, bIsVisible=true, type {classStructOrEnum {refPath=&NiagaraInt32}}}
      Spawn Time {bIsDynamic=false, bIsEditable=true, bIsStaticSwitch=false, bIsVisible=true, type {classStructOrEnum {refPath=&NiagaraFloat}}}
      Use Spawn Probability {bIsDynamic=false, bIsEditable=true, bIsStaticSwitch=false, bIsVisible=true, type {classStructOrEnum {refPath=&NiagaraBool}}}
      Spawn Probability {bIsDynamic=false, bIsEditable=false, bIsStaticSwitch=false, bIsVisible=true, type {classStructOrEnum {refPath=&NiagaraFloat}}}
      UseLoopCountLimit {bIsDynamic=false, bIsEditable=true, bIsStaticSwitch=false, bIsVisible=true, type {classStructOrEnum {refPath=&NiagaraBool}}}
      Loop Count Limit {bIsDynamic=false, bIsEditable=false, bIsStaticSwitch=false, bIsVisible=true, type {classStructOrEnum {refPath=&NiagaraInt32}}}
      Spawn Group {bIsDynamic=false, bIsEditable=true, bIsStaticSwitch=false, bIsVisible=true, type {classStructOrEnum {refPath=&NiagaraInt32}}}
      Age {bIsDynamic=false, bIsEditable=true, bIsStaticSwitch=false, bIsVisible=true, type {classStructOrEnum {refPath=&NiagaraFloat}}}
     ]
     moduleScript {refPath=&SpawnBurst_Instantaneous}
    }
    {
     bIsSetParametersModule=false
     enabled=true
     moduleName=SpawnBurst_Instantaneous005
     inputs [
      Spawn Count {bIsDynamic=false, bIsEditable=true, bIsStaticSwitch=false, bIsVisible=true, type {classStructOrEnum {refPath=&NiagaraInt32}}}
      Spawn Time {bIsDynamic=false, bIsEditable=true, bIsStaticSwitch=false, bIsVisible=true, type {classStructOrEnum {refPath=&NiagaraFloat}}}
      Use Spawn Probability {bIsDynamic=false, bIsEditable=true, bIsStaticSwitch=false, bIsVisible=true, type {classStructOrEnum {refPath=&NiagaraBool}}}
      Spawn Probability {bIsDynamic=false, bIsEditable=false, bIsStaticSwitch=false, bIsVisible=true, type {classStructOrEnum {refPath=&NiagaraFloat}}}
      UseLoopCountLimit {bIsDynamic=false, bIsEditable=true, bIsStaticSwitch=false, bIsVisible=true, type {classStructOrEnum {refPath=&NiagaraBool}}}
      Loop Count Limit {bIsDynamic=false, bIsEditable=false, bIsStaticSwitch=false, bIsVisible=true, type {classStructOrEnum {refPath=&NiagaraInt32}}}
      Spawn Group {bIsDynamic=false, bIsEditable=true, bIsStaticSwitch=false, bIsVisible=true, type {classStructOrEnum {refPath=&NiagaraInt32}}}
      Age {bIsDynamic=false, bIsEditable=true, bIsStaticSwitch=false, bIsVisible=true, type {classStructOrEnum {refPath=&NiagaraFloat}}}
     ]
     moduleScript {refPath=&SpawnBurst_Instantaneous}
    }
   ]
  }
  particleSpawnScript {
   scriptName=ParticleSpawnScript
   modules [
    {
     bIsSetParametersModule=true
     enabled=true
     moduleName=SetVariables_#1
     inputs [Particles.VisibilityTag {bIsDynamic=true, bIsEditable=true, bIsStaticSwitch=false, bIsVisible=true, type {classStructOrEnum {refPath=&NiagaraInt32}}}]
     moduleScript {refPath=&self:DefaultPropSpawn_Merged001_0.NiagaraScriptSource_1.NiagaraGraph_0.NiagaraNodeAssignment_1.SetVariables_#1}
    }
    {
     bIsSetParametersModule=false
     enabled=true
     moduleName=InitializeParticle001
     inputs [
      Lifetime Mode {
       bIsDynamic=false
       bIsEditable=true
       bIsStaticSwitch=true
       bIsVisible=true
       type {classStructOrEnum {refPath=&ENiagara_LifetimeMode}}
      }
      Lifetime Min {
       bIsDynamic=false
       bIsEditable=false
       bIsStaticSwitch=false
       bIsVisible=false
       type {classStructOrEnum {refPath=&NiagaraFloat}}
      }
      Lifetime {bIsDynamic=true, bIsEditable=true, bIsStaticSwitch=false, bIsVisible=true, type {classStructOrEnum {refPath=&NiagaraFloat}}}
      Lifetime Max {
       bIsDynamic=false
       bIsEditable=false
       bIsStaticSwitch=false
       bIsVisible=false
       type {classStructOrEnum {refPath=&NiagaraFloat}}
      }
      Lifetime Randomness Mode {
       bIsDynamic=false
       bIsEditable=false
       bIsStaticSwitch=true
       bIsVisible=false
       type {classStructOrEnum {refPath=&ENiagaraRandomnessMode}}
      }
      Color Mode {
       bIsDynamic=false
       bIsEditable=true
       bIsStaticSwitch=true
       bIsVisible=true
       type {classStructOrEnum {refPath=&ENiagara_ColorInitializationMode}}
      }
      Color Minimum {
       bIsDynamic=false
       bIsEditable=false
       bIsStaticSwitch=false
       bIsVisible=false
       type {classStructOrEnum {refPath=&LinearColor}}
      }
      Color {
       bIsDynamic=true
       bIsEditable=true
       bIsStaticSwitch=false
       bIsVisible=true
       type {classStructOrEnum {refPath=&LinearColor}}
      }
      AdjustHue {bIsDynamic=false, bIsEditable=false, bIsStaticSwitch=false, bIsVisible=false, type {classStructOrEnum {refPath=&NiagaraBool}}}
      Hue Shift Range {
       bIsDynamic=false
       bIsEditable=false
       bIsStaticSwitch=false
       bIsVisible=false
       type {classStructOrEnum {refPath=&Vector2f}}
      }
      Color Maximum {
       bIsDynamic=false
       bIsEditable=false
       bIsStaticSwitch=false
       bIsVisible=false
       type {classStructOrEnum {refPath=&LinearColor}}
      }
      AdjustSaturation {bIsDynamic=false, bIsEditable=false, bIsStaticSwitch=false, bIsVisible=false, type {classStructOrEnum {refPath=&NiagaraBool}}}
      Saturation Range {
       bIsDynamic=false
       bIsEditable=false
       bIsStaticSwitch=false
       bIsVisible=false
       type {classStructOrEnum {refPath=&Vector2f}}
      }
      AdjustValue {bIsDynamic=false, bIsEditable=false, bIsStaticSwitch=false, bIsVisible=false, type {classStructOrEnum {refPath=&NiagaraBool}}}
      Value Range {
       bIsDynamic=false
       bIsEditable=false
       bIsStaticSwitch=false
       bIsVisible=false
       type {classStructOrEnum {refPath=&Vector2f}}
      }
      AdjustAlpha {bIsDynamic=false, bIsEditable=false, bIsStaticSwitch=false, bIsVisible=false, type {classStructOrEnum {refPath=&NiagaraBool}}}
      Alpha Scale Range {
       bIsDynamic=false
       bIsEditable=false
       bIsStaticSwitch=false
       bIsVisible=false
       type {classStructOrEnum {refPath=&Vector2f}}
      }
      Color Channel Mode {
       bIsDynamic=false
       bIsEditable=false
       bIsStaticSwitch=true
       bIsVisible=false
       type {classStructOrEnum {refPath=&ENiagaraChannelCorrelation}}
      }
      Color Randomness Mode {
       bIsDynamic=false
       bIsEditable=false
       bIsStaticSwitch=true
       bIsVisible=false
       type {classStructOrEnum {refPath=&ENiagaraRandomnessMode}}
      }
      Position Mode {
       bIsDynamic=false
       bIsEditable=true
       bIsStaticSwitch=true
       bIsVisible=true
       type {classStructOrEnum {refPath=&ENiagara_PositionInitializationMode}}
      }
      Position {
       bIsDynamic=false
       bIsEditable=false
       bIsStaticSwitch=false
       bIsVisible=false
       type {classStructOrEnum {refPath=&NiagaraPosition}}
      }
      UsePositionOffset {bIsDynamic=false, bIsEditable=true, bIsStaticSwitch=false, bIsVisible=true, type {classStructOrEnum {refPath=&NiagaraBool}}}
      Position Offset {bIsDynamic=false, bIsEditable=true, bIsStaticSwitch=false, bIsVisible=true, type {classStructOrEnum {refPath=&Vector3f}}}
      Position Offset Coordinate Space {
       bIsDynamic=false
       bIsEditable=true
       bIsStaticSwitch=false
       bIsVisible=true
       type {classStructOrEnum {refPath=&ENiagaraCoordinateSpace}}
      }
      Mass Mode {
       bIsDynamic=false
       bIsEditable=true
       bIsStaticSwitch=true
       bIsVisible=true
       type {classStructOrEnum {refPath=&ENiagara_MassInitializationMode}}
      }
      Mass Min {
       bIsDynamic=false
       bIsEditable=false
       bIsStaticSwitch=false
       bIsVisible=false
       type {classStructOrEnum {refPath=&NiagaraFloat}}
      }
      Mass {
       bIsDynamic=false
       bIsEditable=false
       bIsStaticSwitch=false
       bIsVisible=false
       type {classStructOrEnum {refPath=&NiagaraFloat}}
      }
      Mass Max {
       bIsDynamic=false
       bIsEditable=false
       bIsStaticSwitch=false
       bIsVisible=false
       type {classStructOrEnum {refPath=&NiagaraFloat}}
      }
      Mass Randomness Mode {
       bIsDynamic=false
       bIsEditable=false
       bIsStaticSwitch=true
       bIsVisible=false
       type {classStructOrEnum {refPath=&ENiagaraRandomnessMode}}
      }
      Write Material Random {bIsDynamic=false, bIsEditable=true, bIsStaticSwitch=false, bIsVisible=true, type {classStructOrEnum {refPath=&NiagaraBool}}}
      Material Random {bIsDynamic=false, bIsEditable=true, bIsStaticSwitch=false, bIsVisible=true, type {classStructOrEnum {refPath=&NiagaraFloat}}}
      Sprite Size Mode {
       bIsDynamic=false
       bIsEditable=true
       bIsStaticSwitch=true
       bIsVisible=true
       type {classStructOrEnum {refPath=&ENiagara_SizeScaleMode}}
      }
      Sprite Size {
       bIsDynamic=false
       bIsEditable=false
       bIsStaticSwitch=false
       bIsVisible=false
       type {classStructOrEnum {refPath=&Vector2f}}
      }
      Uniform Sprite Size {bIsDynamic=true, bIsEditable=true, bIsStaticSwitch=false, bIsVisible=true, type {classStructOrEnum {refPath=&NiagaraFloat}}}
      Uniform Sprite Size Min {
       bIsDynamic=false
       bIsEditable=false
       bIsStaticSwitch=false
       bIsVisible=false
       type {classStructOrEnum {refPath=&NiagaraFloat}}
      }
      Sprite Size Min {
       bIsDynamic=false
       bIsEditable=false
       bIsStaticSwitch=false
       bIsVisible=false
       type {classStructOrEnum {refPath=&Vector2f}}
      }
      Sprite Size Max {
       bIsDynamic=false
       bIsEditable=false
       bIsStaticSwitch=false
       bIsVisible=false
       type {classStructOrEnum {refPath=&Vector2f}}
      }
      Uniform Sprite Size Max {
       bIsDynamic=false
       bIsEditable=false
       bIsStaticSwitch=false
       bIsVisible=false
       type {classStructOrEnum {refPath=&NiagaraFloat}}
      }
      Sprite Size Randomness Mode {
       bIsDynamic=false
       bIsEditable=false
       bIsStaticSwitch=true
       bIsVisible=false
       type {classStructOrEnum {refPath=&ENiagaraRandomnessMode}}
      }
      Sprite Rotation Mode {
       bIsDynamic=false
       bIsEditable=true
       bIsStaticSwitch=true
       bIsVisible=true
       type {classStructOrEnum {refPath=&ENiagara_SpriteRotationMode}}
      }
      Sprite Rotation Angle Min {
       bIsDynamic=false
       bIsEditable=false
       bIsStaticSwitch=false
       bIsVisible=false
       type {classStructOrEnum {refPath=&NiagaraFloat}}
      }
      Sprite Rotation Angle {bIsDynamic=true, bIsEditable=true, bIsStaticSwitch=false, bIsVisible=true, type {classStructOrEnum {refPath=&NiagaraFloat}}}
      Sprite Rotation Angle Max {
       bIsDynamic=false
       bIsEditable=false
       bIsStaticSwitch=false
       bIsVisible=false
       type {classStructOrEnum {refPath=&NiagaraFloat}}
      }
      Sprite Rotation Randomness Mode {
       bIsDynamic=false
       bIsEditable=false
       bIsStaticSwitch=true
       bIsVisible=false
       type {classStructOrEnum {refPath=&ENiagaraRandomnessMode}}
      }
      Sprite UV Mode {
       bIsDynamic=false
       bIsEditable=true
       bIsStaticSwitch=true
       bIsVisible=true
       type {classStructOrEnum {refPath=&ENiagara_UVFlippingMode}}
      }
      Sprite UV Randomness Mode {
       bIsDynamic=false
       bIsEditable=false
       bIsStaticSwitch=true
       bIsVisible=false
       type {classStructOrEnum {refPath=&ENiagaraRandomnessMode}}
      }
      Sprite UV Scale {
       bIsDynamic=false
       bIsEditable=false
       bIsStaticSwitch=false
       bIsVisible=false
       type {classStructOrEnum {refPath=&Vector2f}}
      }
      Mesh Scale Mode {
       bIsDynamic=false
       bIsEditable=true
       bIsStaticSwitch=true
       bIsVisible=true
       type {classStructOrEnum {refPath=&ENiagara_SizeScaleMode}}
      }
      Mesh Uniform Scale Min {
       bIsDynamic=false
       bIsEditable=false
       bIsStaticSwitch=false
       bIsVisible=false
       type {classStructOrEnum {refPath=&NiagaraFloat}}
      }
      Mesh Uniform Scale {bIsDynamic=true, bIsEditable=true, bIsStaticSwitch=false, bIsVisible=true, type {classStructOrEnum {refPath=&NiagaraFloat}}}
      Mesh Scale Min {
       bIsDynamic=false
       bIsEditable=false
       bIsStaticSwitch=false
       bIsVisible=false
       type {classStructOrEnum {refPath=&Vector3f}}
      }
      Mesh Scale {
       bIsDynamic=false
       bIsEditable=false
       bIsStaticSwitch=false
       bIsVisible=false
       type {classStructOrEnum {refPath=&Vector3f}}
      }
      Mesh Scale Max {
       bIsDynamic=false
       bIsEditable=false
       bIsStaticSwitch=false
       bIsVisible=false
       type {classStructOrEnum {refPath=&Vector3f}}
      }
      Mesh Uniform Scale Max {
       bIsDynamic=false
       bIsEditable=false
       bIsStaticSwitch=false
       bIsVisible=false
       type {classStructOrEnum {refPath=&NiagaraFloat}}
      }
      Mesh Scale Randomness Mode {
       bIsDynamic=false
       bIsEditable=false
       bIsStaticSwitch=true
       bIsVisible=false
       type {classStructOrEnum {refPath=&ENiagaraRandomnessMode}}
      }
      Mesh Renderer Array Visibility Mode {
       bIsDynamic=false
       bIsEditable=true
       bIsStaticSwitch=true
       bIsVisible=true
       type {classStructOrEnum {refPath=&ENiagara_UnsetDirectSetRandom}}
      }
      Mesh Renderer Mesh Index {
       bIsDynamic=false
       bIsEditable=false
       bIsStaticSwitch=false
       bIsVisible=false
       type {classStructOrEnum {refPath=&NiagaraInt32}}
      }
      Wrap Index By Mesh Count {bIsDynamic=false, bIsEditable=false, bIsStaticSwitch=true, bIsVisible=false, type {classStructOrEnum {refPath=&NiagaraBool}}}
      Mesh Renderer Info {
       bIsDynamic=false
       bIsEditable=false
       bIsStaticSwitch=false
       bIsVisible=false
       type {classStructOrEnum {refPath=&NiagaraDataInterfaceMeshRendererInfo}}
      }
      Mesh Renderer Random Seed {
       bIsDynamic=false
       bIsEditable=false
       bIsStaticSwitch=false
       bIsVisible=false
       type {classStructOrEnum {refPath=&NiagaraInt32}}
      }
      Ribbon Width Mode {
       bIsDynamic=false
       bIsEditable=true
       bIsStaticSwitch=true
       bIsVisible=true
       type {classStructOrEnum {refPath=&ENiagara_UnsetDirectSet}}
      }
      Ribbon Width {
       bIsDynamic=false
       bIsEditable=false
       bIsStaticSwitch=false
       bIsVisible=false
       type {classStructOrEnum {refPath=&NiagaraFloat}}
      }
      Ribbon Facing Mode {
       bIsDynamic=false
       bIsEditable=true
       bIsStaticSwitch=true
       bIsVisible=true
       type {classStructOrEnum {refPath=&ENiagara_UnsetDirectSet}}
      }
      Ribbon Facing Vector {
       bIsDynamic=false
       bIsEditable=false
       bIsStaticSwitch=false
       bIsVisible=false
       type {classStructOrEnum {refPath=&Vector3f}}
      }
      Ribbon Twist Mode {
       bIsDynamic=false
       bIsEditable=true
       bIsStaticSwitch=true
       bIsVisible=true
       type {classStructOrEnum {refPath=&ENiagara_UnsetDirectSet}}
      }
      "Ribbon Twist " {
       bIsDynamic=false
       bIsEditable=false
       bIsStaticSwitch=false
       bIsVisible=false
       type {classStructOrEnum {refPath=&NiagaraFloat}}
      }
     ]
     moduleScript {refPath=&InitializeParticle}
    }
    {
     bIsSetParametersModule=true
     enabled=true
     moduleName=SetVariables_#2
     inputs [Particles.InitSpriteRotRate {bIsDynamic=true, bIsEditable=true, bIsStaticSwitch=false, bIsVisible=true, type {classStructOrEnum {refPath=&NiagaraFloat}}}]
     moduleScript {refPath=&self:DefaultPropSpawn_Merged001_0.NiagaraScriptSource_1.NiagaraGraph_0.NiagaraNodeAssignment_0.SetVariables_#2}
    }
    {
     bIsSetParametersModule=false
     enabled=true
     moduleName=AccelerationForce
     inputs [
      Acceleration {bIsDynamic=true, bIsEditable=true, bIsStaticSwitch=false, bIsVisible=true, type {classStructOrEnum {refPath=&Vector3f}}}
      Coordinate Space {
       bIsDynamic=false
       bIsEditable=true
       bIsStaticSwitch=false
       bIsVisible=true
       type {classStructOrEnum {refPath=&ENiagaraCoordinateSpace}}
      }
     ]
     moduleScript {refPath=&AccelerationForce}
    }
    {
     bIsSetParametersModule=false
     enabled=true
     moduleName=ShapeLocation
     inputs [
      Shape Primitive {
       bIsDynamic=false
       bIsEditable=true
       bIsStaticSwitch=true
       bIsVisible=true
       type {classStructOrEnum {refPath=&ENiagara_LocationShapes}}
      }
      Box / Plane Mode {
       bIsDynamic=false
       bIsEditable=false
       bIsStaticSwitch=true
       bIsVisible=false
       type {classStructOrEnum {refPath=&ENiagara_BoxPlaneMode}}
      }
      Cone Mode {
       bIsDynamic=false
       bIsEditable=false
       bIsStaticSwitch=true
       bIsVisible=false
       type {classStructOrEnum {refPath=&ENiagara_ConeMode}}
      }
      Cylinder Height {
       bIsDynamic=false
       bIsEditable=false
       bIsStaticSwitch=false
       bIsVisible=false
       type {classStructOrEnum {refPath=&NiagaraFloat}}
      }
      Sphere Radius {bIsDynamic=true, bIsEditable=true, bIsStaticSwitch=false, bIsVisible=true, type {classStructOrEnum {refPath=&NiagaraFloat}}}
      Ring / Disc Mode {
       bIsDynamic=false
       bIsEditable=false
       bIsStaticSwitch=true
       bIsVisible=false
       type {classStructOrEnum {refPath=&ENiagaraRingDiscMode}}
      }
      Torus Mode {
       bIsDynamic=false
       bIsEditable=false
       bIsStaticSwitch=true
       bIsVisible=false
       type {classStructOrEnum {refPath=&ENiagaraShapeTorusMode}}
      }
      Knot Ring Loops {
       bIsDynamic=false
       bIsEditable=false
       bIsStaticSwitch=false
       bIsVisible=false
       type {classStructOrEnum {refPath=&NiagaraInt32}}
      }
      Knot Helix Loops {
       bIsDynamic=false
       bIsEditable=false
       bIsStaticSwitch=false
       bIsVisible=false
       type {classStructOrEnum {refPath=&NiagaraInt32}}
      }
      Enable Torus Knot Radius {bIsDynamic=false, bIsEditable=false, bIsStaticSwitch=false, bIsVisible=false, type {classStructOrEnum {refPath=&NiagaraBool}}}
      Torus Knot Tube Radius {
       bIsDynamic=false
       bIsEditable=false
       bIsStaticSwitch=false
       bIsVisible=false
       type {classStructOrEnum {refPath=&NiagaraFloat}}
      }
      Cylinder Radius {
       bIsDynamic=false
       bIsEditable=false
       bIsStaticSwitch=false
       bIsVisible=false
       type {classStructOrEnum {refPath=&NiagaraFloat}}
      }
      Cylinder Height Midpoint {
       bIsDynamic=false
       bIsEditable=false
       bIsStaticSwitch=false
       bIsVisible=false
       type {classStructOrEnum {refPath=&NiagaraFloat}}
      }
      Ring Radius {
       bIsDynamic=false
       bIsEditable=false
       bIsStaticSwitch=false
       bIsVisible=false
       type {classStructOrEnum {refPath=&NiagaraFloat}}
      }
      Box Size {
       bIsDynamic=false
       bIsEditable=false
       bIsStaticSwitch=false
       bIsVisible=false
       type {classStructOrEnum {refPath=&Vector3f}}
      }
      Box Midpoint {
       bIsDynamic=false
       bIsEditable=false
       bIsStaticSwitch=false
       bIsVisible=false
       type {classStructOrEnum {refPath=&Vector3f}}
      }
      Cone Length {
       bIsDynamic=false
       bIsEditable=false
       bIsStaticSwitch=false
       bIsVisible=false
       type {classStructOrEnum {refPath=&NiagaraFloat}}
      }
      Plane Size {
       bIsDynamic=false
       bIsEditable=false
       bIsStaticSwitch=false
       bIsVisible=false
       type {classStructOrEnum {refPath=&Vector2f}}
      }
      Wedge Horizontal Angle {
       bIsDynamic=false
       bIsEditable=false
       bIsStaticSwitch=false
       bIsVisible=false
       type {classStructOrEnum {refPath=&NiagaraFloat}}
      }
      Large Radius {
       bIsDynamic=false
       bIsEditable=false
       bIsStaticSwitch=false
       bIsVisible=false
       type {classStructOrEnum {refPath=&NiagaraFloat}}
      }
      Plane Midpoint {
       bIsDynamic=false
       bIsEditable=false
       bIsStaticSwitch=false
       bIsVisible=false
       type {classStructOrEnum {refPath=&Vector2f}}
      }
      Wedge Vertical Angle {
       bIsDynamic=false
       bIsEditable=false
       bIsStaticSwitch=false
       bIsVisible=false
       type {classStructOrEnum {refPath=&NiagaraFloat}}
      }
      Cone Angle {
       bIsDynamic=false
       bIsEditable=false
       bIsStaticSwitch=false
       bIsVisible=false
       type {classStructOrEnum {refPath=&NiagaraFloat}}
      }
      Handle Radius {
       bIsDynamic=false
       bIsEditable=false
       bIsStaticSwitch=false
       bIsVisible=false
       type {classStructOrEnum {refPath=&NiagaraFloat}}
      }
      Cone Inner Angle {
       bIsDynamic=false
       bIsEditable=false
       bIsStaticSwitch=false
       bIsVisible=false
       type {classStructOrEnum {refPath=&NiagaraFloat}}
      }
      Cone Radial Angle {
       bIsDynamic=false
       bIsEditable=false
       bIsStaticSwitch=false
       bIsVisible=false
       type {classStructOrEnum {refPath=&NiagaraFloat}}
      }
      Cone Angle Type {
       bIsDynamic=false
       bIsEditable=false
       bIsStaticSwitch=true
       bIsVisible=false
       type {classStructOrEnum {refPath=&ENiagara_AngleInput}}
      }
      Flatten Endcaps {bIsDynamic=false, bIsEditable=false, bIsStaticSwitch=true, bIsVisible=false, type {classStructOrEnum {refPath=&NiagaraBool}}}
      Apply To Particle Position {
       bIsDynamic=false
       bIsEditable=true
       bIsStaticSwitch=true
       bIsVisible=true
       type {classStructOrEnum {refPath=&ENiagara_AttributeSamplingApplyOutput}}
      }
      Cone Surface Distribution {
       bIsDynamic=false
       bIsEditable=false
       bIsStaticSwitch=false
       bIsVisible=false
       type {classStructOrEnum {refPath=&NiagaraFloat}}
      }
      Ring / Disc Distribution Mode {
       bIsDynamic=false
       bIsEditable=false
       bIsStaticSwitch=true
       bIsVisible=false
       type {classStructOrEnum {refPath=&ENiagaraSphereDistributionMode}}
      }
      Sphere Distribution {
       bIsDynamic=false
       bIsEditable=true
       bIsStaticSwitch=true
       bIsVisible=true
       type {classStructOrEnum {refPath=&ENiagaraSphereDistributionMode}}
      }
      Torus Distribution Mode {
       bIsDynamic=false
       bIsEditable=false
       bIsStaticSwitch=true
       bIsVisible=false
       type {classStructOrEnum {refPath=&ENiagaraTorusDistributionMode}}
      }
      Sphere Surface Distribution {bIsDynamic=false, bIsEditable=true, bIsStaticSwitch=false, bIsVisible=true, type {classStructOrEnum {refPath=&NiagaraFloat}}}
      Hemisphere Distribution {bIsDynamic=false, bIsEditable=true, bIsStaticSwitch=false, bIsVisible=true, type {classStructOrEnum {refPath=&Vector2f}}}
      Cylinder Distribution {
       bIsDynamic=false
       bIsEditable=false
       bIsStaticSwitch=true
       bIsVisible=false
       type {classStructOrEnum {refPath=&ENiagara_CylinderMode}}
      }
      Surface Only {bIsDynamic=false, bIsEditable=false, bIsStaticSwitch=false, bIsVisible=false, type {classStructOrEnum {refPath=&NiagaraBool}}}
      Surface Only Band Thickness {
       bIsDynamic=false
       bIsEditable=false
       bIsStaticSwitch=false
       bIsVisible=false
       type {classStructOrEnum {refPath=&NiagaraFloat}}
      }
      Use Endcaps In Surface Only Mode {bIsDynamic=false, bIsEditable=false, bIsStaticSwitch=true, bIsVisible=false, type {classStructOrEnum {refPath=&NiagaraBool}}}
      Hemicircle X {bIsDynamic=false, bIsEditable=false, bIsStaticSwitch=true, bIsVisible=false, type {classStructOrEnum {refPath=&NiagaraBool}}}
      Hemicircle Y {bIsDynamic=false, bIsEditable=false, bIsStaticSwitch=true, bIsVisible=false, type {classStructOrEnum {refPath=&NiagaraBool}}}
      Disc Coverage {
       bIsDynamic=false
       bIsEditable=false
       bIsStaticSwitch=false
       bIsVisible=false
       type {classStructOrEnum {refPath=&NiagaraFloat}}
      }
      Hemicircle Internal Cap {bIsDynamic=false, bIsEditable=false, bIsStaticSwitch=true, bIsVisible=false, type {classStructOrEnum {refPath=&NiagaraBool}}}
      U Position {
       bIsDynamic=false
       bIsEditable=false
       bIsStaticSwitch=false
       bIsVisible=false
       type {classStructOrEnum {refPath=&NiagaraFloat}}
      }
      Enable Lathe Profile {bIsDynamic=false, bIsEditable=false, bIsStaticSwitch=true, bIsVisible=false, type {classStructOrEnum {refPath=&NiagaraBool}}}
      Lathe Profile {
       bIsDynamic=false
       bIsEditable=false
       bIsStaticSwitch=false
       bIsVisible=false
       type {classStructOrEnum {refPath=&NiagaraDataInterfaceCurve}}
      }
      Surface Distribution {
       bIsDynamic=false
       bIsEditable=false
       bIsStaticSwitch=false
       bIsVisible=false
       type {classStructOrEnum {refPath=&NiagaraFloat}}
      }
      Surface Expansion Mode {
       bIsDynamic=false
       bIsEditable=false
       bIsStaticSwitch=false
       bIsVisible=false
       type {classStructOrEnum {refPath=&ENiagaraExpansionMode}}
      }
      U Distribution {
       bIsDynamic=false
       bIsEditable=false
       bIsStaticSwitch=false
       bIsVisible=false
       type {classStructOrEnum {refPath=&NiagaraFloat}}
      }
      V Position {
       bIsDynamic=false
       bIsEditable=false
       bIsStaticSwitch=false
       bIsVisible=false
       type {classStructOrEnum {refPath=&NiagaraFloat}}
      }
      Radius Position {
       bIsDynamic=false
       bIsEditable=false
       bIsStaticSwitch=false
       bIsVisible=false
       type {classStructOrEnum {refPath=&NiagaraFloat}}
      }
      Sphere Radius Position {
       bIsDynamic=false
       bIsEditable=false
       bIsStaticSwitch=false
       bIsVisible=false
       type {classStructOrEnum {refPath=&NiagaraFloat}}
      }
      V Distribution {
       bIsDynamic=false
       bIsEditable=false
       bIsStaticSwitch=false
       bIsVisible=false
       type {classStructOrEnum {refPath=&NiagaraFloat}}
      }
      Evaluation Type {
       bIsDynamic=false
       bIsEditable=false
       bIsStaticSwitch=true
       bIsVisible=false
       type {classStructOrEnum {refPath=&ENiagaraRandomnessEvaluation}}
      }
      Uniform Distribution {
       bIsDynamic=false
       bIsEditable=false
       bIsStaticSwitch=false
       bIsVisible=false
       type {classStructOrEnum {refPath=&NiagaraFloat}}
      }
      Uniform Spiral Amount {
       bIsDynamic=false
       bIsEditable=false
       bIsStaticSwitch=false
       bIsVisible=false
       type {classStructOrEnum {refPath=&NiagaraFloat}}
      }
      Uniform Spiral Falloff {
       bIsDynamic=false
       bIsEditable=false
       bIsStaticSwitch=false
       bIsVisible=false
       type {classStructOrEnum {refPath=&NiagaraFloat}}
      }
      Uniform Count {
       bIsDynamic=false
       bIsEditable=false
       bIsStaticSwitch=false
       bIsVisible=false
       type {classStructOrEnum {refPath=&NiagaraInt32}}
      }
      Uniform Particle Index {
       bIsDynamic=false
       bIsEditable=false
       bIsStaticSwitch=false
       bIsVisible=false
       type {classStructOrEnum {refPath=&NiagaraInt32}}
      }
      Hemisphere Angle Type {
       bIsDynamic=false
       bIsEditable=true
       bIsStaticSwitch=true
       bIsVisible=true
       type {classStructOrEnum {refPath=&ENiagara_AngleInput}}
      }
      Uniform Endpoint Offset {
       bIsDynamic=false
       bIsEditable=false
       bIsStaticSwitch=false
       bIsVisible=false
       type {classStructOrEnum {refPath=&NiagaraFloat}}
      }
      MaskSpawnGroup {bIsDynamic=false, bIsEditable=true, bIsStaticSwitch=false, bIsVisible=true, type {classStructOrEnum {refPath=&NiagaraBool}}}
      Spawn Group Mask {bIsDynamic=false, bIsEditable=false, bIsStaticSwitch=false, bIsVisible=true, type {classStructOrEnum {refPath=&NiagaraInt32}}}
      Spawn Group {
       bIsDynamic=false
       bIsEditable=false
       bIsStaticSwitch=false
       bIsVisible=false
       type {classStructOrEnum {refPath=&NiagaraInt32}}
      }
      Randomness Mode {
       bIsDynamic=false
       bIsEditable=true
       bIsStaticSwitch=true
       bIsVisible=true
       type {classStructOrEnum {refPath=&ENiagaraRandomnessMode}}
      }
      Override Seed {bIsDynamic=false, bIsEditable=true, bIsStaticSwitch=false, bIsVisible=true, type {classStructOrEnum {refPath=&NiagaraBool}}}
      Random Seed {bIsDynamic=false, bIsEditable=false, bIsStaticSwitch=false, bIsVisible=true, type {classStructOrEnum {refPath=&NiagaraInt32}}}
      Fixed Random Seed {bIsDynamic=false, bIsEditable=false, bIsStaticSwitch=true, bIsVisible=false, type {classStructOrEnum {refPath=&NiagaraBool}}}
      Transform Order {
       bIsDynamic=false
       bIsEditable=true
       bIsStaticSwitch=true
       bIsVisible=true
       type {classStructOrEnum {refPath=&ENiagara_TransformOrder}}
      }
      Scale Mode {
       bIsDynamic=false
       bIsEditable=true
       bIsStaticSwitch=true
       bIsVisible=true
       type {classStructOrEnum {refPath=&ENiagara_ScaleMode}}
      }
      Non Uniform Scale {bIsDynamic=false, bIsEditable=true, bIsStaticSwitch=false, bIsVisible=true, type {classStructOrEnum {refPath=&Vector3f}}}
      UseOwnerScale {bIsDynamic=false, bIsEditable=true, bIsStaticSwitch=false, bIsVisible=true, type {classStructOrEnum {refPath=&NiagaraBool}}}
      Apply Owner Scale {bIsDynamic=false, bIsEditable=true, bIsStaticSwitch=false, bIsVisible=true, type {classStructOrEnum {refPath=&Vector3f}}}
      Rotation Mode {
       bIsDynamic=false
       bIsEditable=true
       bIsStaticSwitch=true
       bIsVisible=true
       type {classStructOrEnum {refPath=&ENiagara_RotationMode}}
      }
      Rotation Axis {
       bIsDynamic=false
       bIsEditable=false
       bIsStaticSwitch=false
       bIsVisible=false
       type {classStructOrEnum {refPath=&Vector3f}}
      }
      Rotation Matrix {
       bIsDynamic=false
       bIsEditable=false
       bIsStaticSwitch=false
       bIsVisible=false
       type {classStructOrEnum {refPath=&NiagaraMatrix}}
      }
      Rotation Quaternion {bIsDynamic=false, bIsEditable=false, bIsStaticSwitch=false, bIsVisible=false, type {classStructOrEnum {refPath=&Quat4f}}}
      Yaw / Pitch / Roll {
       bIsDynamic=false
       bIsEditable=false
       bIsStaticSwitch=false
       bIsVisible=false
       type {classStructOrEnum {refPath=&Vector3f}}
      }
      Rotation Angle {
       bIsDynamic=false
       bIsEditable=false
       bIsStaticSwitch=false
       bIsVisible=false
       type {classStructOrEnum {refPath=&NiagaraFloat}}
      }
      Invert Rotation Quaternion {bIsDynamic=false, bIsEditable=false, bIsStaticSwitch=false, bIsVisible=false, type {classStructOrEnum {refPath=&NiagaraBool}}}
      Additional Quaternion Rotation {bIsDynamic=false, bIsEditable=false, bIsStaticSwitch=false, bIsVisible=false, type {classStructOrEnum {refPath=&NiagaraBool}}}
      Additional Yaw / Pitch / Roll {
       bIsDynamic=false
       bIsEditable=false
       bIsStaticSwitch=false
       bIsVisible=false
       type {classStructOrEnum {refPath=&Vector3f}}
      }
      Rotation Angle Type {
       bIsDynamic=false
       bIsEditable=false
       bIsStaticSwitch=true
       bIsVisible=false
       type {classStructOrEnum {refPath=&ENiagara_AngleInput}}
      }
      Rotation Coordinate Space {
       bIsDynamic=false
       bIsEditable=true
       bIsStaticSwitch=false
       bIsVisible=true
       type {classStructOrEnum {refPath=&ENiagaraCoordinateSpace}}
      }
      Offset Mode {
       bIsDynamic=false
       bIsEditable=true
       bIsStaticSwitch=true
       bIsVisible=true
       type {classStructOrEnum {refPath=&ENiagara_OffsetMode}}
      }
      Shape Origin {
       bIsDynamic=false
       bIsEditable=true
       bIsStaticSwitch=false
       bIsVisible=true
       type {classStructOrEnum {refPath=&NiagaraPosition}}
      }
      Offset {bIsDynamic=false, bIsEditable=true, bIsStaticSwitch=false, bIsVisible=true, type {classStructOrEnum {refPath=&Vector3f}}}
      Offset Coordinate Space {
       bIsDynamic=false
       bIsEditable=true
       bIsStaticSwitch=false
       bIsVisible=true
       type {classStructOrEnum {refPath=&ENiagaraCoordinateSpace}}
      }
      Custom Transform Matrix {
       bIsDynamic=false
       bIsEditable=false
       bIsStaticSwitch=false
       bIsVisible=false
       type {classStructOrEnum {refPath=&NiagaraMatrix}}
      }
      Transform Method {
       bIsDynamic=false
       bIsEditable=true
       bIsStaticSwitch=true
       bIsVisible=true
       type {classStructOrEnum {refPath=&ENiagara_TransformType}}
      }
     ]
     moduleScript {refPath=&ShapeLocation}
    }
    {
     bIsSetParametersModule=false
     enabled=true
     moduleName=CameraOffset
     inputs [Camera Offset Amount {bIsDynamic=true, bIsEditable=true, bIsStaticSwitch=false, bIsVisible=true, type {classStructOrEnum {refPath=&NiagaraFloat}}}]
     moduleScript {refPath=&CameraOffset}
    }
    {
     bIsSetParametersModule=false
     enabled=true
     moduleName=AddVelocity001
     inputs [
      Velocity Mode {
       bIsDynamic=false
       bIsEditable=true
       bIsStaticSwitch=true
       bIsVisible=true
       type {classStructOrEnum {refPath=&ENiagara_VelocityMode}}
      }
      Velocity {bIsDynamic=true, bIsEditable=false, bIsStaticSwitch=false, bIsVisible=false, type {classStructOrEnum {refPath=&Vector3f}}}
      Velocity Speed {bIsDynamic=true, bIsEditable=true, bIsStaticSwitch=false, bIsVisible=true, type {classStructOrEnum {refPath=&NiagaraFloat}}}
      Velocity Speed Scale {
       bIsDynamic=false
       bIsEditable=false
       bIsStaticSwitch=false
       bIsVisible=false
       type {classStructOrEnum {refPath=&NiagaraFloat}}
      }
      Module.Use Falloff Distance {bIsDynamic=false, bIsEditable=true, bIsStaticSwitch=false, bIsVisible=true, type {classStructOrEnum {refPath=&NiagaraBool}}}
      Constrain To Radius {bIsDynamic=false, bIsEditable=false, bIsStaticSwitch=false, bIsVisible=true, type {classStructOrEnum {refPath=&NiagaraFloat}}}
      Distribution Along Cone Axis {
       bIsDynamic=false
       bIsEditable=false
       bIsStaticSwitch=false
       bIsVisible=false
       type {classStructOrEnum {refPath=&NiagaraFloat}}
      }
      Radius Falloff Near / Far {
       bIsDynamic=false
       bIsEditable=false
       bIsStaticSwitch=false
       bIsVisible=false
       type {classStructOrEnum {refPath=&Vector2f}}
      }
      Radius Falloff Exponent {
       bIsDynamic=false
       bIsEditable=false
       bIsStaticSwitch=false
       bIsVisible=false
       type {classStructOrEnum {refPath=&NiagaraFloat}}
      }
      Speed Falloff From Cone Axis {
       bIsDynamic=false
       bIsEditable=false
       bIsStaticSwitch=false
       bIsVisible=false
       type {classStructOrEnum {refPath=&NiagaraFloat}}
      }
      Invert Velocity Falloff {bIsDynamic=false, bIsEditable=false, bIsStaticSwitch=false, bIsVisible=false, type {classStructOrEnum {refPath=&NiagaraBool}}}
      Override Seed {bIsDynamic=false, bIsEditable=true, bIsStaticSwitch=false, bIsVisible=true, type {classStructOrEnum {refPath=&NiagaraBool}}}
      Random Seed {bIsDynamic=false, bIsEditable=false, bIsStaticSwitch=false, bIsVisible=true, type {classStructOrEnum {refPath=&NiagaraInt32}}}
      Fixed Random Seed {bIsDynamic=false, bIsEditable=false, bIsStaticSwitch=true, bIsVisible=false, type {classStructOrEnum {refPath=&NiagaraBool}}}
      Randomness Mode {
       bIsDynamic=false
       bIsEditable=true
       bIsStaticSwitch=true
       bIsVisible=true
       type {classStructOrEnum {refPath=&ENiagaraRandomnessMode}}
      }
      Cone Axis {
       bIsDynamic=false
       bIsEditable=false
       bIsStaticSwitch=false
       bIsVisible=false
       type {classStructOrEnum {refPath=&Vector3f}}
      }
      Cone Angle {
       bIsDynamic=false
       bIsEditable=false
       bIsStaticSwitch=false
       bIsVisible=false
       type {classStructOrEnum {refPath=&NiagaraFloat}}
      }
      Inner Cone Angle {
       bIsDynamic=false
       bIsEditable=false
       bIsStaticSwitch=false
       bIsVisible=false
       type {classStructOrEnum {refPath=&NiagaraFloat}}
      }
      Cone Angle Mode {
       bIsDynamic=false
       bIsEditable=false
       bIsStaticSwitch=true
       bIsVisible=false
       type {classStructOrEnum {refPath=&ENiagara_AngleInput}}
      }
      Rotation Mode {
       bIsDynamic=false
       bIsEditable=false
       bIsStaticSwitch=true
       bIsVisible=false
       type {classStructOrEnum {refPath=&ENiagara_RotationMode}}
      }
      Rotation Axis {
       bIsDynamic=false
       bIsEditable=false
       bIsStaticSwitch=false
       bIsVisible=false
       type {classStructOrEnum {refPath=&Vector3f}}
      }
      Yaw / Pitch / Roll {
       bIsDynamic=false
       bIsEditable=false
       bIsStaticSwitch=false
       bIsVisible=false
       type {classStructOrEnum {refPath=&Vector3f}}
      }
      Rotation Quaternion {bIsDynamic=false, bIsEditable=false, bIsStaticSwitch=false, bIsVisible=false, type {classStructOrEnum {refPath=&Quat4f}}}
      Rotation Matrix {
       bIsDynamic=false
       bIsEditable=false
       bIsStaticSwitch=false
       bIsVisible=false
       type {classStructOrEnum {refPath=&NiagaraMatrix}}
      }
      Rotation Angle {
       bIsDynamic=false
       bIsEditable=false
       bIsStaticSwitch=false
       bIsVisible=false
       type {classStructOrEnum {refPath=&NiagaraFloat}}
      }
      Invert Rotation Quaternion {bIsDynamic=false, bIsEditable=false, bIsStaticSwitch=false, bIsVisible=false, type {classStructOrEnum {refPath=&NiagaraBool}}}
      Additional Quaternion Rotation {bIsDynamic=false, bIsEditable=false, bIsStaticSwitch=false, bIsVisible=false, type {classStructOrEnum {refPath=&NiagaraBool}}}
      Additional Yaw / Pitch / Roll {
       bIsDynamic=false
       bIsEditable=false
       bIsStaticSwitch=false
       bIsVisible=false
       type {classStructOrEnum {refPath=&Vector3f}}
      }
      Rotation Angle Type {
       bIsDynamic=false
       bIsEditable=false
       bIsStaticSwitch=true
       bIsVisible=false
       type {classStructOrEnum {refPath=&ENiagara_AngleInput}}
      }
      Rotation Coordinate Space {
       bIsDynamic=false
       bIsEditable=false
       bIsStaticSwitch=false
       bIsVisible=false
       type {classStructOrEnum {refPath=&ENiagaraCoordinateSpace}}
      }
      Velocity Origin {
       bIsDynamic=false
       bIsEditable=true
       bIsStaticSwitch=false
       bIsVisible=true
       type {classStructOrEnum {refPath=&NiagaraPosition}}
      }
      Origin Offset {bIsDynamic=false, bIsEditable=true, bIsStaticSwitch=false, bIsVisible=true, type {classStructOrEnum {refPath=&Vector3f}}}
      Origin Offset Coordinate Space {
       bIsDynamic=false
       bIsEditable=true
       bIsStaticSwitch=false
       bIsVisible=true
       type {classStructOrEnum {refPath=&ENiagaraCoordinateSpace}}
      }
      Default Position {
       bIsDynamic=false
       bIsEditable=true
       bIsStaticSwitch=false
       bIsVisible=true
       type {classStructOrEnum {refPath=&NiagaraPosition}}
      }
     ]
     moduleScript {refPath=&AddVelocity}
    }
    {
     bIsSetParametersModule=false
     enabled=true
     moduleName=AddVelocity
     inputs [
      Velocity Mode {
       bIsDynamic=false
       bIsEditable=true
       bIsStaticSwitch=true
       bIsVisible=true
       type {classStructOrEnum {refPath=&ENiagara_VelocityMode}}
      }
      Velocity {bIsDynamic=true, bIsEditable=true, bIsStaticSwitch=false, bIsVisible=true, type {classStructOrEnum {refPath=&Vector3f}}}
      Velocity Speed {
       bIsDynamic=false
       bIsEditable=false
       bIsStaticSwitch=false
       bIsVisible=false
       type {classStructOrEnum {refPath=&NiagaraFloat}}
      }
      Velocity Speed Scale {bIsDynamic=false, bIsEditable=true, bIsStaticSwitch=false, bIsVisible=true, type {classStructOrEnum {refPath=&NiagaraFloat}}}
      Constrain To Radius {
       bIsDynamic=false
       bIsEditable=false
       bIsStaticSwitch=false
       bIsVisible=false
       type {classStructOrEnum {refPath=&NiagaraFloat}}
      }
      Distribution Along Cone Axis {
       bIsDynamic=false
       bIsEditable=false
       bIsStaticSwitch=false
       bIsVisible=false
       type {classStructOrEnum {refPath=&NiagaraFloat}}
      }
      Radius Falloff Near / Far {
       bIsDynamic=false
       bIsEditable=false
       bIsStaticSwitch=false
       bIsVisible=false
       type {classStructOrEnum {refPath=&Vector2f}}
      }
      Radius Falloff Exponent {
       bIsDynamic=false
       bIsEditable=false
       bIsStaticSwitch=false
       bIsVisible=false
       type {classStructOrEnum {refPath=&NiagaraFloat}}
      }
      Speed Falloff From Cone Axis {
       bIsDynamic=false
       bIsEditable=false
       bIsStaticSwitch=false
       bIsVisible=false
       type {classStructOrEnum {refPath=&NiagaraFloat}}
      }
      Invert Velocity Falloff {bIsDynamic=false, bIsEditable=false, bIsStaticSwitch=false, bIsVisible=false, type {classStructOrEnum {refPath=&NiagaraBool}}}
      Override Seed {bIsDynamic=false, bIsEditable=false, bIsStaticSwitch=false, bIsVisible=false, type {classStructOrEnum {refPath=&NiagaraBool}}}
      Random Seed {
       bIsDynamic=false
       bIsEditable=false
       bIsStaticSwitch=false
       bIsVisible=false
       type {classStructOrEnum {refPath=&NiagaraInt32}}
      }
      Fixed Random Seed {bIsDynamic=false, bIsEditable=false, bIsStaticSwitch=true, bIsVisible=false, type {classStructOrEnum {refPath=&NiagaraBool}}}
      Randomness Mode {
       bIsDynamic=false
       bIsEditable=false
       bIsStaticSwitch=true
       bIsVisible=false
       type {classStructOrEnum {refPath=&ENiagaraRandomnessMode}}
      }
      Cone Axis {
       bIsDynamic=false
       bIsEditable=false
       bIsStaticSwitch=false
       bIsVisible=false
       type {classStructOrEnum {refPath=&Vector3f}}
      }
      Cone Angle {
       bIsDynamic=false
       bIsEditable=false
       bIsStaticSwitch=false
       bIsVisible=false
       type {classStructOrEnum {refPath=&NiagaraFloat}}
      }
      Inner Cone Angle {
       bIsDynamic=false
       bIsEditable=false
       bIsStaticSwitch=false
       bIsVisible=false
       type {classStructOrEnum {refPath=&NiagaraFloat}}
      }
      Cone Angle Mode {
       bIsDynamic=false
       bIsEditable=false
       bIsStaticSwitch=true
       bIsVisible=false
       type {classStructOrEnum {refPath=&ENiagara_AngleInput}}
      }
      Rotation Mode {
       bIsDynamic=false
       bIsEditable=true
       bIsStaticSwitch=true
       bIsVisible=true
       type {classStructOrEnum {refPath=&ENiagara_RotationMode}}
      }
      Rotation Axis {
       bIsDynamic=false
       bIsEditable=false
       bIsStaticSwitch=false
       bIsVisible=false
       type {classStructOrEnum {refPath=&Vector3f}}
      }
      Yaw / Pitch / Roll {
       bIsDynamic=false
       bIsEditable=false
       bIsStaticSwitch=false
       bIsVisible=false
       type {classStructOrEnum {refPath=&Vector3f}}
      }
      Rotation Quaternion {bIsDynamic=false, bIsEditable=false, bIsStaticSwitch=false, bIsVisible=false, type {classStructOrEnum {refPath=&Quat4f}}}
      Rotation Matrix {
       bIsDynamic=false
       bIsEditable=false
       bIsStaticSwitch=false
       bIsVisible=false
       type {classStructOrEnum {refPath=&NiagaraMatrix}}
      }
      Rotation Angle {
       bIsDynamic=false
       bIsEditable=false
       bIsStaticSwitch=false
       bIsVisible=false
       type {classStructOrEnum {refPath=&NiagaraFloat}}
      }
      Invert Rotation Quaternion {bIsDynamic=false, bIsEditable=false, bIsStaticSwitch=false, bIsVisible=false, type {classStructOrEnum {refPath=&NiagaraBool}}}
      Additional Quaternion Rotation {bIsDynamic=false, bIsEditable=false, bIsStaticSwitch=false, bIsVisible=false, type {classStructOrEnum {refPath=&NiagaraBool}}}
      Additional Yaw / Pitch / Roll {
       bIsDynamic=false
       bIsEditable=false
       bIsStaticSwitch=false
       bIsVisible=false
       type {classStructOrEnum {refPath=&Vector3f}}
      }
      Rotation Angle Type {
       bIsDynamic=false
       bIsEditable=false
       bIsStaticSwitch=true
       bIsVisible=false
       type {classStructOrEnum {refPath=&ENiagara_AngleInput}}
      }
      Rotation Coordinate Space {
       bIsDynamic=false
       bIsEditable=true
       bIsStaticSwitch=false
       bIsVisible=true
       type {classStructOrEnum {refPath=&ENiagaraCoordinateSpace}}
      }
      Velocity Origin {
       bIsDynamic=false
       bIsEditable=false
       bIsStaticSwitch=false
       bIsVisible=false
       type {classStructOrEnum {refPath=&NiagaraPosition}}
      }
      Origin Offset {
       bIsDynamic=false
       bIsEditable=false
       bIsStaticSwitch=false
       bIsVisible=false
       type {classStructOrEnum {refPath=&Vector3f}}
      }
      Origin Offset Coordinate Space {
       bIsDynamic=false
       bIsEditable=false
       bIsStaticSwitch=false
       bIsVisible=false
       type {classStructOrEnum {refPath=&ENiagaraCoordinateSpace}}
      }
      Default Position {
       bIsDynamic=false
       bIsEditable=false
       bIsStaticSwitch=false
       bIsVisible=false
       type {classStructOrEnum {refPath=&NiagaraPosition}}
      }
     ]
     moduleScript {refPath=&AddVelocity}
    }
    {
     bIsSetParametersModule=false
     enabled=true
     moduleName=ApplyInitialForces
     inputs [
      Apply Force to Velocity {bIsDynamic=false, bIsEditable=true, bIsStaticSwitch=false, bIsVisible=true, type {classStructOrEnum {refPath=&NiagaraBool}}}
      Apply Rotational Force to Rotational Velocity {bIsDynamic=false, bIsEditable=true, bIsStaticSwitch=false, bIsVisible=true, type {classStructOrEnum {refPath=&NiagaraBool}}}
      Apply Force to Position {bIsDynamic=false, bIsEditable=true, bIsStaticSwitch=false, bIsVisible=true, type {classStructOrEnum {refPath=&NiagaraBool}}}
      Positional Force Warmup Time {bIsDynamic=false, bIsEditable=false, bIsStaticSwitch=false, bIsVisible=true, type {classStructOrEnum {refPath=&NiagaraFloat}}}
     ]
     moduleScript {refPath=&ApplyInitialForces}
    }
    {
     bIsSetParametersModule=true
     enabled=true
     moduleName=SetVariables_#3
     inputs [Particles.DynamicMaterialParameter_Air {bIsDynamic=true, bIsEditable=true, bIsStaticSwitch=false, bIsVisible=true, type {classStructOrEnum {refPath=&Vector4f}}}]
     moduleScript {refPath=&self:DefaultPropSpawn_Merged001_0.NiagaraScriptSource_1.NiagaraGraph_0.NiagaraNodeAssignment_5.SetVariables_#3}
    }
   ]
  }
  particleUpdateScript {
   scriptName=ParticleUpdateScript
   modules [
    {
     bIsSetParametersModule=false
     enabled=true
     moduleName=ParticleState
     inputs [
      Kill Particles When Lifetime Has Elapsed {bIsDynamic=false, bIsEditable=true, bIsStaticSwitch=true, bIsVisible=true, type {classStructOrEnum {refPath=&NiagaraBool}}}
      Loop Particles Lifetime {bIsDynamic=false, bIsEditable=false, bIsStaticSwitch=false, bIsVisible=false, type {classStructOrEnum {refPath=&NiagaraBool}}}
      Let Infinitely Lived Particles Die When Emitter Deactivates {bIsDynamic=false, bIsEditable=false, bIsStaticSwitch=false, bIsVisible=false, type {classStructOrEnum {refPath=&NiagaraBool}}}
      Lifetime {bIsDynamic=false, bIsEditable=true, bIsStaticSwitch=false, bIsVisible=true, type {classStructOrEnum {refPath=&NiagaraFloat}}}
      DeltaTime {bIsDynamic=false, bIsEditable=true, bIsStaticSwitch=false, bIsVisible=true, type {classStructOrEnum {refPath=&NiagaraFloat}}}
     ]
     moduleScript {refPath=&ParticleState}
    }
    {
     bIsSetParametersModule=false
     enabled=true
     moduleName=ScaleMeshSize
     inputs [
      Scale Factor {bIsDynamic=true, bIsEditable=true, bIsStaticSwitch=false, bIsVisible=true, type {classStructOrEnum {refPath=&Vector3f}}}
      Initial Mesh Scale {bIsDynamic=false, bIsEditable=true, bIsStaticSwitch=false, bIsVisible=true, type {classStructOrEnum {refPath=&Vector3f}}}
     ]
     moduleScript {refPath=&ScaleMeshSize}
    }
    {
     bIsSetParametersModule=false
     enabled=true
     moduleName=SpriteRotationRate
     inputs [
      Rotation Rate {bIsDynamic=true, bIsEditable=true, bIsStaticSwitch=false, bIsVisible=true, type {classStructOrEnum {refPath=&NiagaraFloat}}}
      Delta Time {bIsDynamic=false, bIsEditable=true, bIsStaticSwitch=false, bIsVisible=true, type {classStructOrEnum {refPath=&NiagaraFloat}}}
     ]
     moduleScript {refPath=&SpriteRotationRate}
    }
    {
     bIsSetParametersModule=false
     enabled=true
     moduleName=GravityForce
     inputs [
      Gravity {bIsDynamic=true, bIsEditable=true, bIsStaticSwitch=false, bIsVisible=true, type {classStructOrEnum {refPath=&Vector3f}}}
      Coordinate Space {
       bIsDynamic=false
       bIsEditable=true
       bIsStaticSwitch=true
       bIsVisible=true
       type {classStructOrEnum {refPath=&ENiagaraCoordinateSpace}}
      }
     ]
     moduleScript {refPath=&GravityForce}
    }
    {
     bIsSetParametersModule=false
     enabled=true
     moduleName=CurlNoiseForce
     inputs [
      Noise Strength {bIsDynamic=true, bIsEditable=true, bIsStaticSwitch=false, bIsVisible=true, type {classStructOrEnum {refPath=&NiagaraFloat}}}
      Noise Frequency {bIsDynamic=true, bIsEditable=true, bIsStaticSwitch=false, bIsVisible=true, type {classStructOrEnum {refPath=&NiagaraFloat}}}
      Noise Quality / Cost {
       bIsDynamic=false
       bIsEditable=true
       bIsStaticSwitch=true
       bIsVisible=true
       type {classStructOrEnum {refPath=&ENiagaraCurlNoiseQuality}}
      }
      Random Seed {bIsDynamic=false, bIsEditable=true, bIsStaticSwitch=false, bIsVisible=true, type {classStructOrEnum {refPath=&NiagaraInt32}}}
      Pan Noise {bIsDynamic=false, bIsEditable=true, bIsStaticSwitch=false, bIsVisible=true, type {classStructOrEnum {refPath=&NiagaraBool}}}
      Pan Noise Field {bIsDynamic=false, bIsEditable=false, bIsStaticSwitch=false, bIsVisible=true, type {classStructOrEnum {refPath=&Vector3f}}}
      Write to Intrinsic Parameters {bIsDynamic=false, bIsEditable=true, bIsStaticSwitch=true, bIsVisible=true, type {classStructOrEnum {refPath=&NiagaraBool}}}
      Sample Position {bIsDynamic=false, bIsEditable=true, bIsStaticSwitch=false, bIsVisible=true, type {classStructOrEnum {refPath=&Vector3f}}}
      CurlNoise {
       bIsDynamic=false
       bIsEditable=false
       bIsStaticSwitch=false
       bIsVisible=false
       type {classStructOrEnum {refPath=&NiagaraDataInterfaceCurlNoise}}
      }
      Mask Curl Noise {bIsDynamic=false, bIsEditable=true, bIsStaticSwitch=true, bIsVisible=true, type {classStructOrEnum {refPath=&NiagaraBool}}}
      Curl Noise Cone Mask Angle {
       bIsDynamic=false
       bIsEditable=false
       bIsStaticSwitch=false
       bIsVisible=false
       type {classStructOrEnum {refPath=&NiagaraFloat}}
      }
      Curl Noise Cone Mask Falloff Angle {
       bIsDynamic=false
       bIsEditable=false
       bIsStaticSwitch=false
       bIsVisible=false
       type {classStructOrEnum {refPath=&NiagaraFloat}}
      }
      Curl Noise Cone Mask Axis {
       bIsDynamic=false
       bIsEditable=false
       bIsStaticSwitch=false
       bIsVisible=false
       type {classStructOrEnum {refPath=&Vector3f}}
      }
     ]
     moduleScript {refPath=&CurlNoiseForce}
    }
    {
     bIsSetParametersModule=false
     enabled=true
     moduleName=Drag
     inputs [
      Use Linear Drag {bIsDynamic=false, bIsEditable=true, bIsStaticSwitch=false, bIsVisible=true, type {classStructOrEnum {refPath=&NiagaraBool}}}
      Drag {bIsDynamic=true, bIsEditable=true, bIsStaticSwitch=false, bIsVisible=true, type {classStructOrEnum {refPath=&NiagaraFloat}}}
      Use Rotational Drag {bIsDynamic=false, bIsEditable=true, bIsStaticSwitch=false, bIsVisible=true, type {classStructOrEnum {refPath=&NiagaraBool}}}
      Rotational Drag {bIsDynamic=false, bIsEditable=true, bIsStaticSwitch=false, bIsVisible=true, type {classStructOrEnum {refPath=&NiagaraFloat}}}
      Ignore Mass {bIsDynamic=false, bIsEditable=true, bIsStaticSwitch=false, bIsVisible=true, type {classStructOrEnum {refPath=&NiagaraBool}}}
     ]
     moduleScript {refPath=&Drag}
    }
    {
     bIsSetParametersModule=false
     enabled=true
     moduleName=SolveForcesAndVelocity
     inputs [
      Clamp Velocity {bIsDynamic=false, bIsEditable=true, bIsStaticSwitch=false, bIsVisible=true, type {classStructOrEnum {refPath=&NiagaraBool}}}
      Speed Limit {bIsDynamic=false, bIsEditable=false, bIsStaticSwitch=false, bIsVisible=true, type {classStructOrEnum {refPath=&NiagaraFloat}}}
      Limit Acceleration {bIsDynamic=false, bIsEditable=true, bIsStaticSwitch=false, bIsVisible=true, type {classStructOrEnum {refPath=&NiagaraBool}}}
      Acceleration Limit {bIsDynamic=false, bIsEditable=false, bIsStaticSwitch=false, bIsVisible=true, type {classStructOrEnum {refPath=&NiagaraFloat}}}
      Manually Enable Rotational Solver {bIsDynamic=false, bIsEditable=true, bIsStaticSwitch=false, bIsVisible=true, type {classStructOrEnum {refPath=&NiagaraBool}}}
      Write to Intrinsic Properties {bIsDynamic=false, bIsEditable=true, bIsStaticSwitch=false, bIsVisible=true, type {classStructOrEnum {refPath=&NiagaraBool}}}
      Write to Presolve Properties {bIsDynamic=false, bIsEditable=true, bIsStaticSwitch=true, bIsVisible=true, type {classStructOrEnum {refPath=&NiagaraBool}}}
      DeltaTime {bIsDynamic=false, bIsEditable=true, bIsStaticSwitch=false, bIsVisible=true, type {classStructOrEnum {refPath=&NiagaraFloat}}}
      Mass {bIsDynamic=false, bIsEditable=true, bIsStaticSwitch=false, bIsVisible=true, type {classStructOrEnum {refPath=&NiagaraFloat}}}
      Position {
       bIsDynamic=false
       bIsEditable=true
       bIsStaticSwitch=false
       bIsVisible=true
       type {classStructOrEnum {refPath=&NiagaraPosition}}
      }
      Velocity {bIsDynamic=false, bIsEditable=true, bIsStaticSwitch=false, bIsVisible=true, type {classStructOrEnum {refPath=&Vector3f}}}
      PreviousVelocity {bIsDynamic=false, bIsEditable=true, bIsStaticSwitch=false, bIsVisible=true, type {classStructOrEnum {refPath=&Vector3f}}}
      Force {bIsDynamic=false, bIsEditable=true, bIsStaticSwitch=false, bIsVisible=true, type {classStructOrEnum {refPath=&Vector3f}}}
      Mesh Orientation {bIsDynamic=false, bIsEditable=false, bIsStaticSwitch=false, bIsVisible=false, type {classStructOrEnum {refPath=&Quat4f}}}
      Rotational Velocity {
       bIsDynamic=false
       bIsEditable=false
       bIsStaticSwitch=false
       bIsVisible=false
       type {classStructOrEnum {refPath=&Vector3f}}
      }
      Rotational Force {
       bIsDynamic=false
       bIsEditable=false
       bIsStaticSwitch=false
       bIsVisible=false
       type {classStructOrEnum {refPath=&Vector3f}}
      }
      Rotational Inertia {
       bIsDynamic=false
       bIsEditable=false
       bIsStaticSwitch=false
       bIsVisible=false
       type {classStructOrEnum {refPath=&NiagaraFloat}}
      }
     ]
     moduleScript {refPath=&SolveForcesAndVelocity}
    }
    {
     bIsSetParametersModule=false
     enabled=true
     moduleName=ScaleSpriteSize
     inputs [
      Scale Sprite Size Mode {
       bIsDynamic=false
       bIsEditable=true
       bIsStaticSwitch=true
       bIsVisible=true
       type {classStructOrEnum {refPath=&ENiagara_ScaleSpriteSize}}
      }
      Uniform Curve Sprite Scale {
       bIsDynamic=false
       bIsEditable=false
       bIsStaticSwitch=false
       bIsVisible=false
       type {classStructOrEnum {refPath=&NiagaraDataInterfaceCurve}}
      }
      Non-Uniform Curve Sprite Scale {
       bIsDynamic=false
       bIsEditable=false
       bIsStaticSwitch=false
       bIsVisible=false
       type {classStructOrEnum {refPath=&NiagaraDataInterfaceVector2DCurve}}
      }
      Uniform Curve Index {
       bIsDynamic=false
       bIsEditable=false
       bIsStaticSwitch=false
       bIsVisible=false
       type {classStructOrEnum {refPath=&NiagaraFloat}}
      }
      Non-Uniform Curve Index {
       bIsDynamic=false
       bIsEditable=false
       bIsStaticSwitch=false
       bIsVisible=false
       type {classStructOrEnum {refPath=&NiagaraFloat}}
      }
      Non-Uniform Curve Scale {
       bIsDynamic=false
       bIsEditable=false
       bIsStaticSwitch=false
       bIsVisible=false
       type {classStructOrEnum {refPath=&Vector2f}}
      }
      Uniform Curve Scale {
       bIsDynamic=false
       bIsEditable=false
       bIsStaticSwitch=false
       bIsVisible=false
       type {classStructOrEnum {refPath=&NiagaraFloat}}
      }
      Scale Factor {bIsDynamic=true, bIsEditable=true, bIsStaticSwitch=false, bIsVisible=true, type {classStructOrEnum {refPath=&Vector2f}}}
      Uniform Scale Factor {
       bIsDynamic=false
       bIsEditable=false
       bIsStaticSwitch=false
       bIsVisible=false
       type {classStructOrEnum {refPath=&NiagaraFloat}}
      }
      Initial Sprite Size {bIsDynamic=false, bIsEditable=true, bIsStaticSwitch=false, bIsVisible=true, type {classStructOrEnum {refPath=&Vector2f}}}
     ]
     moduleScript {refPath=&ScaleSpriteSize}
    }
    {
     bIsSetParametersModule=true
     enabled=true
     moduleName=SetVariables_#4
     inputs [
      Particles.DynParSmoke0 {bIsDynamic=true, bIsEditable=true, bIsStaticSwitch=false, bIsVisible=true, type {classStructOrEnum {refPath=&Vector4f}}}
      Particles.DynParSmoke1 {bIsDynamic=true, bIsEditable=true, bIsStaticSwitch=false, bIsVisible=true, type {classStructOrEnum {refPath=&Vector4f}}}
      Particles.DynParSmoke2 {bIsDynamic=true, bIsEditable=true, bIsStaticSwitch=false, bIsVisible=true, type {classStructOrEnum {refPath=&Vector4f}}}
     ]
     moduleScript {refPath=&self:DefaultPropSpawn_Merged001_0.NiagaraScriptSource_1.NiagaraGraph_0.NiagaraNodeAssignment_2.SetVariables_#4}
    }
    {
     bIsSetParametersModule=false
     enabled=true
     moduleName=ScaleColor001
     inputs [
      Scale Mode {bIsDynamic=false, bIsEditable=true, bIsStaticSwitch=true, bIsVisible=true, type {classStructOrEnum {refPath=&NiagaraInt32}}}
      Linear Color Curve {
       bIsDynamic=false
       bIsEditable=false
       bIsStaticSwitch=false
       bIsVisible=false
       type {classStructOrEnum {refPath=&NiagaraDataInterfaceColorCurve}}
      }
      ScaleRGB {bIsDynamic=false, bIsEditable=true, bIsStaticSwitch=false, bIsVisible=true, type {classStructOrEnum {refPath=&NiagaraBool}}}
      Scale RGB {bIsDynamic=false, bIsEditable=true, bIsStaticSwitch=false, bIsVisible=true, type {classStructOrEnum {refPath=&Vector3f}}}
      ScaleRGBA {bIsDynamic=false, bIsEditable=false, bIsStaticSwitch=false, bIsVisible=false, type {classStructOrEnum {refPath=&NiagaraBool}}}
      Scale RGBA {
       bIsDynamic=false
       bIsEditable=false
       bIsStaticSwitch=false
       bIsVisible=false
       type {classStructOrEnum {refPath=&Vector4f}}
      }
      Curve Index {
       bIsDynamic=false
       bIsEditable=false
       bIsStaticSwitch=false
       bIsVisible=false
       type {classStructOrEnum {refPath=&NiagaraFloat}}
      }
      ScaleA {bIsDynamic=false, bIsEditable=true, bIsStaticSwitch=false, bIsVisible=true, type {classStructOrEnum {refPath=&NiagaraBool}}}
      Scale Alpha {bIsDynamic=true, bIsEditable=true, bIsStaticSwitch=false, bIsVisible=true, type {classStructOrEnum {refPath=&NiagaraFloat}}}
      Color Value To Scale {
       bIsDynamic=false
       bIsEditable=true
       bIsStaticSwitch=false
       bIsVisible=true
       type {classStructOrEnum {refPath=&LinearColor}}
      }
      Color Scale {
       bIsDynamic=false
       bIsEditable=false
       bIsStaticSwitch=false
       bIsVisible=false
       type {classStructOrEnum {refPath=&LinearColor}}
      }
     ]
     moduleScript {refPath=&ScaleColor}
    }
    {
     bIsSetParametersModule=false
     enabled=true
     moduleName=FN_Light_Attributes
     inputs [
      Enable Particle Light Enable {bIsDynamic=false, bIsEditable=true, bIsStaticSwitch=true, bIsVisible=true, type {classStructOrEnum {refPath=&NiagaraBool}}}
      Light Enabled {bIsDynamic=false, bIsEditable=false, bIsStaticSwitch=false, bIsVisible=false, type {classStructOrEnum {refPath=&NiagaraBool}}}
      Enable Visibility Tag {bIsDynamic=false, bIsEditable=true, bIsStaticSwitch=true, bIsVisible=true, type {classStructOrEnum {refPath=&NiagaraBool}}}
      VisibilityTag {
       bIsDynamic=false
       bIsEditable=false
       bIsStaticSwitch=false
       bIsVisible=false
       type {classStructOrEnum {refPath=&NiagaraInt32}}
      }
      Create Position Attribute {bIsDynamic=false, bIsEditable=true, bIsStaticSwitch=true, bIsVisible=true, type {classStructOrEnum {refPath=&NiagaraBool}}}
      Position {bIsDynamic=false, bIsEditable=true, bIsStaticSwitch=false, bIsVisible=true, type {classStructOrEnum {refPath=&Vector3f}}}
      Create Radius Attribute {bIsDynamic=false, bIsEditable=true, bIsStaticSwitch=true, bIsVisible=true, type {classStructOrEnum {refPath=&NiagaraBool}}}
      Radius {bIsDynamic=false, bIsEditable=true, bIsStaticSwitch=false, bIsVisible=true, type {classStructOrEnum {refPath=&NiagaraFloat}}}
      Create Exponent Attribute {bIsDynamic=false, bIsEditable=true, bIsStaticSwitch=true, bIsVisible=true, type {classStructOrEnum {refPath=&NiagaraBool}}}
      Exponent {bIsDynamic=false, bIsEditable=true, bIsStaticSwitch=false, bIsVisible=true, type {classStructOrEnum {refPath=&NiagaraFloat}}}
      Create Color Attribute {bIsDynamic=false, bIsEditable=true, bIsStaticSwitch=true, bIsVisible=true, type {classStructOrEnum {refPath=&NiagaraBool}}}
      Color {
       bIsDynamic=false
       bIsEditable=true
       bIsStaticSwitch=false
       bIsVisible=true
       type {classStructOrEnum {refPath=&LinearColor}}
      }
      Scale Color {bIsDynamic=true, bIsEditable=true, bIsStaticSwitch=false, bIsVisible=true, type {classStructOrEnum {refPath=&NiagaraFloat}}}
     ]
     moduleScript {refPath=&FN_Light_Attributes}
    }
    {
     bIsSetParametersModule=true
     enabled=true
     moduleName=SetVariables_#9
     inputs [
      Particles.ShockwaveScale {bIsDynamic=true, bIsEditable=true, bIsStaticSwitch=false, bIsVisible=true, type {classStructOrEnum {refPath=&Vector3f}}}
      Particles.ShockwaveColor {
       bIsDynamic=true
       bIsEditable=true
       bIsStaticSwitch=false
       bIsVisible=true
       type {classStructOrEnum {refPath=&LinearColor}}
      }
      Particles.SecondarySphereColor {
       bIsDynamic=true
       bIsEditable=true
       bIsStaticSwitch=false
       bIsVisible=true
       type {classStructOrEnum {refPath=&LinearColor}}
      }
      Particles.SecondaryDyn {bIsDynamic=true, bIsEditable=true, bIsStaticSwitch=false, bIsVisible=true, type {classStructOrEnum {refPath=&Vector4f}}}
      Particles.PrimaryColor {
       bIsDynamic=true
       bIsEditable=true
       bIsStaticSwitch=false
       bIsVisible=true
       type {classStructOrEnum {refPath=&LinearColor}}
      }
     ]
     moduleScript {refPath=&self:DefaultPropSpawn_Merged001_0.NiagaraScriptSource_1.NiagaraGraph_0.NiagaraNodeAssignment_3.SetVariables_#9}
    }
   ]
  }
  rendererClasses [
   {refPath=&NiagaraSpriteRendererProperties}
   {refPath=&NiagaraLightRendererProperties}
   {refPath=&NiagaraMeshRendererProperties}
  ]
  renderers [
   {rendererIndex=0, rendererClass {refPath=&NiagaraSpriteRendererProperties}}
   {rendererIndex=1, rendererClass {refPath=&NiagaraSpriteRendererProperties}}
   {rendererIndex=2, rendererClass {refPath=&NiagaraSpriteRendererProperties}}
   {rendererIndex=3, rendererClass {refPath=&NiagaraSpriteRendererProperties}}
   {rendererIndex=4, rendererClass {refPath=&NiagaraSpriteRendererProperties}}
   {rendererIndex=5, rendererClass {refPath=&NiagaraLightRendererProperties}}
   {rendererIndex=6, rendererClass {refPath=&NiagaraSpriteRendererProperties}}
   {rendererIndex=7, rendererClass {refPath=&NiagaraMeshRendererProperties}}
   {rendererIndex=8, rendererClass {refPath=&NiagaraMeshRendererProperties}}
   {rendererIndex=9, rendererClass {refPath=&NiagaraMeshRendererProperties}}
  ]
 }
 DefaultPropSpawn_Merged002 {
  bEnabled=false
  emitterSpawnScript {scriptName=EmitterSpawnScript, modules []}
  emitterUpdateScript {
   scriptName=EmitterUpdateScript
   modules [
    {
     bIsSetParametersModule=false
     enabled=true
     moduleName=EmitterState
     inputs [
      Life Cycle Mode {
       bIsDynamic=false
       bIsEditable=true
       bIsStaticSwitch=true
       bIsVisible=true
       type {classStructOrEnum {refPath=&ENiagaraEmitterLifeCycleMode}}
      }
      Inactive Response {
       bIsDynamic=false
       bIsEditable=false
       bIsStaticSwitch=true
       bIsVisible=false
       type {classStructOrEnum {refPath=&ENiagaraInactiveMode}}
      }
      Loop Behavior {
       bIsDynamic=false
       bIsEditable=false
       bIsStaticSwitch=true
       bIsVisible=false
       type {classStructOrEnum {refPath=&ENiagara_EmitterStateOptions}}
      }
      Loop Duration Mode {
       bIsDynamic=false
       bIsEditable=false
       bIsStaticSwitch=true
       bIsVisible=false
       type {classStructOrEnum {refPath=&ENiagara_InfiniteLoopDuration}}
      }
      Loop Count {
       bIsDynamic=false
       bIsEditable=false
       bIsStaticSwitch=false
       bIsVisible=false
       type {classStructOrEnum {refPath=&NiagaraInt32}}
      }
      Loop Duration {
       bIsDynamic=false
       bIsEditable=false
       bIsStaticSwitch=false
       bIsVisible=false
       type {classStructOrEnum {refPath=&NiagaraFloat}}
      }
      Recalculate Duration Each Loop {bIsDynamic=false, bIsEditable=false, bIsStaticSwitch=false, bIsVisible=false, type {classStructOrEnum {refPath=&NiagaraBool}}}
      UseLoopDelay {bIsDynamic=false, bIsEditable=false, bIsStaticSwitch=false, bIsVisible=false, type {classStructOrEnum {refPath=&NiagaraBool}}}
      Loop Delay {
       bIsDynamic=false
       bIsEditable=false
       bIsStaticSwitch=false
       bIsVisible=false
       type {classStructOrEnum {refPath=&NiagaraFloat}}
      }
      Delay First Loop Only {bIsDynamic=false, bIsEditable=false, bIsStaticSwitch=false, bIsVisible=false, type {classStructOrEnum {refPath=&NiagaraBool}}}
      Scalability Mode {
       bIsDynamic=false
       bIsEditable=true
       bIsStaticSwitch=true
       bIsVisible=true
       type {classStructOrEnum {refPath=&ENiagaraEmitterScalabilityMode_Limited}}
      }
      Enable Distance Culling {bIsDynamic=false, bIsEditable=false, bIsStaticSwitch=true, bIsVisible=false, type {classStructOrEnum {refPath=&NiagaraBool}}}
      MinDistance {
       bIsDynamic=false
       bIsEditable=false
       bIsStaticSwitch=false
       bIsVisible=false
       type {classStructOrEnum {refPath=&NiagaraFloat}}
      }
      Min Distance Response {
       bIsDynamic=false
       bIsEditable=false
       bIsStaticSwitch=false
       bIsVisible=false
       type {classStructOrEnum {refPath=&ENiagaraExecutionStateManagement}}
      }
      MaxDistance {
       bIsDynamic=false
       bIsEditable=false
       bIsStaticSwitch=false
       bIsVisible=false
       type {classStructOrEnum {refPath=&NiagaraFloat}}
      }
      Max Distance Response {
       bIsDynamic=false
       bIsEditable=false
       bIsStaticSwitch=false
       bIsVisible=false
       type {classStructOrEnum {refPath=&ENiagaraExecutionStateManagement}}
      }
      Scale Spawn Count {bIsDynamic=false, bIsEditable=false, bIsStaticSwitch=true, bIsVisible=false, type {classStructOrEnum {refPath=&NiagaraBool}}}
      Spawn Count Scale {
       bIsDynamic=false
       bIsEditable=false
       bIsStaticSwitch=false
       bIsVisible=false
       type {classStructOrEnum {refPath=&NiagaraFloat}}
      }
      Spawn Count Scale By Distance {
       bIsDynamic=false
       bIsEditable=false
       bIsStaticSwitch=false
       bIsVisible=false
       type {classStructOrEnum {refPath=&NiagaraDataInterfaceCurve}}
      }
      Enable Visibility Culling {bIsDynamic=false, bIsEditable=false, bIsStaticSwitch=true, bIsVisible=false, type {classStructOrEnum {refPath=&NiagaraBool}}}
      Visibility Cull Response {
       bIsDynamic=false
       bIsEditable=false
       bIsStaticSwitch=false
       bIsVisible=false
       type {classStructOrEnum {refPath=&ENiagaraExecutionStateManagement}}
      }
      Visibility Cull Delay {
       bIsDynamic=false
       bIsEditable=false
       bIsStaticSwitch=false
       bIsVisible=false
       type {classStructOrEnum {refPath=&NiagaraFloat}}
      }
      Reset Age On Awaken {bIsDynamic=false, bIsEditable=false, bIsStaticSwitch=false, bIsVisible=false, type {classStructOrEnum {refPath=&NiagaraBool}}}
     ]
     moduleScript {refPath=&EmitterState}
    }
    {
     bIsSetParametersModule=false
     enabled=true
     moduleName=SpawnBurst_Instantaneous001
     inputs [
      Spawn Count {bIsDynamic=true, bIsEditable=true, bIsStaticSwitch=false, bIsVisible=true, type {classStructOrEnum {refPath=&NiagaraInt32}}}
      Spawn Time {bIsDynamic=false, bIsEditable=true, bIsStaticSwitch=false, bIsVisible=true, type {classStructOrEnum {refPath=&NiagaraFloat}}}
      Use Spawn Probability {bIsDynamic=false, bIsEditable=true, bIsStaticSwitch=false, bIsVisible=true, type {classStructOrEnum {refPath=&NiagaraBool}}}
      Spawn Probability {bIsDynamic=false, bIsEditable=false, bIsStaticSwitch=false, bIsVisible=true, type {classStructOrEnum {refPath=&NiagaraFloat}}}
      UseLoopCountLimit {bIsDynamic=false, bIsEditable=true, bIsStaticSwitch=false, bIsVisible=true, type {classStructOrEnum {refPath=&NiagaraBool}}}
      Loop Count Limit {bIsDynamic=false, bIsEditable=false, bIsStaticSwitch=false, bIsVisible=true, type {classStructOrEnum {refPath=&NiagaraInt32}}}
      Spawn Group {bIsDynamic=false, bIsEditable=true, bIsStaticSwitch=false, bIsVisible=true, type {classStructOrEnum {refPath=&NiagaraInt32}}}
      Age {bIsDynamic=false, bIsEditable=true, bIsStaticSwitch=false, bIsVisible=true, type {classStructOrEnum {refPath=&NiagaraFloat}}}
     ]
     moduleScript {refPath=&SpawnBurst_Instantaneous}
    }
    {
     bIsSetParametersModule=false
     enabled=true
     moduleName=SpawnBurst_Instantaneous
     inputs [
      Spawn Count {bIsDynamic=false, bIsEditable=true, bIsStaticSwitch=false, bIsVisible=true, type {classStructOrEnum {refPath=&NiagaraInt32}}}
      Spawn Time {bIsDynamic=false, bIsEditable=true, bIsStaticSwitch=false, bIsVisible=true, type {classStructOrEnum {refPath=&NiagaraFloat}}}
      Use Spawn Probability {bIsDynamic=false, bIsEditable=true, bIsStaticSwitch=false, bIsVisible=true, type {classStructOrEnum {refPath=&NiagaraBool}}}
      Spawn Probability {bIsDynamic=false, bIsEditable=false, bIsStaticSwitch=false, bIsVisible=true, type {classStructOrEnum {refPath=&NiagaraFloat}}}
      UseLoopCountLimit {bIsDynamic=false, bIsEditable=true, bIsStaticSwitch=false, bIsVisible=true, type {classStructOrEnum {refPath=&NiagaraBool}}}
      Loop Count Limit {bIsDynamic=false, bIsEditable=false, bIsStaticSwitch=false, bIsVisible=true, type {classStructOrEnum {refPath=&NiagaraInt32}}}
      Spawn Group {bIsDynamic=false, bIsEditable=true, bIsStaticSwitch=false, bIsVisible=true, type {classStructOrEnum {refPath=&NiagaraInt32}}}
      Age {bIsDynamic=false, bIsEditable=true, bIsStaticSwitch=false, bIsVisible=true, type {classStructOrEnum {refPath=&NiagaraFloat}}}
     ]
     moduleScript {refPath=&SpawnBurst_Instantaneous}
    }
    {
     bIsSetParametersModule=false
     enabled=true
     moduleName=SpawnBurst_Instantaneous002
     inputs [
      Spawn Count {bIsDynamic=true, bIsEditable=true, bIsStaticSwitch=false, bIsVisible=true, type {classStructOrEnum {refPath=&NiagaraInt32}}}
      Spawn Time {bIsDynamic=false, bIsEditable=true, bIsStaticSwitch=false, bIsVisible=true, type {classStructOrEnum {refPath=&NiagaraFloat}}}
      Use Spawn Probability {bIsDynamic=false, bIsEditable=true, bIsStaticSwitch=false, bIsVisible=true, type {classStructOrEnum {refPath=&NiagaraBool}}}
      Spawn Probability {bIsDynamic=false, bIsEditable=false, bIsStaticSwitch=false, bIsVisible=true, type {classStructOrEnum {refPath=&NiagaraFloat}}}
      UseLoopCountLimit {bIsDynamic=false, bIsEditable=true, bIsStaticSwitch=false, bIsVisible=true, type {classStructOrEnum {refPath=&NiagaraBool}}}
      Loop Count Limit {bIsDynamic=false, bIsEditable=false, bIsStaticSwitch=false, bIsVisible=true, type {classStructOrEnum {refPath=&NiagaraInt32}}}
      Spawn Group {bIsDynamic=false, bIsEditable=true, bIsStaticSwitch=false, bIsVisible=true, type {classStructOrEnum {refPath=&NiagaraInt32}}}
      Age {bIsDynamic=false, bIsEditable=true, bIsStaticSwitch=false, bIsVisible=true, type {classStructOrEnum {refPath=&NiagaraFloat}}}
     ]
     moduleScript {refPath=&SpawnBurst_Instantaneous}
    }
    {
     bIsSetParametersModule=false
     enabled=true
     moduleName=SpawnBurst_Instantaneous003
     inputs [
      Spawn Count {bIsDynamic=false, bIsEditable=true, bIsStaticSwitch=false, bIsVisible=true, type {classStructOrEnum {refPath=&NiagaraInt32}}}
      Spawn Time {bIsDynamic=false, bIsEditable=true, bIsStaticSwitch=false, bIsVisible=true, type {classStructOrEnum {refPath=&NiagaraFloat}}}
      Use Spawn Probability {bIsDynamic=false, bIsEditable=true, bIsStaticSwitch=false, bIsVisible=true, type {classStructOrEnum {refPath=&NiagaraBool}}}
      Spawn Probability {bIsDynamic=false, bIsEditable=false, bIsStaticSwitch=false, bIsVisible=true, type {classStructOrEnum {refPath=&NiagaraFloat}}}
      UseLoopCountLimit {bIsDynamic=false, bIsEditable=true, bIsStaticSwitch=false, bIsVisible=true, type {classStructOrEnum {refPath=&NiagaraBool}}}
      Loop Count Limit {bIsDynamic=false, bIsEditable=false, bIsStaticSwitch=false, bIsVisible=true, type {classStructOrEnum {refPath=&NiagaraInt32}}}
      Spawn Group {bIsDynamic=false, bIsEditable=true, bIsStaticSwitch=false, bIsVisible=true, type {classStructOrEnum {refPath=&NiagaraInt32}}}
      Age {bIsDynamic=false, bIsEditable=true, bIsStaticSwitch=false, bIsVisible=true, type {classStructOrEnum {refPath=&NiagaraFloat}}}
     ]
     moduleScript {refPath=&SpawnBurst_Instantaneous}
    }
    {
     bIsSetParametersModule=false
     enabled=true
     moduleName=SpawnBurst_Instantaneous004
     inputs [
      Spawn Count {bIsDynamic=false, bIsEditable=true, bIsStaticSwitch=false, bIsVisible=true, type {classStructOrEnum {refPath=&NiagaraInt32}}}
      Spawn Time {bIsDynamic=false, bIsEditable=true, bIsStaticSwitch=false, bIsVisible=true, type {classStructOrEnum {refPath=&NiagaraFloat}}}
      Use Spawn Probability {bIsDynamic=false, bIsEditable=true, bIsStaticSwitch=false, bIsVisible=true, type {classStructOrEnum {refPath=&NiagaraBool}}}
      Spawn Probability {bIsDynamic=false, bIsEditable=false, bIsStaticSwitch=false, bIsVisible=true, type {classStructOrEnum {refPath=&NiagaraFloat}}}
      UseLoopCountLimit {bIsDynamic=false, bIsEditable=true, bIsStaticSwitch=false, bIsVisible=true, type {classStructOrEnum {refPath=&NiagaraBool}}}
      Loop Count Limit {bIsDynamic=false, bIsEditable=false, bIsStaticSwitch=false, bIsVisible=true, type {classStructOrEnum {refPath=&NiagaraInt32}}}
      Spawn Group {bIsDynamic=false, bIsEditable=true, bIsStaticSwitch=false, bIsVisible=true, type {classStructOrEnum {refPath=&NiagaraInt32}}}
      Age {bIsDynamic=false, bIsEditable=true, bIsStaticSwitch=false, bIsVisible=true, type {classStructOrEnum {refPath=&NiagaraFloat}}}
     ]
     moduleScript {refPath=&SpawnBurst_Instantaneous}
    }
    {
     bIsSetParametersModule=false
     enabled=true
     moduleName=SpawnBurst_Instantaneous005
     inputs [
      Spawn Count {bIsDynamic=false, bIsEditable=true, bIsStaticSwitch=false, bIsVisible=true, type {classStructOrEnum {refPath=&NiagaraInt32}}}
      Spawn Time {bIsDynamic=false, bIsEditable=true, bIsStaticSwitch=false, bIsVisible=true, type {classStructOrEnum {refPath=&NiagaraFloat}}}
      Use Spawn Probability {bIsDynamic=false, bIsEditable=true, bIsStaticSwitch=false, bIsVisible=true, type {classStructOrEnum {refPath=&NiagaraBool}}}
      Spawn Probability {bIsDynamic=false, bIsEditable=false, bIsStaticSwitch=false, bIsVisible=true, type {classStructOrEnum {refPath=&NiagaraFloat}}}
      UseLoopCountLimit {bIsDynamic=false, bIsEditable=true, bIsStaticSwitch=false, bIsVisible=true, type {classStructOrEnum {refPath=&NiagaraBool}}}
      Loop Count Limit {bIsDynamic=false, bIsEditable=false, bIsStaticSwitch=false, bIsVisible=true, type {classStructOrEnum {refPath=&NiagaraInt32}}}
      Spawn Group {bIsDynamic=false, bIsEditable=true, bIsStaticSwitch=false, bIsVisible=true, type {classStructOrEnum {refPath=&NiagaraInt32}}}
      Age {bIsDynamic=false, bIsEditable=true, bIsStaticSwitch=false, bIsVisible=true, type {classStructOrEnum {refPath=&NiagaraFloat}}}
     ]
     moduleScript {refPath=&SpawnBurst_Instantaneous}
    }
   ]
  }
  particleSpawnScript {
   scriptName=ParticleSpawnScript
   modules [
    {
     bIsSetParametersModule=true
     enabled=true
     moduleName=SetVariables_#1
     inputs [Particles.VisibilityTag {bIsDynamic=true, bIsEditable=true, bIsStaticSwitch=false, bIsVisible=true, type {classStructOrEnum {refPath=&NiagaraInt32}}}]
     moduleScript {refPath=&self:DefaultPropSpawn_Merged002_0.NiagaraScriptSource_1.NiagaraGraph_0.NiagaraNodeAssignment_1.SetVariables_#1}
    }
    {
     bIsSetParametersModule=false
     enabled=true
     moduleName=InitializeParticle001
     inputs [
      Lifetime Mode {
       bIsDynamic=false
       bIsEditable=true
       bIsStaticSwitch=true
       bIsVisible=true
       type {classStructOrEnum {refPath=&ENiagara_LifetimeMode}}
      }
      Lifetime Min {
       bIsDynamic=false
       bIsEditable=false
       bIsStaticSwitch=false
       bIsVisible=false
       type {classStructOrEnum {refPath=&NiagaraFloat}}
      }
      Lifetime {bIsDynamic=true, bIsEditable=true, bIsStaticSwitch=false, bIsVisible=true, type {classStructOrEnum {refPath=&NiagaraFloat}}}
      Lifetime Max {
       bIsDynamic=false
       bIsEditable=false
       bIsStaticSwitch=false
       bIsVisible=false
       type {classStructOrEnum {refPath=&NiagaraFloat}}
      }
      Lifetime Randomness Mode {
       bIsDynamic=false
       bIsEditable=false
       bIsStaticSwitch=true
       bIsVisible=false
       type {classStructOrEnum {refPath=&ENiagaraRandomnessMode}}
      }
      Color Mode {
       bIsDynamic=false
       bIsEditable=true
       bIsStaticSwitch=true
       bIsVisible=true
       type {classStructOrEnum {refPath=&ENiagara_ColorInitializationMode}}
      }
      Color Minimum {
       bIsDynamic=false
       bIsEditable=false
       bIsStaticSwitch=false
       bIsVisible=false
       type {classStructOrEnum {refPath=&LinearColor}}
      }
      Color {
       bIsDynamic=true
       bIsEditable=true
       bIsStaticSwitch=false
       bIsVisible=true
       type {classStructOrEnum {refPath=&LinearColor}}
      }
      AdjustHue {bIsDynamic=false, bIsEditable=false, bIsStaticSwitch=false, bIsVisible=false, type {classStructOrEnum {refPath=&NiagaraBool}}}
      Hue Shift Range {
       bIsDynamic=false
       bIsEditable=false
       bIsStaticSwitch=false
       bIsVisible=false
       type {classStructOrEnum {refPath=&Vector2f}}
      }
      Color Maximum {
       bIsDynamic=false
       bIsEditable=false
       bIsStaticSwitch=false
       bIsVisible=false
       type {classStructOrEnum {refPath=&LinearColor}}
      }
      AdjustSaturation {bIsDynamic=false, bIsEditable=false, bIsStaticSwitch=false, bIsVisible=false, type {classStructOrEnum {refPath=&NiagaraBool}}}
      Saturation Range {
       bIsDynamic=false
       bIsEditable=false
       bIsStaticSwitch=false
       bIsVisible=false
       type {classStructOrEnum {refPath=&Vector2f}}
      }
      AdjustValue {bIsDynamic=false, bIsEditable=false, bIsStaticSwitch=false, bIsVisible=false, type {classStructOrEnum {refPath=&NiagaraBool}}}
      Value Range {
       bIsDynamic=false
       bIsEditable=false
       bIsStaticSwitch=false
       bIsVisible=false
       type {classStructOrEnum {refPath=&Vector2f}}
      }
      AdjustAlpha {bIsDynamic=false, bIsEditable=false, bIsStaticSwitch=false, bIsVisible=false, type {classStructOrEnum {refPath=&NiagaraBool}}}
      Alpha Scale Range {
       bIsDynamic=false
       bIsEditable=false
       bIsStaticSwitch=false
       bIsVisible=false
       type {classStructOrEnum {refPath=&Vector2f}}
      }
      Color Channel Mode {
       bIsDynamic=false
       bIsEditable=false
       bIsStaticSwitch=true
       bIsVisible=false
       type {classStructOrEnum {refPath=&ENiagaraChannelCorrelation}}
      }
      Color Randomness Mode {
       bIsDynamic=false
       bIsEditable=false
       bIsStaticSwitch=true
       bIsVisible=false
       type {classStructOrEnum {refPath=&ENiagaraRandomnessMode}}
      }
      Position Mode {
       bIsDynamic=false
       bIsEditable=true
       bIsStaticSwitch=true
       bIsVisible=true
       type {classStructOrEnum {refPath=&ENiagara_PositionInitializationMode}}
      }
      Position {
       bIsDynamic=false
       bIsEditable=false
       bIsStaticSwitch=false
       bIsVisible=false
       type {classStructOrEnum {refPath=&NiagaraPosition}}
      }
      UsePositionOffset {bIsDynamic=false, bIsEditable=true, bIsStaticSwitch=false, bIsVisible=true, type {classStructOrEnum {refPath=&NiagaraBool}}}
      Position Offset {bIsDynamic=false, bIsEditable=true, bIsStaticSwitch=false, bIsVisible=true, type {classStructOrEnum {refPath=&Vector3f}}}
      Position Offset Coordinate Space {
       bIsDynamic=false
       bIsEditable=true
       bIsStaticSwitch=false
       bIsVisible=true
       type {classStructOrEnum {refPath=&ENiagaraCoordinateSpace}}
      }
      Mass Mode {
       bIsDynamic=false
       bIsEditable=true
       bIsStaticSwitch=true
       bIsVisible=true
       type {classStructOrEnum {refPath=&ENiagara_MassInitializationMode}}
      }
      Mass Min {
       bIsDynamic=false
       bIsEditable=false
       bIsStaticSwitch=false
       bIsVisible=false
       type {classStructOrEnum {refPath=&NiagaraFloat}}
      }
      Mass {
       bIsDynamic=false
       bIsEditable=false
       bIsStaticSwitch=false
       bIsVisible=false
       type {classStructOrEnum {refPath=&NiagaraFloat}}
      }
      Mass Max {
       bIsDynamic=false
       bIsEditable=false
       bIsStaticSwitch=false
       bIsVisible=false
       type {classStructOrEnum {refPath=&NiagaraFloat}}
      }
      Mass Randomness Mode {
       bIsDynamic=false
       bIsEditable=false
       bIsStaticSwitch=true
       bIsVisible=false
       type {classStructOrEnum {refPath=&ENiagaraRandomnessMode}}
      }
      Write Material Random {bIsDynamic=false, bIsEditable=true, bIsStaticSwitch=false, bIsVisible=true, type {classStructOrEnum {refPath=&NiagaraBool}}}
      Material Random {bIsDynamic=false, bIsEditable=true, bIsStaticSwitch=false, bIsVisible=true, type {classStructOrEnum {refPath=&NiagaraFloat}}}
      Sprite Size Mode {
       bIsDynamic=false
       bIsEditable=true
       bIsStaticSwitch=true
       bIsVisible=true
       type {classStructOrEnum {refPath=&ENiagara_SizeScaleMode}}
      }
      Sprite Size {
       bIsDynamic=false
       bIsEditable=false
       bIsStaticSwitch=false
       bIsVisible=false
       type {classStructOrEnum {refPath=&Vector2f}}
      }
      Uniform Sprite Size {bIsDynamic=true, bIsEditable=true, bIsStaticSwitch=false, bIsVisible=true, type {classStructOrEnum {refPath=&NiagaraFloat}}}
      Uniform Sprite Size Min {
       bIsDynamic=false
       bIsEditable=false
       bIsStaticSwitch=false
       bIsVisible=false
       type {classStructOrEnum {refPath=&NiagaraFloat}}
      }
      Sprite Size Min {
       bIsDynamic=false
       bIsEditable=false
       bIsStaticSwitch=false
       bIsVisible=false
       type {classStructOrEnum {refPath=&Vector2f}}
      }
      Sprite Size Max {
       bIsDynamic=false
       bIsEditable=false
       bIsStaticSwitch=false
       bIsVisible=false
       type {classStructOrEnum {refPath=&Vector2f}}
      }
      Uniform Sprite Size Max {
       bIsDynamic=false
       bIsEditable=false
       bIsStaticSwitch=false
       bIsVisible=false
       type {classStructOrEnum {refPath=&NiagaraFloat}}
      }
      Sprite Size Randomness Mode {
       bIsDynamic=false
       bIsEditable=false
       bIsStaticSwitch=true
       bIsVisible=false
       type {classStructOrEnum {refPath=&ENiagaraRandomnessMode}}
      }
      Sprite Rotation Mode {
       bIsDynamic=false
       bIsEditable=true
       bIsStaticSwitch=true
       bIsVisible=true
       type {classStructOrEnum {refPath=&ENiagara_SpriteRotationMode}}
      }
      Sprite Rotation Angle Min {
       bIsDynamic=false
       bIsEditable=false
       bIsStaticSwitch=false
       bIsVisible=false
       type {classStructOrEnum {refPath=&NiagaraFloat}}
      }
      Sprite Rotation Angle {bIsDynamic=true, bIsEditable=true, bIsStaticSwitch=false, bIsVisible=true, type {classStructOrEnum {refPath=&NiagaraFloat}}}
      Sprite Rotation Angle Max {
       bIsDynamic=false
       bIsEditable=false
       bIsStaticSwitch=false
       bIsVisible=false
       type {classStructOrEnum {refPath=&NiagaraFloat}}
      }
      Sprite Rotation Randomness Mode {
       bIsDynamic=false
       bIsEditable=false
       bIsStaticSwitch=true
       bIsVisible=false
       type {classStructOrEnum {refPath=&ENiagaraRandomnessMode}}
      }
      Sprite UV Mode {
       bIsDynamic=false
       bIsEditable=true
       bIsStaticSwitch=true
       bIsVisible=true
       type {classStructOrEnum {refPath=&ENiagara_UVFlippingMode}}
      }
      Sprite UV Randomness Mode {
       bIsDynamic=false
       bIsEditable=false
       bIsStaticSwitch=true
       bIsVisible=false
       type {classStructOrEnum {refPath=&ENiagaraRandomnessMode}}
      }
      Sprite UV Scale {
       bIsDynamic=false
       bIsEditable=false
       bIsStaticSwitch=false
       bIsVisible=false
       type {classStructOrEnum {refPath=&Vector2f}}
      }
      Mesh Scale Mode {
       bIsDynamic=false
       bIsEditable=true
       bIsStaticSwitch=true
       bIsVisible=true
       type {classStructOrEnum {refPath=&ENiagara_SizeScaleMode}}
      }
      Mesh Uniform Scale Min {
       bIsDynamic=false
       bIsEditable=false
       bIsStaticSwitch=false
       bIsVisible=false
       type {classStructOrEnum {refPath=&NiagaraFloat}}
      }
      Mesh Uniform Scale {bIsDynamic=true, bIsEditable=true, bIsStaticSwitch=false, bIsVisible=true, type {classStructOrEnum {refPath=&NiagaraFloat}}}
      Mesh Scale Min {
       bIsDynamic=false
       bIsEditable=false
       bIsStaticSwitch=false
       bIsVisible=false
       type {classStructOrEnum {refPath=&Vector3f}}
      }
      Mesh Scale {
       bIsDynamic=false
       bIsEditable=false
       bIsStaticSwitch=false
       bIsVisible=false
       type {classStructOrEnum {refPath=&Vector3f}}
      }
      Mesh Scale Max {
       bIsDynamic=false
       bIsEditable=false
       bIsStaticSwitch=false
       bIsVisible=false
       type {classStructOrEnum {refPath=&Vector3f}}
      }
      Mesh Uniform Scale Max {
       bIsDynamic=false
       bIsEditable=false
       bIsStaticSwitch=false
       bIsVisible=false
       type {classStructOrEnum {refPath=&NiagaraFloat}}
      }
      Mesh Scale Randomness Mode {
       bIsDynamic=false
       bIsEditable=false
       bIsStaticSwitch=true
       bIsVisible=false
       type {classStructOrEnum {refPath=&ENiagaraRandomnessMode}}
      }
      Mesh Renderer Array Visibility Mode {
       bIsDynamic=false
       bIsEditable=true
       bIsStaticSwitch=true
       bIsVisible=true
       type {classStructOrEnum {refPath=&ENiagara_UnsetDirectSetRandom}}
      }
      Mesh Renderer Mesh Index {
       bIsDynamic=false
       bIsEditable=false
       bIsStaticSwitch=false
       bIsVisible=false
       type {classStructOrEnum {refPath=&NiagaraInt32}}
      }
      Wrap Index By Mesh Count {bIsDynamic=false, bIsEditable=false, bIsStaticSwitch=true, bIsVisible=false, type {classStructOrEnum {refPath=&NiagaraBool}}}
      Mesh Renderer Info {
       bIsDynamic=false
       bIsEditable=false
       bIsStaticSwitch=false
       bIsVisible=false
       type {classStructOrEnum {refPath=&NiagaraDataInterfaceMeshRendererInfo}}
      }
      Mesh Renderer Random Seed {
       bIsDynamic=false
       bIsEditable=false
       bIsStaticSwitch=false
       bIsVisible=false
       type {classStructOrEnum {refPath=&NiagaraInt32}}
      }
      Ribbon Width Mode {
       bIsDynamic=false
       bIsEditable=true
       bIsStaticSwitch=true
       bIsVisible=true
       type {classStructOrEnum {refPath=&ENiagara_UnsetDirectSet}}
      }
      Ribbon Width {
       bIsDynamic=false
       bIsEditable=false
       bIsStaticSwitch=false
       bIsVisible=false
       type {classStructOrEnum {refPath=&NiagaraFloat}}
      }
      Ribbon Facing Mode {
       bIsDynamic=false
       bIsEditable=true
       bIsStaticSwitch=true
       bIsVisible=true
       type {classStructOrEnum {refPath=&ENiagara_UnsetDirectSet}}
      }
      Ribbon Facing Vector {
       bIsDynamic=false
       bIsEditable=false
       bIsStaticSwitch=false
       bIsVisible=false
       type {classStructOrEnum {refPath=&Vector3f}}
      }
      Ribbon Twist Mode {
       bIsDynamic=false
       bIsEditable=true
       bIsStaticSwitch=true
       bIsVisible=true
       type {classStructOrEnum {refPath=&ENiagara_UnsetDirectSet}}
      }
      "Ribbon Twist " {
       bIsDynamic=false
       bIsEditable=false
       bIsStaticSwitch=false
       bIsVisible=false
       type {classStructOrEnum {refPath=&NiagaraFloat}}
      }
     ]
     moduleScript {refPath=&InitializeParticle}
    }
    {
     bIsSetParametersModule=true
     enabled=true
     moduleName=SetVariables_#2
     inputs [Particles.InitSpriteRotRate {bIsDynamic=true, bIsEditable=true, bIsStaticSwitch=false, bIsVisible=true, type {classStructOrEnum {refPath=&NiagaraFloat}}}]
     moduleScript {refPath=&self:DefaultPropSpawn_Merged002_0.NiagaraScriptSource_1.NiagaraGraph_0.NiagaraNodeAssignment_0.SetVariables_#2}
    }
    {
     bIsSetParametersModule=false
     enabled=true
     moduleName=AccelerationForce
     inputs [
      Acceleration {bIsDynamic=true, bIsEditable=true, bIsStaticSwitch=false, bIsVisible=true, type {classStructOrEnum {refPath=&Vector3f}}}
      Coordinate Space {
       bIsDynamic=false
       bIsEditable=true
       bIsStaticSwitch=false
       bIsVisible=true
       type {classStructOrEnum {refPath=&ENiagaraCoordinateSpace}}
      }
     ]
     moduleScript {refPath=&AccelerationForce}
    }
    {
     bIsSetParametersModule=false
     enabled=true
     moduleName=ShapeLocation
     inputs [
      Shape Primitive {
       bIsDynamic=false
       bIsEditable=true
       bIsStaticSwitch=true
       bIsVisible=true
       type {classStructOrEnum {refPath=&ENiagara_LocationShapes}}
      }
      Box / Plane Mode {
       bIsDynamic=false
       bIsEditable=false
       bIsStaticSwitch=true
       bIsVisible=false
       type {classStructOrEnum {refPath=&ENiagara_BoxPlaneMode}}
      }
      Cone Mode {
       bIsDynamic=false
       bIsEditable=false
       bIsStaticSwitch=true
       bIsVisible=false
       type {classStructOrEnum {refPath=&ENiagara_ConeMode}}
      }
      Cylinder Height {
       bIsDynamic=false
       bIsEditable=false
       bIsStaticSwitch=false
       bIsVisible=false
       type {classStructOrEnum {refPath=&NiagaraFloat}}
      }
      Sphere Radius {bIsDynamic=true, bIsEditable=true, bIsStaticSwitch=false, bIsVisible=true, type {classStructOrEnum {refPath=&NiagaraFloat}}}
      Ring / Disc Mode {
       bIsDynamic=false
       bIsEditable=false
       bIsStaticSwitch=true
       bIsVisible=false
       type {classStructOrEnum {refPath=&ENiagaraRingDiscMode}}
      }
      Torus Mode {
       bIsDynamic=false
       bIsEditable=false
       bIsStaticSwitch=true
       bIsVisible=false
       type {classStructOrEnum {refPath=&ENiagaraShapeTorusMode}}
      }
      Knot Ring Loops {
       bIsDynamic=false
       bIsEditable=false
       bIsStaticSwitch=false
       bIsVisible=false
       type {classStructOrEnum {refPath=&NiagaraInt32}}
      }
      Knot Helix Loops {
       bIsDynamic=false
       bIsEditable=false
       bIsStaticSwitch=false
       bIsVisible=false
       type {classStructOrEnum {refPath=&NiagaraInt32}}
      }
      Enable Torus Knot Radius {bIsDynamic=false, bIsEditable=false, bIsStaticSwitch=false, bIsVisible=false, type {classStructOrEnum {refPath=&NiagaraBool}}}
      Torus Knot Tube Radius {
       bIsDynamic=false
       bIsEditable=false
       bIsStaticSwitch=false
       bIsVisible=false
       type {classStructOrEnum {refPath=&NiagaraFloat}}
      }
      Cylinder Radius {
       bIsDynamic=false
       bIsEditable=false
       bIsStaticSwitch=false
       bIsVisible=false
       type {classStructOrEnum {refPath=&NiagaraFloat}}
      }
      Cylinder Height Midpoint {
       bIsDynamic=false
       bIsEditable=false
       bIsStaticSwitch=false
       bIsVisible=false
       type {classStructOrEnum {refPath=&NiagaraFloat}}
      }
      Ring Radius {
       bIsDynamic=false
       bIsEditable=false
       bIsStaticSwitch=false
       bIsVisible=false
       type {classStructOrEnum {refPath=&NiagaraFloat}}
      }
      Box Size {
       bIsDynamic=false
       bIsEditable=false
       bIsStaticSwitch=false
       bIsVisible=false
       type {classStructOrEnum {refPath=&Vector3f}}
      }
      Box Midpoint {
       bIsDynamic=false
       bIsEditable=false
       bIsStaticSwitch=false
       bIsVisible=false
       type {classStructOrEnum {refPath=&Vector3f}}
      }
      Cone Length {
       bIsDynamic=false
       bIsEditable=false
       bIsStaticSwitch=false
       bIsVisible=false
       type {classStructOrEnum {refPath=&NiagaraFloat}}
      }
      Plane Size {
       bIsDynamic=false
       bIsEditable=false
       bIsStaticSwitch=false
       bIsVisible=false
       type {classStructOrEnum {refPath=&Vector2f}}
      }
      Wedge Horizontal Angle {
       bIsDynamic=false
       bIsEditable=false
       bIsStaticSwitch=false
       bIsVisible=false
       type {classStructOrEnum {refPath=&NiagaraFloat}}
      }
      Large Radius {
       bIsDynamic=false
       bIsEditable=false
       bIsStaticSwitch=false
       bIsVisible=false
       type {classStructOrEnum {refPath=&NiagaraFloat}}
      }
      Plane Midpoint {
       bIsDynamic=false
       bIsEditable=false
       bIsStaticSwitch=false
       bIsVisible=false
       type {classStructOrEnum {refPath=&Vector2f}}
      }
      Wedge Vertical Angle {
       bIsDynamic=false
       bIsEditable=false
       bIsStaticSwitch=false
       bIsVisible=false
       type {classStructOrEnum {refPath=&NiagaraFloat}}
      }
      Cone Angle {
       bIsDynamic=false
       bIsEditable=false
       bIsStaticSwitch=false
       bIsVisible=false
       type {classStructOrEnum {refPath=&NiagaraFloat}}
      }
      Handle Radius {
       bIsDynamic=false
       bIsEditable=false
       bIsStaticSwitch=false
       bIsVisible=false
       type {classStructOrEnum {refPath=&NiagaraFloat}}
      }
      Cone Inner Angle {
       bIsDynamic=false
       bIsEditable=false
       bIsStaticSwitch=false
       bIsVisible=false
       type {classStructOrEnum {refPath=&NiagaraFloat}}
      }
      Cone Radial Angle {
       bIsDynamic=false
       bIsEditable=false
       bIsStaticSwitch=false
       bIsVisible=false
       type {classStructOrEnum {refPath=&NiagaraFloat}}
      }
      Cone Angle Type {
       bIsDynamic=false
       bIsEditable=false
       bIsStaticSwitch=true
       bIsVisible=false
       type {classStructOrEnum {refPath=&ENiagara_AngleInput}}
      }
      Flatten Endcaps {bIsDynamic=false, bIsEditable=false, bIsStaticSwitch=true, bIsVisible=false, type {classStructOrEnum {refPath=&NiagaraBool}}}
      Apply To Particle Position {
       bIsDynamic=false
       bIsEditable=true
       bIsStaticSwitch=true
       bIsVisible=true
       type {classStructOrEnum {refPath=&ENiagara_AttributeSamplingApplyOutput}}
      }
      Cone Surface Distribution {
       bIsDynamic=false
       bIsEditable=false
       bIsStaticSwitch=false
       bIsVisible=false
       type {classStructOrEnum {refPath=&NiagaraFloat}}
      }
      Ring / Disc Distribution Mode {
       bIsDynamic=false
       bIsEditable=false
       bIsStaticSwitch=true
       bIsVisible=false
       type {classStructOrEnum {refPath=&ENiagaraSphereDistributionMode}}
      }
      Sphere Distribution {
       bIsDynamic=false
       bIsEditable=true
       bIsStaticSwitch=true
       bIsVisible=true
       type {classStructOrEnum {refPath=&ENiagaraSphereDistributionMode}}
      }
      Torus Distribution Mode {
       bIsDynamic=false
       bIsEditable=false
       bIsStaticSwitch=true
       bIsVisible=false
       type {classStructOrEnum {refPath=&ENiagaraTorusDistributionMode}}
      }
      Sphere Surface Distribution {bIsDynamic=false, bIsEditable=true, bIsStaticSwitch=false, bIsVisible=true, type {classStructOrEnum {refPath=&NiagaraFloat}}}
      Hemisphere Distribution {bIsDynamic=false, bIsEditable=true, bIsStaticSwitch=false, bIsVisible=true, type {classStructOrEnum {refPath=&Vector2f}}}
      Cylinder Distribution {
       bIsDynamic=false
       bIsEditable=false
       bIsStaticSwitch=true
       bIsVisible=false
       type {classStructOrEnum {refPath=&ENiagara_CylinderMode}}
      }
      Surface Only {bIsDynamic=false, bIsEditable=false, bIsStaticSwitch=false, bIsVisible=false, type {classStructOrEnum {refPath=&NiagaraBool}}}
      Surface Only Band Thickness {
       bIsDynamic=false
       bIsEditable=false
       bIsStaticSwitch=false
       bIsVisible=false
       type {classStructOrEnum {refPath=&NiagaraFloat}}
      }
      Use Endcaps In Surface Only Mode {bIsDynamic=false, bIsEditable=false, bIsStaticSwitch=true, bIsVisible=false, type {classStructOrEnum {refPath=&NiagaraBool}}}
      Hemicircle X {bIsDynamic=false, bIsEditable=false, bIsStaticSwitch=true, bIsVisible=false, type {classStructOrEnum {refPath=&NiagaraBool}}}
      Hemicircle Y {bIsDynamic=false, bIsEditable=false, bIsStaticSwitch=true, bIsVisible=false, type {classStructOrEnum {refPath=&NiagaraBool}}}
      Disc Coverage {
       bIsDynamic=false
       bIsEditable=false
       bIsStaticSwitch=false
       bIsVisible=false
       type {classStructOrEnum {refPath=&NiagaraFloat}}
      }
      Hemicircle Internal Cap {bIsDynamic=false, bIsEditable=false, bIsStaticSwitch=true, bIsVisible=false, type {classStructOrEnum {refPath=&NiagaraBool}}}
      U Position {
       bIsDynamic=false
       bIsEditable=false
       bIsStaticSwitch=false
       bIsVisible=false
       type {classStructOrEnum {refPath=&NiagaraFloat}}
      }
      Enable Lathe Profile {bIsDynamic=false, bIsEditable=false, bIsStaticSwitch=true, bIsVisible=false, type {classStructOrEnum {refPath=&NiagaraBool}}}
      Lathe Profile {
       bIsDynamic=false
       bIsEditable=false
       bIsStaticSwitch=false
       bIsVisible=false
       type {classStructOrEnum {refPath=&NiagaraDataInterfaceCurve}}
      }
      Surface Distribution {
       bIsDynamic=false
       bIsEditable=false
       bIsStaticSwitch=false
       bIsVisible=false
       type {classStructOrEnum {refPath=&NiagaraFloat}}
      }
      Surface Expansion Mode {
       bIsDynamic=false
       bIsEditable=false
       bIsStaticSwitch=false
       bIsVisible=false
       type {classStructOrEnum {refPath=&ENiagaraExpansionMode}}
      }
      U Distribution {
       bIsDynamic=false
       bIsEditable=false
       bIsStaticSwitch=false
       bIsVisible=false
       type {classStructOrEnum {refPath=&NiagaraFloat}}
      }
      V Position {
       bIsDynamic=false
       bIsEditable=false
       bIsStaticSwitch=false
       bIsVisible=false
       type {classStructOrEnum {refPath=&NiagaraFloat}}
      }
      Radius Position {
       bIsDynamic=false
       bIsEditable=false
       bIsStaticSwitch=false
       bIsVisible=false
       type {classStructOrEnum {refPath=&NiagaraFloat}}
      }
      Sphere Radius Position {
       bIsDynamic=false
       bIsEditable=false
       bIsStaticSwitch=false
       bIsVisible=false
       type {classStructOrEnum {refPath=&NiagaraFloat}}
      }
      V Distribution {
       bIsDynamic=false
       bIsEditable=false
       bIsStaticSwitch=false
       bIsVisible=false
       type {classStructOrEnum {refPath=&NiagaraFloat}}
      }
      Evaluation Type {
       bIsDynamic=false
       bIsEditable=false
       bIsStaticSwitch=true
       bIsVisible=false
       type {classStructOrEnum {refPath=&ENiagaraRandomnessEvaluation}}
      }
      Uniform Distribution {
       bIsDynamic=false
       bIsEditable=false
       bIsStaticSwitch=false
       bIsVisible=false
       type {classStructOrEnum {refPath=&NiagaraFloat}}
      }
      Uniform Spiral Amount {
       bIsDynamic=false
       bIsEditable=false
       bIsStaticSwitch=false
       bIsVisible=false
       type {classStructOrEnum {refPath=&NiagaraFloat}}
      }
      Uniform Spiral Falloff {
       bIsDynamic=false
       bIsEditable=false
       bIsStaticSwitch=false
       bIsVisible=false
       type {classStructOrEnum {refPath=&NiagaraFloat}}
      }
      Uniform Count {
       bIsDynamic=false
       bIsEditable=false
       bIsStaticSwitch=false
       bIsVisible=false
       type {classStructOrEnum {refPath=&NiagaraInt32}}
      }
      Uniform Particle Index {
       bIsDynamic=false
       bIsEditable=false
       bIsStaticSwitch=false
       bIsVisible=false
       type {classStructOrEnum {refPath=&NiagaraInt32}}
      }
      Hemisphere Angle Type {
       bIsDynamic=false
       bIsEditable=true
       bIsStaticSwitch=true
       bIsVisible=true
       type {classStructOrEnum {refPath=&ENiagara_AngleInput}}
      }
      Uniform Endpoint Offset {
       bIsDynamic=false
       bIsEditable=false
       bIsStaticSwitch=false
       bIsVisible=false
       type {classStructOrEnum {refPath=&NiagaraFloat}}
      }
      MaskSpawnGroup {bIsDynamic=false, bIsEditable=true, bIsStaticSwitch=false, bIsVisible=true, type {classStructOrEnum {refPath=&NiagaraBool}}}
      Spawn Group Mask {bIsDynamic=false, bIsEditable=false, bIsStaticSwitch=false, bIsVisible=true, type {classStructOrEnum {refPath=&NiagaraInt32}}}
      Spawn Group {
       bIsDynamic=false
       bIsEditable=false
       bIsStaticSwitch=false
       bIsVisible=false
       type {classStructOrEnum {refPath=&NiagaraInt32}}
      }
      Randomness Mode {
       bIsDynamic=false
       bIsEditable=true
       bIsStaticSwitch=true
       bIsVisible=true
       type {classStructOrEnum {refPath=&ENiagaraRandomnessMode}}
      }
      Override Seed {bIsDynamic=false, bIsEditable=true, bIsStaticSwitch=false, bIsVisible=true, type {classStructOrEnum {refPath=&NiagaraBool}}}
      Random Seed {bIsDynamic=false, bIsEditable=false, bIsStaticSwitch=false, bIsVisible=true, type {classStructOrEnum {refPath=&NiagaraInt32}}}
      Fixed Random Seed {bIsDynamic=false, bIsEditable=false, bIsStaticSwitch=true, bIsVisible=false, type {classStructOrEnum {refPath=&NiagaraBool}}}
      Transform Order {
       bIsDynamic=false
       bIsEditable=true
       bIsStaticSwitch=true
       bIsVisible=true
       type {classStructOrEnum {refPath=&ENiagara_TransformOrder}}
      }
      Scale Mode {
       bIsDynamic=false
       bIsEditable=true
       bIsStaticSwitch=true
       bIsVisible=true
       type {classStructOrEnum {refPath=&ENiagara_ScaleMode}}
      }
      Non Uniform Scale {bIsDynamic=false, bIsEditable=true, bIsStaticSwitch=false, bIsVisible=true, type {classStructOrEnum {refPath=&Vector3f}}}
      UseOwnerScale {bIsDynamic=false, bIsEditable=true, bIsStaticSwitch=false, bIsVisible=true, type {classStructOrEnum {refPath=&NiagaraBool}}}
      Apply Owner Scale {bIsDynamic=false, bIsEditable=true, bIsStaticSwitch=false, bIsVisible=true, type {classStructOrEnum {refPath=&Vector3f}}}
      Rotation Mode {
       bIsDynamic=false
       bIsEditable=true
       bIsStaticSwitch=true
       bIsVisible=true
       type {classStructOrEnum {refPath=&ENiagara_RotationMode}}
      }
      Rotation Axis {
       bIsDynamic=false
       bIsEditable=false
       bIsStaticSwitch=false
       bIsVisible=false
       type {classStructOrEnum {refPath=&Vector3f}}
      }
      Rotation Matrix {
       bIsDynamic=false
       bIsEditable=false
       bIsStaticSwitch=false
       bIsVisible=false
       type {classStructOrEnum {refPath=&NiagaraMatrix}}
      }
      Rotation Quaternion {bIsDynamic=false, bIsEditable=false, bIsStaticSwitch=false, bIsVisible=false, type {classStructOrEnum {refPath=&Quat4f}}}
      Yaw / Pitch / Roll {
       bIsDynamic=false
       bIsEditable=false
       bIsStaticSwitch=false
       bIsVisible=false
       type {classStructOrEnum {refPath=&Vector3f}}
      }
      Rotation Angle {
       bIsDynamic=false
       bIsEditable=false
       bIsStaticSwitch=false
       bIsVisible=false
       type {classStructOrEnum {refPath=&NiagaraFloat}}
      }
      Invert Rotation Quaternion {bIsDynamic=false, bIsEditable=false, bIsStaticSwitch=false, bIsVisible=false, type {classStructOrEnum {refPath=&NiagaraBool}}}
      Additional Quaternion Rotation {bIsDynamic=false, bIsEditable=false, bIsStaticSwitch=false, bIsVisible=false, type {classStructOrEnum {refPath=&NiagaraBool}}}
      Additional Yaw / Pitch / Roll {
       bIsDynamic=false
       bIsEditable=false
       bIsStaticSwitch=false
       bIsVisible=false
       type {classStructOrEnum {refPath=&Vector3f}}
      }
      Rotation Angle Type {
       bIsDynamic=false
       bIsEditable=false
       bIsStaticSwitch=true
       bIsVisible=false
       type {classStructOrEnum {refPath=&ENiagara_AngleInput}}
      }
      Rotation Coordinate Space {
       bIsDynamic=false
       bIsEditable=true
       bIsStaticSwitch=false
       bIsVisible=true
       type {classStructOrEnum {refPath=&ENiagaraCoordinateSpace}}
      }
      Offset Mode {
       bIsDynamic=false
       bIsEditable=true
       bIsStaticSwitch=true
       bIsVisible=true
       type {classStructOrEnum {refPath=&ENiagara_OffsetMode}}
      }
      Shape Origin {
       bIsDynamic=false
       bIsEditable=true
       bIsStaticSwitch=false
       bIsVisible=true
       type {classStructOrEnum {refPath=&NiagaraPosition}}
      }
      Offset {bIsDynamic=false, bIsEditable=true, bIsStaticSwitch=false, bIsVisible=true, type {classStructOrEnum {refPath=&Vector3f}}}
      Offset Coordinate Space {
       bIsDynamic=false
       bIsEditable=true
       bIsStaticSwitch=false
       bIsVisible=true
       type {classStructOrEnum {refPath=&ENiagaraCoordinateSpace}}
      }
      Custom Transform Matrix {
       bIsDynamic=false
       bIsEditable=false
       bIsStaticSwitch=false
       bIsVisible=false
       type {classStructOrEnum {refPath=&NiagaraMatrix}}
      }
      Transform Method {
       bIsDynamic=false
       bIsEditable=true
       bIsStaticSwitch=true
       bIsVisible=true
       type {classStructOrEnum {refPath=&ENiagara_TransformType}}
      }
     ]
     moduleScript {refPath=&ShapeLocation}
    }
    {
     bIsSetParametersModule=false
     enabled=true
     moduleName=CameraOffset
     inputs [Camera Offset Amount {bIsDynamic=true, bIsEditable=true, bIsStaticSwitch=false, bIsVisible=true, type {classStructOrEnum {refPath=&NiagaraFloat}}}]
     moduleScript {refPath=&CameraOffset}
    }
    {
     bIsSetParametersModule=false
     enabled=true
     moduleName=AddVelocity001
     inputs [
      Velocity Mode {
       bIsDynamic=false
       bIsEditable=true
       bIsStaticSwitch=true
       bIsVisible=true
       type {classStructOrEnum {refPath=&ENiagara_VelocityMode}}
      }
      Velocity {bIsDynamic=true, bIsEditable=false, bIsStaticSwitch=false, bIsVisible=false, type {classStructOrEnum {refPath=&Vector3f}}}
      Velocity Speed {bIsDynamic=true, bIsEditable=true, bIsStaticSwitch=false, bIsVisible=true, type {classStructOrEnum {refPath=&NiagaraFloat}}}
      Velocity Speed Scale {
       bIsDynamic=false
       bIsEditable=false
       bIsStaticSwitch=false
       bIsVisible=false
       type {classStructOrEnum {refPath=&NiagaraFloat}}
      }
      Module.Use Falloff Distance {bIsDynamic=false, bIsEditable=true, bIsStaticSwitch=false, bIsVisible=true, type {classStructOrEnum {refPath=&NiagaraBool}}}
      Constrain To Radius {bIsDynamic=false, bIsEditable=false, bIsStaticSwitch=false, bIsVisible=true, type {classStructOrEnum {refPath=&NiagaraFloat}}}
      Distribution Along Cone Axis {
       bIsDynamic=false
       bIsEditable=false
       bIsStaticSwitch=false
       bIsVisible=false
       type {classStructOrEnum {refPath=&NiagaraFloat}}
      }
      Radius Falloff Near / Far {
       bIsDynamic=false
       bIsEditable=false
       bIsStaticSwitch=false
       bIsVisible=false
       type {classStructOrEnum {refPath=&Vector2f}}
      }
      Radius Falloff Exponent {
       bIsDynamic=false
       bIsEditable=false
       bIsStaticSwitch=false
       bIsVisible=false
       type {classStructOrEnum {refPath=&NiagaraFloat}}
      }
      Speed Falloff From Cone Axis {
       bIsDynamic=false
       bIsEditable=false
       bIsStaticSwitch=false
       bIsVisible=false
       type {classStructOrEnum {refPath=&NiagaraFloat}}
      }
      Invert Velocity Falloff {bIsDynamic=false, bIsEditable=false, bIsStaticSwitch=false, bIsVisible=false, type {classStructOrEnum {refPath=&NiagaraBool}}}
      Override Seed {bIsDynamic=false, bIsEditable=true, bIsStaticSwitch=false, bIsVisible=true, type {classStructOrEnum {refPath=&NiagaraBool}}}
      Random Seed {bIsDynamic=false, bIsEditable=false, bIsStaticSwitch=false, bIsVisible=true, type {classStructOrEnum {refPath=&NiagaraInt32}}}
      Fixed Random Seed {bIsDynamic=false, bIsEditable=false, bIsStaticSwitch=true, bIsVisible=false, type {classStructOrEnum {refPath=&NiagaraBool}}}
      Randomness Mode {
       bIsDynamic=false
       bIsEditable=true
       bIsStaticSwitch=true
       bIsVisible=true
       type {classStructOrEnum {refPath=&ENiagaraRandomnessMode}}
      }
      Cone Axis {
       bIsDynamic=false
       bIsEditable=false
       bIsStaticSwitch=false
       bIsVisible=false
       type {classStructOrEnum {refPath=&Vector3f}}
      }
      Cone Angle {
       bIsDynamic=false
       bIsEditable=false
       bIsStaticSwitch=false
       bIsVisible=false
       type {classStructOrEnum {refPath=&NiagaraFloat}}
      }
      Inner Cone Angle {
       bIsDynamic=false
       bIsEditable=false
       bIsStaticSwitch=false
       bIsVisible=false
       type {classStructOrEnum {refPath=&NiagaraFloat}}
      }
      Cone Angle Mode {
       bIsDynamic=false
       bIsEditable=false
       bIsStaticSwitch=true
       bIsVisible=false
       type {classStructOrEnum {refPath=&ENiagara_AngleInput}}
      }
      Rotation Mode {
       bIsDynamic=false
       bIsEditable=false
       bIsStaticSwitch=true
       bIsVisible=false
       type {classStructOrEnum {refPath=&ENiagara_RotationMode}}
      }
      Rotation Axis {
       bIsDynamic=false
       bIsEditable=false
       bIsStaticSwitch=false
       bIsVisible=false
       type {classStructOrEnum {refPath=&Vector3f}}
      }
      Yaw / Pitch / Roll {
       bIsDynamic=false
       bIsEditable=false
       bIsStaticSwitch=false
       bIsVisible=false
       type {classStructOrEnum {refPath=&Vector3f}}
      }
      Rotation Quaternion {bIsDynamic=false, bIsEditable=false, bIsStaticSwitch=false, bIsVisible=false, type {classStructOrEnum {refPath=&Quat4f}}}
      Rotation Matrix {
       bIsDynamic=false
       bIsEditable=false
       bIsStaticSwitch=false
       bIsVisible=false
       type {classStructOrEnum {refPath=&NiagaraMatrix}}
      }
      Rotation Angle {
       bIsDynamic=false
       bIsEditable=false
       bIsStaticSwitch=false
       bIsVisible=false
       type {classStructOrEnum {refPath=&NiagaraFloat}}
      }
      Invert Rotation Quaternion {bIsDynamic=false, bIsEditable=false, bIsStaticSwitch=false, bIsVisible=false, type {classStructOrEnum {refPath=&NiagaraBool}}}
      Additional Quaternion Rotation {bIsDynamic=false, bIsEditable=false, bIsStaticSwitch=false, bIsVisible=false, type {classStructOrEnum {refPath=&NiagaraBool}}}
      Additional Yaw / Pitch / Roll {
       bIsDynamic=false
       bIsEditable=false
       bIsStaticSwitch=false
       bIsVisible=false
       type {classStructOrEnum {refPath=&Vector3f}}
      }
      Rotation Angle Type {
       bIsDynamic=false
       bIsEditable=false
       bIsStaticSwitch=true
       bIsVisible=false
       type {classStructOrEnum {refPath=&ENiagara_AngleInput}}
      }
      Rotation Coordinate Space {
       bIsDynamic=false
       bIsEditable=false
       bIsStaticSwitch=false
       bIsVisible=false
       type {classStructOrEnum {refPath=&ENiagaraCoordinateSpace}}
      }
      Velocity Origin {
       bIsDynamic=false
       bIsEditable=true
       bIsStaticSwitch=false
       bIsVisible=true
       type {classStructOrEnum {refPath=&NiagaraPosition}}
      }
      Origin Offset {bIsDynamic=false, bIsEditable=true, bIsStaticSwitch=false, bIsVisible=true, type {classStructOrEnum {refPath=&Vector3f}}}
      Origin Offset Coordinate Space {
       bIsDynamic=false
       bIsEditable=true
       bIsStaticSwitch=false
       bIsVisible=true
       type {classStructOrEnum {refPath=&ENiagaraCoordinateSpace}}
      }
      Default Position {
       bIsDynamic=false
       bIsEditable=true
       bIsStaticSwitch=false
       bIsVisible=true
       type {classStructOrEnum {refPath=&NiagaraPosition}}
      }
     ]
     moduleScript {refPath=&AddVelocity}
    }
    {
     bIsSetParametersModule=false
     enabled=true
     moduleName=AddVelocity
     inputs [
      Velocity Mode {
       bIsDynamic=false
       bIsEditable=true
       bIsStaticSwitch=true
       bIsVisible=true
       type {classStructOrEnum {refPath=&ENiagara_VelocityMode}}
      }
      Velocity {bIsDynamic=true, bIsEditable=true, bIsStaticSwitch=false, bIsVisible=true, type {classStructOrEnum {refPath=&Vector3f}}}
      Velocity Speed {
       bIsDynamic=false
       bIsEditable=false
       bIsStaticSwitch=false
       bIsVisible=false
       type {classStructOrEnum {refPath=&NiagaraFloat}}
      }
      Velocity Speed Scale {bIsDynamic=false, bIsEditable=true, bIsStaticSwitch=false, bIsVisible=true, type {classStructOrEnum {refPath=&NiagaraFloat}}}
      Constrain To Radius {
       bIsDynamic=false
       bIsEditable=false
       bIsStaticSwitch=false
       bIsVisible=false
       type {classStructOrEnum {refPath=&NiagaraFloat}}
      }
      Distribution Along Cone Axis {
       bIsDynamic=false
       bIsEditable=false
       bIsStaticSwitch=false
       bIsVisible=false
       type {classStructOrEnum {refPath=&NiagaraFloat}}
      }
      Radius Falloff Near / Far {
       bIsDynamic=false
       bIsEditable=false
       bIsStaticSwitch=false
       bIsVisible=false
       type {classStructOrEnum {refPath=&Vector2f}}
      }
      Radius Falloff Exponent {
       bIsDynamic=false
       bIsEditable=false
       bIsStaticSwitch=false
       bIsVisible=false
       type {classStructOrEnum {refPath=&NiagaraFloat}}
      }
      Speed Falloff From Cone Axis {
       bIsDynamic=false
       bIsEditable=false
       bIsStaticSwitch=false
       bIsVisible=false
       type {classStructOrEnum {refPath=&NiagaraFloat}}
      }
      Invert Velocity Falloff {bIsDynamic=false, bIsEditable=false, bIsStaticSwitch=false, bIsVisible=false, type {classStructOrEnum {refPath=&NiagaraBool}}}
      Override Seed {bIsDynamic=false, bIsEditable=false, bIsStaticSwitch=false, bIsVisible=false, type {classStructOrEnum {refPath=&NiagaraBool}}}
      Random Seed {
       bIsDynamic=false
       bIsEditable=false
       bIsStaticSwitch=false
       bIsVisible=false
       type {classStructOrEnum {refPath=&NiagaraInt32}}
      }
      Fixed Random Seed {bIsDynamic=false, bIsEditable=false, bIsStaticSwitch=true, bIsVisible=false, type {classStructOrEnum {refPath=&NiagaraBool}}}
      Randomness Mode {
       bIsDynamic=false
       bIsEditable=false
       bIsStaticSwitch=true
       bIsVisible=false
       type {classStructOrEnum {refPath=&ENiagaraRandomnessMode}}
      }
      Cone Axis {
       bIsDynamic=false
       bIsEditable=false
       bIsStaticSwitch=false
       bIsVisible=false
       type {classStructOrEnum {refPath=&Vector3f}}
      }
      Cone Angle {
       bIsDynamic=false
       bIsEditable=false
       bIsStaticSwitch=false
       bIsVisible=false
       type {classStructOrEnum {refPath=&NiagaraFloat}}
      }
      Inner Cone Angle {
       bIsDynamic=false
       bIsEditable=false
       bIsStaticSwitch=false
       bIsVisible=false
       type {classStructOrEnum {refPath=&NiagaraFloat}}
      }
      Cone Angle Mode {
       bIsDynamic=false
       bIsEditable=false
       bIsStaticSwitch=true
       bIsVisible=false
       type {classStructOrEnum {refPath=&ENiagara_AngleInput}}
      }
      Rotation Mode {
       bIsDynamic=false
       bIsEditable=true
       bIsStaticSwitch=true
       bIsVisible=true
       type {classStructOrEnum {refPath=&ENiagara_RotationMode}}
      }
      Rotation Axis {
       bIsDynamic=false
       bIsEditable=false
       bIsStaticSwitch=false
       bIsVisible=false
       type {classStructOrEnum {refPath=&Vector3f}}
      }
      Yaw / Pitch / Roll {
       bIsDynamic=false
       bIsEditable=false
       bIsStaticSwitch=false
       bIsVisible=false
       type {classStructOrEnum {refPath=&Vector3f}}
      }
      Rotation Quaternion {bIsDynamic=false, bIsEditable=false, bIsStaticSwitch=false, bIsVisible=false, type {classStructOrEnum {refPath=&Quat4f}}}
      Rotation Matrix {
       bIsDynamic=false
       bIsEditable=false
       bIsStaticSwitch=false
       bIsVisible=false
       type {classStructOrEnum {refPath=&NiagaraMatrix}}
      }
      Rotation Angle {
       bIsDynamic=false
       bIsEditable=false
       bIsStaticSwitch=false
       bIsVisible=false
       type {classStructOrEnum {refPath=&NiagaraFloat}}
      }
      Invert Rotation Quaternion {bIsDynamic=false, bIsEditable=false, bIsStaticSwitch=false, bIsVisible=false, type {classStructOrEnum {refPath=&NiagaraBool}}}
      Additional Quaternion Rotation {bIsDynamic=false, bIsEditable=false, bIsStaticSwitch=false, bIsVisible=false, type {classStructOrEnum {refPath=&NiagaraBool}}}
      Additional Yaw / Pitch / Roll {
       bIsDynamic=false
       bIsEditable=false
       bIsStaticSwitch=false
       bIsVisible=false
       type {classStructOrEnum {refPath=&Vector3f}}
      }
      Rotation Angle Type {
       bIsDynamic=false
       bIsEditable=false
       bIsStaticSwitch=true
       bIsVisible=false
       type {classStructOrEnum {refPath=&ENiagara_AngleInput}}
      }
      Rotation Coordinate Space {
       bIsDynamic=false
       bIsEditable=true
       bIsStaticSwitch=false
       bIsVisible=true
       type {classStructOrEnum {refPath=&ENiagaraCoordinateSpace}}
      }
      Velocity Origin {
       bIsDynamic=false
       bIsEditable=false
       bIsStaticSwitch=false
       bIsVisible=false
       type {classStructOrEnum {refPath=&NiagaraPosition}}
      }
      Origin Offset {
       bIsDynamic=false
       bIsEditable=false
       bIsStaticSwitch=false
       bIsVisible=false
       type {classStructOrEnum {refPath=&Vector3f}}
      }
      Origin Offset Coordinate Space {
       bIsDynamic=false
       bIsEditable=false
       bIsStaticSwitch=false
       bIsVisible=false
       type {classStructOrEnum {refPath=&ENiagaraCoordinateSpace}}
      }
      Default Position {
       bIsDynamic=false
       bIsEditable=false
       bIsStaticSwitch=false
       bIsVisible=false
       type {classStructOrEnum {refPath=&NiagaraPosition}}
      }
     ]
     moduleScript {refPath=&AddVelocity}
    }
    {
     bIsSetParametersModule=false
     enabled=true
     moduleName=ApplyInitialForces
     inputs [
      Apply Force to Velocity {bIsDynamic=false, bIsEditable=true, bIsStaticSwitch=false, bIsVisible=true, type {classStructOrEnum {refPath=&NiagaraBool}}}
      Apply Rotational Force to Rotational Velocity {bIsDynamic=false, bIsEditable=true, bIsStaticSwitch=false, bIsVisible=true, type {classStructOrEnum {refPath=&NiagaraBool}}}
      Apply Force to Position {bIsDynamic=false, bIsEditable=true, bIsStaticSwitch=false, bIsVisible=true, type {classStructOrEnum {refPath=&NiagaraBool}}}
      Positional Force Warmup Time {bIsDynamic=false, bIsEditable=false, bIsStaticSwitch=false, bIsVisible=true, type {classStructOrEnum {refPath=&NiagaraFloat}}}
     ]
     moduleScript {refPath=&ApplyInitialForces}
    }
    {
     bIsSetParametersModule=true
     enabled=true
     moduleName=SetVariables_#3
     inputs [Particles.DynamicMaterialParameter_Air {bIsDynamic=true, bIsEditable=true, bIsStaticSwitch=false, bIsVisible=true, type {classStructOrEnum {refPath=&Vector4f}}}]
     moduleScript {refPath=&self:DefaultPropSpawn_Merged002_0.NiagaraScriptSource_1.NiagaraGraph_0.NiagaraNodeAssignment_5.SetVariables_#3}
    }
   ]
  }
  particleUpdateScript {
   scriptName=ParticleUpdateScript
   modules [
    {
     bIsSetParametersModule=false
     enabled=true
     moduleName=ParticleState
     inputs [
      Kill Particles When Lifetime Has Elapsed {bIsDynamic=false, bIsEditable=true, bIsStaticSwitch=true, bIsVisible=true, type {classStructOrEnum {refPath=&NiagaraBool}}}
      Loop Particles Lifetime {bIsDynamic=false, bIsEditable=false, bIsStaticSwitch=false, bIsVisible=false, type {classStructOrEnum {refPath=&NiagaraBool}}}
      Let Infinitely Lived Particles Die When Emitter Deactivates {bIsDynamic=false, bIsEditable=false, bIsStaticSwitch=false, bIsVisible=false, type {classStructOrEnum {refPath=&NiagaraBool}}}
      Lifetime {bIsDynamic=false, bIsEditable=true, bIsStaticSwitch=false, bIsVisible=true, type {classStructOrEnum {refPath=&NiagaraFloat}}}
      DeltaTime {bIsDynamic=false, bIsEditable=true, bIsStaticSwitch=false, bIsVisible=true, type {classStructOrEnum {refPath=&NiagaraFloat}}}
     ]
     moduleScript {refPath=&ParticleState}
    }
    {
     bIsSetParametersModule=false
     enabled=true
     moduleName=ScaleMeshSize
     inputs [
      Scale Factor {bIsDynamic=true, bIsEditable=true, bIsStaticSwitch=false, bIsVisible=true, type {classStructOrEnum {refPath=&Vector3f}}}
      Initial Mesh Scale {bIsDynamic=false, bIsEditable=true, bIsStaticSwitch=false, bIsVisible=true, type {classStructOrEnum {refPath=&Vector3f}}}
     ]
     moduleScript {refPath=&ScaleMeshSize}
    }
    {
     bIsSetParametersModule=false
     enabled=true
     moduleName=SpriteRotationRate
     inputs [
      Rotation Rate {bIsDynamic=true, bIsEditable=true, bIsStaticSwitch=false, bIsVisible=true, type {classStructOrEnum {refPath=&NiagaraFloat}}}
      Delta Time {bIsDynamic=false, bIsEditable=true, bIsStaticSwitch=false, bIsVisible=true, type {classStructOrEnum {refPath=&NiagaraFloat}}}
     ]
     moduleScript {refPath=&SpriteRotationRate}
    }
    {
     bIsSetParametersModule=false
     enabled=true
     moduleName=GravityForce
     inputs [
      Gravity {bIsDynamic=true, bIsEditable=true, bIsStaticSwitch=false, bIsVisible=true, type {classStructOrEnum {refPath=&Vector3f}}}
      Coordinate Space {
       bIsDynamic=false
       bIsEditable=true
       bIsStaticSwitch=true
       bIsVisible=true
       type {classStructOrEnum {refPath=&ENiagaraCoordinateSpace}}
      }
     ]
     moduleScript {refPath=&GravityForce}
    }
    {
     bIsSetParametersModule=false
     enabled=true
     moduleName=CurlNoiseForce
     inputs [
      Noise Strength {bIsDynamic=true, bIsEditable=true, bIsStaticSwitch=false, bIsVisible=true, type {classStructOrEnum {refPath=&NiagaraFloat}}}
      Noise Frequency {bIsDynamic=true, bIsEditable=true, bIsStaticSwitch=false, bIsVisible=true, type {classStructOrEnum {refPath=&NiagaraFloat}}}
      Noise Quality / Cost {
       bIsDynamic=false
       bIsEditable=true
       bIsStaticSwitch=true
       bIsVisible=true
       type {classStructOrEnum {refPath=&ENiagaraCurlNoiseQuality}}
      }
      Random Seed {bIsDynamic=false, bIsEditable=true, bIsStaticSwitch=false, bIsVisible=true, type {classStructOrEnum {refPath=&NiagaraInt32}}}
      Pan Noise {bIsDynamic=false, bIsEditable=true, bIsStaticSwitch=false, bIsVisible=true, type {classStructOrEnum {refPath=&NiagaraBool}}}
      Pan Noise Field {bIsDynamic=false, bIsEditable=false, bIsStaticSwitch=false, bIsVisible=true, type {classStructOrEnum {refPath=&Vector3f}}}
      Write to Intrinsic Parameters {bIsDynamic=false, bIsEditable=true, bIsStaticSwitch=true, bIsVisible=true, type {classStructOrEnum {refPath=&NiagaraBool}}}
      Sample Position {bIsDynamic=false, bIsEditable=true, bIsStaticSwitch=false, bIsVisible=true, type {classStructOrEnum {refPath=&Vector3f}}}
      CurlNoise {
       bIsDynamic=false
       bIsEditable=false
       bIsStaticSwitch=false
       bIsVisible=false
       type {classStructOrEnum {refPath=&NiagaraDataInterfaceCurlNoise}}
      }
      Mask Curl Noise {bIsDynamic=false, bIsEditable=true, bIsStaticSwitch=true, bIsVisible=true, type {classStructOrEnum {refPath=&NiagaraBool}}}
      Curl Noise Cone Mask Angle {
       bIsDynamic=false
       bIsEditable=false
       bIsStaticSwitch=false
       bIsVisible=false
       type {classStructOrEnum {refPath=&NiagaraFloat}}
      }
      Curl Noise Cone Mask Falloff Angle {
       bIsDynamic=false
       bIsEditable=false
       bIsStaticSwitch=false
       bIsVisible=false
       type {classStructOrEnum {refPath=&NiagaraFloat}}
      }
      Curl Noise Cone Mask Axis {
       bIsDynamic=false
       bIsEditable=false
       bIsStaticSwitch=false
       bIsVisible=false
       type {classStructOrEnum {refPath=&Vector3f}}
      }
     ]
     moduleScript {refPath=&CurlNoiseForce}
    }
    {
     bIsSetParametersModule=false
     enabled=true
     moduleName=Drag
     inputs [
      Use Linear Drag {bIsDynamic=false, bIsEditable=true, bIsStaticSwitch=false, bIsVisible=true, type {classStructOrEnum {refPath=&NiagaraBool}}}
      Drag {bIsDynamic=true, bIsEditable=true, bIsStaticSwitch=false, bIsVisible=true, type {classStructOrEnum {refPath=&NiagaraFloat}}}
      Use Rotational Drag {bIsDynamic=false, bIsEditable=true, bIsStaticSwitch=false, bIsVisible=true, type {classStructOrEnum {refPath=&NiagaraBool}}}
      Rotational Drag {bIsDynamic=false, bIsEditable=true, bIsStaticSwitch=false, bIsVisible=true, type {classStructOrEnum {refPath=&NiagaraFloat}}}
      Ignore Mass {bIsDynamic=false, bIsEditable=true, bIsStaticSwitch=false, bIsVisible=true, type {classStructOrEnum {refPath=&NiagaraBool}}}
     ]
     moduleScript {refPath=&Drag}
    }
    {
     bIsSetParametersModule=false
     enabled=true
     moduleName=SolveForcesAndVelocity
     inputs [
      Clamp Velocity {bIsDynamic=false, bIsEditable=true, bIsStaticSwitch=false, bIsVisible=true, type {classStructOrEnum {refPath=&NiagaraBool}}}
      Speed Limit {bIsDynamic=false, bIsEditable=false, bIsStaticSwitch=false, bIsVisible=true, type {classStructOrEnum {refPath=&NiagaraFloat}}}
      Limit Acceleration {bIsDynamic=false, bIsEditable=true, bIsStaticSwitch=false, bIsVisible=true, type {classStructOrEnum {refPath=&NiagaraBool}}}
      Acceleration Limit {bIsDynamic=false, bIsEditable=false, bIsStaticSwitch=false, bIsVisible=true, type {classStructOrEnum {refPath=&NiagaraFloat}}}
      Manually Enable Rotational Solver {bIsDynamic=false, bIsEditable=true, bIsStaticSwitch=false, bIsVisible=true, type {classStructOrEnum {refPath=&NiagaraBool}}}
      Write to Intrinsic Properties {bIsDynamic=false, bIsEditable=true, bIsStaticSwitch=false, bIsVisible=true, type {classStructOrEnum {refPath=&NiagaraBool}}}
      Write to Presolve Properties {bIsDynamic=false, bIsEditable=true, bIsStaticSwitch=true, bIsVisible=true, type {classStructOrEnum {refPath=&NiagaraBool}}}
      DeltaTime {bIsDynamic=false, bIsEditable=true, bIsStaticSwitch=false, bIsVisible=true, type {classStructOrEnum {refPath=&NiagaraFloat}}}
      Mass {bIsDynamic=false, bIsEditable=true, bIsStaticSwitch=false, bIsVisible=true, type {classStructOrEnum {refPath=&NiagaraFloat}}}
      Position {
       bIsDynamic=false
       bIsEditable=true
       bIsStaticSwitch=false
       bIsVisible=true
       type {classStructOrEnum {refPath=&NiagaraPosition}}
      }
      Velocity {bIsDynamic=false, bIsEditable=true, bIsStaticSwitch=false, bIsVisible=true, type {classStructOrEnum {refPath=&Vector3f}}}
      PreviousVelocity {bIsDynamic=false, bIsEditable=true, bIsStaticSwitch=false, bIsVisible=true, type {classStructOrEnum {refPath=&Vector3f}}}
      Force {bIsDynamic=false, bIsEditable=true, bIsStaticSwitch=false, bIsVisible=true, type {classStructOrEnum {refPath=&Vector3f}}}
      Mesh Orientation {bIsDynamic=false, bIsEditable=false, bIsStaticSwitch=false, bIsVisible=false, type {classStructOrEnum {refPath=&Quat4f}}}
      Rotational Velocity {
       bIsDynamic=false
       bIsEditable=false
       bIsStaticSwitch=false
       bIsVisible=false
       type {classStructOrEnum {refPath=&Vector3f}}
      }
      Rotational Force {
       bIsDynamic=false
       bIsEditable=false
       bIsStaticSwitch=false
       bIsVisible=false
       type {classStructOrEnum {refPath=&Vector3f}}
      }
      Rotational Inertia {
       bIsDynamic=false
       bIsEditable=false
       bIsStaticSwitch=false
       bIsVisible=false
       type {classStructOrEnum {refPath=&NiagaraFloat}}
      }
     ]
     moduleScript {refPath=&SolveForcesAndVelocity}
    }
    {
     bIsSetParametersModule=false
     enabled=true
     moduleName=ScaleSpriteSize
     inputs [
      Scale Sprite Size Mode {
       bIsDynamic=false
       bIsEditable=true
       bIsStaticSwitch=true
       bIsVisible=true
       type {classStructOrEnum {refPath=&ENiagara_ScaleSpriteSize}}
      }
      Uniform Curve Sprite Scale {
       bIsDynamic=false
       bIsEditable=false
       bIsStaticSwitch=false
       bIsVisible=false
       type {classStructOrEnum {refPath=&NiagaraDataInterfaceCurve}}
      }
      Non-Uniform Curve Sprite Scale {
       bIsDynamic=false
       bIsEditable=false
       bIsStaticSwitch=false
       bIsVisible=false
       type {classStructOrEnum {refPath=&NiagaraDataInterfaceVector2DCurve}}
      }
      Uniform Curve Index {
       bIsDynamic=false
       bIsEditable=false
       bIsStaticSwitch=false
       bIsVisible=false
       type {classStructOrEnum {refPath=&NiagaraFloat}}
      }
      Non-Uniform Curve Index {
       bIsDynamic=false
       bIsEditable=false
       bIsStaticSwitch=false
       bIsVisible=false
       type {classStructOrEnum {refPath=&NiagaraFloat}}
      }
      Non-Uniform Curve Scale {
       bIsDynamic=false
       bIsEditable=false
       bIsStaticSwitch=false
       bIsVisible=false
       type {classStructOrEnum {refPath=&Vector2f}}
      }
      Uniform Curve Scale {
       bIsDynamic=false
       bIsEditable=false
       bIsStaticSwitch=false
       bIsVisible=false
       type {classStructOrEnum {refPath=&NiagaraFloat}}
      }
      Scale Factor {bIsDynamic=true, bIsEditable=true, bIsStaticSwitch=false, bIsVisible=true, type {classStructOrEnum {refPath=&Vector2f}}}
      Uniform Scale Factor {
       bIsDynamic=false
       bIsEditable=false
       bIsStaticSwitch=false
       bIsVisible=false
       type {classStructOrEnum {refPath=&NiagaraFloat}}
      }
      Initial Sprite Size {bIsDynamic=false, bIsEditable=true, bIsStaticSwitch=false, bIsVisible=true, type {classStructOrEnum {refPath=&Vector2f}}}
     ]
     moduleScript {refPath=&ScaleSpriteSize}
    }
    {
     bIsSetParametersModule=true
     enabled=true
     moduleName=SetVariables_#4
     inputs [
      Particles.DynParSmoke0 {bIsDynamic=true, bIsEditable=true, bIsStaticSwitch=false, bIsVisible=true, type {classStructOrEnum {refPath=&Vector4f}}}
      Particles.DynParSmoke1 {bIsDynamic=true, bIsEditable=true, bIsStaticSwitch=false, bIsVisible=true, type {classStructOrEnum {refPath=&Vector4f}}}
      Particles.DynParSmoke2 {bIsDynamic=true, bIsEditable=true, bIsStaticSwitch=false, bIsVisible=true, type {classStructOrEnum {refPath=&Vector4f}}}
     ]
     moduleScript {refPath=&self:DefaultPropSpawn_Merged002_0.NiagaraScriptSource_1.NiagaraGraph_0.NiagaraNodeAssignment_2.SetVariables_#4}
    }
    {
     bIsSetParametersModule=false
     enabled=true
     moduleName=ScaleColor001
     inputs [
      Scale Mode {bIsDynamic=false, bIsEditable=true, bIsStaticSwitch=true, bIsVisible=true, type {classStructOrEnum {refPath=&NiagaraInt32}}}
      Linear Color Curve {
       bIsDynamic=false
       bIsEditable=false
       bIsStaticSwitch=false
       bIsVisible=false
       type {classStructOrEnum {refPath=&NiagaraDataInterfaceColorCurve}}
      }
      ScaleRGB {bIsDynamic=false, bIsEditable=true, bIsStaticSwitch=false, bIsVisible=true, type {classStructOrEnum {refPath=&NiagaraBool}}}
      Scale RGB {bIsDynamic=false, bIsEditable=true, bIsStaticSwitch=false, bIsVisible=true, type {classStructOrEnum {refPath=&Vector3f}}}
      ScaleRGBA {bIsDynamic=false, bIsEditable=false, bIsStaticSwitch=false, bIsVisible=false, type {classStructOrEnum {refPath=&NiagaraBool}}}
      Scale RGBA {
       bIsDynamic=false
       bIsEditable=false
       bIsStaticSwitch=false
       bIsVisible=false
       type {classStructOrEnum {refPath=&Vector4f}}
      }
      Curve Index {
       bIsDynamic=false
       bIsEditable=false
       bIsStaticSwitch=false
       bIsVisible=false
       type {classStructOrEnum {refPath=&NiagaraFloat}}
      }
      ScaleA {bIsDynamic=false, bIsEditable=true, bIsStaticSwitch=false, bIsVisible=true, type {classStructOrEnum {refPath=&NiagaraBool}}}
      Scale Alpha {bIsDynamic=true, bIsEditable=true, bIsStaticSwitch=false, bIsVisible=true, type {classStructOrEnum {refPath=&NiagaraFloat}}}
      Color Value To Scale {
       bIsDynamic=false
       bIsEditable=true
       bIsStaticSwitch=false
       bIsVisible=true
       type {classStructOrEnum {refPath=&LinearColor}}
      }
      Color Scale {
       bIsDynamic=false
       bIsEditable=false
       bIsStaticSwitch=false
       bIsVisible=false
       type {classStructOrEnum {refPath=&LinearColor}}
      }
     ]
     moduleScript {refPath=&ScaleColor}
    }
    {
     bIsSetParametersModule=false
     enabled=true
     moduleName=FN_Light_Attributes
     inputs [
      Enable Particle Light Enable {bIsDynamic=false, bIsEditable=true, bIsStaticSwitch=true, bIsVisible=true, type {classStructOrEnum {refPath=&NiagaraBool}}}
      Light Enabled {bIsDynamic=false, bIsEditable=false, bIsStaticSwitch=false, bIsVisible=false, type {classStructOrEnum {refPath=&NiagaraBool}}}
      Enable Visibility Tag {bIsDynamic=false, bIsEditable=true, bIsStaticSwitch=true, bIsVisible=true, type {classStructOrEnum {refPath=&NiagaraBool}}}
      VisibilityTag {
       bIsDynamic=false
       bIsEditable=false
       bIsStaticSwitch=false
       bIsVisible=false
       type {classStructOrEnum {refPath=&NiagaraInt32}}
      }
      Create Position Attribute {bIsDynamic=false, bIsEditable=true, bIsStaticSwitch=true, bIsVisible=true, type {classStructOrEnum {refPath=&NiagaraBool}}}
      Position {bIsDynamic=false, bIsEditable=true, bIsStaticSwitch=false, bIsVisible=true, type {classStructOrEnum {refPath=&Vector3f}}}
      Create Radius Attribute {bIsDynamic=false, bIsEditable=true, bIsStaticSwitch=true, bIsVisible=true, type {classStructOrEnum {refPath=&NiagaraBool}}}
      Radius {bIsDynamic=false, bIsEditable=true, bIsStaticSwitch=false, bIsVisible=true, type {classStructOrEnum {refPath=&NiagaraFloat}}}
      Create Exponent Attribute {bIsDynamic=false, bIsEditable=true, bIsStaticSwitch=true, bIsVisible=true, type {classStructOrEnum {refPath=&NiagaraBool}}}
      Exponent {bIsDynamic=false, bIsEditable=true, bIsStaticSwitch=false, bIsVisible=true, type {classStructOrEnum {refPath=&NiagaraFloat}}}
      Create Color Attribute {bIsDynamic=false, bIsEditable=true, bIsStaticSwitch=true, bIsVisible=true, type {classStructOrEnum {refPath=&NiagaraBool}}}
      Color {
       bIsDynamic=false
       bIsEditable=true
       bIsStaticSwitch=false
       bIsVisible=true
       type {classStructOrEnum {refPath=&LinearColor}}
      }
      Scale Color {bIsDynamic=true, bIsEditable=true, bIsStaticSwitch=false, bIsVisible=true, type {classStructOrEnum {refPath=&NiagaraFloat}}}
     ]
     moduleScript {refPath=&FN_Light_Attributes}
    }
    {
     bIsSetParametersModule=true
     enabled=true
     moduleName=SetVariables_#9
     inputs [
      Particles.ShockwaveScale {bIsDynamic=true, bIsEditable=true, bIsStaticSwitch=false, bIsVisible=true, type {classStructOrEnum {refPath=&Vector3f}}}
      Particles.ShockwaveColor {
       bIsDynamic=true
       bIsEditable=true
       bIsStaticSwitch=false
       bIsVisible=true
       type {classStructOrEnum {refPath=&LinearColor}}
      }
      Particles.SecondarySphereColor {
       bIsDynamic=true
       bIsEditable=true
       bIsStaticSwitch=false
       bIsVisible=true
       type {classStructOrEnum {refPath=&LinearColor}}
      }
      Particles.SecondaryDyn {bIsDynamic=true, bIsEditable=true, bIsStaticSwitch=false, bIsVisible=true, type {classStructOrEnum {refPath=&Vector4f}}}
      Particles.PrimaryColor {
       bIsDynamic=true
       bIsEditable=true
       bIsStaticSwitch=false
       bIsVisible=true
       type {classStructOrEnum {refPath=&LinearColor}}
      }
     ]
     moduleScript {refPath=&self:DefaultPropSpawn_Merged002_0.NiagaraScriptSource_1.NiagaraGraph_0.NiagaraNodeAssignment_3.SetVariables_#9}
    }
   ]
  }
  rendererClasses [
   {refPath=&NiagaraSpriteRendererProperties}
   {refPath=&NiagaraLightRendererProperties}
   {refPath=&NiagaraMeshRendererProperties}
  ]
  renderers [
   {rendererIndex=0, rendererClass {refPath=&NiagaraSpriteRendererProperties}}
   {rendererIndex=1, rendererClass {refPath=&NiagaraSpriteRendererProperties}}
   {rendererIndex=2, rendererClass {refPath=&NiagaraSpriteRendererProperties}}
   {rendererIndex=3, rendererClass {refPath=&NiagaraSpriteRendererProperties}}
   {rendererIndex=4, rendererClass {refPath=&NiagaraSpriteRendererProperties}}
   {rendererIndex=5, rendererClass {refPath=&NiagaraLightRendererProperties}}
   {rendererIndex=6, rendererClass {refPath=&NiagaraSpriteRendererProperties}}
   {rendererIndex=7, rendererClass {refPath=&NiagaraMeshRendererProperties}}
   {rendererIndex=8, rendererClass {refPath=&NiagaraMeshRendererProperties}}
   {rendererIndex=9, rendererClass {refPath=&NiagaraMeshRendererProperties}}
  ]
 }
 DefaultPropSpawn_Merged003 {
  bEnabled=false
  emitterSpawnScript {scriptName=EmitterSpawnScript, modules []}
  emitterUpdateScript {
   scriptName=EmitterUpdateScript
   modules [
    {
     bIsSetParametersModule=false
     enabled=true
     moduleName=EmitterState
     inputs [
      Life Cycle Mode {
       bIsDynamic=false
       bIsEditable=true
       bIsStaticSwitch=true
       bIsVisible=true
       type {classStructOrEnum {refPath=&ENiagaraEmitterLifeCycleMode}}
      }
      Inactive Response {
       bIsDynamic=false
       bIsEditable=false
       bIsStaticSwitch=true
       bIsVisible=false
       type {classStructOrEnum {refPath=&ENiagaraInactiveMode}}
      }
      Loop Behavior {
       bIsDynamic=false
       bIsEditable=false
       bIsStaticSwitch=true
       bIsVisible=false
       type {classStructOrEnum {refPath=&ENiagara_EmitterStateOptions}}
      }
      Loop Duration Mode {
       bIsDynamic=false
       bIsEditable=false
       bIsStaticSwitch=true
       bIsVisible=false
       type {classStructOrEnum {refPath=&ENiagara_InfiniteLoopDuration}}
      }
      Loop Count {
       bIsDynamic=false
       bIsEditable=false
       bIsStaticSwitch=false
       bIsVisible=false
       type {classStructOrEnum {refPath=&NiagaraInt32}}
      }
      Loop Duration {
       bIsDynamic=false
       bIsEditable=false
       bIsStaticSwitch=false
       bIsVisible=false
       type {classStructOrEnum {refPath=&NiagaraFloat}}
      }
      Recalculate Duration Each Loop {bIsDynamic=false, bIsEditable=false, bIsStaticSwitch=false, bIsVisible=false, type {classStructOrEnum {refPath=&NiagaraBool}}}
      UseLoopDelay {bIsDynamic=false, bIsEditable=false, bIsStaticSwitch=false, bIsVisible=false, type {classStructOrEnum {refPath=&NiagaraBool}}}
      Loop Delay {
       bIsDynamic=false
       bIsEditable=false
       bIsStaticSwitch=false
       bIsVisible=false
       type {classStructOrEnum {refPath=&NiagaraFloat}}
      }
      Delay First Loop Only {bIsDynamic=false, bIsEditable=false, bIsStaticSwitch=false, bIsVisible=false, type {classStructOrEnum {refPath=&NiagaraBool}}}
      Scalability Mode {
       bIsDynamic=false
       bIsEditable=true
       bIsStaticSwitch=true
       bIsVisible=true
       type {classStructOrEnum {refPath=&ENiagaraEmitterScalabilityMode_Limited}}
      }
      Enable Distance Culling {bIsDynamic=false, bIsEditable=false, bIsStaticSwitch=true, bIsVisible=false, type {classStructOrEnum {refPath=&NiagaraBool}}}
      MinDistance {
       bIsDynamic=false
       bIsEditable=false
       bIsStaticSwitch=false
       bIsVisible=false
       type {classStructOrEnum {refPath=&NiagaraFloat}}
      }
      Min Distance Response {
       bIsDynamic=false
       bIsEditable=false
       bIsStaticSwitch=false
       bIsVisible=false
       type {classStructOrEnum {refPath=&ENiagaraExecutionStateManagement}}
      }
      MaxDistance {
       bIsDynamic=false
       bIsEditable=false
       bIsStaticSwitch=false
       bIsVisible=false
       type {classStructOrEnum {refPath=&NiagaraFloat}}
      }
      Max Distance Response {
       bIsDynamic=false
       bIsEditable=false
       bIsStaticSwitch=false
       bIsVisible=false
       type {classStructOrEnum {refPath=&ENiagaraExecutionStateManagement}}
      }
      Scale Spawn Count {bIsDynamic=false, bIsEditable=false, bIsStaticSwitch=true, bIsVisible=false, type {classStructOrEnum {refPath=&NiagaraBool}}}
      Spawn Count Scale {
       bIsDynamic=false
       bIsEditable=false
       bIsStaticSwitch=false
       bIsVisible=false
       type {classStructOrEnum {refPath=&NiagaraFloat}}
      }
      Spawn Count Scale By Distance {
       bIsDynamic=false
       bIsEditable=false
       bIsStaticSwitch=false
       bIsVisible=false
       type {classStructOrEnum {refPath=&NiagaraDataInterfaceCurve}}
      }
      Enable Visibility Culling {bIsDynamic=false, bIsEditable=false, bIsStaticSwitch=true, bIsVisible=false, type {classStructOrEnum {refPath=&NiagaraBool}}}
      Visibility Cull Response {
       bIsDynamic=false
       bIsEditable=false
       bIsStaticSwitch=false
       bIsVisible=false
       type {classStructOrEnum {refPath=&ENiagaraExecutionStateManagement}}
      }
      Visibility Cull Delay {
       bIsDynamic=false
       bIsEditable=false
       bIsStaticSwitch=false
       bIsVisible=false
       type {classStructOrEnum {refPath=&NiagaraFloat}}
      }
      Reset Age On Awaken {bIsDynamic=false, bIsEditable=false, bIsStaticSwitch=false, bIsVisible=false, type {classStructOrEnum {refPath=&NiagaraBool}}}
     ]
     moduleScript {refPath=&EmitterState}
    }
    {
     bIsSetParametersModule=false
     enabled=true
     moduleName=SpawnBurst_Instantaneous001
     inputs [
      Spawn Count {bIsDynamic=true, bIsEditable=true, bIsStaticSwitch=false, bIsVisible=true, type {classStructOrEnum {refPath=&NiagaraInt32}}}
      Spawn Time {bIsDynamic=false, bIsEditable=true, bIsStaticSwitch=false, bIsVisible=true, type {classStructOrEnum {refPath=&NiagaraFloat}}}
      Use Spawn Probability {bIsDynamic=false, bIsEditable=true, bIsStaticSwitch=false, bIsVisible=true, type {classStructOrEnum {refPath=&NiagaraBool}}}
      Spawn Probability {bIsDynamic=false, bIsEditable=false, bIsStaticSwitch=false, bIsVisible=true, type {classStructOrEnum {refPath=&NiagaraFloat}}}
      UseLoopCountLimit {bIsDynamic=false, bIsEditable=true, bIsStaticSwitch=false, bIsVisible=true, type {classStructOrEnum {refPath=&NiagaraBool}}}
      Loop Count Limit {bIsDynamic=false, bIsEditable=false, bIsStaticSwitch=false, bIsVisible=true, type {classStructOrEnum {refPath=&NiagaraInt32}}}
      Spawn Group {bIsDynamic=false, bIsEditable=true, bIsStaticSwitch=false, bIsVisible=true, type {classStructOrEnum {refPath=&NiagaraInt32}}}
      Age {bIsDynamic=false, bIsEditable=true, bIsStaticSwitch=false, bIsVisible=true, type {classStructOrEnum {refPath=&NiagaraFloat}}}
     ]
     moduleScript {refPath=&SpawnBurst_Instantaneous}
    }
    {
     bIsSetParametersModule=false
     enabled=true
     moduleName=SpawnBurst_Instantaneous
     inputs [
      Spawn Count {bIsDynamic=false, bIsEditable=true, bIsStaticSwitch=false, bIsVisible=true, type {classStructOrEnum {refPath=&NiagaraInt32}}}
      Spawn Time {bIsDynamic=false, bIsEditable=true, bIsStaticSwitch=false, bIsVisible=true, type {classStructOrEnum {refPath=&NiagaraFloat}}}
      Use Spawn Probability {bIsDynamic=false, bIsEditable=true, bIsStaticSwitch=false, bIsVisible=true, type {classStructOrEnum {refPath=&NiagaraBool}}}
      Spawn Probability {bIsDynamic=false, bIsEditable=false, bIsStaticSwitch=false, bIsVisible=true, type {classStructOrEnum {refPath=&NiagaraFloat}}}
      UseLoopCountLimit {bIsDynamic=false, bIsEditable=true, bIsStaticSwitch=false, bIsVisible=true, type {classStructOrEnum {refPath=&NiagaraBool}}}
      Loop Count Limit {bIsDynamic=false, bIsEditable=false, bIsStaticSwitch=false, bIsVisible=true, type {classStructOrEnum {refPath=&NiagaraInt32}}}
      Spawn Group {bIsDynamic=false, bIsEditable=true, bIsStaticSwitch=false, bIsVisible=true, type {classStructOrEnum {refPath=&NiagaraInt32}}}
      Age {bIsDynamic=false, bIsEditable=true, bIsStaticSwitch=false, bIsVisible=true, type {classStructOrEnum {refPath=&NiagaraFloat}}}
     ]
     moduleScript {refPath=&SpawnBurst_Instantaneous}
    }
    {
     bIsSetParametersModule=false
     enabled=true
     moduleName=SpawnBurst_Instantaneous002
     inputs [
      Spawn Count {bIsDynamic=true, bIsEditable=true, bIsStaticSwitch=false, bIsVisible=true, type {classStructOrEnum {refPath=&NiagaraInt32}}}
      Spawn Time {bIsDynamic=false, bIsEditable=true, bIsStaticSwitch=false, bIsVisible=true, type {classStructOrEnum {refPath=&NiagaraFloat}}}
      Use Spawn Probability {bIsDynamic=false, bIsEditable=true, bIsStaticSwitch=false, bIsVisible=true, type {classStructOrEnum {refPath=&NiagaraBool}}}
      Spawn Probability {bIsDynamic=false, bIsEditable=false, bIsStaticSwitch=false, bIsVisible=true, type {classStructOrEnum {refPath=&NiagaraFloat}}}
      UseLoopCountLimit {bIsDynamic=false, bIsEditable=true, bIsStaticSwitch=false, bIsVisible=true, type {classStructOrEnum {refPath=&NiagaraBool}}}
      Loop Count Limit {bIsDynamic=false, bIsEditable=false, bIsStaticSwitch=false, bIsVisible=true, type {classStructOrEnum {refPath=&NiagaraInt32}}}
      Spawn Group {bIsDynamic=false, bIsEditable=true, bIsStaticSwitch=false, bIsVisible=true, type {classStructOrEnum {refPath=&NiagaraInt32}}}
      Age {bIsDynamic=false, bIsEditable=true, bIsStaticSwitch=false, bIsVisible=true, type {classStructOrEnum {refPath=&NiagaraFloat}}}
     ]
     moduleScript {refPath=&SpawnBurst_Instantaneous}
    }
    {
     bIsSetParametersModule=false
     enabled=true
     moduleName=SpawnBurst_Instantaneous003
     inputs [
      Spawn Count {bIsDynamic=false, bIsEditable=true, bIsStaticSwitch=false, bIsVisible=true, type {classStructOrEnum {refPath=&NiagaraInt32}}}
      Spawn Time {bIsDynamic=false, bIsEditable=true, bIsStaticSwitch=false, bIsVisible=true, type {classStructOrEnum {refPath=&NiagaraFloat}}}
      Use Spawn Probability {bIsDynamic=false, bIsEditable=true, bIsStaticSwitch=false, bIsVisible=true, type {classStructOrEnum {refPath=&NiagaraBool}}}
      Spawn Probability {bIsDynamic=false, bIsEditable=false, bIsStaticSwitch=false, bIsVisible=true, type {classStructOrEnum {refPath=&NiagaraFloat}}}
      UseLoopCountLimit {bIsDynamic=false, bIsEditable=true, bIsStaticSwitch=false, bIsVisible=true, type {classStructOrEnum {refPath=&NiagaraBool}}}
      Loop Count Limit {bIsDynamic=false, bIsEditable=false, bIsStaticSwitch=false, bIsVisible=true, type {classStructOrEnum {refPath=&NiagaraInt32}}}
      Spawn Group {bIsDynamic=false, bIsEditable=true, bIsStaticSwitch=false, bIsVisible=true, type {classStructOrEnum {refPath=&NiagaraInt32}}}
      Age {bIsDynamic=false, bIsEditable=true, bIsStaticSwitch=false, bIsVisible=true, type {classStructOrEnum {refPath=&NiagaraFloat}}}
     ]
     moduleScript {refPath=&SpawnBurst_Instantaneous}
    }
    {
     bIsSetParametersModule=false
     enabled=true
     moduleName=SpawnBurst_Instantaneous004
     inputs [
      Spawn Count {bIsDynamic=false, bIsEditable=true, bIsStaticSwitch=false, bIsVisible=true, type {classStructOrEnum {refPath=&NiagaraInt32}}}
      Spawn Time {bIsDynamic=false, bIsEditable=true, bIsStaticSwitch=false, bIsVisible=true, type {classStructOrEnum {refPath=&NiagaraFloat}}}
      Use Spawn Probability {bIsDynamic=false, bIsEditable=true, bIsStaticSwitch=false, bIsVisible=true, type {classStructOrEnum {refPath=&NiagaraBool}}}
      Spawn Probability {bIsDynamic=false, bIsEditable=false, bIsStaticSwitch=false, bIsVisible=true, type {classStructOrEnum {refPath=&NiagaraFloat}}}
      UseLoopCountLimit {bIsDynamic=false, bIsEditable=true, bIsStaticSwitch=false, bIsVisible=true, type {classStructOrEnum {refPath=&NiagaraBool}}}
      Loop Count Limit {bIsDynamic=false, bIsEditable=false, bIsStaticSwitch=false, bIsVisible=true, type {classStructOrEnum {refPath=&NiagaraInt32}}}
      Spawn Group {bIsDynamic=false, bIsEditable=true, bIsStaticSwitch=false, bIsVisible=true, type {classStructOrEnum {refPath=&NiagaraInt32}}}
      Age {bIsDynamic=false, bIsEditable=true, bIsStaticSwitch=false, bIsVisible=true, type {classStructOrEnum {refPath=&NiagaraFloat}}}
     ]
     moduleScript {refPath=&SpawnBurst_Instantaneous}
    }
    {
     bIsSetParametersModule=false
     enabled=true
     moduleName=SpawnBurst_Instantaneous005
     inputs [
      Spawn Count {bIsDynamic=false, bIsEditable=true, bIsStaticSwitch=false, bIsVisible=true, type {classStructOrEnum {refPath=&NiagaraInt32}}}
      Spawn Time {bIsDynamic=false, bIsEditable=true, bIsStaticSwitch=false, bIsVisible=true, type {classStructOrEnum {refPath=&NiagaraFloat}}}
      Use Spawn Probability {bIsDynamic=false, bIsEditable=true, bIsStaticSwitch=false, bIsVisible=true, type {classStructOrEnum {refPath=&NiagaraBool}}}
      Spawn Probability {bIsDynamic=false, bIsEditable=false, bIsStaticSwitch=false, bIsVisible=true, type {classStructOrEnum {refPath=&NiagaraFloat}}}
      UseLoopCountLimit {bIsDynamic=false, bIsEditable=true, bIsStaticSwitch=false, bIsVisible=true, type {classStructOrEnum {refPath=&NiagaraBool}}}
      Loop Count Limit {bIsDynamic=false, bIsEditable=false, bIsStaticSwitch=false, bIsVisible=true, type {classStructOrEnum {refPath=&NiagaraInt32}}}
      Spawn Group {bIsDynamic=false, bIsEditable=true, bIsStaticSwitch=false, bIsVisible=true, type {classStructOrEnum {refPath=&NiagaraInt32}}}
      Age {bIsDynamic=false, bIsEditable=true, bIsStaticSwitch=false, bIsVisible=true, type {classStructOrEnum {refPath=&NiagaraFloat}}}
     ]
     moduleScript {refPath=&SpawnBurst_Instantaneous}
    }
   ]
  }
  particleSpawnScript {
   scriptName=ParticleSpawnScript
   modules [
    {
     bIsSetParametersModule=true
     enabled=true
     moduleName=SetVariables_#1
     inputs [Particles.VisibilityTag {bIsDynamic=true, bIsEditable=true, bIsStaticSwitch=false, bIsVisible=true, type {classStructOrEnum {refPath=&NiagaraInt32}}}]
     moduleScript {refPath=&self:DefaultPropSpawn_Merged003_0.NiagaraScriptSource_1.NiagaraGraph_0.NiagaraNodeAssignment_1.SetVariables_#1}
    }
    {
     bIsSetParametersModule=false
     enabled=true
     moduleName=InitializeParticle001
     inputs [
      Lifetime Mode {
       bIsDynamic=false
       bIsEditable=true
       bIsStaticSwitch=true
       bIsVisible=true
       type {classStructOrEnum {refPath=&ENiagara_LifetimeMode}}
      }
      Lifetime Min {
       bIsDynamic=false
       bIsEditable=false
       bIsStaticSwitch=false
       bIsVisible=false
       type {classStructOrEnum {refPath=&NiagaraFloat}}
      }
      Lifetime {bIsDynamic=true, bIsEditable=true, bIsStaticSwitch=false, bIsVisible=true, type {classStructOrEnum {refPath=&NiagaraFloat}}}
      Lifetime Max {
       bIsDynamic=false
       bIsEditable=false
       bIsStaticSwitch=false
       bIsVisible=false
       type {classStructOrEnum {refPath=&NiagaraFloat}}
      }
      Lifetime Randomness Mode {
       bIsDynamic=false
       bIsEditable=false
       bIsStaticSwitch=true
       bIsVisible=false
       type {classStructOrEnum {refPath=&ENiagaraRandomnessMode}}
      }
      Color Mode {
       bIsDynamic=false
       bIsEditable=true
       bIsStaticSwitch=true
       bIsVisible=true
       type {classStructOrEnum {refPath=&ENiagara_ColorInitializationMode}}
      }
      Color Minimum {
       bIsDynamic=false
       bIsEditable=false
       bIsStaticSwitch=false
       bIsVisible=false
       type {classStructOrEnum {refPath=&LinearColor}}
      }
      Color {
       bIsDynamic=true
       bIsEditable=true
       bIsStaticSwitch=false
       bIsVisible=true
       type {classStructOrEnum {refPath=&LinearColor}}
      }
      AdjustHue {bIsDynamic=false, bIsEditable=false, bIsStaticSwitch=false, bIsVisible=false, type {classStructOrEnum {refPath=&NiagaraBool}}}
      Hue Shift Range {
       bIsDynamic=false
       bIsEditable=false
       bIsStaticSwitch=false
       bIsVisible=false
       type {classStructOrEnum {refPath=&Vector2f}}
      }
      Color Maximum {
       bIsDynamic=false
       bIsEditable=false
       bIsStaticSwitch=false
       bIsVisible=false
       type {classStructOrEnum {refPath=&LinearColor}}
      }
      AdjustSaturation {bIsDynamic=false, bIsEditable=false, bIsStaticSwitch=false, bIsVisible=false, type {classStructOrEnum {refPath=&NiagaraBool}}}
      Saturation Range {
       bIsDynamic=false
       bIsEditable=false
       bIsStaticSwitch=false
       bIsVisible=false
       type {classStructOrEnum {refPath=&Vector2f}}
      }
      AdjustValue {bIsDynamic=false, bIsEditable=false, bIsStaticSwitch=false, bIsVisible=false, type {classStructOrEnum {refPath=&NiagaraBool}}}
      Value Range {
       bIsDynamic=false
       bIsEditable=false
       bIsStaticSwitch=false
       bIsVisible=false
       type {classStructOrEnum {refPath=&Vector2f}}
      }
      AdjustAlpha {bIsDynamic=false, bIsEditable=false, bIsStaticSwitch=false, bIsVisible=false, type {classStructOrEnum {refPath=&NiagaraBool}}}
      Alpha Scale Range {
       bIsDynamic=false
       bIsEditable=false
       bIsStaticSwitch=false
       bIsVisible=false
       type {classStructOrEnum {refPath=&Vector2f}}
      }
      Color Channel Mode {
       bIsDynamic=false
       bIsEditable=false
       bIsStaticSwitch=true
       bIsVisible=false
       type {classStructOrEnum {refPath=&ENiagaraChannelCorrelation}}
      }
      Color Randomness Mode {
       bIsDynamic=false
       bIsEditable=false
       bIsStaticSwitch=true
       bIsVisible=false
       type {classStructOrEnum {refPath=&ENiagaraRandomnessMode}}
      }
      Position Mode {
       bIsDynamic=false
       bIsEditable=true
       bIsStaticSwitch=true
       bIsVisible=true
       type {classStructOrEnum {refPath=&ENiagara_PositionInitializationMode}}
      }
      Position {
       bIsDynamic=false
       bIsEditable=false
       bIsStaticSwitch=false
       bIsVisible=false
       type {classStructOrEnum {refPath=&NiagaraPosition}}
      }
      UsePositionOffset {bIsDynamic=false, bIsEditable=true, bIsStaticSwitch=false, bIsVisible=true, type {classStructOrEnum {refPath=&NiagaraBool}}}
      Position Offset {bIsDynamic=false, bIsEditable=true, bIsStaticSwitch=false, bIsVisible=true, type {classStructOrEnum {refPath=&Vector3f}}}
      Position Offset Coordinate Space {
       bIsDynamic=false
       bIsEditable=true
       bIsStaticSwitch=false
       bIsVisible=true
       type {classStructOrEnum {refPath=&ENiagaraCoordinateSpace}}
      }
      Mass Mode {
       bIsDynamic=false
       bIsEditable=true
       bIsStaticSwitch=true
       bIsVisible=true
       type {classStructOrEnum {refPath=&ENiagara_MassInitializationMode}}
      }
      Mass Min {
       bIsDynamic=false
       bIsEditable=false
       bIsStaticSwitch=false
       bIsVisible=false
       type {classStructOrEnum {refPath=&NiagaraFloat}}
      }
      Mass {
       bIsDynamic=false
       bIsEditable=false
       bIsStaticSwitch=false
       bIsVisible=false
       type {classStructOrEnum {refPath=&NiagaraFloat}}
      }
      Mass Max {
       bIsDynamic=false
       bIsEditable=false
       bIsStaticSwitch=false
       bIsVisible=false
       type {classStructOrEnum {refPath=&NiagaraFloat}}
      }
      Mass Randomness Mode {
       bIsDynamic=false
       bIsEditable=false
       bIsStaticSwitch=true
       bIsVisible=false
       type {classStructOrEnum {refPath=&ENiagaraRandomnessMode}}
      }
      Write Material Random {bIsDynamic=false, bIsEditable=true, bIsStaticSwitch=false, bIsVisible=true, type {classStructOrEnum {refPath=&NiagaraBool}}}
      Material Random {bIsDynamic=false, bIsEditable=true, bIsStaticSwitch=false, bIsVisible=true, type {classStructOrEnum {refPath=&NiagaraFloat}}}
      Sprite Size Mode {
       bIsDynamic=false
       bIsEditable=true
       bIsStaticSwitch=true
       bIsVisible=true
       type {classStructOrEnum {refPath=&ENiagara_SizeScaleMode}}
      }
      Sprite Size {
       bIsDynamic=false
       bIsEditable=false
       bIsStaticSwitch=false
       bIsVisible=false
       type {classStructOrEnum {refPath=&Vector2f}}
      }
      Uniform Sprite Size {bIsDynamic=true, bIsEditable=true, bIsStaticSwitch=false, bIsVisible=true, type {classStructOrEnum {refPath=&NiagaraFloat}}}
      Uniform Sprite Size Min {
       bIsDynamic=false
       bIsEditable=false
       bIsStaticSwitch=false
       bIsVisible=false
       type {classStructOrEnum {refPath=&NiagaraFloat}}
      }
      Sprite Size Min {
       bIsDynamic=false
       bIsEditable=false
       bIsStaticSwitch=false
       bIsVisible=false
       type {classStructOrEnum {refPath=&Vector2f}}
      }
      Sprite Size Max {
       bIsDynamic=false
       bIsEditable=false
       bIsStaticSwitch=false
       bIsVisible=false
       type {classStructOrEnum {refPath=&Vector2f}}
      }
      Uniform Sprite Size Max {
       bIsDynamic=false
       bIsEditable=false
       bIsStaticSwitch=false
       bIsVisible=false
       type {classStructOrEnum {refPath=&NiagaraFloat}}
      }
      Sprite Size Randomness Mode {
       bIsDynamic=false
       bIsEditable=false
       bIsStaticSwitch=true
       bIsVisible=false
       type {classStructOrEnum {refPath=&ENiagaraRandomnessMode}}
      }
      Sprite Rotation Mode {
       bIsDynamic=false
       bIsEditable=true
       bIsStaticSwitch=true
       bIsVisible=true
       type {classStructOrEnum {refPath=&ENiagara_SpriteRotationMode}}
      }
      Sprite Rotation Angle Min {
       bIsDynamic=false
       bIsEditable=false
       bIsStaticSwitch=false
       bIsVisible=false
       type {classStructOrEnum {refPath=&NiagaraFloat}}
      }
      Sprite Rotation Angle {bIsDynamic=true, bIsEditable=true, bIsStaticSwitch=false, bIsVisible=true, type {classStructOrEnum {refPath=&NiagaraFloat}}}
      Sprite Rotation Angle Max {
       bIsDynamic=false
       bIsEditable=false
       bIsStaticSwitch=false
       bIsVisible=false
       type {classStructOrEnum {refPath=&NiagaraFloat}}
      }
      Sprite Rotation Randomness Mode {
       bIsDynamic=false
       bIsEditable=false
       bIsStaticSwitch=true
       bIsVisible=false
       type {classStructOrEnum {refPath=&ENiagaraRandomnessMode}}
      }
      Sprite UV Mode {
       bIsDynamic=false
       bIsEditable=true
       bIsStaticSwitch=true
       bIsVisible=true
       type {classStructOrEnum {refPath=&ENiagara_UVFlippingMode}}
      }
      Sprite UV Randomness Mode {
       bIsDynamic=false
       bIsEditable=false
       bIsStaticSwitch=true
       bIsVisible=false
       type {classStructOrEnum {refPath=&ENiagaraRandomnessMode}}
      }
      Sprite UV Scale {
       bIsDynamic=false
       bIsEditable=false
       bIsStaticSwitch=false
       bIsVisible=false
       type {classStructOrEnum {refPath=&Vector2f}}
      }
      Mesh Scale Mode {
       bIsDynamic=false
       bIsEditable=true
       bIsStaticSwitch=true
       bIsVisible=true
       type {classStructOrEnum {refPath=&ENiagara_SizeScaleMode}}
      }
      Mesh Uniform Scale Min {
       bIsDynamic=false
       bIsEditable=false
       bIsStaticSwitch=false
       bIsVisible=false
       type {classStructOrEnum {refPath=&NiagaraFloat}}
      }
      Mesh Uniform Scale {bIsDynamic=true, bIsEditable=true, bIsStaticSwitch=false, bIsVisible=true, type {classStructOrEnum {refPath=&NiagaraFloat}}}
      Mesh Scale Min {
       bIsDynamic=false
       bIsEditable=false
       bIsStaticSwitch=false
       bIsVisible=false
       type {classStructOrEnum {refPath=&Vector3f}}
      }
      Mesh Scale {
       bIsDynamic=false
       bIsEditable=false
       bIsStaticSwitch=false
       bIsVisible=false
       type {classStructOrEnum {refPath=&Vector3f}}
      }
      Mesh Scale Max {
       bIsDynamic=false
       bIsEditable=false
       bIsStaticSwitch=false
       bIsVisible=false
       type {classStructOrEnum {refPath=&Vector3f}}
      }
      Mesh Uniform Scale Max {
       bIsDynamic=false
       bIsEditable=false
       bIsStaticSwitch=false
       bIsVisible=false
       type {classStructOrEnum {refPath=&NiagaraFloat}}
      }
      Mesh Scale Randomness Mode {
       bIsDynamic=false
       bIsEditable=false
       bIsStaticSwitch=true
       bIsVisible=false
       type {classStructOrEnum {refPath=&ENiagaraRandomnessMode}}
      }
      Mesh Renderer Array Visibility Mode {
       bIsDynamic=false
       bIsEditable=true
       bIsStaticSwitch=true
       bIsVisible=true
       type {classStructOrEnum {refPath=&ENiagara_UnsetDirectSetRandom}}
      }
      Mesh Renderer Mesh Index {
       bIsDynamic=false
       bIsEditable=false
       bIsStaticSwitch=false
       bIsVisible=false
       type {classStructOrEnum {refPath=&NiagaraInt32}}
      }
      Wrap Index By Mesh Count {bIsDynamic=false, bIsEditable=false, bIsStaticSwitch=true, bIsVisible=false, type {classStructOrEnum {refPath=&NiagaraBool}}}
      Mesh Renderer Info {
       bIsDynamic=false
       bIsEditable=false
       bIsStaticSwitch=false
       bIsVisible=false
       type {classStructOrEnum {refPath=&NiagaraDataInterfaceMeshRendererInfo}}
      }
      Mesh Renderer Random Seed {
       bIsDynamic=false
       bIsEditable=false
       bIsStaticSwitch=false
       bIsVisible=false
       type {classStructOrEnum {refPath=&NiagaraInt32}}
      }
      Ribbon Width Mode {
       bIsDynamic=false
       bIsEditable=true
       bIsStaticSwitch=true
       bIsVisible=true
       type {classStructOrEnum {refPath=&ENiagara_UnsetDirectSet}}
      }
      Ribbon Width {
       bIsDynamic=false
       bIsEditable=false
       bIsStaticSwitch=false
       bIsVisible=false
       type {classStructOrEnum {refPath=&NiagaraFloat}}
      }
      Ribbon Facing Mode {
       bIsDynamic=false
       bIsEditable=true
       bIsStaticSwitch=true
       bIsVisible=true
       type {classStructOrEnum {refPath=&ENiagara_UnsetDirectSet}}
      }
      Ribbon Facing Vector {
       bIsDynamic=false
       bIsEditable=false
       bIsStaticSwitch=false
       bIsVisible=false
       type {classStructOrEnum {refPath=&Vector3f}}
      }
      Ribbon Twist Mode {
       bIsDynamic=false
       bIsEditable=true
       bIsStaticSwitch=true
       bIsVisible=true
       type {classStructOrEnum {refPath=&ENiagara_UnsetDirectSet}}
      }
      "Ribbon Twist " {
       bIsDynamic=false
       bIsEditable=false
       bIsStaticSwitch=false
       bIsVisible=false
       type {classStructOrEnum {refPath=&NiagaraFloat}}
      }
     ]
     moduleScript {refPath=&InitializeParticle}
    }
    {
     bIsSetParametersModule=true
     enabled=true
     moduleName=SetVariables_#2
     inputs [Particles.InitSpriteRotRate {bIsDynamic=true, bIsEditable=true, bIsStaticSwitch=false, bIsVisible=true, type {classStructOrEnum {refPath=&NiagaraFloat}}}]
     moduleScript {refPath=&self:DefaultPropSpawn_Merged003_0.NiagaraScriptSource_1.NiagaraGraph_0.NiagaraNodeAssignment_0.SetVariables_#2}
    }
    {
     bIsSetParametersModule=false
     enabled=true
     moduleName=AccelerationForce
     inputs [
      Acceleration {bIsDynamic=true, bIsEditable=true, bIsStaticSwitch=false, bIsVisible=true, type {classStructOrEnum {refPath=&Vector3f}}}
      Coordinate Space {
       bIsDynamic=false
       bIsEditable=true
       bIsStaticSwitch=false
       bIsVisible=true
       type {classStructOrEnum {refPath=&ENiagaraCoordinateSpace}}
      }
     ]
     moduleScript {refPath=&AccelerationForce}
    }
    {
     bIsSetParametersModule=false
     enabled=true
     moduleName=ShapeLocation
     inputs [
      Shape Primitive {
       bIsDynamic=false
       bIsEditable=true
       bIsStaticSwitch=true
       bIsVisible=true
       type {classStructOrEnum {refPath=&ENiagara_LocationShapes}}
      }
      Box / Plane Mode {
       bIsDynamic=false
       bIsEditable=false
       bIsStaticSwitch=true
       bIsVisible=false
       type {classStructOrEnum {refPath=&ENiagara_BoxPlaneMode}}
      }
      Cone Mode {
       bIsDynamic=false
       bIsEditable=false
       bIsStaticSwitch=true
       bIsVisible=false
       type {classStructOrEnum {refPath=&ENiagara_ConeMode}}
      }
      Cylinder Height {
       bIsDynamic=false
       bIsEditable=false
       bIsStaticSwitch=false
       bIsVisible=false
       type {classStructOrEnum {refPath=&NiagaraFloat}}
      }
      Sphere Radius {bIsDynamic=true, bIsEditable=true, bIsStaticSwitch=false, bIsVisible=true, type {classStructOrEnum {refPath=&NiagaraFloat}}}
      Ring / Disc Mode {
       bIsDynamic=false
       bIsEditable=false
       bIsStaticSwitch=true
       bIsVisible=false
       type {classStructOrEnum {refPath=&ENiagaraRingDiscMode}}
      }
      Torus Mode {
       bIsDynamic=false
       bIsEditable=false
       bIsStaticSwitch=true
       bIsVisible=false
       type {classStructOrEnum {refPath=&ENiagaraShapeTorusMode}}
      }
      Knot Ring Loops {
       bIsDynamic=false
       bIsEditable=false
       bIsStaticSwitch=false
       bIsVisible=false
       type {classStructOrEnum {refPath=&NiagaraInt32}}
      }
      Knot Helix Loops {
       bIsDynamic=false
       bIsEditable=false
       bIsStaticSwitch=false
       bIsVisible=false
       type {classStructOrEnum {refPath=&NiagaraInt32}}
      }
      Enable Torus Knot Radius {bIsDynamic=false, bIsEditable=false, bIsStaticSwitch=false, bIsVisible=false, type {classStructOrEnum {refPath=&NiagaraBool}}}
      Torus Knot Tube Radius {
       bIsDynamic=false
       bIsEditable=false
       bIsStaticSwitch=false
       bIsVisible=false
       type {classStructOrEnum {refPath=&NiagaraFloat}}
      }
      Cylinder Radius {
       bIsDynamic=false
       bIsEditable=false
       bIsStaticSwitch=false
       bIsVisible=false
       type {classStructOrEnum {refPath=&NiagaraFloat}}
      }
      Cylinder Height Midpoint {
       bIsDynamic=false
       bIsEditable=false
       bIsStaticSwitch=false
       bIsVisible=false
       type {classStructOrEnum {refPath=&NiagaraFloat}}
      }
      Ring Radius {
       bIsDynamic=false
       bIsEditable=false
       bIsStaticSwitch=false
       bIsVisible=false
       type {classStructOrEnum {refPath=&NiagaraFloat}}
      }
      Box Size {
       bIsDynamic=false
       bIsEditable=false
       bIsStaticSwitch=false
       bIsVisible=false
       type {classStructOrEnum {refPath=&Vector3f}}
      }
      Box Midpoint {
       bIsDynamic=false
       bIsEditable=false
       bIsStaticSwitch=false
       bIsVisible=false
       type {classStructOrEnum {refPath=&Vector3f}}
      }
      Cone Length {
       bIsDynamic=false
       bIsEditable=false
       bIsStaticSwitch=false
       bIsVisible=false
       type {classStructOrEnum {refPath=&NiagaraFloat}}
      }
      Plane Size {
       bIsDynamic=false
       bIsEditable=false
       bIsStaticSwitch=false
       bIsVisible=false
       type {classStructOrEnum {refPath=&Vector2f}}
      }
      Wedge Horizontal Angle {
       bIsDynamic=false
       bIsEditable=false
       bIsStaticSwitch=false
       bIsVisible=false
       type {classStructOrEnum {refPath=&NiagaraFloat}}
      }
      Large Radius {
       bIsDynamic=false
       bIsEditable=false
       bIsStaticSwitch=false
       bIsVisible=false
       type {classStructOrEnum {refPath=&NiagaraFloat}}
      }
      Plane Midpoint {
       bIsDynamic=false
       bIsEditable=false
       bIsStaticSwitch=false
       bIsVisible=false
       type {classStructOrEnum {refPath=&Vector2f}}
      }
      Wedge Vertical Angle {
       bIsDynamic=false
       bIsEditable=false
       bIsStaticSwitch=false
       bIsVisible=false
       type {classStructOrEnum {refPath=&NiagaraFloat}}
      }
      Cone Angle {
       bIsDynamic=false
       bIsEditable=false
       bIsStaticSwitch=false
       bIsVisible=false
       type {classStructOrEnum {refPath=&NiagaraFloat}}
      }
      Handle Radius {
       bIsDynamic=false
       bIsEditable=false
       bIsStaticSwitch=false
       bIsVisible=false
       type {classStructOrEnum {refPath=&NiagaraFloat}}
      }
      Cone Inner Angle {
       bIsDynamic=false
       bIsEditable=false
       bIsStaticSwitch=false
       bIsVisible=false
       type {classStructOrEnum {refPath=&NiagaraFloat}}
      }
      Cone Radial Angle {
       bIsDynamic=false
       bIsEditable=false
       bIsStaticSwitch=false
       bIsVisible=false
       type {classStructOrEnum {refPath=&NiagaraFloat}}
      }
      Cone Angle Type {
       bIsDynamic=false
       bIsEditable=false
       bIsStaticSwitch=true
       bIsVisible=false
       type {classStructOrEnum {refPath=&ENiagara_AngleInput}}
      }
      Flatten Endcaps {bIsDynamic=false, bIsEditable=false, bIsStaticSwitch=true, bIsVisible=false, type {classStructOrEnum {refPath=&NiagaraBool}}}
      Apply To Particle Position {
       bIsDynamic=false
       bIsEditable=true
       bIsStaticSwitch=true
       bIsVisible=true
       type {classStructOrEnum {refPath=&ENiagara_AttributeSamplingApplyOutput}}
      }
      Cone Surface Distribution {
       bIsDynamic=false
       bIsEditable=false
       bIsStaticSwitch=false
       bIsVisible=false
       type {classStructOrEnum {refPath=&NiagaraFloat}}
      }
      Ring / Disc Distribution Mode {
       bIsDynamic=false
       bIsEditable=false
       bIsStaticSwitch=true
       bIsVisible=false
       type {classStructOrEnum {refPath=&ENiagaraSphereDistributionMode}}
      }
      Sphere Distribution {
       bIsDynamic=false
       bIsEditable=true
       bIsStaticSwitch=true
       bIsVisible=true
       type {classStructOrEnum {refPath=&ENiagaraSphereDistributionMode}}
      }
      Torus Distribution Mode {
       bIsDynamic=false
       bIsEditable=false
       bIsStaticSwitch=true
       bIsVisible=false
       type {classStructOrEnum {refPath=&ENiagaraTorusDistributionMode}}
      }
      Sphere Surface Distribution {bIsDynamic=false, bIsEditable=true, bIsStaticSwitch=false, bIsVisible=true, type {classStructOrEnum {refPath=&NiagaraFloat}}}
      Hemisphere Distribution {bIsDynamic=false, bIsEditable=true, bIsStaticSwitch=false, bIsVisible=true, type {classStructOrEnum {refPath=&Vector2f}}}
      Cylinder Distribution {
       bIsDynamic=false
       bIsEditable=false
       bIsStaticSwitch=true
       bIsVisible=false
       type {classStructOrEnum {refPath=&ENiagara_CylinderMode}}
      }
      Surface Only {bIsDynamic=false, bIsEditable=false, bIsStaticSwitch=false, bIsVisible=false, type {classStructOrEnum {refPath=&NiagaraBool}}}
      Surface Only Band Thickness {
       bIsDynamic=false
       bIsEditable=false
       bIsStaticSwitch=false
       bIsVisible=false
       type {classStructOrEnum {refPath=&NiagaraFloat}}
      }
      Use Endcaps In Surface Only Mode {bIsDynamic=false, bIsEditable=false, bIsStaticSwitch=true, bIsVisible=false, type {classStructOrEnum {refPath=&NiagaraBool}}}
      Hemicircle X {bIsDynamic=false, bIsEditable=false, bIsStaticSwitch=true, bIsVisible=false, type {classStructOrEnum {refPath=&NiagaraBool}}}
      Hemicircle Y {bIsDynamic=false, bIsEditable=false, bIsStaticSwitch=true, bIsVisible=false, type {classStructOrEnum {refPath=&NiagaraBool}}}
      Disc Coverage {
       bIsDynamic=false
       bIsEditable=false
       bIsStaticSwitch=false
       bIsVisible=false
       type {classStructOrEnum {refPath=&NiagaraFloat}}
      }
      Hemicircle Internal Cap {bIsDynamic=false, bIsEditable=false, bIsStaticSwitch=true, bIsVisible=false, type {classStructOrEnum {refPath=&NiagaraBool}}}
      U Position {
       bIsDynamic=false
       bIsEditable=false
       bIsStaticSwitch=false
       bIsVisible=false
       type {classStructOrEnum {refPath=&NiagaraFloat}}
      }
      Enable Lathe Profile {bIsDynamic=false, bIsEditable=false, bIsStaticSwitch=true, bIsVisible=false, type {classStructOrEnum {refPath=&NiagaraBool}}}
      Lathe Profile {
       bIsDynamic=false
       bIsEditable=false
       bIsStaticSwitch=false
       bIsVisible=false
       type {classStructOrEnum {refPath=&NiagaraDataInterfaceCurve}}
      }
      Surface Distribution {
       bIsDynamic=false
       bIsEditable=false
       bIsStaticSwitch=false
       bIsVisible=false
       type {classStructOrEnum {refPath=&NiagaraFloat}}
      }
      Surface Expansion Mode {
       bIsDynamic=false
       bIsEditable=false
       bIsStaticSwitch=false
       bIsVisible=false
       type {classStructOrEnum {refPath=&ENiagaraExpansionMode}}
      }
      U Distribution {
       bIsDynamic=false
       bIsEditable=false
       bIsStaticSwitch=false
       bIsVisible=false
       type {classStructOrEnum {refPath=&NiagaraFloat}}
      }
      V Position {
       bIsDynamic=false
       bIsEditable=false
       bIsStaticSwitch=false
       bIsVisible=false
       type {classStructOrEnum {refPath=&NiagaraFloat}}
      }
      Radius Position {
       bIsDynamic=false
       bIsEditable=false
       bIsStaticSwitch=false
       bIsVisible=false
       type {classStructOrEnum {refPath=&NiagaraFloat}}
      }
      Sphere Radius Position {
       bIsDynamic=false
       bIsEditable=false
       bIsStaticSwitch=false
       bIsVisible=false
       type {classStructOrEnum {refPath=&NiagaraFloat}}
      }
      V Distribution {
       bIsDynamic=false
       bIsEditable=false
       bIsStaticSwitch=false
       bIsVisible=false
       type {classStructOrEnum {refPath=&NiagaraFloat}}
      }
      Evaluation Type {
       bIsDynamic=false
       bIsEditable=false
       bIsStaticSwitch=true
       bIsVisible=false
       type {classStructOrEnum {refPath=&ENiagaraRandomnessEvaluation}}
      }
      Uniform Distribution {
       bIsDynamic=false
       bIsEditable=false
       bIsStaticSwitch=false
       bIsVisible=false
       type {classStructOrEnum {refPath=&NiagaraFloat}}
      }
      Uniform Spiral Amount {
       bIsDynamic=false
       bIsEditable=false
       bIsStaticSwitch=false
       bIsVisible=false
       type {classStructOrEnum {refPath=&NiagaraFloat}}
      }
      Uniform Spiral Falloff {
       bIsDynamic=false
       bIsEditable=false
       bIsStaticSwitch=false
       bIsVisible=false
       type {classStructOrEnum {refPath=&NiagaraFloat}}
      }
      Uniform Count {
       bIsDynamic=false
       bIsEditable=false
       bIsStaticSwitch=false
       bIsVisible=false
       type {classStructOrEnum {refPath=&NiagaraInt32}}
      }
      Uniform Particle Index {
       bIsDynamic=false
       bIsEditable=false
       bIsStaticSwitch=false
       bIsVisible=false
       type {classStructOrEnum {refPath=&NiagaraInt32}}
      }
      Hemisphere Angle Type {
       bIsDynamic=false
       bIsEditable=true
       bIsStaticSwitch=true
       bIsVisible=true
       type {classStructOrEnum {refPath=&ENiagara_AngleInput}}
      }
      Uniform Endpoint Offset {
       bIsDynamic=false
       bIsEditable=false
       bIsStaticSwitch=false
       bIsVisible=false
       type {classStructOrEnum {refPath=&NiagaraFloat}}
      }
      MaskSpawnGroup {bIsDynamic=false, bIsEditable=true, bIsStaticSwitch=false, bIsVisible=true, type {classStructOrEnum {refPath=&NiagaraBool}}}
      Spawn Group Mask {bIsDynamic=false, bIsEditable=false, bIsStaticSwitch=false, bIsVisible=true, type {classStructOrEnum {refPath=&NiagaraInt32}}}
      Spawn Group {
       bIsDynamic=false
       bIsEditable=false
       bIsStaticSwitch=false
       bIsVisible=false
       type {classStructOrEnum {refPath=&NiagaraInt32}}
      }
      Randomness Mode {
       bIsDynamic=false
       bIsEditable=true
       bIsStaticSwitch=true
       bIsVisible=true
       type {classStructOrEnum {refPath=&ENiagaraRandomnessMode}}
      }
      Override Seed {bIsDynamic=false, bIsEditable=true, bIsStaticSwitch=false, bIsVisible=true, type {classStructOrEnum {refPath=&NiagaraBool}}}
      Random Seed {bIsDynamic=false, bIsEditable=false, bIsStaticSwitch=false, bIsVisible=true, type {classStructOrEnum {refPath=&NiagaraInt32}}}
      Fixed Random Seed {bIsDynamic=false, bIsEditable=false, bIsStaticSwitch=true, bIsVisible=false, type {classStructOrEnum {refPath=&NiagaraBool}}}
      Transform Order {
       bIsDynamic=false
       bIsEditable=true
       bIsStaticSwitch=true
       bIsVisible=true
       type {classStructOrEnum {refPath=&ENiagara_TransformOrder}}
      }
      Scale Mode {
       bIsDynamic=false
       bIsEditable=true
       bIsStaticSwitch=true
       bIsVisible=true
       type {classStructOrEnum {refPath=&ENiagara_ScaleMode}}
      }
      Non Uniform Scale {bIsDynamic=false, bIsEditable=true, bIsStaticSwitch=false, bIsVisible=true, type {classStructOrEnum {refPath=&Vector3f}}}
      UseOwnerScale {bIsDynamic=false, bIsEditable=true, bIsStaticSwitch=false, bIsVisible=true, type {classStructOrEnum {refPath=&NiagaraBool}}}
      Apply Owner Scale {bIsDynamic=false, bIsEditable=true, bIsStaticSwitch=false, bIsVisible=true, type {classStructOrEnum {refPath=&Vector3f}}}
      Rotation Mode {
       bIsDynamic=false
       bIsEditable=true
       bIsStaticSwitch=true
       bIsVisible=true
       type {classStructOrEnum {refPath=&ENiagara_RotationMode}}
      }
      Rotation Axis {
       bIsDynamic=false
       bIsEditable=false
       bIsStaticSwitch=false
       bIsVisible=false
       type {classStructOrEnum {refPath=&Vector3f}}
      }
      Rotation Matrix {
       bIsDynamic=false
       bIsEditable=false
       bIsStaticSwitch=false
       bIsVisible=false
       type {classStructOrEnum {refPath=&NiagaraMatrix}}
      }
      Rotation Quaternion {bIsDynamic=false, bIsEditable=false, bIsStaticSwitch=false, bIsVisible=false, type {classStructOrEnum {refPath=&Quat4f}}}
      Yaw / Pitch / Roll {
       bIsDynamic=false
       bIsEditable=false
       bIsStaticSwitch=false
       bIsVisible=false
       type {classStructOrEnum {refPath=&Vector3f}}
      }
      Rotation Angle {
       bIsDynamic=false
       bIsEditable=false
       bIsStaticSwitch=false
       bIsVisible=false
       type {classStructOrEnum {refPath=&NiagaraFloat}}
      }
      Invert Rotation Quaternion {bIsDynamic=false, bIsEditable=false, bIsStaticSwitch=false, bIsVisible=false, type {classStructOrEnum {refPath=&NiagaraBool}}}
      Additional Quaternion Rotation {bIsDynamic=false, bIsEditable=false, bIsStaticSwitch=false, bIsVisible=false, type {classStructOrEnum {refPath=&NiagaraBool}}}
      Additional Yaw / Pitch / Roll {
       bIsDynamic=false
       bIsEditable=false
       bIsStaticSwitch=false
       bIsVisible=false
       type {classStructOrEnum {refPath=&Vector3f}}
      }
      Rotation Angle Type {
       bIsDynamic=false
       bIsEditable=false
       bIsStaticSwitch=true
       bIsVisible=false
       type {classStructOrEnum {refPath=&ENiagara_AngleInput}}
      }
      Rotation Coordinate Space {
       bIsDynamic=false
       bIsEditable=true
       bIsStaticSwitch=false
       bIsVisible=true
       type {classStructOrEnum {refPath=&ENiagaraCoordinateSpace}}
      }
      Offset Mode {
       bIsDynamic=false
       bIsEditable=true
       bIsStaticSwitch=true
       bIsVisible=true
       type {classStructOrEnum {refPath=&ENiagara_OffsetMode}}
      }
      Shape Origin {
       bIsDynamic=false
       bIsEditable=true
       bIsStaticSwitch=false
       bIsVisible=true
       type {classStructOrEnum {refPath=&NiagaraPosition}}
      }
      Offset {bIsDynamic=false, bIsEditable=true, bIsStaticSwitch=false, bIsVisible=true, type {classStructOrEnum {refPath=&Vector3f}}}
      Offset Coordinate Space {
       bIsDynamic=false
       bIsEditable=true
       bIsStaticSwitch=false
       bIsVisible=true
       type {classStructOrEnum {refPath=&ENiagaraCoordinateSpace}}
      }
      Custom Transform Matrix {
       bIsDynamic=false
       bIsEditable=false
       bIsStaticSwitch=false
       bIsVisible=false
       type {classStructOrEnum {refPath=&NiagaraMatrix}}
      }
      Transform Method {
       bIsDynamic=false
       bIsEditable=true
       bIsStaticSwitch=true
       bIsVisible=true
       type {classStructOrEnum {refPath=&ENiagara_TransformType}}
      }
     ]
     moduleScript {refPath=&ShapeLocation}
    }
    {
     bIsSetParametersModule=false
     enabled=true
     moduleName=CameraOffset
     inputs [Camera Offset Amount {bIsDynamic=true, bIsEditable=true, bIsStaticSwitch=false, bIsVisible=true, type {classStructOrEnum {refPath=&NiagaraFloat}}}]
     moduleScript {refPath=&CameraOffset}
    }
    {
     bIsSetParametersModule=false
     enabled=true
     moduleName=AddVelocity001
     inputs [
      Velocity Mode {
       bIsDynamic=false
       bIsEditable=true
       bIsStaticSwitch=true
       bIsVisible=true
       type {classStructOrEnum {refPath=&ENiagara_VelocityMode}}
      }
      Velocity {bIsDynamic=true, bIsEditable=false, bIsStaticSwitch=false, bIsVisible=false, type {classStructOrEnum {refPath=&Vector3f}}}
      Velocity Speed {bIsDynamic=true, bIsEditable=true, bIsStaticSwitch=false, bIsVisible=true, type {classStructOrEnum {refPath=&NiagaraFloat}}}
      Velocity Speed Scale {
       bIsDynamic=false
       bIsEditable=false
       bIsStaticSwitch=false
       bIsVisible=false
       type {classStructOrEnum {refPath=&NiagaraFloat}}
      }
      Module.Use Falloff Distance {bIsDynamic=false, bIsEditable=true, bIsStaticSwitch=false, bIsVisible=true, type {classStructOrEnum {refPath=&NiagaraBool}}}
      Constrain To Radius {bIsDynamic=false, bIsEditable=false, bIsStaticSwitch=false, bIsVisible=true, type {classStructOrEnum {refPath=&NiagaraFloat}}}
      Distribution Along Cone Axis {
       bIsDynamic=false
       bIsEditable=false
       bIsStaticSwitch=false
       bIsVisible=false
       type {classStructOrEnum {refPath=&NiagaraFloat}}
      }
      Radius Falloff Near / Far {
       bIsDynamic=false
       bIsEditable=false
       bIsStaticSwitch=false
       bIsVisible=false
       type {classStructOrEnum {refPath=&Vector2f}}
      }
      Radius Falloff Exponent {
       bIsDynamic=false
       bIsEditable=false
       bIsStaticSwitch=false
       bIsVisible=false
       type {classStructOrEnum {refPath=&NiagaraFloat}}
      }
      Speed Falloff From Cone Axis {
       bIsDynamic=false
       bIsEditable=false
       bIsStaticSwitch=false
       bIsVisible=false
       type {classStructOrEnum {refPath=&NiagaraFloat}}
      }
      Invert Velocity Falloff {bIsDynamic=false, bIsEditable=false, bIsStaticSwitch=false, bIsVisible=false, type {classStructOrEnum {refPath=&NiagaraBool}}}
      Override Seed {bIsDynamic=false, bIsEditable=true, bIsStaticSwitch=false, bIsVisible=true, type {classStructOrEnum {refPath=&NiagaraBool}}}
      Random Seed {bIsDynamic=false, bIsEditable=false, bIsStaticSwitch=false, bIsVisible=true, type {classStructOrEnum {refPath=&NiagaraInt32}}}
      Fixed Random Seed {bIsDynamic=false, bIsEditable=false, bIsStaticSwitch=true, bIsVisible=false, type {classStructOrEnum {refPath=&NiagaraBool}}}
      Randomness Mode {
       bIsDynamic=false
       bIsEditable=true
       bIsStaticSwitch=true
       bIsVisible=true
       type {classStructOrEnum {refPath=&ENiagaraRandomnessMode}}
      }
      Cone Axis {
       bIsDynamic=false
       bIsEditable=false
       bIsStaticSwitch=false
       bIsVisible=false
       type {classStructOrEnum {refPath=&Vector3f}}
      }
      Cone Angle {
       bIsDynamic=false
       bIsEditable=false
       bIsStaticSwitch=false
       bIsVisible=false
       type {classStructOrEnum {refPath=&NiagaraFloat}}
      }
      Inner Cone Angle {
       bIsDynamic=false
       bIsEditable=false
       bIsStaticSwitch=false
       bIsVisible=false
       type {classStructOrEnum {refPath=&NiagaraFloat}}
      }
      Cone Angle Mode {
       bIsDynamic=false
       bIsEditable=false
       bIsStaticSwitch=true
       bIsVisible=false
       type {classStructOrEnum {refPath=&ENiagara_AngleInput}}
      }
      Rotation Mode {
       bIsDynamic=false
       bIsEditable=false
       bIsStaticSwitch=true
       bIsVisible=false
       type {classStructOrEnum {refPath=&ENiagara_RotationMode}}
      }
      Rotation Axis {
       bIsDynamic=false
       bIsEditable=false
       bIsStaticSwitch=false
       bIsVisible=false
       type {classStructOrEnum {refPath=&Vector3f}}
      }
      Yaw / Pitch / Roll {
       bIsDynamic=false
       bIsEditable=false
       bIsStaticSwitch=false
       bIsVisible=false
       type {classStructOrEnum {refPath=&Vector3f}}
      }
      Rotation Quaternion {bIsDynamic=false, bIsEditable=false, bIsStaticSwitch=false, bIsVisible=false, type {classStructOrEnum {refPath=&Quat4f}}}
      Rotation Matrix {
       bIsDynamic=false
       bIsEditable=false
       bIsStaticSwitch=false
       bIsVisible=false
       type {classStructOrEnum {refPath=&NiagaraMatrix}}
      }
      Rotation Angle {
       bIsDynamic=false
       bIsEditable=false
       bIsStaticSwitch=false
       bIsVisible=false
       type {classStructOrEnum {refPath=&NiagaraFloat}}
      }
      Invert Rotation Quaternion {bIsDynamic=false, bIsEditable=false, bIsStaticSwitch=false, bIsVisible=false, type {classStructOrEnum {refPath=&NiagaraBool}}}
      Additional Quaternion Rotation {bIsDynamic=false, bIsEditable=false, bIsStaticSwitch=false, bIsVisible=false, type {classStructOrEnum {refPath=&NiagaraBool}}}
      Additional Yaw / Pitch / Roll {
       bIsDynamic=false
       bIsEditable=false
       bIsStaticSwitch=false
       bIsVisible=false
       type {classStructOrEnum {refPath=&Vector3f}}
      }
      Rotation Angle Type {
       bIsDynamic=false
       bIsEditable=false
       bIsStaticSwitch=true
       bIsVisible=false
       type {classStructOrEnum {refPath=&ENiagara_AngleInput}}
      }
      Rotation Coordinate Space {
       bIsDynamic=false
       bIsEditable=false
       bIsStaticSwitch=false
       bIsVisible=false
       type {classStructOrEnum {refPath=&ENiagaraCoordinateSpace}}
      }
      Velocity Origin {
       bIsDynamic=false
       bIsEditable=true
       bIsStaticSwitch=false
       bIsVisible=true
       type {classStructOrEnum {refPath=&NiagaraPosition}}
      }
      Origin Offset {bIsDynamic=false, bIsEditable=true, bIsStaticSwitch=false, bIsVisible=true, type {classStructOrEnum {refPath=&Vector3f}}}
      Origin Offset Coordinate Space {
       bIsDynamic=false
       bIsEditable=true
       bIsStaticSwitch=false
       bIsVisible=true
       type {classStructOrEnum {refPath=&ENiagaraCoordinateSpace}}
      }
      Default Position {
       bIsDynamic=false
       bIsEditable=true
       bIsStaticSwitch=false
       bIsVisible=true
       type {classStructOrEnum {refPath=&NiagaraPosition}}
      }
     ]
     moduleScript {refPath=&AddVelocity}
    }
    {
     bIsSetParametersModule=false
     enabled=true
     moduleName=AddVelocity
     inputs [
      Velocity Mode {
       bIsDynamic=false
       bIsEditable=true
       bIsStaticSwitch=true
       bIsVisible=true
       type {classStructOrEnum {refPath=&ENiagara_VelocityMode}}
      }
      Velocity {bIsDynamic=true, bIsEditable=true, bIsStaticSwitch=false, bIsVisible=true, type {classStructOrEnum {refPath=&Vector3f}}}
      Velocity Speed {
       bIsDynamic=false
       bIsEditable=false
       bIsStaticSwitch=false
       bIsVisible=false
       type {classStructOrEnum {refPath=&NiagaraFloat}}
      }
      Velocity Speed Scale {bIsDynamic=false, bIsEditable=true, bIsStaticSwitch=false, bIsVisible=true, type {classStructOrEnum {refPath=&NiagaraFloat}}}
      Constrain To Radius {
       bIsDynamic=false
       bIsEditable=false
       bIsStaticSwitch=false
       bIsVisible=false
       type {classStructOrEnum {refPath=&NiagaraFloat}}
      }
      Distribution Along Cone Axis {
       bIsDynamic=false
       bIsEditable=false
       bIsStaticSwitch=false
       bIsVisible=false
       type {classStructOrEnum {refPath=&NiagaraFloat}}
      }
      Radius Falloff Near / Far {
       bIsDynamic=false
       bIsEditable=false
       bIsStaticSwitch=false
       bIsVisible=false
       type {classStructOrEnum {refPath=&Vector2f}}
      }
      Radius Falloff Exponent {
       bIsDynamic=false
       bIsEditable=false
       bIsStaticSwitch=false
       bIsVisible=false
       type {classStructOrEnum {refPath=&NiagaraFloat}}
      }
      Speed Falloff From Cone Axis {
       bIsDynamic=false
       bIsEditable=false
       bIsStaticSwitch=false
       bIsVisible=false
       type {classStructOrEnum {refPath=&NiagaraFloat}}
      }
      Invert Velocity Falloff {bIsDynamic=false, bIsEditable=false, bIsStaticSwitch=false, bIsVisible=false, type {classStructOrEnum {refPath=&NiagaraBool}}}
      Override Seed {bIsDynamic=false, bIsEditable=false, bIsStaticSwitch=false, bIsVisible=false, type {classStructOrEnum {refPath=&NiagaraBool}}}
      Random Seed {
       bIsDynamic=false
       bIsEditable=false
       bIsStaticSwitch=false
       bIsVisible=false
       type {classStructOrEnum {refPath=&NiagaraInt32}}
      }
      Fixed Random Seed {bIsDynamic=false, bIsEditable=false, bIsStaticSwitch=true, bIsVisible=false, type {classStructOrEnum {refPath=&NiagaraBool}}}
      Randomness Mode {
       bIsDynamic=false
       bIsEditable=false
       bIsStaticSwitch=true
       bIsVisible=false
       type {classStructOrEnum {refPath=&ENiagaraRandomnessMode}}
      }
      Cone Axis {
       bIsDynamic=false
       bIsEditable=false
       bIsStaticSwitch=false
       bIsVisible=false
       type {classStructOrEnum {refPath=&Vector3f}}
      }
      Cone Angle {
       bIsDynamic=false
       bIsEditable=false
       bIsStaticSwitch=false
       bIsVisible=false
       type {classStructOrEnum {refPath=&NiagaraFloat}}
      }
      Inner Cone Angle {
       bIsDynamic=false
       bIsEditable=false
       bIsStaticSwitch=false
       bIsVisible=false
       type {classStructOrEnum {refPath=&NiagaraFloat}}
      }
      Cone Angle Mode {
       bIsDynamic=false
       bIsEditable=false
       bIsStaticSwitch=true
       bIsVisible=false
       type {classStructOrEnum {refPath=&ENiagara_AngleInput}}
      }
      Rotation Mode {
       bIsDynamic=false
       bIsEditable=true
       bIsStaticSwitch=true
       bIsVisible=true
       type {classStructOrEnum {refPath=&ENiagara_RotationMode}}
      }
      Rotation Axis {
       bIsDynamic=false
       bIsEditable=false
       bIsStaticSwitch=false
       bIsVisible=false
       type {classStructOrEnum {refPath=&Vector3f}}
      }
      Yaw / Pitch / Roll {
       bIsDynamic=false
       bIsEditable=false
       bIsStaticSwitch=false
       bIsVisible=false
       type {classStructOrEnum {refPath=&Vector3f}}
      }
      Rotation Quaternion {bIsDynamic=false, bIsEditable=false, bIsStaticSwitch=false, bIsVisible=false, type {classStructOrEnum {refPath=&Quat4f}}}
      Rotation Matrix {
       bIsDynamic=false
       bIsEditable=false
       bIsStaticSwitch=false
       bIsVisible=false
       type {classStructOrEnum {refPath=&NiagaraMatrix}}
      }
      Rotation Angle {
       bIsDynamic=false
       bIsEditable=false
       bIsStaticSwitch=false
       bIsVisible=false
       type {classStructOrEnum {refPath=&NiagaraFloat}}
      }
      Invert Rotation Quaternion {bIsDynamic=false, bIsEditable=false, bIsStaticSwitch=false, bIsVisible=false, type {classStructOrEnum {refPath=&NiagaraBool}}}
      Additional Quaternion Rotation {bIsDynamic=false, bIsEditable=false, bIsStaticSwitch=false, bIsVisible=false, type {classStructOrEnum {refPath=&NiagaraBool}}}
      Additional Yaw / Pitch / Roll {
       bIsDynamic=false
       bIsEditable=false
       bIsStaticSwitch=false
       bIsVisible=false
       type {classStructOrEnum {refPath=&Vector3f}}
      }
      Rotation Angle Type {
       bIsDynamic=false
       bIsEditable=false
       bIsStaticSwitch=true
       bIsVisible=false
       type {classStructOrEnum {refPath=&ENiagara_AngleInput}}
      }
      Rotation Coordinate Space {
       bIsDynamic=false
       bIsEditable=true
       bIsStaticSwitch=false
       bIsVisible=true
       type {classStructOrEnum {refPath=&ENiagaraCoordinateSpace}}
      }
      Velocity Origin {
       bIsDynamic=false
       bIsEditable=false
       bIsStaticSwitch=false
       bIsVisible=false
       type {classStructOrEnum {refPath=&NiagaraPosition}}
      }
      Origin Offset {
       bIsDynamic=false
       bIsEditable=false
       bIsStaticSwitch=false
       bIsVisible=false
       type {classStructOrEnum {refPath=&Vector3f}}
      }
      Origin Offset Coordinate Space {
       bIsDynamic=false
       bIsEditable=false
       bIsStaticSwitch=false
       bIsVisible=false
       type {classStructOrEnum {refPath=&ENiagaraCoordinateSpace}}
      }
      Default Position {
       bIsDynamic=false
       bIsEditable=false
       bIsStaticSwitch=false
       bIsVisible=false
       type {classStructOrEnum {refPath=&NiagaraPosition}}
      }
     ]
     moduleScript {refPath=&AddVelocity}
    }
    {
     bIsSetParametersModule=false
     enabled=true
     moduleName=ApplyInitialForces
     inputs [
      Apply Force to Velocity {bIsDynamic=false, bIsEditable=true, bIsStaticSwitch=false, bIsVisible=true, type {classStructOrEnum {refPath=&NiagaraBool}}}
      Apply Rotational Force to Rotational Velocity {bIsDynamic=false, bIsEditable=true, bIsStaticSwitch=false, bIsVisible=true, type {classStructOrEnum {refPath=&NiagaraBool}}}
      Apply Force to Position {bIsDynamic=false, bIsEditable=true, bIsStaticSwitch=false, bIsVisible=true, type {classStructOrEnum {refPath=&NiagaraBool}}}
      Positional Force Warmup Time {bIsDynamic=false, bIsEditable=false, bIsStaticSwitch=false, bIsVisible=true, type {classStructOrEnum {refPath=&NiagaraFloat}}}
     ]
     moduleScript {refPath=&ApplyInitialForces}
    }
    {
     bIsSetParametersModule=true
     enabled=true
     moduleName=SetVariables_#3
     inputs [Particles.DynamicMaterialParameter_Air {bIsDynamic=true, bIsEditable=true, bIsStaticSwitch=false, bIsVisible=true, type {classStructOrEnum {refPath=&Vector4f}}}]
     moduleScript {refPath=&self:DefaultPropSpawn_Merged003_0.NiagaraScriptSource_1.NiagaraGraph_0.NiagaraNodeAssignment_5.SetVariables_#3}
    }
   ]
  }
  particleUpdateScript {
   scriptName=ParticleUpdateScript
   modules [
    {
     bIsSetParametersModule=false
     enabled=true
     moduleName=ParticleState
     inputs [
      Kill Particles When Lifetime Has Elapsed {bIsDynamic=false, bIsEditable=true, bIsStaticSwitch=true, bIsVisible=true, type {classStructOrEnum {refPath=&NiagaraBool}}}
      Loop Particles Lifetime {bIsDynamic=false, bIsEditable=false, bIsStaticSwitch=false, bIsVisible=false, type {classStructOrEnum {refPath=&NiagaraBool}}}
      Let Infinitely Lived Particles Die When Emitter Deactivates {bIsDynamic=false, bIsEditable=false, bIsStaticSwitch=false, bIsVisible=false, type {classStructOrEnum {refPath=&NiagaraBool}}}
      Lifetime {bIsDynamic=false, bIsEditable=true, bIsStaticSwitch=false, bIsVisible=true, type {classStructOrEnum {refPath=&NiagaraFloat}}}
      DeltaTime {bIsDynamic=false, bIsEditable=true, bIsStaticSwitch=false, bIsVisible=true, type {classStructOrEnum {refPath=&NiagaraFloat}}}
     ]
     moduleScript {refPath=&ParticleState}
    }
    {
     bIsSetParametersModule=false
     enabled=true
     moduleName=ScaleMeshSize
     inputs [
      Scale Factor {bIsDynamic=true, bIsEditable=true, bIsStaticSwitch=false, bIsVisible=true, type {classStructOrEnum {refPath=&Vector3f}}}
      Initial Mesh Scale {bIsDynamic=false, bIsEditable=true, bIsStaticSwitch=false, bIsVisible=true, type {classStructOrEnum {refPath=&Vector3f}}}
     ]
     moduleScript {refPath=&ScaleMeshSize}
    }
    {
     bIsSetParametersModule=false
     enabled=true
     moduleName=SpriteRotationRate
     inputs [
      Rotation Rate {bIsDynamic=true, bIsEditable=true, bIsStaticSwitch=false, bIsVisible=true, type {classStructOrEnum {refPath=&NiagaraFloat}}}
      Delta Time {bIsDynamic=false, bIsEditable=true, bIsStaticSwitch=false, bIsVisible=true, type {classStructOrEnum {refPath=&NiagaraFloat}}}
     ]
     moduleScript {refPath=&SpriteRotationRate}
    }
    {
     bIsSetParametersModule=false
     enabled=true
     moduleName=GravityForce
     inputs [
      Gravity {bIsDynamic=true, bIsEditable=true, bIsStaticSwitch=false, bIsVisible=true, type {classStructOrEnum {refPath=&Vector3f}}}
      Coordinate Space {
       bIsDynamic=false
       bIsEditable=true
       bIsStaticSwitch=true
       bIsVisible=true
       type {classStructOrEnum {refPath=&ENiagaraCoordinateSpace}}
      }
     ]
     moduleScript {refPath=&GravityForce}
    }
    {
     bIsSetParametersModule=false
     enabled=true
     moduleName=CurlNoiseForce
     inputs [
      Noise Strength {bIsDynamic=true, bIsEditable=true, bIsStaticSwitch=false, bIsVisible=true, type {classStructOrEnum {refPath=&NiagaraFloat}}}
      Noise Frequency {bIsDynamic=true, bIsEditable=true, bIsStaticSwitch=false, bIsVisible=true, type {classStructOrEnum {refPath=&NiagaraFloat}}}
      Noise Quality / Cost {
       bIsDynamic=false
       bIsEditable=true
       bIsStaticSwitch=true
       bIsVisible=true
       type {classStructOrEnum {refPath=&ENiagaraCurlNoiseQuality}}
      }
      Random Seed {bIsDynamic=false, bIsEditable=true, bIsStaticSwitch=false, bIsVisible=true, type {classStructOrEnum {refPath=&NiagaraInt32}}}
      Pan Noise {bIsDynamic=false, bIsEditable=true, bIsStaticSwitch=false, bIsVisible=true, type {classStructOrEnum {refPath=&NiagaraBool}}}
      Pan Noise Field {bIsDynamic=false, bIsEditable=false, bIsStaticSwitch=false, bIsVisible=true, type {classStructOrEnum {refPath=&Vector3f}}}
      Write to Intrinsic Parameters {bIsDynamic=false, bIsEditable=true, bIsStaticSwitch=true, bIsVisible=true, type {classStructOrEnum {refPath=&NiagaraBool}}}
      Sample Position {bIsDynamic=false, bIsEditable=true, bIsStaticSwitch=false, bIsVisible=true, type {classStructOrEnum {refPath=&Vector3f}}}
      CurlNoise {
       bIsDynamic=false
       bIsEditable=false
       bIsStaticSwitch=false
       bIsVisible=false
       type {classStructOrEnum {refPath=&NiagaraDataInterfaceCurlNoise}}
      }
      Mask Curl Noise {bIsDynamic=false, bIsEditable=true, bIsStaticSwitch=true, bIsVisible=true, type {classStructOrEnum {refPath=&NiagaraBool}}}
      Curl Noise Cone Mask Angle {
       bIsDynamic=false
       bIsEditable=false
       bIsStaticSwitch=false
       bIsVisible=false
       type {classStructOrEnum {refPath=&NiagaraFloat}}
      }
      Curl Noise Cone Mask Falloff Angle {
       bIsDynamic=false
       bIsEditable=false
       bIsStaticSwitch=false
       bIsVisible=false
       type {classStructOrEnum {refPath=&NiagaraFloat}}
      }
      Curl Noise Cone Mask Axis {
       bIsDynamic=false
       bIsEditable=false
       bIsStaticSwitch=false
       bIsVisible=false
       type {classStructOrEnum {refPath=&Vector3f}}
      }
     ]
     moduleScript {refPath=&CurlNoiseForce}
    }
    {
     bIsSetParametersModule=false
     enabled=true
     moduleName=Drag
     inputs [
      Use Linear Drag {bIsDynamic=false, bIsEditable=true, bIsStaticSwitch=false, bIsVisible=true, type {classStructOrEnum {refPath=&NiagaraBool}}}
      Drag {bIsDynamic=true, bIsEditable=true, bIsStaticSwitch=false, bIsVisible=true, type {classStructOrEnum {refPath=&NiagaraFloat}}}
      Use Rotational Drag {bIsDynamic=false, bIsEditable=true, bIsStaticSwitch=false, bIsVisible=true, type {classStructOrEnum {refPath=&NiagaraBool}}}
      Rotational Drag {bIsDynamic=false, bIsEditable=true, bIsStaticSwitch=false, bIsVisible=true, type {classStructOrEnum {refPath=&NiagaraFloat}}}
      Ignore Mass {bIsDynamic=false, bIsEditable=true, bIsStaticSwitch=false, bIsVisible=true, type {classStructOrEnum {refPath=&NiagaraBool}}}
     ]
     moduleScript {refPath=&Drag}
    }
    {
     bIsSetParametersModule=false
     enabled=true
     moduleName=SolveForcesAndVelocity
     inputs [
      Clamp Velocity {bIsDynamic=false, bIsEditable=true, bIsStaticSwitch=false, bIsVisible=true, type {classStructOrEnum {refPath=&NiagaraBool}}}
      Speed Limit {bIsDynamic=false, bIsEditable=false, bIsStaticSwitch=false, bIsVisible=true, type {classStructOrEnum {refPath=&NiagaraFloat}}}
      Limit Acceleration {bIsDynamic=false, bIsEditable=true, bIsStaticSwitch=false, bIsVisible=true, type {classStructOrEnum {refPath=&NiagaraBool}}}
      Acceleration Limit {bIsDynamic=false, bIsEditable=false, bIsStaticSwitch=false, bIsVisible=true, type {classStructOrEnum {refPath=&NiagaraFloat}}}
      Manually Enable Rotational Solver {bIsDynamic=false, bIsEditable=true, bIsStaticSwitch=false, bIsVisible=true, type {classStructOrEnum {refPath=&NiagaraBool}}}
      Write to Intrinsic Properties {bIsDynamic=false, bIsEditable=true, bIsStaticSwitch=false, bIsVisible=true, type {classStructOrEnum {refPath=&NiagaraBool}}}
      Write to Presolve Properties {bIsDynamic=false, bIsEditable=true, bIsStaticSwitch=true, bIsVisible=true, type {classStructOrEnum {refPath=&NiagaraBool}}}
      DeltaTime {bIsDynamic=false, bIsEditable=true, bIsStaticSwitch=false, bIsVisible=true, type {classStructOrEnum {refPath=&NiagaraFloat}}}
      Mass {bIsDynamic=false, bIsEditable=true, bIsStaticSwitch=false, bIsVisible=true, type {classStructOrEnum {refPath=&NiagaraFloat}}}
      Position {
       bIsDynamic=false
       bIsEditable=true
       bIsStaticSwitch=false
       bIsVisible=true
       type {classStructOrEnum {refPath=&NiagaraPosition}}
      }
      Velocity {bIsDynamic=false, bIsEditable=true, bIsStaticSwitch=false, bIsVisible=true, type {classStructOrEnum {refPath=&Vector3f}}}
      PreviousVelocity {bIsDynamic=false, bIsEditable=true, bIsStaticSwitch=false, bIsVisible=true, type {classStructOrEnum {refPath=&Vector3f}}}
      Force {bIsDynamic=false, bIsEditable=true, bIsStaticSwitch=false, bIsVisible=true, type {classStructOrEnum {refPath=&Vector3f}}}
      Mesh Orientation {bIsDynamic=false, bIsEditable=false, bIsStaticSwitch=false, bIsVisible=false, type {classStructOrEnum {refPath=&Quat4f}}}
      Rotational Velocity {
       bIsDynamic=false
       bIsEditable=false
       bIsStaticSwitch=false
       bIsVisible=false
       type {classStructOrEnum {refPath=&Vector3f}}
      }
      Rotational Force {
       bIsDynamic=false
       bIsEditable=false
       bIsStaticSwitch=false
       bIsVisible=false
       type {classStructOrEnum {refPath=&Vector3f}}
      }
      Rotational Inertia {
       bIsDynamic=false
       bIsEditable=false
       bIsStaticSwitch=false
       bIsVisible=false
       type {classStructOrEnum {refPath=&NiagaraFloat}}
      }
     ]
     moduleScript {refPath=&SolveForcesAndVelocity}
    }
    {
     bIsSetParametersModule=false
     enabled=true
     moduleName=ScaleSpriteSize
     inputs [
      Scale Sprite Size Mode {
       bIsDynamic=false
       bIsEditable=true
       bIsStaticSwitch=true
       bIsVisible=true
       type {classStructOrEnum {refPath=&ENiagara_ScaleSpriteSize}}
      }
      Uniform Curve Sprite Scale {
       bIsDynamic=false
       bIsEditable=false
       bIsStaticSwitch=false
       bIsVisible=false
       type {classStructOrEnum {refPath=&NiagaraDataInterfaceCurve}}
      }
      Non-Uniform Curve Sprite Scale {
       bIsDynamic=false
       bIsEditable=false
       bIsStaticSwitch=false
       bIsVisible=false
       type {classStructOrEnum {refPath=&NiagaraDataInterfaceVector2DCurve}}
      }
      Uniform Curve Index {
       bIsDynamic=false
       bIsEditable=false
       bIsStaticSwitch=false
       bIsVisible=false
       type {classStructOrEnum {refPath=&NiagaraFloat}}
      }
      Non-Uniform Curve Index {
       bIsDynamic=false
       bIsEditable=false
       bIsStaticSwitch=false
       bIsVisible=false
       type {classStructOrEnum {refPath=&NiagaraFloat}}
      }
      Non-Uniform Curve Scale {
       bIsDynamic=false
       bIsEditable=false
       bIsStaticSwitch=false
       bIsVisible=false
       type {classStructOrEnum {refPath=&Vector2f}}
      }
      Uniform Curve Scale {
       bIsDynamic=false
       bIsEditable=false
       bIsStaticSwitch=false
       bIsVisible=false
       type {classStructOrEnum {refPath=&NiagaraFloat}}
      }
      Scale Factor {bIsDynamic=true, bIsEditable=true, bIsStaticSwitch=false, bIsVisible=true, type {classStructOrEnum {refPath=&Vector2f}}}
      Uniform Scale Factor {
       bIsDynamic=false
       bIsEditable=false
       bIsStaticSwitch=false
       bIsVisible=false
       type {classStructOrEnum {refPath=&NiagaraFloat}}
      }
      Initial Sprite Size {bIsDynamic=false, bIsEditable=true, bIsStaticSwitch=false, bIsVisible=true, type {classStructOrEnum {refPath=&Vector2f}}}
     ]
     moduleScript {refPath=&ScaleSpriteSize}
    }
    {
     bIsSetParametersModule=true
     enabled=true
     moduleName=SetVariables_#4
     inputs [
      Particles.DynParSmoke0 {bIsDynamic=true, bIsEditable=true, bIsStaticSwitch=false, bIsVisible=true, type {classStructOrEnum {refPath=&Vector4f}}}
      Particles.DynParSmoke1 {bIsDynamic=true, bIsEditable=true, bIsStaticSwitch=false, bIsVisible=true, type {classStructOrEnum {refPath=&Vector4f}}}
      Particles.DynParSmoke2 {bIsDynamic=true, bIsEditable=true, bIsStaticSwitch=false, bIsVisible=true, type {classStructOrEnum {refPath=&Vector4f}}}
     ]
     moduleScript {refPath=&self:DefaultPropSpawn_Merged003_0.NiagaraScriptSource_1.NiagaraGraph_0.NiagaraNodeAssignment_2.SetVariables_#4}
    }
    {
     bIsSetParametersModule=false
     enabled=true
     moduleName=ScaleColor001
     inputs [
      Scale Mode {bIsDynamic=false, bIsEditable=true, bIsStaticSwitch=true, bIsVisible=true, type {classStructOrEnum {refPath=&NiagaraInt32}}}
      Linear Color Curve {
       bIsDynamic=false
       bIsEditable=false
       bIsStaticSwitch=false
       bIsVisible=false
       type {classStructOrEnum {refPath=&NiagaraDataInterfaceColorCurve}}
      }
      ScaleRGB {bIsDynamic=false, bIsEditable=true, bIsStaticSwitch=false, bIsVisible=true, type {classStructOrEnum {refPath=&NiagaraBool}}}
      Scale RGB {bIsDynamic=false, bIsEditable=true, bIsStaticSwitch=false, bIsVisible=true, type {classStructOrEnum {refPath=&Vector3f}}}
      ScaleRGBA {bIsDynamic=false, bIsEditable=false, bIsStaticSwitch=false, bIsVisible=false, type {classStructOrEnum {refPath=&NiagaraBool}}}
      Scale RGBA {
       bIsDynamic=false
       bIsEditable=false
       bIsStaticSwitch=false
       bIsVisible=false
       type {classStructOrEnum {refPath=&Vector4f}}
      }
      Curve Index {
       bIsDynamic=false
       bIsEditable=false
       bIsStaticSwitch=false
       bIsVisible=false
       type {classStructOrEnum {refPath=&NiagaraFloat}}
      }
      ScaleA {bIsDynamic=false, bIsEditable=true, bIsStaticSwitch=false, bIsVisible=true, type {classStructOrEnum {refPath=&NiagaraBool}}}
      Scale Alpha {bIsDynamic=true, bIsEditable=true, bIsStaticSwitch=false, bIsVisible=true, type {classStructOrEnum {refPath=&NiagaraFloat}}}
      Color Value To Scale {
       bIsDynamic=false
       bIsEditable=true
       bIsStaticSwitch=false
       bIsVisible=true
       type {classStructOrEnum {refPath=&LinearColor}}
      }
      Color Scale {
       bIsDynamic=false
       bIsEditable=false
       bIsStaticSwitch=false
       bIsVisible=false
       type {classStructOrEnum {refPath=&LinearColor}}
      }
     ]
     moduleScript {refPath=&ScaleColor}
    }
    {
     bIsSetParametersModule=false
     enabled=true
     moduleName=FN_Light_Attributes
     inputs [
      Enable Particle Light Enable {bIsDynamic=false, bIsEditable=true, bIsStaticSwitch=true, bIsVisible=true, type {classStructOrEnum {refPath=&NiagaraBool}}}
      Light Enabled {bIsDynamic=false, bIsEditable=false, bIsStaticSwitch=false, bIsVisible=false, type {classStructOrEnum {refPath=&NiagaraBool}}}
      Enable Visibility Tag {bIsDynamic=false, bIsEditable=true, bIsStaticSwitch=true, bIsVisible=true, type {classStructOrEnum {refPath=&NiagaraBool}}}
      VisibilityTag {
       bIsDynamic=false
       bIsEditable=false
       bIsStaticSwitch=false
       bIsVisible=false
       type {classStructOrEnum {refPath=&NiagaraInt32}}
      }
      Create Position Attribute {bIsDynamic=false, bIsEditable=true, bIsStaticSwitch=true, bIsVisible=true, type {classStructOrEnum {refPath=&NiagaraBool}}}
      Position {bIsDynamic=false, bIsEditable=true, bIsStaticSwitch=false, bIsVisible=true, type {classStructOrEnum {refPath=&Vector3f}}}
      Create Radius Attribute {bIsDynamic=false, bIsEditable=true, bIsStaticSwitch=true, bIsVisible=true, type {classStructOrEnum {refPath=&NiagaraBool}}}
      Radius {bIsDynamic=false, bIsEditable=true, bIsStaticSwitch=false, bIsVisible=true, type {classStructOrEnum {refPath=&NiagaraFloat}}}
      Create Exponent Attribute {bIsDynamic=false, bIsEditable=true, bIsStaticSwitch=true, bIsVisible=true, type {classStructOrEnum {refPath=&NiagaraBool}}}
      Exponent {bIsDynamic=false, bIsEditable=true, bIsStaticSwitch=false, bIsVisible=true, type {classStructOrEnum {refPath=&NiagaraFloat}}}
      Create Color Attribute {bIsDynamic=false, bIsEditable=true, bIsStaticSwitch=true, bIsVisible=true, type {classStructOrEnum {refPath=&NiagaraBool}}}
      Color {
       bIsDynamic=false
       bIsEditable=true
       bIsStaticSwitch=false
       bIsVisible=true
       type {classStructOrEnum {refPath=&LinearColor}}
      }
      Scale Color {bIsDynamic=true, bIsEditable=true, bIsStaticSwitch=false, bIsVisible=true, type {classStructOrEnum {refPath=&NiagaraFloat}}}
     ]
     moduleScript {refPath=&FN_Light_Attributes}
    }
    {
     bIsSetParametersModule=true
     enabled=true
     moduleName=SetVariables_#9
     inputs [
      Particles.ShockwaveScale {bIsDynamic=true, bIsEditable=true, bIsStaticSwitch=false, bIsVisible=true, type {classStructOrEnum {refPath=&Vector3f}}}
      Particles.ShockwaveColor {
       bIsDynamic=true
       bIsEditable=true
       bIsStaticSwitch=false
       bIsVisible=true
       type {classStructOrEnum {refPath=&LinearColor}}
      }
      Particles.SecondarySphereColor {
       bIsDynamic=true
       bIsEditable=true
       bIsStaticSwitch=false
       bIsVisible=true
       type {classStructOrEnum {refPath=&LinearColor}}
      }
      Particles.SecondaryDyn {bIsDynamic=true, bIsEditable=true, bIsStaticSwitch=false, bIsVisible=true, type {classStructOrEnum {refPath=&Vector4f}}}
      Particles.PrimaryColor {
       bIsDynamic=true
       bIsEditable=true
       bIsStaticSwitch=false
       bIsVisible=true
       type {classStructOrEnum {refPath=&LinearColor}}
      }
     ]
     moduleScript {refPath=&self:DefaultPropSpawn_Merged003_0.NiagaraScriptSource_1.NiagaraGraph_0.NiagaraNodeAssignment_3.SetVariables_#9}
    }
   ]
  }
  rendererClasses [
   {refPath=&NiagaraSpriteRendererProperties}
   {refPath=&NiagaraLightRendererProperties}
   {refPath=&NiagaraMeshRendererProperties}
  ]
  renderers [
   {rendererIndex=0, rendererClass {refPath=&NiagaraSpriteRendererProperties}}
   {rendererIndex=1, rendererClass {refPath=&NiagaraSpriteRendererProperties}}
   {rendererIndex=2, rendererClass {refPath=&NiagaraSpriteRendererProperties}}
   {rendererIndex=3, rendererClass {refPath=&NiagaraSpriteRendererProperties}}
   {rendererIndex=4, rendererClass {refPath=&NiagaraSpriteRendererProperties}}
   {rendererIndex=5, rendererClass {refPath=&NiagaraLightRendererProperties}}
   {rendererIndex=6, rendererClass {refPath=&NiagaraSpriteRendererProperties}}
   {rendererIndex=7, rendererClass {refPath=&NiagaraMeshRendererProperties}}
   {rendererIndex=8, rendererClass {refPath=&NiagaraMeshRendererProperties}}
   {rendererIndex=9, rendererClass {refPath=&NiagaraMeshRendererProperties}}
  ]
 }
 DefaultPropSpawn_Merged004 {
  bEnabled=false
  emitterSpawnScript {scriptName=EmitterSpawnScript, modules []}
  emitterUpdateScript {
   scriptName=EmitterUpdateScript
   modules [
    {
     bIsSetParametersModule=false
     enabled=true
     moduleName=EmitterState
     inputs [
      Life Cycle Mode {
       bIsDynamic=false
       bIsEditable=true
       bIsStaticSwitch=true
       bIsVisible=true
       type {classStructOrEnum {refPath=&ENiagaraEmitterLifeCycleMode}}
      }
      Inactive Response {
       bIsDynamic=false
       bIsEditable=false
       bIsStaticSwitch=true
       bIsVisible=false
       type {classStructOrEnum {refPath=&ENiagaraInactiveMode}}
      }
      Loop Behavior {
       bIsDynamic=false
       bIsEditable=false
       bIsStaticSwitch=true
       bIsVisible=false
       type {classStructOrEnum {refPath=&ENiagara_EmitterStateOptions}}
      }
      Loop Duration Mode {
       bIsDynamic=false
       bIsEditable=false
       bIsStaticSwitch=true
       bIsVisible=false
       type {classStructOrEnum {refPath=&ENiagara_InfiniteLoopDuration}}
      }
      Loop Count {
       bIsDynamic=false
       bIsEditable=false
       bIsStaticSwitch=false
       bIsVisible=false
       type {classStructOrEnum {refPath=&NiagaraInt32}}
      }
      Loop Duration {
       bIsDynamic=false
       bIsEditable=false
       bIsStaticSwitch=false
       bIsVisible=false
       type {classStructOrEnum {refPath=&NiagaraFloat}}
      }
      Recalculate Duration Each Loop {bIsDynamic=false, bIsEditable=false, bIsStaticSwitch=false, bIsVisible=false, type {classStructOrEnum {refPath=&NiagaraBool}}}
      UseLoopDelay {bIsDynamic=false, bIsEditable=false, bIsStaticSwitch=false, bIsVisible=false, type {classStructOrEnum {refPath=&NiagaraBool}}}
      Loop Delay {
       bIsDynamic=false
       bIsEditable=false
       bIsStaticSwitch=false
       bIsVisible=false
       type {classStructOrEnum {refPath=&NiagaraFloat}}
      }
      Delay First Loop Only {bIsDynamic=false, bIsEditable=false, bIsStaticSwitch=false, bIsVisible=false, type {classStructOrEnum {refPath=&NiagaraBool}}}
      Scalability Mode {
       bIsDynamic=false
       bIsEditable=true
       bIsStaticSwitch=true
       bIsVisible=true
       type {classStructOrEnum {refPath=&ENiagaraEmitterScalabilityMode_Limited}}
      }
      Enable Distance Culling {bIsDynamic=false, bIsEditable=false, bIsStaticSwitch=true, bIsVisible=false, type {classStructOrEnum {refPath=&NiagaraBool}}}
      MinDistance {
       bIsDynamic=false
       bIsEditable=false
       bIsStaticSwitch=false
       bIsVisible=false
       type {classStructOrEnum {refPath=&NiagaraFloat}}
      }
      Min Distance Response {
       bIsDynamic=false
       bIsEditable=false
       bIsStaticSwitch=false
       bIsVisible=false
       type {classStructOrEnum {refPath=&ENiagaraExecutionStateManagement}}
      }
      MaxDistance {
       bIsDynamic=false
       bIsEditable=false
       bIsStaticSwitch=false
       bIsVisible=false
       type {classStructOrEnum {refPath=&NiagaraFloat}}
      }
      Max Distance Response {
       bIsDynamic=false
       bIsEditable=false
       bIsStaticSwitch=false
       bIsVisible=false
       type {classStructOrEnum {refPath=&ENiagaraExecutionStateManagement}}
      }
      Scale Spawn Count {bIsDynamic=false, bIsEditable=false, bIsStaticSwitch=true, bIsVisible=false, type {classStructOrEnum {refPath=&NiagaraBool}}}
      Spawn Count Scale {
       bIsDynamic=false
       bIsEditable=false
       bIsStaticSwitch=false
       bIsVisible=false
       type {classStructOrEnum {refPath=&NiagaraFloat}}
      }
      Spawn Count Scale By Distance {
       bIsDynamic=false
       bIsEditable=false
       bIsStaticSwitch=false
       bIsVisible=false
       type {classStructOrEnum {refPath=&NiagaraDataInterfaceCurve}}
      }
      Enable Visibility Culling {bIsDynamic=false, bIsEditable=false, bIsStaticSwitch=true, bIsVisible=false, type {classStructOrEnum {refPath=&NiagaraBool}}}
      Visibility Cull Response {
       bIsDynamic=false
       bIsEditable=false
       bIsStaticSwitch=false
       bIsVisible=false
       type {classStructOrEnum {refPath=&ENiagaraExecutionStateManagement}}
      }
      Visibility Cull Delay {
       bIsDynamic=false
       bIsEditable=false
       bIsStaticSwitch=false
       bIsVisible=false
       type {classStructOrEnum {refPath=&NiagaraFloat}}
      }
      Reset Age On Awaken {bIsDynamic=false, bIsEditable=false, bIsStaticSwitch=false, bIsVisible=false, type {classStructOrEnum {refPath=&NiagaraBool}}}
     ]
     moduleScript {refPath=&EmitterState}
    }
    {
     bIsSetParametersModule=false
     enabled=true
     moduleName=SpawnBurst_Instantaneous001
     inputs [
      Spawn Count {bIsDynamic=true, bIsEditable=true, bIsStaticSwitch=false, bIsVisible=true, type {classStructOrEnum {refPath=&NiagaraInt32}}}
      Spawn Time {bIsDynamic=false, bIsEditable=true, bIsStaticSwitch=false, bIsVisible=true, type {classStructOrEnum {refPath=&NiagaraFloat}}}
      Use Spawn Probability {bIsDynamic=false, bIsEditable=true, bIsStaticSwitch=false, bIsVisible=true, type {classStructOrEnum {refPath=&NiagaraBool}}}
      Spawn Probability {bIsDynamic=false, bIsEditable=false, bIsStaticSwitch=false, bIsVisible=true, type {classStructOrEnum {refPath=&NiagaraFloat}}}
      UseLoopCountLimit {bIsDynamic=false, bIsEditable=true, bIsStaticSwitch=false, bIsVisible=true, type {classStructOrEnum {refPath=&NiagaraBool}}}
      Loop Count Limit {bIsDynamic=false, bIsEditable=false, bIsStaticSwitch=false, bIsVisible=true, type {classStructOrEnum {refPath=&NiagaraInt32}}}
      Spawn Group {bIsDynamic=false, bIsEditable=true, bIsStaticSwitch=false, bIsVisible=true, type {classStructOrEnum {refPath=&NiagaraInt32}}}
      Age {bIsDynamic=false, bIsEditable=true, bIsStaticSwitch=false, bIsVisible=true, type {classStructOrEnum {refPath=&NiagaraFloat}}}
     ]
     moduleScript {refPath=&SpawnBurst_Instantaneous}
    }
    {
     bIsSetParametersModule=false
     enabled=true
     moduleName=SpawnBurst_Instantaneous
     inputs [
      Spawn Count {bIsDynamic=false, bIsEditable=true, bIsStaticSwitch=false, bIsVisible=true, type {classStructOrEnum {refPath=&NiagaraInt32}}}
      Spawn Time {bIsDynamic=false, bIsEditable=true, bIsStaticSwitch=false, bIsVisible=true, type {classStructOrEnum {refPath=&NiagaraFloat}}}
      Use Spawn Probability {bIsDynamic=false, bIsEditable=true, bIsStaticSwitch=false, bIsVisible=true, type {classStructOrEnum {refPath=&NiagaraBool}}}
      Spawn Probability {bIsDynamic=false, bIsEditable=false, bIsStaticSwitch=false, bIsVisible=true, type {classStructOrEnum {refPath=&NiagaraFloat}}}
      UseLoopCountLimit {bIsDynamic=false, bIsEditable=true, bIsStaticSwitch=false, bIsVisible=true, type {classStructOrEnum {refPath=&NiagaraBool}}}
      Loop Count Limit {bIsDynamic=false, bIsEditable=false, bIsStaticSwitch=false, bIsVisible=true, type {classStructOrEnum {refPath=&NiagaraInt32}}}
      Spawn Group {bIsDynamic=false, bIsEditable=true, bIsStaticSwitch=false, bIsVisible=true, type {classStructOrEnum {refPath=&NiagaraInt32}}}
      Age {bIsDynamic=false, bIsEditable=true, bIsStaticSwitch=false, bIsVisible=true, type {classStructOrEnum {refPath=&NiagaraFloat}}}
     ]
     moduleScript {refPath=&SpawnBurst_Instantaneous}
    }
    {
     bIsSetParametersModule=false
     enabled=true
     moduleName=SpawnBurst_Instantaneous002
     inputs [
      Spawn Count {bIsDynamic=true, bIsEditable=true, bIsStaticSwitch=false, bIsVisible=true, type {classStructOrEnum {refPath=&NiagaraInt32}}}
      Spawn Time {bIsDynamic=false, bIsEditable=true, bIsStaticSwitch=false, bIsVisible=true, type {classStructOrEnum {refPath=&NiagaraFloat}}}
      Use Spawn Probability {bIsDynamic=false, bIsEditable=true, bIsStaticSwitch=false, bIsVisible=true, type {classStructOrEnum {refPath=&NiagaraBool}}}
      Spawn Probability {bIsDynamic=false, bIsEditable=false, bIsStaticSwitch=false, bIsVisible=true, type {classStructOrEnum {refPath=&NiagaraFloat}}}
      UseLoopCountLimit {bIsDynamic=false, bIsEditable=true, bIsStaticSwitch=false, bIsVisible=true, type {classStructOrEnum {refPath=&NiagaraBool}}}
      Loop Count Limit {bIsDynamic=false, bIsEditable=false, bIsStaticSwitch=false, bIsVisible=true, type {classStructOrEnum {refPath=&NiagaraInt32}}}
      Spawn Group {bIsDynamic=false, bIsEditable=true, bIsStaticSwitch=false, bIsVisible=true, type {classStructOrEnum {refPath=&NiagaraInt32}}}
      Age {bIsDynamic=false, bIsEditable=true, bIsStaticSwitch=false, bIsVisible=true, type {classStructOrEnum {refPath=&NiagaraFloat}}}
     ]
     moduleScript {refPath=&SpawnBurst_Instantaneous}
    }
    {
     bIsSetParametersModule=false
     enabled=true
     moduleName=SpawnBurst_Instantaneous003
     inputs [
      Spawn Count {bIsDynamic=false, bIsEditable=true, bIsStaticSwitch=false, bIsVisible=true, type {classStructOrEnum {refPath=&NiagaraInt32}}}
      Spawn Time {bIsDynamic=false, bIsEditable=true, bIsStaticSwitch=false, bIsVisible=true, type {classStructOrEnum {refPath=&NiagaraFloat}}}
      Use Spawn Probability {bIsDynamic=false, bIsEditable=true, bIsStaticSwitch=false, bIsVisible=true, type {classStructOrEnum {refPath=&NiagaraBool}}}
      Spawn Probability {bIsDynamic=false, bIsEditable=false, bIsStaticSwitch=false, bIsVisible=true, type {classStructOrEnum {refPath=&NiagaraFloat}}}
      UseLoopCountLimit {bIsDynamic=false, bIsEditable=true, bIsStaticSwitch=false, bIsVisible=true, type {classStructOrEnum {refPath=&NiagaraBool}}}
      Loop Count Limit {bIsDynamic=false, bIsEditable=false, bIsStaticSwitch=false, bIsVisible=true, type {classStructOrEnum {refPath=&NiagaraInt32}}}
      Spawn Group {bIsDynamic=false, bIsEditable=true, bIsStaticSwitch=false, bIsVisible=true, type {classStructOrEnum {refPath=&NiagaraInt32}}}
      Age {bIsDynamic=false, bIsEditable=true, bIsStaticSwitch=false, bIsVisible=true, type {classStructOrEnum {refPath=&NiagaraFloat}}}
     ]
     moduleScript {refPath=&SpawnBurst_Instantaneous}
    }
    {
     bIsSetParametersModule=false
     enabled=true
     moduleName=SpawnBurst_Instantaneous004
     inputs [
      Spawn Count {bIsDynamic=false, bIsEditable=true, bIsStaticSwitch=false, bIsVisible=true, type {classStructOrEnum {refPath=&NiagaraInt32}}}
      Spawn Time {bIsDynamic=false, bIsEditable=true, bIsStaticSwitch=false, bIsVisible=true, type {classStructOrEnum {refPath=&NiagaraFloat}}}
      Use Spawn Probability {bIsDynamic=false, bIsEditable=true, bIsStaticSwitch=false, bIsVisible=true, type {classStructOrEnum {refPath=&NiagaraBool}}}
      Spawn Probability {bIsDynamic=false, bIsEditable=false, bIsStaticSwitch=false, bIsVisible=true, type {classStructOrEnum {refPath=&NiagaraFloat}}}
      UseLoopCountLimit {bIsDynamic=false, bIsEditable=true, bIsStaticSwitch=false, bIsVisible=true, type {classStructOrEnum {refPath=&NiagaraBool}}}
      Loop Count Limit {bIsDynamic=false, bIsEditable=false, bIsStaticSwitch=false, bIsVisible=true, type {classStructOrEnum {refPath=&NiagaraInt32}}}
      Spawn Group {bIsDynamic=false, bIsEditable=true, bIsStaticSwitch=false, bIsVisible=true, type {classStructOrEnum {refPath=&NiagaraInt32}}}
      Age {bIsDynamic=false, bIsEditable=true, bIsStaticSwitch=false, bIsVisible=true, type {classStructOrEnum {refPath=&NiagaraFloat}}}
     ]
     moduleScript {refPath=&SpawnBurst_Instantaneous}
    }
    {
     bIsSetParametersModule=false
     enabled=true
     moduleName=SpawnBurst_Instantaneous005
     inputs [
      Spawn Count {bIsDynamic=false, bIsEditable=true, bIsStaticSwitch=false, bIsVisible=true, type {classStructOrEnum {refPath=&NiagaraInt32}}}
      Spawn Time {bIsDynamic=false, bIsEditable=true, bIsStaticSwitch=false, bIsVisible=true, type {classStructOrEnum {refPath=&NiagaraFloat}}}
      Use Spawn Probability {bIsDynamic=false, bIsEditable=true, bIsStaticSwitch=false, bIsVisible=true, type {classStructOrEnum {refPath=&NiagaraBool}}}
      Spawn Probability {bIsDynamic=false, bIsEditable=false, bIsStaticSwitch=false, bIsVisible=true, type {classStructOrEnum {refPath=&NiagaraFloat}}}
      UseLoopCountLimit {bIsDynamic=false, bIsEditable=true, bIsStaticSwitch=false, bIsVisible=true, type {classStructOrEnum {refPath=&NiagaraBool}}}
      Loop Count Limit {bIsDynamic=false, bIsEditable=false, bIsStaticSwitch=false, bIsVisible=true, type {classStructOrEnum {refPath=&NiagaraInt32}}}
      Spawn Group {bIsDynamic=false, bIsEditable=true, bIsStaticSwitch=false, bIsVisible=true, type {classStructOrEnum {refPath=&NiagaraInt32}}}
      Age {bIsDynamic=false, bIsEditable=true, bIsStaticSwitch=false, bIsVisible=true, type {classStructOrEnum {refPath=&NiagaraFloat}}}
     ]
     moduleScript {refPath=&SpawnBurst_Instantaneous}
    }
   ]
  }
  particleSpawnScript {
   scriptName=ParticleSpawnScript
   modules [
    {
     bIsSetParametersModule=true
     enabled=true
     moduleName=SetVariables_#1
     inputs [Particles.VisibilityTag {bIsDynamic=true, bIsEditable=true, bIsStaticSwitch=false, bIsVisible=true, type {classStructOrEnum {refPath=&NiagaraInt32}}}]
     moduleScript {refPath=&self:DefaultPropSpawn_Merged004_0.NiagaraScriptSource_1.NiagaraGraph_0.NiagaraNodeAssignment_1.SetVariables_#1}
    }
    {
     bIsSetParametersModule=false
     enabled=true
     moduleName=InitializeParticle001
     inputs [
      Lifetime Mode {
       bIsDynamic=false
       bIsEditable=true
       bIsStaticSwitch=true
       bIsVisible=true
       type {classStructOrEnum {refPath=&ENiagara_LifetimeMode}}
      }
      Lifetime Min {
       bIsDynamic=false
       bIsEditable=false
       bIsStaticSwitch=false
       bIsVisible=false
       type {classStructOrEnum {refPath=&NiagaraFloat}}
      }
      Lifetime {bIsDynamic=true, bIsEditable=true, bIsStaticSwitch=false, bIsVisible=true, type {classStructOrEnum {refPath=&NiagaraFloat}}}
      Lifetime Max {
       bIsDynamic=false
       bIsEditable=false
       bIsStaticSwitch=false
       bIsVisible=false
       type {classStructOrEnum {refPath=&NiagaraFloat}}
      }
      Lifetime Randomness Mode {
       bIsDynamic=false
       bIsEditable=false
       bIsStaticSwitch=true
       bIsVisible=false
       type {classStructOrEnum {refPath=&ENiagaraRandomnessMode}}
      }
      Color Mode {
       bIsDynamic=false
       bIsEditable=true
       bIsStaticSwitch=true
       bIsVisible=true
       type {classStructOrEnum {refPath=&ENiagara_ColorInitializationMode}}
      }
      Color Minimum {
       bIsDynamic=false
       bIsEditable=false
       bIsStaticSwitch=false
       bIsVisible=false
       type {classStructOrEnum {refPath=&LinearColor}}
      }
      Color {
       bIsDynamic=true
       bIsEditable=true
       bIsStaticSwitch=false
       bIsVisible=true
       type {classStructOrEnum {refPath=&LinearColor}}
      }
      AdjustHue {bIsDynamic=false, bIsEditable=false, bIsStaticSwitch=false, bIsVisible=false, type {classStructOrEnum {refPath=&NiagaraBool}}}
      Hue Shift Range {
       bIsDynamic=false
       bIsEditable=false
       bIsStaticSwitch=false
       bIsVisible=false
       type {classStructOrEnum {refPath=&Vector2f}}
      }
      Color Maximum {
       bIsDynamic=false
       bIsEditable=false
       bIsStaticSwitch=false
       bIsVisible=false
       type {classStructOrEnum {refPath=&LinearColor}}
      }
      AdjustSaturation {bIsDynamic=false, bIsEditable=false, bIsStaticSwitch=false, bIsVisible=false, type {classStructOrEnum {refPath=&NiagaraBool}}}
      Saturation Range {
       bIsDynamic=false
       bIsEditable=false
       bIsStaticSwitch=false
       bIsVisible=false
       type {classStructOrEnum {refPath=&Vector2f}}
      }
      AdjustValue {bIsDynamic=false, bIsEditable=false, bIsStaticSwitch=false, bIsVisible=false, type {classStructOrEnum {refPath=&NiagaraBool}}}
      Value Range {
       bIsDynamic=false
       bIsEditable=false
       bIsStaticSwitch=false
       bIsVisible=false
       type {classStructOrEnum {refPath=&Vector2f}}
      }
      AdjustAlpha {bIsDynamic=false, bIsEditable=false, bIsStaticSwitch=false, bIsVisible=false, type {classStructOrEnum {refPath=&NiagaraBool}}}
      Alpha Scale Range {
       bIsDynamic=false
       bIsEditable=false
       bIsStaticSwitch=false
       bIsVisible=false
       type {classStructOrEnum {refPath=&Vector2f}}
      }
      Color Channel Mode {
       bIsDynamic=false
       bIsEditable=false
       bIsStaticSwitch=true
       bIsVisible=false
       type {classStructOrEnum {refPath=&ENiagaraChannelCorrelation}}
      }
      Color Randomness Mode {
       bIsDynamic=false
       bIsEditable=false
       bIsStaticSwitch=true
       bIsVisible=false
       type {classStructOrEnum {refPath=&ENiagaraRandomnessMode}}
      }
      Position Mode {
       bIsDynamic=false
       bIsEditable=true
       bIsStaticSwitch=true
       bIsVisible=true
       type {classStructOrEnum {refPath=&ENiagara_PositionInitializationMode}}
      }
      Position {
       bIsDynamic=false
       bIsEditable=false
       bIsStaticSwitch=false
       bIsVisible=false
       type {classStructOrEnum {refPath=&NiagaraPosition}}
      }
      UsePositionOffset {bIsDynamic=false, bIsEditable=true, bIsStaticSwitch=false, bIsVisible=true, type {classStructOrEnum {refPath=&NiagaraBool}}}
      Position Offset {bIsDynamic=false, bIsEditable=true, bIsStaticSwitch=false, bIsVisible=true, type {classStructOrEnum {refPath=&Vector3f}}}
      Position Offset Coordinate Space {
       bIsDynamic=false
       bIsEditable=true
       bIsStaticSwitch=false
       bIsVisible=true
       type {classStructOrEnum {refPath=&ENiagaraCoordinateSpace}}
      }
      Mass Mode {
       bIsDynamic=false
       bIsEditable=true
       bIsStaticSwitch=true
       bIsVisible=true
       type {classStructOrEnum {refPath=&ENiagara_MassInitializationMode}}
      }
      Mass Min {
       bIsDynamic=false
       bIsEditable=false
       bIsStaticSwitch=false
       bIsVisible=false
       type {classStructOrEnum {refPath=&NiagaraFloat}}
      }
      Mass {
       bIsDynamic=false
       bIsEditable=false
       bIsStaticSwitch=false
       bIsVisible=false
       type {classStructOrEnum {refPath=&NiagaraFloat}}
      }
      Mass Max {
       bIsDynamic=false
       bIsEditable=false
       bIsStaticSwitch=false
       bIsVisible=false
       type {classStructOrEnum {refPath=&NiagaraFloat}}
      }
      Mass Randomness Mode {
       bIsDynamic=false
       bIsEditable=false
       bIsStaticSwitch=true
       bIsVisible=false
       type {classStructOrEnum {refPath=&ENiagaraRandomnessMode}}
      }
      Write Material Random {bIsDynamic=false, bIsEditable=true, bIsStaticSwitch=false, bIsVisible=true, type {classStructOrEnum {refPath=&NiagaraBool}}}
      Material Random {bIsDynamic=false, bIsEditable=true, bIsStaticSwitch=false, bIsVisible=true, type {classStructOrEnum {refPath=&NiagaraFloat}}}
      Sprite Size Mode {
       bIsDynamic=false
       bIsEditable=true
       bIsStaticSwitch=true
       bIsVisible=true
       type {classStructOrEnum {refPath=&ENiagara_SizeScaleMode}}
      }
      Sprite Size {
       bIsDynamic=false
       bIsEditable=false
       bIsStaticSwitch=false
       bIsVisible=false
       type {classStructOrEnum {refPath=&Vector2f}}
      }
      Uniform Sprite Size {bIsDynamic=true, bIsEditable=true, bIsStaticSwitch=false, bIsVisible=true, type {classStructOrEnum {refPath=&NiagaraFloat}}}
      Uniform Sprite Size Min {
       bIsDynamic=false
       bIsEditable=false
       bIsStaticSwitch=false
       bIsVisible=false
       type {classStructOrEnum {refPath=&NiagaraFloat}}
      }
      Sprite Size Min {
       bIsDynamic=false
       bIsEditable=false
       bIsStaticSwitch=false
       bIsVisible=false
       type {classStructOrEnum {refPath=&Vector2f}}
      }
      Sprite Size Max {
       bIsDynamic=false
       bIsEditable=false
       bIsStaticSwitch=false
       bIsVisible=false
       type {classStructOrEnum {refPath=&Vector2f}}
      }
      Uniform Sprite Size Max {
       bIsDynamic=false
       bIsEditable=false
       bIsStaticSwitch=false
       bIsVisible=false
       type {classStructOrEnum {refPath=&NiagaraFloat}}
      }
      Sprite Size Randomness Mode {
       bIsDynamic=false
       bIsEditable=false
       bIsStaticSwitch=true
       bIsVisible=false
       type {classStructOrEnum {refPath=&ENiagaraRandomnessMode}}
      }
      Sprite Rotation Mode {
       bIsDynamic=false
       bIsEditable=true
       bIsStaticSwitch=true
       bIsVisible=true
       type {classStructOrEnum {refPath=&ENiagara_SpriteRotationMode}}
      }
      Sprite Rotation Angle Min {
       bIsDynamic=false
       bIsEditable=false
       bIsStaticSwitch=false
       bIsVisible=false
       type {classStructOrEnum {refPath=&NiagaraFloat}}
      }
      Sprite Rotation Angle {bIsDynamic=true, bIsEditable=true, bIsStaticSwitch=false, bIsVisible=true, type {classStructOrEnum {refPath=&NiagaraFloat}}}
      Sprite Rotation Angle Max {
       bIsDynamic=false
       bIsEditable=false
       bIsStaticSwitch=false
       bIsVisible=false
       type {classStructOrEnum {refPath=&NiagaraFloat}}
      }
      Sprite Rotation Randomness Mode {
       bIsDynamic=false
       bIsEditable=false
       bIsStaticSwitch=true
       bIsVisible=false
       type {classStructOrEnum {refPath=&ENiagaraRandomnessMode}}
      }
      Sprite UV Mode {
       bIsDynamic=false
       bIsEditable=true
       bIsStaticSwitch=true
       bIsVisible=true
       type {classStructOrEnum {refPath=&ENiagara_UVFlippingMode}}
      }
      Sprite UV Randomness Mode {
       bIsDynamic=false
       bIsEditable=false
       bIsStaticSwitch=true
       bIsVisible=false
       type {classStructOrEnum {refPath=&ENiagaraRandomnessMode}}
      }
      Sprite UV Scale {
       bIsDynamic=false
       bIsEditable=false
       bIsStaticSwitch=false
       bIsVisible=false
       type {classStructOrEnum {refPath=&Vector2f}}
      }
      Mesh Scale Mode {
       bIsDynamic=false
       bIsEditable=true
       bIsStaticSwitch=true
       bIsVisible=true
       type {classStructOrEnum {refPath=&ENiagara_SizeScaleMode}}
      }
      Mesh Uniform Scale Min {
       bIsDynamic=false
       bIsEditable=false
       bIsStaticSwitch=false
       bIsVisible=false
       type {classStructOrEnum {refPath=&NiagaraFloat}}
      }
      Mesh Uniform Scale {bIsDynamic=true, bIsEditable=true, bIsStaticSwitch=false, bIsVisible=true, type {classStructOrEnum {refPath=&NiagaraFloat}}}
      Mesh Scale Min {
       bIsDynamic=false
       bIsEditable=false
       bIsStaticSwitch=false
       bIsVisible=false
       type {classStructOrEnum {refPath=&Vector3f}}
      }
      Mesh Scale {
       bIsDynamic=false
       bIsEditable=false
       bIsStaticSwitch=false
       bIsVisible=false
       type {classStructOrEnum {refPath=&Vector3f}}
      }
      Mesh Scale Max {
       bIsDynamic=false
       bIsEditable=false
       bIsStaticSwitch=false
       bIsVisible=false
       type {classStructOrEnum {refPath=&Vector3f}}
      }
      Mesh Uniform Scale Max {
       bIsDynamic=false
       bIsEditable=false
       bIsStaticSwitch=false
       bIsVisible=false
       type {classStructOrEnum {refPath=&NiagaraFloat}}
      }
      Mesh Scale Randomness Mode {
       bIsDynamic=false
       bIsEditable=false
       bIsStaticSwitch=true
       bIsVisible=false
       type {classStructOrEnum {refPath=&ENiagaraRandomnessMode}}
      }
      Mesh Renderer Array Visibility Mode {
       bIsDynamic=false
       bIsEditable=true
       bIsStaticSwitch=true
       bIsVisible=true
       type {classStructOrEnum {refPath=&ENiagara_UnsetDirectSetRandom}}
      }
      Mesh Renderer Mesh Index {
       bIsDynamic=false
       bIsEditable=false
       bIsStaticSwitch=false
       bIsVisible=false
       type {classStructOrEnum {refPath=&NiagaraInt32}}
      }
      Wrap Index By Mesh Count {bIsDynamic=false, bIsEditable=false, bIsStaticSwitch=true, bIsVisible=false, type {classStructOrEnum {refPath=&NiagaraBool}}}
      Mesh Renderer Info {
       bIsDynamic=false
       bIsEditable=false
       bIsStaticSwitch=false
       bIsVisible=false
       type {classStructOrEnum {refPath=&NiagaraDataInterfaceMeshRendererInfo}}
      }
      Mesh Renderer Random Seed {
       bIsDynamic=false
       bIsEditable=false
       bIsStaticSwitch=false
       bIsVisible=false
       type {classStructOrEnum {refPath=&NiagaraInt32}}
      }
      Ribbon Width Mode {
       bIsDynamic=false
       bIsEditable=true
       bIsStaticSwitch=true
       bIsVisible=true
       type {classStructOrEnum {refPath=&ENiagara_UnsetDirectSet}}
      }
      Ribbon Width {
       bIsDynamic=false
       bIsEditable=false
       bIsStaticSwitch=false
       bIsVisible=false
       type {classStructOrEnum {refPath=&NiagaraFloat}}
      }
      Ribbon Facing Mode {
       bIsDynamic=false
       bIsEditable=true
       bIsStaticSwitch=true
       bIsVisible=true
       type {classStructOrEnum {refPath=&ENiagara_UnsetDirectSet}}
      }
      Ribbon Facing Vector {
       bIsDynamic=false
       bIsEditable=false
       bIsStaticSwitch=false
       bIsVisible=false
       type {classStructOrEnum {refPath=&Vector3f}}
      }
      Ribbon Twist Mode {
       bIsDynamic=false
       bIsEditable=true
       bIsStaticSwitch=true
       bIsVisible=true
       type {classStructOrEnum {refPath=&ENiagara_UnsetDirectSet}}
      }
      "Ribbon Twist " {
       bIsDynamic=false
       bIsEditable=false
       bIsStaticSwitch=false
       bIsVisible=false
       type {classStructOrEnum {refPath=&NiagaraFloat}}
      }
     ]
     moduleScript {refPath=&InitializeParticle}
    }
    {
     bIsSetParametersModule=true
     enabled=true
     moduleName=SetVariables_#2
     inputs [Particles.InitSpriteRotRate {bIsDynamic=true, bIsEditable=true, bIsStaticSwitch=false, bIsVisible=true, type {classStructOrEnum {refPath=&NiagaraFloat}}}]
     moduleScript {refPath=&self:DefaultPropSpawn_Merged004_0.NiagaraScriptSource_1.NiagaraGraph_0.NiagaraNodeAssignment_0.SetVariables_#2}
    }
    {
     bIsSetParametersModule=false
     enabled=true
     moduleName=AccelerationForce
     inputs [
      Acceleration {bIsDynamic=true, bIsEditable=true, bIsStaticSwitch=false, bIsVisible=true, type {classStructOrEnum {refPath=&Vector3f}}}
      Coordinate Space {
       bIsDynamic=false
       bIsEditable=true
       bIsStaticSwitch=false
       bIsVisible=true
       type {classStructOrEnum {refPath=&ENiagaraCoordinateSpace}}
      }
     ]
     moduleScript {refPath=&AccelerationForce}
    }
    {
     bIsSetParametersModule=false
     enabled=true
     moduleName=ShapeLocation
     inputs [
      Shape Primitive {
       bIsDynamic=false
       bIsEditable=true
       bIsStaticSwitch=true
       bIsVisible=true
       type {classStructOrEnum {refPath=&ENiagara_LocationShapes}}
      }
      Box / Plane Mode {
       bIsDynamic=false
       bIsEditable=false
       bIsStaticSwitch=true
       bIsVisible=false
       type {classStructOrEnum {refPath=&ENiagara_BoxPlaneMode}}
      }
      Cone Mode {
       bIsDynamic=false
       bIsEditable=false
       bIsStaticSwitch=true
       bIsVisible=false
       type {classStructOrEnum {refPath=&ENiagara_ConeMode}}
      }
      Cylinder Height {
       bIsDynamic=false
       bIsEditable=false
       bIsStaticSwitch=false
       bIsVisible=false
       type {classStructOrEnum {refPath=&NiagaraFloat}}
      }
      Sphere Radius {bIsDynamic=true, bIsEditable=true, bIsStaticSwitch=false, bIsVisible=true, type {classStructOrEnum {refPath=&NiagaraFloat}}}
      Ring / Disc Mode {
       bIsDynamic=false
       bIsEditable=false
       bIsStaticSwitch=true
       bIsVisible=false
       type {classStructOrEnum {refPath=&ENiagaraRingDiscMode}}
      }
      Torus Mode {
       bIsDynamic=false
       bIsEditable=false
       bIsStaticSwitch=true
       bIsVisible=false
       type {classStructOrEnum {refPath=&ENiagaraShapeTorusMode}}
      }
      Knot Ring Loops {
       bIsDynamic=false
       bIsEditable=false
       bIsStaticSwitch=false
       bIsVisible=false
       type {classStructOrEnum {refPath=&NiagaraInt32}}
      }
      Knot Helix Loops {
       bIsDynamic=false
       bIsEditable=false
       bIsStaticSwitch=false
       bIsVisible=false
       type {classStructOrEnum {refPath=&NiagaraInt32}}
      }
      Enable Torus Knot Radius {bIsDynamic=false, bIsEditable=false, bIsStaticSwitch=false, bIsVisible=false, type {classStructOrEnum {refPath=&NiagaraBool}}}
      Torus Knot Tube Radius {
       bIsDynamic=false
       bIsEditable=false
       bIsStaticSwitch=false
       bIsVisible=false
       type {classStructOrEnum {refPath=&NiagaraFloat}}
      }
      Cylinder Radius {
       bIsDynamic=false
       bIsEditable=false
       bIsStaticSwitch=false
       bIsVisible=false
       type {classStructOrEnum {refPath=&NiagaraFloat}}
      }
      Cylinder Height Midpoint {
       bIsDynamic=false
       bIsEditable=false
       bIsStaticSwitch=false
       bIsVisible=false
       type {classStructOrEnum {refPath=&NiagaraFloat}}
      }
      Ring Radius {
       bIsDynamic=false
       bIsEditable=false
       bIsStaticSwitch=false
       bIsVisible=false
       type {classStructOrEnum {refPath=&NiagaraFloat}}
      }
      Box Size {
       bIsDynamic=false
       bIsEditable=false
       bIsStaticSwitch=false
       bIsVisible=false
       type {classStructOrEnum {refPath=&Vector3f}}
      }
      Box Midpoint {
       bIsDynamic=false
       bIsEditable=false
       bIsStaticSwitch=false
       bIsVisible=false
       type {classStructOrEnum {refPath=&Vector3f}}
      }
      Cone Length {
       bIsDynamic=false
       bIsEditable=false
       bIsStaticSwitch=false
       bIsVisible=false
       type {classStructOrEnum {refPath=&NiagaraFloat}}
      }
      Plane Size {
       bIsDynamic=false
       bIsEditable=false
       bIsStaticSwitch=false
       bIsVisible=false
       type {classStructOrEnum {refPath=&Vector2f}}
      }
      Wedge Horizontal Angle {
       bIsDynamic=false
       bIsEditable=false
       bIsStaticSwitch=false
       bIsVisible=false
       type {classStructOrEnum {refPath=&NiagaraFloat}}
      }
      Large Radius {
       bIsDynamic=false
       bIsEditable=false
       bIsStaticSwitch=false
       bIsVisible=false
       type {classStructOrEnum {refPath=&NiagaraFloat}}
      }
      Plane Midpoint {
       bIsDynamic=false
       bIsEditable=false
       bIsStaticSwitch=false
       bIsVisible=false
       type {classStructOrEnum {refPath=&Vector2f}}
      }
      Wedge Vertical Angle {
       bIsDynamic=false
       bIsEditable=false
       bIsStaticSwitch=false
       bIsVisible=false
       type {classStructOrEnum {refPath=&NiagaraFloat}}
      }
      Cone Angle {
       bIsDynamic=false
       bIsEditable=false
       bIsStaticSwitch=false
       bIsVisible=false
       type {classStructOrEnum {refPath=&NiagaraFloat}}
      }
      Handle Radius {
       bIsDynamic=false
       bIsEditable=false
       bIsStaticSwitch=false
       bIsVisible=false
       type {classStructOrEnum {refPath=&NiagaraFloat}}
      }
      Cone Inner Angle {
       bIsDynamic=false
       bIsEditable=false
       bIsStaticSwitch=false
       bIsVisible=false
       type {classStructOrEnum {refPath=&NiagaraFloat}}
      }
      Cone Radial Angle {
       bIsDynamic=false
       bIsEditable=false
       bIsStaticSwitch=false
       bIsVisible=false
       type {classStructOrEnum {refPath=&NiagaraFloat}}
      }
      Cone Angle Type {
       bIsDynamic=false
       bIsEditable=false
       bIsStaticSwitch=true
       bIsVisible=false
       type {classStructOrEnum {refPath=&ENiagara_AngleInput}}
      }
      Flatten Endcaps {bIsDynamic=false, bIsEditable=false, bIsStaticSwitch=true, bIsVisible=false, type {classStructOrEnum {refPath=&NiagaraBool}}}
      Apply To Particle Position {
       bIsDynamic=false
       bIsEditable=true
       bIsStaticSwitch=true
       bIsVisible=true
       type {classStructOrEnum {refPath=&ENiagara_AttributeSamplingApplyOutput}}
      }
      Cone Surface Distribution {
       bIsDynamic=false
       bIsEditable=false
       bIsStaticSwitch=false
       bIsVisible=false
       type {classStructOrEnum {refPath=&NiagaraFloat}}
      }
      Ring / Disc Distribution Mode {
       bIsDynamic=false
       bIsEditable=false
       bIsStaticSwitch=true
       bIsVisible=false
       type {classStructOrEnum {refPath=&ENiagaraSphereDistributionMode}}
      }
      Sphere Distribution {
       bIsDynamic=false
       bIsEditable=true
       bIsStaticSwitch=true
       bIsVisible=true
       type {classStructOrEnum {refPath=&ENiagaraSphereDistributionMode}}
      }
      Torus Distribution Mode {
       bIsDynamic=false
       bIsEditable=false
       bIsStaticSwitch=true
       bIsVisible=false
       type {classStructOrEnum {refPath=&ENiagaraTorusDistributionMode}}
      }
      Sphere Surface Distribution {bIsDynamic=false, bIsEditable=true, bIsStaticSwitch=false, bIsVisible=true, type {classStructOrEnum {refPath=&NiagaraFloat}}}
      Hemisphere Distribution {bIsDynamic=false, bIsEditable=true, bIsStaticSwitch=false, bIsVisible=true, type {classStructOrEnum {refPath=&Vector2f}}}
      Cylinder Distribution {
       bIsDynamic=false
       bIsEditable=false
       bIsStaticSwitch=true
       bIsVisible=false
       type {classStructOrEnum {refPath=&ENiagara_CylinderMode}}
      }
      Surface Only {bIsDynamic=false, bIsEditable=false, bIsStaticSwitch=false, bIsVisible=false, type {classStructOrEnum {refPath=&NiagaraBool}}}
      Surface Only Band Thickness {
       bIsDynamic=false
       bIsEditable=false
       bIsStaticSwitch=false
       bIsVisible=false
       type {classStructOrEnum {refPath=&NiagaraFloat}}
      }
      Use Endcaps In Surface Only Mode {bIsDynamic=false, bIsEditable=false, bIsStaticSwitch=true, bIsVisible=false, type {classStructOrEnum {refPath=&NiagaraBool}}}
      Hemicircle X {bIsDynamic=false, bIsEditable=false, bIsStaticSwitch=true, bIsVisible=false, type {classStructOrEnum {refPath=&NiagaraBool}}}
      Hemicircle Y {bIsDynamic=false, bIsEditable=false, bIsStaticSwitch=true, bIsVisible=false, type {classStructOrEnum {refPath=&NiagaraBool}}}
      Disc Coverage {
       bIsDynamic=false
       bIsEditable=false
       bIsStaticSwitch=false
       bIsVisible=false
       type {classStructOrEnum {refPath=&NiagaraFloat}}
      }
      Hemicircle Internal Cap {bIsDynamic=false, bIsEditable=false, bIsStaticSwitch=true, bIsVisible=false, type {classStructOrEnum {refPath=&NiagaraBool}}}
      U Position {
       bIsDynamic=false
       bIsEditable=false
       bIsStaticSwitch=false
       bIsVisible=false
       type {classStructOrEnum {refPath=&NiagaraFloat}}
      }
      Enable Lathe Profile {bIsDynamic=false, bIsEditable=false, bIsStaticSwitch=true, bIsVisible=false, type {classStructOrEnum {refPath=&NiagaraBool}}}
      Lathe Profile {
       bIsDynamic=false
       bIsEditable=false
       bIsStaticSwitch=false
       bIsVisible=false
       type {classStructOrEnum {refPath=&NiagaraDataInterfaceCurve}}
      }
      Surface Distribution {
       bIsDynamic=false
       bIsEditable=false
       bIsStaticSwitch=false
       bIsVisible=false
       type {classStructOrEnum {refPath=&NiagaraFloat}}
      }
      Surface Expansion Mode {
       bIsDynamic=false
       bIsEditable=false
       bIsStaticSwitch=false
       bIsVisible=false
       type {classStructOrEnum {refPath=&ENiagaraExpansionMode}}
      }
      U Distribution {
       bIsDynamic=false
       bIsEditable=false
       bIsStaticSwitch=false
       bIsVisible=false
       type {classStructOrEnum {refPath=&NiagaraFloat}}
      }
      V Position {
       bIsDynamic=false
       bIsEditable=false
       bIsStaticSwitch=false
       bIsVisible=false
       type {classStructOrEnum {refPath=&NiagaraFloat}}
      }
      Radius Position {
       bIsDynamic=false
       bIsEditable=false
       bIsStaticSwitch=false
       bIsVisible=false
       type {classStructOrEnum {refPath=&NiagaraFloat}}
      }
      Sphere Radius Position {
       bIsDynamic=false
       bIsEditable=false
       bIsStaticSwitch=false
       bIsVisible=false
       type {classStructOrEnum {refPath=&NiagaraFloat}}
      }
      V Distribution {
       bIsDynamic=false
       bIsEditable=false
       bIsStaticSwitch=false
       bIsVisible=false
       type {classStructOrEnum {refPath=&NiagaraFloat}}
      }
      Evaluation Type {
       bIsDynamic=false
       bIsEditable=false
       bIsStaticSwitch=true
       bIsVisible=false
       type {classStructOrEnum {refPath=&ENiagaraRandomnessEvaluation}}
      }
      Uniform Distribution {
       bIsDynamic=false
       bIsEditable=false
       bIsStaticSwitch=false
       bIsVisible=false
       type {classStructOrEnum {refPath=&NiagaraFloat}}
      }
      Uniform Spiral Amount {
       bIsDynamic=false
       bIsEditable=false
       bIsStaticSwitch=false
       bIsVisible=false
       type {classStructOrEnum {refPath=&NiagaraFloat}}
      }
      Uniform Spiral Falloff {
       bIsDynamic=false
       bIsEditable=false
       bIsStaticSwitch=false
       bIsVisible=false
       type {classStructOrEnum {refPath=&NiagaraFloat}}
      }
      Uniform Count {
       bIsDynamic=false
       bIsEditable=false
       bIsStaticSwitch=false
       bIsVisible=false
       type {classStructOrEnum {refPath=&NiagaraInt32}}
      }
      Uniform Particle Index {
       bIsDynamic=false
       bIsEditable=false
       bIsStaticSwitch=false
       bIsVisible=false
       type {classStructOrEnum {refPath=&NiagaraInt32}}
      }
      Hemisphere Angle Type {
       bIsDynamic=false
       bIsEditable=true
       bIsStaticSwitch=true
       bIsVisible=true
       type {classStructOrEnum {refPath=&ENiagara_AngleInput}}
      }
      Uniform Endpoint Offset {
       bIsDynamic=false
       bIsEditable=false
       bIsStaticSwitch=false
       bIsVisible=false
       type {classStructOrEnum {refPath=&NiagaraFloat}}
      }
      MaskSpawnGroup {bIsDynamic=false, bIsEditable=true, bIsStaticSwitch=false, bIsVisible=true, type {classStructOrEnum {refPath=&NiagaraBool}}}
      Spawn Group Mask {bIsDynamic=false, bIsEditable=false, bIsStaticSwitch=false, bIsVisible=true, type {classStructOrEnum {refPath=&NiagaraInt32}}}
      Spawn Group {
       bIsDynamic=false
       bIsEditable=false
       bIsStaticSwitch=false
       bIsVisible=false
       type {classStructOrEnum {refPath=&NiagaraInt32}}
      }
      Randomness Mode {
       bIsDynamic=false
       bIsEditable=true
       bIsStaticSwitch=true
       bIsVisible=true
       type {classStructOrEnum {refPath=&ENiagaraRandomnessMode}}
      }
      Override Seed {bIsDynamic=false, bIsEditable=true, bIsStaticSwitch=false, bIsVisible=true, type {classStructOrEnum {refPath=&NiagaraBool}}}
      Random Seed {bIsDynamic=false, bIsEditable=false, bIsStaticSwitch=false, bIsVisible=true, type {classStructOrEnum {refPath=&NiagaraInt32}}}
      Fixed Random Seed {bIsDynamic=false, bIsEditable=false, bIsStaticSwitch=true, bIsVisible=false, type {classStructOrEnum {refPath=&NiagaraBool}}}
      Transform Order {
       bIsDynamic=false
       bIsEditable=true
       bIsStaticSwitch=true
       bIsVisible=true
       type {classStructOrEnum {refPath=&ENiagara_TransformOrder}}
      }
      Scale Mode {
       bIsDynamic=false
       bIsEditable=true
       bIsStaticSwitch=true
       bIsVisible=true
       type {classStructOrEnum {refPath=&ENiagara_ScaleMode}}
      }
      Non Uniform Scale {bIsDynamic=false, bIsEditable=true, bIsStaticSwitch=false, bIsVisible=true, type {classStructOrEnum {refPath=&Vector3f}}}
      UseOwnerScale {bIsDynamic=false, bIsEditable=true, bIsStaticSwitch=false, bIsVisible=true, type {classStructOrEnum {refPath=&NiagaraBool}}}
      Apply Owner Scale {bIsDynamic=false, bIsEditable=true, bIsStaticSwitch=false, bIsVisible=true, type {classStructOrEnum {refPath=&Vector3f}}}
      Rotation Mode {
       bIsDynamic=false
       bIsEditable=true
       bIsStaticSwitch=true
       bIsVisible=true
       type {classStructOrEnum {refPath=&ENiagara_RotationMode}}
      }
      Rotation Axis {
       bIsDynamic=false
       bIsEditable=false
       bIsStaticSwitch=false
       bIsVisible=false
       type {classStructOrEnum {refPath=&Vector3f}}
      }
      Rotation Matrix {
       bIsDynamic=false
       bIsEditable=false
       bIsStaticSwitch=false
       bIsVisible=false
       type {classStructOrEnum {refPath=&NiagaraMatrix}}
      }
      Rotation Quaternion {bIsDynamic=false, bIsEditable=false, bIsStaticSwitch=false, bIsVisible=false, type {classStructOrEnum {refPath=&Quat4f}}}
      Yaw / Pitch / Roll {
       bIsDynamic=false
       bIsEditable=false
       bIsStaticSwitch=false
       bIsVisible=false
       type {classStructOrEnum {refPath=&Vector3f}}
      }
      Rotation Angle {
       bIsDynamic=false
       bIsEditable=false
       bIsStaticSwitch=false
       bIsVisible=false
       type {classStructOrEnum {refPath=&NiagaraFloat}}
      }
      Invert Rotation Quaternion {bIsDynamic=false, bIsEditable=false, bIsStaticSwitch=false, bIsVisible=false, type {classStructOrEnum {refPath=&NiagaraBool}}}
      Additional Quaternion Rotation {bIsDynamic=false, bIsEditable=false, bIsStaticSwitch=false, bIsVisible=false, type {classStructOrEnum {refPath=&NiagaraBool}}}
      Additional Yaw / Pitch / Roll {
       bIsDynamic=false
       bIsEditable=false
       bIsStaticSwitch=false
       bIsVisible=false
       type {classStructOrEnum {refPath=&Vector3f}}
      }
      Rotation Angle Type {
       bIsDynamic=false
       bIsEditable=false
       bIsStaticSwitch=true
       bIsVisible=false
       type {classStructOrEnum {refPath=&ENiagara_AngleInput}}
      }
      Rotation Coordinate Space {
       bIsDynamic=false
       bIsEditable=true
       bIsStaticSwitch=false
       bIsVisible=true
       type {classStructOrEnum {refPath=&ENiagaraCoordinateSpace}}
      }
      Offset Mode {
       bIsDynamic=false
       bIsEditable=true
       bIsStaticSwitch=true
       bIsVisible=true
       type {classStructOrEnum {refPath=&ENiagara_OffsetMode}}
      }
      Shape Origin {
       bIsDynamic=false
       bIsEditable=true
       bIsStaticSwitch=false
       bIsVisible=true
       type {classStructOrEnum {refPath=&NiagaraPosition}}
      }
      Offset {bIsDynamic=false, bIsEditable=true, bIsStaticSwitch=false, bIsVisible=true, type {classStructOrEnum {refPath=&Vector3f}}}
      Offset Coordinate Space {
       bIsDynamic=false
       bIsEditable=true
       bIsStaticSwitch=false
       bIsVisible=true
       type {classStructOrEnum {refPath=&ENiagaraCoordinateSpace}}
      }
      Custom Transform Matrix {
       bIsDynamic=false
       bIsEditable=false
       bIsStaticSwitch=false
       bIsVisible=false
       type {classStructOrEnum {refPath=&NiagaraMatrix}}
      }
      Transform Method {
       bIsDynamic=false
       bIsEditable=true
       bIsStaticSwitch=true
       bIsVisible=true
       type {classStructOrEnum {refPath=&ENiagara_TransformType}}
      }
     ]
     moduleScript {refPath=&ShapeLocation}
    }
    {
     bIsSetParametersModule=false
     enabled=true
     moduleName=CameraOffset
     inputs [Camera Offset Amount {bIsDynamic=true, bIsEditable=true, bIsStaticSwitch=false, bIsVisible=true, type {classStructOrEnum {refPath=&NiagaraFloat}}}]
     moduleScript {refPath=&CameraOffset}
    }
    {
     bIsSetParametersModule=false
     enabled=true
     moduleName=AddVelocity001
     inputs [
      Velocity Mode {
       bIsDynamic=false
       bIsEditable=true
       bIsStaticSwitch=true
       bIsVisible=true
       type {classStructOrEnum {refPath=&ENiagara_VelocityMode}}
      }
      Velocity {bIsDynamic=true, bIsEditable=false, bIsStaticSwitch=false, bIsVisible=false, type {classStructOrEnum {refPath=&Vector3f}}}
      Velocity Speed {bIsDynamic=true, bIsEditable=true, bIsStaticSwitch=false, bIsVisible=true, type {classStructOrEnum {refPath=&NiagaraFloat}}}
      Velocity Speed Scale {
       bIsDynamic=false
       bIsEditable=false
       bIsStaticSwitch=false
       bIsVisible=false
       type {classStructOrEnum {refPath=&NiagaraFloat}}
      }
      Module.Use Falloff Distance {bIsDynamic=false, bIsEditable=true, bIsStaticSwitch=false, bIsVisible=true, type {classStructOrEnum {refPath=&NiagaraBool}}}
      Constrain To Radius {bIsDynamic=false, bIsEditable=false, bIsStaticSwitch=false, bIsVisible=true, type {classStructOrEnum {refPath=&NiagaraFloat}}}
      Distribution Along Cone Axis {
       bIsDynamic=false
       bIsEditable=false
       bIsStaticSwitch=false
       bIsVisible=false
       type {classStructOrEnum {refPath=&NiagaraFloat}}
      }
      Radius Falloff Near / Far {
       bIsDynamic=false
       bIsEditable=false
       bIsStaticSwitch=false
       bIsVisible=false
       type {classStructOrEnum {refPath=&Vector2f}}
      }
      Radius Falloff Exponent {
       bIsDynamic=false
       bIsEditable=false
       bIsStaticSwitch=false
       bIsVisible=false
       type {classStructOrEnum {refPath=&NiagaraFloat}}
      }
      Speed Falloff From Cone Axis {
       bIsDynamic=false
       bIsEditable=false
       bIsStaticSwitch=false
       bIsVisible=false
       type {classStructOrEnum {refPath=&NiagaraFloat}}
      }
      Invert Velocity Falloff {bIsDynamic=false, bIsEditable=false, bIsStaticSwitch=false, bIsVisible=false, type {classStructOrEnum {refPath=&NiagaraBool}}}
      Override Seed {bIsDynamic=false, bIsEditable=true, bIsStaticSwitch=false, bIsVisible=true, type {classStructOrEnum {refPath=&NiagaraBool}}}
      Random Seed {bIsDynamic=false, bIsEditable=false, bIsStaticSwitch=false, bIsVisible=true, type {classStructOrEnum {refPath=&NiagaraInt32}}}
      Fixed Random Seed {bIsDynamic=false, bIsEditable=false, bIsStaticSwitch=true, bIsVisible=false, type {classStructOrEnum {refPath=&NiagaraBool}}}
      Randomness Mode {
       bIsDynamic=false
       bIsEditable=true
       bIsStaticSwitch=true
       bIsVisible=true
       type {classStructOrEnum {refPath=&ENiagaraRandomnessMode}}
      }
      Cone Axis {
       bIsDynamic=false
       bIsEditable=false
       bIsStaticSwitch=false
       bIsVisible=false
       type {classStructOrEnum {refPath=&Vector3f}}
      }
      Cone Angle {
       bIsDynamic=false
       bIsEditable=false
       bIsStaticSwitch=false
       bIsVisible=false
       type {classStructOrEnum {refPath=&NiagaraFloat}}
      }
      Inner Cone Angle {
       bIsDynamic=false
       bIsEditable=false
       bIsStaticSwitch=false
       bIsVisible=false
       type {classStructOrEnum {refPath=&NiagaraFloat}}
      }
      Cone Angle Mode {
       bIsDynamic=false
       bIsEditable=false
       bIsStaticSwitch=true
       bIsVisible=false
       type {classStructOrEnum {refPath=&ENiagara_AngleInput}}
      }
      Rotation Mode {
       bIsDynamic=false
       bIsEditable=false
       bIsStaticSwitch=true
       bIsVisible=false
       type {classStructOrEnum {refPath=&ENiagara_RotationMode}}
      }
      Rotation Axis {
       bIsDynamic=false
       bIsEditable=false
       bIsStaticSwitch=false
       bIsVisible=false
       type {classStructOrEnum {refPath=&Vector3f}}
      }
      Yaw / Pitch / Roll {
       bIsDynamic=false
       bIsEditable=false
       bIsStaticSwitch=false
       bIsVisible=false
       type {classStructOrEnum {refPath=&Vector3f}}
      }
      Rotation Quaternion {bIsDynamic=false, bIsEditable=false, bIsStaticSwitch=false, bIsVisible=false, type {classStructOrEnum {refPath=&Quat4f}}}
      Rotation Matrix {
       bIsDynamic=false
       bIsEditable=false
       bIsStaticSwitch=false
       bIsVisible=false
       type {classStructOrEnum {refPath=&NiagaraMatrix}}
      }
      Rotation Angle {
       bIsDynamic=false
       bIsEditable=false
       bIsStaticSwitch=false
       bIsVisible=false
       type {classStructOrEnum {refPath=&NiagaraFloat}}
      }
      Invert Rotation Quaternion {bIsDynamic=false, bIsEditable=false, bIsStaticSwitch=false, bIsVisible=false, type {classStructOrEnum {refPath=&NiagaraBool}}}
      Additional Quaternion Rotation {bIsDynamic=false, bIsEditable=false, bIsStaticSwitch=false, bIsVisible=false, type {classStructOrEnum {refPath=&NiagaraBool}}}
      Additional Yaw / Pitch / Roll {
       bIsDynamic=false
       bIsEditable=false
       bIsStaticSwitch=false
       bIsVisible=false
       type {classStructOrEnum {refPath=&Vector3f}}
      }
      Rotation Angle Type {
       bIsDynamic=false
       bIsEditable=false
       bIsStaticSwitch=true
       bIsVisible=false
       type {classStructOrEnum {refPath=&ENiagara_AngleInput}}
      }
      Rotation Coordinate Space {
       bIsDynamic=false
       bIsEditable=false
       bIsStaticSwitch=false
       bIsVisible=false
       type {classStructOrEnum {refPath=&ENiagaraCoordinateSpace}}
      }
      Velocity Origin {
       bIsDynamic=false
       bIsEditable=true
       bIsStaticSwitch=false
       bIsVisible=true
       type {classStructOrEnum {refPath=&NiagaraPosition}}
      }
      Origin Offset {bIsDynamic=false, bIsEditable=true, bIsStaticSwitch=false, bIsVisible=true, type {classStructOrEnum {refPath=&Vector3f}}}
      Origin Offset Coordinate Space {
       bIsDynamic=false
       bIsEditable=true
       bIsStaticSwitch=false
       bIsVisible=true
       type {classStructOrEnum {refPath=&ENiagaraCoordinateSpace}}
      }
      Default Position {
       bIsDynamic=false
       bIsEditable=true
       bIsStaticSwitch=false
       bIsVisible=true
       type {classStructOrEnum {refPath=&NiagaraPosition}}
      }
     ]
     moduleScript {refPath=&AddVelocity}
    }
    {
     bIsSetParametersModule=false
     enabled=true
     moduleName=AddVelocity
     inputs [
      Velocity Mode {
       bIsDynamic=false
       bIsEditable=true
       bIsStaticSwitch=true
       bIsVisible=true
       type {classStructOrEnum {refPath=&ENiagara_VelocityMode}}
      }
      Velocity {bIsDynamic=true, bIsEditable=true, bIsStaticSwitch=false, bIsVisible=true, type {classStructOrEnum {refPath=&Vector3f}}}
      Velocity Speed {
       bIsDynamic=false
       bIsEditable=false
       bIsStaticSwitch=false
       bIsVisible=false
       type {classStructOrEnum {refPath=&NiagaraFloat}}
      }
      Velocity Speed Scale {bIsDynamic=false, bIsEditable=true, bIsStaticSwitch=false, bIsVisible=true, type {classStructOrEnum {refPath=&NiagaraFloat}}}
      Constrain To Radius {
       bIsDynamic=false
       bIsEditable=false
       bIsStaticSwitch=false
       bIsVisible=false
       type {classStructOrEnum {refPath=&NiagaraFloat}}
      }
      Distribution Along Cone Axis {
       bIsDynamic=false
       bIsEditable=false
       bIsStaticSwitch=false
       bIsVisible=false
       type {classStructOrEnum {refPath=&NiagaraFloat}}
      }
      Radius Falloff Near / Far {
       bIsDynamic=false
       bIsEditable=false
       bIsStaticSwitch=false
       bIsVisible=false
       type {classStructOrEnum {refPath=&Vector2f}}
      }
      Radius Falloff Exponent {
       bIsDynamic=false
       bIsEditable=false
       bIsStaticSwitch=false
       bIsVisible=false
       type {classStructOrEnum {refPath=&NiagaraFloat}}
      }
      Speed Falloff From Cone Axis {
       bIsDynamic=false
       bIsEditable=false
       bIsStaticSwitch=false
       bIsVisible=false
       type {classStructOrEnum {refPath=&NiagaraFloat}}
      }
      Invert Velocity Falloff {bIsDynamic=false, bIsEditable=false, bIsStaticSwitch=false, bIsVisible=false, type {classStructOrEnum {refPath=&NiagaraBool}}}
      Override Seed {bIsDynamic=false, bIsEditable=false, bIsStaticSwitch=false, bIsVisible=false, type {classStructOrEnum {refPath=&NiagaraBool}}}
      Random Seed {
       bIsDynamic=false
       bIsEditable=false
       bIsStaticSwitch=false
       bIsVisible=false
       type {classStructOrEnum {refPath=&NiagaraInt32}}
      }
      Fixed Random Seed {bIsDynamic=false, bIsEditable=false, bIsStaticSwitch=true, bIsVisible=false, type {classStructOrEnum {refPath=&NiagaraBool}}}
      Randomness Mode {
       bIsDynamic=false
       bIsEditable=false
       bIsStaticSwitch=true
       bIsVisible=false
       type {classStructOrEnum {refPath=&ENiagaraRandomnessMode}}
      }
      Cone Axis {
       bIsDynamic=false
       bIsEditable=false
       bIsStaticSwitch=false
       bIsVisible=false
       type {classStructOrEnum {refPath=&Vector3f}}
      }
      Cone Angle {
       bIsDynamic=false
       bIsEditable=false
       bIsStaticSwitch=false
       bIsVisible=false
       type {classStructOrEnum {refPath=&NiagaraFloat}}
      }
      Inner Cone Angle {
       bIsDynamic=false
       bIsEditable=false
       bIsStaticSwitch=false
       bIsVisible=false
       type {classStructOrEnum {refPath=&NiagaraFloat}}
      }
      Cone Angle Mode {
       bIsDynamic=false
       bIsEditable=false
       bIsStaticSwitch=true
       bIsVisible=false
       type {classStructOrEnum {refPath=&ENiagara_AngleInput}}
      }
      Rotation Mode {
       bIsDynamic=false
       bIsEditable=true
       bIsStaticSwitch=true
       bIsVisible=true
       type {classStructOrEnum {refPath=&ENiagara_RotationMode}}
      }
      Rotation Axis {
       bIsDynamic=false
       bIsEditable=false
       bIsStaticSwitch=false
       bIsVisible=false
       type {classStructOrEnum {refPath=&Vector3f}}
      }
      Yaw / Pitch / Roll {
       bIsDynamic=false
       bIsEditable=false
       bIsStaticSwitch=false
       bIsVisible=false
       type {classStructOrEnum {refPath=&Vector3f}}
      }
      Rotation Quaternion {bIsDynamic=false, bIsEditable=false, bIsStaticSwitch=false, bIsVisible=false, type {classStructOrEnum {refPath=&Quat4f}}}
      Rotation Matrix {
       bIsDynamic=false
       bIsEditable=false
       bIsStaticSwitch=false
       bIsVisible=false
       type {classStructOrEnum {refPath=&NiagaraMatrix}}
      }
      Rotation Angle {
       bIsDynamic=false
       bIsEditable=false
       bIsStaticSwitch=false
       bIsVisible=false
       type {classStructOrEnum {refPath=&NiagaraFloat}}
      }
      Invert Rotation Quaternion {bIsDynamic=false, bIsEditable=false, bIsStaticSwitch=false, bIsVisible=false, type {classStructOrEnum {refPath=&NiagaraBool}}}
      Additional Quaternion Rotation {bIsDynamic=false, bIsEditable=false, bIsStaticSwitch=false, bIsVisible=false, type {classStructOrEnum {refPath=&NiagaraBool}}}
      Additional Yaw / Pitch / Roll {
       bIsDynamic=false
       bIsEditable=false
       bIsStaticSwitch=false
       bIsVisible=false
       type {classStructOrEnum {refPath=&Vector3f}}
      }
      Rotation Angle Type {
       bIsDynamic=false
       bIsEditable=false
       bIsStaticSwitch=true
       bIsVisible=false
       type {classStructOrEnum {refPath=&ENiagara_AngleInput}}
      }
      Rotation Coordinate Space {
       bIsDynamic=false
       bIsEditable=true
       bIsStaticSwitch=false
       bIsVisible=true
       type {classStructOrEnum {refPath=&ENiagaraCoordinateSpace}}
      }
      Velocity Origin {
       bIsDynamic=false
       bIsEditable=false
       bIsStaticSwitch=false
       bIsVisible=false
       type {classStructOrEnum {refPath=&NiagaraPosition}}
      }
      Origin Offset {
       bIsDynamic=false
       bIsEditable=false
       bIsStaticSwitch=false
       bIsVisible=false
       type {classStructOrEnum {refPath=&Vector3f}}
      }
      Origin Offset Coordinate Space {
       bIsDynamic=false
       bIsEditable=false
       bIsStaticSwitch=false
       bIsVisible=false
       type {classStructOrEnum {refPath=&ENiagaraCoordinateSpace}}
      }
      Default Position {
       bIsDynamic=false
       bIsEditable=false
       bIsStaticSwitch=false
       bIsVisible=false
       type {classStructOrEnum {refPath=&NiagaraPosition}}
      }
     ]
     moduleScript {refPath=&AddVelocity}
    }
    {
     bIsSetParametersModule=false
     enabled=true
     moduleName=ApplyInitialForces
     inputs [
      Apply Force to Velocity {bIsDynamic=false, bIsEditable=true, bIsStaticSwitch=false, bIsVisible=true, type {classStructOrEnum {refPath=&NiagaraBool}}}
      Apply Rotational Force to Rotational Velocity {bIsDynamic=false, bIsEditable=true, bIsStaticSwitch=false, bIsVisible=true, type {classStructOrEnum {refPath=&NiagaraBool}}}
      Apply Force to Position {bIsDynamic=false, bIsEditable=true, bIsStaticSwitch=false, bIsVisible=true, type {classStructOrEnum {refPath=&NiagaraBool}}}
      Positional Force Warmup Time {bIsDynamic=false, bIsEditable=false, bIsStaticSwitch=false, bIsVisible=true, type {classStructOrEnum {refPath=&NiagaraFloat}}}
     ]
     moduleScript {refPath=&ApplyInitialForces}
    }
    {
     bIsSetParametersModule=true
     enabled=true
     moduleName=SetVariables_#3
     inputs [Particles.DynamicMaterialParameter_Air {bIsDynamic=true, bIsEditable=true, bIsStaticSwitch=false, bIsVisible=true, type {classStructOrEnum {refPath=&Vector4f}}}]
     moduleScript {refPath=&self:DefaultPropSpawn_Merged004_0.NiagaraScriptSource_1.NiagaraGraph_0.NiagaraNodeAssignment_5.SetVariables_#3}
    }
   ]
  }
  particleUpdateScript {
   scriptName=ParticleUpdateScript
   modules [
    {
     bIsSetParametersModule=false
     enabled=true
     moduleName=ParticleState
     inputs [
      Kill Particles When Lifetime Has Elapsed {bIsDynamic=false, bIsEditable=true, bIsStaticSwitch=true, bIsVisible=true, type {classStructOrEnum {refPath=&NiagaraBool}}}
      Loop Particles Lifetime {bIsDynamic=false, bIsEditable=false, bIsStaticSwitch=false, bIsVisible=false, type {classStructOrEnum {refPath=&NiagaraBool}}}
      Let Infinitely Lived Particles Die When Emitter Deactivates {bIsDynamic=false, bIsEditable=false, bIsStaticSwitch=false, bIsVisible=false, type {classStructOrEnum {refPath=&NiagaraBool}}}
      Lifetime {bIsDynamic=false, bIsEditable=true, bIsStaticSwitch=false, bIsVisible=true, type {classStructOrEnum {refPath=&NiagaraFloat}}}
      DeltaTime {bIsDynamic=false, bIsEditable=true, bIsStaticSwitch=false, bIsVisible=true, type {classStructOrEnum {refPath=&NiagaraFloat}}}
     ]
     moduleScript {refPath=&ParticleState}
    }
    {
     bIsSetParametersModule=false
     enabled=true
     moduleName=ScaleMeshSize
     inputs [
      Scale Factor {bIsDynamic=true, bIsEditable=true, bIsStaticSwitch=false, bIsVisible=true, type {classStructOrEnum {refPath=&Vector3f}}}
      Initial Mesh Scale {bIsDynamic=false, bIsEditable=true, bIsStaticSwitch=false, bIsVisible=true, type {classStructOrEnum {refPath=&Vector3f}}}
     ]
     moduleScript {refPath=&ScaleMeshSize}
    }
    {
     bIsSetParametersModule=false
     enabled=true
     moduleName=SpriteRotationRate
     inputs [
      Rotation Rate {bIsDynamic=true, bIsEditable=true, bIsStaticSwitch=false, bIsVisible=true, type {classStructOrEnum {refPath=&NiagaraFloat}}}
      Delta Time {bIsDynamic=false, bIsEditable=true, bIsStaticSwitch=false, bIsVisible=true, type {classStructOrEnum {refPath=&NiagaraFloat}}}
     ]
     moduleScript {refPath=&SpriteRotationRate}
    }
    {
     bIsSetParametersModule=false
     enabled=true
     moduleName=GravityForce
     inputs [
      Gravity {bIsDynamic=true, bIsEditable=true, bIsStaticSwitch=false, bIsVisible=true, type {classStructOrEnum {refPath=&Vector3f}}}
      Coordinate Space {
       bIsDynamic=false
       bIsEditable=true
       bIsStaticSwitch=true
       bIsVisible=true
       type {classStructOrEnum {refPath=&ENiagaraCoordinateSpace}}
      }
     ]
     moduleScript {refPath=&GravityForce}
    }
    {
     bIsSetParametersModule=false
     enabled=true
     moduleName=CurlNoiseForce
     inputs [
      Noise Strength {bIsDynamic=true, bIsEditable=true, bIsStaticSwitch=false, bIsVisible=true, type {classStructOrEnum {refPath=&NiagaraFloat}}}
      Noise Frequency {bIsDynamic=true, bIsEditable=true, bIsStaticSwitch=false, bIsVisible=true, type {classStructOrEnum {refPath=&NiagaraFloat}}}
      Noise Quality / Cost {
       bIsDynamic=false
       bIsEditable=true
       bIsStaticSwitch=true
       bIsVisible=true
       type {classStructOrEnum {refPath=&ENiagaraCurlNoiseQuality}}
      }
      Random Seed {bIsDynamic=false, bIsEditable=true, bIsStaticSwitch=false, bIsVisible=true, type {classStructOrEnum {refPath=&NiagaraInt32}}}
      Pan Noise {bIsDynamic=false, bIsEditable=true, bIsStaticSwitch=false, bIsVisible=true, type {classStructOrEnum {refPath=&NiagaraBool}}}
      Pan Noise Field {bIsDynamic=false, bIsEditable=false, bIsStaticSwitch=false, bIsVisible=true, type {classStructOrEnum {refPath=&Vector3f}}}
      Write to Intrinsic Parameters {bIsDynamic=false, bIsEditable=true, bIsStaticSwitch=true, bIsVisible=true, type {classStructOrEnum {refPath=&NiagaraBool}}}
      Sample Position {bIsDynamic=false, bIsEditable=true, bIsStaticSwitch=false, bIsVisible=true, type {classStructOrEnum {refPath=&Vector3f}}}
      CurlNoise {
       bIsDynamic=false
       bIsEditable=false
       bIsStaticSwitch=false
       bIsVisible=false
       type {classStructOrEnum {refPath=&NiagaraDataInterfaceCurlNoise}}
      }
      Mask Curl Noise {bIsDynamic=false, bIsEditable=true, bIsStaticSwitch=true, bIsVisible=true, type {classStructOrEnum {refPath=&NiagaraBool}}}
      Curl Noise Cone Mask Angle {
       bIsDynamic=false
       bIsEditable=false
       bIsStaticSwitch=false
       bIsVisible=false
       type {classStructOrEnum {refPath=&NiagaraFloat}}
      }
      Curl Noise Cone Mask Falloff Angle {
       bIsDynamic=false
       bIsEditable=false
       bIsStaticSwitch=false
       bIsVisible=false
       type {classStructOrEnum {refPath=&NiagaraFloat}}
      }
      Curl Noise Cone Mask Axis {
       bIsDynamic=false
       bIsEditable=false
       bIsStaticSwitch=false
       bIsVisible=false
       type {classStructOrEnum {refPath=&Vector3f}}
      }
     ]
     moduleScript {refPath=&CurlNoiseForce}
    }
    {
     bIsSetParametersModule=false
     enabled=true
     moduleName=Drag
     inputs [
      Use Linear Drag {bIsDynamic=false, bIsEditable=true, bIsStaticSwitch=false, bIsVisible=true, type {classStructOrEnum {refPath=&NiagaraBool}}}
      Drag {bIsDynamic=true, bIsEditable=true, bIsStaticSwitch=false, bIsVisible=true, type {classStructOrEnum {refPath=&NiagaraFloat}}}
      Use Rotational Drag {bIsDynamic=false, bIsEditable=true, bIsStaticSwitch=false, bIsVisible=true, type {classStructOrEnum {refPath=&NiagaraBool}}}
      Rotational Drag {bIsDynamic=false, bIsEditable=true, bIsStaticSwitch=false, bIsVisible=true, type {classStructOrEnum {refPath=&NiagaraFloat}}}
      Ignore Mass {bIsDynamic=false, bIsEditable=true, bIsStaticSwitch=false, bIsVisible=true, type {classStructOrEnum {refPath=&NiagaraBool}}}
     ]
     moduleScript {refPath=&Drag}
    }
    {
     bIsSetParametersModule=false
     enabled=true
     moduleName=SolveForcesAndVelocity
     inputs [
      Clamp Velocity {bIsDynamic=false, bIsEditable=true, bIsStaticSwitch=false, bIsVisible=true, type {classStructOrEnum {refPath=&NiagaraBool}}}
      Speed Limit {bIsDynamic=false, bIsEditable=false, bIsStaticSwitch=false, bIsVisible=true, type {classStructOrEnum {refPath=&NiagaraFloat}}}
      Limit Acceleration {bIsDynamic=false, bIsEditable=true, bIsStaticSwitch=false, bIsVisible=true, type {classStructOrEnum {refPath=&NiagaraBool}}}
      Acceleration Limit {bIsDynamic=false, bIsEditable=false, bIsStaticSwitch=false, bIsVisible=true, type {classStructOrEnum {refPath=&NiagaraFloat}}}
      Manually Enable Rotational Solver {bIsDynamic=false, bIsEditable=true, bIsStaticSwitch=false, bIsVisible=true, type {classStructOrEnum {refPath=&NiagaraBool}}}
      Write to Intrinsic Properties {bIsDynamic=false, bIsEditable=true, bIsStaticSwitch=false, bIsVisible=true, type {classStructOrEnum {refPath=&NiagaraBool}}}
      Write to Presolve Properties {bIsDynamic=false, bIsEditable=true, bIsStaticSwitch=true, bIsVisible=true, type {classStructOrEnum {refPath=&NiagaraBool}}}
      DeltaTime {bIsDynamic=false, bIsEditable=true, bIsStaticSwitch=false, bIsVisible=true, type {classStructOrEnum {refPath=&NiagaraFloat}}}
      Mass {bIsDynamic=false, bIsEditable=true, bIsStaticSwitch=false, bIsVisible=true, type {classStructOrEnum {refPath=&NiagaraFloat}}}
      Position {
       bIsDynamic=false
       bIsEditable=true
       bIsStaticSwitch=false
       bIsVisible=true
       type {classStructOrEnum {refPath=&NiagaraPosition}}
      }
      Velocity {bIsDynamic=false, bIsEditable=true, bIsStaticSwitch=false, bIsVisible=true, type {classStructOrEnum {refPath=&Vector3f}}}
      PreviousVelocity {bIsDynamic=false, bIsEditable=true, bIsStaticSwitch=false, bIsVisible=true, type {classStructOrEnum {refPath=&Vector3f}}}
      Force {bIsDynamic=false, bIsEditable=true, bIsStaticSwitch=false, bIsVisible=true, type {classStructOrEnum {refPath=&Vector3f}}}
      Mesh Orientation {bIsDynamic=false, bIsEditable=false, bIsStaticSwitch=false, bIsVisible=false, type {classStructOrEnum {refPath=&Quat4f}}}
      Rotational Velocity {
       bIsDynamic=false
       bIsEditable=false
       bIsStaticSwitch=false
       bIsVisible=false
       type {classStructOrEnum {refPath=&Vector3f}}
      }
      Rotational Force {
       bIsDynamic=false
       bIsEditable=false
       bIsStaticSwitch=false
       bIsVisible=false
       type {classStructOrEnum {refPath=&Vector3f}}
      }
      Rotational Inertia {
       bIsDynamic=false
       bIsEditable=false
       bIsStaticSwitch=false
       bIsVisible=false
       type {classStructOrEnum {refPath=&NiagaraFloat}}
      }
     ]
     moduleScript {refPath=&SolveForcesAndVelocity}
    }
    {
     bIsSetParametersModule=false
     enabled=true
     moduleName=ScaleSpriteSize
     inputs [
      Scale Sprite Size Mode {
       bIsDynamic=false
       bIsEditable=true
       bIsStaticSwitch=true
       bIsVisible=true
       type {classStructOrEnum {refPath=&ENiagara_ScaleSpriteSize}}
      }
      Uniform Curve Sprite Scale {
       bIsDynamic=false
       bIsEditable=false
       bIsStaticSwitch=false
       bIsVisible=false
       type {classStructOrEnum {refPath=&NiagaraDataInterfaceCurve}}
      }
      Non-Uniform Curve Sprite Scale {
       bIsDynamic=false
       bIsEditable=false
       bIsStaticSwitch=false
       bIsVisible=false
       type {classStructOrEnum {refPath=&NiagaraDataInterfaceVector2DCurve}}
      }
      Uniform Curve Index {
       bIsDynamic=false
       bIsEditable=false
       bIsStaticSwitch=false
       bIsVisible=false
       type {classStructOrEnum {refPath=&NiagaraFloat}}
      }
      Non-Uniform Curve Index {
       bIsDynamic=false
       bIsEditable=false
       bIsStaticSwitch=false
       bIsVisible=false
       type {classStructOrEnum {refPath=&NiagaraFloat}}
      }
      Non-Uniform Curve Scale {
       bIsDynamic=false
       bIsEditable=false
       bIsStaticSwitch=false
       bIsVisible=false
       type {classStructOrEnum {refPath=&Vector2f}}
      }
      Uniform Curve Scale {
       bIsDynamic=false
       bIsEditable=false
       bIsStaticSwitch=false
       bIsVisible=false
       type {classStructOrEnum {refPath=&NiagaraFloat}}
      }
      Scale Factor {bIsDynamic=true, bIsEditable=true, bIsStaticSwitch=false, bIsVisible=true, type {classStructOrEnum {refPath=&Vector2f}}}
      Uniform Scale Factor {
       bIsDynamic=false
       bIsEditable=false
       bIsStaticSwitch=false
       bIsVisible=false
       type {classStructOrEnum {refPath=&NiagaraFloat}}
      }
      Initial Sprite Size {bIsDynamic=false, bIsEditable=true, bIsStaticSwitch=false, bIsVisible=true, type {classStructOrEnum {refPath=&Vector2f}}}
     ]
     moduleScript {refPath=&ScaleSpriteSize}
    }
    {
     bIsSetParametersModule=true
     enabled=true
     moduleName=SetVariables_#4
     inputs [
      Particles.DynParSmoke0 {bIsDynamic=true, bIsEditable=true, bIsStaticSwitch=false, bIsVisible=true, type {classStructOrEnum {refPath=&Vector4f}}}
      Particles.DynParSmoke1 {bIsDynamic=true, bIsEditable=true, bIsStaticSwitch=false, bIsVisible=true, type {classStructOrEnum {refPath=&Vector4f}}}
      Particles.DynParSmoke2 {bIsDynamic=true, bIsEditable=true, bIsStaticSwitch=false, bIsVisible=true, type {classStructOrEnum {refPath=&Vector4f}}}
     ]
     moduleScript {refPath=&self:DefaultPropSpawn_Merged004_0.NiagaraScriptSource_1.NiagaraGraph_0.NiagaraNodeAssignment_2.SetVariables_#4}
    }
    {
     bIsSetParametersModule=false
     enabled=true
     moduleName=ScaleColor001
     inputs [
      Scale Mode {bIsDynamic=false, bIsEditable=true, bIsStaticSwitch=true, bIsVisible=true, type {classStructOrEnum {refPath=&NiagaraInt32}}}
      Linear Color Curve {
       bIsDynamic=false
       bIsEditable=false
       bIsStaticSwitch=false
       bIsVisible=false
       type {classStructOrEnum {refPath=&NiagaraDataInterfaceColorCurve}}
      }
      ScaleRGB {bIsDynamic=false, bIsEditable=true, bIsStaticSwitch=false, bIsVisible=true, type {classStructOrEnum {refPath=&NiagaraBool}}}
      Scale RGB {bIsDynamic=false, bIsEditable=true, bIsStaticSwitch=false, bIsVisible=true, type {classStructOrEnum {refPath=&Vector3f}}}
      ScaleRGBA {bIsDynamic=false, bIsEditable=false, bIsStaticSwitch=false, bIsVisible=false, type {classStructOrEnum {refPath=&NiagaraBool}}}
      Scale RGBA {
       bIsDynamic=false
       bIsEditable=false
       bIsStaticSwitch=false
       bIsVisible=false
       type {classStructOrEnum {refPath=&Vector4f}}
      }
      Curve Index {
       bIsDynamic=false
       bIsEditable=false
       bIsStaticSwitch=false
       bIsVisible=false
       type {classStructOrEnum {refPath=&NiagaraFloat}}
      }
      ScaleA {bIsDynamic=false, bIsEditable=true, bIsStaticSwitch=false, bIsVisible=true, type {classStructOrEnum {refPath=&NiagaraBool}}}
      Scale Alpha {bIsDynamic=true, bIsEditable=true, bIsStaticSwitch=false, bIsVisible=true, type {classStructOrEnum {refPath=&NiagaraFloat}}}
      Color Value To Scale {
       bIsDynamic=false
       bIsEditable=true
       bIsStaticSwitch=false
       bIsVisible=true
       type {classStructOrEnum {refPath=&LinearColor}}
      }
      Color Scale {
       bIsDynamic=false
       bIsEditable=false
       bIsStaticSwitch=false
       bIsVisible=false
       type {classStructOrEnum {refPath=&LinearColor}}
      }
     ]
     moduleScript {refPath=&ScaleColor}
    }
    {
     bIsSetParametersModule=false
     enabled=true
     moduleName=FN_Light_Attributes
     inputs [
      Enable Particle Light Enable {bIsDynamic=false, bIsEditable=true, bIsStaticSwitch=true, bIsVisible=true, type {classStructOrEnum {refPath=&NiagaraBool}}}
      Light Enabled {bIsDynamic=false, bIsEditable=false, bIsStaticSwitch=false, bIsVisible=false, type {classStructOrEnum {refPath=&NiagaraBool}}}
      Enable Visibility Tag {bIsDynamic=false, bIsEditable=true, bIsStaticSwitch=true, bIsVisible=true, type {classStructOrEnum {refPath=&NiagaraBool}}}
      VisibilityTag {
       bIsDynamic=false
       bIsEditable=false
       bIsStaticSwitch=false
       bIsVisible=false
       type {classStructOrEnum {refPath=&NiagaraInt32}}
      }
      Create Position Attribute {bIsDynamic=false, bIsEditable=true, bIsStaticSwitch=true, bIsVisible=true, type {classStructOrEnum {refPath=&NiagaraBool}}}
      Position {bIsDynamic=false, bIsEditable=true, bIsStaticSwitch=false, bIsVisible=true, type {classStructOrEnum {refPath=&Vector3f}}}
      Create Radius Attribute {bIsDynamic=false, bIsEditable=true, bIsStaticSwitch=true, bIsVisible=true, type {classStructOrEnum {refPath=&NiagaraBool}}}
      Radius {bIsDynamic=false, bIsEditable=true, bIsStaticSwitch=false, bIsVisible=true, type {classStructOrEnum {refPath=&NiagaraFloat}}}
      Create Exponent Attribute {bIsDynamic=false, bIsEditable=true, bIsStaticSwitch=true, bIsVisible=true, type {classStructOrEnum {refPath=&NiagaraBool}}}
      Exponent {bIsDynamic=false, bIsEditable=true, bIsStaticSwitch=false, bIsVisible=true, type {classStructOrEnum {refPath=&NiagaraFloat}}}
      Create Color Attribute {bIsDynamic=false, bIsEditable=true, bIsStaticSwitch=true, bIsVisible=true, type {classStructOrEnum {refPath=&NiagaraBool}}}
      Color {
       bIsDynamic=false
       bIsEditable=true
       bIsStaticSwitch=false
       bIsVisible=true
       type {classStructOrEnum {refPath=&LinearColor}}
      }
      Scale Color {bIsDynamic=true, bIsEditable=true, bIsStaticSwitch=false, bIsVisible=true, type {classStructOrEnum {refPath=&NiagaraFloat}}}
     ]
     moduleScript {refPath=&FN_Light_Attributes}
    }
    {
     bIsSetParametersModule=true
     enabled=true
     moduleName=SetVariables_#9
     inputs [
      Particles.ShockwaveScale {bIsDynamic=true, bIsEditable=true, bIsStaticSwitch=false, bIsVisible=true, type {classStructOrEnum {refPath=&Vector3f}}}
      Particles.ShockwaveColor {
       bIsDynamic=true
       bIsEditable=true
       bIsStaticSwitch=false
       bIsVisible=true
       type {classStructOrEnum {refPath=&LinearColor}}
      }
      Particles.SecondarySphereColor {
       bIsDynamic=true
       bIsEditable=true
       bIsStaticSwitch=false
       bIsVisible=true
       type {classStructOrEnum {refPath=&LinearColor}}
      }
      Particles.SecondaryDyn {bIsDynamic=true, bIsEditable=true, bIsStaticSwitch=false, bIsVisible=true, type {classStructOrEnum {refPath=&Vector4f}}}
      Particles.PrimaryColor {
       bIsDynamic=true
       bIsEditable=true
       bIsStaticSwitch=false
       bIsVisible=true
       type {classStructOrEnum {refPath=&LinearColor}}
      }
     ]
     moduleScript {refPath=&self:DefaultPropSpawn_Merged004_0.NiagaraScriptSource_1.NiagaraGraph_0.NiagaraNodeAssignment_3.SetVariables_#9}
    }
   ]
  }
  rendererClasses [
   {refPath=&NiagaraSpriteRendererProperties}
   {refPath=&NiagaraLightRendererProperties}
   {refPath=&NiagaraMeshRendererProperties}
  ]
  renderers [
   {rendererIndex=0, rendererClass {refPath=&NiagaraSpriteRendererProperties}}
   {rendererIndex=1, rendererClass {refPath=&NiagaraSpriteRendererProperties}}
   {rendererIndex=2, rendererClass {refPath=&NiagaraSpriteRendererProperties}}
   {rendererIndex=3, rendererClass {refPath=&NiagaraSpriteRendererProperties}}
   {rendererIndex=4, rendererClass {refPath=&NiagaraSpriteRendererProperties}}
   {rendererIndex=5, rendererClass {refPath=&NiagaraLightRendererProperties}}
   {rendererIndex=6, rendererClass {refPath=&NiagaraSpriteRendererProperties}}
   {rendererIndex=7, rendererClass {refPath=&NiagaraMeshRendererProperties}}
   {rendererIndex=8, rendererClass {refPath=&NiagaraMeshRendererProperties}}
   {rendererIndex=9, rendererClass {refPath=&NiagaraMeshRendererProperties}}
  ]
 }
 DefaultPropSpawn_Merged005 {
  emitterSpawnScript {scriptName=EmitterSpawnScript, modules []}
  emitterUpdateScript {
   scriptName=EmitterUpdateScript
   modules [
    {
     bIsSetParametersModule=false
     enabled=true
     moduleName=EmitterState
     inputs [
      Life Cycle Mode {
       bIsDynamic=false
       bIsEditable=true
       bIsStaticSwitch=true
       bIsVisible=true
       type {classStructOrEnum {refPath=&ENiagaraEmitterLifeCycleMode}}
      }
      Inactive Response {
       bIsDynamic=false
       bIsEditable=false
       bIsStaticSwitch=true
       bIsVisible=false
       type {classStructOrEnum {refPath=&ENiagaraInactiveMode}}
      }
      Loop Behavior {
       bIsDynamic=false
       bIsEditable=false
       bIsStaticSwitch=true
       bIsVisible=false
       type {classStructOrEnum {refPath=&ENiagara_EmitterStateOptions}}
      }
      Loop Duration Mode {
       bIsDynamic=false
       bIsEditable=false
       bIsStaticSwitch=true
       bIsVisible=false
       type {classStructOrEnum {refPath=&ENiagara_InfiniteLoopDuration}}
      }
      Loop Count {
       bIsDynamic=false
       bIsEditable=false
       bIsStaticSwitch=false
       bIsVisible=false
       type {classStructOrEnum {refPath=&NiagaraInt32}}
      }
      Loop Duration {
       bIsDynamic=false
       bIsEditable=false
       bIsStaticSwitch=false
       bIsVisible=false
       type {classStructOrEnum {refPath=&NiagaraFloat}}
      }
      Recalculate Duration Each Loop {bIsDynamic=false, bIsEditable=false, bIsStaticSwitch=false, bIsVisible=false, type {classStructOrEnum {refPath=&NiagaraBool}}}
      UseLoopDelay {bIsDynamic=false, bIsEditable=false, bIsStaticSwitch=false, bIsVisible=false, type {classStructOrEnum {refPath=&NiagaraBool}}}
      Loop Delay {
       bIsDynamic=false
       bIsEditable=false
       bIsStaticSwitch=false
       bIsVisible=false
       type {classStructOrEnum {refPath=&NiagaraFloat}}
      }
      Delay First Loop Only {bIsDynamic=false, bIsEditable=false, bIsStaticSwitch=false, bIsVisible=false, type {classStructOrEnum {refPath=&NiagaraBool}}}
      Scalability Mode {
       bIsDynamic=false
       bIsEditable=true
       bIsStaticSwitch=true
       bIsVisible=true
       type {classStructOrEnum {refPath=&ENiagaraEmitterScalabilityMode_Limited}}
      }
      Enable Distance Culling {bIsDynamic=false, bIsEditable=false, bIsStaticSwitch=true, bIsVisible=false, type {classStructOrEnum {refPath=&NiagaraBool}}}
      MinDistance {
       bIsDynamic=false
       bIsEditable=false
       bIsStaticSwitch=false
       bIsVisible=false
       type {classStructOrEnum {refPath=&NiagaraFloat}}
      }
      Min Distance Response {
       bIsDynamic=false
       bIsEditable=false
       bIsStaticSwitch=false
       bIsVisible=false
       type {classStructOrEnum {refPath=&ENiagaraExecutionStateManagement}}
      }
      MaxDistance {
       bIsDynamic=false
       bIsEditable=false
       bIsStaticSwitch=false
       bIsVisible=false
       type {classStructOrEnum {refPath=&NiagaraFloat}}
      }
      Max Distance Response {
       bIsDynamic=false
       bIsEditable=false
       bIsStaticSwitch=false
       bIsVisible=false
       type {classStructOrEnum {refPath=&ENiagaraExecutionStateManagement}}
      }
      Scale Spawn Count {bIsDynamic=false, bIsEditable=false, bIsStaticSwitch=true, bIsVisible=false, type {classStructOrEnum {refPath=&NiagaraBool}}}
      Spawn Count Scale {
       bIsDynamic=false
       bIsEditable=false
       bIsStaticSwitch=false
       bIsVisible=false
       type {classStructOrEnum {refPath=&NiagaraFloat}}
      }
      Spawn Count Scale By Distance {
       bIsDynamic=false
       bIsEditable=false
       bIsStaticSwitch=false
       bIsVisible=false
       type {classStructOrEnum {refPath=&NiagaraDataInterfaceCurve}}
      }
      Enable Visibility Culling {bIsDynamic=false, bIsEditable=false, bIsStaticSwitch=true, bIsVisible=false, type {classStructOrEnum {refPath=&NiagaraBool}}}
      Visibility Cull Response {
       bIsDynamic=false
       bIsEditable=false
       bIsStaticSwitch=false
       bIsVisible=false
       type {classStructOrEnum {refPath=&ENiagaraExecutionStateManagement}}
      }
      Visibility Cull Delay {
       bIsDynamic=false
       bIsEditable=false
       bIsStaticSwitch=false
       bIsVisible=false
       type {classStructOrEnum {refPath=&NiagaraFloat}}
      }
      Reset Age On Awaken {bIsDynamic=false, bIsEditable=false, bIsStaticSwitch=false, bIsVisible=false, type {classStructOrEnum {refPath=&NiagaraBool}}}
     ]
     moduleScript {refPath=&EmitterState}
    }
    {
     bIsSetParametersModule=false
     enabled=true
     moduleName=SpawnBurst_Instantaneous001
     inputs [
      Spawn Count {bIsDynamic=true, bIsEditable=true, bIsStaticSwitch=false, bIsVisible=true, type {classStructOrEnum {refPath=&NiagaraInt32}}}
      Spawn Time {bIsDynamic=false, bIsEditable=true, bIsStaticSwitch=false, bIsVisible=true, type {classStructOrEnum {refPath=&NiagaraFloat}}}
      Use Spawn Probability {bIsDynamic=false, bIsEditable=true, bIsStaticSwitch=false, bIsVisible=true, type {classStructOrEnum {refPath=&NiagaraBool}}}
      Spawn Probability {bIsDynamic=false, bIsEditable=false, bIsStaticSwitch=false, bIsVisible=true, type {classStructOrEnum {refPath=&NiagaraFloat}}}
      UseLoopCountLimit {bIsDynamic=false, bIsEditable=true, bIsStaticSwitch=false, bIsVisible=true, type {classStructOrEnum {refPath=&NiagaraBool}}}
      Loop Count Limit {bIsDynamic=false, bIsEditable=false, bIsStaticSwitch=false, bIsVisible=true, type {classStructOrEnum {refPath=&NiagaraInt32}}}
      Spawn Group {bIsDynamic=false, bIsEditable=true, bIsStaticSwitch=false, bIsVisible=true, type {classStructOrEnum {refPath=&NiagaraInt32}}}
      Age {bIsDynamic=false, bIsEditable=true, bIsStaticSwitch=false, bIsVisible=true, type {classStructOrEnum {refPath=&NiagaraFloat}}}
     ]
     moduleScript {refPath=&SpawnBurst_Instantaneous}
    }
    {
     bIsSetParametersModule=false
     enabled=true
     moduleName=SpawnBurst_Instantaneous
     inputs [
      Spawn Count {bIsDynamic=false, bIsEditable=true, bIsStaticSwitch=false, bIsVisible=true, type {classStructOrEnum {refPath=&NiagaraInt32}}}
      Spawn Time {bIsDynamic=false, bIsEditable=true, bIsStaticSwitch=false, bIsVisible=true, type {classStructOrEnum {refPath=&NiagaraFloat}}}
      Use Spawn Probability {bIsDynamic=false, bIsEditable=true, bIsStaticSwitch=false, bIsVisible=true, type {classStructOrEnum {refPath=&NiagaraBool}}}
      Spawn Probability {bIsDynamic=false, bIsEditable=false, bIsStaticSwitch=false, bIsVisible=true, type {classStructOrEnum {refPath=&NiagaraFloat}}}
      UseLoopCountLimit {bIsDynamic=false, bIsEditable=true, bIsStaticSwitch=false, bIsVisible=true, type {classStructOrEnum {refPath=&NiagaraBool}}}
      Loop Count Limit {bIsDynamic=false, bIsEditable=false, bIsStaticSwitch=false, bIsVisible=true, type {classStructOrEnum {refPath=&NiagaraInt32}}}
      Spawn Group {bIsDynamic=false, bIsEditable=true, bIsStaticSwitch=false, bIsVisible=true, type {classStructOrEnum {refPath=&NiagaraInt32}}}
      Age {bIsDynamic=false, bIsEditable=true, bIsStaticSwitch=false, bIsVisible=true, type {classStructOrEnum {refPath=&NiagaraFloat}}}
     ]
     moduleScript {refPath=&SpawnBurst_Instantaneous}
    }
    {
     bIsSetParametersModule=false
     enabled=true
     moduleName=SpawnBurst_Instantaneous002
     inputs [
      Spawn Count {bIsDynamic=true, bIsEditable=true, bIsStaticSwitch=false, bIsVisible=true, type {classStructOrEnum {refPath=&NiagaraInt32}}}
      Spawn Time {bIsDynamic=false, bIsEditable=true, bIsStaticSwitch=false, bIsVisible=true, type {classStructOrEnum {refPath=&NiagaraFloat}}}
      Use Spawn Probability {bIsDynamic=false, bIsEditable=true, bIsStaticSwitch=false, bIsVisible=true, type {classStructOrEnum {refPath=&NiagaraBool}}}
      Spawn Probability {bIsDynamic=false, bIsEditable=false, bIsStaticSwitch=false, bIsVisible=true, type {classStructOrEnum {refPath=&NiagaraFloat}}}
      UseLoopCountLimit {bIsDynamic=false, bIsEditable=true, bIsStaticSwitch=false, bIsVisible=true, type {classStructOrEnum {refPath=&NiagaraBool}}}
      Loop Count Limit {bIsDynamic=false, bIsEditable=false, bIsStaticSwitch=false, bIsVisible=true, type {classStructOrEnum {refPath=&NiagaraInt32}}}
      Spawn Group {bIsDynamic=false, bIsEditable=true, bIsStaticSwitch=false, bIsVisible=true, type {classStructOrEnum {refPath=&NiagaraInt32}}}
      Age {bIsDynamic=false, bIsEditable=true, bIsStaticSwitch=false, bIsVisible=true, type {classStructOrEnum {refPath=&NiagaraFloat}}}
     ]
     moduleScript {refPath=&SpawnBurst_Instantaneous}
    }
   ]
  }
  particleSpawnScript {
   scriptName=ParticleSpawnScript
   modules [
    {
     bIsSetParametersModule=true
     enabled=true
     moduleName=SetVariables_#1
     inputs [Particles.VisibilityTag {bIsDynamic=true, bIsEditable=true, bIsStaticSwitch=false, bIsVisible=true, type {classStructOrEnum {refPath=&NiagaraInt32}}}]
     moduleScript {refPath=&self:DefaultPropSpawn_Merged005_0.NiagaraScriptSource_1.NiagaraGraph_0.NiagaraNodeAssignment_1.SetVariables_#1}
    }
    {
     bIsSetParametersModule=false
     enabled=true
     moduleName=PartitionParticles
     inputs [
      Partition Mode {
       bIsDynamic=false
       bIsEditable=true
       bIsStaticSwitch=true
       bIsVisible=true
       type {classStructOrEnum {refPath=&ENiagara_PartitionMode_v2}}
      }
      Partition Count Mode {
       bIsDynamic=false
       bIsEditable=false
       bIsStaticSwitch=true
       bIsVisible=false
       type {classStructOrEnum {refPath=&ENiagara_InfiniteLoopDuration}}
      }
      A {
       bIsDynamic=false
       bIsEditable=false
       bIsStaticSwitch=false
       bIsVisible=false
       type {classStructOrEnum {refPath=&NiagaraFloat}}
      }
      Number Of Partitions {bIsDynamic=false, bIsEditable=true, bIsStaticSwitch=false, bIsVisible=true, type {classStructOrEnum {refPath=&NiagaraInt32}}}
      Particles Per Partition {
       bIsDynamic=false
       bIsEditable=false
       bIsStaticSwitch=false
       bIsVisible=false
       type {classStructOrEnum {refPath=&NiagaraInt32}}
      }
      Partition Duration {
       bIsDynamic=false
       bIsEditable=false
       bIsStaticSwitch=false
       bIsVisible=false
       type {classStructOrEnum {refPath=&NiagaraFloat}}
      }
      Time {
       bIsDynamic=false
       bIsEditable=false
       bIsStaticSwitch=false
       bIsVisible=false
       type {classStructOrEnum {refPath=&NiagaraFloat}}
      }
      Partitioning Frequency {
       bIsDynamic=false
       bIsEditable=false
       bIsStaticSwitch=true
       bIsVisible=false
       type {classStructOrEnum {refPath=&ENiagaraRandomnessEvaluation}}
      }
      B {
       bIsDynamic=false
       bIsEditable=false
       bIsStaticSwitch=false
       bIsVisible=false
       type {classStructOrEnum {refPath=&NiagaraFloat}}
      }
      Distance {
       bIsDynamic=false
       bIsEditable=false
       bIsStaticSwitch=false
       bIsVisible=false
       type {classStructOrEnum {refPath=&NiagaraFloat}}
      }
      Comparison Type {
       bIsDynamic=false
       bIsEditable=false
       bIsStaticSwitch=true
       bIsVisible=false
       type {classStructOrEnum {refPath=&ENiagaraBooleanLogicOps_v2}}
      }
      Evaluation Type {
       bIsDynamic=false
       bIsEditable=false
       bIsStaticSwitch=true
       bIsVisible=false
       type {classStructOrEnum {refPath=&ENiagaraRandomnessEvaluation}}
      }
      UseDistanceFalloff {bIsDynamic=false, bIsEditable=false, bIsStaticSwitch=false, bIsVisible=false, type {classStructOrEnum {refPath=&NiagaraBool}}}
      Distance Falloff Exponent {
       bIsDynamic=false
       bIsEditable=false
       bIsStaticSwitch=false
       bIsVisible=false
       type {classStructOrEnum {refPath=&NiagaraFloat}}
      }
      Start Position {
       bIsDynamic=false
       bIsEditable=false
       bIsStaticSwitch=false
       bIsVisible=false
       type {classStructOrEnum {refPath=&NiagaraPosition}}
      }
      End Position {
       bIsDynamic=false
       bIsEditable=false
       bIsStaticSwitch=false
       bIsVisible=false
       type {classStructOrEnum {refPath=&NiagaraPosition}}
      }
      ID to Partition {bIsDynamic=false, bIsEditable=true, bIsStaticSwitch=false, bIsVisible=true, type {classStructOrEnum {refPath=&NiagaraInt32}}}
      Randomness Mode {
       bIsDynamic=false
       bIsEditable=false
       bIsStaticSwitch=true
       bIsVisible=false
       type {classStructOrEnum {refPath=&ENiagaraRandomnessMode}}
      }
      Override Seed {bIsDynamic=false, bIsEditable=false, bIsStaticSwitch=true, bIsVisible=false, type {classStructOrEnum {refPath=&NiagaraBool}}}
      Fixed Random Seed {bIsDynamic=false, bIsEditable=false, bIsStaticSwitch=true, bIsVisible=false, type {classStructOrEnum {refPath=&NiagaraBool}}}
      Color A {
       bIsDynamic=false
       bIsEditable=false
       bIsStaticSwitch=false
       bIsVisible=false
       type {classStructOrEnum {refPath=&LinearColor}}
      }
      Color B {
       bIsDynamic=false
       bIsEditable=false
       bIsStaticSwitch=false
       bIsVisible=false
       type {classStructOrEnum {refPath=&LinearColor}}
      }
      Color Comparison Type {bIsDynamic=false, bIsEditable=false, bIsStaticSwitch=true, bIsVisible=false, type {classStructOrEnum {refPath=&NiagaraInt32}}}
      Color Tolerance {
       bIsDynamic=false
       bIsEditable=false
       bIsStaticSwitch=false
       bIsVisible=false
       type {classStructOrEnum {refPath=&NiagaraFloat}}
      }
      Minimum Saturation {
       bIsDynamic=false
       bIsEditable=false
       bIsStaticSwitch=false
       bIsVisible=false
       type {classStructOrEnum {refPath=&NiagaraFloat}}
      }
      Direction A {
       bIsDynamic=false
       bIsEditable=false
       bIsStaticSwitch=false
       bIsVisible=false
       type {classStructOrEnum {refPath=&Vector3f}}
      }
      Direction B {
       bIsDynamic=false
       bIsEditable=false
       bIsStaticSwitch=false
       bIsVisible=false
       type {classStructOrEnum {refPath=&Vector3f}}
      }
      Direction Tolerance {
       bIsDynamic=false
       bIsEditable=false
       bIsStaticSwitch=false
       bIsVisible=false
       type {classStructOrEnum {refPath=&NiagaraFloat}}
      }
      Write Partition to Ribbon ID {
       bIsDynamic=false
       bIsEditable=true
       bIsStaticSwitch=true
       bIsVisible=true
       type {classStructOrEnum {refPath=&ENiagara_AttributeSamplingApplyOutput}}
      }
      Write Partition Index to Ribbon Link Order {
       bIsDynamic=false
       bIsEditable=false
       bIsStaticSwitch=true
       bIsVisible=false
       type {classStructOrEnum {refPath=&ENiagara_AttributeSamplingApplyOutput}}
      }
      Normalized Index In Partition Includes One {bIsDynamic=false, bIsEditable=false, bIsStaticSwitch=true, bIsVisible=false, type {classStructOrEnum {refPath=&NiagaraBool}}}
      Normalized Index Includes One {bIsDynamic=false, bIsEditable=true, bIsStaticSwitch=true, bIsVisible=true, type {classStructOrEnum {refPath=&NiagaraBool}}}
     ]
     moduleScript {refPath=&PartitionParticles}
    }
    {
     bIsSetParametersModule=false
     enabled=true
     moduleName=InitializeParticle001
     inputs [
      Lifetime Mode {
       bIsDynamic=false
       bIsEditable=true
       bIsStaticSwitch=true
       bIsVisible=true
       type {classStructOrEnum {refPath=&ENiagara_LifetimeMode}}
      }
      Lifetime Min {
       bIsDynamic=false
       bIsEditable=false
       bIsStaticSwitch=false
       bIsVisible=false
       type {classStructOrEnum {refPath=&NiagaraFloat}}
      }
      Lifetime {bIsDynamic=true, bIsEditable=true, bIsStaticSwitch=false, bIsVisible=true, type {classStructOrEnum {refPath=&NiagaraFloat}}}
      Lifetime Max {
       bIsDynamic=false
       bIsEditable=false
       bIsStaticSwitch=false
       bIsVisible=false
       type {classStructOrEnum {refPath=&NiagaraFloat}}
      }
      Lifetime Randomness Mode {
       bIsDynamic=false
       bIsEditable=false
       bIsStaticSwitch=true
       bIsVisible=false
       type {classStructOrEnum {refPath=&ENiagaraRandomnessMode}}
      }
      Color Mode {
       bIsDynamic=false
       bIsEditable=true
       bIsStaticSwitch=true
       bIsVisible=true
       type {classStructOrEnum {refPath=&ENiagara_ColorInitializationMode}}
      }
      Color Minimum {
       bIsDynamic=false
       bIsEditable=false
       bIsStaticSwitch=false
       bIsVisible=false
       type {classStructOrEnum {refPath=&LinearColor}}
      }
      Color {
       bIsDynamic=true
       bIsEditable=true
       bIsStaticSwitch=false
       bIsVisible=true
       type {classStructOrEnum {refPath=&LinearColor}}
      }
      AdjustHue {bIsDynamic=false, bIsEditable=false, bIsStaticSwitch=false, bIsVisible=false, type {classStructOrEnum {refPath=&NiagaraBool}}}
      Hue Shift Range {
       bIsDynamic=false
       bIsEditable=false
       bIsStaticSwitch=false
       bIsVisible=false
       type {classStructOrEnum {refPath=&Vector2f}}
      }
      Color Maximum {
       bIsDynamic=false
       bIsEditable=false
       bIsStaticSwitch=false
       bIsVisible=false
       type {classStructOrEnum {refPath=&LinearColor}}
      }
      AdjustSaturation {bIsDynamic=false, bIsEditable=false, bIsStaticSwitch=false, bIsVisible=false, type {classStructOrEnum {refPath=&NiagaraBool}}}
      Saturation Range {
       bIsDynamic=false
       bIsEditable=false
       bIsStaticSwitch=false
       bIsVisible=false
       type {classStructOrEnum {refPath=&Vector2f}}
      }
      AdjustValue {bIsDynamic=false, bIsEditable=false, bIsStaticSwitch=false, bIsVisible=false, type {classStructOrEnum {refPath=&NiagaraBool}}}
      Value Range {
       bIsDynamic=false
       bIsEditable=false
       bIsStaticSwitch=false
       bIsVisible=false
       type {classStructOrEnum {refPath=&Vector2f}}
      }
      AdjustAlpha {bIsDynamic=false, bIsEditable=false, bIsStaticSwitch=false, bIsVisible=false, type {classStructOrEnum {refPath=&NiagaraBool}}}
      Alpha Scale Range {
       bIsDynamic=false
       bIsEditable=false
       bIsStaticSwitch=false
       bIsVisible=false
       type {classStructOrEnum {refPath=&Vector2f}}
      }
      Color Channel Mode {
       bIsDynamic=false
       bIsEditable=false
       bIsStaticSwitch=true
       bIsVisible=false
       type {classStructOrEnum {refPath=&ENiagaraChannelCorrelation}}
      }
      Color Randomness Mode {
       bIsDynamic=false
       bIsEditable=false
       bIsStaticSwitch=true
       bIsVisible=false
       type {classStructOrEnum {refPath=&ENiagaraRandomnessMode}}
      }
      Position Mode {
       bIsDynamic=false
       bIsEditable=true
       bIsStaticSwitch=true
       bIsVisible=true
       type {classStructOrEnum {refPath=&ENiagara_PositionInitializationMode}}
      }
      Position {
       bIsDynamic=false
       bIsEditable=false
       bIsStaticSwitch=false
       bIsVisible=false
       type {classStructOrEnum {refPath=&NiagaraPosition}}
      }
      UsePositionOffset {bIsDynamic=false, bIsEditable=true, bIsStaticSwitch=false, bIsVisible=true, type {classStructOrEnum {refPath=&NiagaraBool}}}
      Position Offset {bIsDynamic=false, bIsEditable=true, bIsStaticSwitch=false, bIsVisible=true, type {classStructOrEnum {refPath=&Vector3f}}}
      Position Offset Coordinate Space {
       bIsDynamic=false
       bIsEditable=true
       bIsStaticSwitch=false
       bIsVisible=true
       type {classStructOrEnum {refPath=&ENiagaraCoordinateSpace}}
      }
      Mass Mode {
       bIsDynamic=false
       bIsEditable=true
       bIsStaticSwitch=true
       bIsVisible=true
       type {classStructOrEnum {refPath=&ENiagara_MassInitializationMode}}
      }
      Mass Min {
       bIsDynamic=false
       bIsEditable=false
       bIsStaticSwitch=false
       bIsVisible=false
       type {classStructOrEnum {refPath=&NiagaraFloat}}
      }
      Mass {
       bIsDynamic=false
       bIsEditable=false
       bIsStaticSwitch=false
       bIsVisible=false
       type {classStructOrEnum {refPath=&NiagaraFloat}}
      }
      Mass Max {
       bIsDynamic=false
       bIsEditable=false
       bIsStaticSwitch=false
       bIsVisible=false
       type {classStructOrEnum {refPath=&NiagaraFloat}}
      }
      Mass Randomness Mode {
       bIsDynamic=false
       bIsEditable=false
       bIsStaticSwitch=true
       bIsVisible=false
       type {classStructOrEnum {refPath=&ENiagaraRandomnessMode}}
      }
      Write Material Random {bIsDynamic=false, bIsEditable=true, bIsStaticSwitch=false, bIsVisible=true, type {classStructOrEnum {refPath=&NiagaraBool}}}
      Material Random {bIsDynamic=false, bIsEditable=true, bIsStaticSwitch=false, bIsVisible=true, type {classStructOrEnum {refPath=&NiagaraFloat}}}
      Sprite Size Mode {
       bIsDynamic=false
       bIsEditable=true
       bIsStaticSwitch=true
       bIsVisible=true
       type {classStructOrEnum {refPath=&ENiagara_SizeScaleMode}}
      }
      Sprite Size {
       bIsDynamic=false
       bIsEditable=false
       bIsStaticSwitch=false
       bIsVisible=false
       type {classStructOrEnum {refPath=&Vector2f}}
      }
      Uniform Sprite Size {bIsDynamic=true, bIsEditable=true, bIsStaticSwitch=false, bIsVisible=true, type {classStructOrEnum {refPath=&NiagaraFloat}}}
      Uniform Sprite Size Min {
       bIsDynamic=false
       bIsEditable=false
       bIsStaticSwitch=false
       bIsVisible=false
       type {classStructOrEnum {refPath=&NiagaraFloat}}
      }
      Sprite Size Min {
       bIsDynamic=false
       bIsEditable=false
       bIsStaticSwitch=false
       bIsVisible=false
       type {classStructOrEnum {refPath=&Vector2f}}
      }
      Sprite Size Max {
       bIsDynamic=false
       bIsEditable=false
       bIsStaticSwitch=false
       bIsVisible=false
       type {classStructOrEnum {refPath=&Vector2f}}
      }
      Uniform Sprite Size Max {
       bIsDynamic=false
       bIsEditable=false
       bIsStaticSwitch=false
       bIsVisible=false
       type {classStructOrEnum {refPath=&NiagaraFloat}}
      }
      Sprite Size Randomness Mode {
       bIsDynamic=false
       bIsEditable=false
       bIsStaticSwitch=true
       bIsVisible=false
       type {classStructOrEnum {refPath=&ENiagaraRandomnessMode}}
      }
      Sprite Rotation Mode {
       bIsDynamic=false
       bIsEditable=true
       bIsStaticSwitch=true
       bIsVisible=true
       type {classStructOrEnum {refPath=&ENiagara_SpriteRotationMode}}
      }
      Sprite Rotation Angle Min {
       bIsDynamic=false
       bIsEditable=false
       bIsStaticSwitch=false
       bIsVisible=false
       type {classStructOrEnum {refPath=&NiagaraFloat}}
      }
      Sprite Rotation Angle {bIsDynamic=true, bIsEditable=true, bIsStaticSwitch=false, bIsVisible=true, type {classStructOrEnum {refPath=&NiagaraFloat}}}
      Sprite Rotation Angle Max {
       bIsDynamic=false
       bIsEditable=false
       bIsStaticSwitch=false
       bIsVisible=false
       type {classStructOrEnum {refPath=&NiagaraFloat}}
      }
      Sprite Rotation Randomness Mode {
       bIsDynamic=false
       bIsEditable=false
       bIsStaticSwitch=true
       bIsVisible=false
       type {classStructOrEnum {refPath=&ENiagaraRandomnessMode}}
      }
      Sprite UV Mode {
       bIsDynamic=false
       bIsEditable=true
       bIsStaticSwitch=true
       bIsVisible=true
       type {classStructOrEnum {refPath=&ENiagara_UVFlippingMode}}
      }
      Sprite UV Randomness Mode {
       bIsDynamic=false
       bIsEditable=false
       bIsStaticSwitch=true
       bIsVisible=false
       type {classStructOrEnum {refPath=&ENiagaraRandomnessMode}}
      }
      Sprite UV Scale {
       bIsDynamic=false
       bIsEditable=false
       bIsStaticSwitch=false
       bIsVisible=false
       type {classStructOrEnum {refPath=&Vector2f}}
      }
      Mesh Scale Mode {
       bIsDynamic=false
       bIsEditable=true
       bIsStaticSwitch=true
       bIsVisible=true
       type {classStructOrEnum {refPath=&ENiagara_SizeScaleMode}}
      }
      Mesh Uniform Scale Min {
       bIsDynamic=false
       bIsEditable=false
       bIsStaticSwitch=false
       bIsVisible=false
       type {classStructOrEnum {refPath=&NiagaraFloat}}
      }
      Mesh Uniform Scale {
       bIsDynamic=false
       bIsEditable=false
       bIsStaticSwitch=false
       bIsVisible=false
       type {classStructOrEnum {refPath=&NiagaraFloat}}
      }
      Mesh Scale Min {
       bIsDynamic=false
       bIsEditable=false
       bIsStaticSwitch=false
       bIsVisible=false
       type {classStructOrEnum {refPath=&Vector3f}}
      }
      Mesh Scale {
       bIsDynamic=false
       bIsEditable=false
       bIsStaticSwitch=false
       bIsVisible=false
       type {classStructOrEnum {refPath=&Vector3f}}
      }
      Mesh Scale Max {
       bIsDynamic=false
       bIsEditable=false
       bIsStaticSwitch=false
       bIsVisible=false
       type {classStructOrEnum {refPath=&Vector3f}}
      }
      Mesh Uniform Scale Max {
       bIsDynamic=false
       bIsEditable=false
       bIsStaticSwitch=false
       bIsVisible=false
       type {classStructOrEnum {refPath=&NiagaraFloat}}
      }
      Mesh Scale Randomness Mode {
       bIsDynamic=false
       bIsEditable=false
       bIsStaticSwitch=true
       bIsVisible=false
       type {classStructOrEnum {refPath=&ENiagaraRandomnessMode}}
      }
      Mesh Renderer Array Visibility Mode {
       bIsDynamic=false
       bIsEditable=true
       bIsStaticSwitch=true
       bIsVisible=true
       type {classStructOrEnum {refPath=&ENiagara_UnsetDirectSetRandom}}
      }
      Mesh Renderer Mesh Index {
       bIsDynamic=false
       bIsEditable=false
       bIsStaticSwitch=false
       bIsVisible=false
       type {classStructOrEnum {refPath=&NiagaraInt32}}
      }
      Wrap Index By Mesh Count {bIsDynamic=false, bIsEditable=false, bIsStaticSwitch=true, bIsVisible=false, type {classStructOrEnum {refPath=&NiagaraBool}}}
      Mesh Renderer Info {
       bIsDynamic=false
       bIsEditable=false
       bIsStaticSwitch=false
       bIsVisible=false
       type {classStructOrEnum {refPath=&NiagaraDataInterfaceMeshRendererInfo}}
      }
      Mesh Renderer Random Seed {
       bIsDynamic=false
       bIsEditable=false
       bIsStaticSwitch=false
       bIsVisible=false
       type {classStructOrEnum {refPath=&NiagaraInt32}}
      }
      Ribbon Width Mode {
       bIsDynamic=false
       bIsEditable=true
       bIsStaticSwitch=true
       bIsVisible=true
       type {classStructOrEnum {refPath=&ENiagara_UnsetDirectSet}}
      }
      Ribbon Width {
       bIsDynamic=false
       bIsEditable=false
       bIsStaticSwitch=false
       bIsVisible=false
       type {classStructOrEnum {refPath=&NiagaraFloat}}
      }
      Ribbon Facing Mode {
       bIsDynamic=false
       bIsEditable=true
       bIsStaticSwitch=true
       bIsVisible=true
       type {classStructOrEnum {refPath=&ENiagara_UnsetDirectSet}}
      }
      Ribbon Facing Vector {
       bIsDynamic=false
       bIsEditable=false
       bIsStaticSwitch=false
       bIsVisible=false
       type {classStructOrEnum {refPath=&Vector3f}}
      }
      Ribbon Twist Mode {
       bIsDynamic=false
       bIsEditable=true
       bIsStaticSwitch=true
       bIsVisible=true
       type {classStructOrEnum {refPath=&ENiagara_UnsetDirectSet}}
      }
      "Ribbon Twist " {
       bIsDynamic=false
       bIsEditable=false
       bIsStaticSwitch=false
       bIsVisible=false
       type {classStructOrEnum {refPath=&NiagaraFloat}}
      }
     ]
     moduleScript {refPath=&InitializeParticle}
    }
    {
     bIsSetParametersModule=true
     enabled=true
     moduleName=SetVariables_#2
     inputs [Particles.InitSpriteRotRate {bIsDynamic=true, bIsEditable=true, bIsStaticSwitch=false, bIsVisible=true, type {classStructOrEnum {refPath=&NiagaraFloat}}}]
     moduleScript {refPath=&self:DefaultPropSpawn_Merged005_0.NiagaraScriptSource_1.NiagaraGraph_0.NiagaraNodeAssignment_0.SetVariables_#2}
    }
    {
     bIsSetParametersModule=false
     enabled=true
     moduleName=AccelerationForce
     inputs [
      Acceleration {bIsDynamic=true, bIsEditable=true, bIsStaticSwitch=false, bIsVisible=true, type {classStructOrEnum {refPath=&Vector3f}}}
      Coordinate Space {
       bIsDynamic=false
       bIsEditable=true
       bIsStaticSwitch=false
       bIsVisible=true
       type {classStructOrEnum {refPath=&ENiagaraCoordinateSpace}}
      }
     ]
     moduleScript {refPath=&AccelerationForce}
    }
    {
     bIsSetParametersModule=false
     enabled=true
     moduleName=ShapeLocation
     inputs [
      Shape Primitive {
       bIsDynamic=false
       bIsEditable=true
       bIsStaticSwitch=true
       bIsVisible=true
       type {classStructOrEnum {refPath=&ENiagara_LocationShapes}}
      }
      Box / Plane Mode {
       bIsDynamic=false
       bIsEditable=false
       bIsStaticSwitch=true
       bIsVisible=false
       type {classStructOrEnum {refPath=&ENiagara_BoxPlaneMode}}
      }
      Cone Mode {
       bIsDynamic=false
       bIsEditable=false
       bIsStaticSwitch=true
       bIsVisible=false
       type {classStructOrEnum {refPath=&ENiagara_ConeMode}}
      }
      Cylinder Height {
       bIsDynamic=false
       bIsEditable=false
       bIsStaticSwitch=false
       bIsVisible=false
       type {classStructOrEnum {refPath=&NiagaraFloat}}
      }
      Sphere Radius {bIsDynamic=true, bIsEditable=true, bIsStaticSwitch=false, bIsVisible=true, type {classStructOrEnum {refPath=&NiagaraFloat}}}
      Ring / Disc Mode {
       bIsDynamic=false
       bIsEditable=false
       bIsStaticSwitch=true
       bIsVisible=false
       type {classStructOrEnum {refPath=&ENiagaraRingDiscMode}}
      }
      Torus Mode {
       bIsDynamic=false
       bIsEditable=false
       bIsStaticSwitch=true
       bIsVisible=false
       type {classStructOrEnum {refPath=&ENiagaraShapeTorusMode}}
      }
      Knot Ring Loops {
       bIsDynamic=false
       bIsEditable=false
       bIsStaticSwitch=false
       bIsVisible=false
       type {classStructOrEnum {refPath=&NiagaraInt32}}
      }
      Knot Helix Loops {
       bIsDynamic=false
       bIsEditable=false
       bIsStaticSwitch=false
       bIsVisible=false
       type {classStructOrEnum {refPath=&NiagaraInt32}}
      }
      Enable Torus Knot Radius {bIsDynamic=false, bIsEditable=false, bIsStaticSwitch=false, bIsVisible=false, type {classStructOrEnum {refPath=&NiagaraBool}}}
      Torus Knot Tube Radius {
       bIsDynamic=false
       bIsEditable=false
       bIsStaticSwitch=false
       bIsVisible=false
       type {classStructOrEnum {refPath=&NiagaraFloat}}
      }
      Cylinder Radius {
       bIsDynamic=false
       bIsEditable=false
       bIsStaticSwitch=false
       bIsVisible=false
       type {classStructOrEnum {refPath=&NiagaraFloat}}
      }
      Cylinder Height Midpoint {
       bIsDynamic=false
       bIsEditable=false
       bIsStaticSwitch=false
       bIsVisible=false
       type {classStructOrEnum {refPath=&NiagaraFloat}}
      }
      Ring Radius {
       bIsDynamic=false
       bIsEditable=false
       bIsStaticSwitch=false
       bIsVisible=false
       type {classStructOrEnum {refPath=&NiagaraFloat}}
      }
      Box Size {
       bIsDynamic=false
       bIsEditable=false
       bIsStaticSwitch=false
       bIsVisible=false
       type {classStructOrEnum {refPath=&Vector3f}}
      }
      Box Midpoint {
       bIsDynamic=false
       bIsEditable=false
       bIsStaticSwitch=false
       bIsVisible=false
       type {classStructOrEnum {refPath=&Vector3f}}
      }
      Cone Length {
       bIsDynamic=false
       bIsEditable=false
       bIsStaticSwitch=false
       bIsVisible=false
       type {classStructOrEnum {refPath=&NiagaraFloat}}
      }
      Plane Size {
       bIsDynamic=false
       bIsEditable=false
       bIsStaticSwitch=false
       bIsVisible=false
       type {classStructOrEnum {refPath=&Vector2f}}
      }
      Wedge Horizontal Angle {
       bIsDynamic=false
       bIsEditable=false
       bIsStaticSwitch=false
       bIsVisible=false
       type {classStructOrEnum {refPath=&NiagaraFloat}}
      }
      Large Radius {
       bIsDynamic=false
       bIsEditable=false
       bIsStaticSwitch=false
       bIsVisible=false
       type {classStructOrEnum {refPath=&NiagaraFloat}}
      }
      Plane Midpoint {
       bIsDynamic=false
       bIsEditable=false
       bIsStaticSwitch=false
       bIsVisible=false
       type {classStructOrEnum {refPath=&Vector2f}}
      }
      Wedge Vertical Angle {
       bIsDynamic=false
       bIsEditable=false
       bIsStaticSwitch=false
       bIsVisible=false
       type {classStructOrEnum {refPath=&NiagaraFloat}}
      }
      Cone Angle {
       bIsDynamic=false
       bIsEditable=false
       bIsStaticSwitch=false
       bIsVisible=false
       type {classStructOrEnum {refPath=&NiagaraFloat}}
      }
      Handle Radius {
       bIsDynamic=false
       bIsEditable=false
       bIsStaticSwitch=false
       bIsVisible=false
       type {classStructOrEnum {refPath=&NiagaraFloat}}
      }
      Cone Inner Angle {
       bIsDynamic=false
       bIsEditable=false
       bIsStaticSwitch=false
       bIsVisible=false
       type {classStructOrEnum {refPath=&NiagaraFloat}}
      }
      Cone Radial Angle {
       bIsDynamic=false
       bIsEditable=false
       bIsStaticSwitch=false
       bIsVisible=false
       type {classStructOrEnum {refPath=&NiagaraFloat}}
      }
      Cone Angle Type {
       bIsDynamic=false
       bIsEditable=false
       bIsStaticSwitch=true
       bIsVisible=false
       type {classStructOrEnum {refPath=&ENiagara_AngleInput}}
      }
      Flatten Endcaps {bIsDynamic=false, bIsEditable=false, bIsStaticSwitch=true, bIsVisible=false, type {classStructOrEnum {refPath=&NiagaraBool}}}
      Apply To Particle Position {
       bIsDynamic=false
       bIsEditable=true
       bIsStaticSwitch=true
       bIsVisible=true
       type {classStructOrEnum {refPath=&ENiagara_AttributeSamplingApplyOutput}}
      }
      Cone Surface Distribution {
       bIsDynamic=false
       bIsEditable=false
       bIsStaticSwitch=false
       bIsVisible=false
       type {classStructOrEnum {refPath=&NiagaraFloat}}
      }
      Ring / Disc Distribution Mode {
       bIsDynamic=false
       bIsEditable=false
       bIsStaticSwitch=true
       bIsVisible=false
       type {classStructOrEnum {refPath=&ENiagaraSphereDistributionMode}}
      }
      Sphere Distribution {
       bIsDynamic=false
       bIsEditable=true
       bIsStaticSwitch=true
       bIsVisible=true
       type {classStructOrEnum {refPath=&ENiagaraSphereDistributionMode}}
      }
      Torus Distribution Mode {
       bIsDynamic=false
       bIsEditable=false
       bIsStaticSwitch=true
       bIsVisible=false
       type {classStructOrEnum {refPath=&ENiagaraTorusDistributionMode}}
      }
      Sphere Surface Distribution {bIsDynamic=false, bIsEditable=true, bIsStaticSwitch=false, bIsVisible=true, type {classStructOrEnum {refPath=&NiagaraFloat}}}
      Hemisphere Distribution {bIsDynamic=false, bIsEditable=true, bIsStaticSwitch=false, bIsVisible=true, type {classStructOrEnum {refPath=&Vector2f}}}
      Cylinder Distribution {
       bIsDynamic=false
       bIsEditable=false
       bIsStaticSwitch=true
       bIsVisible=false
       type {classStructOrEnum {refPath=&ENiagara_CylinderMode}}
      }
      Surface Only {bIsDynamic=false, bIsEditable=false, bIsStaticSwitch=false, bIsVisible=false, type {classStructOrEnum {refPath=&NiagaraBool}}}
      Surface Only Band Thickness {
       bIsDynamic=false
       bIsEditable=false
       bIsStaticSwitch=false
       bIsVisible=false
       type {classStructOrEnum {refPath=&NiagaraFloat}}
      }
      Use Endcaps In Surface Only Mode {bIsDynamic=false, bIsEditable=false, bIsStaticSwitch=true, bIsVisible=false, type {classStructOrEnum {refPath=&NiagaraBool}}}
      Hemicircle X {bIsDynamic=false, bIsEditable=false, bIsStaticSwitch=true, bIsVisible=false, type {classStructOrEnum {refPath=&NiagaraBool}}}
      Hemicircle Y {bIsDynamic=false, bIsEditable=false, bIsStaticSwitch=true, bIsVisible=false, type {classStructOrEnum {refPath=&NiagaraBool}}}
      Disc Coverage {
       bIsDynamic=false
       bIsEditable=false
       bIsStaticSwitch=false
       bIsVisible=false
       type {classStructOrEnum {refPath=&NiagaraFloat}}
      }
      Hemicircle Internal Cap {bIsDynamic=false, bIsEditable=false, bIsStaticSwitch=true, bIsVisible=false, type {classStructOrEnum {refPath=&NiagaraBool}}}
      U Position {
       bIsDynamic=false
       bIsEditable=false
       bIsStaticSwitch=false
       bIsVisible=false
       type {classStructOrEnum {refPath=&NiagaraFloat}}
      }
      Enable Lathe Profile {bIsDynamic=false, bIsEditable=false, bIsStaticSwitch=true, bIsVisible=false, type {classStructOrEnum {refPath=&NiagaraBool}}}
      Lathe Profile {
       bIsDynamic=false
       bIsEditable=false
       bIsStaticSwitch=false
       bIsVisible=false
       type {classStructOrEnum {refPath=&NiagaraDataInterfaceCurve}}
      }
      Surface Distribution {
       bIsDynamic=false
       bIsEditable=false
       bIsStaticSwitch=false
       bIsVisible=false
       type {classStructOrEnum {refPath=&NiagaraFloat}}
      }
      Surface Expansion Mode {
       bIsDynamic=false
       bIsEditable=false
       bIsStaticSwitch=false
       bIsVisible=false
       type {classStructOrEnum {refPath=&ENiagaraExpansionMode}}
      }
      U Distribution {
       bIsDynamic=false
       bIsEditable=false
       bIsStaticSwitch=false
       bIsVisible=false
       type {classStructOrEnum {refPath=&NiagaraFloat}}
      }
      V Position {
       bIsDynamic=false
       bIsEditable=false
       bIsStaticSwitch=false
       bIsVisible=false
       type {classStructOrEnum {refPath=&NiagaraFloat}}
      }
      Radius Position {
       bIsDynamic=false
       bIsEditable=false
       bIsStaticSwitch=false
       bIsVisible=false
       type {classStructOrEnum {refPath=&NiagaraFloat}}
      }
      Sphere Radius Position {
       bIsDynamic=false
       bIsEditable=false
       bIsStaticSwitch=false
       bIsVisible=false
       type {classStructOrEnum {refPath=&NiagaraFloat}}
      }
      V Distribution {
       bIsDynamic=false
       bIsEditable=false
       bIsStaticSwitch=false
       bIsVisible=false
       type {classStructOrEnum {refPath=&NiagaraFloat}}
      }
      Evaluation Type {
       bIsDynamic=false
       bIsEditable=false
       bIsStaticSwitch=true
       bIsVisible=false
       type {classStructOrEnum {refPath=&ENiagaraRandomnessEvaluation}}
      }
      Uniform Distribution {
       bIsDynamic=false
       bIsEditable=false
       bIsStaticSwitch=false
       bIsVisible=false
       type {classStructOrEnum {refPath=&NiagaraFloat}}
      }
      Uniform Spiral Amount {
       bIsDynamic=false
       bIsEditable=false
       bIsStaticSwitch=false
       bIsVisible=false
       type {classStructOrEnum {refPath=&NiagaraFloat}}
      }
      Uniform Spiral Falloff {
       bIsDynamic=false
       bIsEditable=false
       bIsStaticSwitch=false
       bIsVisible=false
       type {classStructOrEnum {refPath=&NiagaraFloat}}
      }
      Uniform Count {
       bIsDynamic=false
       bIsEditable=false
       bIsStaticSwitch=false
       bIsVisible=false
       type {classStructOrEnum {refPath=&NiagaraInt32}}
      }
      Uniform Particle Index {
       bIsDynamic=false
       bIsEditable=false
       bIsStaticSwitch=false
       bIsVisible=false
       type {classStructOrEnum {refPath=&NiagaraInt32}}
      }
      Hemisphere Angle Type {
       bIsDynamic=false
       bIsEditable=true
       bIsStaticSwitch=true
       bIsVisible=true
       type {classStructOrEnum {refPath=&ENiagara_AngleInput}}
      }
      Uniform Endpoint Offset {
       bIsDynamic=false
       bIsEditable=false
       bIsStaticSwitch=false
       bIsVisible=false
       type {classStructOrEnum {refPath=&NiagaraFloat}}
      }
      MaskSpawnGroup {bIsDynamic=false, bIsEditable=true, bIsStaticSwitch=false, bIsVisible=true, type {classStructOrEnum {refPath=&NiagaraBool}}}
      Spawn Group Mask {bIsDynamic=false, bIsEditable=false, bIsStaticSwitch=false, bIsVisible=true, type {classStructOrEnum {refPath=&NiagaraInt32}}}
      Spawn Group {
       bIsDynamic=false
       bIsEditable=false
       bIsStaticSwitch=false
       bIsVisible=false
       type {classStructOrEnum {refPath=&NiagaraInt32}}
      }
      Randomness Mode {
       bIsDynamic=false
       bIsEditable=true
       bIsStaticSwitch=true
       bIsVisible=true
       type {classStructOrEnum {refPath=&ENiagaraRandomnessMode}}
      }
      Override Seed {bIsDynamic=false, bIsEditable=true, bIsStaticSwitch=false, bIsVisible=true, type {classStructOrEnum {refPath=&NiagaraBool}}}
      Random Seed {bIsDynamic=false, bIsEditable=false, bIsStaticSwitch=false, bIsVisible=true, type {classStructOrEnum {refPath=&NiagaraInt32}}}
      Fixed Random Seed {bIsDynamic=false, bIsEditable=false, bIsStaticSwitch=true, bIsVisible=false, type {classStructOrEnum {refPath=&NiagaraBool}}}
      Transform Order {
       bIsDynamic=false
       bIsEditable=true
       bIsStaticSwitch=true
       bIsVisible=true
       type {classStructOrEnum {refPath=&ENiagara_TransformOrder}}
      }
      Scale Mode {
       bIsDynamic=false
       bIsEditable=true
       bIsStaticSwitch=true
       bIsVisible=true
       type {classStructOrEnum {refPath=&ENiagara_ScaleMode}}
      }
      Non Uniform Scale {bIsDynamic=false, bIsEditable=true, bIsStaticSwitch=false, bIsVisible=true, type {classStructOrEnum {refPath=&Vector3f}}}
      UseOwnerScale {bIsDynamic=false, bIsEditable=true, bIsStaticSwitch=false, bIsVisible=true, type {classStructOrEnum {refPath=&NiagaraBool}}}
      Apply Owner Scale {bIsDynamic=false, bIsEditable=true, bIsStaticSwitch=false, bIsVisible=true, type {classStructOrEnum {refPath=&Vector3f}}}
      Rotation Mode {
       bIsDynamic=false
       bIsEditable=true
       bIsStaticSwitch=true
       bIsVisible=true
       type {classStructOrEnum {refPath=&ENiagara_RotationMode}}
      }
      Rotation Axis {
       bIsDynamic=false
       bIsEditable=false
       bIsStaticSwitch=false
       bIsVisible=false
       type {classStructOrEnum {refPath=&Vector3f}}
      }
      Rotation Matrix {
       bIsDynamic=false
       bIsEditable=false
       bIsStaticSwitch=false
       bIsVisible=false
       type {classStructOrEnum {refPath=&NiagaraMatrix}}
      }
      Rotation Quaternion {bIsDynamic=false, bIsEditable=false, bIsStaticSwitch=false, bIsVisible=false, type {classStructOrEnum {refPath=&Quat4f}}}
      Yaw / Pitch / Roll {
       bIsDynamic=false
       bIsEditable=false
       bIsStaticSwitch=false
       bIsVisible=false
       type {classStructOrEnum {refPath=&Vector3f}}
      }
      Rotation Angle {
       bIsDynamic=false
       bIsEditable=false
       bIsStaticSwitch=false
       bIsVisible=false
       type {classStructOrEnum {refPath=&NiagaraFloat}}
      }
      Invert Rotation Quaternion {bIsDynamic=false, bIsEditable=false, bIsStaticSwitch=false, bIsVisible=false, type {classStructOrEnum {refPath=&NiagaraBool}}}
      Additional Quaternion Rotation {bIsDynamic=false, bIsEditable=false, bIsStaticSwitch=false, bIsVisible=false, type {classStructOrEnum {refPath=&NiagaraBool}}}
      Additional Yaw / Pitch / Roll {
       bIsDynamic=false
       bIsEditable=false
       bIsStaticSwitch=false
       bIsVisible=false
       type {classStructOrEnum {refPath=&Vector3f}}
      }
      Rotation Angle Type {
       bIsDynamic=false
       bIsEditable=false
       bIsStaticSwitch=true
       bIsVisible=false
       type {classStructOrEnum {refPath=&ENiagara_AngleInput}}
      }
      Rotation Coordinate Space {
       bIsDynamic=false
       bIsEditable=true
       bIsStaticSwitch=false
       bIsVisible=true
       type {classStructOrEnum {refPath=&ENiagaraCoordinateSpace}}
      }
      Offset Mode {
       bIsDynamic=false
       bIsEditable=true
       bIsStaticSwitch=true
       bIsVisible=true
       type {classStructOrEnum {refPath=&ENiagara_OffsetMode}}
      }
      Shape Origin {
       bIsDynamic=false
       bIsEditable=true
       bIsStaticSwitch=false
       bIsVisible=true
       type {classStructOrEnum {refPath=&NiagaraPosition}}
      }
      Offset {bIsDynamic=true, bIsEditable=true, bIsStaticSwitch=false, bIsVisible=true, type {classStructOrEnum {refPath=&Vector3f}}}
      Offset Coordinate Space {
       bIsDynamic=false
       bIsEditable=true
       bIsStaticSwitch=false
       bIsVisible=true
       type {classStructOrEnum {refPath=&ENiagaraCoordinateSpace}}
      }
      Custom Transform Matrix {
       bIsDynamic=false
       bIsEditable=false
       bIsStaticSwitch=false
       bIsVisible=false
       type {classStructOrEnum {refPath=&NiagaraMatrix}}
      }
      Transform Method {
       bIsDynamic=false
       bIsEditable=true
       bIsStaticSwitch=true
       bIsVisible=true
       type {classStructOrEnum {refPath=&ENiagara_TransformType}}
      }
     ]
     moduleScript {refPath=&ShapeLocation}
    }
    {
     bIsSetParametersModule=false
     enabled=true
     moduleName=CameraOffset
     inputs [Camera Offset Amount {bIsDynamic=true, bIsEditable=true, bIsStaticSwitch=false, bIsVisible=true, type {classStructOrEnum {refPath=&NiagaraFloat}}}]
     moduleScript {refPath=&CameraOffset}
    }
    {
     bIsSetParametersModule=false
     enabled=true
     moduleName=AddVelocity001
     inputs [
      Velocity Mode {
       bIsDynamic=false
       bIsEditable=true
       bIsStaticSwitch=true
       bIsVisible=true
       type {classStructOrEnum {refPath=&ENiagara_VelocityMode}}
      }
      Velocity {bIsDynamic=true, bIsEditable=false, bIsStaticSwitch=false, bIsVisible=false, type {classStructOrEnum {refPath=&Vector3f}}}
      Velocity Speed {bIsDynamic=true, bIsEditable=true, bIsStaticSwitch=false, bIsVisible=true, type {classStructOrEnum {refPath=&NiagaraFloat}}}
      Velocity Speed Scale {
       bIsDynamic=false
       bIsEditable=false
       bIsStaticSwitch=false
       bIsVisible=false
       type {classStructOrEnum {refPath=&NiagaraFloat}}
      }
      Module.Use Falloff Distance {bIsDynamic=false, bIsEditable=true, bIsStaticSwitch=false, bIsVisible=true, type {classStructOrEnum {refPath=&NiagaraBool}}}
      Constrain To Radius {bIsDynamic=false, bIsEditable=false, bIsStaticSwitch=false, bIsVisible=true, type {classStructOrEnum {refPath=&NiagaraFloat}}}
      Distribution Along Cone Axis {
       bIsDynamic=false
       bIsEditable=false
       bIsStaticSwitch=false
       bIsVisible=false
       type {classStructOrEnum {refPath=&NiagaraFloat}}
      }
      Radius Falloff Near / Far {
       bIsDynamic=false
       bIsEditable=false
       bIsStaticSwitch=false
       bIsVisible=false
       type {classStructOrEnum {refPath=&Vector2f}}
      }
      Radius Falloff Exponent {
       bIsDynamic=false
       bIsEditable=false
       bIsStaticSwitch=false
       bIsVisible=false
       type {classStructOrEnum {refPath=&NiagaraFloat}}
      }
      Speed Falloff From Cone Axis {
       bIsDynamic=false
       bIsEditable=false
       bIsStaticSwitch=false
       bIsVisible=false
       type {classStructOrEnum {refPath=&NiagaraFloat}}
      }
      Invert Velocity Falloff {bIsDynamic=false, bIsEditable=false, bIsStaticSwitch=false, bIsVisible=false, type {classStructOrEnum {refPath=&NiagaraBool}}}
      Override Seed {bIsDynamic=false, bIsEditable=true, bIsStaticSwitch=false, bIsVisible=true, type {classStructOrEnum {refPath=&NiagaraBool}}}
      Random Seed {bIsDynamic=false, bIsEditable=false, bIsStaticSwitch=false, bIsVisible=true, type {classStructOrEnum {refPath=&NiagaraInt32}}}
      Fixed Random Seed {bIsDynamic=false, bIsEditable=false, bIsStaticSwitch=true, bIsVisible=false, type {classStructOrEnum {refPath=&NiagaraBool}}}
      Randomness Mode {
       bIsDynamic=false
       bIsEditable=true
       bIsStaticSwitch=true
       bIsVisible=true
       type {classStructOrEnum {refPath=&ENiagaraRandomnessMode}}
      }
      Cone Axis {
       bIsDynamic=false
       bIsEditable=false
       bIsStaticSwitch=false
       bIsVisible=false
       type {classStructOrEnum {refPath=&Vector3f}}
      }
      Cone Angle {
       bIsDynamic=false
       bIsEditable=false
       bIsStaticSwitch=false
       bIsVisible=false
       type {classStructOrEnum {refPath=&NiagaraFloat}}
      }
      Inner Cone Angle {
       bIsDynamic=false
       bIsEditable=false
       bIsStaticSwitch=false
       bIsVisible=false
       type {classStructOrEnum {refPath=&NiagaraFloat}}
      }
      Cone Angle Mode {
       bIsDynamic=false
       bIsEditable=false
       bIsStaticSwitch=true
       bIsVisible=false
       type {classStructOrEnum {refPath=&ENiagara_AngleInput}}
      }
      Rotation Mode {
       bIsDynamic=false
       bIsEditable=false
       bIsStaticSwitch=true
       bIsVisible=false
       type {classStructOrEnum {refPath=&ENiagara_RotationMode}}
      }
      Rotation Axis {
       bIsDynamic=false
       bIsEditable=false
       bIsStaticSwitch=false
       bIsVisible=false
       type {classStructOrEnum {refPath=&Vector3f}}
      }
      Yaw / Pitch / Roll {
       bIsDynamic=false
       bIsEditable=false
       bIsStaticSwitch=false
       bIsVisible=false
       type {classStructOrEnum {refPath=&Vector3f}}
      }
      Rotation Quaternion {bIsDynamic=false, bIsEditable=false, bIsStaticSwitch=false, bIsVisible=false, type {classStructOrEnum {refPath=&Quat4f}}}
      Rotation Matrix {
       bIsDynamic=false
       bIsEditable=false
       bIsStaticSwitch=false
       bIsVisible=false
       type {classStructOrEnum {refPath=&NiagaraMatrix}}
      }
      Rotation Angle {
       bIsDynamic=false
       bIsEditable=false
       bIsStaticSwitch=false
       bIsVisible=false
       type {classStructOrEnum {refPath=&NiagaraFloat}}
      }
      Invert Rotation Quaternion {bIsDynamic=false, bIsEditable=false, bIsStaticSwitch=false, bIsVisible=false, type {classStructOrEnum {refPath=&NiagaraBool}}}
      Additional Quaternion Rotation {bIsDynamic=false, bIsEditable=false, bIsStaticSwitch=false, bIsVisible=false, type {classStructOrEnum {refPath=&NiagaraBool}}}
      Additional Yaw / Pitch / Roll {
       bIsDynamic=false
       bIsEditable=false
       bIsStaticSwitch=false
       bIsVisible=false
       type {classStructOrEnum {refPath=&Vector3f}}
      }
      Rotation Angle Type {
       bIsDynamic=false
       bIsEditable=false
       bIsStaticSwitch=true
       bIsVisible=false
       type {classStructOrEnum {refPath=&ENiagara_AngleInput}}
      }
      Rotation Coordinate Space {
       bIsDynamic=false
       bIsEditable=false
       bIsStaticSwitch=false
       bIsVisible=false
       type {classStructOrEnum {refPath=&ENiagaraCoordinateSpace}}
      }
      Velocity Origin {
       bIsDynamic=false
       bIsEditable=true
       bIsStaticSwitch=false
       bIsVisible=true
       type {classStructOrEnum {refPath=&NiagaraPosition}}
      }
      Origin Offset {bIsDynamic=true, bIsEditable=true, bIsStaticSwitch=false, bIsVisible=true, type {classStructOrEnum {refPath=&Vector3f}}}
      Origin Offset Coordinate Space {
       bIsDynamic=false
       bIsEditable=true
       bIsStaticSwitch=false
       bIsVisible=true
       type {classStructOrEnum {refPath=&ENiagaraCoordinateSpace}}
      }
      Default Position {
       bIsDynamic=false
       bIsEditable=true
       bIsStaticSwitch=false
       bIsVisible=true
       type {classStructOrEnum {refPath=&NiagaraPosition}}
      }
     ]
     moduleScript {refPath=&AddVelocity}
    }
    {
     bIsSetParametersModule=false
     enabled=true
     moduleName=AddVelocity
     inputs [
      Velocity Mode {
       bIsDynamic=false
       bIsEditable=true
       bIsStaticSwitch=true
       bIsVisible=true
       type {classStructOrEnum {refPath=&ENiagara_VelocityMode}}
      }
      Velocity {bIsDynamic=true, bIsEditable=true, bIsStaticSwitch=false, bIsVisible=true, type {classStructOrEnum {refPath=&Vector3f}}}
      Velocity Speed {
       bIsDynamic=false
       bIsEditable=false
       bIsStaticSwitch=false
       bIsVisible=false
       type {classStructOrEnum {refPath=&NiagaraFloat}}
      }
      Velocity Speed Scale {bIsDynamic=false, bIsEditable=true, bIsStaticSwitch=false, bIsVisible=true, type {classStructOrEnum {refPath=&NiagaraFloat}}}
      Constrain To Radius {
       bIsDynamic=false
       bIsEditable=false
       bIsStaticSwitch=false
       bIsVisible=false
       type {classStructOrEnum {refPath=&NiagaraFloat}}
      }
      Distribution Along Cone Axis {
       bIsDynamic=false
       bIsEditable=false
       bIsStaticSwitch=false
       bIsVisible=false
       type {classStructOrEnum {refPath=&NiagaraFloat}}
      }
      Radius Falloff Near / Far {
       bIsDynamic=false
       bIsEditable=false
       bIsStaticSwitch=false
       bIsVisible=false
       type {classStructOrEnum {refPath=&Vector2f}}
      }
      Radius Falloff Exponent {
       bIsDynamic=false
       bIsEditable=false
       bIsStaticSwitch=false
       bIsVisible=false
       type {classStructOrEnum {refPath=&NiagaraFloat}}
      }
      Speed Falloff From Cone Axis {
       bIsDynamic=false
       bIsEditable=false
       bIsStaticSwitch=false
       bIsVisible=false
       type {classStructOrEnum {refPath=&NiagaraFloat}}
      }
      Invert Velocity Falloff {bIsDynamic=false, bIsEditable=false, bIsStaticSwitch=false, bIsVisible=false, type {classStructOrEnum {refPath=&NiagaraBool}}}
      Override Seed {bIsDynamic=false, bIsEditable=false, bIsStaticSwitch=false, bIsVisible=false, type {classStructOrEnum {refPath=&NiagaraBool}}}
      Random Seed {
       bIsDynamic=false
       bIsEditable=false
       bIsStaticSwitch=false
       bIsVisible=false
       type {classStructOrEnum {refPath=&NiagaraInt32}}
      }
      Fixed Random Seed {bIsDynamic=false, bIsEditable=false, bIsStaticSwitch=true, bIsVisible=false, type {classStructOrEnum {refPath=&NiagaraBool}}}
      Randomness Mode {
       bIsDynamic=false
       bIsEditable=false
       bIsStaticSwitch=true
       bIsVisible=false
       type {classStructOrEnum {refPath=&ENiagaraRandomnessMode}}
      }
      Cone Axis {
       bIsDynamic=false
       bIsEditable=false
       bIsStaticSwitch=false
       bIsVisible=false
       type {classStructOrEnum {refPath=&Vector3f}}
      }
      Cone Angle {
       bIsDynamic=false
       bIsEditable=false
       bIsStaticSwitch=false
       bIsVisible=false
       type {classStructOrEnum {refPath=&NiagaraFloat}}
      }
      Inner Cone Angle {
       bIsDynamic=false
       bIsEditable=false
       bIsStaticSwitch=false
       bIsVisible=false
       type {classStructOrEnum {refPath=&NiagaraFloat}}
      }
      Cone Angle Mode {
       bIsDynamic=false
       bIsEditable=false
       bIsStaticSwitch=true
       bIsVisible=false
       type {classStructOrEnum {refPath=&ENiagara_AngleInput}}
      }
      Rotation Mode {
       bIsDynamic=false
       bIsEditable=true
       bIsStaticSwitch=true
       bIsVisible=true
       type {classStructOrEnum {refPath=&ENiagara_RotationMode}}
      }
      Rotation Axis {
       bIsDynamic=false
       bIsEditable=false
       bIsStaticSwitch=false
       bIsVisible=false
       type {classStructOrEnum {refPath=&Vector3f}}
      }
      Yaw / Pitch / Roll {
       bIsDynamic=false
       bIsEditable=false
       bIsStaticSwitch=false
       bIsVisible=false
       type {classStructOrEnum {refPath=&Vector3f}}
      }
      Rotation Quaternion {bIsDynamic=false, bIsEditable=false, bIsStaticSwitch=false, bIsVisible=false, type {classStructOrEnum {refPath=&Quat4f}}}
      Rotation Matrix {
       bIsDynamic=false
       bIsEditable=false
       bIsStaticSwitch=false
       bIsVisible=false
       type {classStructOrEnum {refPath=&NiagaraMatrix}}
      }
      Rotation Angle {
       bIsDynamic=false
       bIsEditable=false
       bIsStaticSwitch=false
       bIsVisible=false
       type {classStructOrEnum {refPath=&NiagaraFloat}}
      }
      Invert Rotation Quaternion {bIsDynamic=false, bIsEditable=false, bIsStaticSwitch=false, bIsVisible=false, type {classStructOrEnum {refPath=&NiagaraBool}}}
      Additional Quaternion Rotation {bIsDynamic=false, bIsEditable=false, bIsStaticSwitch=false, bIsVisible=false, type {classStructOrEnum {refPath=&NiagaraBool}}}
      Additional Yaw / Pitch / Roll {
       bIsDynamic=false
       bIsEditable=false
       bIsStaticSwitch=false
       bIsVisible=false
       type {classStructOrEnum {refPath=&Vector3f}}
      }
      Rotation Angle Type {
       bIsDynamic=false
       bIsEditable=false
       bIsStaticSwitch=true
       bIsVisible=false
       type {classStructOrEnum {refPath=&ENiagara_AngleInput}}
      }
      Rotation Coordinate Space {
       bIsDynamic=false
       bIsEditable=true
       bIsStaticSwitch=false
       bIsVisible=true
       type {classStructOrEnum {refPath=&ENiagaraCoordinateSpace}}
      }
      Velocity Origin {
       bIsDynamic=false
       bIsEditable=false
       bIsStaticSwitch=false
       bIsVisible=false
       type {classStructOrEnum {refPath=&NiagaraPosition}}
      }
      Origin Offset {
       bIsDynamic=false
       bIsEditable=false
       bIsStaticSwitch=false
       bIsVisible=false
       type {classStructOrEnum {refPath=&Vector3f}}
      }
      Origin Offset Coordinate Space {
       bIsDynamic=false
       bIsEditable=false
       bIsStaticSwitch=false
       bIsVisible=false
       type {classStructOrEnum {refPath=&ENiagaraCoordinateSpace}}
      }
      Default Position {
       bIsDynamic=false
       bIsEditable=false
       bIsStaticSwitch=false
       bIsVisible=false
       type {classStructOrEnum {refPath=&NiagaraPosition}}
      }
     ]
     moduleScript {refPath=&AddVelocity}
    }
    {
     bIsSetParametersModule=false
     enabled=true
     moduleName=ApplyInitialForces
     inputs [
      Apply Force to Velocity {bIsDynamic=false, bIsEditable=true, bIsStaticSwitch=false, bIsVisible=true, type {classStructOrEnum {refPath=&NiagaraBool}}}
      Apply Rotational Force to Rotational Velocity {bIsDynamic=false, bIsEditable=true, bIsStaticSwitch=false, bIsVisible=true, type {classStructOrEnum {refPath=&NiagaraBool}}}
      Apply Force to Position {bIsDynamic=false, bIsEditable=true, bIsStaticSwitch=false, bIsVisible=true, type {classStructOrEnum {refPath=&NiagaraBool}}}
      Positional Force Warmup Time {bIsDynamic=false, bIsEditable=false, bIsStaticSwitch=false, bIsVisible=true, type {classStructOrEnum {refPath=&NiagaraFloat}}}
     ]
     moduleScript {refPath=&ApplyInitialForces}
    }
    {
     bIsSetParametersModule=true
     enabled=true
     moduleName=SetVariables_#3
     inputs [Particles.DynamicMaterialParameter_Air {bIsDynamic=true, bIsEditable=true, bIsStaticSwitch=false, bIsVisible=true, type {classStructOrEnum {refPath=&Vector4f}}}]
     moduleScript {refPath=&self:DefaultPropSpawn_Merged005_0.NiagaraScriptSource_1.NiagaraGraph_0.NiagaraNodeAssignment_5.SetVariables_#3}
    }
   ]
  }
  particleUpdateScript {
   scriptName=ParticleUpdateScript
   modules [
    {
     bIsSetParametersModule=false
     enabled=true
     moduleName=ParticleState
     inputs [
      Kill Particles When Lifetime Has Elapsed {bIsDynamic=false, bIsEditable=true, bIsStaticSwitch=true, bIsVisible=true, type {classStructOrEnum {refPath=&NiagaraBool}}}
      Loop Particles Lifetime {bIsDynamic=false, bIsEditable=false, bIsStaticSwitch=false, bIsVisible=false, type {classStructOrEnum {refPath=&NiagaraBool}}}
      Let Infinitely Lived Particles Die When Emitter Deactivates {bIsDynamic=false, bIsEditable=false, bIsStaticSwitch=false, bIsVisible=false, type {classStructOrEnum {refPath=&NiagaraBool}}}
      Lifetime {bIsDynamic=false, bIsEditable=true, bIsStaticSwitch=false, bIsVisible=true, type {classStructOrEnum {refPath=&NiagaraFloat}}}
      DeltaTime {bIsDynamic=false, bIsEditable=true, bIsStaticSwitch=false, bIsVisible=true, type {classStructOrEnum {refPath=&NiagaraFloat}}}
     ]
     moduleScript {refPath=&ParticleState}
    }
    {
     bIsSetParametersModule=false
     enabled=true
     moduleName=SpriteRotationRate
     inputs [
      Rotation Rate {bIsDynamic=true, bIsEditable=true, bIsStaticSwitch=false, bIsVisible=true, type {classStructOrEnum {refPath=&NiagaraFloat}}}
      Delta Time {bIsDynamic=false, bIsEditable=true, bIsStaticSwitch=false, bIsVisible=true, type {classStructOrEnum {refPath=&NiagaraFloat}}}
     ]
     moduleScript {refPath=&SpriteRotationRate}
    }
    {
     bIsSetParametersModule=false
     enabled=true
     moduleName=GravityForce
     inputs [
      Gravity {bIsDynamic=true, bIsEditable=true, bIsStaticSwitch=false, bIsVisible=true, type {classStructOrEnum {refPath=&Vector3f}}}
      Coordinate Space {
       bIsDynamic=false
       bIsEditable=true
       bIsStaticSwitch=true
       bIsVisible=true
       type {classStructOrEnum {refPath=&ENiagaraCoordinateSpace}}
      }
     ]
     moduleScript {refPath=&GravityForce}
    }
    {
     bIsSetParametersModule=false
     enabled=true
     moduleName=CurlNoiseForce
     inputs [
      Noise Strength {bIsDynamic=true, bIsEditable=true, bIsStaticSwitch=false, bIsVisible=true, type {classStructOrEnum {refPath=&NiagaraFloat}}}
      Noise Frequency {bIsDynamic=true, bIsEditable=true, bIsStaticSwitch=false, bIsVisible=true, type {classStructOrEnum {refPath=&NiagaraFloat}}}
      Noise Quality / Cost {
       bIsDynamic=false
       bIsEditable=true
       bIsStaticSwitch=true
       bIsVisible=true
       type {classStructOrEnum {refPath=&ENiagaraCurlNoiseQuality}}
      }
      Random Seed {bIsDynamic=false, bIsEditable=true, bIsStaticSwitch=false, bIsVisible=true, type {classStructOrEnum {refPath=&NiagaraInt32}}}
      Pan Noise {bIsDynamic=false, bIsEditable=true, bIsStaticSwitch=false, bIsVisible=true, type {classStructOrEnum {refPath=&NiagaraBool}}}
      Pan Noise Field {bIsDynamic=false, bIsEditable=false, bIsStaticSwitch=false, bIsVisible=true, type {classStructOrEnum {refPath=&Vector3f}}}
      Write to Intrinsic Parameters {bIsDynamic=false, bIsEditable=true, bIsStaticSwitch=true, bIsVisible=true, type {classStructOrEnum {refPath=&NiagaraBool}}}
      Sample Position {bIsDynamic=false, bIsEditable=true, bIsStaticSwitch=false, bIsVisible=true, type {classStructOrEnum {refPath=&Vector3f}}}
      CurlNoise {
       bIsDynamic=false
       bIsEditable=false
       bIsStaticSwitch=false
       bIsVisible=false
       type {classStructOrEnum {refPath=&NiagaraDataInterfaceCurlNoise}}
      }
      Mask Curl Noise {bIsDynamic=false, bIsEditable=true, bIsStaticSwitch=true, bIsVisible=true, type {classStructOrEnum {refPath=&NiagaraBool}}}
      Curl Noise Cone Mask Angle {
       bIsDynamic=false
       bIsEditable=false
       bIsStaticSwitch=false
       bIsVisible=false
       type {classStructOrEnum {refPath=&NiagaraFloat}}
      }
      Curl Noise Cone Mask Falloff Angle {
       bIsDynamic=false
       bIsEditable=false
       bIsStaticSwitch=false
       bIsVisible=false
       type {classStructOrEnum {refPath=&NiagaraFloat}}
      }
      Curl Noise Cone Mask Axis {
       bIsDynamic=false
       bIsEditable=false
       bIsStaticSwitch=false
       bIsVisible=false
       type {classStructOrEnum {refPath=&Vector3f}}
      }
     ]
     moduleScript {refPath=&CurlNoiseForce}
    }
    {
     bIsSetParametersModule=false
     enabled=true
     moduleName=Drag
     inputs [
      Use Linear Drag {bIsDynamic=false, bIsEditable=true, bIsStaticSwitch=false, bIsVisible=true, type {classStructOrEnum {refPath=&NiagaraBool}}}
      Drag {bIsDynamic=true, bIsEditable=true, bIsStaticSwitch=false, bIsVisible=true, type {classStructOrEnum {refPath=&NiagaraFloat}}}
      Use Rotational Drag {bIsDynamic=false, bIsEditable=true, bIsStaticSwitch=false, bIsVisible=true, type {classStructOrEnum {refPath=&NiagaraBool}}}
      Rotational Drag {bIsDynamic=false, bIsEditable=true, bIsStaticSwitch=false, bIsVisible=true, type {classStructOrEnum {refPath=&NiagaraFloat}}}
      Ignore Mass {bIsDynamic=false, bIsEditable=true, bIsStaticSwitch=false, bIsVisible=true, type {classStructOrEnum {refPath=&NiagaraBool}}}
     ]
     moduleScript {refPath=&Drag}
    }
    {
     bIsSetParametersModule=false
     enabled=true
     moduleName=SolveForcesAndVelocity
     inputs [
      Clamp Velocity {bIsDynamic=false, bIsEditable=true, bIsStaticSwitch=false, bIsVisible=true, type {classStructOrEnum {refPath=&NiagaraBool}}}
      Speed Limit {bIsDynamic=false, bIsEditable=false, bIsStaticSwitch=false, bIsVisible=true, type {classStructOrEnum {refPath=&NiagaraFloat}}}
      Limit Acceleration {bIsDynamic=false, bIsEditable=true, bIsStaticSwitch=false, bIsVisible=true, type {classStructOrEnum {refPath=&NiagaraBool}}}
      Acceleration Limit {bIsDynamic=false, bIsEditable=false, bIsStaticSwitch=false, bIsVisible=true, type {classStructOrEnum {refPath=&NiagaraFloat}}}
      Manually Enable Rotational Solver {bIsDynamic=false, bIsEditable=true, bIsStaticSwitch=false, bIsVisible=true, type {classStructOrEnum {refPath=&NiagaraBool}}}
      Write to Intrinsic Properties {bIsDynamic=false, bIsEditable=true, bIsStaticSwitch=false, bIsVisible=true, type {classStructOrEnum {refPath=&NiagaraBool}}}
      Write to Presolve Properties {bIsDynamic=false, bIsEditable=true, bIsStaticSwitch=true, bIsVisible=true, type {classStructOrEnum {refPath=&NiagaraBool}}}
      DeltaTime {bIsDynamic=false, bIsEditable=true, bIsStaticSwitch=false, bIsVisible=true, type {classStructOrEnum {refPath=&NiagaraFloat}}}
      Mass {bIsDynamic=false, bIsEditable=true, bIsStaticSwitch=false, bIsVisible=true, type {classStructOrEnum {refPath=&NiagaraFloat}}}
      Position {
       bIsDynamic=false
       bIsEditable=true
       bIsStaticSwitch=false
       bIsVisible=true
       type {classStructOrEnum {refPath=&NiagaraPosition}}
      }
      Velocity {bIsDynamic=false, bIsEditable=true, bIsStaticSwitch=false, bIsVisible=true, type {classStructOrEnum {refPath=&Vector3f}}}
      PreviousVelocity {bIsDynamic=false, bIsEditable=true, bIsStaticSwitch=false, bIsVisible=true, type {classStructOrEnum {refPath=&Vector3f}}}
      Force {bIsDynamic=false, bIsEditable=true, bIsStaticSwitch=false, bIsVisible=true, type {classStructOrEnum {refPath=&Vector3f}}}
      Mesh Orientation {bIsDynamic=false, bIsEditable=false, bIsStaticSwitch=false, bIsVisible=false, type {classStructOrEnum {refPath=&Quat4f}}}
      Rotational Velocity {
       bIsDynamic=false
       bIsEditable=false
       bIsStaticSwitch=false
       bIsVisible=false
       type {classStructOrEnum {refPath=&Vector3f}}
      }
      Rotational Force {
       bIsDynamic=false
       bIsEditable=false
       bIsStaticSwitch=false
       bIsVisible=false
       type {classStructOrEnum {refPath=&Vector3f}}
      }
      Rotational Inertia {
       bIsDynamic=false
       bIsEditable=false
       bIsStaticSwitch=false
       bIsVisible=false
       type {classStructOrEnum {refPath=&NiagaraFloat}}
      }
     ]
     moduleScript {refPath=&SolveForcesAndVelocity}
    }
    {
     bIsSetParametersModule=false
     enabled=true
     moduleName=ScaleSpriteSize
     inputs [
      Scale Sprite Size Mode {
       bIsDynamic=false
       bIsEditable=true
       bIsStaticSwitch=true
       bIsVisible=true
       type {classStructOrEnum {refPath=&ENiagara_ScaleSpriteSize}}
      }
      Uniform Curve Sprite Scale {
       bIsDynamic=false
       bIsEditable=false
       bIsStaticSwitch=false
       bIsVisible=false
       type {classStructOrEnum {refPath=&NiagaraDataInterfaceCurve}}
      }
      Non-Uniform Curve Sprite Scale {
       bIsDynamic=false
       bIsEditable=false
       bIsStaticSwitch=false
       bIsVisible=false
       type {classStructOrEnum {refPath=&NiagaraDataInterfaceVector2DCurve}}
      }
      Uniform Curve Index {
       bIsDynamic=false
       bIsEditable=false
       bIsStaticSwitch=false
       bIsVisible=false
       type {classStructOrEnum {refPath=&NiagaraFloat}}
      }
      Non-Uniform Curve Index {
       bIsDynamic=false
       bIsEditable=false
       bIsStaticSwitch=false
       bIsVisible=false
       type {classStructOrEnum {refPath=&NiagaraFloat}}
      }
      Non-Uniform Curve Scale {
       bIsDynamic=false
       bIsEditable=false
       bIsStaticSwitch=false
       bIsVisible=false
       type {classStructOrEnum {refPath=&Vector2f}}
      }
      Uniform Curve Scale {
       bIsDynamic=false
       bIsEditable=false
       bIsStaticSwitch=false
       bIsVisible=false
       type {classStructOrEnum {refPath=&NiagaraFloat}}
      }
      Scale Factor {bIsDynamic=true, bIsEditable=true, bIsStaticSwitch=false, bIsVisible=true, type {classStructOrEnum {refPath=&Vector2f}}}
      Uniform Scale Factor {
       bIsDynamic=false
       bIsEditable=false
       bIsStaticSwitch=false
       bIsVisible=false
       type {classStructOrEnum {refPath=&NiagaraFloat}}
      }
      Initial Sprite Size {bIsDynamic=false, bIsEditable=true, bIsStaticSwitch=false, bIsVisible=true, type {classStructOrEnum {refPath=&Vector2f}}}
     ]
     moduleScript {refPath=&ScaleSpriteSize}
    }
    {
     bIsSetParametersModule=true
     enabled=true
     moduleName=SetVariables_#4
     inputs [
      Particles.DynParSmoke0 {bIsDynamic=true, bIsEditable=true, bIsStaticSwitch=false, bIsVisible=true, type {classStructOrEnum {refPath=&Vector4f}}}
      Particles.DynParSmoke1 {bIsDynamic=true, bIsEditable=true, bIsStaticSwitch=false, bIsVisible=true, type {classStructOrEnum {refPath=&Vector4f}}}
      Particles.DynParSmoke2 {bIsDynamic=true, bIsEditable=true, bIsStaticSwitch=false, bIsVisible=true, type {classStructOrEnum {refPath=&Vector4f}}}
     ]
     moduleScript {refPath=&self:DefaultPropSpawn_Merged005_0.NiagaraScriptSource_1.NiagaraGraph_0.NiagaraNodeAssignment_2.SetVariables_#4}
    }
    {
     bIsSetParametersModule=false
     enabled=true
     moduleName=ScaleColor001
     inputs [
      Scale Mode {bIsDynamic=false, bIsEditable=true, bIsStaticSwitch=true, bIsVisible=true, type {classStructOrEnum {refPath=&NiagaraInt32}}}
      Linear Color Curve {
       bIsDynamic=false
       bIsEditable=false
       bIsStaticSwitch=false
       bIsVisible=false
       type {classStructOrEnum {refPath=&NiagaraDataInterfaceColorCurve}}
      }
      ScaleRGB {bIsDynamic=false, bIsEditable=true, bIsStaticSwitch=false, bIsVisible=true, type {classStructOrEnum {refPath=&NiagaraBool}}}
      Scale RGB {bIsDynamic=false, bIsEditable=true, bIsStaticSwitch=false, bIsVisible=true, type {classStructOrEnum {refPath=&Vector3f}}}
      ScaleRGBA {bIsDynamic=false, bIsEditable=false, bIsStaticSwitch=false, bIsVisible=false, type {classStructOrEnum {refPath=&NiagaraBool}}}
      Scale RGBA {
       bIsDynamic=false
       bIsEditable=false
       bIsStaticSwitch=false
       bIsVisible=false
       type {classStructOrEnum {refPath=&Vector4f}}
      }
      Curve Index {
       bIsDynamic=false
       bIsEditable=false
       bIsStaticSwitch=false
       bIsVisible=false
       type {classStructOrEnum {refPath=&NiagaraFloat}}
      }
      ScaleA {bIsDynamic=false, bIsEditable=true, bIsStaticSwitch=false, bIsVisible=true, type {classStructOrEnum {refPath=&NiagaraBool}}}
      Scale Alpha {bIsDynamic=true, bIsEditable=true, bIsStaticSwitch=false, bIsVisible=true, type {classStructOrEnum {refPath=&NiagaraFloat}}}
      Color Value To Scale {
       bIsDynamic=false
       bIsEditable=true
       bIsStaticSwitch=false
       bIsVisible=true
       type {classStructOrEnum {refPath=&LinearColor}}
      }
      Color Scale {
       bIsDynamic=false
       bIsEditable=false
       bIsStaticSwitch=false
       bIsVisible=false
       type {classStructOrEnum {refPath=&LinearColor}}
      }
     ]
     moduleScript {refPath=&ScaleColor}
    }
   ]
  }
  rendererClasses [{refPath=&NiagaraSpriteRendererProperties}]
  renderers [
   {rendererIndex=0, rendererClass {refPath=&NiagaraSpriteRendererProperties}}
   {rendererIndex=1, rendererClass {refPath=&NiagaraSpriteRendererProperties}}
   {rendererIndex=2, rendererClass {refPath=&NiagaraSpriteRendererProperties}}
   {rendererIndex=3, rendererClass {refPath=&NiagaraSpriteRendererProperties}}
  ]
 }
 Flare {
  bEnabled=false
  emitterSpawnScript {scriptName=EmitterSpawnScript, modules []}
  emitterUpdateScript {
   scriptName=EmitterUpdateScript
   modules [
    {
     bIsSetParametersModule=false
     enabled=true
     moduleName=EmitterState
     inputs [
      Life Cycle Mode {
       bIsDynamic=false
       bIsEditable=true
       bIsStaticSwitch=true
       bIsVisible=true
       type {classStructOrEnum {refPath=&ENiagaraEmitterLifeCycleMode}}
      }
      Inactive Response {
       bIsDynamic=false
       bIsEditable=false
       bIsStaticSwitch=true
       bIsVisible=false
       type {classStructOrEnum {refPath=&ENiagaraInactiveMode}}
      }
      Loop Behavior {
       bIsDynamic=false
       bIsEditable=false
       bIsStaticSwitch=true
       bIsVisible=false
       type {classStructOrEnum {refPath=&ENiagara_EmitterStateOptions}}
      }
      Loop Duration Mode {
       bIsDynamic=false
       bIsEditable=false
       bIsStaticSwitch=true
       bIsVisible=false
       type {classStructOrEnum {refPath=&ENiagara_InfiniteLoopDuration}}
      }
      Loop Count {
       bIsDynamic=false
       bIsEditable=false
       bIsStaticSwitch=false
       bIsVisible=false
       type {classStructOrEnum {refPath=&NiagaraInt32}}
      }
      Loop Duration {
       bIsDynamic=false
       bIsEditable=false
       bIsStaticSwitch=false
       bIsVisible=false
       type {classStructOrEnum {refPath=&NiagaraFloat}}
      }
      Recalculate Duration Each Loop {bIsDynamic=false, bIsEditable=false, bIsStaticSwitch=false, bIsVisible=false, type {classStructOrEnum {refPath=&NiagaraBool}}}
      UseLoopDelay {bIsDynamic=false, bIsEditable=false, bIsStaticSwitch=false, bIsVisible=false, type {classStructOrEnum {refPath=&NiagaraBool}}}
      Loop Delay {
       bIsDynamic=false
       bIsEditable=false
       bIsStaticSwitch=false
       bIsVisible=false
       type {classStructOrEnum {refPath=&NiagaraFloat}}
      }
      Delay First Loop Only {bIsDynamic=false, bIsEditable=false, bIsStaticSwitch=false, bIsVisible=false, type {classStructOrEnum {refPath=&NiagaraBool}}}
      Scalability Mode {
       bIsDynamic=false
       bIsEditable=true
       bIsStaticSwitch=true
       bIsVisible=true
       type {classStructOrEnum {refPath=&ENiagaraEmitterScalabilityMode_Limited}}
      }
      Enable Distance Culling {bIsDynamic=false, bIsEditable=false, bIsStaticSwitch=true, bIsVisible=false, type {classStructOrEnum {refPath=&NiagaraBool}}}
      MinDistance {
       bIsDynamic=false
       bIsEditable=false
       bIsStaticSwitch=false
       bIsVisible=false
       type {classStructOrEnum {refPath=&NiagaraFloat}}
      }
      Min Distance Response {
       bIsDynamic=false
       bIsEditable=false
       bIsStaticSwitch=false
       bIsVisible=false
       type {classStructOrEnum {refPath=&ENiagaraExecutionStateManagement}}
      }
      MaxDistance {
       bIsDynamic=false
       bIsEditable=false
       bIsStaticSwitch=false
       bIsVisible=false
       type {classStructOrEnum {refPath=&NiagaraFloat}}
      }
      Max Distance Response {
       bIsDynamic=false
       bIsEditable=false
       bIsStaticSwitch=false
       bIsVisible=false
       type {classStructOrEnum {refPath=&ENiagaraExecutionStateManagement}}
      }
      Scale Spawn Count {bIsDynamic=false, bIsEditable=false, bIsStaticSwitch=true, bIsVisible=false, type {classStructOrEnum {refPath=&NiagaraBool}}}
      Spawn Count Scale {
       bIsDynamic=false
       bIsEditable=false
       bIsStaticSwitch=false
       bIsVisible=false
       type {classStructOrEnum {refPath=&NiagaraFloat}}
      }
      Spawn Count Scale By Distance {
       bIsDynamic=false
       bIsEditable=false
       bIsStaticSwitch=false
       bIsVisible=false
       type {classStructOrEnum {refPath=&NiagaraDataInterfaceCurve}}
      }
      Enable Visibility Culling {bIsDynamic=false, bIsEditable=false, bIsStaticSwitch=true, bIsVisible=false, type {classStructOrEnum {refPath=&NiagaraBool}}}
      Visibility Cull Response {
       bIsDynamic=false
       bIsEditable=false
       bIsStaticSwitch=false
       bIsVisible=false
       type {classStructOrEnum {refPath=&ENiagaraExecutionStateManagement}}
      }
      Visibility Cull Delay {
       bIsDynamic=false
       bIsEditable=false
       bIsStaticSwitch=false
       bIsVisible=false
       type {classStructOrEnum {refPath=&NiagaraFloat}}
      }
      Reset Age On Awaken {bIsDynamic=false, bIsEditable=false, bIsStaticSwitch=false, bIsVisible=false, type {classStructOrEnum {refPath=&NiagaraBool}}}
     ]
     moduleScript {refPath=&EmitterState}
    }
    {
     bIsSetParametersModule=false
     enabled=true
     moduleName=SpawnBurst_Instantaneous
     inputs [
      Spawn Count {bIsDynamic=false, bIsEditable=true, bIsStaticSwitch=false, bIsVisible=true, type {classStructOrEnum {refPath=&NiagaraInt32}}}
      Spawn Time {bIsDynamic=false, bIsEditable=true, bIsStaticSwitch=false, bIsVisible=true, type {classStructOrEnum {refPath=&NiagaraFloat}}}
      Use Spawn Probability {bIsDynamic=false, bIsEditable=true, bIsStaticSwitch=false, bIsVisible=true, type {classStructOrEnum {refPath=&NiagaraBool}}}
      Spawn Probability {bIsDynamic=false, bIsEditable=false, bIsStaticSwitch=false, bIsVisible=true, type {classStructOrEnum {refPath=&NiagaraFloat}}}
      UseLoopCountLimit {bIsDynamic=false, bIsEditable=true, bIsStaticSwitch=false, bIsVisible=true, type {classStructOrEnum {refPath=&NiagaraBool}}}
      Loop Count Limit {bIsDynamic=false, bIsEditable=false, bIsStaticSwitch=false, bIsVisible=true, type {classStructOrEnum {refPath=&NiagaraInt32}}}
      Spawn Group {bIsDynamic=false, bIsEditable=true, bIsStaticSwitch=false, bIsVisible=true, type {classStructOrEnum {refPath=&NiagaraInt32}}}
      Age {bIsDynamic=false, bIsEditable=true, bIsStaticSwitch=false, bIsVisible=true, type {classStructOrEnum {refPath=&NiagaraFloat}}}
     ]
     moduleScript {refPath=&SpawnBurst_Instantaneous}
    }
   ]
  }
  particleSpawnScript {
   scriptName=ParticleSpawnScript
   modules [
    {
     bIsSetParametersModule=false
     enabled=true
     moduleName=PartitionParticles
     inputs [
      Partition Mode {
       bIsDynamic=false
       bIsEditable=true
       bIsStaticSwitch=true
       bIsVisible=true
       type {classStructOrEnum {refPath=&ENiagara_PartitionMode_v2}}
      }
      Partition Count Mode {
       bIsDynamic=false
       bIsEditable=false
       bIsStaticSwitch=true
       bIsVisible=false
       type {classStructOrEnum {refPath=&ENiagara_InfiniteLoopDuration}}
      }
      A {
       bIsDynamic=false
       bIsEditable=false
       bIsStaticSwitch=false
       bIsVisible=false
       type {classStructOrEnum {refPath=&NiagaraFloat}}
      }
      Number Of Partitions {bIsDynamic=false, bIsEditable=true, bIsStaticSwitch=false, bIsVisible=true, type {classStructOrEnum {refPath=&NiagaraInt32}}}
      Particles Per Partition {
       bIsDynamic=false
       bIsEditable=false
       bIsStaticSwitch=false
       bIsVisible=false
       type {classStructOrEnum {refPath=&NiagaraInt32}}
      }
      Partition Duration {
       bIsDynamic=false
       bIsEditable=false
       bIsStaticSwitch=false
       bIsVisible=false
       type {classStructOrEnum {refPath=&NiagaraFloat}}
      }
      Time {
       bIsDynamic=false
       bIsEditable=false
       bIsStaticSwitch=false
       bIsVisible=false
       type {classStructOrEnum {refPath=&NiagaraFloat}}
      }
      Partitioning Frequency {
       bIsDynamic=false
       bIsEditable=false
       bIsStaticSwitch=true
       bIsVisible=false
       type {classStructOrEnum {refPath=&ENiagaraRandomnessEvaluation}}
      }
      B {
       bIsDynamic=false
       bIsEditable=false
       bIsStaticSwitch=false
       bIsVisible=false
       type {classStructOrEnum {refPath=&NiagaraFloat}}
      }
      Distance {
       bIsDynamic=false
       bIsEditable=false
       bIsStaticSwitch=false
       bIsVisible=false
       type {classStructOrEnum {refPath=&NiagaraFloat}}
      }
      Comparison Type {
       bIsDynamic=false
       bIsEditable=false
       bIsStaticSwitch=true
       bIsVisible=false
       type {classStructOrEnum {refPath=&ENiagaraBooleanLogicOps_v2}}
      }
      Evaluation Type {
       bIsDynamic=false
       bIsEditable=false
       bIsStaticSwitch=true
       bIsVisible=false
       type {classStructOrEnum {refPath=&ENiagaraRandomnessEvaluation}}
      }
      UseDistanceFalloff {bIsDynamic=false, bIsEditable=false, bIsStaticSwitch=false, bIsVisible=false, type {classStructOrEnum {refPath=&NiagaraBool}}}
      Distance Falloff Exponent {
       bIsDynamic=false
       bIsEditable=false
       bIsStaticSwitch=false
       bIsVisible=false
       type {classStructOrEnum {refPath=&NiagaraFloat}}
      }
      Start Position {
       bIsDynamic=false
       bIsEditable=false
       bIsStaticSwitch=false
       bIsVisible=false
       type {classStructOrEnum {refPath=&NiagaraPosition}}
      }
      End Position {
       bIsDynamic=false
       bIsEditable=false
       bIsStaticSwitch=false
       bIsVisible=false
       type {classStructOrEnum {refPath=&NiagaraPosition}}
      }
      ID to Partition {bIsDynamic=false, bIsEditable=true, bIsStaticSwitch=false, bIsVisible=true, type {classStructOrEnum {refPath=&NiagaraInt32}}}
      Randomness Mode {
       bIsDynamic=false
       bIsEditable=false
       bIsStaticSwitch=true
       bIsVisible=false
       type {classStructOrEnum {refPath=&ENiagaraRandomnessMode}}
      }
      Override Seed {bIsDynamic=false, bIsEditable=false, bIsStaticSwitch=true, bIsVisible=false, type {classStructOrEnum {refPath=&NiagaraBool}}}
      Fixed Random Seed {bIsDynamic=false, bIsEditable=false, bIsStaticSwitch=true, bIsVisible=false, type {classStructOrEnum {refPath=&NiagaraBool}}}
      Color A {
       bIsDynamic=false
       bIsEditable=false
       bIsStaticSwitch=false
       bIsVisible=false
       type {classStructOrEnum {refPath=&LinearColor}}
      }
      Color B {
       bIsDynamic=false
       bIsEditable=false
       bIsStaticSwitch=false
       bIsVisible=false
       type {classStructOrEnum {refPath=&LinearColor}}
      }
      Color Comparison Type {bIsDynamic=false, bIsEditable=false, bIsStaticSwitch=true, bIsVisible=false, type {classStructOrEnum {refPath=&NiagaraInt32}}}
      Color Tolerance {
       bIsDynamic=false
       bIsEditable=false
       bIsStaticSwitch=false
       bIsVisible=false
       type {classStructOrEnum {refPath=&NiagaraFloat}}
      }
      Minimum Saturation {
       bIsDynamic=false
       bIsEditable=false
       bIsStaticSwitch=false
       bIsVisible=false
       type {classStructOrEnum {refPath=&NiagaraFloat}}
      }
      Direction A {
       bIsDynamic=false
       bIsEditable=false
       bIsStaticSwitch=false
       bIsVisible=false
       type {classStructOrEnum {refPath=&Vector3f}}
      }
      Direction B {
       bIsDynamic=false
       bIsEditable=false
       bIsStaticSwitch=false
       bIsVisible=false
       type {classStructOrEnum {refPath=&Vector3f}}
      }
      Direction Tolerance {
       bIsDynamic=false
       bIsEditable=false
       bIsStaticSwitch=false
       bIsVisible=false
       type {classStructOrEnum {refPath=&NiagaraFloat}}
      }
      Write Partition to Ribbon ID {
       bIsDynamic=false
       bIsEditable=true
       bIsStaticSwitch=true
       bIsVisible=true
       type {classStructOrEnum {refPath=&ENiagara_AttributeSamplingApplyOutput}}
      }
      Write Partition Index to Ribbon Link Order {
       bIsDynamic=false
       bIsEditable=false
       bIsStaticSwitch=true
       bIsVisible=false
       type {classStructOrEnum {refPath=&ENiagara_AttributeSamplingApplyOutput}}
      }
      Normalized Index In Partition Includes One {bIsDynamic=false, bIsEditable=false, bIsStaticSwitch=true, bIsVisible=false, type {classStructOrEnum {refPath=&NiagaraBool}}}
      Normalized Index Includes One {bIsDynamic=false, bIsEditable=true, bIsStaticSwitch=true, bIsVisible=true, type {classStructOrEnum {refPath=&NiagaraBool}}}
     ]
     moduleScript {refPath=&PartitionParticles}
    }
    {
     bIsSetParametersModule=false
     enabled=true
     moduleName=InitializeParticle
     inputs [
      Lifetime Mode {
       bIsDynamic=false
       bIsEditable=true
       bIsStaticSwitch=true
       bIsVisible=true
       type {classStructOrEnum {refPath=&ENiagara_LifetimeMode}}
      }
      Lifetime Min {
       bIsDynamic=false
       bIsEditable=false
       bIsStaticSwitch=false
       bIsVisible=false
       type {classStructOrEnum {refPath=&NiagaraFloat}}
      }
      Lifetime {bIsDynamic=false, bIsEditable=true, bIsStaticSwitch=false, bIsVisible=true, type {classStructOrEnum {refPath=&NiagaraFloat}}}
      Lifetime Max {
       bIsDynamic=false
       bIsEditable=false
       bIsStaticSwitch=false
       bIsVisible=false
       type {classStructOrEnum {refPath=&NiagaraFloat}}
      }
      Lifetime Randomness Mode {
       bIsDynamic=false
       bIsEditable=false
       bIsStaticSwitch=true
       bIsVisible=false
       type {classStructOrEnum {refPath=&ENiagaraRandomnessMode}}
      }
      Color Mode {
       bIsDynamic=false
       bIsEditable=true
       bIsStaticSwitch=true
       bIsVisible=true
       type {classStructOrEnum {refPath=&ENiagara_ColorInitializationMode}}
      }
      Color Minimum {
       bIsDynamic=false
       bIsEditable=false
       bIsStaticSwitch=false
       bIsVisible=false
       type {classStructOrEnum {refPath=&LinearColor}}
      }
      Color {
       bIsDynamic=false
       bIsEditable=true
       bIsStaticSwitch=false
       bIsVisible=true
       type {classStructOrEnum {refPath=&LinearColor}}
      }
      AdjustHue {bIsDynamic=false, bIsEditable=false, bIsStaticSwitch=false, bIsVisible=false, type {classStructOrEnum {refPath=&NiagaraBool}}}
      Hue Shift Range {
       bIsDynamic=false
       bIsEditable=false
       bIsStaticSwitch=false
       bIsVisible=false
       type {classStructOrEnum {refPath=&Vector2f}}
      }
      Color Maximum {
       bIsDynamic=false
       bIsEditable=false
       bIsStaticSwitch=false
       bIsVisible=false
       type {classStructOrEnum {refPath=&LinearColor}}
      }
      AdjustSaturation {bIsDynamic=false, bIsEditable=false, bIsStaticSwitch=false, bIsVisible=false, type {classStructOrEnum {refPath=&NiagaraBool}}}
      Saturation Range {
       bIsDynamic=false
       bIsEditable=false
       bIsStaticSwitch=false
       bIsVisible=false
       type {classStructOrEnum {refPath=&Vector2f}}
      }
      AdjustValue {bIsDynamic=false, bIsEditable=false, bIsStaticSwitch=false, bIsVisible=false, type {classStructOrEnum {refPath=&NiagaraBool}}}
      Value Range {
       bIsDynamic=false
       bIsEditable=false
       bIsStaticSwitch=false
       bIsVisible=false
       type {classStructOrEnum {refPath=&Vector2f}}
      }
      AdjustAlpha {bIsDynamic=false, bIsEditable=false, bIsStaticSwitch=false, bIsVisible=false, type {classStructOrEnum {refPath=&NiagaraBool}}}
      Alpha Scale Range {
       bIsDynamic=false
       bIsEditable=false
       bIsStaticSwitch=false
       bIsVisible=false
       type {classStructOrEnum {refPath=&Vector2f}}
      }
      Color Channel Mode {
       bIsDynamic=false
       bIsEditable=false
       bIsStaticSwitch=true
       bIsVisible=false
       type {classStructOrEnum {refPath=&ENiagaraChannelCorrelation}}
      }
      Color Randomness Mode {
       bIsDynamic=false
       bIsEditable=false
       bIsStaticSwitch=true
       bIsVisible=false
       type {classStructOrEnum {refPath=&ENiagaraRandomnessMode}}
      }
      Position Mode {
       bIsDynamic=false
       bIsEditable=true
       bIsStaticSwitch=true
       bIsVisible=true
       type {classStructOrEnum {refPath=&ENiagara_PositionInitializationMode}}
      }
      Position {
       bIsDynamic=false
       bIsEditable=false
       bIsStaticSwitch=false
       bIsVisible=false
       type {classStructOrEnum {refPath=&NiagaraPosition}}
      }
      UsePositionOffset {bIsDynamic=false, bIsEditable=true, bIsStaticSwitch=false, bIsVisible=true, type {classStructOrEnum {refPath=&NiagaraBool}}}
      Position Offset {bIsDynamic=true, bIsEditable=true, bIsStaticSwitch=false, bIsVisible=true, type {classStructOrEnum {refPath=&Vector3f}}}
      Position Offset Coordinate Space {
       bIsDynamic=false
       bIsEditable=true
       bIsStaticSwitch=false
       bIsVisible=true
       type {classStructOrEnum {refPath=&ENiagaraCoordinateSpace}}
      }
      Mass Mode {
       bIsDynamic=false
       bIsEditable=true
       bIsStaticSwitch=true
       bIsVisible=true
       type {classStructOrEnum {refPath=&ENiagara_MassInitializationMode}}
      }
      Mass Min {
       bIsDynamic=false
       bIsEditable=false
       bIsStaticSwitch=false
       bIsVisible=false
       type {classStructOrEnum {refPath=&NiagaraFloat}}
      }
      Mass {
       bIsDynamic=false
       bIsEditable=false
       bIsStaticSwitch=false
       bIsVisible=false
       type {classStructOrEnum {refPath=&NiagaraFloat}}
      }
      Mass Max {
       bIsDynamic=false
       bIsEditable=false
       bIsStaticSwitch=false
       bIsVisible=false
       type {classStructOrEnum {refPath=&NiagaraFloat}}
      }
      Mass Randomness Mode {
       bIsDynamic=false
       bIsEditable=false
       bIsStaticSwitch=true
       bIsVisible=false
       type {classStructOrEnum {refPath=&ENiagaraRandomnessMode}}
      }
      Write Material Random {bIsDynamic=false, bIsEditable=true, bIsStaticSwitch=false, bIsVisible=true, type {classStructOrEnum {refPath=&NiagaraBool}}}
      Material Random {bIsDynamic=false, bIsEditable=true, bIsStaticSwitch=false, bIsVisible=true, type {classStructOrEnum {refPath=&NiagaraFloat}}}
      Sprite Size Mode {
       bIsDynamic=false
       bIsEditable=true
       bIsStaticSwitch=true
       bIsVisible=true
       type {classStructOrEnum {refPath=&ENiagara_SizeScaleMode}}
      }
      Sprite Size {
       bIsDynamic=false
       bIsEditable=false
       bIsStaticSwitch=false
       bIsVisible=false
       type {classStructOrEnum {refPath=&Vector2f}}
      }
      Uniform Sprite Size {bIsDynamic=false, bIsEditable=true, bIsStaticSwitch=false, bIsVisible=true, type {classStructOrEnum {refPath=&NiagaraFloat}}}
      Uniform Sprite Size Min {
       bIsDynamic=false
       bIsEditable=false
       bIsStaticSwitch=false
       bIsVisible=false
       type {classStructOrEnum {refPath=&NiagaraFloat}}
      }
      Sprite Size Min {
       bIsDynamic=false
       bIsEditable=false
       bIsStaticSwitch=false
       bIsVisible=false
       type {classStructOrEnum {refPath=&Vector2f}}
      }
      Sprite Size Max {
       bIsDynamic=false
       bIsEditable=false
       bIsStaticSwitch=false
       bIsVisible=false
       type {classStructOrEnum {refPath=&Vector2f}}
      }
      Uniform Sprite Size Max {
       bIsDynamic=false
       bIsEditable=false
       bIsStaticSwitch=false
       bIsVisible=false
       type {classStructOrEnum {refPath=&NiagaraFloat}}
      }
      Sprite Size Randomness Mode {
       bIsDynamic=false
       bIsEditable=false
       bIsStaticSwitch=true
       bIsVisible=false
       type {classStructOrEnum {refPath=&ENiagaraRandomnessMode}}
      }
      Sprite Rotation Mode {
       bIsDynamic=false
       bIsEditable=true
       bIsStaticSwitch=true
       bIsVisible=true
       type {classStructOrEnum {refPath=&ENiagara_SpriteRotationMode}}
      }
      Sprite Rotation Angle Min {bIsDynamic=false, bIsEditable=true, bIsStaticSwitch=false, bIsVisible=true, type {classStructOrEnum {refPath=&NiagaraFloat}}}
      Sprite Rotation Angle {
       bIsDynamic=false
       bIsEditable=false
       bIsStaticSwitch=false
       bIsVisible=false
       type {classStructOrEnum {refPath=&NiagaraFloat}}
      }
      Sprite Rotation Angle Max {bIsDynamic=false, bIsEditable=true, bIsStaticSwitch=false, bIsVisible=true, type {classStructOrEnum {refPath=&NiagaraFloat}}}
      Sprite Rotation Randomness Mode {
       bIsDynamic=false
       bIsEditable=true
       bIsStaticSwitch=true
       bIsVisible=true
       type {classStructOrEnum {refPath=&ENiagaraRandomnessMode}}
      }
      Sprite UV Mode {
       bIsDynamic=false
       bIsEditable=true
       bIsStaticSwitch=true
       bIsVisible=true
       type {classStructOrEnum {refPath=&ENiagara_UVFlippingMode}}
      }
      Sprite UV Randomness Mode {
       bIsDynamic=false
       bIsEditable=false
       bIsStaticSwitch=true
       bIsVisible=false
       type {classStructOrEnum {refPath=&ENiagaraRandomnessMode}}
      }
      Sprite UV Scale {
       bIsDynamic=false
       bIsEditable=false
       bIsStaticSwitch=false
       bIsVisible=false
       type {classStructOrEnum {refPath=&Vector2f}}
      }
      Mesh Scale Mode {
       bIsDynamic=false
       bIsEditable=true
       bIsStaticSwitch=true
       bIsVisible=true
       type {classStructOrEnum {refPath=&ENiagara_SizeScaleMode}}
      }
      Mesh Uniform Scale Min {
       bIsDynamic=false
       bIsEditable=false
       bIsStaticSwitch=false
       bIsVisible=false
       type {classStructOrEnum {refPath=&NiagaraFloat}}
      }
      Mesh Uniform Scale {
       bIsDynamic=false
       bIsEditable=false
       bIsStaticSwitch=false
       bIsVisible=false
       type {classStructOrEnum {refPath=&NiagaraFloat}}
      }
      Mesh Scale Min {
       bIsDynamic=false
       bIsEditable=false
       bIsStaticSwitch=false
       bIsVisible=false
       type {classStructOrEnum {refPath=&Vector3f}}
      }
      Mesh Scale {
       bIsDynamic=false
       bIsEditable=false
       bIsStaticSwitch=false
       bIsVisible=false
       type {classStructOrEnum {refPath=&Vector3f}}
      }
      Mesh Scale Max {
       bIsDynamic=false
       bIsEditable=false
       bIsStaticSwitch=false
       bIsVisible=false
       type {classStructOrEnum {refPath=&Vector3f}}
      }
      Mesh Uniform Scale Max {
       bIsDynamic=false
       bIsEditable=false
       bIsStaticSwitch=false
       bIsVisible=false
       type {classStructOrEnum {refPath=&NiagaraFloat}}
      }
      Mesh Scale Randomness Mode {
       bIsDynamic=false
       bIsEditable=false
       bIsStaticSwitch=true
       bIsVisible=false
       type {classStructOrEnum {refPath=&ENiagaraRandomnessMode}}
      }
      Mesh Renderer Array Visibility Mode {
       bIsDynamic=false
       bIsEditable=true
       bIsStaticSwitch=true
       bIsVisible=true
       type {classStructOrEnum {refPath=&ENiagara_UnsetDirectSetRandom}}
      }
      Mesh Renderer Mesh Index {
       bIsDynamic=false
       bIsEditable=false
       bIsStaticSwitch=false
       bIsVisible=false
       type {classStructOrEnum {refPath=&NiagaraInt32}}
      }
      Wrap Index By Mesh Count {bIsDynamic=false, bIsEditable=false, bIsStaticSwitch=true, bIsVisible=false, type {classStructOrEnum {refPath=&NiagaraBool}}}
      Mesh Renderer Info {
       bIsDynamic=false
       bIsEditable=false
       bIsStaticSwitch=false
       bIsVisible=false
       type {classStructOrEnum {refPath=&NiagaraDataInterfaceMeshRendererInfo}}
      }
      Mesh Renderer Random Seed {
       bIsDynamic=false
       bIsEditable=false
       bIsStaticSwitch=false
       bIsVisible=false
       type {classStructOrEnum {refPath=&NiagaraInt32}}
      }
      Ribbon Width Mode {
       bIsDynamic=false
       bIsEditable=true
       bIsStaticSwitch=true
       bIsVisible=true
       type {classStructOrEnum {refPath=&ENiagara_UnsetDirectSet}}
      }
      Ribbon Width {
       bIsDynamic=false
       bIsEditable=false
       bIsStaticSwitch=false
       bIsVisible=false
       type {classStructOrEnum {refPath=&NiagaraFloat}}
      }
      Ribbon Facing Mode {
       bIsDynamic=false
       bIsEditable=true
       bIsStaticSwitch=true
       bIsVisible=true
       type {classStructOrEnum {refPath=&ENiagara_UnsetDirectSet}}
      }
      Ribbon Facing Vector {
       bIsDynamic=false
       bIsEditable=false
       bIsStaticSwitch=false
       bIsVisible=false
       type {classStructOrEnum {refPath=&Vector3f}}
      }
      Ribbon Twist Mode {
       bIsDynamic=false
       bIsEditable=true
       bIsStaticSwitch=true
       bIsVisible=true
       type {classStructOrEnum {refPath=&ENiagara_UnsetDirectSet}}
      }
      "Ribbon Twist " {
       bIsDynamic=false
       bIsEditable=false
       bIsStaticSwitch=false
       bIsVisible=false
       type {classStructOrEnum {refPath=&NiagaraFloat}}
      }
     ]
     moduleScript {refPath=&InitializeParticle}
    }
    {
     bIsSetParametersModule=false
     enabled=true
     moduleName=DynamicMaterialParameters
     inputs [
      Write Parameter Index 0 {
       bIsDynamic=false
       bIsEditable=true
       bIsStaticSwitch=true
       bIsVisible=true
       type {classStructOrEnum {refPath=/Niagara/Enums/Utility/ENiagara_DynamicMaterialParametersInputMode}}
      }
      Index 0 Vector 1 {
       bIsDynamic=false
       bIsEditable=false
       bIsStaticSwitch=false
       bIsVisible=false
       type {classStructOrEnum {refPath=&Vector3f}}
      }
      Module.Index0Float0Write {bIsDynamic=false, bIsEditable=true, bIsStaticSwitch=true, bIsVisible=true, type {classStructOrEnum {refPath=&NiagaraBool}}}
      Index 0 Param 1 {bIsDynamic=true, bIsEditable=true, bIsStaticSwitch=false, bIsVisible=true, type {classStructOrEnum {refPath=&NiagaraFloat}}}
      Module.Index0Float1Write {bIsDynamic=false, bIsEditable=true, bIsStaticSwitch=true, bIsVisible=true, type {classStructOrEnum {refPath=&NiagaraBool}}}
      Index 0 Param 2 {bIsDynamic=false, bIsEditable=true, bIsStaticSwitch=false, bIsVisible=true, type {classStructOrEnum {refPath=&NiagaraFloat}}}
      Module.Index0Float2Write {bIsDynamic=false, bIsEditable=true, bIsStaticSwitch=true, bIsVisible=true, type {classStructOrEnum {refPath=&NiagaraBool}}}
      Index 0 Param 3 {bIsDynamic=false, bIsEditable=true, bIsStaticSwitch=false, bIsVisible=true, type {classStructOrEnum {refPath=&NiagaraFloat}}}
      Module.Index0Float3Write {bIsDynamic=false, bIsEditable=true, bIsStaticSwitch=true, bIsVisible=true, type {classStructOrEnum {refPath=&NiagaraBool}}}
      Index 0 Param 4 {bIsDynamic=false, bIsEditable=true, bIsStaticSwitch=false, bIsVisible=true, type {classStructOrEnum {refPath=&NiagaraFloat}}}
      Index 0 Vector2D 1 {
       bIsDynamic=false
       bIsEditable=false
       bIsStaticSwitch=false
       bIsVisible=false
       type {classStructOrEnum {refPath=&Vector2f}}
      }
      Index 0 Vector2D 2 {
       bIsDynamic=false
       bIsEditable=false
       bIsStaticSwitch=false
       bIsVisible=false
       type {classStructOrEnum {refPath=&Vector2f}}
      }
      Index 0 Vector 4 {
       bIsDynamic=false
       bIsEditable=false
       bIsStaticSwitch=false
       bIsVisible=false
       type {classStructOrEnum {refPath=&Vector4f}}
      }
      Index 0 Linear Color {
       bIsDynamic=false
       bIsEditable=false
       bIsStaticSwitch=false
       bIsVisible=false
       type {classStructOrEnum {refPath=&LinearColor}}
      }
      Write Parameter Index 1 {
       bIsDynamic=false
       bIsEditable=true
       bIsStaticSwitch=true
       bIsVisible=true
       type {classStructOrEnum {refPath=&ENiagara_DynamicMaterialParametersInputMode}}
      }
      Index 1 Vector 1 {
       bIsDynamic=false
       bIsEditable=false
       bIsStaticSwitch=false
       bIsVisible=false
       type {classStructOrEnum {refPath=&Vector3f}}
      }
      Index 1 Param 1 {
       bIsDynamic=false
       bIsEditable=false
       bIsStaticSwitch=false
       bIsVisible=false
       type {classStructOrEnum {refPath=&NiagaraFloat}}
      }
      Index 1 Param 2 {
       bIsDynamic=false
       bIsEditable=false
       bIsStaticSwitch=false
       bIsVisible=false
       type {classStructOrEnum {refPath=&NiagaraFloat}}
      }
      Index 1 Param 3 {
       bIsDynamic=false
       bIsEditable=false
       bIsStaticSwitch=false
       bIsVisible=false
       type {classStructOrEnum {refPath=&NiagaraFloat}}
      }
      Index 1 Param 4 {
       bIsDynamic=false
       bIsEditable=false
       bIsStaticSwitch=false
       bIsVisible=false
       type {classStructOrEnum {refPath=&NiagaraFloat}}
      }
      Index 1 Vector2D 1 {
       bIsDynamic=false
       bIsEditable=false
       bIsStaticSwitch=false
       bIsVisible=false
       type {classStructOrEnum {refPath=&Vector2f}}
      }
      Index 1 Vector2D 2 {
       bIsDynamic=false
       bIsEditable=false
       bIsStaticSwitch=false
       bIsVisible=false
       type {classStructOrEnum {refPath=&Vector2f}}
      }
      Index 1 Vector 4 {
       bIsDynamic=false
       bIsEditable=false
       bIsStaticSwitch=false
       bIsVisible=false
       type {classStructOrEnum {refPath=&Vector4f}}
      }
      Index 1 Linear Color {
       bIsDynamic=false
       bIsEditable=false
       bIsStaticSwitch=false
       bIsVisible=false
       type {classStructOrEnum {refPath=&LinearColor}}
      }
      Write Parameter Index 2 {
       bIsDynamic=false
       bIsEditable=true
       bIsStaticSwitch=true
       bIsVisible=true
       type {classStructOrEnum {refPath=&ENiagara_DynamicMaterialParametersInputMode}}
      }
      Index 2 Vector 1 {
       bIsDynamic=false
       bIsEditable=false
       bIsStaticSwitch=false
       bIsVisible=false
       type {classStructOrEnum {refPath=&Vector3f}}
      }
      Index 2 Param 1 {
       bIsDynamic=false
       bIsEditable=false
       bIsStaticSwitch=false
       bIsVisible=false
       type {classStructOrEnum {refPath=&NiagaraFloat}}
      }
      Index 2 Param 2 {
       bIsDynamic=false
       bIsEditable=false
       bIsStaticSwitch=false
       bIsVisible=false
       type {classStructOrEnum {refPath=&NiagaraFloat}}
      }
      Index 2 Param 3 {
       bIsDynamic=false
       bIsEditable=false
       bIsStaticSwitch=false
       bIsVisible=false
       type {classStructOrEnum {refPath=&NiagaraFloat}}
      }
      Index 2 Param 4 {
       bIsDynamic=false
       bIsEditable=false
       bIsStaticSwitch=false
       bIsVisible=false
       type {classStructOrEnum {refPath=&NiagaraFloat}}
      }
      Index 2 Vector2D 1 {
       bIsDynamic=false
       bIsEditable=false
       bIsStaticSwitch=false
       bIsVisible=false
       type {classStructOrEnum {refPath=&Vector2f}}
      }
      Index 2 Vector2D 2 {
       bIsDynamic=false
       bIsEditable=false
       bIsStaticSwitch=false
       bIsVisible=false
       type {classStructOrEnum {refPath=&Vector2f}}
      }
      Index 2 Vector 4 {
       bIsDynamic=false
       bIsEditable=false
       bIsStaticSwitch=false
       bIsVisible=false
       type {classStructOrEnum {refPath=&Vector4f}}
      }
      Index 2 Linear Color {
       bIsDynamic=false
       bIsEditable=false
       bIsStaticSwitch=false
       bIsVisible=false
       type {classStructOrEnum {refPath=&LinearColor}}
      }
      Write Parameter Index 3 {
       bIsDynamic=false
       bIsEditable=true
       bIsStaticSwitch=true
       bIsVisible=true
       type {classStructOrEnum {refPath=&ENiagara_DynamicMaterialParametersInputMode}}
      }
      Index 3 Vector 1 {
       bIsDynamic=false
       bIsEditable=false
       bIsStaticSwitch=false
       bIsVisible=false
       type {classStructOrEnum {refPath=&Vector3f}}
      }
      Index 3 Param 1 {
       bIsDynamic=false
       bIsEditable=false
       bIsStaticSwitch=false
       bIsVisible=false
       type {classStructOrEnum {refPath=&NiagaraFloat}}
      }
      Index 3 Param 2 {
       bIsDynamic=false
       bIsEditable=false
       bIsStaticSwitch=false
       bIsVisible=false
       type {classStructOrEnum {refPath=&NiagaraFloat}}
      }
      Index 3 Param 3 {
       bIsDynamic=false
       bIsEditable=false
       bIsStaticSwitch=false
       bIsVisible=false
       type {classStructOrEnum {refPath=&NiagaraFloat}}
      }
      Index 3 Param 4 {
       bIsDynamic=false
       bIsEditable=false
       bIsStaticSwitch=false
       bIsVisible=false
       type {classStructOrEnum {refPath=&NiagaraFloat}}
      }
      Index 3 Vector2D 1 {
       bIsDynamic=false
       bIsEditable=false
       bIsStaticSwitch=false
       bIsVisible=false
       type {classStructOrEnum {refPath=&Vector2f}}
      }
      Index 3 Vector2D 2 {
       bIsDynamic=false
       bIsEditable=false
       bIsStaticSwitch=false
       bIsVisible=false
       type {classStructOrEnum {refPath=&Vector2f}}
      }
      Index 3 Vector 4 {
       bIsDynamic=false
       bIsEditable=false
       bIsStaticSwitch=false
       bIsVisible=false
       type {classStructOrEnum {refPath=&Vector4f}}
      }
      Index 3 Linear Color {
       bIsDynamic=false
       bIsEditable=false
       bIsStaticSwitch=false
       bIsVisible=false
       type {classStructOrEnum {refPath=&LinearColor}}
      }
      Dynamic Parameter Output {
       bIsDynamic=false
       bIsEditable=true
       bIsStaticSwitch=true
       bIsVisible=true
       type {classStructOrEnum {refPath=&ENiagara_AttributeSamplingApplyOutput}}
      }
     ]
     moduleScript {refPath=&DynamicMaterialParameters}
    }
    {
     bIsSetParametersModule=true
     enabled=true
     moduleName=SetVariables_#5
     inputs [Particles.LightSpecularScale {bIsDynamic=true, bIsEditable=true, bIsStaticSwitch=false, bIsVisible=true, type {classStructOrEnum {refPath=&NiagaraFloat}}}]
     moduleScript {refPath=&self:Flare_0.NiagaraScriptSource_0.NiagaraGraph_0.NiagaraNodeAssignment_0.SetVariables_#5}
    }
   ]
  }
  particleUpdateScript {
   scriptName=ParticleUpdateScript
   modules [
    {
     bIsSetParametersModule=false
     enabled=true
     moduleName=ParticleState
     inputs [
      Kill Particles When Lifetime Has Elapsed {bIsDynamic=false, bIsEditable=true, bIsStaticSwitch=true, bIsVisible=true, type {classStructOrEnum {refPath=&NiagaraBool}}}
      Loop Particles Lifetime {bIsDynamic=false, bIsEditable=false, bIsStaticSwitch=false, bIsVisible=false, type {classStructOrEnum {refPath=&NiagaraBool}}}
      Let Infinitely Lived Particles Die When Emitter Deactivates {bIsDynamic=false, bIsEditable=false, bIsStaticSwitch=false, bIsVisible=false, type {classStructOrEnum {refPath=&NiagaraBool}}}
      Lifetime {bIsDynamic=false, bIsEditable=true, bIsStaticSwitch=false, bIsVisible=true, type {classStructOrEnum {refPath=&NiagaraFloat}}}
      DeltaTime {bIsDynamic=false, bIsEditable=true, bIsStaticSwitch=false, bIsVisible=true, type {classStructOrEnum {refPath=&NiagaraFloat}}}
     ]
     moduleScript {refPath=&ParticleState}
    }
    {
     bIsSetParametersModule=false
     enabled=true
     moduleName=SolveForcesAndVelocity
     inputs [
      Clamp Velocity {bIsDynamic=false, bIsEditable=true, bIsStaticSwitch=false, bIsVisible=true, type {classStructOrEnum {refPath=&NiagaraBool}}}
      Speed Limit {bIsDynamic=false, bIsEditable=false, bIsStaticSwitch=false, bIsVisible=true, type {classStructOrEnum {refPath=&NiagaraFloat}}}
      Limit Acceleration {bIsDynamic=false, bIsEditable=true, bIsStaticSwitch=false, bIsVisible=true, type {classStructOrEnum {refPath=&NiagaraBool}}}
      Acceleration Limit {bIsDynamic=false, bIsEditable=false, bIsStaticSwitch=false, bIsVisible=true, type {classStructOrEnum {refPath=&NiagaraFloat}}}
      Manually Enable Rotational Solver {bIsDynamic=false, bIsEditable=true, bIsStaticSwitch=false, bIsVisible=true, type {classStructOrEnum {refPath=&NiagaraBool}}}
      Write to Intrinsic Properties {bIsDynamic=false, bIsEditable=true, bIsStaticSwitch=false, bIsVisible=true, type {classStructOrEnum {refPath=&NiagaraBool}}}
      Write to Presolve Properties {bIsDynamic=false, bIsEditable=true, bIsStaticSwitch=true, bIsVisible=true, type {classStructOrEnum {refPath=&NiagaraBool}}}
      DeltaTime {bIsDynamic=false, bIsEditable=true, bIsStaticSwitch=false, bIsVisible=true, type {classStructOrEnum {refPath=&NiagaraFloat}}}
      Mass {bIsDynamic=false, bIsEditable=true, bIsStaticSwitch=false, bIsVisible=true, type {classStructOrEnum {refPath=&NiagaraFloat}}}
      Position {
       bIsDynamic=false
       bIsEditable=true
       bIsStaticSwitch=false
       bIsVisible=true
       type {classStructOrEnum {refPath=&NiagaraPosition}}
      }
      Velocity {bIsDynamic=false, bIsEditable=true, bIsStaticSwitch=false, bIsVisible=true, type {classStructOrEnum {refPath=&Vector3f}}}
      PreviousVelocity {bIsDynamic=false, bIsEditable=true, bIsStaticSwitch=false, bIsVisible=true, type {classStructOrEnum {refPath=&Vector3f}}}
      Force {bIsDynamic=false, bIsEditable=true, bIsStaticSwitch=false, bIsVisible=true, type {classStructOrEnum {refPath=&Vector3f}}}
      Mesh Orientation {bIsDynamic=false, bIsEditable=false, bIsStaticSwitch=false, bIsVisible=false, type {classStructOrEnum {refPath=&Quat4f}}}
      Rotational Velocity {
       bIsDynamic=false
       bIsEditable=false
       bIsStaticSwitch=false
       bIsVisible=false
       type {classStructOrEnum {refPath=&Vector3f}}
      }
      Rotational Force {
       bIsDynamic=false
       bIsEditable=false
       bIsStaticSwitch=false
       bIsVisible=false
       type {classStructOrEnum {refPath=&Vector3f}}
      }
      Rotational Inertia {
       bIsDynamic=false
       bIsEditable=false
       bIsStaticSwitch=false
       bIsVisible=false
       type {classStructOrEnum {refPath=&NiagaraFloat}}
      }
     ]
     moduleScript {refPath=&SolveForcesAndVelocity}
    }
    {
     bIsSetParametersModule=false
     enabled=true
     moduleName=ScaleColor
     inputs [
      Scale Mode {bIsDynamic=false, bIsEditable=true, bIsStaticSwitch=true, bIsVisible=true, type {classStructOrEnum {refPath=&NiagaraInt32}}}
      Linear Color Curve {
       bIsDynamic=false
       bIsEditable=false
       bIsStaticSwitch=false
       bIsVisible=false
       type {classStructOrEnum {refPath=&NiagaraDataInterfaceColorCurve}}
      }
      ScaleRGB {bIsDynamic=false, bIsEditable=true, bIsStaticSwitch=false, bIsVisible=true, type {classStructOrEnum {refPath=&NiagaraBool}}}
      Scale RGB {bIsDynamic=false, bIsEditable=true, bIsStaticSwitch=false, bIsVisible=true, type {classStructOrEnum {refPath=&Vector3f}}}
      ScaleRGBA {bIsDynamic=false, bIsEditable=false, bIsStaticSwitch=false, bIsVisible=false, type {classStructOrEnum {refPath=&NiagaraBool}}}
      Scale RGBA {
       bIsDynamic=false
       bIsEditable=false
       bIsStaticSwitch=false
       bIsVisible=false
       type {classStructOrEnum {refPath=&Vector4f}}
      }
      Curve Index {
       bIsDynamic=false
       bIsEditable=false
       bIsStaticSwitch=false
       bIsVisible=false
       type {classStructOrEnum {refPath=&NiagaraFloat}}
      }
      ScaleA {bIsDynamic=false, bIsEditable=true, bIsStaticSwitch=false, bIsVisible=true, type {classStructOrEnum {refPath=&NiagaraBool}}}
      Scale Alpha {bIsDynamic=true, bIsEditable=true, bIsStaticSwitch=false, bIsVisible=true, type {classStructOrEnum {refPath=&NiagaraFloat}}}
      Color Value To Scale {
       bIsDynamic=false
       bIsEditable=true
       bIsStaticSwitch=false
       bIsVisible=true
       type {classStructOrEnum {refPath=&LinearColor}}
      }
      Color Scale {
       bIsDynamic=false
       bIsEditable=false
       bIsStaticSwitch=false
       bIsVisible=false
       type {classStructOrEnum {refPath=&LinearColor}}
      }
     ]
     moduleScript {refPath=&ScaleColor}
    }
    {
     bIsSetParametersModule=false
     enabled=true
     moduleName=ScaleSpriteSize
     inputs [
      Scale Sprite Size Mode {
       bIsDynamic=false
       bIsEditable=true
       bIsStaticSwitch=true
       bIsVisible=true
       type {classStructOrEnum {refPath=&ENiagara_ScaleSpriteSize}}
      }
      Uniform Curve Sprite Scale {
       bIsDynamic=false
       bIsEditable=false
       bIsStaticSwitch=false
       bIsVisible=false
       type {classStructOrEnum {refPath=&NiagaraDataInterfaceCurve}}
      }
      Non-Uniform Curve Sprite Scale {
       bIsDynamic=false
       bIsEditable=false
       bIsStaticSwitch=false
       bIsVisible=false
       type {classStructOrEnum {refPath=&NiagaraDataInterfaceVector2DCurve}}
      }
      Uniform Curve Index {
       bIsDynamic=false
       bIsEditable=false
       bIsStaticSwitch=false
       bIsVisible=false
       type {classStructOrEnum {refPath=&NiagaraFloat}}
      }
      Non-Uniform Curve Index {
       bIsDynamic=false
       bIsEditable=false
       bIsStaticSwitch=false
       bIsVisible=false
       type {classStructOrEnum {refPath=&NiagaraFloat}}
      }
      Non-Uniform Curve Scale {
       bIsDynamic=false
       bIsEditable=false
       bIsStaticSwitch=false
       bIsVisible=false
       type {classStructOrEnum {refPath=&Vector2f}}
      }
      Uniform Curve Scale {
       bIsDynamic=false
       bIsEditable=false
       bIsStaticSwitch=false
       bIsVisible=false
       type {classStructOrEnum {refPath=&NiagaraFloat}}
      }
      Scale Factor {bIsDynamic=true, bIsEditable=true, bIsStaticSwitch=false, bIsVisible=true, type {classStructOrEnum {refPath=&Vector2f}}}
      Uniform Scale Factor {
       bIsDynamic=false
       bIsEditable=false
       bIsStaticSwitch=false
       bIsVisible=false
       type {classStructOrEnum {refPath=&NiagaraFloat}}
      }
      Initial Sprite Size {bIsDynamic=false, bIsEditable=true, bIsStaticSwitch=false, bIsVisible=true, type {classStructOrEnum {refPath=&Vector2f}}}
     ]
     moduleScript {refPath=&ScaleSpriteSize}
    }
    {
     bIsSetParametersModule=false
     enabled=true
     moduleName=FN_Light_Attributes
     inputs [
      Enable Particle Light Enable {bIsDynamic=false, bIsEditable=true, bIsStaticSwitch=true, bIsVisible=true, type {classStructOrEnum {refPath=&NiagaraBool}}}
      Light Enabled {bIsDynamic=false, bIsEditable=false, bIsStaticSwitch=false, bIsVisible=false, type {classStructOrEnum {refPath=&NiagaraBool}}}
      Enable Visibility Tag {bIsDynamic=false, bIsEditable=true, bIsStaticSwitch=true, bIsVisible=true, type {classStructOrEnum {refPath=&NiagaraBool}}}
      VisibilityTag {
       bIsDynamic=false
       bIsEditable=false
       bIsStaticSwitch=false
       bIsVisible=false
       type {classStructOrEnum {refPath=&NiagaraInt32}}
      }
      Create Position Attribute {bIsDynamic=false, bIsEditable=true, bIsStaticSwitch=true, bIsVisible=true, type {classStructOrEnum {refPath=&NiagaraBool}}}
      Position {bIsDynamic=false, bIsEditable=true, bIsStaticSwitch=false, bIsVisible=true, type {classStructOrEnum {refPath=&Vector3f}}}
      Create Radius Attribute {bIsDynamic=false, bIsEditable=true, bIsStaticSwitch=true, bIsVisible=true, type {classStructOrEnum {refPath=&NiagaraBool}}}
      Radius {bIsDynamic=false, bIsEditable=true, bIsStaticSwitch=false, bIsVisible=true, type {classStructOrEnum {refPath=&NiagaraFloat}}}
      Create Exponent Attribute {bIsDynamic=false, bIsEditable=true, bIsStaticSwitch=true, bIsVisible=true, type {classStructOrEnum {refPath=&NiagaraBool}}}
      Exponent {bIsDynamic=false, bIsEditable=true, bIsStaticSwitch=false, bIsVisible=true, type {classStructOrEnum {refPath=&NiagaraFloat}}}
      Create Color Attribute {bIsDynamic=false, bIsEditable=true, bIsStaticSwitch=true, bIsVisible=true, type {classStructOrEnum {refPath=&NiagaraBool}}}
      Color {
       bIsDynamic=false
       bIsEditable=true
       bIsStaticSwitch=false
       bIsVisible=true
       type {classStructOrEnum {refPath=&LinearColor}}
      }
      Scale Color {bIsDynamic=true, bIsEditable=true, bIsStaticSwitch=false, bIsVisible=true, type {classStructOrEnum {refPath=&NiagaraFloat}}}
     ]
     moduleScript {refPath=&FN_Light_Attributes}
    }
   ]
  }
  rendererClasses [{refPath=&NiagaraSpriteRendererProperties}, {refPath=&NiagaraLightRendererProperties}]
  renderers [
   {rendererIndex=0, rendererClass {refPath=&NiagaraSpriteRendererProperties}}
   {rendererIndex=1, rendererClass {refPath=&NiagaraLightRendererProperties}}
  ]
 }
 Flare002 {
  emitterSpawnScript {scriptName=EmitterSpawnScript, modules []}
  emitterUpdateScript {
   scriptName=EmitterUpdateScript
   modules [
    {
     bIsSetParametersModule=false
     enabled=true
     moduleName=EmitterState
     inputs [
      Life Cycle Mode {
       bIsDynamic=false
       bIsEditable=true
       bIsStaticSwitch=true
       bIsVisible=true
       type {classStructOrEnum {refPath=&ENiagaraEmitterLifeCycleMode}}
      }
      Inactive Response {
       bIsDynamic=false
       bIsEditable=false
       bIsStaticSwitch=true
       bIsVisible=false
       type {classStructOrEnum {refPath=&ENiagaraInactiveMode}}
      }
      Loop Behavior {
       bIsDynamic=false
       bIsEditable=false
       bIsStaticSwitch=true
       bIsVisible=false
       type {classStructOrEnum {refPath=&ENiagara_EmitterStateOptions}}
      }
      Loop Duration Mode {
       bIsDynamic=false
       bIsEditable=false
       bIsStaticSwitch=true
       bIsVisible=false
       type {classStructOrEnum {refPath=&ENiagara_InfiniteLoopDuration}}
      }
      Loop Count {
       bIsDynamic=false
       bIsEditable=false
       bIsStaticSwitch=false
       bIsVisible=false
       type {classStructOrEnum {refPath=&NiagaraInt32}}
      }
      Loop Duration {
       bIsDynamic=false
       bIsEditable=false
       bIsStaticSwitch=false
       bIsVisible=false
       type {classStructOrEnum {refPath=&NiagaraFloat}}
      }
      Recalculate Duration Each Loop {bIsDynamic=false, bIsEditable=false, bIsStaticSwitch=false, bIsVisible=false, type {classStructOrEnum {refPath=&NiagaraBool}}}
      UseLoopDelay {bIsDynamic=false, bIsEditable=false, bIsStaticSwitch=false, bIsVisible=false, type {classStructOrEnum {refPath=&NiagaraBool}}}
      Loop Delay {
       bIsDynamic=false
       bIsEditable=false
       bIsStaticSwitch=false
       bIsVisible=false
       type {classStructOrEnum {refPath=&NiagaraFloat}}
      }
      Delay First Loop Only {bIsDynamic=false, bIsEditable=false, bIsStaticSwitch=false, bIsVisible=false, type {classStructOrEnum {refPath=&NiagaraBool}}}
      Scalability Mode {
       bIsDynamic=false
       bIsEditable=true
       bIsStaticSwitch=true
       bIsVisible=true
       type {classStructOrEnum {refPath=&ENiagaraEmitterScalabilityMode_Limited}}
      }
      Enable Distance Culling {bIsDynamic=false, bIsEditable=false, bIsStaticSwitch=true, bIsVisible=false, type {classStructOrEnum {refPath=&NiagaraBool}}}
      MinDistance {
       bIsDynamic=false
       bIsEditable=false
       bIsStaticSwitch=false
       bIsVisible=false
       type {classStructOrEnum {refPath=&NiagaraFloat}}
      }
      Min Distance Response {
       bIsDynamic=false
       bIsEditable=false
       bIsStaticSwitch=false
       bIsVisible=false
       type {classStructOrEnum {refPath=&ENiagaraExecutionStateManagement}}
      }
      MaxDistance {
       bIsDynamic=false
       bIsEditable=false
       bIsStaticSwitch=false
       bIsVisible=false
       type {classStructOrEnum {refPath=&NiagaraFloat}}
      }
      Max Distance Response {
       bIsDynamic=false
       bIsEditable=false
       bIsStaticSwitch=false
       bIsVisible=false
       type {classStructOrEnum {refPath=&ENiagaraExecutionStateManagement}}
      }
      Scale Spawn Count {bIsDynamic=false, bIsEditable=false, bIsStaticSwitch=true, bIsVisible=false, type {classStructOrEnum {refPath=&NiagaraBool}}}
      Spawn Count Scale {
       bIsDynamic=false
       bIsEditable=false
       bIsStaticSwitch=false
       bIsVisible=false
       type {classStructOrEnum {refPath=&NiagaraFloat}}
      }
      Spawn Count Scale By Distance {
       bIsDynamic=false
       bIsEditable=false
       bIsStaticSwitch=false
       bIsVisible=false
       type {classStructOrEnum {refPath=&NiagaraDataInterfaceCurve}}
      }
      Enable Visibility Culling {bIsDynamic=false, bIsEditable=false, bIsStaticSwitch=true, bIsVisible=false, type {classStructOrEnum {refPath=&NiagaraBool}}}
      Visibility Cull Response {
       bIsDynamic=false
       bIsEditable=false
       bIsStaticSwitch=false
       bIsVisible=false
       type {classStructOrEnum {refPath=&ENiagaraExecutionStateManagement}}
      }
      Visibility Cull Delay {
       bIsDynamic=false
       bIsEditable=false
       bIsStaticSwitch=false
       bIsVisible=false
       type {classStructOrEnum {refPath=&NiagaraFloat}}
      }
      Reset Age On Awaken {bIsDynamic=false, bIsEditable=false, bIsStaticSwitch=false, bIsVisible=false, type {classStructOrEnum {refPath=&NiagaraBool}}}
     ]
     moduleScript {refPath=&EmitterState}
    }
    {
     bIsSetParametersModule=false
     enabled=true
     moduleName=SpawnBurst_Instantaneous
     inputs [
      Spawn Count {bIsDynamic=false, bIsEditable=true, bIsStaticSwitch=false, bIsVisible=true, type {classStructOrEnum {refPath=&NiagaraInt32}}}
      Spawn Time {bIsDynamic=false, bIsEditable=true, bIsStaticSwitch=false, bIsVisible=true, type {classStructOrEnum {refPath=&NiagaraFloat}}}
      Use Spawn Probability {bIsDynamic=false, bIsEditable=true, bIsStaticSwitch=false, bIsVisible=true, type {classStructOrEnum {refPath=&NiagaraBool}}}
      Spawn Probability {bIsDynamic=false, bIsEditable=false, bIsStaticSwitch=false, bIsVisible=true, type {classStructOrEnum {refPath=&NiagaraFloat}}}
      UseLoopCountLimit {bIsDynamic=false, bIsEditable=true, bIsStaticSwitch=false, bIsVisible=true, type {classStructOrEnum {refPath=&NiagaraBool}}}
      Loop Count Limit {bIsDynamic=false, bIsEditable=false, bIsStaticSwitch=false, bIsVisible=true, type {classStructOrEnum {refPath=&NiagaraInt32}}}
      Spawn Group {bIsDynamic=false, bIsEditable=true, bIsStaticSwitch=false, bIsVisible=true, type {classStructOrEnum {refPath=&NiagaraInt32}}}
      Age {bIsDynamic=false, bIsEditable=true, bIsStaticSwitch=false, bIsVisible=true, type {classStructOrEnum {refPath=&NiagaraFloat}}}
     ]
     moduleScript {refPath=&SpawnBurst_Instantaneous}
    }
   ]
  }
  particleSpawnScript {
   scriptName=ParticleSpawnScript
   modules [
    {
     bIsSetParametersModule=false
     enabled=true
     moduleName=PartitionParticles
     inputs [
      Partition Mode {
       bIsDynamic=false
       bIsEditable=true
       bIsStaticSwitch=true
       bIsVisible=true
       type {classStructOrEnum {refPath=&ENiagara_PartitionMode_v2}}
      }
      Partition Count Mode {
       bIsDynamic=false
       bIsEditable=false
       bIsStaticSwitch=true
       bIsVisible=false
       type {classStructOrEnum {refPath=&ENiagara_InfiniteLoopDuration}}
      }
      A {
       bIsDynamic=false
       bIsEditable=false
       bIsStaticSwitch=false
       bIsVisible=false
       type {classStructOrEnum {refPath=&NiagaraFloat}}
      }
      Number Of Partitions {bIsDynamic=false, bIsEditable=true, bIsStaticSwitch=false, bIsVisible=true, type {classStructOrEnum {refPath=&NiagaraInt32}}}
      Particles Per Partition {
       bIsDynamic=false
       bIsEditable=false
       bIsStaticSwitch=false
       bIsVisible=false
       type {classStructOrEnum {refPath=&NiagaraInt32}}
      }
      Partition Duration {
       bIsDynamic=false
       bIsEditable=false
       bIsStaticSwitch=false
       bIsVisible=false
       type {classStructOrEnum {refPath=&NiagaraFloat}}
      }
      Time {
       bIsDynamic=false
       bIsEditable=false
       bIsStaticSwitch=false
       bIsVisible=false
       type {classStructOrEnum {refPath=&NiagaraFloat}}
      }
      Partitioning Frequency {
       bIsDynamic=false
       bIsEditable=false
       bIsStaticSwitch=true
       bIsVisible=false
       type {classStructOrEnum {refPath=&ENiagaraRandomnessEvaluation}}
      }
      B {
       bIsDynamic=false
       bIsEditable=false
       bIsStaticSwitch=false
       bIsVisible=false
       type {classStructOrEnum {refPath=&NiagaraFloat}}
      }
      Distance {
       bIsDynamic=false
       bIsEditable=false
       bIsStaticSwitch=false
       bIsVisible=false
       type {classStructOrEnum {refPath=&NiagaraFloat}}
      }
      Comparison Type {
       bIsDynamic=false
       bIsEditable=false
       bIsStaticSwitch=true
       bIsVisible=false
       type {classStructOrEnum {refPath=&ENiagaraBooleanLogicOps_v2}}
      }
      Evaluation Type {
       bIsDynamic=false
       bIsEditable=false
       bIsStaticSwitch=true
       bIsVisible=false
       type {classStructOrEnum {refPath=&ENiagaraRandomnessEvaluation}}
      }
      UseDistanceFalloff {bIsDynamic=false, bIsEditable=false, bIsStaticSwitch=false, bIsVisible=false, type {classStructOrEnum {refPath=&NiagaraBool}}}
      Distance Falloff Exponent {
       bIsDynamic=false
       bIsEditable=false
       bIsStaticSwitch=false
       bIsVisible=false
       type {classStructOrEnum {refPath=&NiagaraFloat}}
      }
      Start Position {
       bIsDynamic=false
       bIsEditable=false
       bIsStaticSwitch=false
       bIsVisible=false
       type {classStructOrEnum {refPath=&NiagaraPosition}}
      }
      End Position {
       bIsDynamic=false
       bIsEditable=false
       bIsStaticSwitch=false
       bIsVisible=false
       type {classStructOrEnum {refPath=&NiagaraPosition}}
      }
      ID to Partition {bIsDynamic=false, bIsEditable=true, bIsStaticSwitch=false, bIsVisible=true, type {classStructOrEnum {refPath=&NiagaraInt32}}}
      Randomness Mode {
       bIsDynamic=false
       bIsEditable=false
       bIsStaticSwitch=true
       bIsVisible=false
       type {classStructOrEnum {refPath=&ENiagaraRandomnessMode}}
      }
      Override Seed {bIsDynamic=false, bIsEditable=false, bIsStaticSwitch=true, bIsVisible=false, type {classStructOrEnum {refPath=&NiagaraBool}}}
      Fixed Random Seed {bIsDynamic=false, bIsEditable=false, bIsStaticSwitch=true, bIsVisible=false, type {classStructOrEnum {refPath=&NiagaraBool}}}
      Color A {
       bIsDynamic=false
       bIsEditable=false
       bIsStaticSwitch=false
       bIsVisible=false
       type {classStructOrEnum {refPath=&LinearColor}}
      }
      Color B {
       bIsDynamic=false
       bIsEditable=false
       bIsStaticSwitch=false
       bIsVisible=false
       type {classStructOrEnum {refPath=&LinearColor}}
      }
      Color Comparison Type {bIsDynamic=false, bIsEditable=false, bIsStaticSwitch=true, bIsVisible=false, type {classStructOrEnum {refPath=&NiagaraInt32}}}
      Color Tolerance {
       bIsDynamic=false
       bIsEditable=false
       bIsStaticSwitch=false
       bIsVisible=false
       type {classStructOrEnum {refPath=&NiagaraFloat}}
      }
      Minimum Saturation {
       bIsDynamic=false
       bIsEditable=false
       bIsStaticSwitch=false
       bIsVisible=false
       type {classStructOrEnum {refPath=&NiagaraFloat}}
      }
      Direction A {
       bIsDynamic=false
       bIsEditable=false
       bIsStaticSwitch=false
       bIsVisible=false
       type {classStructOrEnum {refPath=&Vector3f}}
      }
      Direction B {
       bIsDynamic=false
       bIsEditable=false
       bIsStaticSwitch=false
       bIsVisible=false
       type {classStructOrEnum {refPath=&Vector3f}}
      }
      Direction Tolerance {
       bIsDynamic=false
       bIsEditable=false
       bIsStaticSwitch=false
       bIsVisible=false
       type {classStructOrEnum {refPath=&NiagaraFloat}}
      }
      Write Partition to Ribbon ID {
       bIsDynamic=false
       bIsEditable=true
       bIsStaticSwitch=true
       bIsVisible=true
       type {classStructOrEnum {refPath=&ENiagara_AttributeSamplingApplyOutput}}
      }
      Write Partition Index to Ribbon Link Order {
       bIsDynamic=false
       bIsEditable=false
       bIsStaticSwitch=true
       bIsVisible=false
       type {classStructOrEnum {refPath=&ENiagara_AttributeSamplingApplyOutput}}
      }
      Normalized Index In Partition Includes One {bIsDynamic=false, bIsEditable=false, bIsStaticSwitch=true, bIsVisible=false, type {classStructOrEnum {refPath=&NiagaraBool}}}
      Normalized Index Includes One {bIsDynamic=false, bIsEditable=true, bIsStaticSwitch=true, bIsVisible=true, type {classStructOrEnum {refPath=&NiagaraBool}}}
     ]
     moduleScript {refPath=&PartitionParticles}
    }
    {
     bIsSetParametersModule=false
     enabled=true
     moduleName=InitializeParticle
     inputs [
      Lifetime Mode {
       bIsDynamic=false
       bIsEditable=true
       bIsStaticSwitch=true
       bIsVisible=true
       type {classStructOrEnum {refPath=&ENiagara_LifetimeMode}}
      }
      Lifetime Min {
       bIsDynamic=false
       bIsEditable=false
       bIsStaticSwitch=false
       bIsVisible=false
       type {classStructOrEnum {refPath=&NiagaraFloat}}
      }
      Lifetime {bIsDynamic=false, bIsEditable=true, bIsStaticSwitch=false, bIsVisible=true, type {classStructOrEnum {refPath=&NiagaraFloat}}}
      Lifetime Max {
       bIsDynamic=false
       bIsEditable=false
       bIsStaticSwitch=false
       bIsVisible=false
       type {classStructOrEnum {refPath=&NiagaraFloat}}
      }
      Lifetime Randomness Mode {
       bIsDynamic=false
       bIsEditable=false
       bIsStaticSwitch=true
       bIsVisible=false
       type {classStructOrEnum {refPath=&ENiagaraRandomnessMode}}
      }
      Color Mode {
       bIsDynamic=false
       bIsEditable=true
       bIsStaticSwitch=true
       bIsVisible=true
       type {classStructOrEnum {refPath=&ENiagara_ColorInitializationMode}}
      }
      Color Minimum {
       bIsDynamic=false
       bIsEditable=false
       bIsStaticSwitch=false
       bIsVisible=false
       type {classStructOrEnum {refPath=&LinearColor}}
      }
      Color {
       bIsDynamic=false
       bIsEditable=true
       bIsStaticSwitch=false
       bIsVisible=true
       type {classStructOrEnum {refPath=&LinearColor}}
      }
      AdjustHue {bIsDynamic=false, bIsEditable=false, bIsStaticSwitch=false, bIsVisible=false, type {classStructOrEnum {refPath=&NiagaraBool}}}
      Hue Shift Range {
       bIsDynamic=false
       bIsEditable=false
       bIsStaticSwitch=false
       bIsVisible=false
       type {classStructOrEnum {refPath=&Vector2f}}
      }
      Color Maximum {
       bIsDynamic=false
       bIsEditable=false
       bIsStaticSwitch=false
       bIsVisible=false
       type {classStructOrEnum {refPath=&LinearColor}}
      }
      AdjustSaturation {bIsDynamic=false, bIsEditable=false, bIsStaticSwitch=false, bIsVisible=false, type {classStructOrEnum {refPath=&NiagaraBool}}}
      Saturation Range {
       bIsDynamic=false
       bIsEditable=false
       bIsStaticSwitch=false
       bIsVisible=false
       type {classStructOrEnum {refPath=&Vector2f}}
      }
      AdjustValue {bIsDynamic=false, bIsEditable=false, bIsStaticSwitch=false, bIsVisible=false, type {classStructOrEnum {refPath=&NiagaraBool}}}
      Value Range {
       bIsDynamic=false
       bIsEditable=false
       bIsStaticSwitch=false
       bIsVisible=false
       type {classStructOrEnum {refPath=&Vector2f}}
      }
      AdjustAlpha {bIsDynamic=false, bIsEditable=false, bIsStaticSwitch=false, bIsVisible=false, type {classStructOrEnum {refPath=&NiagaraBool}}}
      Alpha Scale Range {
       bIsDynamic=false
       bIsEditable=false
       bIsStaticSwitch=false
       bIsVisible=false
       type {classStructOrEnum {refPath=&Vector2f}}
      }
      Color Channel Mode {
       bIsDynamic=false
       bIsEditable=false
       bIsStaticSwitch=true
       bIsVisible=false
       type {classStructOrEnum {refPath=&ENiagaraChannelCorrelation}}
      }
      Color Randomness Mode {
       bIsDynamic=false
       bIsEditable=false
       bIsStaticSwitch=true
       bIsVisible=false
       type {classStructOrEnum {refPath=&ENiagaraRandomnessMode}}
      }
      Position Mode {
       bIsDynamic=false
       bIsEditable=true
       bIsStaticSwitch=true
       bIsVisible=true
       type {classStructOrEnum {refPath=&ENiagara_PositionInitializationMode}}
      }
      Position {
       bIsDynamic=false
       bIsEditable=false
       bIsStaticSwitch=false
       bIsVisible=false
       type {classStructOrEnum {refPath=&NiagaraPosition}}
      }
      UsePositionOffset {bIsDynamic=false, bIsEditable=true, bIsStaticSwitch=false, bIsVisible=true, type {classStructOrEnum {refPath=&NiagaraBool}}}
      Position Offset {bIsDynamic=true, bIsEditable=true, bIsStaticSwitch=false, bIsVisible=true, type {classStructOrEnum {refPath=&Vector3f}}}
      Position Offset Coordinate Space {
       bIsDynamic=false
       bIsEditable=true
       bIsStaticSwitch=false
       bIsVisible=true
       type {classStructOrEnum {refPath=&ENiagaraCoordinateSpace}}
      }
      Mass Mode {
       bIsDynamic=false
       bIsEditable=true
       bIsStaticSwitch=true
       bIsVisible=true
       type {classStructOrEnum {refPath=&ENiagara_MassInitializationMode}}
      }
      Mass Min {
       bIsDynamic=false
       bIsEditable=false
       bIsStaticSwitch=false
       bIsVisible=false
       type {classStructOrEnum {refPath=&NiagaraFloat}}
      }
      Mass {
       bIsDynamic=false
       bIsEditable=false
       bIsStaticSwitch=false
       bIsVisible=false
       type {classStructOrEnum {refPath=&NiagaraFloat}}
      }
      Mass Max {
       bIsDynamic=false
       bIsEditable=false
       bIsStaticSwitch=false
       bIsVisible=false
       type {classStructOrEnum {refPath=&NiagaraFloat}}
      }
      Mass Randomness Mode {
       bIsDynamic=false
       bIsEditable=false
       bIsStaticSwitch=true
       bIsVisible=false
       type {classStructOrEnum {refPath=&ENiagaraRandomnessMode}}
      }
      Write Material Random {bIsDynamic=false, bIsEditable=true, bIsStaticSwitch=false, bIsVisible=true, type {classStructOrEnum {refPath=&NiagaraBool}}}
      Material Random {bIsDynamic=false, bIsEditable=true, bIsStaticSwitch=false, bIsVisible=true, type {classStructOrEnum {refPath=&NiagaraFloat}}}
      Sprite Size Mode {
       bIsDynamic=false
       bIsEditable=true
       bIsStaticSwitch=true
       bIsVisible=true
       type {classStructOrEnum {refPath=&ENiagara_SizeScaleMode}}
      }
      Sprite Size {
       bIsDynamic=false
       bIsEditable=false
       bIsStaticSwitch=false
       bIsVisible=false
       type {classStructOrEnum {refPath=&Vector2f}}
      }
      Uniform Sprite Size {bIsDynamic=false, bIsEditable=true, bIsStaticSwitch=false, bIsVisible=true, type {classStructOrEnum {refPath=&NiagaraFloat}}}
      Uniform Sprite Size Min {
       bIsDynamic=false
       bIsEditable=false
       bIsStaticSwitch=false
       bIsVisible=false
       type {classStructOrEnum {refPath=&NiagaraFloat}}
      }
      Sprite Size Min {
       bIsDynamic=false
       bIsEditable=false
       bIsStaticSwitch=false
       bIsVisible=false
       type {classStructOrEnum {refPath=&Vector2f}}
      }
      Sprite Size Max {
       bIsDynamic=false
       bIsEditable=false
       bIsStaticSwitch=false
       bIsVisible=false
       type {classStructOrEnum {refPath=&Vector2f}}
      }
      Uniform Sprite Size Max {
       bIsDynamic=false
       bIsEditable=false
       bIsStaticSwitch=false
       bIsVisible=false
       type {classStructOrEnum {refPath=&NiagaraFloat}}
      }
      Sprite Size Randomness Mode {
       bIsDynamic=false
       bIsEditable=false
       bIsStaticSwitch=true
       bIsVisible=false
       type {classStructOrEnum {refPath=&ENiagaraRandomnessMode}}
      }
      Sprite Rotation Mode {
       bIsDynamic=false
       bIsEditable=true
       bIsStaticSwitch=true
       bIsVisible=true
       type {classStructOrEnum {refPath=&ENiagara_SpriteRotationMode}}
      }
      Sprite Rotation Angle Min {bIsDynamic=false, bIsEditable=true, bIsStaticSwitch=false, bIsVisible=true, type {classStructOrEnum {refPath=&NiagaraFloat}}}
      Sprite Rotation Angle {
       bIsDynamic=false
       bIsEditable=false
       bIsStaticSwitch=false
       bIsVisible=false
       type {classStructOrEnum {refPath=&NiagaraFloat}}
      }
      Sprite Rotation Angle Max {bIsDynamic=false, bIsEditable=true, bIsStaticSwitch=false, bIsVisible=true, type {classStructOrEnum {refPath=&NiagaraFloat}}}
      Sprite Rotation Randomness Mode {
       bIsDynamic=false
       bIsEditable=true
       bIsStaticSwitch=true
       bIsVisible=true
       type {classStructOrEnum {refPath=&ENiagaraRandomnessMode}}
      }
      Sprite UV Mode {
       bIsDynamic=false
       bIsEditable=true
       bIsStaticSwitch=true
       bIsVisible=true
       type {classStructOrEnum {refPath=&ENiagara_UVFlippingMode}}
      }
      Sprite UV Randomness Mode {
       bIsDynamic=false
       bIsEditable=false
       bIsStaticSwitch=true
       bIsVisible=false
       type {classStructOrEnum {refPath=&ENiagaraRandomnessMode}}
      }
      Sprite UV Scale {
       bIsDynamic=false
       bIsEditable=false
       bIsStaticSwitch=false
       bIsVisible=false
       type {classStructOrEnum {refPath=&Vector2f}}
      }
      Mesh Scale Mode {
       bIsDynamic=false
       bIsEditable=true
       bIsStaticSwitch=true
       bIsVisible=true
       type {classStructOrEnum {refPath=&ENiagara_SizeScaleMode}}
      }
      Mesh Uniform Scale Min {
       bIsDynamic=false
       bIsEditable=false
       bIsStaticSwitch=false
       bIsVisible=false
       type {classStructOrEnum {refPath=&NiagaraFloat}}
      }
      Mesh Uniform Scale {
       bIsDynamic=false
       bIsEditable=false
       bIsStaticSwitch=false
       bIsVisible=false
       type {classStructOrEnum {refPath=&NiagaraFloat}}
      }
      Mesh Scale Min {
       bIsDynamic=false
       bIsEditable=false
       bIsStaticSwitch=false
       bIsVisible=false
       type {classStructOrEnum {refPath=&Vector3f}}
      }
      Mesh Scale {
       bIsDynamic=false
       bIsEditable=false
       bIsStaticSwitch=false
       bIsVisible=false
       type {classStructOrEnum {refPath=&Vector3f}}
      }
      Mesh Scale Max {
       bIsDynamic=false
       bIsEditable=false
       bIsStaticSwitch=false
       bIsVisible=false
       type {classStructOrEnum {refPath=&Vector3f}}
      }
      Mesh Uniform Scale Max {
       bIsDynamic=false
       bIsEditable=false
       bIsStaticSwitch=false
       bIsVisible=false
       type {classStructOrEnum {refPath=&NiagaraFloat}}
      }
      Mesh Scale Randomness Mode {
       bIsDynamic=false
       bIsEditable=false
       bIsStaticSwitch=true
       bIsVisible=false
       type {classStructOrEnum {refPath=&ENiagaraRandomnessMode}}
      }
      Mesh Renderer Array Visibility Mode {
       bIsDynamic=false
       bIsEditable=true
       bIsStaticSwitch=true
       bIsVisible=true
       type {classStructOrEnum {refPath=&ENiagara_UnsetDirectSetRandom}}
      }
      Mesh Renderer Mesh Index {
       bIsDynamic=false
       bIsEditable=false
       bIsStaticSwitch=false
       bIsVisible=false
       type {classStructOrEnum {refPath=&NiagaraInt32}}
      }
      Wrap Index By Mesh Count {bIsDynamic=false, bIsEditable=false, bIsStaticSwitch=true, bIsVisible=false, type {classStructOrEnum {refPath=&NiagaraBool}}}
      Mesh Renderer Info {
       bIsDynamic=false
       bIsEditable=false
       bIsStaticSwitch=false
       bIsVisible=false
       type {classStructOrEnum {refPath=&NiagaraDataInterfaceMeshRendererInfo}}
      }
      Mesh Renderer Random Seed {
       bIsDynamic=false
       bIsEditable=false
       bIsStaticSwitch=false
       bIsVisible=false
       type {classStructOrEnum {refPath=&NiagaraInt32}}
      }
      Ribbon Width Mode {
       bIsDynamic=false
       bIsEditable=true
       bIsStaticSwitch=true
       bIsVisible=true
       type {classStructOrEnum {refPath=&ENiagara_UnsetDirectSet}}
      }
      Ribbon Width {
       bIsDynamic=false
       bIsEditable=false
       bIsStaticSwitch=false
       bIsVisible=false
       type {classStructOrEnum {refPath=&NiagaraFloat}}
      }
      Ribbon Facing Mode {
       bIsDynamic=false
       bIsEditable=true
       bIsStaticSwitch=true
       bIsVisible=true
       type {classStructOrEnum {refPath=&ENiagara_UnsetDirectSet}}
      }
      Ribbon Facing Vector {
       bIsDynamic=false
       bIsEditable=false
       bIsStaticSwitch=false
       bIsVisible=false
       type {classStructOrEnum {refPath=&Vector3f}}
      }
      Ribbon Twist Mode {
       bIsDynamic=false
       bIsEditable=true
       bIsStaticSwitch=true
       bIsVisible=true
       type {classStructOrEnum {refPath=&ENiagara_UnsetDirectSet}}
      }
      "Ribbon Twist " {
       bIsDynamic=false
       bIsEditable=false
       bIsStaticSwitch=false
       bIsVisible=false
       type {classStructOrEnum {refPath=&NiagaraFloat}}
      }
     ]
     moduleScript {refPath=&InitializeParticle}
    }
    {
     bIsSetParametersModule=false
     enabled=true
     moduleName=DynamicMaterialParameters
     inputs [
      Write Parameter Index 0 {
       bIsDynamic=false
       bIsEditable=true
       bIsStaticSwitch=true
       bIsVisible=true
       type {classStructOrEnum {refPath=&ENiagara_DynamicMaterialParametersInputMode}}
      }
      Index 0 Vector 1 {
       bIsDynamic=false
       bIsEditable=false
       bIsStaticSwitch=false
       bIsVisible=false
       type {classStructOrEnum {refPath=&Vector3f}}
      }
      Module.Index0Float0Write {bIsDynamic=false, bIsEditable=true, bIsStaticSwitch=true, bIsVisible=true, type {classStructOrEnum {refPath=&NiagaraBool}}}
      Index 0 Param 1 {bIsDynamic=true, bIsEditable=true, bIsStaticSwitch=false, bIsVisible=true, type {classStructOrEnum {refPath=&NiagaraFloat}}}
      Module.Index0Float1Write {bIsDynamic=false, bIsEditable=true, bIsStaticSwitch=true, bIsVisible=true, type {classStructOrEnum {refPath=&NiagaraBool}}}
      Index 0 Param 2 {bIsDynamic=false, bIsEditable=true, bIsStaticSwitch=false, bIsVisible=true, type {classStructOrEnum {refPath=&NiagaraFloat}}}
      Module.Index0Float2Write {bIsDynamic=false, bIsEditable=true, bIsStaticSwitch=true, bIsVisible=true, type {classStructOrEnum {refPath=&NiagaraBool}}}
      Index 0 Param 3 {bIsDynamic=false, bIsEditable=true, bIsStaticSwitch=false, bIsVisible=true, type {classStructOrEnum {refPath=&NiagaraFloat}}}
      Module.Index0Float3Write {bIsDynamic=false, bIsEditable=true, bIsStaticSwitch=true, bIsVisible=true, type {classStructOrEnum {refPath=&NiagaraBool}}}
      Index 0 Param 4 {bIsDynamic=false, bIsEditable=true, bIsStaticSwitch=false, bIsVisible=true, type {classStructOrEnum {refPath=&NiagaraFloat}}}
      Index 0 Vector2D 1 {
       bIsDynamic=false
       bIsEditable=false
       bIsStaticSwitch=false
       bIsVisible=false
       type {classStructOrEnum {refPath=&Vector2f}}
      }
      Index 0 Vector2D 2 {
       bIsDynamic=false
       bIsEditable=false
       bIsStaticSwitch=false
       bIsVisible=false
       type {classStructOrEnum {refPath=&Vector2f}}
      }
      Index 0 Vector 4 {
       bIsDynamic=false
       bIsEditable=false
       bIsStaticSwitch=false
       bIsVisible=false
       type {classStructOrEnum {refPath=&Vector4f}}
      }
      Index 0 Linear Color {
       bIsDynamic=false
       bIsEditable=false
       bIsStaticSwitch=false
       bIsVisible=false
       type {classStructOrEnum {refPath=&LinearColor}}
      }
      Write Parameter Index 1 {
       bIsDynamic=false
       bIsEditable=true
       bIsStaticSwitch=true
       bIsVisible=true
       type {classStructOrEnum {refPath=&ENiagara_DynamicMaterialParametersInputMode}}
      }
      Index 1 Vector 1 {
       bIsDynamic=false
       bIsEditable=false
       bIsStaticSwitch=false
       bIsVisible=false
       type {classStructOrEnum {refPath=&Vector3f}}
      }
      Index 1 Param 1 {
       bIsDynamic=false
       bIsEditable=false
       bIsStaticSwitch=false
       bIsVisible=false
       type {classStructOrEnum {refPath=&NiagaraFloat}}
      }
      Index 1 Param 2 {
       bIsDynamic=false
       bIsEditable=false
       bIsStaticSwitch=false
       bIsVisible=false
       type {classStructOrEnum {refPath=&NiagaraFloat}}
      }
      Index 1 Param 3 {
       bIsDynamic=false
       bIsEditable=false
       bIsStaticSwitch=false
       bIsVisible=false
       type {classStructOrEnum {refPath=&NiagaraFloat}}
      }
      Index 1 Param 4 {
       bIsDynamic=false
       bIsEditable=false
       bIsStaticSwitch=false
       bIsVisible=false
       type {classStructOrEnum {refPath=&NiagaraFloat}}
      }
      Index 1 Vector2D 1 {
       bIsDynamic=false
       bIsEditable=false
       bIsStaticSwitch=false
       bIsVisible=false
       type {classStructOrEnum {refPath=&Vector2f}}
      }
      Index 1 Vector2D 2 {
       bIsDynamic=false
       bIsEditable=false
       bIsStaticSwitch=false
       bIsVisible=false
       type {classStructOrEnum {refPath=&Vector2f}}
      }
      Index 1 Vector 4 {
       bIsDynamic=false
       bIsEditable=false
       bIsStaticSwitch=false
       bIsVisible=false
       type {classStructOrEnum {refPath=&Vector4f}}
      }
      Index 1 Linear Color {
       bIsDynamic=false
       bIsEditable=false
       bIsStaticSwitch=false
       bIsVisible=false
       type {classStructOrEnum {refPath=&LinearColor}}
      }
      Write Parameter Index 2 {
       bIsDynamic=false
       bIsEditable=true
       bIsStaticSwitch=true
       bIsVisible=true
       type {classStructOrEnum {refPath=&ENiagara_DynamicMaterialParametersInputMode}}
      }
      Index 2 Vector 1 {
       bIsDynamic=false
       bIsEditable=false
       bIsStaticSwitch=false
       bIsVisible=false
       type {classStructOrEnum {refPath=&Vector3f}}
      }
      Index 2 Param 1 {
       bIsDynamic=false
       bIsEditable=false
       bIsStaticSwitch=false
       bIsVisible=false
       type {classStructOrEnum {refPath=&NiagaraFloat}}
      }
      Index 2 Param 2 {
       bIsDynamic=false
       bIsEditable=false
       bIsStaticSwitch=false
       bIsVisible=false
       type {classStructOrEnum {refPath=&NiagaraFloat}}
      }
      Index 2 Param 3 {
       bIsDynamic=false
       bIsEditable=false
       bIsStaticSwitch=false
       bIsVisible=false
       type {classStructOrEnum {refPath=&NiagaraFloat}}
      }
      Index 2 Param 4 {
       bIsDynamic=false
       bIsEditable=false
       bIsStaticSwitch=false
       bIsVisible=false
       type {classStructOrEnum {refPath=&NiagaraFloat}}
      }
      Index 2 Vector2D 1 {
       bIsDynamic=false
       bIsEditable=false
       bIsStaticSwitch=false
       bIsVisible=false
       type {classStructOrEnum {refPath=&Vector2f}}
      }
      Index 2 Vector2D 2 {
       bIsDynamic=false
       bIsEditable=false
       bIsStaticSwitch=false
       bIsVisible=false
       type {classStructOrEnum {refPath=&Vector2f}}
      }
      Index 2 Vector 4 {
       bIsDynamic=false
       bIsEditable=false
       bIsStaticSwitch=false
       bIsVisible=false
       type {classStructOrEnum {refPath=&Vector4f}}
      }
      Index 2 Linear Color {
       bIsDynamic=false
       bIsEditable=false
       bIsStaticSwitch=false
       bIsVisible=false
       type {classStructOrEnum {refPath=&LinearColor}}
      }
      Write Parameter Index 3 {
       bIsDynamic=false
       bIsEditable=true
       bIsStaticSwitch=true
       bIsVisible=true
       type {classStructOrEnum {refPath=&ENiagara_DynamicMaterialParametersInputMode}}
      }
      Index 3 Vector 1 {
       bIsDynamic=false
       bIsEditable=false
       bIsStaticSwitch=false
       bIsVisible=false
       type {classStructOrEnum {refPath=&Vector3f}}
      }
      Index 3 Param 1 {
       bIsDynamic=false
       bIsEditable=false
       bIsStaticSwitch=false
       bIsVisible=false
       type {classStructOrEnum {refPath=&NiagaraFloat}}
      }
      Index 3 Param 2 {
       bIsDynamic=false
       bIsEditable=false
       bIsStaticSwitch=false
       bIsVisible=false
       type {classStructOrEnum {refPath=&NiagaraFloat}}
      }
      Index 3 Param 3 {
       bIsDynamic=false
       bIsEditable=false
       bIsStaticSwitch=false
       bIsVisible=false
       type {classStructOrEnum {refPath=&NiagaraFloat}}
      }
      Index 3 Param 4 {
       bIsDynamic=false
       bIsEditable=false
       bIsStaticSwitch=false
       bIsVisible=false
       type {classStructOrEnum {refPath=&NiagaraFloat}}
      }
      Index 3 Vector2D 1 {
       bIsDynamic=false
       bIsEditable=false
       bIsStaticSwitch=false
       bIsVisible=false
       type {classStructOrEnum {refPath=&Vector2f}}
      }
      Index 3 Vector2D 2 {
       bIsDynamic=false
       bIsEditable=false
       bIsStaticSwitch=false
       bIsVisible=false
       type {classStructOrEnum {refPath=&Vector2f}}
      }
      Index 3 Vector 4 {
       bIsDynamic=false
       bIsEditable=false
       bIsStaticSwitch=false
       bIsVisible=false
       type {classStructOrEnum {refPath=&Vector4f}}
      }
      Index 3 Linear Color {
       bIsDynamic=false
       bIsEditable=false
       bIsStaticSwitch=false
       bIsVisible=false
       type {classStructOrEnum {refPath=&LinearColor}}
      }
      Dynamic Parameter Output {
       bIsDynamic=false
       bIsEditable=true
       bIsStaticSwitch=true
       bIsVisible=true
       type {classStructOrEnum {refPath=&ENiagara_AttributeSamplingApplyOutput}}
      }
     ]
     moduleScript {refPath=&DynamicMaterialParameters}
    }
    {
     bIsSetParametersModule=true
     enabled=true
     moduleName=SetVariables_#5
     inputs [Particles.LightSpecularScale {bIsDynamic=true, bIsEditable=true, bIsStaticSwitch=false, bIsVisible=true, type {classStructOrEnum {refPath=&NiagaraFloat}}}]
     moduleScript {refPath=&self:Flare002_0.NiagaraScriptSource_0.NiagaraGraph_0.NiagaraNodeAssignment_0.SetVariables_#5}
    }
   ]
  }
  particleUpdateScript {
   scriptName=ParticleUpdateScript
   modules [
    {
     bIsSetParametersModule=false
     enabled=true
     moduleName=ParticleState
     inputs [
      Kill Particles When Lifetime Has Elapsed {bIsDynamic=false, bIsEditable=true, bIsStaticSwitch=true, bIsVisible=true, type {classStructOrEnum {refPath=&NiagaraBool}}}
      Loop Particles Lifetime {bIsDynamic=false, bIsEditable=false, bIsStaticSwitch=false, bIsVisible=false, type {classStructOrEnum {refPath=&NiagaraBool}}}
      Let Infinitely Lived Particles Die When Emitter Deactivates {bIsDynamic=false, bIsEditable=false, bIsStaticSwitch=false, bIsVisible=false, type {classStructOrEnum {refPath=&NiagaraBool}}}
      Lifetime {bIsDynamic=false, bIsEditable=true, bIsStaticSwitch=false, bIsVisible=true, type {classStructOrEnum {refPath=&NiagaraFloat}}}
      DeltaTime {bIsDynamic=false, bIsEditable=true, bIsStaticSwitch=false, bIsVisible=true, type {classStructOrEnum {refPath=&NiagaraFloat}}}
     ]
     moduleScript {refPath=&ParticleState}
    }
    {
     bIsSetParametersModule=false
     enabled=true
     moduleName=SolveForcesAndVelocity
     inputs [
      Clamp Velocity {bIsDynamic=false, bIsEditable=true, bIsStaticSwitch=false, bIsVisible=true, type {classStructOrEnum {refPath=&NiagaraBool}}}
      Speed Limit {bIsDynamic=false, bIsEditable=false, bIsStaticSwitch=false, bIsVisible=true, type {classStructOrEnum {refPath=&NiagaraFloat}}}
      Limit Acceleration {bIsDynamic=false, bIsEditable=true, bIsStaticSwitch=false, bIsVisible=true, type {classStructOrEnum {refPath=&NiagaraBool}}}
      Acceleration Limit {bIsDynamic=false, bIsEditable=false, bIsStaticSwitch=false, bIsVisible=true, type {classStructOrEnum {refPath=&NiagaraFloat}}}
      Manually Enable Rotational Solver {bIsDynamic=false, bIsEditable=true, bIsStaticSwitch=false, bIsVisible=true, type {classStructOrEnum {refPath=&NiagaraBool}}}
      Write to Intrinsic Properties {bIsDynamic=false, bIsEditable=true, bIsStaticSwitch=false, bIsVisible=true, type {classStructOrEnum {refPath=&NiagaraBool}}}
      Write to Presolve Properties {bIsDynamic=false, bIsEditable=true, bIsStaticSwitch=true, bIsVisible=true, type {classStructOrEnum {refPath=&NiagaraBool}}}
      DeltaTime {bIsDynamic=false, bIsEditable=true, bIsStaticSwitch=false, bIsVisible=true, type {classStructOrEnum {refPath=&NiagaraFloat}}}
      Mass {bIsDynamic=false, bIsEditable=true, bIsStaticSwitch=false, bIsVisible=true, type {classStructOrEnum {refPath=&NiagaraFloat}}}
      Position {
       bIsDynamic=false
       bIsEditable=true
       bIsStaticSwitch=false
       bIsVisible=true
       type {classStructOrEnum {refPath=&NiagaraPosition}}
      }
      Velocity {bIsDynamic=false, bIsEditable=true, bIsStaticSwitch=false, bIsVisible=true, type {classStructOrEnum {refPath=&Vector3f}}}
      PreviousVelocity {bIsDynamic=false, bIsEditable=true, bIsStaticSwitch=false, bIsVisible=true, type {classStructOrEnum {refPath=&Vector3f}}}
      Force {bIsDynamic=false, bIsEditable=true, bIsStaticSwitch=false, bIsVisible=true, type {classStructOrEnum {refPath=&Vector3f}}}
      Mesh Orientation {bIsDynamic=false, bIsEditable=false, bIsStaticSwitch=false, bIsVisible=false, type {classStructOrEnum {refPath=&Quat4f}}}
      Rotational Velocity {
       bIsDynamic=false
       bIsEditable=false
       bIsStaticSwitch=false
       bIsVisible=false
       type {classStructOrEnum {refPath=&Vector3f}}
      }
      Rotational Force {
       bIsDynamic=false
       bIsEditable=false
       bIsStaticSwitch=false
       bIsVisible=false
       type {classStructOrEnum {refPath=&Vector3f}}
      }
      Rotational Inertia {
       bIsDynamic=false
       bIsEditable=false
       bIsStaticSwitch=false
       bIsVisible=false
       type {classStructOrEnum {refPath=&NiagaraFloat}}
      }
     ]
     moduleScript {refPath=&SolveForcesAndVelocity}
    }
    {
     bIsSetParametersModule=false
     enabled=true
     moduleName=ScaleColor
     inputs [
      Scale Mode {bIsDynamic=false, bIsEditable=true, bIsStaticSwitch=true, bIsVisible=true, type {classStructOrEnum {refPath=&NiagaraInt32}}}
      Linear Color Curve {
       bIsDynamic=false
       bIsEditable=false
       bIsStaticSwitch=false
       bIsVisible=false
       type {classStructOrEnum {refPath=&NiagaraDataInterfaceColorCurve}}
      }
      ScaleRGB {bIsDynamic=false, bIsEditable=true, bIsStaticSwitch=false, bIsVisible=true, type {classStructOrEnum {refPath=&NiagaraBool}}}
      Scale RGB {bIsDynamic=false, bIsEditable=true, bIsStaticSwitch=false, bIsVisible=true, type {classStructOrEnum {refPath=&Vector3f}}}
      ScaleRGBA {bIsDynamic=false, bIsEditable=false, bIsStaticSwitch=false, bIsVisible=false, type {classStructOrEnum {refPath=&NiagaraBool}}}
      Scale RGBA {
       bIsDynamic=false
       bIsEditable=false
       bIsStaticSwitch=false
       bIsVisible=false
       type {classStructOrEnum {refPath=&Vector4f}}
      }
      Curve Index {
       bIsDynamic=false
       bIsEditable=false
       bIsStaticSwitch=false
       bIsVisible=false
       type {classStructOrEnum {refPath=&NiagaraFloat}}
      }
      ScaleA {bIsDynamic=false, bIsEditable=true, bIsStaticSwitch=false, bIsVisible=true, type {classStructOrEnum {refPath=&NiagaraBool}}}
      Scale Alpha {bIsDynamic=true, bIsEditable=true, bIsStaticSwitch=false, bIsVisible=true, type {classStructOrEnum {refPath=&NiagaraFloat}}}
      Color Value To Scale {
       bIsDynamic=false
       bIsEditable=true
       bIsStaticSwitch=false
       bIsVisible=true
       type {classStructOrEnum {refPath=&LinearColor}}
      }
      Color Scale {
       bIsDynamic=false
       bIsEditable=false
       bIsStaticSwitch=false
       bIsVisible=false
       type {classStructOrEnum {refPath=&LinearColor}}
      }
     ]
     moduleScript {refPath=&ScaleColor}
    }
    {
     bIsSetParametersModule=false
     enabled=true
     moduleName=ScaleSpriteSize
     inputs [
      Scale Sprite Size Mode {
       bIsDynamic=false
       bIsEditable=true
       bIsStaticSwitch=true
       bIsVisible=true
       type {classStructOrEnum {refPath=&ENiagara_ScaleSpriteSize}}
      }
      Uniform Curve Sprite Scale {
       bIsDynamic=false
       bIsEditable=false
       bIsStaticSwitch=false
       bIsVisible=false
       type {classStructOrEnum {refPath=&NiagaraDataInterfaceCurve}}
      }
      Non-Uniform Curve Sprite Scale {
       bIsDynamic=false
       bIsEditable=false
       bIsStaticSwitch=false
       bIsVisible=false
       type {classStructOrEnum {refPath=&NiagaraDataInterfaceVector2DCurve}}
      }
      Uniform Curve Index {
       bIsDynamic=false
       bIsEditable=false
       bIsStaticSwitch=false
       bIsVisible=false
       type {classStructOrEnum {refPath=&NiagaraFloat}}
      }
      Non-Uniform Curve Index {
       bIsDynamic=false
       bIsEditable=false
       bIsStaticSwitch=false
       bIsVisible=false
       type {classStructOrEnum {refPath=&NiagaraFloat}}
      }
      Non-Uniform Curve Scale {
       bIsDynamic=false
       bIsEditable=false
       bIsStaticSwitch=false
       bIsVisible=false
       type {classStructOrEnum {refPath=&Vector2f}}
      }
      Uniform Curve Scale {
       bIsDynamic=false
       bIsEditable=false
       bIsStaticSwitch=false
       bIsVisible=false
       type {classStructOrEnum {refPath=&NiagaraFloat}}
      }
      Scale Factor {bIsDynamic=true, bIsEditable=true, bIsStaticSwitch=false, bIsVisible=true, type {classStructOrEnum {refPath=&Vector2f}}}
      Uniform Scale Factor {
       bIsDynamic=false
       bIsEditable=false
       bIsStaticSwitch=false
       bIsVisible=false
       type {classStructOrEnum {refPath=&NiagaraFloat}}
      }
      Initial Sprite Size {bIsDynamic=false, bIsEditable=true, bIsStaticSwitch=false, bIsVisible=true, type {classStructOrEnum {refPath=&Vector2f}}}
     ]
     moduleScript {refPath=&ScaleSpriteSize}
    }
    {
     bIsSetParametersModule=false
     enabled=true
     moduleName=FN_Light_Attributes
     inputs [
      Enable Particle Light Enable {bIsDynamic=false, bIsEditable=true, bIsStaticSwitch=true, bIsVisible=true, type {classStructOrEnum {refPath=&NiagaraBool}}}
      Light Enabled {bIsDynamic=false, bIsEditable=false, bIsStaticSwitch=false, bIsVisible=false, type {classStructOrEnum {refPath=&NiagaraBool}}}
      Enable Visibility Tag {bIsDynamic=false, bIsEditable=true, bIsStaticSwitch=true, bIsVisible=true, type {classStructOrEnum {refPath=&NiagaraBool}}}
      VisibilityTag {
       bIsDynamic=false
       bIsEditable=false
       bIsStaticSwitch=false
       bIsVisible=false
       type {classStructOrEnum {refPath=&NiagaraInt32}}
      }
      Create Position Attribute {bIsDynamic=false, bIsEditable=true, bIsStaticSwitch=true, bIsVisible=true, type {classStructOrEnum {refPath=&NiagaraBool}}}
      Position {bIsDynamic=true, bIsEditable=true, bIsStaticSwitch=false, bIsVisible=true, type {classStructOrEnum {refPath=&Vector3f}}}
      Create Radius Attribute {bIsDynamic=false, bIsEditable=true, bIsStaticSwitch=true, bIsVisible=true, type {classStructOrEnum {refPath=&NiagaraBool}}}
      Radius {bIsDynamic=false, bIsEditable=true, bIsStaticSwitch=false, bIsVisible=true, type {classStructOrEnum {refPath=&NiagaraFloat}}}
      Create Exponent Attribute {bIsDynamic=false, bIsEditable=true, bIsStaticSwitch=true, bIsVisible=true, type {classStructOrEnum {refPath=&NiagaraBool}}}
      Exponent {bIsDynamic=false, bIsEditable=true, bIsStaticSwitch=false, bIsVisible=true, type {classStructOrEnum {refPath=&NiagaraFloat}}}
      Create Color Attribute {bIsDynamic=false, bIsEditable=true, bIsStaticSwitch=true, bIsVisible=true, type {classStructOrEnum {refPath=&NiagaraBool}}}
      Color {
       bIsDynamic=false
       bIsEditable=true
       bIsStaticSwitch=false
       bIsVisible=true
       type {classStructOrEnum {refPath=&LinearColor}}
      }
      Scale Color {bIsDynamic=true, bIsEditable=true, bIsStaticSwitch=false, bIsVisible=true, type {classStructOrEnum {refPath=&NiagaraFloat}}}
     ]
     moduleScript {refPath=&FN_Light_Attributes}
    }
   ]
  }
  rendererClasses [{refPath=&NiagaraSpriteRendererProperties}, {refPath=&NiagaraLightRendererProperties}]
  renderers [
   {rendererIndex=0, rendererClass {refPath=&NiagaraSpriteRendererProperties}}
   {rendererIndex=1, rendererClass {refPath=&NiagaraLightRendererProperties}}
  ]
 }
 Shockwave {
  bEnabled=false
  emitterSpawnScript {scriptName=EmitterSpawnScript, modules []}
  emitterUpdateScript {
   scriptName=EmitterUpdateScript
   modules [
    {
     bIsSetParametersModule=false
     enabled=true
     moduleName=EmitterState
     inputs [
      Life Cycle Mode {
       bIsDynamic=false
       bIsEditable=true
       bIsStaticSwitch=true
       bIsVisible=true
       type {classStructOrEnum {refPath=&ENiagaraEmitterLifeCycleMode}}
      }
      Inactive Response {
       bIsDynamic=false
       bIsEditable=false
       bIsStaticSwitch=true
       bIsVisible=false
       type {classStructOrEnum {refPath=&ENiagaraInactiveMode}}
      }
      Loop Behavior {
       bIsDynamic=false
       bIsEditable=false
       bIsStaticSwitch=true
       bIsVisible=false
       type {classStructOrEnum {refPath=&ENiagara_EmitterStateOptions}}
      }
      Loop Duration Mode {
       bIsDynamic=false
       bIsEditable=false
       bIsStaticSwitch=true
       bIsVisible=false
       type {classStructOrEnum {refPath=&ENiagara_InfiniteLoopDuration}}
      }
      Loop Count {
       bIsDynamic=false
       bIsEditable=false
       bIsStaticSwitch=false
       bIsVisible=false
       type {classStructOrEnum {refPath=&NiagaraInt32}}
      }
      Loop Duration {
       bIsDynamic=false
       bIsEditable=false
       bIsStaticSwitch=false
       bIsVisible=false
       type {classStructOrEnum {refPath=&NiagaraFloat}}
      }
      Recalculate Duration Each Loop {bIsDynamic=false, bIsEditable=false, bIsStaticSwitch=false, bIsVisible=false, type {classStructOrEnum {refPath=&NiagaraBool}}}
      UseLoopDelay {bIsDynamic=false, bIsEditable=false, bIsStaticSwitch=false, bIsVisible=false, type {classStructOrEnum {refPath=&NiagaraBool}}}
      Loop Delay {
       bIsDynamic=false
       bIsEditable=false
       bIsStaticSwitch=false
       bIsVisible=false
       type {classStructOrEnum {refPath=&NiagaraFloat}}
      }
      Delay First Loop Only {bIsDynamic=false, bIsEditable=false, bIsStaticSwitch=false, bIsVisible=false, type {classStructOrEnum {refPath=&NiagaraBool}}}
      Scalability Mode {
       bIsDynamic=false
       bIsEditable=true
       bIsStaticSwitch=true
       bIsVisible=true
       type {classStructOrEnum {refPath=&ENiagaraEmitterScalabilityMode_Limited}}
      }
      Enable Distance Culling {bIsDynamic=false, bIsEditable=false, bIsStaticSwitch=true, bIsVisible=false, type {classStructOrEnum {refPath=&NiagaraBool}}}
      MinDistance {
       bIsDynamic=false
       bIsEditable=false
       bIsStaticSwitch=false
       bIsVisible=false
       type {classStructOrEnum {refPath=&NiagaraFloat}}
      }
      Min Distance Response {
       bIsDynamic=false
       bIsEditable=false
       bIsStaticSwitch=false
       bIsVisible=false
       type {classStructOrEnum {refPath=&ENiagaraExecutionStateManagement}}
      }
      MaxDistance {
       bIsDynamic=false
       bIsEditable=false
       bIsStaticSwitch=false
       bIsVisible=false
       type {classStructOrEnum {refPath=&NiagaraFloat}}
      }
      Max Distance Response {
       bIsDynamic=false
       bIsEditable=false
       bIsStaticSwitch=false
       bIsVisible=false
       type {classStructOrEnum {refPath=&ENiagaraExecutionStateManagement}}
      }
      Scale Spawn Count {bIsDynamic=false, bIsEditable=false, bIsStaticSwitch=true, bIsVisible=false, type {classStructOrEnum {refPath=&NiagaraBool}}}
      Spawn Count Scale {
       bIsDynamic=false
       bIsEditable=false
       bIsStaticSwitch=false
       bIsVisible=false
       type {classStructOrEnum {refPath=&NiagaraFloat}}
      }
      Spawn Count Scale By Distance {
       bIsDynamic=false
       bIsEditable=false
       bIsStaticSwitch=false
       bIsVisible=false
       type {classStructOrEnum {refPath=&NiagaraDataInterfaceCurve}}
      }
      Enable Visibility Culling {bIsDynamic=false, bIsEditable=false, bIsStaticSwitch=true, bIsVisible=false, type {classStructOrEnum {refPath=&NiagaraBool}}}
      Visibility Cull Response {
       bIsDynamic=false
       bIsEditable=false
       bIsStaticSwitch=false
       bIsVisible=false
       type {classStructOrEnum {refPath=&ENiagaraExecutionStateManagement}}
      }
      Visibility Cull Delay {
       bIsDynamic=false
       bIsEditable=false
       bIsStaticSwitch=false
       bIsVisible=false
       type {classStructOrEnum {refPath=&NiagaraFloat}}
      }
      Reset Age On Awaken {bIsDynamic=false, bIsEditable=false, bIsStaticSwitch=false, bIsVisible=false, type {classStructOrEnum {refPath=&NiagaraBool}}}
     ]
     moduleScript {refPath=&EmitterState}
    }
    {
     bIsSetParametersModule=false
     enabled=true
     moduleName=SpawnBurst_Instantaneous
     inputs [
      Spawn Count {bIsDynamic=false, bIsEditable=true, bIsStaticSwitch=false, bIsVisible=true, type {classStructOrEnum {refPath=&NiagaraInt32}}}
      Spawn Time {bIsDynamic=false, bIsEditable=true, bIsStaticSwitch=false, bIsVisible=true, type {classStructOrEnum {refPath=&NiagaraFloat}}}
      Use Spawn Probability {bIsDynamic=false, bIsEditable=true, bIsStaticSwitch=false, bIsVisible=true, type {classStructOrEnum {refPath=&NiagaraBool}}}
      Spawn Probability {bIsDynamic=false, bIsEditable=false, bIsStaticSwitch=false, bIsVisible=true, type {classStructOrEnum {refPath=&NiagaraFloat}}}
      UseLoopCountLimit {bIsDynamic=false, bIsEditable=true, bIsStaticSwitch=false, bIsVisible=true, type {classStructOrEnum {refPath=&NiagaraBool}}}
      Loop Count Limit {bIsDynamic=false, bIsEditable=false, bIsStaticSwitch=false, bIsVisible=true, type {classStructOrEnum {refPath=&NiagaraInt32}}}
      Spawn Group {bIsDynamic=false, bIsEditable=true, bIsStaticSwitch=false, bIsVisible=true, type {classStructOrEnum {refPath=&NiagaraInt32}}}
      Age {bIsDynamic=false, bIsEditable=true, bIsStaticSwitch=false, bIsVisible=true, type {classStructOrEnum {refPath=&NiagaraFloat}}}
     ]
     moduleScript {refPath=&SpawnBurst_Instantaneous}
    }
   ]
  }
  particleSpawnScript {
   scriptName=ParticleSpawnScript
   modules [
    {
     bIsSetParametersModule=false
     enabled=true
     moduleName=PartitionParticles
     inputs [
      Partition Mode {
       bIsDynamic=false
       bIsEditable=true
       bIsStaticSwitch=true
       bIsVisible=true
       type {classStructOrEnum {refPath=&ENiagara_PartitionMode_v2}}
      }
      Partition Count Mode {
       bIsDynamic=false
       bIsEditable=false
       bIsStaticSwitch=true
       bIsVisible=false
       type {classStructOrEnum {refPath=&ENiagara_InfiniteLoopDuration}}
      }
      A {
       bIsDynamic=false
       bIsEditable=false
       bIsStaticSwitch=false
       bIsVisible=false
       type {classStructOrEnum {refPath=&NiagaraFloat}}
      }
      Number Of Partitions {bIsDynamic=false, bIsEditable=true, bIsStaticSwitch=false, bIsVisible=true, type {classStructOrEnum {refPath=&NiagaraInt32}}}
      Particles Per Partition {
       bIsDynamic=false
       bIsEditable=false
       bIsStaticSwitch=false
       bIsVisible=false
       type {classStructOrEnum {refPath=&NiagaraInt32}}
      }
      Partition Duration {
       bIsDynamic=false
       bIsEditable=false
       bIsStaticSwitch=false
       bIsVisible=false
       type {classStructOrEnum {refPath=&NiagaraFloat}}
      }
      Time {
       bIsDynamic=false
       bIsEditable=false
       bIsStaticSwitch=false
       bIsVisible=false
       type {classStructOrEnum {refPath=&NiagaraFloat}}
      }
      Partitioning Frequency {
       bIsDynamic=false
       bIsEditable=false
       bIsStaticSwitch=true
       bIsVisible=false
       type {classStructOrEnum {refPath=&ENiagaraRandomnessEvaluation}}
      }
      B {
       bIsDynamic=false
       bIsEditable=false
       bIsStaticSwitch=false
       bIsVisible=false
       type {classStructOrEnum {refPath=&NiagaraFloat}}
      }
      Distance {
       bIsDynamic=false
       bIsEditable=false
       bIsStaticSwitch=false
       bIsVisible=false
       type {classStructOrEnum {refPath=&NiagaraFloat}}
      }
      Comparison Type {
       bIsDynamic=false
       bIsEditable=false
       bIsStaticSwitch=true
       bIsVisible=false
       type {classStructOrEnum {refPath=&ENiagaraBooleanLogicOps_v2}}
      }
      Evaluation Type {
       bIsDynamic=false
       bIsEditable=false
       bIsStaticSwitch=true
       bIsVisible=false
       type {classStructOrEnum {refPath=&ENiagaraRandomnessEvaluation}}
      }
      UseDistanceFalloff {bIsDynamic=false, bIsEditable=false, bIsStaticSwitch=false, bIsVisible=false, type {classStructOrEnum {refPath=&NiagaraBool}}}
      Distance Falloff Exponent {
       bIsDynamic=false
       bIsEditable=false
       bIsStaticSwitch=false
       bIsVisible=false
       type {classStructOrEnum {refPath=&NiagaraFloat}}
      }
      Start Position {
       bIsDynamic=false
       bIsEditable=false
       bIsStaticSwitch=false
       bIsVisible=false
       type {classStructOrEnum {refPath=&NiagaraPosition}}
      }
      End Position {
       bIsDynamic=false
       bIsEditable=false
       bIsStaticSwitch=false
       bIsVisible=false
       type {classStructOrEnum {refPath=&NiagaraPosition}}
      }
      ID to Partition {bIsDynamic=false, bIsEditable=true, bIsStaticSwitch=false, bIsVisible=true, type {classStructOrEnum {refPath=&NiagaraInt32}}}
      Randomness Mode {
       bIsDynamic=false
       bIsEditable=false
       bIsStaticSwitch=true
       bIsVisible=false
       type {classStructOrEnum {refPath=&ENiagaraRandomnessMode}}
      }
      Override Seed {bIsDynamic=false, bIsEditable=false, bIsStaticSwitch=true, bIsVisible=false, type {classStructOrEnum {refPath=&NiagaraBool}}}
      Fixed Random Seed {bIsDynamic=false, bIsEditable=false, bIsStaticSwitch=true, bIsVisible=false, type {classStructOrEnum {refPath=&NiagaraBool}}}
      Color A {
       bIsDynamic=false
       bIsEditable=false
       bIsStaticSwitch=false
       bIsVisible=false
       type {classStructOrEnum {refPath=&LinearColor}}
      }
      Color B {
       bIsDynamic=false
       bIsEditable=false
       bIsStaticSwitch=false
       bIsVisible=false
       type {classStructOrEnum {refPath=&LinearColor}}
      }
      Color Comparison Type {bIsDynamic=false, bIsEditable=false, bIsStaticSwitch=true, bIsVisible=false, type {classStructOrEnum {refPath=&NiagaraInt32}}}
      Color Tolerance {
       bIsDynamic=false
       bIsEditable=false
       bIsStaticSwitch=false
       bIsVisible=false
       type {classStructOrEnum {refPath=&NiagaraFloat}}
      }
      Minimum Saturation {
       bIsDynamic=false
       bIsEditable=false
       bIsStaticSwitch=false
       bIsVisible=false
       type {classStructOrEnum {refPath=&NiagaraFloat}}
      }
      Direction A {
       bIsDynamic=false
       bIsEditable=false
       bIsStaticSwitch=false
       bIsVisible=false
       type {classStructOrEnum {refPath=&Vector3f}}
      }
      Direction B {
       bIsDynamic=false
       bIsEditable=false
       bIsStaticSwitch=false
       bIsVisible=false
       type {classStructOrEnum {refPath=&Vector3f}}
      }
      Direction Tolerance {
       bIsDynamic=false
       bIsEditable=false
       bIsStaticSwitch=false
       bIsVisible=false
       type {classStructOrEnum {refPath=&NiagaraFloat}}
      }
      Write Partition to Ribbon ID {
       bIsDynamic=false
       bIsEditable=true
       bIsStaticSwitch=true
       bIsVisible=true
       type {classStructOrEnum {refPath=&ENiagara_AttributeSamplingApplyOutput}}
      }
      Write Partition Index to Ribbon Link Order {
       bIsDynamic=false
       bIsEditable=false
       bIsStaticSwitch=true
       bIsVisible=false
       type {classStructOrEnum {refPath=&ENiagara_AttributeSamplingApplyOutput}}
      }
      Normalized Index In Partition Includes One {bIsDynamic=false, bIsEditable=false, bIsStaticSwitch=true, bIsVisible=false, type {classStructOrEnum {refPath=&NiagaraBool}}}
      Normalized Index Includes One {bIsDynamic=false, bIsEditable=true, bIsStaticSwitch=true, bIsVisible=true, type {classStructOrEnum {refPath=&NiagaraBool}}}
     ]
     moduleScript {refPath=&PartitionParticles}
    }
    {
     bIsSetParametersModule=false
     enabled=true
     moduleName=InitializeParticle
     inputs [
      Lifetime Mode {
       bIsDynamic=false
       bIsEditable=true
       bIsStaticSwitch=true
       bIsVisible=true
       type {classStructOrEnum {refPath=&ENiagara_LifetimeMode}}
      }
      Lifetime Min {
       bIsDynamic=false
       bIsEditable=false
       bIsStaticSwitch=false
       bIsVisible=false
       type {classStructOrEnum {refPath=&NiagaraFloat}}
      }
      Lifetime {bIsDynamic=false, bIsEditable=true, bIsStaticSwitch=false, bIsVisible=true, type {classStructOrEnum {refPath=&NiagaraFloat}}}
      Lifetime Max {
       bIsDynamic=false
       bIsEditable=false
       bIsStaticSwitch=false
       bIsVisible=false
       type {classStructOrEnum {refPath=&NiagaraFloat}}
      }
      Lifetime Randomness Mode {
       bIsDynamic=false
       bIsEditable=false
       bIsStaticSwitch=true
       bIsVisible=false
       type {classStructOrEnum {refPath=&ENiagaraRandomnessMode}}
      }
      Color Mode {
       bIsDynamic=false
       bIsEditable=true
       bIsStaticSwitch=true
       bIsVisible=true
       type {classStructOrEnum {refPath=&ENiagara_ColorInitializationMode}}
      }
      Color Minimum {
       bIsDynamic=false
       bIsEditable=false
       bIsStaticSwitch=false
       bIsVisible=false
       type {classStructOrEnum {refPath=&LinearColor}}
      }
      Color {
       bIsDynamic=false
       bIsEditable=true
       bIsStaticSwitch=false
       bIsVisible=true
       type {classStructOrEnum {refPath=&LinearColor}}
      }
      AdjustHue {bIsDynamic=false, bIsEditable=false, bIsStaticSwitch=false, bIsVisible=false, type {classStructOrEnum {refPath=&NiagaraBool}}}
      Hue Shift Range {
       bIsDynamic=false
       bIsEditable=false
       bIsStaticSwitch=false
       bIsVisible=false
       type {classStructOrEnum {refPath=&Vector2f}}
      }
      Color Maximum {
       bIsDynamic=false
       bIsEditable=false
       bIsStaticSwitch=false
       bIsVisible=false
       type {classStructOrEnum {refPath=&LinearColor}}
      }
      AdjustSaturation {bIsDynamic=false, bIsEditable=false, bIsStaticSwitch=false, bIsVisible=false, type {classStructOrEnum {refPath=&NiagaraBool}}}
      Saturation Range {
       bIsDynamic=false
       bIsEditable=false
       bIsStaticSwitch=false
       bIsVisible=false
       type {classStructOrEnum {refPath=&Vector2f}}
      }
      AdjustValue {bIsDynamic=false, bIsEditable=false, bIsStaticSwitch=false, bIsVisible=false, type {classStructOrEnum {refPath=&NiagaraBool}}}
      Value Range {
       bIsDynamic=false
       bIsEditable=false
       bIsStaticSwitch=false
       bIsVisible=false
       type {classStructOrEnum {refPath=&Vector2f}}
      }
      AdjustAlpha {bIsDynamic=false, bIsEditable=false, bIsStaticSwitch=false, bIsVisible=false, type {classStructOrEnum {refPath=&NiagaraBool}}}
      Alpha Scale Range {
       bIsDynamic=false
       bIsEditable=false
       bIsStaticSwitch=false
       bIsVisible=false
       type {classStructOrEnum {refPath=&Vector2f}}
      }
      Color Channel Mode {
       bIsDynamic=false
       bIsEditable=false
       bIsStaticSwitch=true
       bIsVisible=false
       type {classStructOrEnum {refPath=&ENiagaraChannelCorrelation}}
      }
      Color Randomness Mode {
       bIsDynamic=false
       bIsEditable=false
       bIsStaticSwitch=true
       bIsVisible=false
       type {classStructOrEnum {refPath=&ENiagaraRandomnessMode}}
      }
      Position Mode {
       bIsDynamic=false
       bIsEditable=true
       bIsStaticSwitch=true
       bIsVisible=true
       type {classStructOrEnum {refPath=&ENiagara_PositionInitializationMode}}
      }
      Position {
       bIsDynamic=false
       bIsEditable=false
       bIsStaticSwitch=false
       bIsVisible=false
       type {classStructOrEnum {refPath=&NiagaraPosition}}
      }
      UsePositionOffset {bIsDynamic=false, bIsEditable=true, bIsStaticSwitch=false, bIsVisible=true, type {classStructOrEnum {refPath=&NiagaraBool}}}
      Position Offset {bIsDynamic=true, bIsEditable=true, bIsStaticSwitch=false, bIsVisible=true, type {classStructOrEnum {refPath=&Vector3f}}}
      Position Offset Coordinate Space {
       bIsDynamic=false
       bIsEditable=true
       bIsStaticSwitch=false
       bIsVisible=true
       type {classStructOrEnum {refPath=&ENiagaraCoordinateSpace}}
      }
      Mass Mode {
       bIsDynamic=false
       bIsEditable=true
       bIsStaticSwitch=true
       bIsVisible=true
       type {classStructOrEnum {refPath=&ENiagara_MassInitializationMode}}
      }
      Mass Min {
       bIsDynamic=false
       bIsEditable=false
       bIsStaticSwitch=false
       bIsVisible=false
       type {classStructOrEnum {refPath=&NiagaraFloat}}
      }
      Mass {
       bIsDynamic=false
       bIsEditable=false
       bIsStaticSwitch=false
       bIsVisible=false
       type {classStructOrEnum {refPath=&NiagaraFloat}}
      }
      Mass Max {
       bIsDynamic=false
       bIsEditable=false
       bIsStaticSwitch=false
       bIsVisible=false
       type {classStructOrEnum {refPath=&NiagaraFloat}}
      }
      Mass Randomness Mode {
       bIsDynamic=false
       bIsEditable=false
       bIsStaticSwitch=true
       bIsVisible=false
       type {classStructOrEnum {refPath=&ENiagaraRandomnessMode}}
      }
      Write Material Random {bIsDynamic=false, bIsEditable=true, bIsStaticSwitch=false, bIsVisible=true, type {classStructOrEnum {refPath=&NiagaraBool}}}
      Material Random {bIsDynamic=false, bIsEditable=true, bIsStaticSwitch=false, bIsVisible=true, type {classStructOrEnum {refPath=&NiagaraFloat}}}
      Sprite Size Mode {
       bIsDynamic=false
       bIsEditable=true
       bIsStaticSwitch=true
       bIsVisible=true
       type {classStructOrEnum {refPath=&ENiagara_SizeScaleMode}}
      }
      Sprite Size {
       bIsDynamic=false
       bIsEditable=false
       bIsStaticSwitch=false
       bIsVisible=false
       type {classStructOrEnum {refPath=&Vector2f}}
      }
      Uniform Sprite Size {
       bIsDynamic=false
       bIsEditable=false
       bIsStaticSwitch=false
       bIsVisible=false
       type {classStructOrEnum {refPath=&NiagaraFloat}}
      }
      Uniform Sprite Size Min {
       bIsDynamic=false
       bIsEditable=false
       bIsStaticSwitch=false
       bIsVisible=false
       type {classStructOrEnum {refPath=&NiagaraFloat}}
      }
      Sprite Size Min {
       bIsDynamic=false
       bIsEditable=false
       bIsStaticSwitch=false
       bIsVisible=false
       type {classStructOrEnum {refPath=&Vector2f}}
      }
      Sprite Size Max {
       bIsDynamic=false
       bIsEditable=false
       bIsStaticSwitch=false
       bIsVisible=false
       type {classStructOrEnum {refPath=&Vector2f}}
      }
      Uniform Sprite Size Max {
       bIsDynamic=false
       bIsEditable=false
       bIsStaticSwitch=false
       bIsVisible=false
       type {classStructOrEnum {refPath=&NiagaraFloat}}
      }
      Sprite Size Randomness Mode {
       bIsDynamic=false
       bIsEditable=false
       bIsStaticSwitch=true
       bIsVisible=false
       type {classStructOrEnum {refPath=&ENiagaraRandomnessMode}}
      }
      Sprite Rotation Mode {
       bIsDynamic=false
       bIsEditable=true
       bIsStaticSwitch=true
       bIsVisible=true
       type {classStructOrEnum {refPath=&ENiagara_SpriteRotationMode}}
      }
      Sprite Rotation Angle Min {
       bIsDynamic=false
       bIsEditable=false
       bIsStaticSwitch=false
       bIsVisible=false
       type {classStructOrEnum {refPath=&NiagaraFloat}}
      }
      Sprite Rotation Angle {
       bIsDynamic=false
       bIsEditable=false
       bIsStaticSwitch=false
       bIsVisible=false
       type {classStructOrEnum {refPath=&NiagaraFloat}}
      }
      Sprite Rotation Angle Max {
       bIsDynamic=false
       bIsEditable=false
       bIsStaticSwitch=false
       bIsVisible=false
       type {classStructOrEnum {refPath=&NiagaraFloat}}
      }
      Sprite Rotation Randomness Mode {
       bIsDynamic=false
       bIsEditable=false
       bIsStaticSwitch=true
       bIsVisible=false
       type {classStructOrEnum {refPath=&ENiagaraRandomnessMode}}
      }
      Sprite UV Mode {
       bIsDynamic=false
       bIsEditable=true
       bIsStaticSwitch=true
       bIsVisible=true
       type {classStructOrEnum {refPath=&ENiagara_UVFlippingMode}}
      }
      Sprite UV Randomness Mode {
       bIsDynamic=false
       bIsEditable=false
       bIsStaticSwitch=true
       bIsVisible=false
       type {classStructOrEnum {refPath=&ENiagaraRandomnessMode}}
      }
      Sprite UV Scale {
       bIsDynamic=false
       bIsEditable=false
       bIsStaticSwitch=false
       bIsVisible=false
       type {classStructOrEnum {refPath=&Vector2f}}
      }
      Mesh Scale Mode {
       bIsDynamic=false
       bIsEditable=true
       bIsStaticSwitch=true
       bIsVisible=true
       type {classStructOrEnum {refPath=&ENiagara_SizeScaleMode}}
      }
      Mesh Uniform Scale Min {
       bIsDynamic=false
       bIsEditable=false
       bIsStaticSwitch=false
       bIsVisible=false
       type {classStructOrEnum {refPath=&NiagaraFloat}}
      }
      Mesh Uniform Scale {bIsDynamic=true, bIsEditable=true, bIsStaticSwitch=false, bIsVisible=true, type {classStructOrEnum {refPath=&NiagaraFloat}}}
      Mesh Scale Min {
       bIsDynamic=false
       bIsEditable=false
       bIsStaticSwitch=false
       bIsVisible=false
       type {classStructOrEnum {refPath=&Vector3f}}
      }
      Mesh Scale {
       bIsDynamic=false
       bIsEditable=false
       bIsStaticSwitch=false
       bIsVisible=false
       type {classStructOrEnum {refPath=&Vector3f}}
      }
      Mesh Scale Max {
       bIsDynamic=false
       bIsEditable=false
       bIsStaticSwitch=false
       bIsVisible=false
       type {classStructOrEnum {refPath=&Vector3f}}
      }
      Mesh Uniform Scale Max {
       bIsDynamic=false
       bIsEditable=false
       bIsStaticSwitch=false
       bIsVisible=false
       type {classStructOrEnum {refPath=&NiagaraFloat}}
      }
      Mesh Scale Randomness Mode {
       bIsDynamic=false
       bIsEditable=false
       bIsStaticSwitch=true
       bIsVisible=false
       type {classStructOrEnum {refPath=&ENiagaraRandomnessMode}}
      }
      Mesh Renderer Array Visibility Mode {
       bIsDynamic=false
       bIsEditable=true
       bIsStaticSwitch=true
       bIsVisible=true
       type {classStructOrEnum {refPath=&ENiagara_UnsetDirectSetRandom}}
      }
      Mesh Renderer Mesh Index {
       bIsDynamic=false
       bIsEditable=false
       bIsStaticSwitch=false
       bIsVisible=false
       type {classStructOrEnum {refPath=&NiagaraInt32}}
      }
      Wrap Index By Mesh Count {bIsDynamic=false, bIsEditable=false, bIsStaticSwitch=true, bIsVisible=false, type {classStructOrEnum {refPath=&NiagaraBool}}}
      Mesh Renderer Info {
       bIsDynamic=false
       bIsEditable=false
       bIsStaticSwitch=false
       bIsVisible=false
       type {classStructOrEnum {refPath=&NiagaraDataInterfaceMeshRendererInfo}}
      }
      Mesh Renderer Random Seed {
       bIsDynamic=false
       bIsEditable=false
       bIsStaticSwitch=false
       bIsVisible=false
       type {classStructOrEnum {refPath=&NiagaraInt32}}
      }
      Ribbon Width Mode {
       bIsDynamic=false
       bIsEditable=true
       bIsStaticSwitch=true
       bIsVisible=true
       type {classStructOrEnum {refPath=&ENiagara_UnsetDirectSet}}
      }
      Ribbon Width {
       bIsDynamic=false
       bIsEditable=false
       bIsStaticSwitch=false
       bIsVisible=false
       type {classStructOrEnum {refPath=&NiagaraFloat}}
      }
      Ribbon Facing Mode {
       bIsDynamic=false
       bIsEditable=true
       bIsStaticSwitch=true
       bIsVisible=true
       type {classStructOrEnum {refPath=&ENiagara_UnsetDirectSet}}
      }
      Ribbon Facing Vector {
       bIsDynamic=false
       bIsEditable=false
       bIsStaticSwitch=false
       bIsVisible=false
       type {classStructOrEnum {refPath=&Vector3f}}
      }
      Ribbon Twist Mode {
       bIsDynamic=false
       bIsEditable=true
       bIsStaticSwitch=true
       bIsVisible=true
       type {classStructOrEnum {refPath=&ENiagara_UnsetDirectSet}}
      }
      "Ribbon Twist " {
       bIsDynamic=false
       bIsEditable=false
       bIsStaticSwitch=false
       bIsVisible=false
       type {classStructOrEnum {refPath=&NiagaraFloat}}
      }
     ]
     moduleScript {refPath=&InitializeParticle}
    }
    {
     bIsSetParametersModule=false
     enabled=true
     moduleName=InitialMeshOrientation
     inputs [
      Mesh Orientation Mode {
       bIsDynamic=false
       bIsEditable=true
       bIsStaticSwitch=true
       bIsVisible=true
       type {classStructOrEnum {refPath=/Niagara/Enums/Transforms/ENiagara_MeshOrientationOptions}}
      }
      Orientation Quaternion {bIsDynamic=false, bIsEditable=false, bIsStaticSwitch=false, bIsVisible=false, type {classStructOrEnum {refPath=&Quat4f}}}
      Orientation Matrix {
       bIsDynamic=false
       bIsEditable=false
       bIsStaticSwitch=false
       bIsVisible=false
       type {classStructOrEnum {refPath=&NiagaraMatrix}}
      }
      Orientation Vector {
       bIsDynamic=false
       bIsEditable=false
       bIsStaticSwitch=false
       bIsVisible=false
       type {classStructOrEnum {refPath=&Vector3f}}
      }
      Orientation Coordinate Space {
       bIsDynamic=false
       bIsEditable=false
       bIsStaticSwitch=true
       bIsVisible=false
       type {classStructOrEnum {refPath=&ENiagaraCoordinateSpace}}
      }
      Mesh Axis To Orient {
       bIsDynamic=false
       bIsEditable=false
       bIsStaticSwitch=false
       bIsVisible=false
       type {classStructOrEnum {refPath=&Vector3f}}
      }
      Random Seed {
       bIsDynamic=false
       bIsEditable=false
       bIsStaticSwitch=false
       bIsVisible=false
       type {classStructOrEnum {refPath=&NiagaraInt32}}
      }
      Use Rotation Vector {bIsDynamic=false, bIsEditable=true, bIsStaticSwitch=false, bIsVisible=true, type {classStructOrEnum {refPath=&NiagaraBool}}}
      Rotation {bIsDynamic=true, bIsEditable=true, bIsStaticSwitch=false, bIsVisible=true, type {classStructOrEnum {refPath=&Vector3f}}}
      Rotation Coordinate Space {
       bIsDynamic=false
       bIsEditable=true
       bIsStaticSwitch=true
       bIsVisible=true
       type {classStructOrEnum {refPath=/Niagara/Enums/ENiagaraMeshTransforms}}
      }
     ]
     moduleScript {refPath=&InitialMeshOrientation}
    }
   ]
  }
  particleUpdateScript {
   scriptName=ParticleUpdateScript
   modules [
    {
     bIsSetParametersModule=false
     enabled=true
     moduleName=ParticleState
     inputs [
      Kill Particles When Lifetime Has Elapsed {bIsDynamic=false, bIsEditable=true, bIsStaticSwitch=true, bIsVisible=true, type {classStructOrEnum {refPath=&NiagaraBool}}}
      Loop Particles Lifetime {bIsDynamic=false, bIsEditable=false, bIsStaticSwitch=false, bIsVisible=false, type {classStructOrEnum {refPath=&NiagaraBool}}}
      Let Infinitely Lived Particles Die When Emitter Deactivates {bIsDynamic=false, bIsEditable=false, bIsStaticSwitch=false, bIsVisible=false, type {classStructOrEnum {refPath=&NiagaraBool}}}
      Lifetime {bIsDynamic=false, bIsEditable=true, bIsStaticSwitch=false, bIsVisible=true, type {classStructOrEnum {refPath=&NiagaraFloat}}}
      DeltaTime {bIsDynamic=false, bIsEditable=true, bIsStaticSwitch=false, bIsVisible=true, type {classStructOrEnum {refPath=&NiagaraFloat}}}
     ]
     moduleScript {refPath=&ParticleState}
    }
    {
     bIsSetParametersModule=false
     enabled=true
     moduleName=DynamicMaterialParameters
     inputs [
      Write Parameter Index 0 {
       bIsDynamic=false
       bIsEditable=true
       bIsStaticSwitch=true
       bIsVisible=true
       type {classStructOrEnum {refPath=&ENiagara_DynamicMaterialParametersInputMode}}
      }
      Index 0 Vector 1 {
       bIsDynamic=false
       bIsEditable=false
       bIsStaticSwitch=false
       bIsVisible=false
       type {classStructOrEnum {refPath=&Vector3f}}
      }
      Module.Index0Float0Write {bIsDynamic=false, bIsEditable=true, bIsStaticSwitch=true, bIsVisible=true, type {classStructOrEnum {refPath=&NiagaraBool}}}
      Index 0 Param 1 {bIsDynamic=false, bIsEditable=true, bIsStaticSwitch=false, bIsVisible=true, type {classStructOrEnum {refPath=&NiagaraFloat}}}
      Module.Index0Float1Write {bIsDynamic=false, bIsEditable=true, bIsStaticSwitch=true, bIsVisible=true, type {classStructOrEnum {refPath=&NiagaraBool}}}
      Index 0 Param 2 {bIsDynamic=false, bIsEditable=true, bIsStaticSwitch=false, bIsVisible=true, type {classStructOrEnum {refPath=&NiagaraFloat}}}
      Module.Index0Float2Write {bIsDynamic=false, bIsEditable=true, bIsStaticSwitch=true, bIsVisible=true, type {classStructOrEnum {refPath=&NiagaraBool}}}
      Index 0 Param 3 {bIsDynamic=false, bIsEditable=true, bIsStaticSwitch=false, bIsVisible=true, type {classStructOrEnum {refPath=&NiagaraFloat}}}
      Module.Index0Float3Write {bIsDynamic=false, bIsEditable=true, bIsStaticSwitch=true, bIsVisible=true, type {classStructOrEnum {refPath=&NiagaraBool}}}
      Index 0 Param 4 {bIsDynamic=false, bIsEditable=true, bIsStaticSwitch=false, bIsVisible=true, type {classStructOrEnum {refPath=&NiagaraFloat}}}
      Index 0 Vector2D 1 {
       bIsDynamic=false
       bIsEditable=false
       bIsStaticSwitch=false
       bIsVisible=false
       type {classStructOrEnum {refPath=&Vector2f}}
      }
      Index 0 Vector2D 2 {
       bIsDynamic=false
       bIsEditable=false
       bIsStaticSwitch=false
       bIsVisible=false
       type {classStructOrEnum {refPath=&Vector2f}}
      }
      Index 0 Vector 4 {
       bIsDynamic=false
       bIsEditable=false
       bIsStaticSwitch=false
       bIsVisible=false
       type {classStructOrEnum {refPath=&Vector4f}}
      }
      Index 0 Linear Color {
       bIsDynamic=false
       bIsEditable=false
       bIsStaticSwitch=false
       bIsVisible=false
       type {classStructOrEnum {refPath=&LinearColor}}
      }
      Write Parameter Index 1 {
       bIsDynamic=false
       bIsEditable=true
       bIsStaticSwitch=true
       bIsVisible=true
       type {classStructOrEnum {refPath=&ENiagara_DynamicMaterialParametersInputMode}}
      }
      Index 1 Vector 1 {
       bIsDynamic=false
       bIsEditable=false
       bIsStaticSwitch=false
       bIsVisible=false
       type {classStructOrEnum {refPath=&Vector3f}}
      }
      Index 1 Param 1 {
       bIsDynamic=false
       bIsEditable=false
       bIsStaticSwitch=false
       bIsVisible=false
       type {classStructOrEnum {refPath=&NiagaraFloat}}
      }
      Index 1 Param 2 {
       bIsDynamic=false
       bIsEditable=false
       bIsStaticSwitch=false
       bIsVisible=false
       type {classStructOrEnum {refPath=&NiagaraFloat}}
      }
      Index 1 Param 3 {
       bIsDynamic=false
       bIsEditable=false
       bIsStaticSwitch=false
       bIsVisible=false
       type {classStructOrEnum {refPath=&NiagaraFloat}}
      }
      Index 1 Param 4 {
       bIsDynamic=false
       bIsEditable=false
       bIsStaticSwitch=false
       bIsVisible=false
       type {classStructOrEnum {refPath=&NiagaraFloat}}
      }
      Index 1 Vector2D 1 {
       bIsDynamic=false
       bIsEditable=false
       bIsStaticSwitch=false
       bIsVisible=false
       type {classStructOrEnum {refPath=&Vector2f}}
      }
      Index 1 Vector2D 2 {
       bIsDynamic=false
       bIsEditable=false
       bIsStaticSwitch=false
       bIsVisible=false
       type {classStructOrEnum {refPath=&Vector2f}}
      }
      Index 1 Vector 4 {
       bIsDynamic=false
       bIsEditable=false
       bIsStaticSwitch=false
       bIsVisible=false
       type {classStructOrEnum {refPath=&Vector4f}}
      }
      Index 1 Linear Color {
       bIsDynamic=false
       bIsEditable=false
       bIsStaticSwitch=false
       bIsVisible=false
       type {classStructOrEnum {refPath=&LinearColor}}
      }
      Write Parameter Index 2 {
       bIsDynamic=false
       bIsEditable=true
       bIsStaticSwitch=true
       bIsVisible=true
       type {classStructOrEnum {refPath=&ENiagara_DynamicMaterialParametersInputMode}}
      }
      Index 2 Vector 1 {
       bIsDynamic=false
       bIsEditable=false
       bIsStaticSwitch=false
       bIsVisible=false
       type {classStructOrEnum {refPath=&Vector3f}}
      }
      Index 2 Param 1 {
       bIsDynamic=false
       bIsEditable=false
       bIsStaticSwitch=false
       bIsVisible=false
       type {classStructOrEnum {refPath=&NiagaraFloat}}
      }
      Index 2 Param 2 {
       bIsDynamic=false
       bIsEditable=false
       bIsStaticSwitch=false
       bIsVisible=false
       type {classStructOrEnum {refPath=&NiagaraFloat}}
      }
      Index 2 Param 3 {
       bIsDynamic=false
       bIsEditable=false
       bIsStaticSwitch=false
       bIsVisible=false
       type {classStructOrEnum {refPath=&NiagaraFloat}}
      }
      Index 2 Param 4 {
       bIsDynamic=false
       bIsEditable=false
       bIsStaticSwitch=false
       bIsVisible=false
       type {classStructOrEnum {refPath=&NiagaraFloat}}
      }
      Index 2 Vector2D 1 {
       bIsDynamic=false
       bIsEditable=false
       bIsStaticSwitch=false
       bIsVisible=false
       type {classStructOrEnum {refPath=&Vector2f}}
      }
      Index 2 Vector2D 2 {
       bIsDynamic=false
       bIsEditable=false
       bIsStaticSwitch=false
       bIsVisible=false
       type {classStructOrEnum {refPath=&Vector2f}}
      }
      Index 2 Vector 4 {
       bIsDynamic=false
       bIsEditable=false
       bIsStaticSwitch=false
       bIsVisible=false
       type {classStructOrEnum {refPath=&Vector4f}}
      }
      Index 2 Linear Color {
       bIsDynamic=false
       bIsEditable=false
       bIsStaticSwitch=false
       bIsVisible=false
       type {classStructOrEnum {refPath=&LinearColor}}
      }
      Write Parameter Index 3 {
       bIsDynamic=false
       bIsEditable=true
       bIsStaticSwitch=true
       bIsVisible=true
       type {classStructOrEnum {refPath=&ENiagara_DynamicMaterialParametersInputMode}}
      }
      Index 3 Vector 1 {
       bIsDynamic=false
       bIsEditable=false
       bIsStaticSwitch=false
       bIsVisible=false
       type {classStructOrEnum {refPath=&Vector3f}}
      }
      Index 3 Param 1 {
       bIsDynamic=false
       bIsEditable=false
       bIsStaticSwitch=false
       bIsVisible=false
       type {classStructOrEnum {refPath=&NiagaraFloat}}
      }
      Index 3 Param 2 {
       bIsDynamic=false
       bIsEditable=false
       bIsStaticSwitch=false
       bIsVisible=false
       type {classStructOrEnum {refPath=&NiagaraFloat}}
      }
      Index 3 Param 3 {
       bIsDynamic=false
       bIsEditable=false
       bIsStaticSwitch=false
       bIsVisible=false
       type {classStructOrEnum {refPath=&NiagaraFloat}}
      }
      Index 3 Param 4 {
       bIsDynamic=false
       bIsEditable=false
       bIsStaticSwitch=false
       bIsVisible=false
       type {classStructOrEnum {refPath=&NiagaraFloat}}
      }
      Index 3 Vector2D 1 {
       bIsDynamic=false
       bIsEditable=false
       bIsStaticSwitch=false
       bIsVisible=false
       type {classStructOrEnum {refPath=&Vector2f}}
      }
      Index 3 Vector2D 2 {
       bIsDynamic=false
       bIsEditable=false
       bIsStaticSwitch=false
       bIsVisible=false
       type {classStructOrEnum {refPath=&Vector2f}}
      }
      Index 3 Vector 4 {
       bIsDynamic=false
       bIsEditable=false
       bIsStaticSwitch=false
       bIsVisible=false
       type {classStructOrEnum {refPath=&Vector4f}}
      }
      Index 3 Linear Color {
       bIsDynamic=false
       bIsEditable=false
       bIsStaticSwitch=false
       bIsVisible=false
       type {classStructOrEnum {refPath=&LinearColor}}
      }
      Dynamic Parameter Output {
       bIsDynamic=false
       bIsEditable=true
       bIsStaticSwitch=true
       bIsVisible=true
       type {classStructOrEnum {refPath=&ENiagara_AttributeSamplingApplyOutput}}
      }
     ]
     moduleScript {refPath=&DynamicMaterialParameters}
    }
    {
     bIsSetParametersModule=false
     enabled=true
     moduleName=ScaleColor
     inputs [
      Scale Mode {bIsDynamic=false, bIsEditable=true, bIsStaticSwitch=true, bIsVisible=true, type {classStructOrEnum {refPath=&NiagaraInt32}}}
      Linear Color Curve {
       bIsDynamic=false
       bIsEditable=false
       bIsStaticSwitch=false
       bIsVisible=false
       type {classStructOrEnum {refPath=&NiagaraDataInterfaceColorCurve}}
      }
      ScaleRGB {bIsDynamic=false, bIsEditable=true, bIsStaticSwitch=false, bIsVisible=true, type {classStructOrEnum {refPath=&NiagaraBool}}}
      Scale RGB {bIsDynamic=true, bIsEditable=true, bIsStaticSwitch=false, bIsVisible=true, type {classStructOrEnum {refPath=&Vector3f}}}
      ScaleRGBA {bIsDynamic=false, bIsEditable=false, bIsStaticSwitch=false, bIsVisible=false, type {classStructOrEnum {refPath=&NiagaraBool}}}
      Scale RGBA {bIsDynamic=true, bIsEditable=false, bIsStaticSwitch=false, bIsVisible=false, type {classStructOrEnum {refPath=&Vector4f}}}
      Curve Index {
       bIsDynamic=false
       bIsEditable=false
       bIsStaticSwitch=false
       bIsVisible=false
       type {classStructOrEnum {refPath=&NiagaraFloat}}
      }
      ScaleA {bIsDynamic=false, bIsEditable=true, bIsStaticSwitch=false, bIsVisible=true, type {classStructOrEnum {refPath=&NiagaraBool}}}
      Scale Alpha {bIsDynamic=true, bIsEditable=true, bIsStaticSwitch=false, bIsVisible=true, type {classStructOrEnum {refPath=&NiagaraFloat}}}
      Color Value To Scale {
       bIsDynamic=false
       bIsEditable=true
       bIsStaticSwitch=false
       bIsVisible=true
       type {classStructOrEnum {refPath=&LinearColor}}
      }
      Color Scale {
       bIsDynamic=false
       bIsEditable=false
       bIsStaticSwitch=false
       bIsVisible=false
       type {classStructOrEnum {refPath=&LinearColor}}
      }
     ]
     moduleScript {refPath=&ScaleColor}
    }
    {
     bIsSetParametersModule=false
     enabled=true
     moduleName=ScaleMeshSize
     inputs [
      Scale Factor {bIsDynamic=true, bIsEditable=true, bIsStaticSwitch=false, bIsVisible=true, type {classStructOrEnum {refPath=&Vector3f}}}
      Initial Mesh Scale {bIsDynamic=false, bIsEditable=true, bIsStaticSwitch=false, bIsVisible=true, type {classStructOrEnum {refPath=&Vector3f}}}
     ]
     moduleScript {refPath=&ScaleMeshSize}
    }
    {
     bIsSetParametersModule=false
     enabled=true
     moduleName=CameraOffset
     inputs [Camera Offset Amount {bIsDynamic=false, bIsEditable=true, bIsStaticSwitch=false, bIsVisible=true, type {classStructOrEnum {refPath=&NiagaraFloat}}}]
     moduleScript {refPath=&CameraOffset}
    }
    {
     bIsSetParametersModule=true
     enabled=true
     moduleName=SetVariables_#8
     inputs [
      Particles.ShockwaveScale {bIsDynamic=true, bIsEditable=true, bIsStaticSwitch=false, bIsVisible=true, type {classStructOrEnum {refPath=&Vector3f}}}
      Particles.ShockwaveColor {
       bIsDynamic=true
       bIsEditable=true
       bIsStaticSwitch=false
       bIsVisible=true
       type {classStructOrEnum {refPath=&LinearColor}}
      }
      Particles.SecondarySphereColor {
       bIsDynamic=true
       bIsEditable=true
       bIsStaticSwitch=false
       bIsVisible=true
       type {classStructOrEnum {refPath=&LinearColor}}
      }
      Particles.SecondaryDyn {bIsDynamic=true, bIsEditable=true, bIsStaticSwitch=false, bIsVisible=true, type {classStructOrEnum {refPath=&Vector4f}}}
     ]
     moduleScript {refPath=&self:Shockwave_0.NiagaraScriptSource_0.NiagaraGraph_0.NiagaraNodeAssignment_0.SetVariables_#8}
    }
   ]
  }
  rendererClasses [{refPath=&NiagaraMeshRendererProperties}]
  renderers [
   {rendererIndex=0, rendererClass {refPath=&NiagaraMeshRendererProperties}}
   {rendererIndex=1, rendererClass {refPath=&NiagaraMeshRendererProperties}}
   {rendererIndex=2, rendererClass {refPath=&NiagaraMeshRendererProperties}}
  ]
 }
 Shockwave002 {
  emitterSpawnScript {scriptName=EmitterSpawnScript, modules []}
  emitterUpdateScript {
   scriptName=EmitterUpdateScript
   modules [
    {
     bIsSetParametersModule=false
     enabled=true
     moduleName=EmitterState
     inputs [
      Life Cycle Mode {
       bIsDynamic=false
       bIsEditable=true
       bIsStaticSwitch=true
       bIsVisible=true
       type {classStructOrEnum {refPath=&ENiagaraEmitterLifeCycleMode}}
      }
      Inactive Response {
       bIsDynamic=false
       bIsEditable=false
       bIsStaticSwitch=true
       bIsVisible=false
       type {classStructOrEnum {refPath=&ENiagaraInactiveMode}}
      }
      Loop Behavior {
       bIsDynamic=false
       bIsEditable=false
       bIsStaticSwitch=true
       bIsVisible=false
       type {classStructOrEnum {refPath=&ENiagara_EmitterStateOptions}}
      }
      Loop Duration Mode {
       bIsDynamic=false
       bIsEditable=false
       bIsStaticSwitch=true
       bIsVisible=false
       type {classStructOrEnum {refPath=&ENiagara_InfiniteLoopDuration}}
      }
      Loop Count {
       bIsDynamic=false
       bIsEditable=false
       bIsStaticSwitch=false
       bIsVisible=false
       type {classStructOrEnum {refPath=&NiagaraInt32}}
      }
      Loop Duration {
       bIsDynamic=false
       bIsEditable=false
       bIsStaticSwitch=false
       bIsVisible=false
       type {classStructOrEnum {refPath=&NiagaraFloat}}
      }
      Recalculate Duration Each Loop {bIsDynamic=false, bIsEditable=false, bIsStaticSwitch=false, bIsVisible=false, type {classStructOrEnum {refPath=&NiagaraBool}}}
      UseLoopDelay {bIsDynamic=false, bIsEditable=false, bIsStaticSwitch=false, bIsVisible=false, type {classStructOrEnum {refPath=&NiagaraBool}}}
      Loop Delay {
       bIsDynamic=false
       bIsEditable=false
       bIsStaticSwitch=false
       bIsVisible=false
       type {classStructOrEnum {refPath=&NiagaraFloat}}
      }
      Delay First Loop Only {bIsDynamic=false, bIsEditable=false, bIsStaticSwitch=false, bIsVisible=false, type {classStructOrEnum {refPath=&NiagaraBool}}}
      Scalability Mode {
       bIsDynamic=false
       bIsEditable=true
       bIsStaticSwitch=true
       bIsVisible=true
       type {classStructOrEnum {refPath=&ENiagaraEmitterScalabilityMode_Limited}}
      }
      Enable Distance Culling {bIsDynamic=false, bIsEditable=false, bIsStaticSwitch=true, bIsVisible=false, type {classStructOrEnum {refPath=&NiagaraBool}}}
      MinDistance {
       bIsDynamic=false
       bIsEditable=false
       bIsStaticSwitch=false
       bIsVisible=false
       type {classStructOrEnum {refPath=&NiagaraFloat}}
      }
      Min Distance Response {
       bIsDynamic=false
       bIsEditable=false
       bIsStaticSwitch=false
       bIsVisible=false
       type {classStructOrEnum {refPath=&ENiagaraExecutionStateManagement}}
      }
      MaxDistance {
       bIsDynamic=false
       bIsEditable=false
       bIsStaticSwitch=false
       bIsVisible=false
       type {classStructOrEnum {refPath=&NiagaraFloat}}
      }
      Max Distance Response {
       bIsDynamic=false
       bIsEditable=false
       bIsStaticSwitch=false
       bIsVisible=false
       type {classStructOrEnum {refPath=&ENiagaraExecutionStateManagement}}
      }
      Scale Spawn Count {bIsDynamic=false, bIsEditable=false, bIsStaticSwitch=true, bIsVisible=false, type {classStructOrEnum {refPath=&NiagaraBool}}}
      Spawn Count Scale {
       bIsDynamic=false
       bIsEditable=false
       bIsStaticSwitch=false
       bIsVisible=false
       type {classStructOrEnum {refPath=&NiagaraFloat}}
      }
      Spawn Count Scale By Distance {
       bIsDynamic=false
       bIsEditable=false
       bIsStaticSwitch=false
       bIsVisible=false
       type {classStructOrEnum {refPath=&NiagaraDataInterfaceCurve}}
      }
      Enable Visibility Culling {bIsDynamic=false, bIsEditable=false, bIsStaticSwitch=true, bIsVisible=false, type {classStructOrEnum {refPath=&NiagaraBool}}}
      Visibility Cull Response {
       bIsDynamic=false
       bIsEditable=false
       bIsStaticSwitch=false
       bIsVisible=false
       type {classStructOrEnum {refPath=&ENiagaraExecutionStateManagement}}
      }
      Visibility Cull Delay {
       bIsDynamic=false
       bIsEditable=false
       bIsStaticSwitch=false
       bIsVisible=false
       type {classStructOrEnum {refPath=&NiagaraFloat}}
      }
      Reset Age On Awaken {bIsDynamic=false, bIsEditable=false, bIsStaticSwitch=false, bIsVisible=false, type {classStructOrEnum {refPath=&NiagaraBool}}}
     ]
     moduleScript {refPath=&EmitterState}
    }
    {
     bIsSetParametersModule=false
     enabled=true
     moduleName=SpawnBurst_Instantaneous
     inputs [
      Spawn Count {bIsDynamic=false, bIsEditable=true, bIsStaticSwitch=false, bIsVisible=true, type {classStructOrEnum {refPath=&NiagaraInt32}}}
      Spawn Time {bIsDynamic=false, bIsEditable=true, bIsStaticSwitch=false, bIsVisible=true, type {classStructOrEnum {refPath=&NiagaraFloat}}}
      Use Spawn Probability {bIsDynamic=false, bIsEditable=true, bIsStaticSwitch=false, bIsVisible=true, type {classStructOrEnum {refPath=&NiagaraBool}}}
      Spawn Probability {bIsDynamic=false, bIsEditable=false, bIsStaticSwitch=false, bIsVisible=true, type {classStructOrEnum {refPath=&NiagaraFloat}}}
      UseLoopCountLimit {bIsDynamic=false, bIsEditable=true, bIsStaticSwitch=false, bIsVisible=true, type {classStructOrEnum {refPath=&NiagaraBool}}}
      Loop Count Limit {bIsDynamic=false, bIsEditable=false, bIsStaticSwitch=false, bIsVisible=true, type {classStructOrEnum {refPath=&NiagaraInt32}}}
      Spawn Group {bIsDynamic=false, bIsEditable=true, bIsStaticSwitch=false, bIsVisible=true, type {classStructOrEnum {refPath=&NiagaraInt32}}}
      Age {bIsDynamic=false, bIsEditable=true, bIsStaticSwitch=false, bIsVisible=true, type {classStructOrEnum {refPath=&NiagaraFloat}}}
     ]
     moduleScript {refPath=&SpawnBurst_Instantaneous}
    }
   ]
  }
  particleSpawnScript {
   scriptName=ParticleSpawnScript
   modules [
    {
     bIsSetParametersModule=false
     enabled=true
     moduleName=PartitionParticles
     inputs [
      Partition Mode {
       bIsDynamic=false
       bIsEditable=true
       bIsStaticSwitch=true
       bIsVisible=true
       type {classStructOrEnum {refPath=&ENiagara_PartitionMode_v2}}
      }
      Partition Count Mode {
       bIsDynamic=false
       bIsEditable=false
       bIsStaticSwitch=true
       bIsVisible=false
       type {classStructOrEnum {refPath=&ENiagara_InfiniteLoopDuration}}
      }
      A {
       bIsDynamic=false
       bIsEditable=false
       bIsStaticSwitch=false
       bIsVisible=false
       type {classStructOrEnum {refPath=&NiagaraFloat}}
      }
      Number Of Partitions {bIsDynamic=false, bIsEditable=true, bIsStaticSwitch=false, bIsVisible=true, type {classStructOrEnum {refPath=&NiagaraInt32}}}
      Particles Per Partition {
       bIsDynamic=false
       bIsEditable=false
       bIsStaticSwitch=false
       bIsVisible=false
       type {classStructOrEnum {refPath=&NiagaraInt32}}
      }
      Partition Duration {
       bIsDynamic=false
       bIsEditable=false
       bIsStaticSwitch=false
       bIsVisible=false
       type {classStructOrEnum {refPath=&NiagaraFloat}}
      }
      Time {
       bIsDynamic=false
       bIsEditable=false
       bIsStaticSwitch=false
       bIsVisible=false
       type {classStructOrEnum {refPath=&NiagaraFloat}}
      }
      Partitioning Frequency {
       bIsDynamic=false
       bIsEditable=false
       bIsStaticSwitch=true
       bIsVisible=false
       type {classStructOrEnum {refPath=&ENiagaraRandomnessEvaluation}}
      }
      B {
       bIsDynamic=false
       bIsEditable=false
       bIsStaticSwitch=false
       bIsVisible=false
       type {classStructOrEnum {refPath=&NiagaraFloat}}
      }
      Distance {
       bIsDynamic=false
       bIsEditable=false
       bIsStaticSwitch=false
       bIsVisible=false
       type {classStructOrEnum {refPath=&NiagaraFloat}}
      }
      Comparison Type {
       bIsDynamic=false
       bIsEditable=false
       bIsStaticSwitch=true
       bIsVisible=false
       type {classStructOrEnum {refPath=&ENiagaraBooleanLogicOps_v2}}
      }
      Evaluation Type {
       bIsDynamic=false
       bIsEditable=false
       bIsStaticSwitch=true
       bIsVisible=false
       type {classStructOrEnum {refPath=&ENiagaraRandomnessEvaluation}}
      }
      UseDistanceFalloff {bIsDynamic=false, bIsEditable=false, bIsStaticSwitch=false, bIsVisible=false, type {classStructOrEnum {refPath=&NiagaraBool}}}
      Distance Falloff Exponent {
       bIsDynamic=false
       bIsEditable=false
       bIsStaticSwitch=false
       bIsVisible=false
       type {classStructOrEnum {refPath=&NiagaraFloat}}
      }
      Start Position {
       bIsDynamic=false
       bIsEditable=false
       bIsStaticSwitch=false
       bIsVisible=false
       type {classStructOrEnum {refPath=&NiagaraPosition}}
      }
      End Position {
       bIsDynamic=false
       bIsEditable=false
       bIsStaticSwitch=false
       bIsVisible=false
       type {classStructOrEnum {refPath=&NiagaraPosition}}
      }
      ID to Partition {bIsDynamic=false, bIsEditable=true, bIsStaticSwitch=false, bIsVisible=true, type {classStructOrEnum {refPath=&NiagaraInt32}}}
      Randomness Mode {
       bIsDynamic=false
       bIsEditable=false
       bIsStaticSwitch=true
       bIsVisible=false
       type {classStructOrEnum {refPath=&ENiagaraRandomnessMode}}
      }
      Override Seed {bIsDynamic=false, bIsEditable=false, bIsStaticSwitch=true, bIsVisible=false, type {classStructOrEnum {refPath=&NiagaraBool}}}
      Fixed Random Seed {bIsDynamic=false, bIsEditable=false, bIsStaticSwitch=true, bIsVisible=false, type {classStructOrEnum {refPath=&NiagaraBool}}}
      Color A {
       bIsDynamic=false
       bIsEditable=false
       bIsStaticSwitch=false
       bIsVisible=false
       type {classStructOrEnum {refPath=&LinearColor}}
      }
      Color B {
       bIsDynamic=false
       bIsEditable=false
       bIsStaticSwitch=false
       bIsVisible=false
       type {classStructOrEnum {refPath=&LinearColor}}
      }
      Color Comparison Type {bIsDynamic=false, bIsEditable=false, bIsStaticSwitch=true, bIsVisible=false, type {classStructOrEnum {refPath=&NiagaraInt32}}}
      Color Tolerance {
       bIsDynamic=false
       bIsEditable=false
       bIsStaticSwitch=false
       bIsVisible=false
       type {classStructOrEnum {refPath=&NiagaraFloat}}
      }
      Minimum Saturation {
       bIsDynamic=false
       bIsEditable=false
       bIsStaticSwitch=false
       bIsVisible=false
       type {classStructOrEnum {refPath=&NiagaraFloat}}
      }
      Direction A {
       bIsDynamic=false
       bIsEditable=false
       bIsStaticSwitch=false
       bIsVisible=false
       type {classStructOrEnum {refPath=&Vector3f}}
      }
      Direction B {
       bIsDynamic=false
       bIsEditable=false
       bIsStaticSwitch=false
       bIsVisible=false
       type {classStructOrEnum {refPath=&Vector3f}}
      }
      Direction Tolerance {
       bIsDynamic=false
       bIsEditable=false
       bIsStaticSwitch=false
       bIsVisible=false
       type {classStructOrEnum {refPath=&NiagaraFloat}}
      }
      Write Partition to Ribbon ID {
       bIsDynamic=false
       bIsEditable=true
       bIsStaticSwitch=true
       bIsVisible=true
       type {classStructOrEnum {refPath=&ENiagara_AttributeSamplingApplyOutput}}
      }
      Write Partition Index to Ribbon Link Order {
       bIsDynamic=false
       bIsEditable=false
       bIsStaticSwitch=true
       bIsVisible=false
       type {classStructOrEnum {refPath=&ENiagara_AttributeSamplingApplyOutput}}
      }
      Normalized Index In Partition Includes One {bIsDynamic=false, bIsEditable=false, bIsStaticSwitch=true, bIsVisible=false, type {classStructOrEnum {refPath=&NiagaraBool}}}
      Normalized Index Includes One {bIsDynamic=false, bIsEditable=true, bIsStaticSwitch=true, bIsVisible=true, type {classStructOrEnum {refPath=&NiagaraBool}}}
     ]
     moduleScript {refPath=&PartitionParticles}
    }
    {
     bIsSetParametersModule=false
     enabled=true
     moduleName=InitializeParticle
     inputs [
      Lifetime Mode {
       bIsDynamic=false
       bIsEditable=true
       bIsStaticSwitch=true
       bIsVisible=true
       type {classStructOrEnum {refPath=&ENiagara_LifetimeMode}}
      }
      Lifetime Min {
       bIsDynamic=false
       bIsEditable=false
       bIsStaticSwitch=false
       bIsVisible=false
       type {classStructOrEnum {refPath=&NiagaraFloat}}
      }
      Lifetime {bIsDynamic=false, bIsEditable=true, bIsStaticSwitch=false, bIsVisible=true, type {classStructOrEnum {refPath=&NiagaraFloat}}}
      Lifetime Max {
       bIsDynamic=false
       bIsEditable=false
       bIsStaticSwitch=false
       bIsVisible=false
       type {classStructOrEnum {refPath=&NiagaraFloat}}
      }
      Lifetime Randomness Mode {
       bIsDynamic=false
       bIsEditable=false
       bIsStaticSwitch=true
       bIsVisible=false
       type {classStructOrEnum {refPath=&ENiagaraRandomnessMode}}
      }
      Color Mode {
       bIsDynamic=false
       bIsEditable=true
       bIsStaticSwitch=true
       bIsVisible=true
       type {classStructOrEnum {refPath=&ENiagara_ColorInitializationMode}}
      }
      Color Minimum {
       bIsDynamic=false
       bIsEditable=false
       bIsStaticSwitch=false
       bIsVisible=false
       type {classStructOrEnum {refPath=&LinearColor}}
      }
      Color {
       bIsDynamic=false
       bIsEditable=true
       bIsStaticSwitch=false
       bIsVisible=true
       type {classStructOrEnum {refPath=&LinearColor}}
      }
      AdjustHue {bIsDynamic=false, bIsEditable=false, bIsStaticSwitch=false, bIsVisible=false, type {classStructOrEnum {refPath=&NiagaraBool}}}
      Hue Shift Range {
       bIsDynamic=false
       bIsEditable=false
       bIsStaticSwitch=false
       bIsVisible=false
       type {classStructOrEnum {refPath=&Vector2f}}
      }
      Color Maximum {
       bIsDynamic=false
       bIsEditable=false
       bIsStaticSwitch=false
       bIsVisible=false
       type {classStructOrEnum {refPath=&LinearColor}}
      }
      AdjustSaturation {bIsDynamic=false, bIsEditable=false, bIsStaticSwitch=false, bIsVisible=false, type {classStructOrEnum {refPath=&NiagaraBool}}}
      Saturation Range {
       bIsDynamic=false
       bIsEditable=false
       bIsStaticSwitch=false
       bIsVisible=false
       type {classStructOrEnum {refPath=&Vector2f}}
      }
      AdjustValue {bIsDynamic=false, bIsEditable=false, bIsStaticSwitch=false, bIsVisible=false, type {classStructOrEnum {refPath=&NiagaraBool}}}
      Value Range {
       bIsDynamic=false
       bIsEditable=false
       bIsStaticSwitch=false
       bIsVisible=false
       type {classStructOrEnum {refPath=&Vector2f}}
      }
      AdjustAlpha {bIsDynamic=false, bIsEditable=false, bIsStaticSwitch=false, bIsVisible=false, type {classStructOrEnum {refPath=&NiagaraBool}}}
      Alpha Scale Range {
       bIsDynamic=false
       bIsEditable=false
       bIsStaticSwitch=false
       bIsVisible=false
       type {classStructOrEnum {refPath=&Vector2f}}
      }
      Color Channel Mode {
       bIsDynamic=false
       bIsEditable=false
       bIsStaticSwitch=true
       bIsVisible=false
       type {classStructOrEnum {refPath=&ENiagaraChannelCorrelation}}
      }
      Color Randomness Mode {
       bIsDynamic=false
       bIsEditable=false
       bIsStaticSwitch=true
       bIsVisible=false
       type {classStructOrEnum {refPath=&ENiagaraRandomnessMode}}
      }
      Position Mode {
       bIsDynamic=false
       bIsEditable=true
       bIsStaticSwitch=true
       bIsVisible=true
       type {classStructOrEnum {refPath=&ENiagara_PositionInitializationMode}}
      }
      Position {
       bIsDynamic=false
       bIsEditable=false
       bIsStaticSwitch=false
       bIsVisible=false
       type {classStructOrEnum {refPath=&NiagaraPosition}}
      }
      UsePositionOffset {bIsDynamic=false, bIsEditable=true, bIsStaticSwitch=false, bIsVisible=true, type {classStructOrEnum {refPath=&NiagaraBool}}}
      Position Offset {bIsDynamic=true, bIsEditable=true, bIsStaticSwitch=false, bIsVisible=true, type {classStructOrEnum {refPath=&Vector3f}}}
      Position Offset Coordinate Space {
       bIsDynamic=false
       bIsEditable=true
       bIsStaticSwitch=false
       bIsVisible=true
       type {classStructOrEnum {refPath=&ENiagaraCoordinateSpace}}
      }
      Mass Mode {
       bIsDynamic=false
       bIsEditable=true
       bIsStaticSwitch=true
       bIsVisible=true
       type {classStructOrEnum {refPath=&ENiagara_MassInitializationMode}}
      }
      Mass Min {
       bIsDynamic=false
       bIsEditable=false
       bIsStaticSwitch=false
       bIsVisible=false
       type {classStructOrEnum {refPath=&NiagaraFloat}}
      }
      Mass {
       bIsDynamic=false
       bIsEditable=false
       bIsStaticSwitch=false
       bIsVisible=false
       type {classStructOrEnum {refPath=&NiagaraFloat}}
      }
      Mass Max {
       bIsDynamic=false
       bIsEditable=false
       bIsStaticSwitch=false
       bIsVisible=false
       type {classStructOrEnum {refPath=&NiagaraFloat}}
      }
      Mass Randomness Mode {
       bIsDynamic=false
       bIsEditable=false
       bIsStaticSwitch=true
       bIsVisible=false
       type {classStructOrEnum {refPath=&ENiagaraRandomnessMode}}
      }
      Write Material Random {bIsDynamic=false, bIsEditable=true, bIsStaticSwitch=false, bIsVisible=true, type {classStructOrEnum {refPath=&NiagaraBool}}}
      Material Random {bIsDynamic=false, bIsEditable=true, bIsStaticSwitch=false, bIsVisible=true, type {classStructOrEnum {refPath=&NiagaraFloat}}}
      Sprite Size Mode {
       bIsDynamic=false
       bIsEditable=true
       bIsStaticSwitch=true
       bIsVisible=true
       type {classStructOrEnum {refPath=&ENiagara_SizeScaleMode}}
      }
      Sprite Size {
       bIsDynamic=false
       bIsEditable=false
       bIsStaticSwitch=false
       bIsVisible=false
       type {classStructOrEnum {refPath=&Vector2f}}
      }
      Uniform Sprite Size {
       bIsDynamic=false
       bIsEditable=false
       bIsStaticSwitch=false
       bIsVisible=false
       type {classStructOrEnum {refPath=&NiagaraFloat}}
      }
      Uniform Sprite Size Min {
       bIsDynamic=false
       bIsEditable=false
       bIsStaticSwitch=false
       bIsVisible=false
       type {classStructOrEnum {refPath=&NiagaraFloat}}
      }
      Sprite Size Min {
       bIsDynamic=false
       bIsEditable=false
       bIsStaticSwitch=false
       bIsVisible=false
       type {classStructOrEnum {refPath=&Vector2f}}
      }
      Sprite Size Max {
       bIsDynamic=false
       bIsEditable=false
       bIsStaticSwitch=false
       bIsVisible=false
       type {classStructOrEnum {refPath=&Vector2f}}
      }
      Uniform Sprite Size Max {
       bIsDynamic=false
       bIsEditable=false
       bIsStaticSwitch=false
       bIsVisible=false
       type {classStructOrEnum {refPath=&NiagaraFloat}}
      }
      Sprite Size Randomness Mode {
       bIsDynamic=false
       bIsEditable=false
       bIsStaticSwitch=true
       bIsVisible=false
       type {classStructOrEnum {refPath=&ENiagaraRandomnessMode}}
      }
      Sprite Rotation Mode {
       bIsDynamic=false
       bIsEditable=true
       bIsStaticSwitch=true
       bIsVisible=true
       type {classStructOrEnum {refPath=&ENiagara_SpriteRotationMode}}
      }
      Sprite Rotation Angle Min {
       bIsDynamic=false
       bIsEditable=false
       bIsStaticSwitch=false
       bIsVisible=false
       type {classStructOrEnum {refPath=&NiagaraFloat}}
      }
      Sprite Rotation Angle {
       bIsDynamic=false
       bIsEditable=false
       bIsStaticSwitch=false
       bIsVisible=false
       type {classStructOrEnum {refPath=&NiagaraFloat}}
      }
      Sprite Rotation Angle Max {
       bIsDynamic=false
       bIsEditable=false
       bIsStaticSwitch=false
       bIsVisible=false
       type {classStructOrEnum {refPath=&NiagaraFloat}}
      }
      Sprite Rotation Randomness Mode {
       bIsDynamic=false
       bIsEditable=false
       bIsStaticSwitch=true
       bIsVisible=false
       type {classStructOrEnum {refPath=&ENiagaraRandomnessMode}}
      }
      Sprite UV Mode {
       bIsDynamic=false
       bIsEditable=true
       bIsStaticSwitch=true
       bIsVisible=true
       type {classStructOrEnum {refPath=&ENiagara_UVFlippingMode}}
      }
      Sprite UV Randomness Mode {
       bIsDynamic=false
       bIsEditable=false
       bIsStaticSwitch=true
       bIsVisible=false
       type {classStructOrEnum {refPath=&ENiagaraRandomnessMode}}
      }
      Sprite UV Scale {
       bIsDynamic=false
       bIsEditable=false
       bIsStaticSwitch=false
       bIsVisible=false
       type {classStructOrEnum {refPath=&Vector2f}}
      }
      Mesh Scale Mode {
       bIsDynamic=false
       bIsEditable=true
       bIsStaticSwitch=true
       bIsVisible=true
       type {classStructOrEnum {refPath=&ENiagara_SizeScaleMode}}
      }
      Mesh Uniform Scale Min {
       bIsDynamic=false
       bIsEditable=false
       bIsStaticSwitch=false
       bIsVisible=false
       type {classStructOrEnum {refPath=&NiagaraFloat}}
      }
      Mesh Uniform Scale {bIsDynamic=true, bIsEditable=true, bIsStaticSwitch=false, bIsVisible=true, type {classStructOrEnum {refPath=&NiagaraFloat}}}
      Mesh Scale Min {
       bIsDynamic=false
       bIsEditable=false
       bIsStaticSwitch=false
       bIsVisible=false
       type {classStructOrEnum {refPath=&Vector3f}}
      }
      Mesh Scale {
       bIsDynamic=false
       bIsEditable=false
       bIsStaticSwitch=false
       bIsVisible=false
       type {classStructOrEnum {refPath=&Vector3f}}
      }
      Mesh Scale Max {
       bIsDynamic=false
       bIsEditable=false
       bIsStaticSwitch=false
       bIsVisible=false
       type {classStructOrEnum {refPath=&Vector3f}}
      }
      Mesh Uniform Scale Max {
       bIsDynamic=false
       bIsEditable=false
       bIsStaticSwitch=false
       bIsVisible=false
       type {classStructOrEnum {refPath=&NiagaraFloat}}
      }
      Mesh Scale Randomness Mode {
       bIsDynamic=false
       bIsEditable=false
       bIsStaticSwitch=true
       bIsVisible=false
       type {classStructOrEnum {refPath=&ENiagaraRandomnessMode}}
      }
      Mesh Renderer Array Visibility Mode {
       bIsDynamic=false
       bIsEditable=true
       bIsStaticSwitch=true
       bIsVisible=true
       type {classStructOrEnum {refPath=&ENiagara_UnsetDirectSetRandom}}
      }
      Mesh Renderer Mesh Index {
       bIsDynamic=false
       bIsEditable=false
       bIsStaticSwitch=false
       bIsVisible=false
       type {classStructOrEnum {refPath=&NiagaraInt32}}
      }
      Wrap Index By Mesh Count {bIsDynamic=false, bIsEditable=false, bIsStaticSwitch=true, bIsVisible=false, type {classStructOrEnum {refPath=&NiagaraBool}}}
      Mesh Renderer Info {
       bIsDynamic=false
       bIsEditable=false
       bIsStaticSwitch=false
       bIsVisible=false
       type {classStructOrEnum {refPath=&NiagaraDataInterfaceMeshRendererInfo}}
      }
      Mesh Renderer Random Seed {
       bIsDynamic=false
       bIsEditable=false
       bIsStaticSwitch=false
       bIsVisible=false
       type {classStructOrEnum {refPath=&NiagaraInt32}}
      }
      Ribbon Width Mode {
       bIsDynamic=false
       bIsEditable=true
       bIsStaticSwitch=true
       bIsVisible=true
       type {classStructOrEnum {refPath=&ENiagara_UnsetDirectSet}}
      }
      Ribbon Width {
       bIsDynamic=false
       bIsEditable=false
       bIsStaticSwitch=false
       bIsVisible=false
       type {classStructOrEnum {refPath=&NiagaraFloat}}
      }
      Ribbon Facing Mode {
       bIsDynamic=false
       bIsEditable=true
       bIsStaticSwitch=true
       bIsVisible=true
       type {classStructOrEnum {refPath=&ENiagara_UnsetDirectSet}}
      }
      Ribbon Facing Vector {
       bIsDynamic=false
       bIsEditable=false
       bIsStaticSwitch=false
       bIsVisible=false
       type {classStructOrEnum {refPath=&Vector3f}}
      }
      Ribbon Twist Mode {
       bIsDynamic=false
       bIsEditable=true
       bIsStaticSwitch=true
       bIsVisible=true
       type {classStructOrEnum {refPath=&ENiagara_UnsetDirectSet}}
      }
      "Ribbon Twist " {
       bIsDynamic=false
       bIsEditable=false
       bIsStaticSwitch=false
       bIsVisible=false
       type {classStructOrEnum {refPath=&NiagaraFloat}}
      }
     ]
     moduleScript {refPath=&InitializeParticle}
    }
    {
     bIsSetParametersModule=false
     enabled=true
     moduleName=InitialMeshOrientation
     inputs [
      Mesh Orientation Mode {
       bIsDynamic=false
       bIsEditable=true
       bIsStaticSwitch=true
       bIsVisible=true
       type {classStructOrEnum {refPath=&ENiagara_MeshOrientationOptions}}
      }
      Orientation Quaternion {bIsDynamic=false, bIsEditable=false, bIsStaticSwitch=false, bIsVisible=false, type {classStructOrEnum {refPath=&Quat4f}}}
      Orientation Matrix {
       bIsDynamic=false
       bIsEditable=false
       bIsStaticSwitch=false
       bIsVisible=false
       type {classStructOrEnum {refPath=&NiagaraMatrix}}
      }
      Orientation Vector {
       bIsDynamic=false
       bIsEditable=false
       bIsStaticSwitch=false
       bIsVisible=false
       type {classStructOrEnum {refPath=&Vector3f}}
      }
      Orientation Coordinate Space {
       bIsDynamic=false
       bIsEditable=false
       bIsStaticSwitch=true
       bIsVisible=false
       type {classStructOrEnum {refPath=&ENiagaraCoordinateSpace}}
      }
      Mesh Axis To Orient {
       bIsDynamic=false
       bIsEditable=false
       bIsStaticSwitch=false
       bIsVisible=false
       type {classStructOrEnum {refPath=&Vector3f}}
      }
      Random Seed {
       bIsDynamic=false
       bIsEditable=false
       bIsStaticSwitch=false
       bIsVisible=false
       type {classStructOrEnum {refPath=&NiagaraInt32}}
      }
      Use Rotation Vector {bIsDynamic=false, bIsEditable=true, bIsStaticSwitch=false, bIsVisible=true, type {classStructOrEnum {refPath=&NiagaraBool}}}
      Rotation {bIsDynamic=true, bIsEditable=true, bIsStaticSwitch=false, bIsVisible=true, type {classStructOrEnum {refPath=&Vector3f}}}
      Rotation Coordinate Space {
       bIsDynamic=false
       bIsEditable=true
       bIsStaticSwitch=true
       bIsVisible=true
       type {classStructOrEnum {refPath=&ENiagaraMeshTransforms}}
      }
     ]
     moduleScript {refPath=&InitialMeshOrientation}
    }
   ]
  }
  particleUpdateScript {
   scriptName=ParticleUpdateScript
   modules [
    {
     bIsSetParametersModule=false
     enabled=true
     moduleName=ParticleState
     inputs [
      Kill Particles When Lifetime Has Elapsed {bIsDynamic=false, bIsEditable=true, bIsStaticSwitch=true, bIsVisible=true, type {classStructOrEnum {refPath=&NiagaraBool}}}
      Loop Particles Lifetime {bIsDynamic=false, bIsEditable=false, bIsStaticSwitch=false, bIsVisible=false, type {classStructOrEnum {refPath=&NiagaraBool}}}
      Let Infinitely Lived Particles Die When Emitter Deactivates {bIsDynamic=false, bIsEditable=false, bIsStaticSwitch=false, bIsVisible=false, type {classStructOrEnum {refPath=&NiagaraBool}}}
      Lifetime {bIsDynamic=false, bIsEditable=true, bIsStaticSwitch=false, bIsVisible=true, type {classStructOrEnum {refPath=&NiagaraFloat}}}
      DeltaTime {bIsDynamic=false, bIsEditable=true, bIsStaticSwitch=false, bIsVisible=true, type {classStructOrEnum {refPath=&NiagaraFloat}}}
     ]
     moduleScript {refPath=&ParticleState}
    }
    {
     bIsSetParametersModule=false
     enabled=true
     moduleName=DynamicMaterialParameters
     inputs [
      Write Parameter Index 0 {
       bIsDynamic=false
       bIsEditable=true
       bIsStaticSwitch=true
       bIsVisible=true
       type {classStructOrEnum {refPath=&ENiagara_DynamicMaterialParametersInputMode}}
      }
      Index 0 Vector 1 {
       bIsDynamic=false
       bIsEditable=false
       bIsStaticSwitch=false
       bIsVisible=false
       type {classStructOrEnum {refPath=&Vector3f}}
      }
      Module.Index0Float0Write {bIsDynamic=false, bIsEditable=true, bIsStaticSwitch=true, bIsVisible=true, type {classStructOrEnum {refPath=&NiagaraBool}}}
      Index 0 Param 1 {bIsDynamic=false, bIsEditable=true, bIsStaticSwitch=false, bIsVisible=true, type {classStructOrEnum {refPath=&NiagaraFloat}}}
      Module.Index0Float1Write {bIsDynamic=false, bIsEditable=true, bIsStaticSwitch=true, bIsVisible=true, type {classStructOrEnum {refPath=&NiagaraBool}}}
      Index 0 Param 2 {bIsDynamic=false, bIsEditable=true, bIsStaticSwitch=false, bIsVisible=true, type {classStructOrEnum {refPath=&NiagaraFloat}}}
      Module.Index0Float2Write {bIsDynamic=false, bIsEditable=true, bIsStaticSwitch=true, bIsVisible=true, type {classStructOrEnum {refPath=&NiagaraBool}}}
      Index 0 Param 3 {bIsDynamic=false, bIsEditable=true, bIsStaticSwitch=false, bIsVisible=true, type {classStructOrEnum {refPath=&NiagaraFloat}}}
      Module.Index0Float3Write {bIsDynamic=false, bIsEditable=true, bIsStaticSwitch=true, bIsVisible=true, type {classStructOrEnum {refPath=&NiagaraBool}}}
      Index 0 Param 4 {bIsDynamic=false, bIsEditable=true, bIsStaticSwitch=false, bIsVisible=true, type {classStructOrEnum {refPath=&NiagaraFloat}}}
      Index 0 Vector2D 1 {
       bIsDynamic=false
       bIsEditable=false
       bIsStaticSwitch=false
       bIsVisible=false
       type {classStructOrEnum {refPath=&Vector2f}}
      }
      Index 0 Vector2D 2 {
       bIsDynamic=false
       bIsEditable=false
       bIsStaticSwitch=false
       bIsVisible=false
       type {classStructOrEnum {refPath=&Vector2f}}
      }
      Index 0 Vector 4 {
       bIsDynamic=false
       bIsEditable=false
       bIsStaticSwitch=false
       bIsVisible=false
       type {classStructOrEnum {refPath=&Vector4f}}
      }
      Index 0 Linear Color {
       bIsDynamic=false
       bIsEditable=false
       bIsStaticSwitch=false
       bIsVisible=false
       type {classStructOrEnum {refPath=&LinearColor}}
      }
      Write Parameter Index 1 {
       bIsDynamic=false
       bIsEditable=true
       bIsStaticSwitch=true
       bIsVisible=true
       type {classStructOrEnum {refPath=&ENiagara_DynamicMaterialParametersInputMode}}
      }
      Index 1 Vector 1 {
       bIsDynamic=false
       bIsEditable=false
       bIsStaticSwitch=false
       bIsVisible=false
       type {classStructOrEnum {refPath=&Vector3f}}
      }
      Index 1 Param 1 {
       bIsDynamic=false
       bIsEditable=false
       bIsStaticSwitch=false
       bIsVisible=false
       type {classStructOrEnum {refPath=&NiagaraFloat}}
      }
      Index 1 Param 2 {
       bIsDynamic=false
       bIsEditable=false
       bIsStaticSwitch=false
       bIsVisible=false
       type {classStructOrEnum {refPath=&NiagaraFloat}}
      }
      Index 1 Param 3 {
       bIsDynamic=false
       bIsEditable=false
       bIsStaticSwitch=false
       bIsVisible=false
       type {classStructOrEnum {refPath=&NiagaraFloat}}
      }
      Index 1 Param 4 {
       bIsDynamic=false
       bIsEditable=false
       bIsStaticSwitch=false
       bIsVisible=false
       type {classStructOrEnum {refPath=&NiagaraFloat}}
      }
      Index 1 Vector2D 1 {
       bIsDynamic=false
       bIsEditable=false
       bIsStaticSwitch=false
       bIsVisible=false
       type {classStructOrEnum {refPath=&Vector2f}}
      }
      Index 1 Vector2D 2 {
       bIsDynamic=false
       bIsEditable=false
       bIsStaticSwitch=false
       bIsVisible=false
       type {classStructOrEnum {refPath=&Vector2f}}
      }
      Index 1 Vector 4 {
       bIsDynamic=false
       bIsEditable=false
       bIsStaticSwitch=false
       bIsVisible=false
       type {classStructOrEnum {refPath=&Vector4f}}
      }
      Index 1 Linear Color {
       bIsDynamic=false
       bIsEditable=false
       bIsStaticSwitch=false
       bIsVisible=false
       type {classStructOrEnum {refPath=&LinearColor}}
      }
      Write Parameter Index 2 {
       bIsDynamic=false
       bIsEditable=true
       bIsStaticSwitch=true
       bIsVisible=true
       type {classStructOrEnum {refPath=&ENiagara_DynamicMaterialParametersInputMode}}
      }
      Index 2 Vector 1 {
       bIsDynamic=false
       bIsEditable=false
       bIsStaticSwitch=false
       bIsVisible=false
       type {classStructOrEnum {refPath=&Vector3f}}
      }
      Index 2 Param 1 {
       bIsDynamic=false
       bIsEditable=false
       bIsStaticSwitch=false
       bIsVisible=false
       type {classStructOrEnum {refPath=&NiagaraFloat}}
      }
      Index 2 Param 2 {
       bIsDynamic=false
       bIsEditable=false
       bIsStaticSwitch=false
       bIsVisible=false
       type {classStructOrEnum {refPath=&NiagaraFloat}}
      }
      Index 2 Param 3 {
       bIsDynamic=false
       bIsEditable=false
       bIsStaticSwitch=false
       bIsVisible=false
       type {classStructOrEnum {refPath=&NiagaraFloat}}
      }
      Index 2 Param 4 {
       bIsDynamic=false
       bIsEditable=false
       bIsStaticSwitch=false
       bIsVisible=false
       type {classStructOrEnum {refPath=&NiagaraFloat}}
      }
      Index 2 Vector2D 1 {
       bIsDynamic=false
       bIsEditable=false
       bIsStaticSwitch=false
       bIsVisible=false
       type {classStructOrEnum {refPath=&Vector2f}}
      }
      Index 2 Vector2D 2 {
       bIsDynamic=false
       bIsEditable=false
       bIsStaticSwitch=false
       bIsVisible=false
       type {classStructOrEnum {refPath=&Vector2f}}
      }
      Index 2 Vector 4 {
       bIsDynamic=false
       bIsEditable=false
       bIsStaticSwitch=false
       bIsVisible=false
       type {classStructOrEnum {refPath=&Vector4f}}
      }
      Index 2 Linear Color {
       bIsDynamic=false
       bIsEditable=false
       bIsStaticSwitch=false
       bIsVisible=false
       type {classStructOrEnum {refPath=&LinearColor}}
      }
      Write Parameter Index 3 {
       bIsDynamic=false
       bIsEditable=true
       bIsStaticSwitch=true
       bIsVisible=true
       type {classStructOrEnum {refPath=&ENiagara_DynamicMaterialParametersInputMode}}
      }
      Index 3 Vector 1 {
       bIsDynamic=false
       bIsEditable=false
       bIsStaticSwitch=false
       bIsVisible=false
       type {classStructOrEnum {refPath=&Vector3f}}
      }
      Index 3 Param 1 {
       bIsDynamic=false
       bIsEditable=false
       bIsStaticSwitch=false
       bIsVisible=false
       type {classStructOrEnum {refPath=&NiagaraFloat}}
      }
      Index 3 Param 2 {
       bIsDynamic=false
       bIsEditable=false
       bIsStaticSwitch=false
       bIsVisible=false
       type {classStructOrEnum {refPath=&NiagaraFloat}}
      }
      Index 3 Param 3 {
       bIsDynamic=false
       bIsEditable=false
       bIsStaticSwitch=false
       bIsVisible=false
       type {classStructOrEnum {refPath=&NiagaraFloat}}
      }
      Index 3 Param 4 {
       bIsDynamic=false
       bIsEditable=false
       bIsStaticSwitch=false
       bIsVisible=false
       type {classStructOrEnum {refPath=&NiagaraFloat}}
      }
      Index 3 Vector2D 1 {
       bIsDynamic=false
       bIsEditable=false
       bIsStaticSwitch=false
       bIsVisible=false
       type {classStructOrEnum {refPath=&Vector2f}}
      }
      Index 3 Vector2D 2 {
       bIsDynamic=false
       bIsEditable=false
       bIsStaticSwitch=false
       bIsVisible=false
       type {classStructOrEnum {refPath=&Vector2f}}
      }
      Index 3 Vector 4 {
       bIsDynamic=false
       bIsEditable=false
       bIsStaticSwitch=false
       bIsVisible=false
       type {classStructOrEnum {refPath=&Vector4f}}
      }
      Index 3 Linear Color {
       bIsDynamic=false
       bIsEditable=false
       bIsStaticSwitch=false
       bIsVisible=false
       type {classStructOrEnum {refPath=&LinearColor}}
      }
      Dynamic Parameter Output {
       bIsDynamic=false
       bIsEditable=true
       bIsStaticSwitch=true
       bIsVisible=true
       type {classStructOrEnum {refPath=&ENiagara_AttributeSamplingApplyOutput}}
      }
     ]
     moduleScript {refPath=&DynamicMaterialParameters}
    }
    {
     bIsSetParametersModule=false
     enabled=true
     moduleName=ScaleColor
     inputs [
      Scale Mode {bIsDynamic=false, bIsEditable=true, bIsStaticSwitch=true, bIsVisible=true, type {classStructOrEnum {refPath=&NiagaraInt32}}}
      Linear Color Curve {
       bIsDynamic=false
       bIsEditable=false
       bIsStaticSwitch=false
       bIsVisible=false
       type {classStructOrEnum {refPath=&NiagaraDataInterfaceColorCurve}}
      }
      ScaleRGB {bIsDynamic=false, bIsEditable=true, bIsStaticSwitch=false, bIsVisible=true, type {classStructOrEnum {refPath=&NiagaraBool}}}
      Scale RGB {bIsDynamic=true, bIsEditable=true, bIsStaticSwitch=false, bIsVisible=true, type {classStructOrEnum {refPath=&Vector3f}}}
      ScaleRGBA {bIsDynamic=false, bIsEditable=false, bIsStaticSwitch=false, bIsVisible=false, type {classStructOrEnum {refPath=&NiagaraBool}}}
      Scale RGBA {bIsDynamic=true, bIsEditable=false, bIsStaticSwitch=false, bIsVisible=false, type {classStructOrEnum {refPath=&Vector4f}}}
      Curve Index {
       bIsDynamic=false
       bIsEditable=false
       bIsStaticSwitch=false
       bIsVisible=false
       type {classStructOrEnum {refPath=&NiagaraFloat}}
      }
      ScaleA {bIsDynamic=false, bIsEditable=true, bIsStaticSwitch=false, bIsVisible=true, type {classStructOrEnum {refPath=&NiagaraBool}}}
      Scale Alpha {bIsDynamic=true, bIsEditable=true, bIsStaticSwitch=false, bIsVisible=true, type {classStructOrEnum {refPath=&NiagaraFloat}}}
      Color Value To Scale {
       bIsDynamic=false
       bIsEditable=true
       bIsStaticSwitch=false
       bIsVisible=true
       type {classStructOrEnum {refPath=&LinearColor}}
      }
      Color Scale {
       bIsDynamic=false
       bIsEditable=false
       bIsStaticSwitch=false
       bIsVisible=false
       type {classStructOrEnum {refPath=&LinearColor}}
      }
     ]
     moduleScript {refPath=&ScaleColor}
    }
    {
     bIsSetParametersModule=false
     enabled=true
     moduleName=ScaleMeshSize
     inputs [
      Scale Factor {bIsDynamic=true, bIsEditable=true, bIsStaticSwitch=false, bIsVisible=true, type {classStructOrEnum {refPath=&Vector3f}}}
      Initial Mesh Scale {bIsDynamic=false, bIsEditable=true, bIsStaticSwitch=false, bIsVisible=true, type {classStructOrEnum {refPath=&Vector3f}}}
     ]
     moduleScript {refPath=&ScaleMeshSize}
    }
    {
     bIsSetParametersModule=false
     enabled=true
     moduleName=CameraOffset
     inputs [Camera Offset Amount {bIsDynamic=false, bIsEditable=true, bIsStaticSwitch=false, bIsVisible=true, type {classStructOrEnum {refPath=&NiagaraFloat}}}]
     moduleScript {refPath=&CameraOffset}
    }
    {
     bIsSetParametersModule=true
     enabled=true
     moduleName=SetVariables_#8
     inputs [
      Particles.ShockwaveScale {bIsDynamic=true, bIsEditable=true, bIsStaticSwitch=false, bIsVisible=true, type {classStructOrEnum {refPath=&Vector3f}}}
      Particles.ShockwaveColor {
       bIsDynamic=true
       bIsEditable=true
       bIsStaticSwitch=false
       bIsVisible=true
       type {classStructOrEnum {refPath=&LinearColor}}
      }
      Particles.SecondarySphereColor {
       bIsDynamic=true
       bIsEditable=true
       bIsStaticSwitch=false
       bIsVisible=true
       type {classStructOrEnum {refPath=&LinearColor}}
      }
      Particles.SecondaryDyn {bIsDynamic=true, bIsEditable=true, bIsStaticSwitch=false, bIsVisible=true, type {classStructOrEnum {refPath=&Vector4f}}}
     ]
     moduleScript {refPath=&self:Shockwave002_0.NiagaraScriptSource_0.NiagaraGraph_0.NiagaraNodeAssignment_0.SetVariables_#8}
    }
   ]
  }
  rendererClasses [{refPath=&NiagaraMeshRendererProperties}]
  renderers [
   {rendererIndex=0, rendererClass {refPath=&NiagaraMeshRendererProperties}}
   {rendererIndex=1, rendererClass {refPath=&NiagaraMeshRendererProperties}}
   {rendererIndex=2, rendererClass {refPath=&NiagaraMeshRendererProperties}}
  ]
 }
}
summary {
 systemName=NS_Emote_SpriteSender_PropSpawn
 emitters [
  {bEnabled=false, emitterName=DefaultPropSpawn_Merged, rendererClasses [{refPath=&NiagaraSpriteRendererProperties}]}
  {
   bEnabled=false
   emitterName=Flare
   rendererClasses [{refPath=&NiagaraSpriteRendererProperties}, {refPath=&NiagaraLightRendererProperties}]
  }
  {bEnabled=false, emitterName=BigSparkles, rendererClasses [{refPath=&NiagaraSpriteRendererProperties}]}
  {bEnabled=false, emitterName=Shockwave, rendererClasses [{refPath=&NiagaraMeshRendererProperties}]}
  {
   bEnabled=false
   emitterName=DefaultPropSpawn_Merged001
   rendererClasses [
    {refPath=&NiagaraSpriteRendererProperties}
    {refPath=&NiagaraLightRendererProperties}
    {refPath=&NiagaraMeshRendererProperties}
   ]
  }
  {
   bEnabled=false
   emitterName=DefaultPropSpawn_Merged002
   rendererClasses [
    {refPath=&NiagaraSpriteRendererProperties}
    {refPath=&NiagaraLightRendererProperties}
    {refPath=&NiagaraMeshRendererProperties}
   ]
  }
  {
   bEnabled=false
   emitterName=DefaultPropSpawn_Merged003
   rendererClasses [
    {refPath=&NiagaraSpriteRendererProperties}
    {refPath=&NiagaraLightRendererProperties}
    {refPath=&NiagaraMeshRendererProperties}
   ]
  }
  {
   bEnabled=false
   emitterName=DefaultPropSpawn_Merged004
   rendererClasses [
    {refPath=&NiagaraSpriteRendererProperties}
    {refPath=&NiagaraLightRendererProperties}
    {refPath=&NiagaraMeshRendererProperties}
   ]
  }
  {emitterName=DefaultPropSpawn_Merged005, rendererClasses [{refPath=&NiagaraSpriteRendererProperties}]}
  {
   emitterName=Flare002
   rendererClasses [{refPath=&NiagaraSpriteRendererProperties}, {refPath=&NiagaraLightRendererProperties}]
  }
  {emitterName=BigSparkles002, rendererClasses [{refPath=&NiagaraSpriteRendererProperties}]}
  {emitterName=Shockwave002, rendererClasses [{refPath=&NiagaraMeshRendererProperties}]}
 ]
 userVariables [
  User.bIsFrontEnd {
   description=""
   defaultValue {struct {refPath=&NiagaraBool}, value {value=0}}
   type {classStructOrEnum {refPath=&NiagaraBool}}
  }
  User.bIsFrontEndPreview {
   description=""
   defaultValue {struct {refPath=&NiagaraBool}, value {value=0}}
   type {classStructOrEnum {refPath=&NiagaraBool}}
  }
 ]
}
